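; I_MPItemShop_Row is a Blueprint interface for row widgets in the multi-product Item Shop.
; It declares setup and reset calls that take row data, an offer group view model, a row state and parent focus state.
; Query functions resolve focus targets by location or offer ID and supply the scroll target, destination and padding, alongside a scroll-complete callback.
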(fn I_GetClosestFocusTarget
  (params (TargetLocation :type Vector%202D%20Structure))
  (return 0))

(fn I_GetDesiredFocusTarget
  (return 0))

(fn I_GetDesiredScrollDestination
  (return "IntoView"))

(fn I_GetDesiredScrollPadding
  (return 0.0))

(fn I_GetFocusTargetFromOfferID
  (params (OfferID :type String))
  (return 0))

(fn I_GetScrollTarget
  (return 0))

(fn I_InitializeRow
  (params (FortItemShopRow :type Fort%20Item%20Shop%20Row%20Structure) (ContentHasFocus :type Boolean)))

(fn I_OnScrollComplete
  (params (DestinationIndex :type Integer)))

(fn I_ResetRow)

(fn I_SetFocusedOfferID
  (params (OfferID :type String)))

(fn I_SetOfferGroupVM
  (params (OfferGroupVM :type Fort%20Item%20Shop%20Offer%20Group%20VM%20Object%20Reference)))

(fn I_SetParentHasFocusState
  (params (HasFocus :type Boolean)))

(fn I_SetRowState
  (params (InRowState :type E_ItemShop_RowState%20Enum :default "NewEnumerator0")))
