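# Item shop multi-product card widget, built on the UIKit item list card base, showing cosmetic ownership as a check or owned/total variant count.
# On UpdatePresentation, it casts the FortItemVM's item to an Athena cosmetic and tallies variant tags, skipping start-unlocked ones.
# Partial variant ownership shows the count; full variant ownership or plain ownership shows the check.
# It also sets card size on mobile and skips parent unhover handling on touch.
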WidgetBlueprint /ItemShop/MultiProduct/WBP_ItemShop_ItemCard
graphs [
 CountOwnedVariants {nodes [
   n1 = CountOwnedVariants!() -> (then, VariantChannels:Array of Mcp Variant Channel Info Structures, CosmeticItem:Athena Cosmetic Item Definition Object Reference) {is=title K2Node_FunctionEntry}
   > n2 = For Each Loop!(Exec, Array:Array of Mcp Variant Channel Info Structures=@n1) -> (
    LoopBody
    Array Element_OwnedVariantTags:Gameplay Tag Container Structure
    Array Element_ItemVariantIsUsedFor:Item Definition Base Object Reference
    Array Element_CustomData:String
    Array Element_VariantChannelTag:Gameplay Tag Structure
    Array Element_ActiveVariantTag:Gameplay Tag Structure
    Array Index
    Completed
   )
   n3 = GetVariantChannel!(
    self:Athena Cosmetic Item Definition Object Reference=Get CosmeticItem() -> (CosmeticItem:Athena Cosmetic Item Definition Object Reference)
    InVariantChannelTag:Gameplay Tag Structure=@n2.Array Element_VariantChannelTag
   ) -> (ReturnValue:Fort Cosmetic Variant Object Reference)
   > n4 = For Each Loop!(
    Exec
    Array:Array of Gameplay Tag Structures=GetAllVariantTags!(self:Fort Cosmetic Variant Object Reference=@n3) -> (ReturnValue_GameplayTags:Array of Gameplay Tag Structures)
   ) -> (LoopBody, Array Element:Gameplay Tag Structure, Array Index, Completed)
   > Branch(Condition=NOT Boolean(A=GetVariantStartUnlocked!(self:Fort Cosmetic Variant Object Reference=@n3, InVariantTag:Gameplay Tag Structure=@n4) -> (ReturnValue:bool)))
   > Increment Int(Value=Get TotalVariants() -> (TotalVariants:int=0))
   > Branch(Condition=HasTag!(TagContainer:"Gameplay Tag Container Structure (by ref)"=@n2, Tag:Gameplay Tag Structure=@n4, bExactMatch:bool=true) -> (ReturnValue:bool))
   > Increment Int(Value=Get OwnedVariants() -> (OwnedVariants:int=0))
  ]}
 EventGraph {nodes [
   n1 = Event PreConstruct()
   > Parent: PreConstruct(IsDesignTime=@n1.IsDesignTime)
   > Branch(Condition=IsMobileGame())
   > Set ItemCardSize(ItemCardSize:Vector 2D Structure=(X=128,Y=156)) -> (Output_Get:Vector 2D Structure)
   Event UpdatePresentation() {is=title K2Node_Event}
   > Set Variants Info!(
    execute
    self
    ItemInstance:Fort Item Object Reference=Get ItemInstance!(self:Fort Item VM Object Reference=Get FortItemVM() -> (FortItemVM:Fort Item VM Object Reference)) -> (ItemInstance:Fort Item Object Reference)
   ) -> (then) {is=title K2Node_CallFunction}
   n2 = Event SetListItem!() -> (OutputDelegate, then, InListItem:Object Reference) {is=title K2Node_Event}
   > n3 = Cast To FortItemVM!(execute, Object=@n2.InListItem) -> (then, CastFailed, AsFort Item VM:Fort Item VM Object Reference) {is=title K2Node_DynamicCast}
   > Set FortItemVM(FortItemVM:Fort Item VM Object Reference=@n3) -> (Output_Get:Fort Item VM Object Reference)
   Event On Unhovered()
   > n4 = Branch(Condition=IsUsingTouch(PlayerController=GetOwningPlayer()))
   Parent: On Unhovered(execute=@n4.else) {is=title K2Node_CallParentFunction}
  ]}
 ResetVariantsInfo {nodes [
   ResetVariantsInfo() {is=title K2Node_FunctionEntry}
   > Set Visibility(Visibility=Collapsed, self=Get O_OwnedAmount() -> (O_OwnedAmount:Overlay Object Reference))
   > Set Visibility(Visibility=Collapsed, self=Get StatusIndicator_Owned() -> (StatusIndicator_Owned:WBP UIKit Status Indicator Object Reference))
   > Set OwnedVariants(OwnedVariants:int=0) -> (Output_Get:int=0)
   > Set TotalVariants(TotalVariants:int=0) -> (Output_Get:int=0)
  ]}
 Set Variants Info {nodes [
   n1 = Set Variants Info!() -> (then, ItemInstance:Fort Item Object Reference) {is=title K2Node_FunctionEntry}
   > ResetVariantsInfo() {is=title K2Node_CallFunction}
   > n2 = Cast To AthenaCosmeticAccountItem!(execute, Object=@n1) -> (then, CastFailed, AsAthena Cosmetic Account Item:Athena Cosmetic Account Item Object Reference) {is=title K2Node_DynamicCast}
   > n3 = Cast To AthenaCosmeticItemDefinition!(execute, Object=GetItemDefinition!(self:Fort Item Object Reference=@n2) -> (ReturnValue:Fort Item Definition Object Reference)) -> (then, CastFailed, AsAthena Cosmetic Item Definition:Athena Cosmetic Item Definition Object Reference) {is=title K2Node_DynamicCast}
   > CountOwnedVariants!(
    execute
    self
    VariantChannels:Array of Mcp Variant Channel Info Structures=GetVariantChannels!(self:Athena Cosmetic Account Item Object Reference=@n2) -> (InOutTagData:Array of Mcp Variant Channel Info Structures)
    CosmeticItem:Athena Cosmetic Item Definition Object Reference=@n3
   ) -> (then) {is=title K2Node_CallFunction}
   > SetVariantsCountVisibility()
   > n4 = SetOwnedVariantsText()
   SetIsOwned!(
    execute=[@n4.then, @n3.CastFailed, @n2.CastFailed]
    self
    IsOwned:bool=Get bIsOwned!(self:Fort Item VM Object Reference=Get FortItemVM()) -> (bIsOwned:bool)
   ) -> (then)
  ]}
 SetIsOwned {nodes [
   n1 = SetIsOwned!() -> (then, IsOwned:bool=~) {is=title K2Node_FunctionEntry}
   > n2 = Set IsOwned(IsOwned:bool=@n1) -> (Output_Get:bool)
   n3 = NOT Boolean(A=Get PartiallyOwnsVariants() -> (PartiallyOwnsVariants:bool))
   > Set Visibility(Visibility=Select(Option 0=Collapsed, Option 1=SelfHitTestInvisible, Index=AND Boolean(A=@n2, B=@n3)), self=Get StatusIndicator_Owned())
  ]}
 SetOwnedVariantsText {nodes [
   SetOwnedVariantsText() {is=title K2Node_FunctionEntry}
   > SetText!(
    execute
    self:WBP UIKit Highlight Base Object Reference=Get Highlight_OwnedAmount() -> (Highlight_OwnedAmount:WBP UIKit Highlight Object Reference)
    Text:Text=Format Text!(Format="LOCTEXT(\"{0}/{1}\")", 0:int=Get OwnedVariants() -> (OwnedVariants=0), 1:int=Get TotalVariants() -> (TotalVariants=0)) -> (Result)
   ) -> (then)
  ]}
 SetVariantsCountVisibility {nodes [
   SetVariantsCountVisibility() {is=title K2Node_FunctionEntry}
   > n1 = Branch(Condition="Not Equal (Integer)"(A=Get OwnedVariants() -> (OwnedVariants=0)))
   > n2 = Branch(Condition="Equal (Integer)"(A=Get OwnedVariants() -> (OwnedVariants=0), B=Get TotalVariants() -> (TotalVariants=0)))
   > Set Visibility(Visibility=SelfHitTestInvisible, self=Get StatusIndicator_Owned())
   > Set PartiallyOwnsVariants(PartiallyOwnsVariants:bool=false) -> (Output_Get:bool)
   Set Visibility(execute=@n2.else, Visibility=SelfHitTestInvisible, self=Get O_OwnedAmount())
   > Set PartiallyOwnsVariants(PartiallyOwnsVariants=true)
   Set PartiallyOwnsVariants(execute=@n1.else, PartiallyOwnsVariants=false)
  ]}
]
variables [
 IsOwned {type=boolean}
 OwnedVariants {type=integer}
 PartiallyOwnsVariants {type=boolean}
 TotalVariants {type=integer}
]
widgets {
 info {
  inheritedWidgetCount=0
  namedSlotCount=1
  parentClass=/FortUILibrary/Systems/Items/WBP_UIKit_ItemList_CardBase.WBP_UIKit_ItemList_CardBase_C
  rootWidgetClass=null
  widgetCount=4
 }
 widgets [
  GridPanel_Owned {widgetClassPath=&GridPanel, widget=GridPanel_Owned, properties {columnFill [1], rowFill [1]}}
  O_OwnedAmount {
   widgetClassPath=&Overlay
   bIsVariable=true
   parent=GridPanel_Owned
   widget=O_OwnedAmount
   slot {class=GridSlot, properties {horizontalAlignment=HAlign_Right, verticalAlignment=VAlign_Top, padding=<ltrb 8 8 8 0>}}
  }
  Highlight_OwnedAmount {
   widgetClassPath=/FortUILibrary/UIKit/Highlight/WBP_UIKit_Highlight
   bIsVariable=true
   parent=O_OwnedAmount
   widget=Highlight_OwnedAmount
   properties {highlightSize=Compact, highlightType=Social}
   slot {class=OverlaySlot}
  }
  StatusIndicator_Owned {
   widgetClassPath=/FortUILibrary/UIKit/StatusIndicator/WBP_UIKit_StatusIndicator
   bIsVariable=true
   parent=GridPanel_Owned
   widget=StatusIndicator_Owned
   properties {icon=/FortUILibrary/Textures/MaterialInstances/MI_UIKit_Icon_Check, overrideStatusIndicatorVisibility=true}
   slot {class=GridSlot, properties {horizontalAlignment=HAlign_Right, verticalAlignment=VAlign_Top, padding=<ltrb 0 8 8 0>}}
  }
 ]
}
