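# Full-screen-height billboard row for the Item Shop, showing an offer group through WBP_Billboard_Details.
# Becoming the active row sets the selected group, eases text in, clears 3D preview, sends analytics and installs custom up/down navigation; other states reverse text and navigation.
# String metadata picks horizontal alignment, with padding adjusted for mobile, safe zone and aspect ratio.
# On mobile, a pulsing arrow shows while active with gamepad input.
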WidgetBlueprint /ItemShop/MultiProduct/Billboard/WBP_Row_Billboard
graphs [
 Clear3DPreview {nodes [
   Clear3DPreview() {is=title K2Node_FunctionEntry}
   n1 = Get FortItemShopSelectionVM() -> (FortItemShopSelectionVM:Fort Item Shop Selection VM Object Reference)
   > Set ItemsToPreview!(
    execute
    ItemsToPreview:Array of Item Shop Preview Data Object References=Make Array!() -> (Array:Array of Item Shop Preview Data Object References)
    self:Fort Item Shop Selection VM Object Reference=@n1
   ) -> (then, Output_Get:Array of Item Shop Preview Data Object References)
   > Set PreviewedItem!(execute, PreviewedItem:Item Shop Preview Data Object Reference, self:Fort Item Shop Selection VM Object Reference=@n1) -> (then, Output_Get:Item Shop Preview Data Object Reference)
  ]}
 EventGraph {nodes [
   n1 = OnRowActivated!() -> (OutputDelegate, then, IsImmediate:bool=~) {is=title K2Node_CustomEvent}
   > Ease in Text!(
    execute
    self:WBP Billboard Details Object Reference=Get WBP_Billboard_Details() -> (WBP_Billboard_Details:WBP Billboard Details Object Reference)
    IsImmediate:bool=@n1.IsImmediate
   ) -> (then)
   > K2Node_Composite_0 = MobileArrowManager!(Activated in:Exec, Deactivated in:Exec=@n3.then) -> (Activated out:Exec, Deactivated out:Exec) {is=title K2Node_Composite}
   n2 = OnRowDeactivated!() -> (OutputDelegate, then, IsImmediate:bool=~) {is=title K2Node_CustomEvent}
   > n3 = Ease out Text!(execute, self:WBP Billboard Details Object Reference=Get WBP_Billboard_Details(), IsImmediate:bool=@n2.IsImmediate) -> (then)
   Event I_ResetRow() {is=title K2Node_Event}
   > ResetRow()
   n4 = Event I_InitializeRow!() -> (
    OutputDelegate
    then
    FortItemShopRow_OfferGroup:Fort Item Shop Offer Group VM Object Reference
    FortItemShopRow_Offers:Array of Fort Item Shop Offer VM Object References
    FortItemShopRow_bIsFirstRowInOfferGroup:bool=~
    FortItemShopRow_bIsLastRowInOfferGroup:bool=~
    ContentHasFocus:bool
   ) {is=title K2Node_Event}
   > InitializeRow!(execute, self, OfferGroupVM:Fort Item Shop Offer Group VM Object Reference=@n4.FortItemShopRow_OfferGroup) -> (then)
   n5 = Event I_SetRowState!() -> (OutputDelegate, then, InRowState:E_ItemShop_RowState Enum=NewEnumerator4) {is=title K2Node_Event}
   > SetRowState!(execute, self, InRowState:E_ItemShop_RowState Enum=@n5.InRowState) -> (then)
   n6 = Event I_SetOfferGroupVM!() -> (OutputDelegate, then, OfferGroupVM:Fort Item Shop Offer Group VM Object Reference) {is=title K2Node_Event}
   > SetFortItemShopOfferGroupVM!(execute, self, Viewmodel:Fort Item Shop Offer Group VM Object Reference=@n6.OfferGroupVM) -> (then)
   Event OnAddedToFocusPath!() -> (OutputDelegate, then, InFocusEvent:Focus Event Structure) {is=title K2Node_Event}
   > n7 = Branch(Condition=IsUsingTouch(PlayerController=GetOwningPlayer()))
   Call On Row Focused!(execute:Exec=@n7.else, self:WBP Row Billboard Object Reference, RowWidget:User Widget Object Reference=Self-Reference()) -> (then:Exec) {is=title K2Node_CallDelegate}
   n8 = Event PreConstruct()
   > n9 = Branch(Condition=@n8.IsDesignTime)
   > SetLayoutAlignment!(execute, self, Alignment:EHorizontalAlignment Enum=Get LayoutAlignment() -> (LayoutAlignment:EHorizontalAlignment Enum=HAlign_Fill)) -> (then)
   Ease out Text(execute=@n9.else, self=Get WBP_Billboard_Details(), IsImmediate=true)
   Event Construct()
   > n10 = GetViewportWidgetGeometry!(execute) -> (then, ReturnValue:Geometry Structure)
   > SetSize!(
    execute
    self:Spacer Object Reference=Get Spacer_FullscreenHeight() -> (Spacer_FullscreenHeight:Spacer Object Reference)
    InSize_X:double=0
    InSize_Y:double=GetLocalSize!(Geometry:"Geometry Structure (by ref)"=@n10) -> (ReturnValue_X:double, ReturnValue_Y:double).ReturnValue_Y
   ) -> (then)
  ]}
 MobileArrowManager {
  path=EventGraph.K2Node_Composite_0.MobileArrowManager
  nodes [
   n1 = OnInputChanged!() -> (OutputDelegate, then, bNewInputType:ECommonInputType Enum) {is=title K2Node_CustomEvent}
   n2 = IsMobileGame()
   n3 = "Equal (Enum)"(A=GetCurrentInputType(self=Get CommonInputSubsystem()), B=Gamepad)
   n4 = Get ArrowButtonMobile() -> (ArrowButtonMobile:WBP UIKit View More Arrow Button Object Reference)
   SetVisibility(execute=[@n1.then, @n7.then], self=@n4, InVisibility=Select(Option 0=Collapsed, Option 1=Visible, Index=AND Boolean(A=@n3, B=@n2)))
   > PlayPulseAnimation(self:WBP UIKit View More Arrow Button Object Reference=@n4)
   > Outputs!(Activated out:Exec, Deactivated out:Exec=@n8.then)
   n5 = Inputs!() -> (Activated in:Exec, Deactivated in:Exec)
   n6 = Get CommonInputSubsystem()
   > n7 = Bind Event to On Input Method Changed(self=@n6, Delegate=@n1)
   Unbind Event from On Input Method Changed(execute=@n5.Deactivated in, self=@n6, Delegate=@n1)
   > n8 = SetVisibility(self=Get ArrowButtonMobile(), InVisibility=Collapsed)
  ]
 }
 I_GetClosestFocusTarget {nodes [
   I_GetClosestFocusTarget!() -> (then, TargetLocation:Vector 2D Structure) {is=title K2Node_FunctionEntry}
   > n1 = I_GetDesiredFocusTarget!(execute, self) -> (then, TileWidget:Widget Object Reference)
   > Return Node!(execute, Target:Widget Object Reference=@n1)
  ]}
 I_GetDesiredFocusTarget {nodes [
   I_GetDesiredFocusTarget() {is=title K2Node_FunctionEntry}
   > Return Node!(
    execute
    TileWidget:Widget Object Reference=GetDesiredFocusTarget!(self:WBP Billboard Details Object Reference=Get WBP_Billboard_Details()) -> (Target:Widget Object Reference)
   )
  ]}
 I_GetDesiredScrollDestination {nodes [
   I_GetDesiredScrollDestination() {is=title K2Node_FunctionEntry}
   > Return Node!(execute, Destination:EDescendantScrollDestination Enum=TopOrLeft)
  ]}
 I_GetDesiredScrollPadding {nodes [
   I_GetDesiredScrollPadding() {is=title K2Node_FunctionEntry}
   > Return Node!(execute, ScrollPadding:double=0)
  ]}
 I_GetFocusTargetFromOfferID {nodes [
   I_GetFocusTargetFromOfferID!() -> (then, OfferID:String) {is=title K2Node_FunctionEntry}
   > n1 = I_GetDesiredFocusTarget() {is=title K2Node_CallFunction}
   > Return Node!(execute, FocusTarget:Widget Object Reference=@n1)
  ]}
 I_GetScrollTarget {nodes [
   I_GetScrollTarget() {is=title K2Node_FunctionEntry}
   > Return Node!(execute, ScrollTarget:Widget Object Reference=Get Overlay_All() -> (Overlay_All:Overlay Object Reference))
  ]}
 InitializeRow {nodes [
   n1 = InitializeRow!() -> (then, OfferGroupVM:Fort Item Shop Offer Group VM Object Reference) {is=title K2Node_FunctionEntry}
   > SetFortItemShopOfferGroupVM(Viewmodel=@n1)
  ]}
 OnNavigatedFrom {nodes [
   n1 = OnNavigatedFrom!() -> (then, Navigation:EUINavigation Enum) {is=title K2Node_FunctionEntry}
   > Call On Row Navigated From!(
    execute:Exec
    self:WBP Row Billboard Object Reference
    Direction:EUINavigation Enum=@n1
    AnchorPosition:Vector 2D Structure=GetAbsolutePosition!(Geometry:"Geometry Structure (by ref)"=GetCachedGeometry!(self:Widget Object Reference) -> (ReturnValue:"Geometry Structure (by ref)")) -> (ReturnValue:Vector 2D Structure)
   ) -> (then:Exec) {is=title K2Node_CallDelegate}
   > Return Node!(execute, ReturnValue:Widget Object Reference)
  ]}
 OnOfferFocused {nodes [OnOfferFocused!() -> (then, OfferID:String) {is=title K2Node_FunctionEntry}]}
 OnRowFocused {nodes [OnRowFocused!() -> (then, RowWidget:User Widget Object Reference) {is=title K2Node_FunctionEntry}]}
 OnRowNavigatedFrom {nodes [OnRowNavigatedFrom!() -> (then, Direction:EUINavigation Enum, AnchorPosition:Vector 2D Structure) {is=title K2Node_FunctionEntry}]}
 ResetRow {nodes [
   ResetRow() {is=title K2Node_FunctionEntry}
   > StopAnimationsAndLatentActions(self:User Widget Object Reference)
   > SetFortItemShopOfferGroupVM()
  ]}
 SetLayoutAlignment {nodes [
   n1 = SetLayoutAlignment!() -> (then, Alignment:EHorizontalAlignment Enum=HAlign_Fill) {is=title K2Node_FunctionEntry}
   > n2 = Set LayoutAlignment(LayoutAlignment:EHorizontalAlignment Enum=@n1) -> (Output_Get:EHorizontalAlignment Enum=HAlign_Fill)
   > Set HorizontalAlignment!(
    execute
    HorizontalAlignment:EHorizontalAlignment Enum=@n2
    self:Overlay Slot Object Reference=SlotAsOverlaySlot(Widget=Get WBP_Billboard_Details())
   ) -> (then, Output_Get:EHorizontalAlignment Enum=HAlign_Fill)
   > n3 = GetSafeZonePadding!(execute) -> (
    then
    SafePadding_X:double
    SafePadding_Y:double
    SafePadding_Z:double
    SafePadding_W:double
    SafePaddingScale:Vector 2D Structure
    SpillOverPadding:Vector 4 Structure
   )
   n4 = IsMobileGame()
   n5 = "float > float"(A=@n3.SafePadding_Y)
   n6 = Select(
    Option 0:Margin Structure=Get DetailsPadding_LeftAligned() -> (DetailsPadding_LeftAligned:Margin Structure)
    Option 1:Margin Structure=Select(
     Option 0=Get DetailsPadding_LeftAligned_Mobile() -> (DetailsPadding_LeftAligned_Mobile:Margin Structure)
     Option 1=Get DetailsPadding_LeftAligned_Mobile_TopSafezone() -> (DetailsPadding_LeftAligned_Mobile_TopSafezone:Margin Structure)
     Index=@n5
    )
    Index=@n4
   ) -> (ReturnValue:Margin Structure)
   n7 = Select(
    Option 0=Get DetailsPadding_CenterAligned_Mobile() -> (DetailsPadding_CenterAligned_Mobile:Margin Structure)
    Option 1=Get DetailsPadding_CenterAligned_Mobile_TopSafezone() -> (DetailsPadding_CenterAligned_Mobile_TopSafezone:Margin Structure)
    Index=@n5
   )
   n8 = Get DetailsPadding_CenterAligned() -> (DetailsPadding_CenterAligned:Margin Structure)
   n9 = Select(
    Option 0=Get DetailsPadding_RightAligned_Mobile() -> (DetailsPadding_RightAligned_Mobile:Margin Structure)
    Option 1=Get DetailsPadding_RightAligned_Mobile_TopSafezone() -> (DetailsPadding_RightAligned_Mobile_TopSafezone:Margin Structure)
    Index=@n5
   )
   n10 = Get DetailsPadding_RightAligned() -> (DetailsPadding_RightAligned:Margin Structure)
   > Set Padding!(
    execute
    Padding:Margin Structure=Select!(
     HAlign_Fill:Margin Structure=@n6
     HAlign_Left:Margin Structure=@n6
     HAlign_Center:Margin Structure=Select!(Option 0:Margin Structure=@n8, Option 1:Margin Structure=@n7, Index:bool=@n4) -> (ReturnValue)
     HAlign_Right:Margin Structure=Select(Option 0=@n10, Option 1=@n9, Index=@n4)
     Index:EHorizontalAlignment Enum=Get LayoutAlignment() -> (LayoutAlignment=HAlign_Fill)
    ) -> (ReturnValue)
    self:Overlay Slot Object Reference=SlotAsOverlaySlot(Widget=Get WBP_Billboard_Details())
   ) -> (then, Output_Get:Margin Structure)
   > Branch(Condition=IsMobileGame())
   n11 = Get WBP_Billboard_Details()
   > SetVerticalAlignment(self=SlotAsOverlaySlot(Widget=@n11), InVerticalAlignment=VAlign_Top)
   n12 = "Equal (Enum)"(A:EHorizontalAlignment Enum=Get LayoutAlignment() -> (LayoutAlignment=HAlign_Fill), B:EHorizontalAlignment Enum=HAlign_Center)
   n13 = GetViewportSize!() -> (ReturnValue_X:double, ReturnValue_Y:double)
   n14 = "float >= float"(A=float / float(A=@n13, B=@n13.ReturnValue_Y), B=1.34)
   > SetPadding!(
    execute
    self:Vertical Box Slot Object Reference=SlotAsVerticalBoxSlot(Widget=Get WBP_Billboard_CTAButtonAndPrice!(self:WBP Billboard Details Object Reference=@n11) -> (WBP_Billboard_CTAButtonAndPrice:WBP Billboard CTAButton And Price Object Reference))
    InPadding_Left:float=SelectFloat(A=-320, B=0, bPickA=AND Boolean(A=@n12, B=@n14))
    InPadding_Top:float=48
    InPadding_Right:float=0
    InPadding_Bottom:float=0
   ) -> (then)
  ]}
 SetRowState {nodes [
   n1 = SetRowState!() -> (then, InRowState:E_ItemShop_RowState Enum=NewEnumerator0) {is=title K2Node_FunctionEntry}
   > n2 = Branch(Condition="Equal (Enum)"(A:E_ItemShop_RowState Enum=@n1, B:E_ItemShop_RowState Enum=NewEnumerator0))
   > Set SelectedOfferGroup!(
    execute
    SelectedOfferGroup:Fort Item Shop Offer Group VM Object Reference=Get FortItemShopOfferGroupVM() -> (FortItemShopOfferGroupVM:Fort Item Shop Offer Group VM Object Reference)
    self:Fort Item Shop Selection VM Object Reference=Get FortItemShopSelectionVM()
   ) -> (then, Output_Get:Fort Item Shop Offer Group VM Object Reference)
   n3 = "Not Equal (Enum)"!(A:E_ItemShop_RowState Enum=@n1, B:E_ItemShop_RowState Enum=Get RowState() -> (RowState:E_ItemShop_RowState Enum=NewEnumerator4)) -> (ReturnValue:bool)
   > Branch(Condition=@n3)
   > Set RowState(RowState:E_ItemShop_RowState Enum=Get InRowState() -> (InRowState:E_ItemShop_RowState Enum=NewEnumerator4)) -> (Output_Get:E_ItemShop_RowState Enum=NewEnumerator4)
   > OnRowActivated!(execute, self, IsImmediate:bool=false) -> (then) {is=title K2Node_CallFunction}
   > Clear3DPreview() {is=title K2Node_CallFunction}
   n4 = Create Event()
   > SetNavigationRuleCustomBoundary!(execute, self:Widget Object Reference, Direction:EUINavigation Enum=Up, InCustomDelegate:Delegate=@n4) -> (then)
   > SetNavigationRuleCustomBoundary(Direction=Down, InCustomDelegate=@n4)
   > n5 = Get!(execute:Exec) -> (then:Exec, else:Exec, FortItemShopOfferGroupVM:Fort Item Shop Offer Group VM Object Reference)
   n6 = Get Offers!(self:Fort Item Shop Offer Group VM Object Reference=@n5) -> (Offers:Array of Fort Item Shop Offer VM Object References)
   > Branch(Condition=Is Valid Index!(TargetArray:Array of Fort Item Shop Offer VM Object References=@n6, IndexToTest:int) -> (ReturnValue:bool))
   > Call On Offer Focused!(
    execute:Exec
    self:WBP Row Billboard Object Reference
    OfferID:String=Get OfferId!(self:Fort Item Shop Offer VM Object Reference="Get (a copy)"(Array:Array of Fort Item Shop Offer VM Object References=@n6) -> (Output:Fort Item Shop Offer VM Object Reference)) -> (OfferId:String)
   ) -> (then:Exec) {is=title K2Node_CallDelegate}
   n7 = Get FortItemShopOfferGroupVM()
   n8 = Get Offers(self=@n7)
   > SendImpressionAnalyticsForRow!(
    execute
    self:Fort Item Shop Selection VM Object Reference=Get FortItemShopSelectionVM()
    Row:"Fort Item Shop Row Structure (by ref)"=GetRowFromOffer!(
     self:Fort Item Shop Selection VM Object Reference=Get FortItemShopSelectionVM()
     Offer:Fort Item Shop Offer VM Object Reference="Get (a copy)"(Array=@n8)
     PreferredSection:String=Get!(self:Fort Item Shop Offer Group VM Object Reference=@n7) -> (OwningLayout_SectionId:String, OwningLayout_SubsectionId:String)
    ) -> (ReturnValue:Fort Item Shop Row Structure)
    OffersNum:int=1
   ) -> (then)
   Branch(execute=@n2.else, Condition=@n3)
   > Set RowState(RowState=Get InRowState() -> (InRowState=NewEnumerator4)) -> (Output_Get=NewEnumerator4)
   > OnRowDeactivated!(execute, self, IsImmediate:bool=false) -> (then) {is=title K2Node_CallFunction}
   > SetNavigationRuleBase!(execute, self:Widget Object Reference, Direction:EUINavigation Enum=Up, Rule:EUINavigationRule Enum=Escape) -> (then)
   > SetNavigationRuleBase(Direction=Down, Rule=Escape)
  ]}
 VM_StringMetadataUpdated {nodes [
   n1 = VM_StringMetadataUpdated!() -> (then, StringMetadata:Map of Names to Strings) {is=title K2Node_FunctionEntry}
   > n2 = GetAlignmentFromStringMetadata!(execute, StringMetadata:Map of Names to Strings=@n1) -> (then, Alignment:EHorizontalAlignment Enum=HAlign_Fill, Found:bool)
   > Branch(Condition=@n2.Found)
   > SetLayoutAlignment(Alignment=@n2) {is=title K2Node_CallFunction}
  ]}
]
variables [
 DetailsPadding_CenterAligned {type=Margin, default=<ltrb 0 0 89 160>}
 DetailsPadding_CenterAligned_Mobile {type=Margin, default=<ltrb 0 128 89 0>}
 DetailsPadding_CenterAligned_Mobile_TopSafezone {type=Margin, default=<ltrb 0 216 89 0>}
 DetailsPadding_LeftAligned {type=Margin, default=<ltrb 60 0 0 192>}
 DetailsPadding_LeftAligned_Mobile {type=Margin, default=<ltrb 60 128 0 0>}
 DetailsPadding_LeftAligned_Mobile_TopSafezone {type=Margin, default=<ltrb 60 216 0 0>}
 DetailsPadding_RightAligned {type=Margin, default=<ltrb 0 0 110 192>}
 DetailsPadding_RightAligned_Mobile {type=Margin, default=<ltrb 0 128 80 0>}
 DetailsPadding_RightAligned_Mobile_TopSafezone {type=Margin, default=<ltrb 0 216 80 0>}
 LayoutAlignment {type=EHorizontalAlignment, default=HAlign_Left}
 OnOfferFocused {}
 OnRowFocused {}
 OnRowNavigatedFrom {}
 RowState {type=E_ItemShop_RowState, default=None}
]
widgets {
 info {inheritedWidgetCount=0, parentClass=&UserWidget, rootWidgetClass=&Overlay, widgetCount=4}
 widgets [
  Overlay_All {widgetClassPath=&Overlay, bIsVariable=true, widget=Overlay_All}
  WBP_Billboard_Details {
   widgetClassPath=&/WBP_Billboard_Details
   bIsVariable=true
   parent=Overlay_All
   widget=WBP_Billboard_Details
   properties {widthOverride=960}
   slot {class=OverlaySlot, properties {verticalAlignment=VAlign_Bottom, padding=<ltrb 60 0 0 200>}}
  }
  Spacer_FullscreenHeight {
   widgetClassPath=&Spacer
   bIsVariable=true
   parent=Overlay_All
   widget=Spacer_FullscreenHeight
   properties {size=<0 1080>}
   slot {class=OverlaySlot}
  }
  ArrowButtonMobile {
   widgetClassPath=/FortUILibrary/UIKit/Buttons/ArrowButtons/WBP_UIKit_ViewMore_ArrowButton
   bIsVariable=true
   parent=Overlay_All
   widget=ArrowButtonMobile
   properties {visibility=Collapsed}
   slot {class=OverlaySlot, properties {horizontalAlignment=HAlign_Center, verticalAlignment=VAlign_Bottom, padding=<ltrb 0 0 88 96>}}
  }
 ]
}
