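# Item Shop tile component that shows a bottom tag strip for an offer.
# Offer view model tags select either a simple text with simple background, or an auto-scrolling row of short description, divider and joined tags.
# Hover reveals the strip by animation or immediately, and touch input shows it unhovered.
# A config flag can hide the text in compact UI, and dynamic font materials take the view model's text color.
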WidgetBlueprint /ItemShop/MultiProduct/ItemShopTileComponents/WBP_ItemShopTile_BottomTags
animations [OnHover {
  binding_count=4
  length_seconds=0.1042
  object=&self:OnHover
  sequence {
   bindings [
    Image_Background {
     id=#1
     possessed_class=Image
     tracks [{
       type=MovieSceneWidgetMaterialTrack
       sections [{
         type=MovieSceneParameterSection
         end_frame=25
         start_frame=0
         channels [IsHovered {type=MovieSceneScriptingFloatChannel, keys [[0,0],[25,1]]}]
        }]
      }]
    }
    GridSlot_3 {
     id=#2
     parent_id=#3
     possessed_class=GridSlot
     tracks [{
       type=MovieSceneMarginTrack
       display_name=Padding
       sections [{
         type=MovieSceneMarginSection
         end_frame=25
         start_frame=0
         channels [
          Left {type=MovieSceneScriptingFloatChannel, default=8}
          Top {type=MovieSceneScriptingFloatChannel, keys [[0,32],[25,0]]}
          Right {type=MovieSceneScriptingFloatChannel, default=8}
          Bottom {type=MovieSceneScriptingFloatChannel, default=0}
         ]
        }]
      }]
    }
    Spacer {
     id=#4
     possessed_class=Spacer
     tracks [{
       type=MovieSceneDoubleVectorTrack
       display_name=Size
       sections [{
         type=MovieSceneDoubleVectorSection
         end_frame=25
         start_frame=0
         channels [
          Vector.X {type=MovieSceneScriptingDoubleChannel, default=0}
          Vector.Y {type=MovieSceneScriptingDoubleChannel, keys [[0,0],[25,32]]}
         ]
        }]
      }]
    }
    Switcher_Text {id=#3, possessed_class=CommonVisibilitySwitcher, tracks []}
   ]
   display_rate {numerator=240}
  }
 }]
graphs [
 EventGraph {nodes [
   n1 = OnInputMethodChanged_Event!() -> (OutputDelegate, then, bNewInputType:ECommonInputType Enum) {is=title K2Node_CustomEvent}
   > SetHoveredState!(execute, self:BPI Item Shop Tile Interface, IsHovered:bool="Equal (Enum)"(A=@n1.bNewInputType, B=Touch), IsImmediate:bool=true) -> (then)
   n2 = SetUsingSimpleTag!() -> (OutputDelegate, then, IsUsingSimpleTag:bool=~) {is=title K2Node_CustomEvent}
   > n3 = Load From Soft Material!(
    execute:Exec
    Material:Material Interface Soft Object Reference=Select(
     Option 0:Material Interface Soft Object Reference=/ItemShop/MultiProduct/Art/M_UI_ShopTile_BottomTags
     Option 1:Material Interface Soft Object Reference=/ItemShop/MultiProduct/Art/M_UI_ShopTile_BottomTags_Simple
     Index=@n2.IsUsingSimpleTag
    ) -> (ReturnValue:Material Interface Soft Object Reference)
   ) -> (then:Exec, Completed:Exec, Result:Material Interface Object Reference) {is=title MVVMK2Node_LoadSoftMaterial}
   SetBrushFromMaterial(execute=@n3.Completed, self=Get Image_Background() -> (Image_Background:Image Object Reference), Material=@n3)
   > SetActiveWidget(
    self=Get Switcher_Text() -> (Switcher_Text:Common Visibility Switcher Object Reference)
    Widget=Select(
     Option 0:Widget Object Reference=Get AutoScroll_Tags() -> (AutoScroll_Tags:WBP Item Shop Auto Scroll Horizontally Object Reference)
     Option 1:Widget Object Reference=Get Text_Simple() -> (Text_Simple:Common Text Object Reference)
     Index=@n2.IsUsingSimpleTag
    ) -> (ReturnValue:Widget Object Reference)
   )
   > InitializeTile(self:BPI Item Shop Tile Interface)
   Event InitializeTile() {is=title K2Node_Event}
   n4 = GetConfigDataBoolWithDefault!(EntryName:String=shop_hide_marquee_text_mobile, bDefaultValue:bool=false) -> (ReturnValue:bool)
   n5 = NOT Boolean(A=@n4)
   n6 = "Equal (Enum)"(A=GetCurrentInputType(self=Get CommonInputSubsystem()), B=Touch)
   n7 = AND Boolean(A=@n4, B=IsUsingCompactUI())
   > SetHoveredState(IsHovered=AND Boolean(A=AND Boolean(A=@n6, B=@n5), B=NOT Boolean(A=@n7)), IsImmediate=true)
   n8 = Event SetTileDimensions!() -> (OutputDelegate, then, Dimensions_X:double, Dimensions_Y:double) {is=title K2Node_Event}
   > Set WidthOverride!(
    execute
    WidthOverride:float=@n8.Dimensions_X
    self:Size Box Object Reference=Get VisualAttachment() -> (VisualAttachment:Common Visual Attachment Object Reference)
   ) -> (then, Output_Get:float)
   n9 = Event PreConstruct()
   > n10 = Branch(Condition=@n9.IsDesignTime)
   > SetUsingSimpleTag!(execute, self, IsUsingSimpleTag:bool=Get PreviewSimpleTag() -> (PreviewSimpleTag:bool)) -> (then) {is=title K2Node_CallFunction}
   > VM_FinalTextColorUpdated!(execute, self, Final Color:Linear Color Structure=(R=0.0056,G=0.0666,B=0.147,A=1)) -> (then)
   SetFontMaterial!(
    execute=@n10.else
    self:Text Object Reference=[
     Get Text_OfferType() -> (Text_OfferType:Common Text Object Reference)
     Get Text_Tags() -> (Text_Tags:Common Text Object Reference)
     Get Text_Divider() -> (Text_Divider:Common Text Object Reference)
    ]
    InMaterial:Material Interface Object Reference=/ItemShop/MultiProduct/Art/M_UI_ItemShop_BottomTags_Font
   ) -> (then)
   n11 = Event SetOfferVM!() -> (OutputDelegate, then, OfferVM:Fort Item Shop Offer VM Object Reference) {is=title K2Node_Event}
   > Set FortItemShopOfferVM(FortItemShopOfferVM:Fort Item Shop Offer VM Object Reference=@n11.OfferVM) -> (Output_Get:Fort Item Shop Offer VM Object Reference)
   n12 = Event SetHoveredState!() -> (OutputDelegate, then, IsHovered:bool, IsImmediate:bool) {is=title K2Node_Event}
   > n13 = Branch(Condition=@n12.IsImmediate)
   > Set!(
    execute
    Size_X:double=0
    Size_Y:double=Select(Option 0:float=0, Option 1:float=32, Index=@n12.IsHovered) -> (ReturnValue:float)
    self:Spacer Object Reference=Get Spacer() -> (Spacer:Spacer Object Reference)
   ) -> (then, Output_Get:Vector 2D Structure)
   > SetGridSlotPadding!(
    execute
    Widget:Widget Object Reference=Get Switcher_Text()
    Padding_Left:float=12
    Padding_Top:float=Select(Option 0=32, Option 1=0, Index=@n12.IsHovered)
    Padding_Right:float=12
    Padding_Bottom:float=0
   ) -> (then)
   > n14 = GetDynamicMaterial(self=Get Image_Background())
   > SetScalarParameterValue(self=@n14, ParameterName=IsHovered, Value="To Float (Boolean)"(InBool=@n12.IsHovered))
   > Branch(Condition=@n12.IsHovered)
   > ResetAutoScroll(self:WBP Item Shop Auto Scroll Horizontally Object Reference=Get AutoScroll_Tags())
   n15 = IsUsingCompactUI()
   n16 = GetConfigDataBoolWithDefault(EntryName=shop_hide_marquee_text_mobile, bDefaultValue=false)
   Branch(execute=@n13.else, Condition=NOT Boolean(A=AND Boolean(A=@n16, B=@n15)))
   > n17 = Branch(Condition=@n12.IsHovered)
   n18 = Get OnHover() -> (OnHover:Widget Animation Object Reference)
   > PlayAnimationForward(InAnimation=@n18)
   > ResetAutoScroll(self=Get AutoScroll_Tags())
   PlayAnimationReverse(execute=@n17.else, InAnimation=@n18)
   Event Construct()
   > Assign On Input Method Changed!(execute:Exec, self:Common Input Object Reference=Get CommonInputSubsystem(), Delegate:"Delegate (by ref)"=@n1) -> (then:Exec) {is=title K2Node_AssignDelegate}
   > Branch(Condition=AND Boolean(A=GetConfigDataBoolWithDefault(EntryName=shop_hide_marquee_text_mobile, bDefaultValue=false), B=IsUsingCompactUI()))
   > Set Visibility(Visibility=Collapsed, self=Get Switcher_Text())
   Event Destruct()
   > Unbind Event from On Input Method Changed(self=Get CommonInputSubsystem(), Delegate=@n1)
  ]}
 GetOfferVM {nodes [
   GetOfferVM() {is=title K2Node_FunctionEntry}
   > Return Node!(execute, OfferVM:Fort Item Shop Offer VM Object Reference)
  ]}
 GetTileAnchor {nodes [
   GetTileAnchor() {is=title K2Node_FunctionEntry}
   > Return Node!(execute, Anchor:Vector 2D Structure)
  ]}
 GetTileWidget {nodes [
   GetTileWidget() {is=title K2Node_FunctionEntry}
   > Return Node!(execute, NewParam:User Widget Object Reference)
  ]}
 GetVisibilityFromTexts {nodes [
   n1 = GetVisibilityFromTexts!() -> (then, ShortDescription:Text, OfferType:EFortItemShopOfferType Enum=LevelBundle, TagTexts:Array of Texts) {is=title K2Node_FunctionEntry}
   n2 = OR Boolean(
    A=CanShowOfferDescription!(Description:Text=@n1, OfferType:EFortItemShopOfferType Enum=@n1.OfferType) -> (CanShow:bool)
    B=NOT Boolean(A=Is Empty!(TargetArray:Array of Texts=@n1.TagTexts) -> (ReturnValue:bool))
   )
   n3 = "Not Equal (Enum)"!(A:EFortItemShopOfferType Enum=@n1.OfferType, B:EFortItemShopOfferType Enum=BattlePass) -> (ReturnValue:bool)
   > Return Node!(
    execute
    ReturnValue:ESlateVisibility Enum=Select(Option 0:ESlateVisibility Enum=Collapsed, Option 1:ESlateVisibility Enum=SelfHitTestInvisible, Index=AND Boolean(A=@n2, B=@n3)) -> (ReturnValue:ESlateVisibility Enum)
   )
  ]}
 VM_FinalTextColorUpdated {nodes [
   n1 = VM_FinalTextColorUpdated!() -> (then, Final Color_R:float=~, Final Color_G:float=~, Final Color_B:float=~, Final Color_A:float=~) {is=title K2Node_FunctionEntry}
   > n2 = GetDynamicFontMaterial!(execute, self:Text Object Reference=Get Text_OfferType()) -> (then, ReturnValue:Material Instance Dynamic Object Reference)
   > n3 = GetDynamicFontMaterial(self=Get Text_Divider())
   > n4 = GetDynamicFontMaterial(self=Get Text_Tags())
   > SetVectorParameterValue(
    self=[@n2, @n3, @n4]
    ParameterName=Color
    Value=Make Linear Color!(R:float=@n1, G:float=@n1.Final Color_G, B:float=@n1.Final Color_B, A:float=1) -> (LinearColor:Linear Color Structure) {is=title K2Node_MakeStruct}
   )
  ]}
 VM_TagsUpdated {nodes [
   VM_TagsUpdated!() -> (then, Visibility:ESlateVisibility Enum=Visible) {is=title K2Node_FunctionEntry}
   > n1 = Set Tags(Tags:Array of Texts=GetTagTexts!(self:Fort Item Shop Offer VM Object Reference=Get FortItemShopOfferVM() -> (FortItemShopOfferVM:Fort Item Shop Offer VM Object Reference)) -> (ReturnValue:Array of Texts)) -> (Output_Get:Array of Texts)
   n2 = Is Empty(TargetArray=@n1)
   > SetUsingSimpleTag(IsUsingSimpleTag=@n2)
   n3 = Get Visibility() -> (Visibility:ESlateVisibility Enum=Visible)
   > Branch(Condition=AND Boolean(A=NOT Boolean(A=@n2), B="Equal (Enum)"(A:ESlateVisibility Enum=@n3, B:ESlateVisibility Enum=SelfHitTestInvisible)))
   > n4 = Branch(Condition=TextIsEmpty(InText=Get ShortDescription() -> (ShortDescription:Text)))
   n5 = Get Tags() -> (Tags:Array of Texts)
   n6 = For Each Loop(Exec=[@n9.then, @n4.then], Array:Array of Texts=@n5) -> (Array Element:Text)
   Set Text(execute=@n6.Completed, Text=Get TagsText() -> (TagsText:Text), self=Get Text_Tags())
   n7 = "Not Equal (Integer)"(A=@n6.Array Index, B=Last Index!(TargetArray:Array of Texts=@n5) -> (ReturnValue=0))
   n8 = Get TagsDividerText() -> (TagsDividerText:Text)
   Set TagsText(
    execute=@n6.LoopBody
    TagsText:Text=Format Text!(
     Format="LOCTEXT(\"{TagsText}{Tag}{TagDivider}\")"
     TagsText:Text=Get TagsText()
     Tag:Text=@n6
     TagDivider:Text=Select(Option 0:Text, Option 1:Text=@n8, Index=@n7) -> (ReturnValue:Text)
    ) -> (Result)
   ) -> (Output_Get:Text)
   Set Text(execute=@n4.else, Text=Get ShortDescription(), self=Get Text_OfferType())
   > n9 = Set Text(Text=Get DividerText() -> (DividerText:Text), self=Get Text_Divider())
  ]}
]
variables [
 DividerText {type=string, default=+}
 PreviewSimpleTag {type=boolean}
 ShortDescription {type=string, default=""}
 TagsDividerText {type=string, default=", "}
]
widgets {
 info {inheritedWidgetCount=0, namedSlotCount=1, parentClass=&UserWidget, rootWidgetClass=&GridPanel, widgetCount=12}
 widgets [
  GP_Root {widgetClassPath=&GridPanel, bIsVariable=true, widget=GP_Root, properties {bOverride_Cursor=true}}
  Spacer {widgetClassPath=&Spacer, bIsVariable=true, parent=GP_Root, widget=Spacer, properties {size=<0 32>}, slot {class=GridSlot}}
  VisualAttachment {
   widgetClassPath=&CommonVisualAttachment
   bIsVariable=true
   parent=GP_Root
   widget=VisualAttachment
   properties {bOverride_HeightOverride=true, bOverride_WidthOverride=true, heightOverride=32, widthOverride=256, contentAnchor=<0 1>}
   slot {class=GridSlot, properties {verticalAlignment=VAlign_Bottom}}
  }
  GridPanel_38 {
   widgetClassPath=&GridPanel
   parent=VisualAttachment
   widget=GridPanel_38
   properties {columnFill [1], rowFill [1]}
   slot {class=SizeBoxSlot}
  }
  Image_Background {
   widgetClassPath=&Image
   bIsVariable=true
   parent=GridPanel_38
   widget=Image_Background
   properties {brush {
     bIsDynamicallyLoaded=false
     drawAs=Image
     imageType=FullColor
     mirroring=NoMirror
     resourceName=None
     resourceObject=&M_UI_ShopTile_BottomTags
     tiling=NoTile
     imageSize=<32 32>
     margin=<ltrb 0 0 0 0>
     outlineSettings {
      bUseBrushTransparency=false
      roundingType=HalfHeightRadius
      width=0
      color {colorUseRule=UseColor_Specified, specifiedColor=<rgba 0 0 0 0>}
      cornerRadii=<0 0 0 0>
     }
     tintColor {colorUseRule=UseColor_Specified, specifiedColor=<rgba 1 1 1 1>}
     uVRegion {bIsValid=false, max=<0 0>, min=<0 0>}
    }}
   slot {class=GridSlot}
  }
  Switcher_Text {
   widgetClassPath=&CommonVisibilitySwitcher
   bIsVariable=true
   parent=GridPanel_38
   widget=Switcher_Text
   slot {class=GridSlot, properties {padding=<ltrb 12 0 12 0>}}
  }
  Text_Simple {
   widgetClassPath=&CommonTextBlock
   bIsVariable=true
   parent=Switcher_Text
   widget=Text_Simple
   properties {
    clipping=ClipToBounds
    justification=Center
    mobileFontSizeMultiplier=1.4
    scrollStyle=&TextScrollStyle-Base_C
    text="[PH] Outfit"
    textTransformPolicy=ToUpper
    colorAndOpacity {colorUseRule=UseColor_Specified, specifiedColor=<rgba 1 1 1 0.8>}
    font {
     bForceMonospaced=false
     bMaterialIsStencil=false
     fontMaterial=null
     fontObject=&HeadingNow
     letterSpacing=0
     monospacedWidth=1
     size=9.75
     skewAmount=0
     typefaceFontName=75 Medium Italic
     outlineSettings {
      bApplyOutlineToDropShadows=false
      bMiteredCorners=false
      bSeparateFillAlpha=false
      outlineMaterial=null
      outlineSize=0
      outlineColor=<rgba 0 0 0 1>
     }
    }
    shadowOffset=<0 0>
   }
   slot {class=CommonVisibilitySwitcherSlot, properties {horizontalAlignment=HAlign_Center, verticalAlignment=VAlign_Center}}
  }
  AutoScroll_Tags {
   widgetClassPath=/ItemShop/MultiProduct/WBP_ItemShop_AutoScrollHorizontally
   bIsVariable=true
   parent=Switcher_Text
   widget=AutoScroll_Tags
   slot {class=CommonVisibilitySwitcherSlot, properties {horizontalAlignment=HAlign_Center, verticalAlignment=VAlign_Center}}
  }
  HorizontalBox_0 {widgetClassPath=&HorizontalBox, namedSlotHost=AutoScroll_Tags, widget=HorizontalBox_0}
  Text_OfferType {
   widgetClassPath=&CommonTextBlock
   bIsVariable=true
   parent=HorizontalBox_0
   widget=Text_OfferType
   properties {
    bIsScrollingEnabled=false
    mobileFontSizeMultiplier=1.4
    text="[PH] Bundle"
    textTransformPolicy=ToUpper
    font {
     bForceMonospaced=false
     bMaterialIsStencil=false
     fontMaterial=null
     fontObject=&HeadingNow
     letterSpacing=0
     monospacedWidth=1
     size=9.75
     skewAmount=0
     typefaceFontName=75 Medium Italic
     outlineSettings {
      bApplyOutlineToDropShadows=false
      bMiteredCorners=false
      bSeparateFillAlpha=false
      outlineMaterial=null
      outlineSize=0
      outlineColor=<rgba 0 0 0 1>
     }
    }
    shadowOffset=<0 0>
   }
   slot {class=HorizontalBoxSlot, properties {horizontalAlignment=HAlign_Center, verticalAlignment=VAlign_Center}}
  }
  Text_Divider {
   widgetClassPath=&CommonTextBlock
   bIsVariable=true
   parent=HorizontalBox_0
   widget=Text_Divider
   properties {
    bIsScrollingEnabled=false
    mobileFontSizeMultiplier=1.4
    text=+
    textTransformPolicy=ToUpper
    font {
     bForceMonospaced=false
     bMaterialIsStencil=false
     fontMaterial=null
     fontObject=&HeadingNow
     letterSpacing=0
     monospacedWidth=1
     size=11.25
     skewAmount=0
     typefaceFontName=75 Medium Italic
     outlineSettings {
      bApplyOutlineToDropShadows=false
      bMiteredCorners=false
      bSeparateFillAlpha=false
      outlineMaterial=null
      outlineSize=0
      outlineColor=<rgba 0 0 0 1>
     }
    }
    shadowOffset=<0 0>
   }
   slot {class=HorizontalBoxSlot, properties {horizontalAlignment=HAlign_Center, verticalAlignment=VAlign_Center, padding=<ltrb 10 0 6 1>}}
  }
  Text_Tags {
   widgetClassPath=&CommonTextBlock
   bIsVariable=true
   parent=HorizontalBox_0
   widget=Text_Tags
   properties {
    bIsScrollingEnabled=false
    mobileFontSizeMultiplier=1.4
    text=Selectable Styles
    textTransformPolicy=ToUpper
    font {
     bForceMonospaced=false
     bMaterialIsStencil=false
     fontMaterial=null
     fontObject=&HeadingNow
     letterSpacing=0
     monospacedWidth=1
     size=9.75
     skewAmount=0
     typefaceFontName=75 Medium Italic
     outlineSettings {
      bApplyOutlineToDropShadows=false
      bMiteredCorners=false
      bSeparateFillAlpha=false
      outlineMaterial=null
      outlineSize=0
      outlineColor=<rgba 0 0 0 1>
     }
    }
    shadowOffset=<0 0>
   }
   slot {class=HorizontalBoxSlot, properties {horizontalAlignment=HAlign_Center, verticalAlignment=VAlign_Center, padding=<ltrb 0 0 2 0>}}
  }
 ]
}
