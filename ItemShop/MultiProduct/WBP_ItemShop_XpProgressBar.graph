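# Item shop widget that shows XP progress as a glowing, outlined UIKit bar beside a rich-text label with placeholder text.
# PreConstruct shows or collapses the label according to the ShowText variable.
# GetProgressRatio safely divides current by total progress and clamps it to 0–1, while GetProgressBarVisibility returns Collapsed when the total is zero.
# Neither helper function is called within this widget's own graphs.
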WidgetBlueprint /ItemShop/MultiProduct/WBP_ItemShop_XpProgressBar
graphs [
 EventGraph {nodes [
   Event PreConstruct()
   > Set Visibility(
    Visibility=Select(Option 0=Collapsed, Option 1=SelfHitTestInvisible, Index=Get ShowText() -> (ShowText:bool))
    self=Get Text_Progress() -> (Text_Progress:Common Rich Text Block Object Reference)
   )
  ]}
 GetProgressBarVisibility {nodes [
   n1 = GetProgressBarVisibility!() -> (then, TotalProgress:int) {is=title K2Node_FunctionEntry}
   > Return Node!(execute, Visibility:ESlateVisibility Enum=Select(Option 0=SelfHitTestInvisible, Option 1=Collapsed, Index="Equal (Integer)"(A=@n1)))
  ]}
 GetProgressRatio {nodes [
   n1 = GetProgressRatio!() -> (then, CurrentProgress:int, TotalProgress:int) {is=title K2Node_FunctionEntry}
   n2 = "To Float (Integer)"(InInt=@n1.TotalProgress)
   n3 = "To Float (Integer)"(InInt=@n1)
   > Return Node!(execute, ReturnValue:double="Clamp (Float)"(Value=SafeDivide(A=@n3, B=@n2), Min=0, Max=1))
  ]}
]
variables [ShowText {type=boolean, default=true}]
widgets {
 info {inheritedWidgetCount=0, parentClass=&UserWidget, rootWidgetClass=&HorizontalBox, widgetCount=3}
 widgets [
  HorizontalBox_0 {widgetClassPath=&HorizontalBox, widget=HorizontalBox_0}
  WBP_UIKit_ProgressBar {
   widgetClassPath=/FortUILibrary/UIKit/ProgressBar/WBP_UIKit_ProgressBar
   parent=HorizontalBox_0
   widget=WBP_UIKit_ProgressBar
   properties {
    glowIntensity=0.1
    height=14
    outline_PX=4
    progressValue=0.75
    useCircular=false
    useGlow=true
    useOutline=true
    width=244
    colorA=<rgba 0.0185 0.521 1 1>
    colorB=<rgba 0.3231 0.2159 1 1>
    colorC=<rgba 0.7835 0.0999 1 1>
    emptyColor=<rgba 0.3231 0.2874 0.5972 0.15>
   }
   slot {class=HorizontalBoxSlot, properties {verticalAlignment=VAlign_Center, size {sizeRule=Fill, value=1}}}
  }
  Text_Progress {
   widgetClassPath=&CommonRichTextBlock
   parent=HorizontalBox_0
   widget=Text_Progress
   properties {
    bOverrideDefaultStyle=true
    mobileTextBlockScale=1.33
    text="[PH] 100,000<text color=\"C8C8C8\"> / 100K XP</>"
    defaultTextStyleOverride {
     overflowPolicy=Clip
     transformPolicy=None
     colorAndOpacity {colorUseRule=UseColor_Specified, specifiedColor=<rgba 1 1 1 1>}
     font {
      bForceMonospaced=false
      bMaterialIsStencil=false
      fontMaterial=null
      fontObject=&HeadingNow
      letterSpacing=0
      monospacedWidth=1
      size=13.5
      skewAmount=0
      typefaceFontName=75 Medium
      outlineSettings {
       bApplyOutlineToDropShadows=false
       bMiteredCorners=false
       bSeparateFillAlpha=false
       outlineMaterial=null
       outlineSize=0
       outlineColor=<rgba 0 0 0 1>
      }
     }
     highlightColor {colorUseRule=UseColor_Specified, specifiedColor=<rgba 0 0 0 1>}
     highlightShape {
      bIsDynamicallyLoaded=false
      drawAs=Image
      imageType=NoImage
      mirroring=NoMirror
      resourceName=None
      resourceObject=null
      tiling=NoTile
      imageSize=<32 32>
      margin=<ltrb 0 0 0 0>
      outlineSettings {
       bUseBrushTransparency=false
       roundingType=HalfHeightRadius
       width=0
       color {colorUseRule=UseColor_Specified, specifiedColor=<rgba 0 0 0 0>}
       cornerRadii=<0 0 0 0>
      }
      tintColor {colorUseRule=UseColor_Specified, specifiedColor=<rgba 1 1 1 1>}
      uVRegion {bIsValid=false, max=<0 0>, min=<0 0>}
     }
     selectedBackgroundColor {colorUseRule=UseColor_Foreground, specifiedColor=<rgba 1 0 1 1>}
     shadowColorAndOpacity=<rgba 0 0 0 1>
     shadowOffset=<0 0>
     strikeBrush {
      bIsDynamicallyLoaded=false
      drawAs=Image
      imageType=NoImage
      mirroring=NoMirror
      resourceName=None
      resourceObject=null
      tiling=NoTile
      imageSize=<32 32>
      margin=<ltrb 0 0 0 0>
      outlineSettings {
       bUseBrushTransparency=false
       roundingType=HalfHeightRadius
       width=0
       color {colorUseRule=UseColor_Specified, specifiedColor=<rgba 0 0 0 0>}
       cornerRadii=<0 0 0 0>
      }
      tintColor {colorUseRule=UseColor_Specified, specifiedColor=<rgba 1 1 1 1>}
      uVRegion {bIsValid=false, max=<0 0>, min=<0 0>}
     }
     underlineBrush {
      bIsDynamicallyLoaded=false
      drawAs=Image
      imageType=NoImage
      mirroring=NoMirror
      resourceName=None
      resourceObject=null
      tiling=NoTile
      imageSize=<32 32>
      margin=<ltrb 0 0 0 0>
      outlineSettings {
       bUseBrushTransparency=false
       roundingType=HalfHeightRadius
       width=0
       color {colorUseRule=UseColor_Specified, specifiedColor=<rgba 0 0 0 0>}
       cornerRadii=<0 0 0 0>
      }
      tintColor {colorUseRule=UseColor_Specified, specifiedColor=<rgba 1 1 1 1>}
      uVRegion {bIsValid=false, max=<0 0>, min=<0 0>}
     }
    }
   }
   slot {class=HorizontalBoxSlot, properties {verticalAlignment=VAlign_Center, padding=<ltrb 16 0 0 0>}}
  }
 ]
}
