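# Item Shop tile component widget that shows XP progress toward unlocking an offer for purchase.
# A backing Border stacks a progress bar, progress text and description, with a spacer for minimum height.
# SetOfferVM stores the offer view model and forwards it to the child XP bar.
# GetDescriptionText selects locked, available or purchased text from progress and ownership; GetOfferVM, GetTileAnchor and GetTileWidget are empty stubs.
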WidgetBlueprint /ItemShop/MultiProduct/ItemShopTileComponents/WBP_ItemShopTile_XpInfo
graphs [
 EventGraph {nodes [
   n1 = Event SetOfferVM!() -> (OutputDelegate, then, OfferVM:Fort Item Shop Offer VM Object Reference) {is=title K2Node_Event}
   > n2 = Set FortItemShopOfferVM(FortItemShopOfferVM:Fort Item Shop Offer VM Object Reference=@n1.OfferVM) -> (Output_Get:Fort Item Shop Offer VM Object Reference)
   > Set FortItemShopOfferVM!(
    execute
    FortItemShopOfferVM=@n2
    self:WBP Item Shop Xp Progress Bar Object Reference=Get WBP_ItemShop_ProgressBar() -> (WBP_ItemShop_ProgressBar:WBP Item Shop Xp Progress Bar Object Reference)
   ) -> (then, Output_Get)
  ]}
 GetDescriptionText {nodes [
   n1 = GetDescriptionText!() -> (then, CurrentProgress:int, TotalProgress:int, IsOwned:bool=~) {is=title K2Node_FunctionEntry}
   n2 = "integer >= integer"(A=@n1, B=@n1.TotalProgress)
   > Return Node!(
    execute
    ReturnValue:Text=Select(
     Option 0:Text=Select(Option 0="LOCTEXT(\"Unlocks ability to purchase\")", Option 1="LOCTEXT(\"Now available for purchase\")", Index=@n2)
     Option 1:Text="LOCTEXT(\"Purchased\")"
     Index=@n1.IsOwned
    ) -> (ReturnValue:Text)
   )
  ]}
 GetOfferVM {nodes [
   GetOfferVM() {is=title K2Node_FunctionEntry}
   > Return Node!(execute, OfferVM:Fort Item Shop Offer VM Object Reference)
  ]}
 GetTileAnchor {nodes [
   GetTileAnchor() {is=title K2Node_FunctionEntry}
   > Return Node!(execute, Anchor:Vector 2D Structure)
  ]}
 GetTileWidget {nodes [
   GetTileWidget() {is=title K2Node_FunctionEntry}
   > Return Node!(execute, NewParam:User Widget Object Reference)
  ]}
]
variables []
widgets {
 info {inheritedWidgetCount=0, parentClass=&UserWidget, rootWidgetClass=&Border, widgetCount=7}
 widgets [
  Border_0 {
   widgetClassPath=&Border
   widget=Border_0
   properties {
    background {
     bIsDynamicallyLoaded=false
     drawAs=Image
     imageType=FullColor
     mirroring=NoMirror
     resourceName=None
     resourceObject=/FortUILibrary/Systems/ItemShopTile/Materials/MI_UI_ItemShopTile_XpInfoBacking
     tiling=NoTile
     imageSize=<32 32>
     margin=<ltrb 0 0 0 0>
     outlineSettings {
      bUseBrushTransparency=false
      roundingType=HalfHeightRadius
      width=0
      color {colorUseRule=UseColor_Specified, specifiedColor=<rgba 0 0 0 0>}
      cornerRadii=<0 0 0 0>
     }
     tintColor {colorUseRule=UseColor_Specified, specifiedColor=<rgba 1 1 1 1>}
     uVRegion {bIsValid=false, max=<0 0>, min=<0 0>}
    }
    padding=<ltrb 0 0 0 0>
   }
  }
  Overlay_0 {widgetClassPath=&Overlay, parent=Border_0, widget=Overlay_0, slot {class=BorderSlot, properties {padding=<ltrb 0 0 0 0>}}}
  VerticalBox_1 {
   widgetClassPath=&VerticalBox
   parent=Overlay_0
   widget=VerticalBox_1
   slot {class=OverlaySlot, properties {horizontalAlignment=HAlign_Fill, verticalAlignment=VAlign_Fill, padding=<ltrb 10 12 10 10>}}
  }
  WBP_ItemShop_ProgressBar {
   widgetClassPath=/ItemShop/MultiProduct/WBP_ItemShop_XpProgressBar
   bIsVariable=true
   parent=VerticalBox_1
   widget=WBP_ItemShop_ProgressBar
   properties {showText=false}
   slot {class=VerticalBoxSlot, properties {verticalAlignment=VAlign_Top, size {sizeRule=Fill, value=1}}}
  }
  Text_Progress {
   widgetClassPath=&CommonRichTextBlock
   parent=VerticalBox_1
   widget=Text_Progress
   properties {
    bOverrideDefaultStyle=true
    mobileTextBlockScale=1.33
    scrollStyle=&TextScrollStyle-Base_C
    text="[PH] 100,000<text color=\"C8C8C8\"> / 100K XP</>"
    defaultTextStyleOverride {
     overflowPolicy=Clip
     transformPolicy=None
     colorAndOpacity {colorUseRule=UseColor_Specified, specifiedColor=<rgba 1 1 1 1>}
     font {
      bForceMonospaced=false
      bMaterialIsStencil=false
      fontMaterial=null
      fontObject=&HeadingNow
      letterSpacing=0
      monospacedWidth=1
      size=13.5
      skewAmount=0
      typefaceFontName=75 Medium
      outlineSettings {
       bApplyOutlineToDropShadows=false
       bMiteredCorners=false
       bSeparateFillAlpha=false
       outlineMaterial=null
       outlineSize=0
       outlineColor=<rgba 0 0 0 1>
      }
     }
     highlightColor {colorUseRule=UseColor_Specified, specifiedColor=<rgba 0 0 0 1>}
     highlightShape {
      bIsDynamicallyLoaded=false
      drawAs=Image
      imageType=NoImage
      mirroring=NoMirror
      resourceName=None
      resourceObject=null
      tiling=NoTile
      imageSize=<32 32>
      margin=<ltrb 0 0 0 0>
      outlineSettings {
       bUseBrushTransparency=false
       roundingType=HalfHeightRadius
       width=0
       color {colorUseRule=UseColor_Specified, specifiedColor=<rgba 0 0 0 0>}
       cornerRadii=<0 0 0 0>
      }
      tintColor {colorUseRule=UseColor_Specified, specifiedColor=<rgba 1 1 1 1>}
      uVRegion {bIsValid=false, max=<0 0>, min=<0 0>}
     }
     selectedBackgroundColor {colorUseRule=UseColor_Foreground, specifiedColor=<rgba 1 0 1 1>}
     shadowColorAndOpacity=<rgba 0 0 0 1>
     shadowOffset=<0 0>
     strikeBrush {
      bIsDynamicallyLoaded=false
      drawAs=Image
      imageType=NoImage
      mirroring=NoMirror
      resourceName=None
      resourceObject=null
      tiling=NoTile
      imageSize=<32 32>
      margin=<ltrb 0 0 0 0>
      outlineSettings {
       bUseBrushTransparency=false
       roundingType=HalfHeightRadius
       width=0
       color {colorUseRule=UseColor_Specified, specifiedColor=<rgba 0 0 0 0>}
       cornerRadii=<0 0 0 0>
      }
      tintColor {colorUseRule=UseColor_Specified, specifiedColor=<rgba 1 1 1 1>}
      uVRegion {bIsValid=false, max=<0 0>, min=<0 0>}
     }
     underlineBrush {
      bIsDynamicallyLoaded=false
      drawAs=Image
      imageType=NoImage
      mirroring=NoMirror
      resourceName=None
      resourceObject=null
      tiling=NoTile
      imageSize=<32 32>
      margin=<ltrb 0 0 0 0>
      outlineSettings {
       bUseBrushTransparency=false
       roundingType=HalfHeightRadius
       width=0
       color {colorUseRule=UseColor_Specified, specifiedColor=<rgba 0 0 0 0>}
       cornerRadii=<0 0 0 0>
      }
      tintColor {colorUseRule=UseColor_Specified, specifiedColor=<rgba 1 1 1 1>}
      uVRegion {bIsValid=false, max=<0 0>, min=<0 0>}
     }
    }
   }
   slot {class=VerticalBoxSlot, properties {verticalAlignment=VAlign_Bottom}}
  }
  Text_Description {
   widgetClassPath=&CommonTextBlock
   parent=VerticalBox_1
   widget=Text_Description
   properties {
    lineHeightPercentage=1.05
    mobileFontSizeMultiplier=1.5
    scrollStyle=&TextScrollStyle-Base_C
    text="[PH] Unlocks ability to purchase"
    colorAndOpacity {colorUseRule=UseColor_Specified, specifiedColor=<rgba 0.5776 0.5776 0.5776 1>}
    font {
     bForceMonospaced=false
     bMaterialIsStencil=false
     fontMaterial=null
     fontObject=&HeadingNow
     letterSpacing=0
     monospacedWidth=1
     size=11.25
     skewAmount=0
     typefaceFontName=74 Regular
     outlineSettings {
      bApplyOutlineToDropShadows=false
      bMiteredCorners=false
      bSeparateFillAlpha=false
      outlineMaterial=null
      outlineSize=0
      outlineColor=<rgba 0 0 0 1>
     }
    }
    shadowOffset=<0 0>
   }
   slot {class=VerticalBoxSlot, properties {verticalAlignment=VAlign_Bottom, padding=<ltrb 0 4 0 0>}}
  }
  Spacer_MinHeight {widgetClassPath=&Spacer, parent=Overlay_0, widget=Spacer_MinHeight, properties {size=<1 90>}, slot {class=OverlaySlot}}
 ]
}
