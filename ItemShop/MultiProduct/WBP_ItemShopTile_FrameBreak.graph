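# Item shop offer tile widget, derived from WBP_ItemShopTile_Base, with a separated swiping image, offer name, artist, price, tags, XP info and a 3D preview.
# Registers child components with the base, then sizes itself from the offer's tile size.
# View-model callbacks update artist, colors and XP visibility.
# Hover plays a smart-card slide, scrolls names and records analytics; touch forces details shown, and row state drives info opacity and click behavior.
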WidgetBlueprint /ItemShop/MultiProduct/WBP_ItemShopTile_FrameBreak
animations [Hover {
  binding_count=1
  length_seconds=0.1
  object=&self:Hover
  sequence {bindings [Component_SmartCardList {
     id=#1
     possessed_class=WBP_ItemShopTile_SmartCardWrapper_C
     tracks [{
       type=MovieScene2DTransformTrack
       sections [{
         type=MovieScene2DTransformSection
         end_frame=2
         start_frame=0
         channels [
          Translation.X {type=MovieSceneScriptingFloatChannel, const [0,2]}
          Translation.Y {type=MovieSceneScriptingFloatChannel, keys [[0,0],[2,-8]]}
         ]
        }]
      }]
    }]}
 }]
graphs [
 EventGraph {nodes [
   n1 = HandleInputMethodChanged!() -> (OutputDelegate, then, bNewInputType:ECommonInputType Enum) {is=title K2Node_CustomEvent}
   > UpdateVisualStateForInput!(execute, self, CurrentInputType:ECommonInputType Enum=@n1.bNewInputType) -> (then)
   n2 = Event PreConstruct()
   > SetupComponents()
   > Parent: PreConstruct(IsDesignTime=@n2.IsDesignTime)
   > SetCompactStyle()
   n3 = Event SetTileDimensions!() -> (OutputDelegate, then, Dimensions:Vector 2D Structure) {is=title K2Node_Event}
   n4 = BreakVector2D(InVec=@n3.Dimensions)
   > SetWidthOverride(self=Get SizeBox_All() -> (SizeBox_All:Size Box Object Reference), InWidthOverride=@n4)
   > SetHeightOverride(self=Get SizeBox_All(), InHeightOverride=@n4.Y)
   > UpdateGridSizeOverride!(
    execute
    self:WBP Item Shop Tile Smart Card Wrapper Object Reference=Get Component_SmartCardList() -> (Component_SmartCardList:WBP Item Shop Tile Smart Card Wrapper Object Reference)
    UpdatedWidth:double=float - float(A=@n4, B=24)
    RefreshList:bool=true
   ) -> (then)
   > Parent: SetTileDimensions!(execute, Dimensions:Vector 2D Structure=@n3.Dimensions) -> (then) {is=title K2Node_CallParentFunction}
   Event On Hovered()
   > Parent: On Hovered() {is=title K2Node_CallParentFunction}
   > n5 = Get!(execute:Exec) -> (then:Exec, else:Exec, FortItemShopOfferVM:Fort Item Shop Offer VM Object Reference)
   > Call On Offer Hovered!(
    execute:Exec
    self:WBP Item Shop Tile Base Object Reference
    OfferId:String=Get OfferId!(self:Fort Item Shop Offer VM Object Reference=@n5) -> (OfferId:String)
   ) -> (then:Exec) {is=title K2Node_CallDelegate}
   Event On Clicked()
   > n6 = Switch on E_ItemShop_RowState!(execute, Selection:E_ItemShop_RowState Enum=Get RowState() -> (RowState:E_ItemShop_RowState Enum=NewEnumerator4)) -> (NewEnumerator4:Exec, NewEnumerator1:Exec, NewEnumerator0:Exec, NewEnumerator2:Exec, NewEnumerator3:Exec) {is=title K2Node_SwitchEnum}
   Set FeaturedOfferId!(
    execute=@n6.NewEnumerator0
    FeaturedOfferId:String=Get OfferId(self=Get FortItemShopOfferVM() -> (FortItemShopOfferVM:Fort Item Shop Offer VM Object Reference))
    self:Fort Item Shop VM Object Reference=Get FortItemShopVM() -> (FortItemShopVM:Fort Item Shop VM Object Reference)
   ) -> (then, Output_Get:String)
   > n7 = Cast To FortPlayerController(Object=GetOwningPlayer())
   n8 = GetDisplayAssetName!(self:Fort Item Shop Offer VM Object Reference=Get FortItemShopOfferVM()) -> (ReturnValue:String)
   > Branch(Condition=NOT Boolean(A=IsEmpty(InString=@n8)))
   > "Fire Event Element Interacted (Non-Localized)"!(execute, FortPC:Fort Player Controller Object Reference=@n7, ElementName:String=@n8, AdditionalInfo:Flow Analytics Additional Info Structure) -> (then)
   Call On Request Scroll to Row!(execute:Exec=[@n6.NewEnumerator1, @n6.NewEnumerator2, @n6.NewEnumerator3], self:WBP Item Shop Tile Frame Break Object Reference) -> (then:Exec) {is=title K2Node_CallDelegate}
   n9 = Event SetOfferVM!() -> (OutputDelegate, then, OfferVM:Fort Item Shop Offer VM Object Reference) {is=title K2Node_Event}
   > SetFortItemShopOfferVM!(execute, self, Viewmodel:Fort Item Shop Offer VM Object Reference=@n9.OfferVM) -> (then)
   > Parent: SetOfferVM!(execute, OfferVM:Fort Item Shop Offer VM Object Reference=@n9.OfferVM) -> (then) {is=title K2Node_CallParentFunction}
   Event Construct()
   > Bind Event to On Input Method Changed(self=Get CommonInputSubsystem(), Delegate=@n1)
   n10 = Event SetHoveredState!() -> (OutputDelegate, then, IsHovered:bool, IsImmediate:bool) {is=title K2Node_Event}
   > Parent: SetHoveredState!(execute, IsHovered:bool=@n10.IsHovered, IsImmediate:bool=@n10.IsImmediate) -> (then) {is=title K2Node_CallParentFunction}
   > n11 = Branch(Condition=IsUsingTouch(PlayerController=GetOwningPlayer()))
   SetOfferDetailsShown!(execute=@n11.else, self, ShouldShow:bool=@n10.IsHovered, IsImmediate:bool=@n10.IsImmediate) -> (then)
   > UpdateHoverRecordAnalytics!(execute, self, IsActive:bool=@n10.IsHovered) -> (then)
   > n12 = Branch(Condition=Get AutoWrapText!(self:Text Layout Widget Object Reference=Get Text_Name() -> (Text_Name:Common Text Object Reference)) -> (AutoWrapText:bool))
   SetScrollingEnabled!(execute=@n12.else, self:Common Text Object Reference=Get Text_Name(), bInIsScrollingEnabled:bool=@n10.IsHovered) -> (then)
   Event Destruct()
   > Unbind Event from On Input Method Changed(self=Get CommonInputSubsystem(), Delegate=@n1)
   n13 = Event SetRowState!() -> (OutputDelegate, then, InRowState:E_ItemShop_RowState Enum=NewEnumerator4) {is=title K2Node_Event}
   > n14 = Set RowState(RowState:E_ItemShop_RowState Enum=@n13.InRowState) -> (Output_Get:E_ItemShop_RowState Enum=NewEnumerator4)
   > Parent: SetRowState!(execute, InRowState:E_ItemShop_RowState Enum=@n14) -> (then) {is=title K2Node_CallParentFunction}
   > SetRowState_Implementation()
   Event InitializeTile() {is=title K2Node_Event}
   > InitializeTile_Implementation()
   > Parent: InitializeTile() {is=title K2Node_CallParentFunction}
  ]}
 GetLockIconVisibility {nodes [
   n1 = GetLockIconVisibility!() -> (then, CurrentProgress:int, TotalProgress:int) {is=title K2Node_FunctionEntry}
   n2 = "integer < integer"(A=@n1, B=@n1.TotalProgress)
   n3 = "integer > integer"(A=Get TotalProgress() -> (TotalProgress:int=0))
   > Return Node!(execute, Visibility:ESlateVisibility Enum=Select(Option 0=Collapsed, Option 1=SelfHitTestInvisible, Index=AND Boolean(A=@n2, B=@n3)))
  ]}
 GetOfferVM {nodes [
   GetOfferVM() {is=title K2Node_FunctionEntry}
   > Return Node!(execute, OfferVM:Fort Item Shop Offer VM Object Reference=Get FortItemShopOfferVM())
  ]}
 GetTileAnchor {nodes [
   GetTileAnchor() {is=title K2Node_FunctionEntry}
   n1 = GetCachedGeometry!(self:Widget Object Reference) -> (ReturnValue:"Geometry Structure (by ref)")
   n2 = GetAbsoluteSize!(Geometry:"Geometry Structure (by ref)"=@n1) -> (ReturnValue_X:double, ReturnValue_Y:double)
   n3 = Get!() -> (1x1TileSizeRatio_X:double, 1x1TileSizeRatio_Y:double)
   > Return Node!(
    execute
    Anchor:Vector 2D Structure=vector2d + vector2d(
     A:Vector 2D Structure=GetAbsolutePosition!(Geometry:"Geometry Structure (by ref)"=@n1) -> (ReturnValue:Vector 2D Structure)
     B:Vector 2D Structure=vector2d / vector2d!(A_X:double=@n3, A_Y:double=@n2.ReturnValue_Y, B:Vector 2D Structure=(X=2,Y=2)) -> (ReturnValue:Vector 2D Structure) {is=title K2Node_PromotableOperator}
    ) -> (ReturnValue:Vector 2D Structure) {is=title K2Node_PromotableOperator}
   )
  ]}
 GetTileWidget {nodes [
   GetTileWidget() {is=title K2Node_FunctionEntry}
   > Return Node!(execute, NewParam:User Widget Object Reference=Self-Reference())
  ]}
 InitializeTile_Implementation {nodes [
   InitializeTile_Implementation() {is=title K2Node_FunctionEntry}
   > SetOfferVM!(execute, self:WBP Item Shop Tile Base Object Reference, OfferVM:Fort Item Shop Offer VM Object Reference) -> (then)
   > SetAllNavigationRules!(execute, self:Widget Object Reference, Rule:EUINavigationRule Enum=Escape, WidgetToFocus:Name=None) -> (then)
   n1 = Get SizeBox_All()
   > UpdateGridSizeOverride(
    self=Get Component_SmartCardList()
    UpdatedWidth=float - float(A=Get WidthOverride!(self:Size Box Object Reference=@n1) -> (WidthOverride:float), B:int=24)
    RefreshList=false
   )
  ]}
 RegisterOutline {nodes [
   RegisterOutline!() -> (then, IsFrameBreaking:bool=~) {is=title K2Node_FunctionEntry}
   > UnregisterModularBlock!(
    execute
    self:UIKit Modular Button Object Reference
    Block:UIKit Block Object Reference=Get ModularBlock_Outline() -> (ModularBlock_Outline:WBP UIKit Block Outline Object Reference)
   ) -> (then)
   > UnregisterModularBlock(Block=Get ModularBlock_Outline!(self:WBP Item Shop Tile Swiping Offer Image Separate Object Reference=Get Component_SwipingOfferImage_Separated() -> (Component_SwipingOfferImage_Separated:WBP Item Shop Tile Swiping Offer Image Separate Object Reference)) -> (ModularBlock_Outline))
   n1 = Get Component_SwipingOfferImage_Separated()
   > RegisterModularBlock!(
    execute
    self:UIKit Modular Button Object Reference
    Block:UIKit Block Object Reference=Select(
     Option 0:UIKit Block Object Reference=Get ModularBlock_Outline!() -> (ModularBlock_Outline)
     Option 1:UIKit Block Object Reference=Get ModularBlock_Outline!(self:WBP Item Shop Tile Swiping Offer Image Separate Object Reference=@n1) -> (ModularBlock_Outline)
     Index=Get IsFrameBreaking() -> (IsFrameBreaking:bool)
    ) -> (ReturnValue:UIKit Block Object Reference)
   ) -> (then)
  ]}
 SetCompactStyle {nodes [
   SetCompactStyle() {is=title K2Node_FunctionEntry}
   > Branch(Condition=IsUsingCompactUI())
   n1 = SlotAsSizeBoxSlot!(Widget:Widget Object Reference=Get Text_Name()) -> (ReturnValue:Size Box Slot Object Reference)
   n2 = Get!(self:Size Box Slot Object Reference=@n1) -> (Padding_Left:float=~, Padding_Top:float=~, Padding_Right:float=~, Padding_Bottom:float=~)
   > Set!(
    execute
    Padding_Left:float=@n2
    Padding_Top:float=@n2.Padding_Top
    Padding_Right:float=@n2.Padding_Right
    Padding_Bottom:float=6
    self:Size Box Slot Object Reference=@n1
   ) -> (then, Output_Get:Margin Structure)
  ]}
 SetDesignTimePreview {nodes [
   n1 = SetDesignTimePreview!() -> (then, PresentationMaterial:Material Interface Object Reference, TileSize:EItemShopTileSize Enum=Size_1_x_1, IsHovered:bool=~) {is=title K2Node_FunctionEntry}
   > DesignTimePreviewMaterial!(
    execute
    self:WBP Item Shop Tile Swiping Offer Image Separate Object Reference=Get Component_SwipingOfferImage_Separated()
    MaterialToPreview:"Material Interface Object Reference (by ref)"=@n1
   ) -> (then)
   > Set DesignTimePreviewTileSize(DesignTimePreviewTileSize:EItemShopTileSize Enum=Get TileSize() -> (TileSize:EItemShopTileSize Enum=Size_1_x_1)) -> (Output_Get:EItemShopTileSize Enum=Size_1_x_1)
   > n2 = Set DesignTimePreviewHoveredState(DesignTimePreviewHoveredState:bool=Get IsHovered() -> (IsHovered:bool)) -> (Output_Get:bool)
   > SetHoveredState!(execute, self:WBP Item Shop Tile Base Object Reference, IsHovered:bool=@n2, IsImmediate:bool=true) -> (then)
   > Set Visibility(Visibility=Collapsed, self=Get ModularBlock_Outline!() -> (ModularBlock_Outline))
  ]}
 SetOfferDetailsShown {nodes [
   SetOfferDetailsShown!() -> (then, ShouldShow:bool=~, IsImmediate:bool=~) {is=title K2Node_FunctionEntry}
   n1 = GetOwningPlayer()
   n2 = Get IsOfferDetailsShown() -> (IsOfferDetailsShown:bool)
   n3 = Get ShouldShow() -> (ShouldShow:bool)
   > Branch(Condition=OR Boolean!(
     A="Not Equal (Boolean)"(A=@n3, B=@n2) -> (ReturnValue=~)
     B=IsUsingTouch(PlayerController=@n1)
     C:bool=Get IsImmediate() -> (IsImmediate:bool)
    ) -> (ReturnValue))
   > n4 = Branch(Condition=Get ShouldShow())
   n5 = Get IsImmediate()
   > n6 = Branch(Condition=@n5)
   Set IsOfferDetailsShown(execute=[@n7.then, @n6.then], IsOfferDetailsShown:bool=true) -> (Output_Get:bool)
   n7 = PlayAnimationForward(execute=@n6.else, InAnimation=Get Hover() -> (Hover:Widget Animation Object Reference))
   n8 = Branch(execute=@n4.else, Condition=@n5)
   Set IsOfferDetailsShown(execute=[@n9.then, @n8.then], IsOfferDetailsShown=false)
   n9 = PlayAnimationReverse(execute=@n8.else, InAnimation=Get Hover(), PlaybackSpeed=2)
  ]}
 SetRowState_Implementation {nodes [
   SetRowState_Implementation() {is=title K2Node_FunctionEntry}
   > SetOfferDetailsShown(ShouldShow=IsUsingTouch(PlayerController=GetOwningPlayer()), IsImmediate=true) {is=title K2Node_CallFunction}
   > SetRenderOpacity(
    self=[
     Get VerticalBox_OfferInfo() -> (VerticalBox_OfferInfo:Vertical Box Object Reference)
     Get Border_OfferInfo() -> (Border_OfferInfo:Border Object Reference)
    ]
    InOpacity=Select!(
     NewEnumerator4:float=0
     NewEnumerator1:float=0.3
     NewEnumerator0:float=1
     NewEnumerator2:float=1
     NewEnumerator3:float=0.3
     Index:E_ItemShop_RowState Enum=Get RowState() -> (RowState=NewEnumerator4)
    ) -> (ReturnValue:float)
   )
  ]}
 SetTileSize {nodes [
   n1 = SetTileSize!() -> (then, TileSize:EItemShopTileSize Enum=Size_1_x_1) {is=title K2Node_FunctionEntry}
   > n2 = GetTileDimensionsFromTileSize!(execute, Context:Object Reference, TileSize:EItemShopTileSize Enum=@n1, ItemShopVM:Fort Item Shop VM Object Reference=Get FortItemShopVM()) -> (then, ReturnValue:Vector 2D Structure)
   > SetTileDimensions!(execute, self:WBP Item Shop Tile Base Object Reference, Dimensions:Vector 2D Structure=@n2) -> (then)
  ]}
 SetupComponents {nodes [
   SetupComponents() {is=title K2Node_FunctionEntry}
   > RegisterComponents!(
    execute
    self:WBP Item Shop Tile Base Object Reference
    Components:Array of BPI Item Shop Tile Interfaces=Make Array!(
     "[0]":BPI Item Shop Tile Interface=Get Component_SwipingOfferImage_Separated()
     "[1]":BPI Item Shop Tile Interface=Get Component_Price() -> (Component_Price:WBP Item Shop Tile Price Object Reference)
     "[2]":BPI Item Shop Tile Interface=Get Component_3DPreview() -> (Component_3DPreview:WBP Item Shop Tile 3DPreview Object Reference)
     "[3]":BPI Item Shop Tile Interface=Get Component_Highlight() -> (Component_Highlight:WBP Item Shop Tile Highlight 2 Object Reference)
     "[4]":BPI Item Shop Tile Interface=Get Component_BottomTags() -> (Component_BottomTags:WBP Item Shop Tile Bottom Tags Object Reference)
     "[5]":BPI Item Shop Tile Interface=Get Component_SmartCardList()
     "[6]":BPI Item Shop Tile Interface=Get Component_XpInfo() -> (Component_XpInfo:WBP Item Shop Tile Xp Info Object Reference)
    ) -> (Array:Array of BPI Item Shop Tile Interfaces)
   ) -> (then)
   > SetBlocksTiming!(
    execute
    self:UIKit Modular Button Object Reference
    Timing_Hovering:float=0
    Timing_Unhovering:float=0
    Timing_Focusing:float=0.1
    Timing_Unfocusing:float=0
    Timing_Pressing:float=0.1
    Timing_Releasing:float=0
    Timing_Disabling:float=0
    Timing_Enabling:float=0
    Timing_Selecting:float=0
    Timing_DeselectingIdle:float=0
    Timing_DeselectingFocused:float=0
   ) -> (then)
  ]}
 UpdateHoverRecordAnalytics {nodes [
   n1 = UpdateHoverRecordAnalytics!() -> (then, IsActive:bool=~) {is=title K2Node_FunctionEntry}
   > Branch(Condition=IsItemPreviewEnabled!(self:Fort Item Shop VM Object Reference=Get FortItemShopVM()) -> (ReturnValue:bool))
   > n2 = Branch(Condition=@n1)
   > StartOrResumeTileHoverRecord!(
    execute
    self:Fort Item Shop Selection VM Object Reference=Get FortItemShopSelectionVM() -> (FortItemShopSelectionVM:Fort Item Shop Selection VM Object Reference)
    Offer:Fort Item Shop Offer VM Object Reference=Get FortItemShopOfferVM()
   ) -> (then)
   PauseAndComputeTileHoverRecord(execute=@n2.else, self:Fort Item Shop Selection VM Object Reference=Get FortItemShopSelectionVM())
  ]}
 UpdateVisualStateForInput {nodes [
   UpdateVisualStateForInput!() -> (then, CurrentInputType:ECommonInputType Enum) {is=title K2Node_FunctionEntry}
   > n1 = Get!(execute:Exec) -> (then:Exec, else:Exec, FortItemShopOfferVM:Fort Item Shop Offer VM Object Reference)
   > SetTileSize!(execute, self, TileSize:EItemShopTileSize Enum=Get TileSize!(self:Fort Item Shop Offer VM Object Reference=@n1) -> (TileSize=Size_1_x_1)) -> (then) {is=title K2Node_CallFunction}
   > n2 = Branch(Condition="Equal (Enum)"(A=Get CurrentInputType() -> (CurrentInputType:ECommonInputType Enum), B=Touch))
   > SetOfferDetailsShown(ShouldShow=true, IsImmediate=true)
   n3 = Get RowState() -> (RowState=NewEnumerator4)
   SetOfferDetailsShown(
    execute=@n2.else
    ShouldShow=AND Boolean(A="Equal (Enum)"(A:E_ItemShop_RowState Enum=@n3, B:E_ItemShop_RowState Enum=NewEnumerator0), B=IsHovered())
    IsImmediate=true
   )
  ]}
 VM_GrantedItemsUpdated {nodes [
   n1 = VM_GrantedItemsUpdated!() -> (then, GrantedItems:Array of Fort Item VM Object References) {is=title K2Node_FunctionEntry}
   > n2 = Set Text(
    Text=GetArtistNameFromGrantedItems!(GrantedItems:Array of Fort Item VM Object References=@n1) -> (ArtistName:Text)
    self=Get Text_Artist() -> (Text_Artist:Common Text Object Reference)
   )
   > Set Visibility(
    Visibility=Select!(Option 0:ESlateVisibility Enum=SelfHitTestInvisible, Option 1:ESlateVisibility Enum=Collapsed, Index:bool=TextIsEmpty(InText=@n2)) -> (ReturnValue:ESlateVisibility Enum)
    self=Get Text_Artist()
   )
  ]}
 VM_IsFilterActiveUpdated {nodes [
   n1 = VM_IsFilterActiveUpdated!() -> (then, IsFilterActive:bool=~) {is=title K2Node_FunctionEntry}
   > Set VerticalAlignment!(
    execute
    VerticalAlignment:EVerticalAlignment Enum=Select(Option 0:EVerticalAlignment Enum=VAlign_Top, Option 1:EVerticalAlignment Enum=VAlign_Bottom, Index=@n1) -> (ReturnValue:EVerticalAlignment Enum)
    self:Vertical Box Slot Object Reference=SlotAsVerticalBoxSlot(Widget=Get Component_SmartCardList())
   ) -> (then, Output_Get:EVerticalAlignment Enum=VAlign_Fill)
  ]}
 VM_MetaDisplayDataUpdated {nodes [
   n1 = VM_MetaDisplayDataUpdated!() -> (then, FinalColor:"Linear Color Structure (by ref)") {is=title K2Node_FunctionEntry}
   > n2 = GetDynamicMaterial(self:Border Object Reference=Get Border_OfferInfo())
   > SetVectorParameterValue(self=@n2, ParameterName=Color, Value=@n1)
   > SetVectorParameterValue(self=@n2, ParameterName=Color2, Value=@n1)
   > SetScalarParameterValue(self=@n2, ParameterName=AlphaEnd, Value=BreakColor(InColor=@n1).A)
  ]}
 VM_TotalProgressUpdated {nodes [
   n1 = VM_TotalProgressUpdated!() -> (then, TotalProgress:int) {is=title K2Node_FunctionEntry}
   n2 = "Equal (Integer)"(A=@n1)
   > Set Visibility(
    Visibility=Select(Option 0:ESlateVisibility Enum=SelfHitTestInvisible, Option 1:ESlateVisibility Enum=Collapsed, Index=@n2) -> (ReturnValue:ESlateVisibility Enum)
    self=Get Component_XpInfo()
   )
   > n3 = GetDynamicMaterial(self=Get Border_OfferInfo())
   > SetVectorParameterValue(
    self=@n3
    ParameterName=bIndividualRoundedCorners
    Value=Select(Option 0:Linear Color Structure=(R=0,G=0,B=0,A=0), Option 1:Linear Color Structure=(R=0,G=1,B=0,A=1), Index=@n2) -> (ReturnValue:Linear Color Structure)
   )
   > Set Visibility(
    Visibility=Select(Option 0:ESlateVisibility Enum=Collapsed, Option 1:ESlateVisibility Enum=SelfHitTestInvisible, Index=@n2) -> (ReturnValue:ESlateVisibility Enum)
    self=Get Component_BottomTags()
   )
  ]}
]
variables [
 DesignTimeMaterialPreview {type=MaterialInterface, default=/OfferCatalog/MI_OfferImages/MI_0_Character_Base}
 DesignTimePreviewHoveredState {type=boolean, default=true}
 DesignTimePreviewTileSize {type=EItemShopTileSize, default=Size_1_x_1}
 IsOfferDetailsShown {type=boolean}
 RowState {type=E_ItemShop_RowState, default=None}
]
widgets {
 info {
  inheritedWidgetCount=1
  parentClass=/ItemShop/MultiProduct/ItemShopTile/WBP_ItemShopTile_Base.WBP_ItemShopTile_Base_C
  rootWidgetClass=&SizeBox
  widgetCount=19
 }
 widgets [
  SizeBox_All {
   widgetClassPath=&SizeBox
   bIsVariable=true
   widget=SizeBox_All
   properties {bOverride_HeightOverride=true, bOverride_WidthOverride=true, heightOverride=408, widthOverride=256}
  }
  Overlay_All {widgetClassPath=&Overlay, bIsVariable=true, parent=SizeBox_All, widget=Overlay_All, slot {class=SizeBoxSlot}}
  Component_SwipingOfferImage_Separated {
   widgetClassPath=/ItemShop/MultiProduct/ItemShopTileComponents/WBP_ItemShopTile_SwipingOfferImage_Separate
   bIsVariable=true
   parent=Overlay_All
   widget=Component_SwipingOfferImage_Separated
   slot {class=OverlaySlot, properties {horizontalAlignment=HAlign_Fill, verticalAlignment=VAlign_Fill}}
  }
  VerticalBox_0 {
   widgetClassPath=&VerticalBox
   parent=Overlay_All
   widget=VerticalBox_0
   slot {class=OverlaySlot, properties {horizontalAlignment=HAlign_Fill, verticalAlignment=VAlign_Fill}}
  }
  Component_SmartCardList {
   widgetClassPath=/ItemShop/MultiProduct/ItemShopTileComponents/WBP_ItemShopTile_SmartCardWrapper
   bIsVariable=true
   parent=VerticalBox_0
   widget=Component_SmartCardList
   slot {class=VerticalBoxSlot, properties {padding=<ltrb 12 12 12 -52>, size {sizeRule=Fill, value=1}}}
  }
  Border_OfferInfo {
   widgetClassPath=&Border
   bIsVariable=true
   parent=VerticalBox_0
   widget=Border_OfferInfo
   properties {
    verticalAlignment=VAlign_Bottom
    background {
     bIsDynamicallyLoaded=false
     drawAs=Image
     imageType=FullColor
     mirroring=NoMirror
     resourceName=None
     resourceObject=/FortUILibrary/Systems/ItemShopTile/Materials/MI_UI_ItemShopTile_OfferInfoBacking
     tiling=NoTile
     imageSize=<32 32>
     margin=<ltrb 0 0 0 0>
     outlineSettings {
      bUseBrushTransparency=false
      roundingType=HalfHeightRadius
      width=0
      color {colorUseRule=UseColor_Specified, specifiedColor=<rgba 0 0 0 0>}
      cornerRadii=<0 0 0 0>
     }
     tintColor {colorUseRule=UseColor_Specified, specifiedColor=<rgba 1 1 1 1>}
     uVRegion {bIsValid=false, max=<0 0>, min=<0 0>}
    }
    padding=<ltrb 0 60 0 0>
   }
   slot {class=VerticalBoxSlot}
  }
  VerticalBox_OfferInfo {
   widgetClassPath=&VerticalBox
   bIsVariable=true
   parent=Border_OfferInfo
   widget=VerticalBox_OfferInfo
   slot {class=BorderSlot, properties {verticalAlignment=VAlign_Bottom, padding=<ltrb 0 60 0 0>}}
  }
  Component_Highlight {
   widgetClassPath=/ItemShop/MultiProduct/ItemShopTileComponents/WBP_ItemShopTile_Highlight_2
   bIsVariable=true
   parent=VerticalBox_OfferInfo
   widget=Component_Highlight
   properties {bOverride_Cursor=true, highlightSize=Compact, highlightText="[PH] Highlight"}
   slot {class=VerticalBoxSlot, properties {horizontalAlignment=HAlign_Left, verticalAlignment=VAlign_Bottom, padding=<ltrb 11 0 12 6>}}
  }
  Text_Artist {
   widgetClassPath=&CommonTextBlock
   bIsVariable=true
   parent=VerticalBox_OfferInfo
   widget=Text_Artist
   properties {
    mobileFontSizeMultiplier=1.2
    scrollStyle=&TextScrollStyle-Base_C
    text=Artist Name
    colorAndOpacity {colorUseRule=UseColor_Specified, specifiedColor=<rgba 1 1 1 0.7>}
    font {
     bForceMonospaced=false
     bMaterialIsStencil=false
     fontMaterial=null
     fontObject=&HeadingNow
     letterSpacing=0
     monospacedWidth=1
     size=11.25
     skewAmount=0
     typefaceFontName=74 Regular
     outlineSettings {
      bApplyOutlineToDropShadows=false
      bMiteredCorners=false
      bSeparateFillAlpha=false
      outlineMaterial=null
      outlineSize=0
      outlineColor=<rgba 0 0 0 1>
     }
    }
    shadowOffset=<0 0>
   }
   slot {class=VerticalBoxSlot, properties {padding=<ltrb 12 0 12 4>}}
  }
  SizeBox_NameText {
   widgetClassPath=&SizeBox
   bIsVariable=true
   parent=VerticalBox_OfferInfo
   widget=SizeBox_NameText
   properties {bOverride_MaxDesiredHeight=true, maxDesiredHeight=122}
   slot {class=VerticalBoxSlot, properties {padding=<ltrb 12 0 12 2>}}
  }
  Text_Name {
   widgetClassPath=&CommonTextBlock
   bIsVariable=true
   parent=SizeBox_NameText
   widget=Text_Name
   properties {
    bIsScrollingEnabled=false
    clipping=OnDemand
    mobileFontSizeMultiplier=1.2
    scrollStyle=/FortUILibrary/Styles/Typography/Foundation/TextScrollStyle_ShortDelay.TextScrollStyle_ShortDelay_C
    text="[PH] Offer Name Very Long"
    textOverflowPolicy=MultilineEllipsis
    font {
     bForceMonospaced=false
     bMaterialIsStencil=false
     fontMaterial=/FortUILibrary/UIKit/Buttons/MaterialInstances/ItemShopTile/MI_UIKit_ItemShopTile_TitleText
     fontObject=&HeadingNow
     letterSpacing=0
     monospacedWidth=1
     size=15.75
     skewAmount=0
     typefaceFontName=75 Medium
     outlineSettings {
      bApplyOutlineToDropShadows=false
      bMiteredCorners=false
      bSeparateFillAlpha=false
      outlineMaterial=null
      outlineSize=0
      outlineColor=<rgba 0 0 0 1>
     }
    }
    shadowOffset=<0 0>
   }
   slot {class=SizeBoxSlot, properties {padding=<ltrb 0 0 0 4>}}
  }
  Overlay_PriceStateSwitcher {
   widgetClassPath=&Overlay
   bIsVariable=true
   parent=VerticalBox_OfferInfo
   widget=Overlay_PriceStateSwitcher
   slot {class=VerticalBoxSlot, properties {verticalAlignment=VAlign_Bottom, padding=<ltrb 12 0 12 8>}}
  }
  Component_Price {
   widgetClassPath=/ItemShop/MultiProduct/ItemShopTileComponents/WBP_ItemShopTile_Price
   bIsVariable=true
   parent=Overlay_PriceStateSwitcher
   widget=Component_Price
   slot {class=OverlaySlot}
  }
  Image_Lock {
   widgetClassPath=&Image
   bIsVariable=true
   parent=Overlay_PriceStateSwitcher
   widget=Image_Lock
   properties {
    visibility=Collapsed
    brush {
     bIsDynamicallyLoaded=false
     drawAs=Image
     imageType=FullColor
     mirroring=NoMirror
     resourceName=None
     resourceObject=/FortUILibrary/Textures/Icons/ToBeDeprecated/T_UI_Icon_Global_Locked
     tiling=NoTile
     imageSize=<20 20>
     margin=<ltrb 0 0 0 0>
     outlineSettings {
      bUseBrushTransparency=false
      roundingType=HalfHeightRadius
      width=0
      color {colorUseRule=UseColor_Specified, specifiedColor=<rgba 0 0 0 0>}
      cornerRadii=<0 0 0 0>
     }
     tintColor {colorUseRule=UseColor_Specified, specifiedColor=<rgba 1 1 1 1>}
     uVRegion {bIsValid=false, max=<0 0>, min=<0 0>}
    }
   }
   slot {class=OverlaySlot, properties {horizontalAlignment=HAlign_Right, verticalAlignment=VAlign_Bottom}}
  }
  Component_BottomTags {
   widgetClassPath=/ItemShop/MultiProduct/ItemShopTileComponents/WBP_ItemShopTile_BottomTags
   bIsVariable=true
   parent=VerticalBox_OfferInfo
   widget=Component_BottomTags
   slot {class=VerticalBoxSlot}
  }
  Component_XpInfo {
   widgetClassPath=/ItemShop/MultiProduct/ItemShopTileComponents/WBP_ItemShopTile_XpInfo
   bIsVariable=true
   parent=VerticalBox_0
   widget=Component_XpInfo
   properties {visibility=Collapsed}
   slot {class=VerticalBoxSlot, properties {padding=<ltrb 0 -1 0 -1>}}
  }
  ModularBlock_Outline {
   widgetClassPath=/FortUILibrary/UIKit/Buttons/Blocks/WBP_UIKit_Block_Outline
   bIsVariable=true
   parent=Overlay_All
   widget=ModularBlock_Outline
   properties {bottomLeftCornerRadius=18, bottomRightCornerRadius=18, radiusInPixels=true, topLeftCornerRadius=18, topRightCornerRadius=18}
   slot {class=OverlaySlot, properties {horizontalAlignment=HAlign_Fill, verticalAlignment=VAlign_Fill}}
  }
  Component_3DPreview {
   widgetClassPath=/ItemShop/MultiProduct/ItemShopTileComponents/WBP_ItemShopTile_3DPreview
   bIsVariable=true
   parent=Overlay_All
   widget=Component_3DPreview
   slot {class=OverlaySlot, properties {horizontalAlignment=HAlign_Fill, verticalAlignment=VAlign_Fill}}
  }
  InputActionWidget {widgetClassPath=&CommonActionWidget, bInherited=true, widget=null}
 ]
}
