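# WBP_MPItemShop_Row is an Item Shop row widget that shows one offer group's tiles in a horizontal scroll box.
# CreateTileGrid fills a dynamic entry box and maps column indices to tiles for arrow scrolling, capping columns unless the group is a scrollable list.
# Row state and parent focus drive hitbox, peek animation, navigation, previews and delayed visited marking.
# Closest-tile search favors same-column targets, and delegates report focus, hover and list-view requests.
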WidgetBlueprint /ItemShop/MultiProduct/WBP_MPItemShop_Row
animations [OnInactiveRowHovered {
  binding_count=1
  length_seconds=0.2
  object=&self:OnInactiveRowHovered
  sequence {bindings [Button_ViewMore {
     id=#1
     possessed_class=WBP_UIKit_Button_Regular_C
     tracks [{
       type=MovieSceneFloatTrack
       display_name=Render Opacity
       sections [{type=MovieSceneFloatSection, end_frame=4, start_frame=0, channels [None {type=MovieSceneScriptingFloatChannel, keys [[0,0.3],[4,0.5]]}]}]
      }]
    }]}
 }]
graphs [
 CanNavigateToSideNav {nodes [
   CanNavigateToSideNav() {is=title K2Node_FunctionEntry}
   n1 = GetScrollOffset!(self:Scroll Box Object Reference=Get Scrollbox_RowContent() -> (Scrollbox_RowContent:Athena Scroll Box Object Reference)) -> (ReturnValue:float)
   n2 = GetDisplayType!(self:Fort Item Shop Offer Group VM Object Reference=Get FortItemShopOfferGroupVM() -> (FortItemShopOfferGroupVM:Fort Item Shop Offer Group VM Object Reference)) -> (ReturnValue:EFortItemShopOfferDisplayType Enum=TileGrid)
   > Return Node!(
    execute
    ReturnValue:bool=OR Boolean!(
     A="Nearly Equal (Float)"(A=@n1, B=0, ErrorTolerance=1)
     B=IsInputType!(InputType:ECommonInputType Enum=Gamepad, IsNot:bool=true) -> (Result:bool)
     C:bool="Not Equal (Enum)"!(A:EFortItemShopOfferDisplayType Enum=@n2, B:EFortItemShopOfferDisplayType Enum=ScrollableList) -> (ReturnValue:bool)
    ) -> (ReturnValue)
   )
  ]}
 CancelRequestMarkOfferGroupAsVisited {nodes [
   CancelRequestMarkOfferGroupAsVisited() {is=title K2Node_FunctionEntry}
   > Clear and Invalidate Timer by Handle(Handle=Get MarkOfferGroupAsVisitedHandler() -> (MarkOfferGroupAsVisitedHandler:Timer Handle Structure))
  ]}
 CreateTileGrid {nodes [
   CreateTileGrid!() -> (then, Offers:Array of Fort Item Shop Offer VM Object References) {is=title K2Node_FunctionEntry}
   > ResetAllTiles()
   > Reset!(
    execute
    self:Dynamic Entry Box Object Reference=Get DynamicEntryBox_Tiles() -> (DynamicEntryBox_Tiles:Fort Dynamic Entry Box Object Reference)
    bDeleteWidgets:bool=false
   ) -> (then)
   > Clear!(
    execute
    TargetMap:Map of Integers to Widget Object References=Get TileIndexMap() -> (TileIndexMap:Map of Integers to Widget Object References)
   ) -> (then) {is=title K2Node_CallFunction}
   n1 = BreakVector2D(InVec=GetTileSpacing!() -> (TileSpacing:Vector 2D Structure))
   > Set!(execute, EntrySpacing_X:double=@n1, EntrySpacing_Y:double=@n1.Y, self:Dynamic Entry Box Base Object Reference=Get DynamicEntryBox_Tiles()) -> (then, Output_Get:Vector 2D Structure)
   > n2 = For Each Loop with Break!(
    Exec:Exec
    Array:Array of Fort Item Shop Offer VM Object References=Get Offers() -> (Offers:Array of Fort Item Shop Offer VM Object References)
    Break:Exec=@n6.else
   ) -> (LoopBody:Exec, Array Element:Fort Item Shop Offer VM Object Reference, Array Index:int, Completed:Exec)
   n3 = "Equal (Enum)"(
    A:EFortItemShopOfferDisplayType Enum=GetDisplayType(self=Get FortItemShopOfferGroupVM()) -> (ReturnValue=TileGrid)
    B:EFortItemShopOfferDisplayType Enum=ScrollableList
   )
   n4 = Get CappedColumnCount() -> (CappedColumnCount:int=0)
   n5 = Get ColumnIndex() -> (ColumnIndex:int=0)
   > n6 = Branch(Condition=OR Boolean!(A="integer < integer"(A=@n5, B=@n4), B=@n3) -> (ReturnValue))
   > Set LoopOfferVM(LoopOfferVM:Fort Item Shop Offer VM Object Reference=@n2) -> (Output_Get:Fort Item Shop Offer VM Object Reference)
   > n7 = Create Entry of Class!(
    execute
    self:Dynamic Entry Box Object Reference=Get DynamicEntryBox_Tiles()
    EntryClass:User Widget Class Reference=Get TileWidgetClass() -> (TileWidgetClass:User Widget Class Reference)
   ) -> (then, ReturnValue:User Widget Object Reference)
   > n8 = Cast To WBP_ItemShopTile_Base!(execute, Object=@n7) -> (then, CastFailed, AsWBP Item Shop Tile Base:WBP Item Shop Tile Base Object Reference) {is=title K2Node_DynamicCast}
   > n9 = Set LoopTile(LoopTile:WBP Item Shop Tile Base Object Reference=@n8) -> (Output_Get:WBP Item Shop Tile Base Object Reference)
   > SetSectionID!(
    execute
    self:BPI Item Shop Tile Interface=@n9
    InSectionID:String=Get!(self:Fort Item Shop Offer Group VM Object Reference=Get FortItemShopOfferGroupVM()) -> (OwningLayout_SectionId:String, OwningLayout_SubsectionId:String)
   ) -> (then)
   > InitializeTile(self:WBP Item Shop Tile Base Object Reference=@n9)
   > SetOfferVM!(
    execute
    self:WBP Item Shop Tile Base Object Reference=@n9
    OfferVM:Fort Item Shop Offer VM Object Reference=Get LoopOfferVM() -> (LoopOfferVM:Fort Item Shop Offer VM Object Reference)
   ) -> (then)
   > Bind Event to On Offer Hovered(self:WBP Item Shop Tile Base Object Reference=Get LoopTile() -> (LoopTile:WBP Item Shop Tile Base Object Reference), Delegate=Create Event()) {is=title K2Node_AddDelegate}
   > Bind Event to On Tile Focused(self:WBP Item Shop Tile Base Object Reference=Get LoopTile(), Delegate=Create Event()) {is=title K2Node_AddDelegate}
   > Bind Event to On Request Scroll to Row(self:WBP Item Shop Tile Base Object Reference=Get LoopTile(), Delegate=Create Event()) {is=title K2Node_AddDelegate}
   n10 = Get FortItemShopOfferGroupVM()
   > n11 = Branch(Condition="Equal (Enum)"(A=GetDisplayType(self=@n10) -> (ReturnValue=TileGrid), B=ScrollableList))
   n12 = Get LoopOfferVM()
   n13 = Get ColumnIndex() -> (ColumnIndex=0)
   n14 = int + int(
    A=@n13
    B=GetGridSpanFromTileSize!(TileSize:EItemShopTileSize Enum=Get TileSize!(self:Fort Item Shop Offer VM Object Reference=@n12) -> (TileSize:EItemShopTileSize Enum=Size_1_x_1)) -> (GridSpan_X:int, GridSpan_Y:int)
   )
   > n15 = For Loop(FirstIndex=@n13, LastIndex=int - int(A=@n14))
   > Add!(
    execute
    TargetMap:Map of Integers to Widget Object References=Get TileIndexMap()
    Key:"Integer (by ref)"=@n15
    Value:"Widget Object Reference (by ref)"=Get LoopTile()
   ) -> (then) {is=title K2Node_CallFunction}
   Set ColumnIndex(execute=[@n15.Completed, @n11.else], ColumnIndex:int=@n14) -> (Output_Get:int=0)
   n16 = "integer > integer"(A=Get ColumnIndex() -> (ColumnIndex=0), B=Get CappedColumnCount() -> (CappedColumnCount=0))
   Set Visibility(
    execute=@n2.Completed
    Visibility=Select(Option 0=Collapsed, Option 1=SelfHitTestInvisible, Index=AND Boolean(A=@n3, B=@n16))
    self=Get Spacer_EndPadding() -> (Spacer_EndPadding:Spacer Object Reference)
   )
   n17 = Get ColumnIndex() -> (ColumnIndex=0)
   > Set TotalColumnCount(TotalColumnCount:int=@n17) -> (Output_Get:int=0)
   int + int(A=@n17)
  ]}
 EventGraph {nodes [
   "On Hovered (Button_RowHitbox)"() {is=title K2Node_ComponentBoundEvent}
   > n1 = Branch(Condition=IsUsingTouch(PlayerController=GetOwningPlayer()))
   Branch(execute=@n1.else, Condition=Get ParentHasFocus() -> (ParentHasFocus:bool))
   > n2 = Switch on E_ItemShop_RowState!(execute, Selection:E_ItemShop_RowState Enum=Get RowState() -> (RowState:E_ItemShop_RowState Enum=NewEnumerator4)) -> (NewEnumerator4:Exec, NewEnumerator1:Exec, NewEnumerator0:Exec, NewEnumerator2:Exec, NewEnumerator3:Exec) {is=title K2Node_SwitchEnum}
   SetRowPeekState!(execute=[@n2.NewEnumerator1, @n2.NewEnumerator3], self, IsInPeekState:bool=true) -> (then)
   "On Unhovered (Button_RowHitbox)"() {is=title K2Node_ComponentBoundEvent}
   > n3 = Branch(Condition=IsUsingTouch(PlayerController=GetOwningPlayer()))
   Branch(execute=@n3.else, Condition=Get ParentHasFocus())
   > Branch(Condition=Get IsRowPeeked() -> (IsRowPeeked:bool))
   > SetRowPeekState(IsInPeekState=false)
   n4 = "On Clicked (Button_RowHitbox)"() {is=title K2Node_ComponentBoundEvent}
   Call On Row Focused!(execute:Exec=[@n4.then, @n13.then], self:WBP MPItem Shop Row Object Reference, RowWidget:User Widget Object Reference=Self-Reference()) -> (then:Exec) {is=title K2Node_CallDelegate}
   n5 = "On Clicked (Button_ViewMore)"!() -> (OutputDelegate, then, Button:Common Button Base Object Reference) {is=title K2Node_ComponentBoundEvent}
   > Fire Event Element Interacted(Element=@n5.Button)
   > Call On Row List View Requested!(
    execute:Exec
    self:WBP MPItem Shop Row Object Reference
    OfferGroupVM:Fort Item Shop Offer Group VM Object Reference=Get FortItemShopOfferGroupVM()
   ) -> (then:Exec) {is=title K2Node_CallDelegate}
   > Set ShouldFocusListViewButton(ShouldFocusListViewButton:bool=true) -> (Output_Get:bool)
   "On User Scrolled (Scrollbox_RowContent)"!() -> (OutputDelegate, then, CurrentOffset:float) {is=title K2Node_ComponentBoundEvent}
   > UpdateNavigationToSideNav()
   > UpdateScrollboxPresentation()
   "On Clicked (Button_ScrollRight)"!() -> (OutputDelegate, then, Button:Common Button Base Object Reference) {is=title K2Node_ComponentBoundEvent}
   n6 = Get TotalColumnCount() -> (TotalColumnCount:int=0)
   n7 = Get ColumnsToScroll() -> (ColumnsToScroll:int=0)
   n8 = Get CurrentColumnIndex() -> (CurrentColumnIndex:int=0)
   n9 = GetTileAtColumnIndex!(self, DesiredIndex:int="Min (Integer)"(A=int + int(A=@n7, B=@n8), B=int - int(A=@n6)) -> (ReturnValue=0)) -> (Tile:Widget Object Reference, TileIndex:int=0)
   > ScrollWidgetIntoView!(
    execute
    self:Scroll Box Object Reference=Get Scrollbox_RowContent()
    WidgetToFind:Widget Object Reference=@n9
    AnimateScroll:bool=true
    ScrollDestination:EDescendantScrollDestination Enum=TopOrLeft
    Padding:float=Get NavigationScrollPadding!(self:Scroll Box Object Reference=Get Scrollbox_RowContent()) -> (NavigationScrollPadding:float)
   ) -> (then)
   > Set CurrentColumnIndex(CurrentColumnIndex:int=@n9.TileIndex) -> (Output_Get:int=0)
   "On Clicked (Button_ScrollLeft)"!() -> (OutputDelegate, then, Button:Common Button Base Object Reference) {is=title K2Node_ComponentBoundEvent}
   n10 = Get CurrentColumnIndex() -> (CurrentColumnIndex=0)
   n11 = Get ColumnsToScroll() -> (ColumnsToScroll=0)
   n12 = GetTileAtColumnIndex(DesiredIndex="Max (Integer)"(A=int - int(A=@n10, B=@n11), B=0)) -> (TileIndex=0)
   > ScrollWidgetIntoView(
    self=Get Scrollbox_RowContent()
    WidgetToFind=@n12
    AnimateScroll=true
    ScrollDestination=TopOrLeft
    Padding=Get NavigationScrollPadding(self=Get Scrollbox_RowContent())
   )
   > Set CurrentColumnIndex(CurrentColumnIndex=@n12.TileIndex) -> (Output_Get=0)
   n13 = HandleScrollToRowRequested() {is=title K2Node_CustomEvent}
   n14 = OnTileFocused!() -> (OutputDelegate, then, FocusedTile:User Widget Object Reference) {is=title K2Node_CustomEvent}
   > Set FocusedTile(FocusedTile:User Widget Object Reference=@n14.FocusedTile) -> (Output_Get:User Widget Object Reference)
   > Set ShouldFocusListViewButton(ShouldFocusListViewButton=false)
   n15 = OnInputMethodChanged!() -> (OutputDelegate, then, bNewInputType:ECommonInputType Enum) {is=title K2Node_CustomEvent}
   > UpdateNavigationToSideNav()
   n16 = OnTileHovered!() -> (OutputDelegate, then, OfferId:String) {is=title K2Node_CustomEvent}
   > Call On Offer Hovered!(execute:Exec, self:WBP MPItem Shop Row Object Reference, OfferId:String=@n16.OfferId) -> (then:Exec) {is=title K2Node_CallDelegate}
   Event OnRemovedFromFocusPath!() -> (OutputDelegate, then, InFocusEvent:Focus Event Structure) {is=title K2Node_Event}
   > Set FocusedOfferId(FocusedOfferId:String) -> (Output_Get:String)
   n17 = Event I_SetFocusedOfferID!() -> (OutputDelegate, then, OfferID:String) {is=title K2Node_Event}
   > Set FocusedOfferId(FocusedOfferId=@n17.OfferID)
   Event PreConstruct()
   > Set TileWidgetClass(TileWidgetClass:User Widget Class Reference=Select(
     Option 0:User Widget Class Reference=&/WBP_ItemShopTile_Default.WBP_ItemShopTile_Default_C
     Option 1:User Widget Class Reference=&/WBP_ItemShopTile_FrameBreak.WBP_ItemShopTile_FrameBreak_C
     Index=IsUsingTileV2Layout!() -> (ReturnValue:bool)
    ) -> (ReturnValue:User Widget Class Reference)) -> (Output_Get:User Widget Class Reference)
   > SetTilesLayout()
   n18 = Event I_SetOfferGroupVM!() -> (OutputDelegate, then, OfferGroupVM:Fort Item Shop Offer Group VM Object Reference) {is=title K2Node_Event}
   > SetFortItemShopOfferGroupVM!(execute, self, Viewmodel:Fort Item Shop Offer Group VM Object Reference=@n18.OfferGroupVM) -> (then)
   n19 = Event I_InitializeRow!() -> (OutputDelegate, then, FortItemShopRow:Fort Item Shop Row Structure, ContentHasFocus:bool) {is=title K2Node_Event}
   > InitializeRow!(execute, self, FortItemShopRow:Fort Item Shop Row Structure=@n19.FortItemShopRow, ContentHasFocus:bool=@n19.ContentHasFocus) -> (then)
   n20 = Event I_SetParentHasFocusState!() -> (OutputDelegate, then, HasFocus:bool) {is=title K2Node_Event}
   > SetParentHasFocusState!(execute, self, inParentHasFocus:bool=@n20.HasFocus) -> (then)
   Event OnInitialized()
   > SetMobileStyle() {is=title K2Node_CallFunction}
   Event OnAddedToFocusPath!() -> (OutputDelegate, then, InFocusEvent:Focus Event Structure) {is=title K2Node_Event}
   n21 = Get CommonInputSubsystem()
   > Branch(Condition="Equal (Enum)"(A:ECommonInputType Enum=GetCurrentInputType(self=@n21), B:ECommonInputType Enum=Gamepad))
   > Call On Row Focused(RowWidget=Self-Reference())
   Event I_ResetRow() {is=title K2Node_Event}
   > ResetRow()
   Event I_OnScrollComplete!() -> (OutputDelegate, then, DestinationIndex:int=0) {is=title K2Node_Event}
   > Branch(Condition=ShouldLoadPresentationsOnScrollComplete!(self:Fort MPItem Shop Row Object Reference) -> (ReturnValue:bool))
   > LoadAllPresentations()
   Event Construct()
   n22 = Get CommonInputSubsystem()
   > Bind Event to On Input Method Changed(self=@n22, Delegate=@n15)
   n23 = Event I_SetRowState!() -> (OutputDelegate, then, InRowState:E_ItemShop_RowState Enum=NewEnumerator4) {is=title K2Node_Event}
   > SetRowState!(execute, self, InRowState:E_ItemShop_RowState Enum=@n23.InRowState) -> (then)
   Event Destruct()
   > Unbind Event from On Input Method Changed(self=@n22, Delegate=@n15)
  ]}
 Get Closest Focus Target {nodes [
   Get Closest Focus Target!() -> (then, TargetLocation:Vector 2D Structure) {is=title K2Node_FunctionEntry}
   n1 = Get Visibility!(self:Widget Object Reference=Get Button_ViewMore() -> (Button_ViewMore:WBP UIKit Button Regular Object Reference)) -> (Visibility:ESlateVisibility Enum=Visible)
   n2 = GetAbsolutePosition!(Geometry:"Geometry Structure (by ref)"=GetCachedGeometry!(self:Widget Object Reference) -> (ReturnValue:"Geometry Structure (by ref)")) -> (ReturnValue_X:double, ReturnValue_Y:double)
   n3 = Get!() -> (TargetLocation_X:double, TargetLocation_Y:double)
   > n4 = Branch(Condition=AND Boolean(
     A="float < float"(A=@n2.ReturnValue_Y, B=@n3.TargetLocation_Y)
     B="Equal (Enum)"(A:ESlateVisibility Enum=@n1, B:ESlateVisibility Enum=Visible)
    ))
   > Return Node!(execute, Widget:Widget Object Reference=Get Button_ViewMore())
   n5 = For Each Loop(
    Exec=@n4.else
    Array:Array of Widget Object References=GetAllEntries!(self:Dynamic Entry Box Base Object Reference=Get DynamicEntryBox_Tiles()) -> (ReturnValue:Array of User Widget Object References)
   ) -> (Array Element:Widget Object Reference)
   Return Node(execute=@n5.Completed, Widget=Get ClosestTile() -> (ClosestTile:WBP Item Shop Tile Base Object Reference))
   > n6 = Cast To WBP_ItemShopTile_Base(Object=@n5)
   > Set Current Tile(Current Tile:WBP Item Shop Tile Base Object Reference=@n6) -> (Output_Get:WBP Item Shop Tile Base Object Reference)
   > n7 = Sequence()
   n8 = GetTileAnchor!(
    execute=@n7.then_1
    self:WBP Item Shop Tile Base Object Reference=Get Current Tile() -> (Current Tile:WBP Item Shop Tile Base Object Reference)
   ) -> (then, Anchor:Vector 2D Structure)
   n9 = Get TargetLocation() -> (TargetLocation:Vector 2D Structure)
   > Set DistanceToCurrentTile(DistanceToCurrentTile:double=Vector2D Length!(A:Vector 2D Structure=vector2d - vector2d(A:Vector 2D Structure=@n8, B:Vector 2D Structure=@n9) -> (ReturnValue:Vector 2D Structure) {is=title K2Node_PromotableOperator}) -> (ReturnValue:double)) -> (Output_Get:double)
   > n10 = Do Once Then Do!(execute:Exec, Reset:Exec, StartClosed:bool) -> (DoOnce:Exec, ThenDo:Exec) {is=title K2Node_MacroInstance}
   n11 = NOT Boolean(A=Get IsClosestTileInSameColumn() -> (IsClosestTileInSameColumn:bool))
   n12 = Get IsCurrentTileInSameColumn() -> (IsCurrentTileInSameColumn:bool)
   n13 = "float < float"(A=Get DistanceToCurrentTile() -> (DistanceToCurrentTile:double), B=Get DistanceToClosestTile() -> (DistanceToClosestTile:double))
   n14 = "Equal (Boolean)"(A:bool=Get IsClosestTileInSameColumn(), B:bool=Get IsCurrentTileInSameColumn()) -> (ReturnValue:bool=~) {is=title K2Node_PromotableOperator}
   n15 = Branch(execute=@n10.ThenDo, Condition=OR Boolean(A=AND Boolean(A=@n12, B=@n11), B=AND Boolean(A=@n14, B=@n13)))
   Set DistanceToClosestTile(execute=[@n15.then, @n10.DoOnce], DistanceToClosestTile:double=Get DistanceToCurrentTile()) -> (Output_Get:double)
   > Set ClosestTile(ClosestTile:WBP Item Shop Tile Base Object Reference=Get Current Tile()) -> (Output_Get:WBP Item Shop Tile Base Object Reference)
   > Set IsClosestTileInSameColumn(IsClosestTileInSameColumn:bool=Get IsCurrentTileInSameColumn()) -> (Output_Get:bool)
   n16 = GetCachedGeometry(self=Get Current Tile())
   n17 = GetAbsolutePosition(Geometry=@n16)
   n18 = GetAbsoluteSize!(Geometry:"Geometry Structure (by ref)"=@n16) -> (ReturnValue_X:double, ReturnValue_Y:double)
   Set IsCurrentTileInSameColumn(
    execute=@n7.then_0
    IsCurrentTileInSameColumn:bool="In Range (Float)"!(Value:double=Get(), Min:double=@n17, Max:double=float + float(A=@n17, B=@n18), InclusiveMin:bool=true, InclusiveMax:bool=true) -> (ReturnValue:bool)
   ) -> (Output_Get:bool)
  ]}
 GetDesiredFocusTile {nodes [
   GetDesiredFocusTile() {is=title K2Node_FunctionEntry}
   > n1 = GetTileByOfferId!(execute, self, OfferId:String=Get FocusedOfferId() -> (FocusedOfferId:String)) -> (then, TileWidget:Widget Object Reference)
   > n2 = Is Valid(InputObject=@n1)
   Return Node!(execute=@n2.Is Not Valid, TileWidget:Widget Object Reference=GetFirstTile!(self) -> (TileWidget:Widget Object Reference))
   > Return Node(TileWidget=@n1)
  ]}
 GetFirstTile {nodes [
   GetFirstTile() {is=title K2Node_FunctionEntry}
   n1 = GetAllEntries(self=Get DynamicEntryBox_Tiles())
   > n2 = Branch(Condition=Is Valid Index!(TargetArray:Array of User Widget Object References=@n1, IndexToTest:int) -> (ReturnValue:bool))
   > Return Node(TileWidget="Get (a copy)"(Array:Array of User Widget Object References=@n1) -> (Output:User Widget Object Reference))
   Return Node(execute=@n2.else)
  ]}
 GetLastTile {nodes [
   GetLastTile() {is=title K2Node_FunctionEntry}
   n1 = GetAllEntries(self=Get DynamicEntryBox_Tiles())
   n2 = Last Index!(TargetArray:Array of User Widget Object References=@n1) -> (ReturnValue=0)
   > n3 = Branch(Condition=Is Valid Index(TargetArray=@n1, IndexToTest=@n2))
   > Return Node!(execute, LastTileWidget:Widget Object Reference="Get (a copy)"(Array=@n1, Dimension 1=@n2))
   Return Node(execute=@n3.else)
  ]}
 GetTileAtColumnIndex {nodes [
   GetTileAtColumnIndex!() -> (then, DesiredIndex:int) {is=title K2Node_FunctionEntry}
   n1 = Find(TargetMap:Map of Integers to Widget Object References=Get TileIndexMap(), Key:"Integer (by ref)"=Get DesiredIndex() -> (DesiredIndex:int=0)) -> (Value:Widget Object Reference)
   > n2 = Branch(Condition=@n1.ReturnValue)
   > n3 = Values!(execute, TargetMap:Map of Integers to Widget Object References=Get TileIndexMap()) -> (then, Values:Array of Widget Object References)
   > Return Node!(
    execute
    Tile:Widget Object Reference=@n1
    TileIndex:int=Find Item!(TargetArray:Array of Widget Object References=@n3, ItemToFind:"Widget Object Reference (by ref)"=@n1) -> (ReturnValue)
   )
   n4 = Get DesiredIndex() -> (DesiredIndex=0)
   PrintString!(
    execute=@n2.else
    InString=Append(A="Tile not found at desired column index ", B="To String (Integer)"(InInt=@n4))
    bPrintToScreen=false
    bPrintToLog
    TextColor
    Duration
    Key
   ) -> (then)
   > Return Node(TileIndex=-1)
  ]}
 GetTileByOfferId {nodes [
   GetTileByOfferId!() -> (then, OfferId:String) {is=title K2Node_FunctionEntry}
   > n1 = For Each Loop(Array=GetAllEntries(self=Get DynamicEntryBox_Tiles()))
   Return Node!(execute=@n1.Completed, TileWidget:Widget Object Reference)
   > n2 = Cast To WBP_ItemShopTile_Base(Object=@n1)
   > n3 = GetOfferVM!(execute, self:WBP Item Shop Tile Base Object Reference=@n2) -> (then, OfferVM:Fort Item Shop Offer VM Object Reference)
   > Branch(Condition="Equal Exactly (String)"(A=Get OfferId() -> (OfferId:String), B=Get OfferId!(self:Fort Item Shop Offer VM Object Reference=@n3) -> (OfferId)))
   > Return Node(TileWidget=@n1)
  ]}
 GetTileSpacing {nodes [
   GetTileSpacing() {is=title K2Node_FunctionEntry}
   n1 = GetItemShopConfiguration!() -> (ReturnValue:Athena Item Shop Configuration Object Reference)
   n2 = Get TileSpacing!(self:Athena Item Shop Configuration Object Reference=@n1) -> (TileSpacing:float)
   > Return Node!(
    execute
    TileSpacing:Vector 2D Structure=Select(
     Option 0:Vector 2D Structure=MakeVector2D(X=@n2, Y=@n2)
     Option 1:Vector 2D Structure=Get FrameBreakTileSpacing!(self:Athena Item Shop Configuration Object Reference=@n1) -> (FrameBreakTileSpacing:Vector 2D Structure)
     Index=IsUsingTileV2Layout()
    ) -> (ReturnValue:Vector 2D Structure)
   )
  ]}
 GetViewMoreButtonText {nodes [
   n1 = GetViewMoreButtonText!() -> (then, OfferVMs:Array of Fort Item Shop Offer VM Object References) {is=title K2Node_FunctionEntry}
   > Return Node!(
    execute
    Result:Text=Format Text!(Format="LOCTEXT(\"View More ({OfferCount})\")", OfferCount:int=Length(TargetArray:Array of Fort Item Shop Offer VM Object References=@n1)) -> (Result)
   )
  ]}
 I_GetClosestFocusTarget {nodes [
   n1 = I_GetClosestFocusTarget!() -> (then, TargetLocation:Vector 2D Structure) {is=title K2Node_FunctionEntry}
   > Return Node!(
    execute
    Target:Widget Object Reference=Get Closest Focus Target!(self, TargetLocation:Vector 2D Structure=@n1) -> (Widget:Widget Object Reference) {is=title K2Node_CallFunction}
   )
  ]}
 I_GetDesiredFocusTarget {nodes [
   I_GetDesiredFocusTarget() {is=title K2Node_FunctionEntry}
   > n1 = Branch(Condition=Get ShouldFocusListViewButton() -> (ShouldFocusListViewButton:bool))
   > Return Node!(execute, TileWidget:Widget Object Reference=Get Button_ViewMore())
   n2 = GetDesiredFocusTile!(execute=@n1.else, self) -> (then, TileWidget:Widget Object Reference) {is=title K2Node_CallFunction}
   > Return Node(TileWidget=@n2)
  ]}
 I_GetDesiredScrollDestination {nodes [
   I_GetDesiredScrollDestination() {is=title K2Node_FunctionEntry}
   > Return Node!(execute, Destination:EDescendantScrollDestination Enum=TopOrLeft)
  ]}
 I_GetDesiredScrollPadding {nodes [
   I_GetDesiredScrollPadding() {is=title K2Node_FunctionEntry}
   > Return Node!(
    execute
    ScrollPadding:double=Select(Option 0:double=256, Option 1:double=128, Index=IsInSubGame!(SubGame:ESubGame Enum=Campaign) -> (ReturnValue:bool)) -> (ReturnValue:double)
   )
  ]}
 I_GetFocusTargetFromOfferID {nodes [
   n1 = I_GetFocusTargetFromOfferID!() -> (then, OfferID:String) {is=title K2Node_FunctionEntry}
   > n2 = GetTileByOfferId(OfferId=@n1) {is=title K2Node_CallFunction}
   > Return Node!(execute, FocusTarget:Widget Object Reference=@n2)
  ]}
 I_GetScrollTarget {nodes [
   I_GetScrollTarget() {is=title K2Node_FunctionEntry}
   > Return Node!(execute, ScrollTarget:Widget Object Reference=Get Scrollbox_RowContent())
  ]}
 InitializeRow {nodes [
   n1 = InitializeRow!() -> (then, FortItemShopRow:Fort Item Shop Row Structure, ContentHasFocus:bool=~) {is=title K2Node_FunctionEntry}
   > n2 = Set FortItemShopRow!(execute, FortItemShopRow:Fort Item Shop Row Structure=@n1) -> (
    then
    Output_Get_OfferGroup:Fort Item Shop Offer Group VM Object Reference
    Output_Get_Offers:Array of Fort Item Shop Offer VM Object References
    Output_Get_bIsFirstRowInOfferGroup:bool=~
    Output_Get_bIsLastRowInOfferGroup:bool=~
   )
   > SetFortItemShopOfferGroupVM(Viewmodel=@n2)
   > CreateTileGrid!(execute, self, Offers:Array of Fort Item Shop Offer VM Object References=@n2.Output_Get_Offers) -> (then) {is=title K2Node_CallFunction}
   > UpdateHeaderSpacer()
   > SetParentHasFocusState(inParentHasFocus=Get ContentHasFocus() -> (ContentHasFocus:bool))
  ]}
 LoadAllPresentations {nodes [
   LoadAllPresentations() {is=title K2Node_FunctionEntry}
   > n1 = For Each Loop(Array=GetAllEntries(self=Get DynamicEntryBox_Tiles()))
   > n2 = Cast To WBP_ItemShopTile_Base(Object=@n1)
   > n3 = GetOfferVM(self=@n2)
   > RequestLoadPresentations!(
    execute
    self:WBP Item Shop Tile Base Object Reference=@n2
    DisplayAssetPath:String=Get DisplayAssetPath!(self:Fort Item Shop Offer VM Object Reference=@n3) -> (DisplayAssetPath:String)
   ) -> (then)
  ]}
 MarkOfferGroupAsVisited {nodes [
   MarkOfferGroupAsVisited() {is=title K2Node_FunctionEntry}
   n1 = Get FortItemShopOfferGroupVM()
   > Is Valid(InputObject=@n1)
   > Set SeenLevel!(execute, SeenLevel:EFortOfferSeenLevel Enum=OfferSectionVisited, self:Fort Item Shop Offer Group VM Object Reference=@n1) -> (then, Output_Get:EFortOfferSeenLevel Enum=Unseen)
  ]}
 OnNavigatedFrom {nodes [
   OnNavigatedFrom!() -> (then, Navigation:EUINavigation Enum) {is=title K2Node_FunctionEntry}
   > n1 = Cast To WBP_ItemShopTile_Base(Object=Get FocusedTile() -> (FocusedTile:User Widget Object Reference))
   > n2 = GetTileAnchor(self=@n1)
   > Call On Row Navigated From!(
    execute:Exec
    self:WBP MPItem Shop Row Object Reference
    Direction:EUINavigation Enum=Get Navigation() -> (Navigation:EUINavigation Enum)
    FocusedTilePosition:Vector 2D Structure=@n2
   ) -> (then:Exec) {is=title K2Node_CallDelegate}
   > Return Node!(execute, ReturnValue:Widget Object Reference)
  ]}
 OnOfferHovered {nodes [OnOfferHovered!() -> (then, OfferId:String) {is=title K2Node_FunctionEntry}]}
 OnRowFocused {nodes [OnRowFocused!() -> (then, RowWidget:User Widget Object Reference) {is=title K2Node_FunctionEntry}]}
 OnRowListViewRequested {nodes [OnRowListViewRequested!() -> (then, OfferGroupVM:Fort Item Shop Offer Group VM Object Reference) {is=title K2Node_FunctionEntry}]}
 OnRowNavigatedFrom {nodes [OnRowNavigatedFrom!() -> (then, Direction:EUINavigation Enum, FocusedTilePosition:Vector 2D Structure) {is=title K2Node_FunctionEntry}]}
 RequestMarkOfferGroupAsVisited {nodes [
   RequestMarkOfferGroupAsVisited() {is=title K2Node_FunctionEntry}
   > n1 = Branch(Condition="float > float"(A=Get MarkOfferGroupAsVisitedDelay() -> (MarkOfferGroupAsVisitedDelay:float)))
   n2 = Is Timer Active by Handle!(Handle:Timer Handle Structure=Get MarkOfferGroupAsVisitedHandler()) -> (ReturnValue:bool)
   n3 = MakeLiteralByte!(Value:Byte=Literal enum EFortOfferSeenLevel!(Enum:EFortOfferSeenLevel Enum=OfferSectionVisited) -> (ReturnValue:EFortOfferSeenLevel Enum) {is=title K2Node_EnumLiteral}) -> (ReturnValue:Byte=0)
   n4 = Get SeenLevel!(self:Fort Item Shop Offer Group VM Object Reference=Get FortItemShopOfferGroupVM()) -> (SeenLevel:EFortOfferSeenLevel Enum=Unseen)
   > Branch(Condition=AND Boolean(A="Byte < Byte"(A:Byte=@n4, B:Byte=@n3) -> (ReturnValue:bool=~) {is=title K2Node_PromotableOperator}, B=NOT Boolean(A=@n2)))
   > n5 = Set Timer by Event(Delegate=Create Event(), Time=Get MarkOfferGroupAsVisitedDelay(), bLooping=false)
   > Set MarkOfferGroupAsVisitedHandler(MarkOfferGroupAsVisitedHandler:Timer Handle Structure=@n5) -> (Output_Get:Timer Handle Structure)
   MarkOfferGroupAsVisited(execute=@n1.else) {is=title K2Node_CallFunction}
  ]}
 ResetAllTiles {nodes [
   ResetAllTiles() {is=title K2Node_FunctionEntry}
   > n1 = For Each Loop(Array=GetAllEntries(self=Get DynamicEntryBox_Tiles()))
   > n2 = Cast To WBP_ItemShopTile_Base(Object=@n1)
   > ResetTile(self:WBP Item Shop Tile Base Object Reference=@n2)
   > Unbind Event from On Tile Focused(self:WBP Item Shop Tile Base Object Reference=@n2, Delegate=Create Event()) {is=title K2Node_RemoveDelegate}
   > Unbind Event from On Request Scroll to Row(self:WBP Item Shop Tile Base Object Reference=@n2, Delegate=Create Event()) {is=title K2Node_RemoveDelegate}
   > Unbind Event from On Offer Hovered(self:WBP Item Shop Tile Base Object Reference=@n2, Delegate=Create Event()) {is=title K2Node_RemoveDelegate}
  ]}
 ResetRow {nodes [
   ResetRow() {is=title K2Node_FunctionEntry}
   > ResetScrollBox()
   > ResetAllTiles() {is=title K2Node_CallFunction}
   > Reset(self=Get DynamicEntryBox_Tiles(), bDeleteWidgets=false)
   > SetFortItemShopOfferGroupVM()
  ]}
 ResetScrollBox {nodes [
   ResetScrollBox() {is=title K2Node_FunctionEntry}
   > Set CurrentColumnIndex(CurrentColumnIndex=0) -> (Output_Get=0)
   > ScrollToStart(self:Scroll Box Object Reference=Get Scrollbox_RowContent())
   > UpdateNavigationToSideNav()
  ]}
 SetMobileStyle {nodes [
   SetMobileStyle() {is=title K2Node_FunctionEntry}
   > Branch(Condition=IsMobileGame())
   n1 = Get Button_ViewMore()
   > Set HorizontalAlignment!(
    execute
    HorizontalAlignment:EHorizontalAlignment Enum=HAlign_Left
    self:Vertical Box Slot Object Reference=SlotAsVerticalBoxSlot(Widget=@n1)
   ) -> (then, Output_Get:EHorizontalAlignment Enum=HAlign_Fill)
   > Set ButtonSize!(execute, ButtonSize:E_UI_ButtonSize Enum=NewEnumerator2, self:WBP UIKit Button CTA Base Object Reference=@n1) -> (then, Output_Get:E_UI_ButtonSize Enum=NewEnumerator0)
   n2 = SlotAsOverlaySlot(Widget=Get CommonHardwareVisibilityBorder_ScrollArrows() -> (CommonHardwareVisibilityBorder_ScrollArrows:Common Hardware Visibility Border Object Reference))
   n3 = Get!(self:Overlay Slot Object Reference=@n2) -> (Padding_Left:float=~, Padding_Top:float=~, Padding_Right:float=~, Padding_Bottom:float=~)
   > Set!(
    execute
    Padding_Left:float=@n3
    Padding_Top:float=16
    Padding_Right:float=@n3.Padding_Right
    Padding_Bottom:float=@n3.Padding_Bottom
    self:Overlay Slot Object Reference=@n2
   ) -> (then, Output_Get:Margin Structure)
  ]}
 SetParentHasFocusState {nodes [
   n1 = SetParentHasFocusState!() -> (then, inParentHasFocus:bool=~) {is=title K2Node_FunctionEntry}
   > Set ParentHasFocus(ParentHasFocus:bool=@n1) -> (Output_Get:bool)
   > n2 = Sequence!(execute) -> (then_0, then_1, then_2:Exec)
   n3 = Get ParentHasFocus()
   n4 = GetCurrentInputType(self=Get CommonInputSubsystem())
   Branch(
    execute=@n2.then_1
    Condition=AND Boolean(A=NOT Boolean(A=@n3), B="Equal (Enum)"(A:ECommonInputType Enum=@n4, B:ECommonInputType Enum=Gamepad))
   )
   > n5 = For Each Loop(Array=GetAllEntries(self=Get DynamicEntryBox_Tiles()))
   > n6 = Cast To WBP_ItemShopTile_Base(Object=@n5)
   > On Disable Hover(self:WBP Item Shop Tile Base Object Reference=@n6)
   UpdateScrollboxPresentation(execute=@n2.then_2)
   Update Row Hitbox Visibility(execute=@n2.then_0)
  ]}
 SetRowPeekState {nodes [
   n1 = SetRowPeekState!() -> (then, IsInPeekState:bool=~) {is=title K2Node_FunctionEntry}
   > n2 = Set IsRowPeeked(IsRowPeeked:bool=@n1) -> (Output_Get:bool)
   > n3 = Switch on E_ItemShop_RowState(Selection=Get RowState() -> (RowState=NewEnumerator4))
   n4 = PlayAnimation(
    execute=[@n3.NewEnumerator1, @n3.NewEnumerator3]
    InAnimation=Get OnInactiveRowHovered() -> (OnInactiveRowHovered:Widget Animation Object Reference)
    PlayMode=Select(Option 0:EUMGSequencePlayMode Enum=Reverse, Option 1:EUMGSequencePlayMode Enum, Index=@n2) -> (ReturnValue:EUMGSequencePlayMode Enum)
   )
   n5 = For Each Loop(Exec=[@n4.then, @n3.NewEnumerator0, @n3.NewEnumerator2], Array=GetAllEntries(self=Get DynamicEntryBox_Tiles()))
   > n6 = Cast To WBP_ItemShopTile_Base(Object=@n5)
   > SetPeekState!(execute, self:WBP Item Shop Tile Base Object Reference=@n6, IsPeeked:bool=Get IsRowPeeked()) -> (then)
  ]}
 SetRowState {nodes [
   n1 = SetRowState!() -> (then, InRowState:E_ItemShop_RowState Enum=NewEnumerator0) {is=title K2Node_FunctionEntry}
   > Set HasStateChanged(HasStateChanged:bool="Not Equal (Enum)"(A:E_ItemShop_RowState Enum=@n1, B:E_ItemShop_RowState Enum=Get RowState() -> (RowState=NewEnumerator4))) -> (Output_Get:bool)
   > Set RowState(RowState:E_ItemShop_RowState Enum=Get InRowState() -> (InRowState:E_ItemShop_RowState Enum=NewEnumerator4)) -> (Output_Get:E_ItemShop_RowState Enum=NewEnumerator4)
   > n2 = Sequence()
   > n3 = For Each Loop(Array=GetAllEntries(self=Get DynamicEntryBox_Tiles()))
   > n4 = Cast To WBP_ItemShopTile_Base(Object=@n3)
   > SetRowState(self:WBP Item Shop Tile Base Object Reference=@n4, InRowState=Get InRowState() -> (InRowState=NewEnumerator4)) {is=title K2Node_CallFunction}
   Update Row Hitbox Visibility(execute=@n2.then_1)
   n5 = Select!(
    NewEnumerator4:bool=false
    NewEnumerator1:bool=false
    NewEnumerator0:bool=true
    NewEnumerator2:bool=true
    NewEnumerator3:bool=false
    Index:E_ItemShop_RowState Enum=Get InRowState() -> (InRowState=NewEnumerator4)
   ) -> (ReturnValue:bool)
   > SetIsInteractionEnabled!(execute, self:Common Button Base Object Reference=Get Button_ViewMore(), bInIsInteractionEnabled:bool=@n5) -> (then)
   > Set RenderOpacity!(execute, RenderOpacity:float=SelectFloat(A=1, B=0.3, bPickA=@n5), self:Widget Object Reference=Get Button_ViewMore()) -> (then, Output_Get:float)
   > Branch(Condition=Get HasStateChanged() -> (HasStateChanged:bool))
   > UpdateScrollboxPresentation()
   n6 = Switch on E_ItemShop_RowState(execute=@n2.then_2, Selection=Get InRowState() -> (InRowState=NewEnumerator4))
   n7 = Create Event()
   SetNavigationRuleCustomBoundary!(execute=@n6.NewEnumerator0, self:Widget Object Reference, Direction:EUINavigation Enum=Up, InCustomDelegate:Delegate=@n7) -> (then)
   > SetNavigationRuleCustomBoundary(Direction=Down, InCustomDelegate=@n7)
   > RequestMarkOfferGroupAsVisited() {is=title K2Node_CallFunction}
   n8 = Get DynamicEntryBox_Tiles()
   > SendImpressionAnalyticsForRow!(
    execute
    self:Fort Item Shop Selection VM Object Reference=Get FortItemShopSelectionVM() -> (FortItemShopSelectionVM:Fort Item Shop Selection VM Object Reference)
    Row:"Fort Item Shop Row Structure (by ref)"=Get FortItemShopRow() -> (FortItemShopRow:Fort Item Shop Row Structure)
    OffersNum:int=Length(TargetArray:Array of User Widget Object References=GetAllEntries(self=@n8))
   ) -> (then)
   > Set SelectedOfferGroup!(
    execute
    SelectedOfferGroup:Fort Item Shop Offer Group VM Object Reference=Get FortItemShopOfferGroupVM()
    self:Fort Item Shop Selection VM Object Reference=Get FortItemShopSelectionVM()
   ) -> (then, Output_Get:Fort Item Shop Offer Group VM Object Reference)
   > UpdateNavigationToSideNav()
   > n9 = Branch(Condition=IsItemPreviewEnabledForOfferGroup!(OfferGroupVM:Fort Item Shop Offer Group VM Object Reference=Get FortItemShopOfferGroupVM()) -> (IsEnabled:bool))
   > Branch(Condition=Get ParentHasFocus())
   > n10 = Cast To WBP_ItemShopTile_Base(Object=GetFirstTile() {is=title K2Node_CallFunction})
   > Request3DPreview(self:WBP Item Shop Tile Base Object Reference=@n10)
   n11 = Get FortItemShopSelectionVM()
   Set ItemsToPreview!(
    execute=@n9.else
    ItemsToPreview:Array of Item Shop Preview Data Object References=Make Array("[0]":Item Shop Preview Data Object Reference) -> (Array:Array of Item Shop Preview Data Object References)
    self:Fort Item Shop Selection VM Object Reference=@n11
   ) -> (then, Output_Get:Array of Item Shop Preview Data Object References)
   > Set bIsItemPreviewCycleEnabled!(execute, bIsItemPreviewCycleEnabled:bool=false, self:Fort Item Shop Selection VM Object Reference=@n11) -> (then, Output_Get:bool)
   SetAllNavigationRules!(
    execute=[@n6.NewEnumerator1, @n6.NewEnumerator2, @n6.NewEnumerator3]
    self:Widget Object Reference
    Rule:EUINavigationRule Enum=Escape
    WidgetToFocus:Name=None
   ) -> (then)
   > CancelRequestMarkOfferGroupAsVisited() {is=title K2Node_CallFunction}
   > Set FocusedOfferId()
  ]}
 SetTilesLayout {nodes [
   SetTilesLayout() {is=title K2Node_FunctionEntry}
   > Set IsUsingTileV2Layout(IsUsingTileV2Layout:bool=IsUsingTileV2Layout()) -> (Output_Get:bool)
   n1 = Select!(
    Option 0:Margin Structure=Make Margin(Left=0, Top=Get TileV1BottomButtonsPadding() -> (TileV1BottomButtonsPadding:double), Right=0, Bottom=0)
    Option 1:Margin Structure=Make Margin(Left=0, Top=Get TileV2BottomButtonsPadding() -> (TileV2BottomButtonsPadding:double), Right=0, Bottom=0)
    Index:bool=Get IsUsingTileV2Layout() -> (IsUsingTileV2Layout:bool)
   ) -> (ReturnValue:Margin Structure)
   > Set Padding!(execute, Padding:Margin Structure=@n1, self:Border Object Reference=Get CommonHardwareVisibilityBorder_ScrollArrows()) -> (then, Output_Get:Margin Structure)
   > SetPadding!(execute, self:Overlay Slot Object Reference=SlotAsOverlaySlot(Widget=Get Button_ViewMore()), InPadding:Margin Structure=@n1) -> (then)
   n2 = Get FrameBreakPadding!(self:Athena Item Shop Configuration Object Reference=GetItemShopConfiguration()) -> (FrameBreakPadding:float)
   > SetPadding(
    self:Scroll Box Slot Object Reference=SlotAsScrollBoxSlot!(Widget:Widget Object Reference=Get DynamicEntryBox_Tiles()) -> (ReturnValue:Scroll Box Slot Object Reference)
    InPadding=Select(
     Option 0=Make Margin(Left=0, Top=0, Right=0, Bottom=0)
     Option 1=Make Margin(Left=0, Top=@n2, Right=0, Bottom=0)
     Index=Get IsUsingTileV2Layout()
    )
   )
  ]}
 Update Row Hitbox Visibility {nodes [
   Update Row Hitbox Visibility() {is=title K2Node_FunctionEntry}
   n1 = NOT Boolean(A=Get ParentHasFocus())
   n2 = Select!(
    NewEnumerator4:bool=false
    NewEnumerator1:bool=true
    NewEnumerator0:bool=false
    NewEnumerator2:bool=false
    NewEnumerator3:bool=true
    Index:E_ItemShop_RowState Enum=Get RowState() -> (RowState=NewEnumerator4)
   ) -> (ReturnValue:bool)
   > SetVisibility(
    self=Get Button_RowHitbox() -> (Button_RowHitbox:Button Object Reference)
    InVisibility=Select!(Option 0:ESlateVisibility Enum=SelfHitTestInvisible, Option 1:ESlateVisibility Enum=Visible, Index:bool=OR Boolean(A=@n1, B=@n2)) -> (ReturnValue:ESlateVisibility Enum)
   )
  ]}
 UpdateHeaderSpacer {nodes [
   UpdateHeaderSpacer() {is=title K2Node_FunctionEntry}
   n1 = GetTileSpacing() {is=title K2Node_CallFunction}
   n2 = IsUsingTileV2Layout()
   n3 = IsMobileGame()
   n4 = GetDisplayType(self=Get FortItemShopOfferGroupVM()) -> (ReturnValue=TileGrid)
   n5 = IsItemPreviewEnabled!(self:Fort Item Shop VM Object Reference=Get FortItemShopVM() -> (FortItemShopVM:Fort Item Shop VM Object Reference)) -> (ReturnValue:bool)
   n6 = SelectFloat(
    A=float - float(A=SelectFloat(A=258, B=150, bPickA=@n3), B=Select(Option 0:double=66, Option 1:double=0, Index=@n5) -> (ReturnValue:double))
    B=float + float!(
     A=SelectFloat(A=129, B=150, bPickA=@n3)
     B=Select(Option 0=32, Option 1=0, Index=@n5)
     C:double=Select!(TileGrid:double=0, ExpandableList:double=80, Billboard:double=0, ScrollableList:double=0, Index:EFortItemShopOfferDisplayType Enum=@n4) -> (ReturnValue)
    ) -> (ReturnValue)
    bPickA=IsLastOfferGroupInCategory!(
     self:Fort Item Shop VM Object Reference=Get FortItemShopVM()
     OfferGroup:Fort Item Shop Offer Group VM Object Reference=Get FortItemShopOfferGroupVM()
    ) -> (ReturnValue:bool)
   )
   K2Node_Composite_0 = TileV2Difference!() -> (ReturnValue:double=~) {is=title K2Node_Composite}
   n7 = float + float!(A=@K2Node_Composite_0, B=@n6) -> (ReturnValue)
   n8 = Get FortItemShopOfferGroupVM()
   n9 = Get FortItemShopVM()
   n10 = NOT Boolean(A=IsLastOfferGroupInShop!(self:Fort Item Shop VM Object Reference=@n9, OfferGroup:Fort Item Shop Offer Group VM Object Reference=@n8) -> (ReturnValue:bool))
   n11 = Get!() -> (
    FortItemShopRow_OfferGroup:Fort Item Shop Offer Group VM Object Reference
    FortItemShopRow_Offers:Array of Fort Item Shop Offer VM Object References
    FortItemShopRow_bIsFirstRowInOfferGroup:bool=~
    FortItemShopRow_bIsLastRowInOfferGroup:bool=~
   )
   n12 = IsLastOfferGroupInSection!(self:Fort Item Shop VM Object Reference=@n9, OfferGroup:Fort Item Shop Offer Group VM Object Reference=@n8) -> (ReturnValue:bool)
   > Set!(
    execute
    Size_X:double=0
    Size_Y:double=SelectFloat(
     A=SelectFloat(A=@n7, B=@n6, bPickA=@n2)
     B=BreakVector2D(InVec=@n1).Y
     bPickA=AND Boolean!(A=@n11.FortItemShopRow_bIsLastRowInOfferGroup, B=@n12, C:bool=@n10) -> (ReturnValue)
    )
    self:Spacer Object Reference=Get Spacer_BetweenRows() -> (Spacer_BetweenRows:Spacer Object Reference)
   ) -> (then, Output_Get:Vector 2D Structure)
  ]}
 TileV2Difference {
  path=UpdateHeaderSpacer.K2Node_Composite_0.TileV2Difference
  nodes [
   Inputs!()
   n1 = GetItemShopConfiguration()
   n2 = Get FrameBreakPadding(self=@n1)
   n3 = Get!(self:Athena Item Shop Configuration Object Reference=@n1) -> (FrameBreakSingleTileDimensions_X:double, FrameBreakSingleTileDimensions_Y:double)
   Outputs!(ReturnValue:double=float - float(
     A=Get!(self=@n1) -> (SingleTileDimensions_X:double, SingleTileDimensions_Y:double).SingleTileDimensions_Y
     B=float + float(A=@n3.FrameBreakSingleTileDimensions_Y, B=@n2)
    ))
  ]
 }
 UpdateNavigationToSideNav {nodes [
   UpdateNavigationToSideNav() {is=title K2Node_FunctionEntry}
   > Set bCanNavigateToSideNav!(
    execute
    bCanNavigateToSideNav:bool=CanNavigateToSideNav!(self) -> (ReturnValue:bool) {is=title K2Node_CallFunction}
    self:Fort Item Shop Selection VM Object Reference=Get FortItemShopSelectionVM()
   ) -> (then, Output_Get:bool)
  ]}
 UpdateScrollboxPresentation {nodes [
   UpdateScrollboxPresentation() {is=title K2Node_FunctionEntry}
   > n1 = Get!(execute:Exec) -> (then:Exec, else:Exec, FortItemShopOfferGroupVM:Fort Item Shop Offer Group VM Object Reference)
   > n2 = Branch(Condition="Equal (Enum)"(A:EFortItemShopOfferDisplayType Enum=GetDisplayType(self=@n1) -> (ReturnValue=TileGrid), B:EFortItemShopOfferDisplayType Enum=ScrollableList))
   > ForceLayoutPrepass(self:Widget Object Reference=Get Scrollbox_RowContent())
   n3 = Get Scrollbox_RowContent()
   n4 = GetScrollOffset(self=@n3)
   > Set CanNavigateLeft(CanNavigateLeft:bool=NOT Boolean(A="Nearly Equal (Float)"(A=@n4, B=0, ErrorTolerance=1))) -> (Output_Get:bool)
   n5 = GetScrollOffsetOfEnd!(self:Scroll Box Object Reference=@n3) -> (ReturnValue:float)
   n6 = Get!(self:Spacer Object Reference=Get Spacer_EndPadding()) -> (Size_X:double, Size_Y:double)
   > Set CanNavigateRight(CanNavigateRight:bool=NOT Boolean(A="Nearly Equal (Float)"(A=@n4, B=@n5, ErrorTolerance=@n6))) -> (Output_Get:bool)
   > SetVisibility(
    self=Get Image_Fade_Left() -> (Image_Fade_Left:Image Object Reference)
    InVisibility=Select!(
     Option 0:ESlateVisibility Enum=Collapsed
     Option 1:ESlateVisibility Enum=SelfHitTestInvisible
     Index:bool=Get CanNavigateLeft() -> (CanNavigateLeft:bool)
    ) -> (ReturnValue:ESlateVisibility Enum)
   )
   > SetVisibility(
    self=Get Image_Fade_Right() -> (Image_Fade_Right:Image Object Reference)
    InVisibility=Select(Option 0=Collapsed, Option 1=SelfHitTestInvisible, Index=Get CanNavigateRight() -> (CanNavigateRight:bool))
   )
   n7 = IsUsingInputType!(InputType:ECommonInputType Enum=Touch, PlayerController:Player Controller Object Reference=GetOwningPlayer()) -> (ReturnValue:bool)
   n8 = NOT Boolean(A=IsItemPreviewEnabled(self=Get FortItemShopVM()))
   n9 = Get ParentHasFocus()
   n10 = Get RowState() -> (RowState=NewEnumerator4)
   > n11 = Branch(Condition=AND Boolean(
     A=NOT Boolean(A=@n7)
     B="Equal (Enum)"(A:E_ItemShop_RowState Enum=@n10, B:E_ItemShop_RowState Enum=NewEnumerator0)
     C=OR Boolean(A=@n9, B=@n8)
    ))
   n12 = Get HorizontalBox_NavigationArrows() -> (HorizontalBox_NavigationArrows:Horizontal Box Object Reference)
   > SetVisibility(self=@n12, InVisibility=SelfHitTestInvisible)
   n13 = Get CanNavigateLeft()
   > SetIsInteractionEnabled(self=Get Button_ScrollLeft() -> (Button_ScrollLeft:WBP UIKit Button Generic Object Reference), bInIsInteractionEnabled=@n13)
   > Set RenderOpacity(RenderOpacity=SelectFloat(A=1, B=0.4, bPickA=@n13), self=Get Button_ScrollLeft())
   n14 = Get CanNavigateRight()
   > SetIsInteractionEnabled(self=Get Button_ScrollRight() -> (Button_ScrollRight:WBP UIKit Button Generic Object Reference), bInIsInteractionEnabled=@n14)
   > Set RenderOpacity(RenderOpacity=SelectFloat(A=1, B=0.4, bPickA=@n14), self=Get Button_ScrollRight())
   SetVisibility(execute=@n11.else, self=@n12, InVisibility=Collapsed)
   SetVisibility(
    execute=[@n2.else, @n1.else]
    self=[Get HorizontalBox_NavigationArrows(), Get Image_Fade_Left(), Get Image_Fade_Right()]
    InVisibility=Collapsed
   )
  ]}
 VM_CanNavigateToSideNavUpdated {nodes [
   n1 = VM_CanNavigateToSideNavUpdated!() -> (then, CanNavigate:bool=~) {is=title K2Node_FunctionEntry}
   > SetNavigationRuleBase!(
    execute
    self:Widget Object Reference=Get DynamicEntryBox_Tiles()
    Direction:EUINavigation Enum=Left
    Rule:EUINavigationRule Enum=Select(Option 0:EUINavigationRule Enum=Stop, Option 1:EUINavigationRule Enum=Escape, Index=@n1) -> (ReturnValue:EUINavigationRule Enum)
   ) -> (then)
  ]}
 VM_DisplayTypeUpdated {nodes [
   n1 = VM_DisplayTypeUpdated!() -> (then, DisplayType:EFortItemShopOfferDisplayType Enum=TileGrid) {is=title K2Node_FunctionEntry}
   n2 = "Equal (Enum)"(A:EFortItemShopOfferDisplayType Enum=@n1, B:EFortItemShopOfferDisplayType Enum=ExpandableList)
   n3 = "integer > integer"(
    A=Length(TargetArray:Array of Fort Item Shop Offer VM Object References=Get!() -> (
      FortItemShopRow_OfferGroup:Fort Item Shop Offer Group VM Object Reference
      FortItemShopRow_Offers:Array of Fort Item Shop Offer VM Object References
      FortItemShopRow_bIsFirstRowInOfferGroup:bool=~
      FortItemShopRow_bIsLastRowInOfferGroup:bool=~
     ).FortItemShopRow_Offers)
    B=GetTileGridRowWidth!() -> (ReturnValue:int=0)
   )
   > Set Visibility(
    Visibility=Select(Option 0:ESlateVisibility Enum=Collapsed, Option 1:ESlateVisibility Enum=Visible, Index=AND Boolean!(A=@n2, B=@n3) -> (ReturnValue)) -> (ReturnValue:ESlateVisibility Enum)
    self=Get Button_ViewMore()
   )
   > Set Clipping!(
    execute
    Clipping:EWidgetClipping Enum=Select!(
     TileGrid:EWidgetClipping Enum=Inherit
     ExpandableList:EWidgetClipping Enum=Inherit
     Billboard:EWidgetClipping Enum=Inherit
     ScrollableList:EWidgetClipping Enum=ClipToBounds
     Index:EFortItemShopOfferDisplayType Enum=Get DisplayType() -> (DisplayType:EFortItemShopOfferDisplayType Enum=TileGrid)
    ) -> (ReturnValue:EWidgetClipping Enum)
    self:Widget Object Reference=Get Overlay_Clipper() -> (Overlay_Clipper:Overlay Object Reference)
   ) -> (then, Output_Get:EWidgetClipping Enum=Inherit)
   > UpdateScrollboxPresentation() {is=title K2Node_CallFunction}
  ]}
]
variables [
 CappedColumnCount {type=integer, default=4}
 ColumnsToScroll {type=integer, default=4}
 CurrentColumnIndex {type=integer}
 FocusedOfferId {type=string, default=""}
 FocusedTile {type=UserWidget, default=null}
 FortItemShopRow {type=FortItemShopRow, default {bIsFirstRowInOfferGroup=true, bIsLastRowInOfferGroup=true, offerGroup=null, offers []}}
 IsRowPeeked {type=boolean}
 MarkOfferGroupAsVisitedDelay {type=number, default=2}
 MarkOfferGroupAsVisitedHandler {type=TimerHandle}
 OnOfferHovered {}
 OnRowFocused {}
 OnRowListViewRequested {}
 OnRowNavigatedFrom {}
 ParentHasFocus {type=boolean}
 RowCount {type=integer, default=1}
 RowState {type=E_ItemShop_RowState, default=None}
 ShouldFocusListViewButton {type=boolean}
 TileIndexMap {default {}}
 TileV1BottomButtonsPadding {type=number, default=24}
 TileV2BottomButtonsPadding {type=number, default=24}
 TileWidgetClass {type=Class, default=null}
 TotalColumnCount {type=integer}
]
widgets {
 info {inheritedWidgetCount=0, parentClass=/Script/FortniteUI.FortMPItemShopRow, rootWidgetClass=&Overlay, widgetCount=21}
 widgets [
  Overlay_2 {widgetClassPath=&Overlay, widget=Overlay_2}
  VerticalBox_5 {widgetClassPath=&VerticalBox, parent=Overlay_2, widget=VerticalBox_5, slot {class=OverlaySlot}}
  Overlay_RowContent {widgetClassPath=&Overlay, parent=VerticalBox_5, widget=Overlay_RowContent, slot {class=VerticalBoxSlot}}
  Overlay_Clipper {
   widgetClassPath=&Overlay
   bIsVariable=true
   parent=Overlay_RowContent
   widget=Overlay_Clipper
   properties {clipping=ClipToBounds}
   slot {class=OverlaySlot, properties {padding=<ltrb -132 -24 0 -24>}}
  }
  Scrollbox_RowContent {
   widgetClassPath=/Script/FortniteUI.AthenaScrollBox
   bIsVariable=true
   parent=Overlay_Clipper
   widget=Scrollbox_RowContent
   properties {
    allowOverscroll=false
    bAllowRightClickDragScrolling=false
    bEnableClickThrough=true
    clipping=Inherit
    consumeMouseWheel=Never
    navigationScrollPadding=108
    orientation=Orient_Horizontal
    scrollBarVisibility=Collapsed
    scrollWhenFocusChanges=AnimatedScroll
    wheelScrollMultiplier=0
    scrollbarPadding=<ltrb 0 0 0 0>
    scrollbarThickness=<0 0>
    widgetBarStyle {
     thickness=0
     draggedThumbImage {
      bIsDynamicallyLoaded=false
      drawAs=Box
      imageType=FullColor
      mirroring=NoMirror
      resourceName=../../../Engine/Content/Slate/Common/Scrollbar_Thumb.png
      resourceObject=null
      tiling=NoTile
      imageSize=<0 0>
      margin=<ltrb 0.25 0.25 0.25 0.25>
      outlineSettings {
       bUseBrushTransparency=false
       roundingType=HalfHeightRadius
       width=0
       color {colorUseRule=UseColor_Specified, specifiedColor=<rgba 0 0 0 0>}
       cornerRadii=<0 0 0 0>
      }
      tintColor {colorUseRule=UseColor_Specified, specifiedColor=<rgba 1 1 1 1>}
      uVRegion {bIsValid=false, max=<0 0>, min=<0 0>}
     }
     horizontalBackgroundImage {
      bIsDynamicallyLoaded=false
      drawAs=NoDrawType
      imageType=NoImage
      mirroring=NoMirror
      resourceName=None
      resourceObject=null
      tiling=NoTile
      imageSize=<0 0>
      margin=<ltrb 0 0 0 0>
      outlineSettings {
       bUseBrushTransparency=false
       roundingType=HalfHeightRadius
       width=0
       color {colorUseRule=UseColor_Specified, specifiedColor=<rgba 0 0 0 0>}
       cornerRadii=<0 0 0 0>
      }
      tintColor {colorUseRule=UseColor_Specified, specifiedColor=<rgba 1 1 1 1>}
      uVRegion {bIsValid=false, max=<0 0>, min=<0 0>}
     }
     horizontalBottomSlotImage {
      bIsDynamicallyLoaded=false
      drawAs=Image
      imageType=FullColor
      mirroring=NoMirror
      resourceName=../../../Engine/Content/Slate/Common/Scrollbar_Background_Horizontal.png
      resourceObject=null
      tiling=NoTile
      imageSize=<8 8>
      margin=<ltrb 0 0 0 0>
      outlineSettings {
       bUseBrushTransparency=false
       roundingType=HalfHeightRadius
       width=0
       color {colorUseRule=UseColor_Specified, specifiedColor=<rgba 0 0 0 0>}
       cornerRadii=<0 0 0 0>
      }
      tintColor {colorUseRule=UseColor_Specified, specifiedColor=<rgba 1 1 1 1>}
      uVRegion {bIsValid=false, max=<0 0>, min=<0 0>}
     }
     horizontalTopSlotImage {
      bIsDynamicallyLoaded=false
      drawAs=Image
      imageType=FullColor
      mirroring=NoMirror
      resourceName=../../../Engine/Content/Slate/Common/Scrollbar_Background_Horizontal.png
      resourceObject=null
      tiling=NoTile
      imageSize=<8 8>
      margin=<ltrb 0 0 0 0>
      outlineSettings {
       bUseBrushTransparency=false
       roundingType=HalfHeightRadius
       width=0
       color {colorUseRule=UseColor_Specified, specifiedColor=<rgba 0 0 0 0>}
       cornerRadii=<0 0 0 0>
      }
      tintColor {colorUseRule=UseColor_Specified, specifiedColor=<rgba 1 1 1 1>}
      uVRegion {bIsValid=false, max=<0 0>, min=<0 0>}
     }
     hoveredThumbImage {
      bIsDynamicallyLoaded=false
      drawAs=Box
      imageType=FullColor
      mirroring=NoMirror
      resourceName=../../../Engine/Content/Slate/Common/Scrollbar_Thumb.png
      resourceObject=null
      tiling=NoTile
      imageSize=<0 0>
      margin=<ltrb 0.25 0.25 0.25 0.25>
      outlineSettings {
       bUseBrushTransparency=false
       roundingType=HalfHeightRadius
       width=0
       color {colorUseRule=UseColor_Specified, specifiedColor=<rgba 0 0 0 0>}
       cornerRadii=<0 0 0 0>
      }
      tintColor {colorUseRule=UseColor_Specified, specifiedColor=<rgba 1 1 1 1>}
      uVRegion {bIsValid=false, max=<0 0>, min=<0 0>}
     }
     normalThumbImage {
      bIsDynamicallyLoaded=false
      drawAs=Box
      imageType=FullColor
      mirroring=NoMirror
      resourceName=../../../Engine/Content/Slate/Common/Scrollbar_Thumb.png
      resourceObject=null
      tiling=NoTile
      imageSize=<0 0>
      margin=<ltrb 0.25 0.25 0.25 0.25>
      outlineSettings {
       bUseBrushTransparency=false
       roundingType=HalfHeightRadius
       width=0
       color {colorUseRule=UseColor_Specified, specifiedColor=<rgba 0 0 0 0>}
       cornerRadii=<0 0 0 0>
      }
      tintColor {colorUseRule=UseColor_Specified, specifiedColor=<rgba 1 1 1 1>}
      uVRegion {bIsValid=false, max=<0 0>, min=<0 0>}
     }
     verticalBackgroundImage {
      bIsDynamicallyLoaded=false
      drawAs=NoDrawType
      imageType=NoImage
      mirroring=NoMirror
      resourceName=None
      resourceObject=null
      tiling=NoTile
      imageSize=<0 0>
      margin=<ltrb 0 0 0 0>
      outlineSettings {
       bUseBrushTransparency=false
       roundingType=HalfHeightRadius
       width=0
       color {colorUseRule=UseColor_Specified, specifiedColor=<rgba 0 0 0 0>}
       cornerRadii=<0 0 0 0>
      }
      tintColor {colorUseRule=UseColor_Specified, specifiedColor=<rgba 1 1 1 1>}
      uVRegion {bIsValid=false, max=<0 0>, min=<0 0>}
     }
     verticalBottomSlotImage {
      bIsDynamicallyLoaded=false
      drawAs=Image
      imageType=FullColor
      mirroring=NoMirror
      resourceName=../../../Engine/Content/Slate/Common/Scrollbar_Background_Vertical.png
      resourceObject=null
      tiling=NoTile
      imageSize=<8 8>
      margin=<ltrb 0 0 0 0>
      outlineSettings {
       bUseBrushTransparency=false
       roundingType=HalfHeightRadius
       width=0
       color {colorUseRule=UseColor_Specified, specifiedColor=<rgba 0 0 0 0>}
       cornerRadii=<0 0 0 0>
      }
      tintColor {colorUseRule=UseColor_Specified, specifiedColor=<rgba 1 1 1 1>}
      uVRegion {bIsValid=false, max=<0 0>, min=<0 0>}
     }
     verticalTopSlotImage {
      bIsDynamicallyLoaded=false
      drawAs=Image
      imageType=FullColor
      mirroring=NoMirror
      resourceName=../../../Engine/Content/Slate/Common/Scrollbar_Background_Vertical.png
      resourceObject=null
      tiling=NoTile
      imageSize=<8 8>
      margin=<ltrb 0 0 0 0>
      outlineSettings {
       bUseBrushTransparency=false
       roundingType=HalfHeightRadius
       width=0
       color {colorUseRule=UseColor_Specified, specifiedColor=<rgba 0 0 0 0>}
       cornerRadii=<0 0 0 0>
      }
      tintColor {colorUseRule=UseColor_Specified, specifiedColor=<rgba 1 1 1 1>}
      uVRegion {bIsValid=false, max=<0 0>, min=<0 0>}
     }
    }
    widgetStyle {
     barThickness=9
     bottomShadowBrush {
      bIsDynamicallyLoaded=false
      drawAs=NoDrawType
      imageType=FullColor
      mirroring=NoMirror
      resourceName=../../../Engine/Content/Slate/Common/ScrollBoxShadowBottom.png
      resourceObject=null
      tiling=NoTile
      imageSize=<16 8>
      margin=<ltrb 0.5 0 0.5 1>
      outlineSettings {
       bUseBrushTransparency=false
       roundingType=HalfHeightRadius
       width=0
       color {colorUseRule=UseColor_Specified, specifiedColor=<rgba 0 0 0 0>}
       cornerRadii=<0 0 0 0>
      }
      tintColor {colorUseRule=UseColor_Specified, specifiedColor=<rgba 1 1 1 1>}
      uVRegion {bIsValid=false, max=<0 0>, min=<0 0>}
     }
     horizontalScrolledContentPadding=<ltrb 0 0 0 0>
     leftShadowBrush {
      bIsDynamicallyLoaded=false
      drawAs=NoDrawType
      imageType=FullColor
      mirroring=NoMirror
      resourceName=../../../Engine/Content/Slate/Common/ScrollBoxShadowLeft.png
      resourceObject=null
      tiling=NoTile
      imageSize=<8 16>
      margin=<ltrb 1 0.5 0 0.5>
      outlineSettings {
       bUseBrushTransparency=false
       roundingType=HalfHeightRadius
       width=0
       color {colorUseRule=UseColor_Specified, specifiedColor=<rgba 0 0 0 0>}
       cornerRadii=<0 0 0 0>
      }
      tintColor {colorUseRule=UseColor_Specified, specifiedColor=<rgba 1 1 1 1>}
      uVRegion {bIsValid=false, max=<0 0>, min=<0 0>}
     }
     rightShadowBrush {
      bIsDynamicallyLoaded=false
      drawAs=NoDrawType
      imageType=FullColor
      mirroring=NoMirror
      resourceName=../../../Engine/Content/Slate/Common/ScrollBoxShadowRight.png
      resourceObject=null
      tiling=NoTile
      imageSize=<8 16>
      margin=<ltrb 0 0.5 1 0.5>
      outlineSettings {
       bUseBrushTransparency=false
       roundingType=HalfHeightRadius
       width=0
       color {colorUseRule=UseColor_Specified, specifiedColor=<rgba 0 0 0 0>}
       cornerRadii=<0 0 0 0>
      }
      tintColor {colorUseRule=UseColor_Specified, specifiedColor=<rgba 1 1 1 1>}
      uVRegion {bIsValid=false, max=<0 0>, min=<0 0>}
     }
     topShadowBrush {
      bIsDynamicallyLoaded=false
      drawAs=NoDrawType
      imageType=FullColor
      mirroring=NoMirror
      resourceName=../../../Engine/Content/Slate/Common/ScrollBoxShadowTop.png
      resourceObject=null
      tiling=NoTile
      imageSize=<16 8>
      margin=<ltrb 0.5 1 0.5 0>
      outlineSettings {
       bUseBrushTransparency=false
       roundingType=HalfHeightRadius
       width=0
       color {colorUseRule=UseColor_Specified, specifiedColor=<rgba 0 0 0 0>}
       cornerRadii=<0 0 0 0>
      }
      tintColor {colorUseRule=UseColor_Specified, specifiedColor=<rgba 1 1 1 1>}
      uVRegion {bIsValid=false, max=<0 0>, min=<0 0>}
     }
     verticalScrolledContentPadding=<ltrb 0 0 0 0>
    }
   }
   slot {class=OverlaySlot, properties {padding=<ltrb 24 24 0 24>}}
  }
  Spacer_StartPadding {
   widgetClassPath=&Spacer
   parent=Scrollbox_RowContent
   widget=Spacer_StartPadding
   properties {size=<108 1>}
   slot {class=ScrollBoxSlot}
  }
  DynamicEntryBox_Tiles {
   widgetClassPath=/Script/FortniteUI.FortDynamicEntryBox
   bIsVariable=true
   parent=Scrollbox_RowContent
   widget=DynamicEntryBox_Tiles
   properties {
    entryWidgetClass=&WBP_ItemShopTile_Default_C
    numDesignerPreviewEntries=4
    entrySpacing=<24 0>
   }
   slot {class=ScrollBoxSlot, properties {horizontalAlignment=HAlign_Left, verticalAlignment=VAlign_Top}}
  }
  Spacer_EndPadding {
   widgetClassPath=&Spacer
   bIsVariable=true
   parent=Scrollbox_RowContent
   widget=Spacer_EndPadding
   properties {visibility=Collapsed, size=<84 1>}
   slot {class=ScrollBoxSlot}
  }
  Image_Fade_Left {
   widgetClassPath=&Image
   bIsVariable=true
   parent=Overlay_RowContent
   widget=Image_Fade_Left
   properties {
    visibility=SelfHitTestInvisible
    brush {
     bIsDynamicallyLoaded=false
     drawAs=Image
     imageType=FullColor
     mirroring=NoMirror
     resourceName=None
     resourceObject=/FortUILibrary/Systems/Items/MaterialInstances/MI_UI_ListFade_Left
     tiling=NoTile
     imageSize=<84 32>
     margin=<ltrb 0 0 0 0>
     outlineSettings {
      bUseBrushTransparency=false
      roundingType=HalfHeightRadius
      width=0
      color {colorUseRule=UseColor_Specified, specifiedColor=<rgba 0 0 0 0>}
      cornerRadii=<0 0 0 0>
     }
     tintColor {colorUseRule=UseColor_Specified, specifiedColor=<rgba 1 1 1 1>}
     uVRegion {bIsValid=false, max=<0 0>, min=<0 0>}
    }
   }
   slot {class=OverlaySlot, properties {verticalAlignment=VAlign_Fill, padding=<ltrb -132 -24 0 -24>}}
  }
  Image_Fade_Right {
   widgetClassPath=&Image
   bIsVariable=true
   parent=Overlay_RowContent
   widget=Image_Fade_Right
   properties {
    visibility=SelfHitTestInvisible
    brush {
     bIsDynamicallyLoaded=false
     drawAs=Image
     imageType=FullColor
     mirroring=NoMirror
     resourceName=None
     resourceObject=/FortUILibrary/Systems/Items/MaterialInstances/MI_UI_ListFade_Right
     tiling=NoTile
     imageSize=<100 32>
     margin=<ltrb 0 0 0 0>
     outlineSettings {
      bUseBrushTransparency=false
      roundingType=HalfHeightRadius
      width=0
      color {colorUseRule=UseColor_Specified, specifiedColor=<rgba 0 0 0 0>}
      cornerRadii=<0 0 0 0>
     }
     tintColor {colorUseRule=UseColor_Specified, specifiedColor=<rgba 1 1 1 1>}
     uVRegion {bIsValid=false, max=<0 0>, min=<0 0>}
    }
   }
   slot {class=OverlaySlot, properties {horizontalAlignment=HAlign_Right, verticalAlignment=VAlign_Fill, padding=<ltrb 0 -24 0 -24>}}
  }
  Overlay_BottomContent {widgetClassPath=&Overlay, parent=VerticalBox_5, widget=Overlay_BottomContent, slot {class=VerticalBoxSlot}}
  WBP_CaptureForPostBufferUpdate {
   widgetClassPath=&WBP_CaptureForPostBufferUpdate
   bIsVariable=true
   parent=Overlay_BottomContent
   slot {class=OverlaySlot, properties {horizontalAlignment=HAlign_Fill, verticalAlignment=VAlign_Fill}}
  }
  Button_ViewMore {
   widgetClassPath=&WBP_UIKit_Button_Regular
   bIsVariable=true
   parent=Overlay_BottomContent
   widget=Button_ViewMore
   properties {mobileMinWidth=536, text="View More (0)"}
   slot {class=OverlaySlot, properties {horizontalAlignment=HAlign_Fill, padding=<ltrb 0 24 0 24>}}
  }
  CommonHardwareVisibilityBorder_ScrollArrows {
   widgetClassPath=&CommonHardwareVisibilityBorder
   bIsVariable=true
   parent=Overlay_BottomContent
   widget=CommonHardwareVisibilityBorder_ScrollArrows
   properties {visibilityQuery {
     autoDescription=" NONE( Input.Gamepad )"
     tokenStreamVersion=0
     userDescription=No gamepad
     queryTokenStream [0,1,3,1,0]
     tagDictionary [{tagName=Input.Gamepad}]
    }}
   slot {class=OverlaySlot, properties {horizontalAlignment=HAlign_Center}}
  }
  HorizontalBox_NavigationArrows {
   widgetClassPath=&HorizontalBox
   bIsVariable=true
   parent=CommonHardwareVisibilityBorder_ScrollArrows
   widget=HorizontalBox_NavigationArrows
   properties {visibility=Collapsed}
   slot {class=BorderSlot, properties {padding=<ltrb 0 0 0 0>}}
  }
  Button_ScrollLeft {
   widgetClassPath=&WBP_UIKit_Button_Generic
   bIsVariable=true
   parent=HorizontalBox_NavigationArrows
   widget=Button_ScrollLeft
   properties {
    bIsFocusable=false
    iconMaterial=/ItemShop/MultiProduct/Materials/Buttons/MI_UI_Icon_Button_Previous
    minHeight=48
    mobileMinHeight=80
    outlineMaterial=/FortUILibrary/UIKit/Buttons/MaterialInstances/TileOutline/MI_UIKit_Outline
    shouldTextImpactLayout=false
    text=""
    useTextBlock=false
    buttonPadding=<ltrb 8 0 8 0>
    promptNudge=<-50 0>
   }
   slot {class=HorizontalBoxSlot}
  }
  Spacer_102 {
   widgetClassPath=&Spacer
   parent=HorizontalBox_NavigationArrows
   widget=Spacer_102
   properties {size=<32 1>}
   slot {class=HorizontalBoxSlot}
  }
  Button_ScrollRight {
   widgetClassPath=&WBP_UIKit_Button_Generic
   bIsVariable=true
   parent=HorizontalBox_NavigationArrows
   widget=Button_ScrollRight
   properties {
    bIsFocusable=false
    holdData=null
    iconMaterial=/ItemShop/MultiProduct/Materials/Buttons/MI_UI_Icon_Button_Next
    minHeight=48
    mobileMinHeight=80
    outlineMaterial=&MI_UIKit_Outline
    promptHorizontalAlignment=HAlign_Right
    shouldTextImpactLayout=false
    text=""
    useTextBlock=false
    buttonPadding=<ltrb 8 0 8 0>
    promptNudge=<50 0>
   }
   slot {class=HorizontalBoxSlot}
  }
  Spacer_BetweenRows {
   widgetClassPath=&Spacer
   bIsVariable=true
   parent=Overlay_BottomContent
   widget=Spacer_BetweenRows
   properties {size=<0 24>}
   slot {class=OverlaySlot, properties {horizontalAlignment=HAlign_Fill}}
  }
  CommonHardwareVisibilityBorder_0 {
   widgetClassPath=&CommonHardwareVisibilityBorder
   parent=Overlay_2
   widget=CommonHardwareVisibilityBorder_0
   properties {visibilityQuery {
     autoDescription=" ALL( Input.MouseAndKeyboard )"
     tokenStreamVersion=0
     userDescription=""
     queryTokenStream [0,1,2,1,0]
     tagDictionary [{tagName=Input.MouseAndKeyboard}]
    }}
   slot {class=OverlaySlot, properties {horizontalAlignment=HAlign_Fill, verticalAlignment=VAlign_Fill}}
  }
  Button_RowHitbox {
   widgetClassPath=&Button
   parent=CommonHardwareVisibilityBorder_0
   widget=Button_RowHitbox
   properties {
    isFocusable=false
    renderOpacity=0
    touchMethod=PreciseTap
    visibility=SelfHitTestInvisible
    widgetStyle {
     clickedSlateSound {resourceObject=null}
     disabled {
      bIsDynamicallyLoaded=false
      drawAs=NoDrawType
      imageType=NoImage
      mirroring=NoMirror
      resourceName=None
      resourceObject=null
      tiling=NoTile
      imageSize=<0 0>
      margin=<ltrb 0 0 0 0>
      outlineSettings {
       bUseBrushTransparency=false
       roundingType=HalfHeightRadius
       width=0
       color {colorUseRule=UseColor_Specified, specifiedColor=<rgba 0 0 0 0>}
       cornerRadii=<0 0 0 0>
      }
      tintColor {colorUseRule=UseColor_Specified, specifiedColor=<rgba 1 1 1 1>}
      uVRegion {bIsValid=false, max=<0 0>, min=<0 0>}
     }
     disabledForeground {colorUseRule=UseColor_Specified, specifiedColor=<rgba 0.5271 0.5271 0.5271 1>}
     hovered {
      bIsDynamicallyLoaded=false
      drawAs=RoundedBox
      imageType=NoImage
      mirroring=NoMirror
      resourceName=None
      resourceObject=null
      tiling=NoTile
      imageSize=<32 32>
      margin=<ltrb 0 0 0 0>
      outlineSettings {
       bUseBrushTransparency=true
       roundingType=FixedRadius
       width=1
       color {colorUseRule=UseColor_Specified, specifiedColor=<rgba 0.7243 0.7243 0.7243 1>}
       cornerRadii=<4 4 4 4>
      }
      tintColor {colorUseRule=UseColor_Specified, specifiedColor=<rgba 0.7243 0.7243 0.7243 1>}
      uVRegion {bIsValid=false, max=<0 0>, min=<0 0>}
     }
     hoveredForeground {colorUseRule=UseColor_Specified, specifiedColor=<rgba 0 1 1 1>}
     hoveredSlateSound {resourceObject=null}
     normal {
      bIsDynamicallyLoaded=false
      drawAs=RoundedBox
      imageType=NoImage
      mirroring=NoMirror
      resourceName=None
      resourceObject=null
      tiling=NoTile
      imageSize=<32 32>
      margin=<ltrb 0 0 0 0>
      outlineSettings {
       bUseBrushTransparency=true
       roundingType=FixedRadius
       width=1
       color {colorUseRule=UseColor_Specified, specifiedColor=<rgba 0.6951 0.6951 0.6951 1>}
       cornerRadii=<4 4 4 4>
      }
      tintColor {colorUseRule=UseColor_Specified, specifiedColor=<rgba 0.4955 0 0 1>}
      uVRegion {bIsValid=false, max=<0 0>, min=<0 0>}
     }
     normalForeground {colorUseRule=UseColor_Specified, specifiedColor=<rgba 0 1 1 1>}
     normalPadding=<ltrb 12 1.5 12 1.5>
     pressed {
      bIsDynamicallyLoaded=false
      drawAs=RoundedBox
      imageType=NoImage
      mirroring=NoMirror
      resourceName=None
      resourceObject=null
      tiling=NoTile
      imageSize=<32 32>
      margin=<ltrb 0 0 0 0>
      outlineSettings {
       bUseBrushTransparency=true
       roundingType=FixedRadius
       width=1
       color {colorUseRule=UseColor_Specified, specifiedColor=<rgba 0.7243 0.7243 0.7243 1>}
       cornerRadii=<4 4 4 4>
      }
      tintColor {colorUseRule=UseColor_Specified, specifiedColor=<rgba 0.3843 0.3843 0.3843 1>}
      uVRegion {bIsValid=false, max=<0 0>, min=<0 0>}
     }
     pressedForeground {colorUseRule=UseColor_Specified, specifiedColor=<rgba 0 1 1 1>}
     pressedPadding=<ltrb 12 1.5 12 1.5>
     pressedSlateSound {resourceObject=null}
    }
   }
   slot {class=BorderSlot, properties {padding=<ltrb 0 0 0 0>}}
  }
 ]
}
