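# Item shop row wrapper widget that shows either a tile grid or billboard row based on the offer group's display type.
# InitializeRow switches named slots, reusing the child if its class matches, otherwise creating one, then binds its events.
# Row interface calls forward to the active child, whose focus, list-view and navigation events are re-broadcast.
# SetRowState gives the inactive row a fixed state and binds navigation events only on the active row in one specific state.
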WidgetBlueprint /ItemShop/MultiProduct/WBP_MPItemShop_RowWrapper
graphs [
 BillboardInitialization {nodes [
   n1 = BillboardInitialization!() -> (then, RowWidget:Object Reference) {is=title K2Node_FunctionEntry}
   > n2 = Cast To WBP_Row_Billboard!(execute, Object=@n1) -> (then, CastFailed, AsWBP Row Billboard:WBP Row Billboard Object Reference) {is=title K2Node_DynamicCast}
   > Set RowRef(RowRef:WBP Row Billboard Object Reference=@n2) -> (Output_Get:WBP Row Billboard Object Reference)
   n3 = Create Event()
   n4 = Get RowRef() -> (RowRef:WBP Row Billboard Object Reference)
   > Unbind Event from On Row Focused(self:WBP Row Billboard Object Reference=@n4, Delegate=@n3) {is=title K2Node_RemoveDelegate}
   > Bind Event to On Row Focused(self:WBP Row Billboard Object Reference=@n4, Delegate=@n3) {is=title K2Node_AddDelegate}
   n5 = Create Event()
   n6 = Get RowRef()
   > Unbind Event from On Offer Focused(self:WBP Row Billboard Object Reference=@n6, Delegate=@n5) {is=title K2Node_RemoveDelegate}
   > Bind Event to On Offer Focused(self:WBP Row Billboard Object Reference=@n6, Delegate=@n5) {is=title K2Node_AddDelegate}
  ]}
 CreateOrGetWidgetInstance {nodes [
   CreateOrGetWidgetInstance!() -> (then, NamedSlot:Panel Widget Object Reference, RowClass:User Widget Class Reference) {is=title K2Node_FunctionEntry}
   > n1 = Set ChildInstance(ChildInstance:Widget Object Reference=GetChildAt!(self:Panel Widget Object Reference=Get NamedSlot() -> (NamedSlot:Panel Widget Object Reference), Index:int=0) -> (ReturnValue:Widget Object Reference)) -> (Output_Get:Widget Object Reference)
   > n2 = Is Valid(InputObject=@n1)
   n3 = Get ChildInstance() -> (ChildInstance:Widget Object Reference)
   > n4 = Branch(Condition="Equal (Class)"(A=Get RowClass() -> (RowClass:User Widget Class Reference), B=Get Class(Object=@n3) -> (ReturnValue:Widget Class Reference)))
   > n5 = Set WidgetInstance(WidgetInstance:Widget Object Reference=Get ChildInstance()) -> (Output_Get:Widget Object Reference)
   Return Node!(execute=[@n7.then, @n5.then], RowInstance:Widget Object Reference=Get WidgetInstance() -> (WidgetInstance:Widget Object Reference))
   n6 = Create Widget!(
    execute:Exec=[@n8.then, @n4.else]
    Class:User Widget Class Reference=Get RowClass()
    OwningPlayer:Player Controller Object Reference=GetOwningPlayer()
   ) -> (then:Exec, ReturnValue:User Widget Object Reference)
   > Set WidgetInstance(WidgetInstance=@n6)
   > n7 = AddChild(self=Get NamedSlot(), Content=Get WidgetInstance())
   n8 = ClearChildren(execute=@n2.Is Not Valid, self=Get NamedSlot())
  ]}
 EventGraph {nodes [
   OnRowWidgetFocused!() -> (OutputDelegate, then, RowWidget:User Widget Object Reference) {is=title K2Node_CustomEvent}
   > Call On Row Focused!(execute:Exec, self:WBP MPItem Shop Row Wrapper Object Reference, RowWidget:User Widget Object Reference=Self-Reference()) -> (then:Exec) {is=title K2Node_CallDelegate}
   n1 = OnRowWidgetListViewRequested!() -> (OutputDelegate, then, OfferGroupVM:Fort Item Shop Offer Group VM Object Reference) {is=title K2Node_CustomEvent}
   > Call On List View Requested!(
    execute:Exec
    self:WBP MPItem Shop Row Wrapper Object Reference
    OfferGroupVM:Fort Item Shop Offer Group VM Object Reference=@n1.OfferGroupVM
   ) -> (then:Exec) {is=title K2Node_CallDelegate}
   n2 = OnWidgetOfferFocused!() -> (OutputDelegate, then, OfferId:String) {is=title K2Node_CustomEvent}
   > Call On Offer Focused!(execute:Exec, self:WBP MPItem Shop Row Wrapper Object Reference, OfferID:String=@n2.OfferId) -> (then:Exec) {is=title K2Node_CallDelegate}
   n3 = OnRowWidgetNavigatedFrom!() -> (OutputDelegate, then, Direction:EUINavigation Enum, FocusedTilePosition:Vector 2D Structure) {is=title K2Node_CustomEvent}
   > Call On Row Navigated From!(
    execute:Exec
    self:WBP MPItem Shop Row Wrapper Object Reference
    Direction:EUINavigation Enum=@n3.Direction
    FocusedTilePosition:Vector 2D Structure=@n3.FocusedTilePosition
   ) -> (then:Exec) {is=title K2Node_CallDelegate}
   n4 = Event I_InitializeRow!() -> (OutputDelegate, then, FortItemShopRow:Fort Item Shop Row Structure, ContentHasFocus:bool) {is=title K2Node_Event}
   > InitializeRow!(execute, self, FortItemShopRow:Fort Item Shop Row Structure=@n4.FortItemShopRow, ParentHasFocus:bool=@n4.ContentHasFocus) -> (then)
   Event I_ResetRow() {is=title K2Node_Event}
   > I_ResetRow!(execute:Exec, self:Object Reference=Get ActiveRowWidget() -> (ActiveRowWidget:Widget Object Reference)) -> (then:Exec) {is=title K2Node_Message}
   n5 = Event I_SetFocusedOfferID!() -> (OutputDelegate, then, OfferID:String) {is=title K2Node_Event}
   > I_SetFocusedOfferID!(execute:Exec, self:Object Reference=Get ActiveRowWidget(), OfferID:String=@n5.OfferID) -> (then:Exec) {is=title K2Node_Message}
   n6 = Event I_SetParentHasFocusState!() -> (OutputDelegate, then, HasFocus:bool) {is=title K2Node_Event}
   > I_SetParentHasFocusState!(execute:Exec, self:Object Reference=Get ActiveRowWidget(), HasFocus:bool=@n6.HasFocus) -> (then:Exec) {is=title K2Node_Message}
   n7 = Event I_OnScrollComplete!() -> (OutputDelegate, then, DestinationIndex:int=0) {is=title K2Node_Event}
   > I_OnScrollComplete!(execute:Exec, self:Object Reference=Get ActiveRowWidget(), DestinationIndex:int=@n7.DestinationIndex) -> (then:Exec) {is=title K2Node_Message}
   n8 = Event I_SetOfferGroupVM!() -> (OutputDelegate, then, OfferGroupVM:Fort Item Shop Offer Group VM Object Reference) {is=title K2Node_Event}
   > I_SetOfferGroupVM!(execute:Exec, self:Object Reference=Get ActiveRowWidget(), OfferGroupVM:Fort Item Shop Offer Group VM Object Reference=@n8.OfferGroupVM) -> (then:Exec) {is=title K2Node_Message}
   n9 = Event I_SetRowState!() -> (OutputDelegate, then, InRowState:E_ItemShop_RowState Enum=NewEnumerator4) {is=title K2Node_Event}
   > SetRowState!(execute, self, InRowState:E_ItemShop_RowState Enum=@n9.InRowState) -> (then)
  ]}
 HandleRowInitialization {nodes [
   n1 = HandleRowInitialization!() -> (then, NamedSlot:Panel Widget Object Reference, RowClass:User Widget Class Reference) {is=title K2Node_FunctionEntry}
   > n2 = CreateOrGetWidgetInstance!(execute, self, NamedSlot:Panel Widget Object Reference=@n1, RowClass:User Widget Class Reference=@n1.RowClass) -> (then, RowInstance:Widget Object Reference) {is=title K2Node_CallFunction}
   > Set RowInstance(RowInstance:Widget Object Reference=@n2) -> (Output_Get:Widget Object Reference)
   > SetActiveWidget(
    self:Widget Switcher Object Reference=Get Switcher_ViewType() -> (Switcher_ViewType:Common Activatable Widget Switcher Object Reference)
    Widget=Get NamedSlot()
   )
   > I_InitializeRow!(
    execute:Exec
    self:Object Reference=Get RowInstance() -> (RowInstance:Widget Object Reference)
    FortItemShopRow:Fort Item Shop Row Structure=Get FortItemShopRow() -> (FortItemShopRow:Fort Item Shop Row Structure)
    ContentHasFocus:bool=Get IsContentFocused() -> (IsContentFocused:bool)
   ) -> (then:Exec) {is=title K2Node_Message}
   > n3 = Set ActiveRowWidget(ActiveRowWidget:Widget Object Reference=Get RowInstance()) -> (Output_Get:Widget Object Reference)
   > Return Node!(execute, RowWidget:Widget Object Reference=@n3)
  ]}
 I_GetClosestFocusTarget {nodes [
   n1 = I_GetClosestFocusTarget!() -> (then, TargetLocation:Vector 2D Structure) {is=title K2Node_FunctionEntry}
   > n2 = I_GetClosestFocusTarget!(execute:Exec, self:Object Reference=Get ActiveRowWidget(), TargetLocation:Vector 2D Structure=@n1) -> (then:Exec, Target:Widget Object Reference) {is=title K2Node_Message}
   > Return Node!(execute, Target:Widget Object Reference=@n2)
  ]}
 I_GetDesiredFocusTarget {nodes [
   I_GetDesiredFocusTarget() {is=title K2Node_FunctionEntry}
   > n1 = I_GetDesiredFocusTarget!(execute:Exec, self:Object Reference=Get ActiveRowWidget()) -> (then:Exec, TileWidget:Widget Object Reference) {is=title K2Node_Message}
   > Return Node!(execute, TileWidget:Widget Object Reference=@n1)
  ]}
 I_GetDesiredScrollDestination {nodes [
   I_GetDesiredScrollDestination() {is=title K2Node_FunctionEntry}
   > n1 = I_GetDesiredScrollDestination!(execute:Exec, self:Object Reference=Get ActiveRowWidget()) -> (then:Exec, Destination:EDescendantScrollDestination Enum=IntoView) {is=title K2Node_Message}
   > Return Node!(execute, Destination:EDescendantScrollDestination Enum=@n1)
  ]}
 I_GetDesiredScrollPadding {nodes [
   I_GetDesiredScrollPadding() {is=title K2Node_FunctionEntry}
   > n1 = I_GetDesiredScrollPadding!(execute:Exec, self:Object Reference=Get ActiveRowWidget()) -> (then:Exec, ScrollPadding:double) {is=title K2Node_Message}
   > Return Node!(execute, ScrollPadding:double=@n1)
  ]}
 I_GetFocusTargetFromOfferID {nodes [
   n1 = I_GetFocusTargetFromOfferID!() -> (then, OfferID:String) {is=title K2Node_FunctionEntry}
   > n2 = I_GetFocusTargetFromOfferID!(execute:Exec, self:Object Reference=Get ActiveRowWidget(), OfferID:String=@n1) -> (then:Exec, FocusTarget:Widget Object Reference) {is=title K2Node_Message}
   > Return Node!(execute, FocusTarget:Widget Object Reference=@n2)
  ]}
 I_GetScrollTarget {nodes [
   I_GetScrollTarget() {is=title K2Node_FunctionEntry}
   > n1 = I_GetScrollTarget!(execute:Exec, self:Object Reference=Get ActiveRowWidget()) -> (then:Exec, ScrollTarget:Widget Object Reference) {is=title K2Node_Message}
   > Return Node!(execute, ScrollTarget:Widget Object Reference=@n1)
  ]}
 InitializeRow {nodes [
   n1 = InitializeRow!() -> (then, FortItemShopRow:Fort Item Shop Row Structure, ParentHasFocus:bool=~) {is=title K2Node_FunctionEntry}
   > Set FortItemShopRow(FortItemShopRow:Fort Item Shop Row Structure=@n1) -> (Output_Get:Fort Item Shop Row Structure)
   > Set IsContentFocused(IsContentFocused:bool=@n1.ParentHasFocus) -> (Output_Get:bool)
   > n2 = Switch on EFortItemShopOfferDisplayType!(
    execute
    Selection:EFortItemShopOfferDisplayType Enum=GetDisplayType!(self:Fort Item Shop Offer Group VM Object Reference=Get!() -> (
      FortItemShopRow_OfferGroup:Fort Item Shop Offer Group VM Object Reference
      FortItemShopRow_Offers:Array of Fort Item Shop Offer VM Object References
      FortItemShopRow_bIsFirstRowInOfferGroup:bool=~
      FortItemShopRow_bIsLastRowInOfferGroup:bool=~
     )) -> (ReturnValue:EFortItemShopOfferDisplayType Enum=TileGrid)
   ) -> (TileGrid:Exec, ExpandableList:Exec, Billboard:Exec, ScrollableList:Exec) {is=title K2Node_SwitchEnum}
   n3 = HandleRowInitialization!(
    execute=[@n2.ExpandableList, @n2.ScrollableList, @n2.TileGrid]
    self
    NamedSlot:Panel Widget Object Reference=Get NamedSlot_ViewType_TileGrid() -> (NamedSlot_ViewType_TileGrid:Named Slot Object Reference)
    RowClass:User Widget Class Reference=Get TileGridWidgetClass() -> (TileGridWidgetClass:User Widget Class Reference)
   ) -> (then, RowWidget:Widget Object Reference) {is=title K2Node_CallFunction}
   > TileGridInitialization!(execute, self, RowWidget:Object Reference=@n3) -> (then)
   n4 = HandleRowInitialization(
    execute=@n2.Billboard
    NamedSlot=Get NamedSlot_ViewType_Billboard() -> (NamedSlot_ViewType_Billboard:Named Slot Object Reference)
    RowClass=Get BillboardWidgetClass() -> (BillboardWidgetClass:User Widget Class Reference)
   )
   > BillboardInitialization!(execute, self, RowWidget:Object Reference=@n4) -> (then) {is=title K2Node_CallFunction}
  ]}
 OnListViewRequested {nodes [OnListViewRequested!() -> (then, OfferGroupVM:Fort Item Shop Offer Group VM Object Reference) {is=title K2Node_FunctionEntry}]}
 OnOfferFocused {nodes [OnOfferFocused!() -> (then, OfferID:String) {is=title K2Node_FunctionEntry}]}
 OnRowFocused {nodes [OnRowFocused!() -> (then, RowWidget:User Widget Object Reference) {is=title K2Node_FunctionEntry}]}
 OnRowNavigatedFrom {nodes [OnRowNavigatedFrom!() -> (then, Direction:EUINavigation Enum, FocusedTilePosition:Vector 2D Structure) {is=title K2Node_FunctionEntry}]}
 SetRowState {nodes [
   SetRowState!() -> (then, InRowState:E_ItemShop_RowState Enum=NewEnumerator0) {is=title K2Node_FunctionEntry}
   > n1 = For Each Loop(Array:Array of Named Slot Object References=Make Array!("[0]":Named Slot Object Reference=Get NamedSlot_ViewType_Billboard(), "[1]":Named Slot Object Reference=Get NamedSlot_ViewType_TileGrid()) -> (Array:Array of Named Slot Object References)) -> (Array Element:Named Slot Object Reference)
   > n2 = Set LoopRow(LoopRow:Widget Object Reference=GetContent!(self:Content Widget Object Reference=@n1) -> (ReturnValue:Widget Object Reference)) -> (Output_Get:Widget Object Reference)
   > n3 = Cast To I_MPItemShop_Row!(execute, Object=@n2) -> (then, CastFailed, AsI MPItem Shop Row:I MPItem Shop Row Interface) {is=title K2Node_DynamicCast}
   n4 = Get ActiveRowWidget()
   n5 = Get LoopRow() -> (LoopRow:Widget Object Reference)
   > I_SetRowState!(
    execute
    self:I MPItem Shop Row Interface=@n3
    InRowState:E_ItemShop_RowState Enum=Select(
     Option 0:E_ItemShop_RowState Enum=NewEnumerator3
     Option 1:E_ItemShop_RowState Enum=Get InRowState() -> (InRowState:E_ItemShop_RowState Enum=NewEnumerator4)
     Index="Equal (Object)"(A=@n5, B=@n4)
    ) -> (ReturnValue:E_ItemShop_RowState Enum)
   ) -> (then)
   > n6 = Cast To WBP_MPItemShop_Row!(execute, Object=Get LoopRow()) -> (then, CastFailed, AsWBP MPItem Shop Row:WBP MPItem Shop Row Object Reference) {is=title K2Node_DynamicCast}
   > n7 = Set TileGrid(TileGrid:WBP MPItem Shop Row Object Reference=@n6) -> (Output_Get:WBP MPItem Shop Row Object Reference)
   n8 = Create Event()
   > Unbind Event from On Row Navigated From(self:WBP MPItem Shop Row Object Reference=@n7, Delegate=@n8) {is=title K2Node_RemoveDelegate}
   n9 = AND Boolean(
    A="Equal (Object)"(A=Get LoopRow(), B=Get ActiveRowWidget())
    B="Equal (Enum)"(A:E_ItemShop_RowState Enum=Get InRowState() -> (InRowState=NewEnumerator4), B:E_ItemShop_RowState Enum=NewEnumerator0)
   )
   > Branch(Condition=@n9)
   > Bind Event to On Row Navigated From(self:WBP MPItem Shop Row Object Reference=Get TileGrid() -> (TileGrid:WBP MPItem Shop Row Object Reference), Delegate=@n8) {is=title K2Node_AddDelegate}
   n10 = Cast To WBP_Row_Billboard(execute=@n6.CastFailed, Object=Get LoopRow())
   > n11 = Set Billboard(Billboard:WBP Row Billboard Object Reference=@n10) -> (Output_Get:WBP Row Billboard Object Reference)
   n12 = Create Event()
   > Unbind Event from On Row Navigated From(self:WBP Row Billboard Object Reference=@n11, Delegate=@n12)
   > Branch(Condition=@n9)
   > Bind Event to On Row Navigated From(self:WBP Row Billboard Object Reference=Get Billboard() -> (Billboard:WBP Row Billboard Object Reference), Delegate=@n12)
  ]}
 TileGridInitialization {nodes [
   n1 = TileGridInitialization!() -> (then, RowWidget:Object Reference) {is=title K2Node_FunctionEntry}
   > n2 = Cast To WBP_MPItemShop_Row(Object=@n1)
   > Set RowRef(RowRef:WBP MPItem Shop Row Object Reference=@n2) -> (Output_Get:WBP MPItem Shop Row Object Reference)
   n3 = Create Event()
   n4 = Get RowRef() -> (RowRef:WBP MPItem Shop Row Object Reference)
   > Unbind Event from On Row Focused(self:WBP MPItem Shop Row Object Reference=@n4, Delegate=@n3)
   > Bind Event to On Row Focused(self:WBP MPItem Shop Row Object Reference=@n4, Delegate=@n3)
   n5 = Create Event()
   n6 = Get RowRef()
   > Unbind Event from On Offer Hovered(self:WBP MPItem Shop Row Object Reference=@n6, Delegate=@n5) {is=title K2Node_RemoveDelegate}
   > Bind Event to On Offer Hovered(self:WBP MPItem Shop Row Object Reference=@n6, Delegate=@n5) {is=title K2Node_AddDelegate}
   n7 = Create Event()
   n8 = Get RowRef()
   > Unbind Event from On Row List View Requested(self:WBP MPItem Shop Row Object Reference=@n8, Delegate=@n7) {is=title K2Node_RemoveDelegate}
   > Bind Event to On Row List View Requested(self:WBP MPItem Shop Row Object Reference=@n8, Delegate=@n7) {is=title K2Node_AddDelegate}
  ]}
]
variables [
 ActiveRowWidget {type=Widget, default=null}
 BillboardWidgetClass {type=Class, default=/ItemShop/MultiProduct/Billboard/WBP_Row_Billboard.WBP_Row_Billboard_C}
 FortItemShopRow {type=FortItemShopRow, default {bIsFirstRowInOfferGroup=true, bIsLastRowInOfferGroup=true, offerGroup=null, offers []}}
 IsContentFocused {type=boolean}
 OnListViewRequested {}
 OnOfferFocused {}
 OnRowFocused {}
 OnRowNavigatedFrom {}
 TileGridWidgetClass {type=Class, default=&/WBP_MPItemShop_Row.WBP_MPItemShop_Row_C}
]
widgets {
 info {inheritedWidgetCount=0, parentClass=&UserWidget, rootWidgetClass=&Overlay, widgetCount=4}
 widgets [
  Overlay_30 {widgetClassPath=&Overlay, widget=Overlay_30}
  Switcher_ViewType {
   widgetClassPath=&CommonActivatableWidgetSwitcher
   bIsVariable=true
   parent=Overlay_30
   widget=Switcher_ViewType
   properties {transitionDuration=0}
   slot {class=OverlaySlot, properties {horizontalAlignment=HAlign_Fill, verticalAlignment=VAlign_Fill}}
  }
  NamedSlot_ViewType_TileGrid {
   widgetClassPath=&NamedSlot
   bIsVariable=true
   parent=Switcher_ViewType
   widget=NamedSlot_ViewType_TileGrid
   slot {class=WidgetSwitcherSlot}
  }
  NamedSlot_ViewType_Billboard {
   widgetClassPath=&NamedSlot
   bIsVariable=true
   parent=Switcher_ViewType
   widget=NamedSlot_ViewType_Billboard
   slot {class=WidgetSwitcherSlot}
  }
 ]
}
