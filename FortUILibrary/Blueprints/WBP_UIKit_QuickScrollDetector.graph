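# CommonUI activatable widget that detects quick scrolling from the rate of scroll-start calls.
# OnScrollStarted maintains a fixed-length sliding window of real-time timestamps, flagging quick scrolling when its span is under DetectionTimeRange.
# A delta gate broadcasts QuickScrollStateChanged only on changes, and OnScrollFinished resets the state to false.
# An optional debug panel, shown by default only at design time, displays settings, color-coded state and timestamps.
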WidgetBlueprint /FortUILibrary/Blueprints/WBP_UIKit_QuickScrollDetector
graphs [
 EventGraph {nodes [
   n1 = OnQuickScrollStateChanged!() -> (OutputDelegate, then, IsQuickScrolling:bool=~) {is=title K2Node_CustomEvent}
   > n2 = Delta Gate Wildcard!(execute:Exec, Variable:"Boolean (by ref)"=Get IsQuickScrolling() -> (IsQuickScrolling:bool), New Value:bool=@n1.IsQuickScrolling) -> (Success:Exec, Updated Variable:"Boolean (by ref)", Previous Variable:bool=~) {is=title K2Node_MacroInstance}
   > Call Quick Scroll State Changed!(execute:Exec, self:WBP UIKit Quick Scroll Detector Object Reference, IsQuickScrolling:bool=@n2) -> (then:Exec) {is=title K2Node_CallDelegate}
   OnScrollFinished() {is=title K2Node_CustomEvent}
   > OnQuickScrollStateChanged!(execute, self, IsQuickScrolling:bool=false) -> (then) {is=title K2Node_CallFunction}
   > UpdateDebugUI!(execute, self, bIsQuickScrolling:bool=false) -> (then)
   n3 = ShowDebug!() -> (OutputDelegate, then, ShouldShow:bool=~) {is=title K2Node_CustomEvent}
   > Set Visibility(
    Visibility=Select(Option 0=Collapsed, Option 1=SelfHitTestInvisible, Index=@n3.ShouldShow)
    self=Get Border_BG() -> (Border_BG:Border Object Reference)
   )
   OnScrollStarted() {is=title K2Node_CustomEvent}
   > Remove Index!(
    execute
    TargetArray:"Array of Float (double-precision)s"=Get TimeStamps() -> (TimeStamps:"Array of Float (double-precision)s")
    IndexToRemove:int=0
   ) -> (then)
   n4 = GetAccurateRealTime!() -> (Seconds:int=0, PartialSeconds:double)
   > Add(
    TargetArray:"Array of Float (double-precision)s"=Get TimeStamps()
    NewItem:"Float (double-precision) (by ref)"=float + float(A:int=@n4, B=@n4.PartialSeconds)
   ) {is=title K2Node_CallArrayFunction}
   n5 = int - int(A=Length(TargetArray:"Array of Float (double-precision)s"=Get TimeStamps()))
   n6 = Get TimeStamps()
   n7 = "float < float"(
    A=float - float(A="Get (a copy)"(Array:"Array of Float (double-precision)s"=@n6, Dimension 1=@n5) -> (Output:double), B="Get (a copy)"(Array=@n6))
    B=Get DetectionTimeRange() -> (DetectionTimeRange:float)
   )
   > OnQuickScrollStateChanged(IsQuickScrolling=@n7)
   > UpdateDebugUI(bIsQuickScrolling=@n7)
   n8 = Event PreConstruct()
   > ShowDebug!(execute, self, ShouldShow:bool=@n8.IsDesignTime) -> (then) {is=title K2Node_CallFunction}
   Event On Activated()
   > Clear(TargetArray:"Array of Float (double-precision)s"=Get TimeStamps())
   > Resize!(
    execute
    TargetArray:"Array of Float (double-precision)s"=Get TimeStamps()
    Size:int=Get ScrollRequestsPerTimeRangeThreshold() -> (ScrollRequestsPerTimeRangeThreshold:int=0)
   ) -> (then) {is=title K2Node_CallArrayFunction}
   > UpdateDebugUI(bIsQuickScrolling=false)
   > InitializeDebugHeader()
  ]}
 InitializeDebugHeader {nodes [
   InitializeDebugHeader() {is=title K2Node_FunctionEntry}
   n1 = "To String (Integer)"(InInt=Get ScrollRequestsPerTimeRangeThreshold() -> (ScrollRequestsPerTimeRangeThreshold=0))
   n2 = "To String (Float)"(InDouble=Get DetectionTimeRange())
   > Set Text(
    Text="To Text (String)"(InString=Append!(A="quick scroll settings: ", B=@n1, C:String=" scrolls over ", D:String=@n2, E:String=" seconds") -> (ReturnValue))
    self=Get Text_Status_Header() -> (Text_Status_Header:Common Text Object Reference)
   )
  ]}
 QuickScrollStateChanged {nodes [QuickScrollStateChanged!() -> (then, IsQuickScrolling:bool=~) {is=title K2Node_FunctionEntry}]}
 ShiftArray {nodes [
   ShiftArray() {is=title K2Node_FunctionEntry}
   n1 = Get TimeStamps()
   > n2 = Reverse for Each Loop!(Exec:Exec, Array:"Array of Float (double-precision)s"=@n1) -> (LoopBody:Exec, ArrayIndex:int, ArrayElement:double=~, Completed:Exec) {is=title K2Node_MacroInstance}
   > Set Array Elem!(
    execute
    TargetArray:"Array of Float (double-precision)s"=Get TimeStamps()
    Index:int=int + int(A=@n2)
    Item:"Float (double-precision) (by ref)"=@n2.ArrayElement
    bSizeToFit:bool=false
   ) -> (then)
   Remove Index(TargetArray=@n1, IndexToRemove=0)
  ]}
 UpdateDebugText {nodes [
   UpdateDebugText() {is=title K2Node_FunctionEntry}
   > n1 = For Each Loop(Array:"Array of Float (double-precision)s"=Get TimeStamps()) -> (Array Element:double)
   Set Text(
    execute=@n1.Completed
    Text="To Text (String)"(InString=Get PartialString() -> (PartialString:String))
    self=Get Text_timestamps() -> (Text_timestamps:Common Text Object Reference)
   )
   Set PartialString(execute=@n1.LoopBody, PartialString:String=Append!(A=Get PartialString(), B=" | ", C="To String (Float)"(InDouble=@n1)) -> (ReturnValue)) -> (Output_Get:String)
  ]}
 UpdateDebugUI {nodes [
   UpdateDebugUI!() -> (then, bIsQuickScrolling:bool=~) {is=title K2Node_FunctionEntry}
   n1 = Get Border_BG()
   > n2 = Branch(Condition="Equal (Enum)"(
     A:ESlateVisibility Enum=Get Visibility!(self:Widget Object Reference=@n1) -> (Visibility:ESlateVisibility Enum=Visible)
     B:ESlateVisibility Enum=Collapsed
    ))
   UpdateDebugText(execute=@n2.else) {is=title K2Node_CallFunction}
   > SetBrushColor(
    self=Get Border_BG()
    InBrushColor=SelectColor(A=(R=1,G=0,B=0,A=1), B=(R=0.0311,G=1,B=0,A=1), bPickA=Get bIsQuickScrolling() -> (bIsQuickScrolling:bool))
   )
   > Set Text(
    Text=SelectText(A="INVTEXT(\"ON\")", B="INVTEXT(\"off\")", bPickA=Get bIsQuickScrolling())
    self=Get Text_Status() -> (Text_Status:Common Text Object Reference)
   )
  ]}
]
variables [
 DetectionTimeRange {type=number, default=1}
 IsQuickScrolling {type=boolean}
 QuickScrollStateChanged {}
 ScrollRequestsPerTimeRangeThreshold {type=integer, default=5}
 TimeStamps {type="array<number>", default []}
]
widgets {
 info {inheritedWidgetCount=0, parentClass=&CommonActivatableWidget, rootWidgetClass=&Border, widgetCount=6}
 widgets [
  Border_BG {widgetClassPath=&Border, bIsVariable=true, widget=Border_BG, properties=~}
  VerticalBox_0 {widgetClassPath=&VerticalBox, parent=Border_BG, widget=VerticalBox_0, slot {class=BorderSlot}}
  Text_Status_Header {
   widgetClassPath=&CommonTextBlock
   bIsVariable=true
   parent=VerticalBox_0
   widget=Text_Status_Header
   properties {
    text="quick scroll: "
    colorAndOpacity {colorUseRule=UseColor_Specified, specifiedColor=<rgba 0 0 0 1>}
    font {
     bForceMonospaced=false
     bMaterialIsStencil=false
     fontMaterial=null
     fontObject=&NotoSans
     letterSpacing=0
     monospacedWidth=1
     size=9
     skewAmount=0
     typefaceFontName=NotoSans-Regular
     outlineSettings {
      bApplyOutlineToDropShadows=false
      bMiteredCorners=false
      bSeparateFillAlpha=false
      outlineMaterial=null
      outlineSize=0
      outlineColor=<rgba 0 0 0 1>
     }
    }
    shadowOffset=<0 0>
   }
   slot {class=VerticalBoxSlot}
  }
  Text_Status {
   widgetClassPath=&CommonTextBlock
   bIsVariable=true
   parent=VerticalBox_0
   widget=Text_Status
   properties {
    text=waiting for scroll
    colorAndOpacity {colorUseRule=UseColor_Specified, specifiedColor=<rgba 0 0 0 1>}
    font {
     bForceMonospaced=false
     bMaterialIsStencil=false
     fontMaterial=null
     fontObject=&NotoSans
     letterSpacing=0
     monospacedWidth=1
     size=15
     skewAmount=0
     typefaceFontName=NotoSans-Regular
     outlineSettings {
      bApplyOutlineToDropShadows=false
      bMiteredCorners=false
      bSeparateFillAlpha=false
      outlineMaterial=null
      outlineSize=0
      outlineColor=<rgba 0 0 0 1>
     }
    }
    shadowOffset=<0 0>
   }
   slot {class=VerticalBoxSlot}
  }
  Text_Stacks_Header {
   widgetClassPath=&CommonTextBlock
   bIsVariable=true
   parent=VerticalBox_0
   widget=Text_Stacks_Header
   properties {
    text=request timestamps:
    colorAndOpacity {colorUseRule=UseColor_Specified, specifiedColor=<rgba 0 0 0 1>}
    font {
     bForceMonospaced=false
     bMaterialIsStencil=false
     fontMaterial=null
     fontObject=&NotoSans
     letterSpacing=0
     monospacedWidth=1
     size=9
     skewAmount=0
     typefaceFontName=NotoSans-Regular
     outlineSettings {
      bApplyOutlineToDropShadows=false
      bMiteredCorners=false
      bSeparateFillAlpha=false
      outlineMaterial=null
      outlineSize=0
      outlineColor=<rgba 0 0 0 1>
     }
    }
    shadowOffset=<0 0>
   }
   slot {class=VerticalBoxSlot}
  }
  Text_timestamps {
   widgetClassPath=&CommonTextBlock
   bIsVariable=true
   parent=VerticalBox_0
   widget=Text_timestamps
   properties {
    text=Text Block
    colorAndOpacity {colorUseRule=UseColor_Specified, specifiedColor=<rgba 0 0 0 1>}
    font {
     bForceMonospaced=false
     bMaterialIsStencil=false
     fontMaterial=null
     fontObject=&NotoSans
     letterSpacing=0
     monospacedWidth=1
     size=9
     skewAmount=0
     typefaceFontName=NotoSans-Regular
     outlineSettings {
      bApplyOutlineToDropShadows=false
      bMiteredCorners=false
      bSeparateFillAlpha=false
      outlineMaterial=null
      outlineSize=0
      outlineColor=<rgba 0 0 0 1>
     }
    }
    shadowOffset=<0 0>
   }
   slot {class=VerticalBoxSlot, properties {padding=<ltrb 4 2 4 2>}}
  }
 ]
}
