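# Sound cue for a navigator reveal whoosh in the front-end item shop, routed to the Fort_UI sound class.
# It has no root or sound nodes, so it produces no audio as configured.
# It retains a reduced volume multiplier and a concurrency override that caps simultaneous instances, stops the oldest, limits rapid retriggers and scales volume.
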SoundCue /Game/Sounds/Fort_UI_Sounds/FrontEnd/ItemShop/Fort_UI_Shop_Whoosh_NavigatorReveal_Cue
root=null
nodes []
properties {
 bOverrideConcurrency=true
 maxDistance=2097152
 soundClassObject=&Fort_UI
 volumeMultiplier=0.4
 concurrencyOverrides {maxCount=2, resolutionRule=StopOldest, retriggerTime=0.13, volumeScale=0.5}
}
