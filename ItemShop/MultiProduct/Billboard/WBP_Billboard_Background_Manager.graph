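# Item shop widget that manages the background behind Billboard offer groups using two alternating background widgets.
# When the offer group changes, VM_OfferGroupUpdated activates the idle background and deactivates the other, or deactivates the active one for non-Billboard groups.
# Activation-start events bring the active background to the top grid layer.
# A mouse-position ticker, shown only for mouse and keyboard, runs for Billboard groups unless their parallaxEnabled metadata disables it.
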WidgetBlueprint /ItemShop/MultiProduct/Billboard/WBP_Billboard_Background_Manager
graphs [
 CreateBindings {nodes [
   CreateBindings!() -> (then, ActivationStarted:"Delegate (by ref)", ActivationComplete:"Delegate (by ref)", DeactivationComplete:"Delegate (by ref)") {is=title K2Node_FunctionEntry}
   > Unbind Event from On Activation Started(
    self:WBP Billboard Background Object Reference=[
     Get BillboardBackground_1() -> (BillboardBackground_1:WBP Billboard Background Object Reference)
     Get BillboardBackground_2() -> (BillboardBackground_2:WBP Billboard Background Object Reference)
    ]
    Delegate=Get ActivationStarted() -> (ActivationStarted:"Delegate (by ref)")
   ) {is=title K2Node_RemoveDelegate}
   > Bind Event to On Activation Started(self:WBP Billboard Background Object Reference=[Get BillboardBackground_1(), Get BillboardBackground_2()], Delegate=Get ActivationStarted()) {is=title K2Node_AddDelegate}
   > Unbind Event from On Activation Complete(
    self:WBP Billboard Background Object Reference=[Get BillboardBackground_2(), Get BillboardBackground_1()]
    Delegate=Get ActivationComplete() -> (ActivationComplete:"Delegate (by ref)")
   ) {is=title K2Node_RemoveDelegate}
   > Bind Event to On Activation Complete(self:WBP Billboard Background Object Reference=[Get BillboardBackground_1(), Get BillboardBackground_2()], Delegate=Get ActivationComplete()) {is=title K2Node_AddDelegate}
   > Unbind Event from On Deactivation Complete(
    self:WBP Billboard Background Object Reference=[Get BillboardBackground_1(), Get BillboardBackground_2()]
    Delegate=Get DeactivationComplete() -> (DeactivationComplete:"Delegate (by ref)")
   ) {is=title K2Node_RemoveDelegate}
   > Bind Event to On Deactivation Complete(
    self:WBP Billboard Background Object Reference=[Get BillboardBackground_1(), Get BillboardBackground_2()]
    Delegate=Get DeactivationComplete()
   ) {is=title K2Node_AddDelegate}
  ]}
 EventGraph {nodes [
   Activate() {is=title K2Node_CustomEvent}
   n1 = Get RequestedOfferGroup() -> (RequestedOfferGroup:Fort Item Shop Offer Group VM Object Reference)
   n2 = Get SelectedOfferGroup!(self:Fort Item Shop Selection VM Object Reference=Get FortItemShopSelectionVM() -> (FortItemShopSelectionVM:Fort Item Shop Selection VM Object Reference)) -> (SelectedOfferGroup:Fort Item Shop Offer Group VM Object Reference)
   > Branch(Condition=AND Boolean(A="Equal (Object)"(A=@n2, B=@n1), B=IsValid(Object=@n1)))
   > UpdateMouseTickerEnabled!(
    execute
    self:WBP Billboard Mouse Position Ticker Object Reference=Get WBP_Billboard_MousePositionTicker() -> (WBP_Billboard_MousePositionTicker:WBP Billboard Mouse Position Ticker Object Reference)
    bDesiredEnabled:bool=GetBillboardMouseTickerDesiredState!(self) -> (DesiredState:bool)
   ) -> (then)
   OnBackgroundActivationStarted!() -> (OutputDelegate, then, Widget:WBP Billboard Background Object Reference) {is=title K2Node_CustomEvent}
   > ResetGridSlotLayers()
   > SetLayer!(
    execute
    self:Grid Slot Object Reference=SlotAsGridSlot(Widget=Get ActiveBackground() -> (ActiveBackground:WBP Billboard Background Object Reference))
    InLayer:int=1
   ) -> (then)
   Event Construct()
   > CreateBindings!(
    execute
    self
    ActivationStarted:"Delegate (by ref)"=Create Event()
    ActivationComplete:"Delegate (by ref)"=Create Event()
    DeactivationComplete:"Delegate (by ref)"=Create Event()
   ) -> (then) {is=title K2Node_CallFunction}
   OnBackgroundDeactivationComplete!() -> (OutputDelegate, then, Widget:WBP Billboard Background Object Reference) {is=title K2Node_CustomEvent}
   OnBackgroundActivationComplete!() -> (OutputDelegate, then, Widget:WBP Billboard Background Object Reference) {is=title K2Node_CustomEvent}
  ]}
 GetBillboardMouseTickerDesiredState {nodes [
   GetBillboardMouseTickerDesiredState() {is=title K2Node_FunctionEntry}
   n1 = Get RequestedOfferGroup()
   n2 = Find(
    TargetMap:Map of Names to Strings=Get StringMetadata!(self:Fort Item Shop Offer Group VM Object Reference=@n1) -> (StringMetadata:Map of Names to Strings)
    Key:"Name (by ref)"=parallaxEnabled
   )
   n3 = "Equal, Case Insensitive (String)"(A=@n2, B=true)
   n4 = NOT Boolean(A=@n2.ReturnValue)
   n5 = GetDisplayType!(self:Fort Item Shop Offer Group VM Object Reference=@n1) -> (ReturnValue:EFortItemShopOfferDisplayType Enum=TileGrid)
   > Return Node!(
    execute
    DesiredState:bool=AND Boolean(A="Equal (Enum)"(A:EFortItemShopOfferDisplayType Enum=@n5, B:EFortItemShopOfferDisplayType Enum=Billboard), B=OR Boolean(A=@n4, B=@n3))
   )
  ]}
 HandleActivation {nodes [
   n1 = HandleActivation!() -> (then, ActiveBackground:WBP Billboard Background Object Reference) {is=title K2Node_FunctionEntry}
   > Set ActiveBackground(ActiveBackground:WBP Billboard Background Object Reference=@n1) -> (Output_Get:WBP Billboard Background Object Reference)
   > SetOfferGroup!(
    execute
    self:WBP Billboard Background Object Reference=Get ActiveBackground()
    OfferGroupVM:Fort Item Shop Offer Group VM Object Reference=Get RequestedOfferGroup()
   ) -> (then)
   > OnActivated(self:WBP Billboard Background Object Reference=Get ActiveBackground())
  ]}
 HandleDeactivation {nodes [
   n1 = HandleDeactivation!() -> (then, InactiveBackground:WBP Billboard Background Object Reference) {is=title K2Node_FunctionEntry}
   > n2 = Set InactiveBackground(InactiveBackground:WBP Billboard Background Object Reference=@n1) -> (Output_Get:WBP Billboard Background Object Reference)
   > OnDeactivated(self:WBP Billboard Background Object Reference=@n2)
  ]}
 ResetGridSlotLayers {nodes [
   ResetGridSlotLayers() {is=title K2Node_FunctionEntry}
   > SetGridSlotLayer!(execute, Widget:Widget Object Reference=Get BillboardBackground_1(), Layer:int=0) -> (then)
   > SetGridSlotLayer(Widget=Get BillboardBackground_2(), Layer=0)
  ]}
 VM_OfferGroupUpdated {nodes [
   n1 = VM_OfferGroupUpdated!() -> (then, OfferGroupVM:Fort Item Shop Offer Group VM Object Reference) {is=title K2Node_FunctionEntry}
   > Is Valid(InputObject=@n1)
   > n2 = Branch(Condition="Equal (Object)"(A=Get OfferGroupVM() -> (OfferGroupVM:Fort Item Shop Offer Group VM Object Reference), B=Get RequestedOfferGroup()))
   > n3 = Get!(execute:Exec) -> (then:Exec, else:Exec, ActiveBackground:WBP Billboard Background Object Reference)
   > HandleActivation!(execute, self, ActiveBackground:WBP Billboard Background Object Reference=@n3) -> (then) {is=title K2Node_CallFunction}
   n4 = Set RequestedOfferGroup(execute=@n2.else, RequestedOfferGroup:Fort Item Shop Offer Group VM Object Reference=Get OfferGroupVM()) -> (Output_Get:Fort Item Shop Offer Group VM Object Reference)
   > n5 = Branch(Condition="Equal (Enum)"(A=GetDisplayType(self=@n4) -> (ReturnValue=TileGrid), B=Billboard))
   > UpdateMouseTickerEnabled(self=Get WBP_Billboard_MousePositionTicker(), bDesiredEnabled=GetBillboardMouseTickerDesiredState() {is=title K2Node_CallFunction})
   > n6 = Get()
   > n7 = Branch(Condition="Equal (Object)"(A=@n6, B=Get BillboardBackground_1()))
   > HandleActivation(ActiveBackground=Get BillboardBackground_2())
   > HandleDeactivation!(execute, self, InactiveBackground:WBP Billboard Background Object Reference=Get BillboardBackground_1()) -> (then) {is=title K2Node_CallFunction}
   HandleActivation(execute=@n7.else, ActiveBackground=Get BillboardBackground_1())
   > HandleDeactivation(InactiveBackground=Get BillboardBackground_2())
   HandleActivation(execute=@n6.else, ActiveBackground=Get BillboardBackground_1())
   UpdateMouseTickerEnabled(execute=@n5.else, self=Get WBP_Billboard_MousePositionTicker(), bDesiredEnabled=false)
   > n8 = Get()
   > OnDeactivated(self=@n8)
   > Set ActiveBackground()
   Reroute Node(InputPin=@n8.else)
  ]}
]
variables [
 ActiveBackground {type=WBP_Billboard_Background_C, default=null}
 BringToFrontAfterActivation {type=boolean, default=true}
 DeactivationDelayTimer {type=TimerHandle}
 InactiveBackground {type=WBP_Billboard_Background_C, default=null}
 NonBillboardTransitionTimerDelay {type=number, default=0.28}
 RequestedOfferGroup {type=FortItemShopOfferGroupVM, default=null}
]
widgets {
 info {inheritedWidgetCount=0, parentClass=&UserWidget, rootWidgetClass=&GridPanel, widgetCount=5}
 widgets [
  GridPanel_All {widgetClassPath=&GridPanel, widget=GridPanel_All, properties {columnFill [1], rowFill [1]}}
  BillboardBackground_2 {
   widgetClassPath=&/WBP_Billboard_Background
   bIsVariable=true
   parent=GridPanel_All
   widget=BillboardBackground_2
   properties {emptyBrush=null}
   slot {class=GridSlot}
  }
  BillboardBackground_1 {
   widgetClassPath=&WBP_Billboard_Background
   bIsVariable=true
   parent=GridPanel_All
   widget=BillboardBackground_1
   properties {emptyBrush=null}
   slot {class=GridSlot, properties {layer=1}}
  }
  CommonHardwareVisibilityBorder_29 {
   widgetClassPath=&CommonHardwareVisibilityBorder
   parent=GridPanel_All
   widget=CommonHardwareVisibilityBorder_29
   properties {visibilityQuery {
     autoDescription=" ALL( Input.MouseAndKeyboard )"
     tokenStreamVersion=0
     userDescription=""
     queryTokenStream [0,1,2,1,0]
     tagDictionary [{tagName=Input.MouseAndKeyboard}]
    }}
   slot {class=GridSlot}
  }
  WBP_Billboard_MousePositionTicker {
   widgetClassPath=&/WBP_Billboard_MousePositionTicker
   bIsVariable=true
   parent=CommonHardwareVisibilityBorder_29
   widget=WBP_Billboard_MousePositionTicker
   properties {mass=0.6, stiffness=0.7, target Velocity Amount=1, visibility=Collapsed}
   slot {class=BorderSlot, properties {padding=<ltrb 0 0 0 0>}}
  }
 ]
}
