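# Item Shop filter dialog, an activatable widget with Show Only and Item Type multi-select sections driven by the shop's FilterVM.
# Entry clicks toggle options and refresh counts; Apply commits or clears filters, Clear resets selection, Close restores the previous selection.
# Active-filter count drives the Apply badge and label plus the Clear button's state.
# Button padding adapts to gamepad prompts and mirrors for Arabic; compact UI resizes the dialog and buttons.
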WidgetBlueprint /ItemShop/MultiProduct/Filter/WBP_ShopFilter_Dialog
graphs [
 BP_GetDesiredFocusTarget {nodes [
   Get Desired Focus Target()
   > Return Node!(
    execute
    ReturnValue:Widget Object Reference=GetDesiredFocusTarget!(self:WBP UIKit Sort Filter Section Multi Select Object Reference=Get Section_ItemType() -> (Section_ItemType:WBP UIKit Sort Filter Section Multi Select Object Reference)) -> (FocusTarget:Widget Object Reference)
   )
  ]}
 EventGraph {nodes [
   "On Clicked (Button_Apply)"!() -> (OutputDelegate, then, Button:Common Button Base Object Reference) {is=title K2Node_ComponentBoundEvent}
   n1 = Get FilterVM!(self:Fort Item Shop VM Object Reference=Get FortItemShopVM() -> (FortItemShopVM:Fort Item Shop VM Object Reference)) -> (FilterVM:Fort Item Shop Filter VM Object Reference)
   > n2 = Branch(Condition="integer > integer"(A=GetActiveFiltersNum!(self:Fort Item Shop Filter VM Object Reference=@n1) -> (ReturnValue:int=0)))
   > n3 = ApplyFilters(self:Fort Item Shop Filter VM Object Reference=Get FilterVM(self=Get FortItemShopVM()))
   DeactivateWidget(execute=[@n4.then, @n3.then])
   ClearFilters(execute=@n2.else, self:Fort Item Shop Filter VM Object Reference=Get FilterVM(self=Get FortItemShopVM()))
   > n4 = UpdateContentOrdering(self:Fort Item Shop VM Object Reference=Get FortItemShopVM())
   "On Clicked (Button_Clear)"!() -> (OutputDelegate, then, Button:Common Button Base Object Reference) {is=title K2Node_ComponentBoundEvent}
   > ClearFiltersSelection(self:Fort Item Shop Filter VM Object Reference=Get FilterVM(self=Get FortItemShopVM()))
   "On Clicked (Button_Close)"!() -> (OutputDelegate, then, Button:Common Button Base Object Reference) {is=title K2Node_ComponentBoundEvent}
   > RestoreLastFiltersSelection(self:Fort Item Shop Filter VM Object Reference=Get FilterVM(self=Get FortItemShopVM()))
   > DeactivateWidget()
   n5 = "On Entry Clicked (Section_ShowOnly)"!() -> (OutputDelegate, then, SortFilterItemVM:Fort Sort Filter Item VM Object Reference) {is=title K2Node_ComponentBoundEvent}
   > n6 = Cast To FortOptionItemVM!(execute, Object=@n5.SortFilterItemVM) -> (then, CastFailed, AsFort Option Item VM:Fort Option Item VM Object Reference) {is=title K2Node_DynamicCast}
   > ToggleActive(self:Fort Option Item VM Object Reference=@n6)
   > RefreshCounts(self:Fort Item Shop Filter VM Object Reference=Get FilterVM(self=Get FortItemShopVM()))
   n7 = Get FortItemShopVM()
   > SetEntries!(
    execute
    self:WBP UIKit Sort Filter Section Multi Select Object Reference=Get Section_ItemType()
    SortFilterItemVMs:Array of Fort Sort Filter Item VM Object References=Get Filters!(self:Fort Item Shop Filter VM Object Reference=Get FilterVM(self=@n7)) -> (Filters:Array of Fort Item Shop Offer Filter VM Object References)
   ) -> (then)
   n8 = "On Entry Clicked (Section_ItemType)"!() -> (OutputDelegate, then, SortFilterItemVM:Fort Sort Filter Item VM Object Reference) {is=title K2Node_ComponentBoundEvent}
   > n9 = Cast To FortOptionItemVM(Object=@n8.SortFilterItemVM)
   > ToggleActive(self=@n9)
   > RefreshCounts(self=Get FilterVM(self=Get FortItemShopVM()))
   n10 = HandleInputMethodChanged!() -> (OutputDelegate, then, bNewInputType:ECommonInputType Enum) {is=title K2Node_CustomEvent}
   n11 = "Equal (Enum)"(A=@n10.bNewInputType, B=Gamepad)
   > Set Button Padding for Input Action!(
    exec
    Target:WBP UIKit Button Generic Base Object Reference=Get Button_Apply() -> (Button_Apply:WBP UIKit Button Generic Object Reference)
    ShouldShowInputAction:bool=@n11
   ) -> (then:Exec) {is=title K2Node_MacroInstance}
   > Set Button Padding for Input Action(Target=Get Button_Clear() -> (Button_Clear:WBP UIKit Button Generic Object Reference), ShouldShowInputAction=@n11)
   > Set Button Padding for Input Action(Target=Get Button_Close() -> (Button_Close:WBP UIKit Button Generic Object Reference), ShouldShowInputAction=@n11)
   n12 = Event PreConstruct()
   > n13 = Branch(Condition=@n12.IsDesignTime)
   n14 = Construct Object from Class(execute=[@n13.then, @n16.then], Class=&/DialogVM_ShopFilter.DialogVM_ShopFilter_C) -> (ReturnValue:Dialog VM Shop Filter Object Reference)
   > SetDialogData!(
    execute
    self:WBP UIKit Dialog Internal Object Reference=Get WBP_UIKit_Dialog_Base() -> (WBP_UIKit_Dialog_Base:WBP UIKit Dialog Internal Object Reference)
    Data:UI Kit Dialog Viewmodel Object Reference=@n14
   ) -> (then)
   n15 = Get StatusIndicator!(self:WBP UIKit Button Generic Base Object Reference=Get Button_Apply()) -> (StatusIndicator:WBP UIKit Status Indicator Object Reference)
   > Set OverrideStatusIndicatorVisibility!(execute, OverrideStatusIndicatorVisibility:bool=true, self:WBP UIKit Status Indicator Base Object Reference=@n15) -> (then, Output_Get:bool)
   > Set OverrideStatusIndicatorState!(execute, OverrideStatusIndicatorState:bool=true, self:WBP UIKit Status Indicator Base Object Reference=@n15) -> (then, Output_Get:bool)
   n16 = SetMobileLayout(execute=@n13.else)
   Event On Activated()
   > Bind Event to On Input Method Changed(self=Get CommonInputSubsystem(), Delegate=@n10)
   > Fire Event Sub View Opened!(execute, View:Widget Object Reference=Self-Reference(), AdditionalInfo:Flow Analytics Additional Info Structure, bIsOverlay:bool=true) -> (then)
   > HandleInputMethodChanged!(execute, self, bNewInputType:ECommonInputType Enum=GetCurrentInputType(self=Get CommonInputSubsystem())) -> (then) {is=title K2Node_CallFunction}
   > SetNavigationRules()
   Event On Deactivated()
   > Unbind Event from On Input Method Changed(self=Get CommonInputSubsystem(), Delegate=@n10)
  ]}
 GetNavigationFocusTarget {nodes [
   n1 = GetNavigationFocusTarget!() -> (then, Navigation:EUINavigation Enum) {is=title K2Node_FunctionEntry}
   > n2 = Cast To WBP_UIKit_SortFilter_Section_MultiSelect!(
    execute
    Object=Select!(
     Left:WBP UIKit Sort Filter Section Multi Select Object Reference
     Right:WBP UIKit Sort Filter Section Multi Select Object Reference
     Up:WBP UIKit Sort Filter Section Multi Select Object Reference=Get Section_ShowOnly() -> (Section_ShowOnly:WBP UIKit Sort Filter Section Multi Select Object Reference)
     Down:WBP UIKit Sort Filter Section Multi Select Object Reference=Get Section_ItemType()
     Next:WBP UIKit Sort Filter Section Multi Select Object Reference
     Previous:WBP UIKit Sort Filter Section Multi Select Object Reference
     Invalid:WBP UIKit Sort Filter Section Multi Select Object Reference
     Index:EUINavigation Enum=@n1
    ) -> (ReturnValue:WBP UIKit Sort Filter Section Multi Select Object Reference)
   ) -> (then, CastFailed, AsWBP UIKit Sort Filter Section Multi Select:WBP UIKit Sort Filter Section Multi Select Object Reference) {is=title K2Node_DynamicCast}
   > Return Node(ReturnValue=GetDesiredFocusTarget(self=@n2))
   Return Node(execute=@n2.CastFailed)
  ]}
 MakeOptionArray {nodes [
   n1 = MakeOptionArray!() -> (then, GameCompatibilityOption:Fort Sort Filter Item VM Object Reference, OwnershipOption:Fort Sort Filter Item VM Object Reference) {is=title K2Node_FunctionEntry}
   > Return Node!(
    execute
    Array:Array of Fort Sort Filter Item VM Object References=Make Array!("[0]":Fort Sort Filter Item VM Object Reference=@n1, "[1]":Fort Sort Filter Item VM Object Reference=@n1.OwnershipOption) -> (Array:Array of Fort Sort Filter Item VM Object References)
   )
  ]}
 SetButtonPaddingForInputAction {nodes [
   n1 = Inputs!() -> (exec:Exec, Target:WBP UIKit Button Generic Base Object Reference, ShouldShowInputAction:bool=~)
   n2 = GetCurrentCulture!() -> (ReturnValue:String)
   n3 = Select!(Option 0:float=48, Option 1:float=72, Index:bool=@n1.ShouldShowInputAction) -> (ReturnValue:float)
   n4 = Get!(self:WBP UIKit Button Generic Base Object Reference=@n1) -> (TextPadding_Left:float=~, TextPadding_Top:float=~, TextPadding_Right:float=~, TextPadding_Bottom:float=~)
   > SetTextPadding!(
    execute
    self:WBP UIKit Button Generic Base Object Reference=@n1
    TextPadding:Margin Structure=Select!(
     Option 0_Left:float=@n3
     Option 0_Top:float=@n4.TextPadding_Top
     Option 0_Right:float=@n4.TextPadding_Right
     Option 0_Bottom:float=@n4.TextPadding_Bottom
     Option 1_Left:float=@n4
     Option 1_Top:float=@n4.TextPadding_Top
     Option 1_Right:float=@n3
     Option 1_Bottom:float=@n4.TextPadding_Bottom
     Index="Equal Exactly (String)"(A=@n2, B=ar)
    ) -> (ReturnValue:Margin Structure)
   ) -> (then)
   > Outputs!(then:Exec)
  ]}
 SetMobileLayout {nodes [
   SetMobileLayout() {is=title K2Node_FunctionEntry}
   > Branch(Condition=IsUsingCompactUI())
   n1 = Get WBP_UIKit_Dialog_Base()
   n2 = Get!(self:WBP UIKit Dialog Internal Object Reference=@n1) -> (
    DialogSIze_WidthOverride_107_#1:S UI Bool Float Structure
    DialogSIze_HeightOverride_108_#2:S UI Bool Float Structure
    DialogSIze_MinDesiredWidth_109_#3:S UI Bool Float Structure
    DialogSIze_MinDesiredHeight_110_#4:S UI Bool Float Structure
    DialogSIze_MaxDesiredWidth_111_#5:S UI Bool Float Structure
    DialogSIze_MaxDesiredHeight_112_#6:S UI Bool Float Structure
    DialogSIze_MinAspectRatio_113_#7:S UI Bool Float Structure
    DialogSIze_MaxAspectRatio_114_#8:S UI Bool Float Structure
   )
   > Set!(
    execute
    DialogSIze_WidthOverride_107_#1_Enable_3_#9:bool=true
    DialogSIze_WidthOverride_107_#1_Value_4_#10:double=1600
    DialogSIze_HeightOverride_108_#2:S UI Bool Float Structure=@n2.DialogSIze_HeightOverride_108_#2
    DialogSIze_MinDesiredWidth_109_#3:S UI Bool Float Structure=@n2.DialogSIze_MinDesiredWidth_109_#3
    DialogSIze_MinDesiredHeight_110_#4:S UI Bool Float Structure=@n2.DialogSIze_MinDesiredHeight_110_#4
    DialogSIze_MaxDesiredWidth_111_#5:S UI Bool Float Structure=@n2.DialogSIze_MaxDesiredWidth_111_#5
    DialogSIze_MaxDesiredHeight_112_#6:S UI Bool Float Structure=@n2.DialogSIze_MaxDesiredHeight_112_#6
    DialogSIze_MinAspectRatio_113_#7:S UI Bool Float Structure=@n2.DialogSIze_MinAspectRatio_113_#7
    DialogSIze_MaxAspectRatio_114_#8:S UI Bool Float Structure=@n2.DialogSIze_MaxAspectRatio_114_#8
    self:WBP UIKit Dialog Internal Object Reference=@n1
   ) -> (then, Output_Get:S UI Size Box Structure)
   n3 = Get Button_Apply()
   > SetButtonMinSize!(
    execute
    self:WBP UIKit Button Generic Base Object Reference=[Get Button_Clear(), @n3, Get Button_Close()]
    Min Width:int=0
    Min Height:int=80
   ) -> (then)
   > SetMaxDimensions!(execute, self:Common Button Base Object Reference=@n3, InMaxWidth:int=700, InMaxHeight:int=0) -> (then)
  ]}
 SetNavigationRules {nodes [
   SetNavigationRules() {is=title K2Node_FunctionEntry}
   > SetNavigationRuleBase!(execute, self:Widget Object Reference, Direction:EUINavigation Enum=Left, Rule:EUINavigationRule Enum=Stop) -> (then)
   > SetNavigationRuleBase(Direction=Right, Rule=Stop)
   > SetNavigationRuleBase(Direction=Up, Rule=Stop)
   > SetNavigationRuleBase(Direction=Down, Rule=Stop)
   > SetNavigationRuleBase(Direction=Next, Rule=Stop)
   > SetNavigationRuleBase(Direction=Previous, Rule=Stop)
  ]}
 ToggleActiveOption {nodes [
   n1 = ToggleActiveOption!() -> (then, OptionItemVM:Fort Option Item VM Object Reference) {is=title K2Node_FunctionEntry}
   > ToggleActive(self=@n1)
   > RefreshCounts(self=Get FilterVM(self=Get FortItemShopVM()))
  ]}
 VM_ActiveFiltersNumUpdated {nodes [
   n1 = VM_ActiveFiltersNumUpdated!() -> (then, ActiveFiltersNum:int) {is=title K2Node_FunctionEntry}
   > Set NoActiveFilters(NoActiveFilters:bool="integer <= integer"(A=@n1)) -> (Output_Get:bool)
   > n2 = Sequence!(execute) -> (then_0, then_1, then_2:Exec)
   n3 = Get StatusIndicator(self=Get Button_Apply())
   > "SetText (Text)"(
    self=Get Indicator_Text!(self:WBP UIKit Status Indicator Base Object Reference=@n3) -> (Indicator_Text:Common Text Object Reference)
    InText="To Text (Integer)"(Value=Get ActiveFiltersNum() -> (ActiveFiltersNum:int=0))
   )
   > Set Visibility(
    Visibility=Select!(
     Option 0:ESlateVisibility Enum=SelfHitTestInvisible
     Option 1:ESlateVisibility Enum=Collapsed
     Index=Get NoActiveFilters() -> (NoActiveFilters:bool)
    ) -> (ReturnValue:ESlateVisibility Enum)
    self=@n3
   )
   n4 = GetCurrentCulture()
   n5 = Select(Option 0:float=72, Option 1:float=48, Index=Get NoActiveFilters()) -> (ReturnValue:float)
   n6 = Get Button_Apply()
   n7 = Get!(self:WBP UIKit Button Generic Base Object Reference=@n6) -> (TextPadding_Left:float=~, TextPadding_Top:float=~, TextPadding_Right:float=~, TextPadding_Bottom:float=~)
   > SetTextPadding(
    self=@n6
    TextPadding=Select!(
     Option 0_Left:float=@n7
     Option 0_Top:float=@n7.TextPadding_Top
     Option 0_Right:float=@n5
     Option 0_Bottom:float=@n7.TextPadding_Bottom
     Option 1_Left:float=@n5
     Option 1_Top:float=@n7.TextPadding_Top
     Option 1_Right:float=@n7.TextPadding_Right
     Option 1_Bottom:float=@n7.TextPadding_Bottom
     Index="Equal Exactly (String)"(A=@n4, B=ar)
    ) -> (ReturnValue:Margin Structure)
   )
   SetText!(
    execute=@n2.then_1
    self:WBP UIKit Button Generic Base Object Reference=Get Button_Apply()
    Text:Text=Select!(Option 0:Text="LOCTEXT(\"See results\")", Option 1:Text="LOCTEXT(\"Show All Items\")", Index=Get NoActiveFilters()) -> (ReturnValue:Text)
   ) -> (then)
   n8 = Get Button_Clear()
   SetBackgroundMaterials!(
    execute=@n2.then_2
    self:WBP UIKit Button Generic Base Object Reference=@n8
    Background MI:Material Instance Object Reference=Select(
     Option 0:Material Instance Object Reference=/FortUILibrary/UIKit/Buttons/MaterialInstances/CTAButtons/MI_UIKit_BgRegular
     Option 1:Material Instance Object Reference=/FortUILibrary/UIKit/Buttons/MaterialInstances/CTAButtons/MI_UIKit_BgQuiet
     Index=Get NoActiveFilters()
    ) -> (ReturnValue:Material Instance Object Reference)
    Outline MI:Material Instance Object Reference=/FortUILibrary/UIKit/Buttons/MaterialInstances/TileOutline/MI_UIKit_Outline
   ) -> (then)
   > SetIsInteractionEnabled!(execute, self:Common Button Base Object Reference=@n8, bInIsInteractionEnabled:bool=NOT Boolean(A=Get NoActiveFilters())) -> (then)
   > Set RenderOpacity!(
    execute
    RenderOpacity:float=Select(Option 0:float=1, Option 1:float=0.5, Index=Get NoActiveFilters()) -> (ReturnValue:float)
    self:Widget Object Reference=Get Block_Text!(self:WBP UIKit Button Generic Base Object Reference=@n8) -> (Block_Text:WBP UIKit Block Text Object Reference)
   ) -> (then, Output_Get:float)
  ]}
 VM_OfferGroupsWithOffersUpdated {nodes [
   VM_OfferGroupsWithOffersUpdated!() -> (then, OfferGroups:Array of Fort Item Shop Offer Group VM Object References) {is=title K2Node_FunctionEntry}
   > RefreshCounts(self=Get FilterVM(self=Get FortItemShopVM()))
  ]}
]
variables []
widgets {
 info {
  inheritedWidgetCount=0
  namedSlotCount=1
  parentClass=&CommonActivatableWidget
  rootWidgetClass=/FortUILibrary/Systems/Dialog/WBP_UIKit_Dialog_Internal.WBP_UIKit_Dialog_Internal_C
  widgetCount=11
 }
 widgets [
  WBP_UIKit_Dialog_Base {
   widgetClassPath=/FortUILibrary/Systems/Dialog/WBP_UIKit_Dialog_Internal
   bIsVariable=true
   widget=WBP_UIKit_Dialog_Base
   properties {
    buttons {
     each Button Horizontal Fill=true
     each Button Min Width=0
     each Button Size=Medium
     focuse Button=0
     space Between Buttons=30
     cTA Buttons Types [Loud,Regular]
     group Placement {
      horizontal Alignment=HAlign_Fill
      vertical Alignment=VAlign_Bottom
      grid {column=1, column Span=0, layer=0, row=2, row Span=0, nudge=<0 0>}
      padding=<ltrb 0 0 0 0>
     }
    }
    dialogSIze {
     height Override {enable=false, value=800}
     max Aspect Ratio {enable=false, value=0}
     max Desired Height {enable=true, value=988}
     max Desired Width {enable=false, value=0}
     min Aspect Ratio {enable=false, value=0}
     min Desired Height {enable=true, value=552}
     min Desired Width {enable=false, value=0}
     width Override {enable=true, value=1312}
    }
   }
  }
  VerticalBox_0 {widgetClassPath=&VerticalBox, namedSlotHost=WBP_UIKit_Dialog_Base, widget=VerticalBox_0}
  Border_1 {
   widgetClassPath=&Border
   parent=VerticalBox_0
   widget=Border_1
   properties {
    clipping=ClipToBounds
    background {
     bIsDynamicallyLoaded=false
     drawAs=NoDrawType
     imageType=NoImage
     mirroring=NoMirror
     resourceName=None
     resourceObject=null
     tiling=NoTile
     imageSize=<32 32>
     margin=<ltrb 0 0 0 0>
     outlineSettings {
      bUseBrushTransparency=false
      roundingType=HalfHeightRadius
      width=0
      color {colorUseRule=UseColor_Specified, specifiedColor=<rgba 0 0 0 0>}
      cornerRadii=<0 0 0 0>
     }
     tintColor {colorUseRule=UseColor_Specified, specifiedColor=<rgba 1 1 1 1>}
     uVRegion {bIsValid=false, max=<0 0>, min=<0 0>}
    }
    padding=<ltrb 80 0 80 0>
   }
   slot {class=VerticalBoxSlot, properties {padding=<ltrb 0 0 0 64>, size {sizeRule=Fill, value=1}}}
  }
  VerticalBox_1 {
   widgetClassPath=&VerticalBox
   parent=Border_1
   widget=VerticalBox_1
   slot {class=BorderSlot, properties {padding=<ltrb 80 0 80 0>}}
  }
  Section_ShowOnly {
   widgetClassPath=/FortUILibrary/UIKit/SortFilter/WBP_UIKit_SortFilter_Section_MultiSelect
   bIsVariable=true
   parent=VerticalBox_1
   widget=Section_ShowOnly
   properties {columnCount=2, designPreviewCount=2, sectionTitle=Show Only, useEvenlySpacedColumns=false}
   slot {class=VerticalBoxSlot, properties {padding=<ltrb 0 0 0 48>}}
  }
  Section_ItemType {
   widgetClassPath=&WBP_UIKit_SortFilter_Section_MultiSelect
   bIsVariable=true
   parent=VerticalBox_1
   widget=Section_ItemType
   properties {designPreviewCount=27, sectionTitle=Item Type}
   slot {class=VerticalBoxSlot, properties {size {sizeRule=Fill, value=1}}}
  }
  HorizontalBox_0 {
   widgetClassPath=&HorizontalBox
   parent=VerticalBox_0
   widget=HorizontalBox_0
   slot {class=VerticalBoxSlot, properties {padding=<ltrb 80 0 80 48>}}
  }
  Button_Apply {
   widgetClassPath=&WBP_UIKit_Button_Generic
   bIsVariable=true
   parent=HorizontalBox_0
   widget=Button_Apply
   properties {
    bIsFocusable=false
    backgroundMaterial=&MI_UIKit_BgRegular
    configuration=Number
    enableFocusedSelectedState=false
    statusIndicatorVerticalAlignment=VAlign_Center
    text=Apply
    useDarkInputIcon=true
    useIconBlock=false
    useStatusIndicator=true
    buttonPadding=<ltrb 0 0 0 0>
    font {
     bForceMonospaced=false
     bMaterialIsStencil=false
     fontMaterial=/FortUILibrary/UIKit/Buttons/MaterialInstances/CTAButtons/MI_UIKit_TextQuiet
     fontObject=&HeadingNow
     letterSpacing=48
     monospacedWidth=1
     size=18
     skewAmount=0
     typefaceFontName=76 Bold
     outlineSettings {
      bApplyOutlineToDropShadows=false
      bMiteredCorners=false
      bSeparateFillAlpha=false
      outlineMaterial=null
      outlineSize=0
      outlineColor=<rgba 0 0 0 1>
     }
    }
    promptNudge=<24 0>
    statusIndicatorPadding=<ltrb 0 0 24 0>
    textPadding=<ltrb 96 0 96 0>
    triggeringInputAction {dataTable=&InputActionDataTable, rowName=GenericFaceButtonTop}
   }
   slot {class=HorizontalBoxSlot, properties {padding=<ltrb 0 0 24 0>}}
  }
  Button_Clear {
   widgetClassPath=&WBP_UIKit_Button_Generic
   bIsVariable=true
   parent=HorizontalBox_0
   widget=Button_Clear
   properties {
    bDisplayInputActionWhenNotInteractable=true
    bIsFocusable=false
    configuration=Number
    enableFocusedSelectedState=false
    statusIndicatorVerticalAlignment=VAlign_Center
    text=Clear
    useDarkInputIcon=true
    buttonPadding=<ltrb 0 0 0 0>
    font {
     bForceMonospaced=false
     bMaterialIsStencil=false
     fontMaterial=&MI_UIKit_TextQuiet
     fontObject=&HeadingNow
     letterSpacing=48
     monospacedWidth=1
     size=18
     skewAmount=0
     typefaceFontName=76 Bold
     outlineSettings {
      bApplyOutlineToDropShadows=false
      bMiteredCorners=false
      bSeparateFillAlpha=false
      outlineMaterial=null
      outlineSize=0
      outlineColor=<rgba 0 0 0 1>
     }
    }
    promptNudge=<24 0>
    statusIndicatorPadding=<ltrb 0 0 48 0>
    textPadding=<ltrb 72 0 48 0>
    triggeringInputAction {dataTable=&InputActionDataTable, rowName=GenericFaceButtonLeft}
   }
   slot {class=HorizontalBoxSlot}
  }
  Spacer_104 {
   widgetClassPath=&Spacer
   parent=HorizontalBox_0
   widget=Spacer_104
   slot {class=HorizontalBoxSlot, properties {size {sizeRule=Fill, value=1}}}
  }
  Button_Close {
   widgetClassPath=&WBP_UIKit_Button_Generic
   bIsVariable=true
   parent=HorizontalBox_0
   widget=Button_Close
   properties {
    bDisplayInputActionWhenNotInteractable=true
    bHideInputActionWithKeyboard=true
    bIsFocusable=false
    configuration=Number
    enableFocusedSelectedState=false
    statusIndicatorVerticalAlignment=VAlign_Center
    text=Close
    useDarkInputIcon=true
    buttonPadding=<ltrb 0 0 0 0>
    font {
     bForceMonospaced=false
     bMaterialIsStencil=false
     fontMaterial=&MI_UIKit_TextQuiet
     fontObject=&HeadingNow
     letterSpacing=48
     monospacedWidth=1
     size=18
     skewAmount=0
     typefaceFontName=76 Bold
     outlineSettings {
      bApplyOutlineToDropShadows=false
      bMiteredCorners=false
      bSeparateFillAlpha=false
      outlineMaterial=null
      outlineSize=0
      outlineColor=<rgba 0 0 0 1>
     }
    }
    promptNudge=<24 0>
    statusIndicatorPadding=<ltrb 0 0 48 0>
    textPadding=<ltrb 72 0 48 0>
    triggeringInputAction {dataTable=&InputActionDataTable, rowName=GenericBack}
   }
   slot {class=HorizontalBoxSlot, properties {horizontalAlignment=HAlign_Right, padding=<ltrb 24 0 0 0>}}
  }
 ]
}
