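# Item Shop widget holding one gradient image, shown or hidden to suit the vault world background.
# When active, UpdateVaultWorldBackground checks that dynamic PDP backgrounds are enabled and no filter is active, then loads the featured layout's preferred background with a callback.
# A successful load shows the gradient for Light backgrounds, collapses it for Dark and returns false.
# Otherwise it collapses the gradient and returns ToUpdateBackground true; inactive calls just return false.
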WidgetBlueprint /ItemShop/MultiProduct/WBP_ItemShop_BackgroundGradient
graphs [
 EventGraph {nodes [
   Event PreConstruct()
   Event Construct()
   Event Tick()
  ]}
 UpdateVaultWorldBackground {nodes [
   n1 = UpdateVaultWorldBackground!() -> (then, IsActive:bool=~, Fort Item Shop VM:Fort Item Shop VM Object Reference, OnLoadedCallback:"Delegate (by ref)") {is=title K2Node_FunctionEntry}
   > n2 = Branch(Condition=@n1)
   n3 = Get bIsActive!(self:Fort Item Shop Filter VM Object Reference=Get FilterVM!(self:Fort Item Shop VM Object Reference=@n1.Fort Item Shop VM) -> (FilterVM:Fort Item Shop Filter VM Object Reference)) -> (bIsActive:bool)
   > n4 = Branch(Condition=AND Boolean(
     A=AreDynamicBackgroundsEnabledInPdp!(self:Fort Item Shop VM Object Reference=@n1.Fort Item Shop VM) -> (ReturnValue:bool)
     B=NOT Boolean(A=@n3)
    ))
   n5 = Get Fort Item Shop VM() -> (Fort Item Shop VM:Fort Item Shop VM Object Reference)
   n6 = GetPreferredBackgroundForLayout!(
    self:Fort Item Shop VM Object Reference=@n5
    InLayout:"Fort Item Shop Layout Structure (by ref)"=Get FeaturedLayout!(self:Fort Item Shop VM Object Reference=@n5) -> (FeaturedLayout:Fort Item Shop Layout Structure)
   ) -> (ReturnValue:Fort Item Shop CMSBackground Structure)
   > n7 = LoadBackground!(
    execute
    self:Fort Item Shop VM Object Reference=@n5
    Background:"Fort Item Shop CMSBackground Structure (by ref)"=@n6
    OnLoadedCallback:"Delegate (by ref)"=@n1.OnLoadedCallback
   ) -> (then, ReturnValue:bool)
   > n8 = Branch(Condition=@n7)
   > Set Visibility(
    Visibility=Select!(
     Light:ESlateVisibility Enum=SelfHitTestInvisible
     Dark:ESlateVisibility Enum=Collapsed
     Index:EFortItemShopBackgroundType Enum=Break Fort Item Shop CMSBackground!(FortItemShopCMSBackground:"Fort Item Shop CMSBackground Structure (by ref)"=@n6) -> (Type:EFortItemShopBackgroundType Enum) {is=title K2Node_BreakStruct}
    ) -> (ReturnValue)
    self=Get Image_BackgroundGradient() -> (Image_BackgroundGradient:Image Object Reference)
   )
   > Return Node!(execute, ToUpdateBackground:bool=false)
   Set Visibility(execute=[@n4.else, @n8.else], Visibility=Collapsed, self=Get Image_BackgroundGradient())
   > Return Node(ToUpdateBackground=true)
   Return Node(execute=@n2.else, ToUpdateBackground=false)
  ]}
]
variables []
widgets {
 info {inheritedWidgetCount=0, parentClass=&UserWidget, rootWidgetClass=&Image, widgetCount=1}
 widgets [Image_BackgroundGradient {
   widgetClassPath=&Image
   bIsVariable=true
   widget=Image_BackgroundGradient
   properties {
    visibility=SelfHitTestInvisible
    brush {
     bIsDynamicallyLoaded=false
     drawAs=Image
     imageType=FullColor
     mirroring=NoMirror
     resourceName=None
     resourceObject=/FortUILibrary/UIKit/ModulateBackground/Materials/MI_UI_ModulateBackground
     tiling=NoTile
     imageSize=<800 32>
     margin=<ltrb 0 0 0 0>
     outlineSettings {
      bUseBrushTransparency=false
      roundingType=HalfHeightRadius
      width=0
      color {colorUseRule=UseColor_Specified, specifiedColor=<rgba 0 0 0 0>}
      cornerRadii=<0 0 0 0>
     }
     tintColor {colorUseRule=UseColor_Specified, specifiedColor=<rgba 1 1 1 1>}
     uVRegion {bIsValid=false, max=<0 0>, min=<0 0>}
    }
   }
  }]
}
