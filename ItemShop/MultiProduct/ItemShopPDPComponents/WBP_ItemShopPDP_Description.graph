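# Item Shop product detail page widget that shows the description header for the selected offer or previewed item, alongside an XP progress bar.
# A visibility switcher shows the item header when items are focused or the offer is a single Jam Track, otherwise the offer header with highlight tag.
# Previewed-item viewmodel updates feed item details, compatibility wording and Jam Track subtitle visibility.
# View changes toggle the item body text.
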WidgetBlueprint /ItemShop/MultiProduct/ItemShopPDPComponents/WBP_ItemShopPDP_Description
graphs [
 EventGraph {nodes [
   Event PreConstruct()
   n1 = Get ItemDescription_Offer() -> (ItemDescription_Offer:WBP MPItem Shop Offer Inspection Item Description Object Reference)
   > SetRarityText!(execute, self:WBP UIKit Item Description Base Object Reference=@n1, Text:Text) -> (then)
   > SetRarityTagVisibility!(execute, self:WBP UIKit Item Description Base Object Reference=@n1, ShowTag:bool=true) -> (then)
   > SetStyleText!(execute, self:WBP UIKit Item Description Base Object Reference=@n1, Text:Text) -> (then)
   n2 = Event OnOfferSet!() -> (
    OutputDelegate
    then
    OfferVM:Fort Item Shop Offer VM Object Reference
    ParentOfferVM:Fort Item Shop Offer VM Object Reference
    OfferGroupVM:Fort Item Shop Offer Group VM Object Reference
   ) {is=title K2Node_Event}
   > Set FortItemShopOfferVM(FortItemShopOfferVM:Fort Item Shop Offer VM Object Reference=@n2.OfferVM) -> (Output_Get:Fort Item Shop Offer VM Object Reference)
   > OnOfferSet!(
    execute
    self:BPI Item Shop PDP Interface=Get Highlight() -> (Highlight:WBP Item Shop Tile Highlight 2 Object Reference)
    OfferVM:Fort Item Shop Offer VM Object Reference=Get FortItemShopOfferVM() -> (FortItemShopOfferVM:Fort Item Shop Offer VM Object Reference)
    ParentOfferVM:Fort Item Shop Offer VM Object Reference
    OfferGroupVM:Fort Item Shop Offer Group VM Object Reference
   ) -> (then)
   n3 = Event OnFocusAreaChanged!() -> (OutputDelegate, then, IsFocusingItems:bool) {is=title K2Node_Event}
   n4 = IsSingleJamTrackOffer!(self) -> (ReturnValue:bool)
   > SetActiveWidget(
    self=Get Switcher_Header() -> (Switcher_Header:Common Visibility Switcher Object Reference)
    Widget=Select(
     Option 0:Widget Object Reference=Get Header_Offer() -> (Header_Offer:Vertical Box Object Reference)
     Option 1:Widget Object Reference=Get Header_Item() -> (Header_Item:Vertical Box Object Reference)
     Index=OR Boolean(A=@n3.IsFocusingItems, B=@n4)
    ) -> (ReturnValue:Widget Object Reference)
   )
   n5 = Event OnViewChanged!() -> (OutputDelegate, then, NewView:EFortItemShopPdpView Enum=OfferInspect) {is=title K2Node_Event}
   > SetBodyTextVisibility!(
    execute
    self:WBP UIKit Item Description Base Object Reference=Get ItemDescription_Item() -> (ItemDescription_Item:WBP MPItem Shop Offer Inspection Item Description Object Reference)
    Show Text:bool=Select!(OfferInspect:bool=false, ItemDetails:bool=true, Index:EFortItemShopPdpView Enum=@n5.NewView) -> (ReturnValue:bool)
   ) -> (then)
   > ChangeView!(
    execute
    self:WBP MPItem Shop Offer Inspection Item Description Object Reference=Get ItemDescription_Item()
    NewView:EFortItemShopPdpView Enum=@n5.NewView
   ) -> (then)
  ]}
 GetCompatibilityText {nodes [
   n1 = GetCompatibilityText!() -> (then, Compatible Modes:Set of Gameplay Tag Structures, GrantedItems:Array of Fort Item VM Object References) {is=title K2Node_FunctionEntry}
   n2 = Length(TargetArray:Array of Fort Item VM Object References=@n1.GrantedItems)
   n3 = "integer > integer"(A=Length!(TargetSet:Set of Gameplay Tag Structures=@n1) -> (ReturnValue:int=0) {is=title K2Node_CallFunction}, B=1)
   > Return Node!(
    execute
    ReturnValue:Text=Select!(
     Option 0:Text=Select(Option 0="LOCTEXT(\"This item only works in:\")", Option 1="LOCTEXT(\"This item works in:\")", Index=@n3)
     Option 1:Text=Select(Option 0="LOCTEXT(\"These items may only work in:\")", Option 1="LOCTEXT(\"These items may work in:\")", Index=@n3)
     Index:bool="integer > integer"(A=@n2, B=1)
    ) -> (ReturnValue:Text)
   )
  ]}
 GetOfferCategoryText {nodes [
   n1 = GetOfferCategoryText!() -> (then, ShortDescription:"Text (by ref)", OfferType:EFortItemShopOfferType Enum=LevelBundleOffer) {is=title K2Node_FunctionEntry}
   n2 = "Equal (Enum)"(A:EFortItemShopOfferType Enum=@n1.OfferType, B:EFortItemShopOfferType Enum=Special)
   n3 = "Equal (Enum)"(A=@n1.OfferType, B=None)
   n4 = "Equal (Enum)"(A=@n1.OfferType, B=LevelBundle)
   > Return Node(ReturnValue=Select(Option 0=Get ShortDescription() -> (ShortDescription:"Text (by ref)"), Index=OR Boolean!(A=@n3, B=@n4, C:bool=@n2) -> (ReturnValue)))
  ]}
 IsSingleJamTrackOffer {nodes [
   IsSingleJamTrackOffer() {is=title K2Node_FunctionEntry}
   > n1 = Get!(execute:Exec) -> (then:Exec, else:Exec, FortItemShopOfferVM:Fort Item Shop Offer VM Object Reference)
   n2 = Get GrantedItems!(self:Fort Item Shop Offer VM Object Reference=@n1) -> (GrantedItems:Array of Fort Item VM Object References)
   > n3 = Branch(Condition="Equal (Integer)"(A=Length(TargetArray=@n2) {is=title K2Node_CallArrayFunction}, B=1))
   n4 = Get ItemDefinition!(self:Fort Item VM Object Reference="Get (a copy)"(Array:Array of Fort Item VM Object References=@n2) -> (Output:Fort Item VM Object Reference)) -> (ItemDefinition:Fort Item Definition Object Reference)
   > Return Node(ReturnValue:bool="Equal (Enum)"(
     A:EFortItemType Enum=GetItemType!(ItemDefinition:Item Definition Base Object Reference=@n4) -> (ReturnValue:EFortItemType Enum=WorldItem)
     B:EFortItemType Enum=SparksSong
    ))
   Return Node(execute=@n3.else, ReturnValue=false)
  ]}
 VM_PreviewedItemUpdated {nodes [
   VM_PreviewedItemUpdated!() -> (then, PreviewedItem:Item Shop Preview Data Object Reference) {is=title K2Node_FunctionEntry}
   > n1 = Get!(execute) -> (then, else, PreviewedItem:Item Shop Preview Data Object Reference)
   > n2 = Get!(execute, self:Item Shop Preview Data Object Reference=@n1) -> (then, else, ItemVM:Fort Item VM Object Reference)
   > SetFortItemDetailsVM!(
    execute
    self:WBP UIKit Description With Subtitle Object Reference=Get ItemDescription_Item()
    Viewmodel:Fort Item Details VM Object Reference=Get ItemDetails!(self:Fort Item VM Object Reference=@n2) -> (ItemDetails:Fort Item Details VM Object Reference)
   ) -> (then)
   n3 = Length(TargetSet=GetCompatibleProducts!(self:Fort Item VM Object Reference=@n2) -> (ReturnValue:Set of Gameplay Tag Structures)) -> (ReturnValue=0) {is=title K2Node_CallFunction}
   > SetCompatibilityText!(
    execute
    self:WBP UIKit Item Description Base Object Reference=Get ItemDescription_Item()
    Text:Text=Select(Option 0="LOCTEXT(\"This item only works in:\")", Option 1="LOCTEXT(\"This item works in:\")", Index="integer > integer"(A=@n3, B=1))
   ) -> (then)
   n4 = Get FortItemShopVM() -> (FortItemShopVM:Fort Item Shop VM Object Reference)
   n5 = Get ItemDefinition(self=Get ItemVM!(self:Item Shop Preview Data Object Reference=Get PreviewedItem!(self:Fort Item Shop VM Object Reference=@n4) -> (PreviewedItem:Item Shop Preview Data Object Reference)) -> (ItemVM:Fort Item VM Object Reference))
   > ShowDynamicSubtitle!(
    execute
    self:WBP UIKit Description With Subtitle Object Reference=Get ItemDescription_Item()
    Show:bool="Equal (Enum)"(A=GetItemType(ItemDefinition=@n5) -> (ReturnValue=WorldItem), B=SparksSong)
   ) -> (then)
  ]}
]
variables []
widgets {
 info {inheritedWidgetCount=0, parentClass=&CommonUserWidget, rootWidgetClass=&Overlay, widgetCount=11}
 widgets [
  Overlay_2 {widgetClassPath=&Overlay, widget=Overlay_2}
  Switcher_Header {
   widgetClassPath=&CommonVisibilitySwitcher
   bIsVariable=true
   parent=Overlay_2
   widget=Switcher_Header
   slot {class=OverlaySlot, properties {verticalAlignment=VAlign_Fill}}
  }
  Header_Offer {
   widgetClassPath=&VerticalBox
   bIsVariable=true
   parent=Switcher_Header
   widget=Header_Offer
   slot {class=CommonVisibilitySwitcherSlot}
  }
  Overlay_0 {
   widgetClassPath=&Overlay
   parent=Header_Offer
   widget=Overlay_0
   slot {class=VerticalBoxSlot, properties {horizontalAlignment=HAlign_Left, verticalAlignment=VAlign_Top, padding=<ltrb 0 0 0 8>}}
  }
  Highlight {
   widgetClassPath=/ItemShop/MultiProduct/ItemShopTileComponents/WBP_ItemShopTile_Highlight_2
   bIsVariable=true
   parent=Overlay_0
   widget=Highlight
   properties {highlightText="[PH] Highlight"}
   slot {class=OverlaySlot}
  }
  Spacer_HighlightHeight_1 {widgetClassPath=&Spacer, parent=Overlay_0, widget=Spacer_HighlightHeight_1, properties {size=<1 32>}, slot {class=OverlaySlot}}
  ItemDescription_Offer {
   widgetClassPath=/ItemShop/MultiProduct/WBP_MPItemShop_OfferInspection_ItemDescription
   bIsVariable=true
   parent=Header_Offer
   widget=ItemDescription_Offer
   properties {
    bodyTextRow=3
    compatibilityRow=4
    flowDirectionPreference=LeftToRight
    hideSubtitle=true
    tagSize=Compact
    bodyTextPadding=<ltrb 0 0 0 0>
    bottonSlotPadding=<ltrb 0 8 0 0>
    compatibilityPadding=<ltrb 0 8 0 0>
    gameTags [BR,Juno]
    titleFont {
     bForceMonospaced=false
     bMaterialIsStencil=false
     fontMaterial=null
     fontObject=&HeadingNow
     letterSpacing=0
     monospacedWidth=1
     size=30
     skewAmount=0
     typefaceFontName=86 Bold Italic
     outlineSettings {
      bApplyOutlineToDropShadows=false
      bMiteredCorners=false
      bSeparateFillAlpha=false
      outlineMaterial=null
      outlineSize=0
      outlineColor=<rgba 0 0 0 1>
     }
    }
   }
   slot {class=VerticalBoxSlot}
  }
  Header_Item {
   widgetClassPath=&VerticalBox
   bIsVariable=true
   parent=Switcher_Header
   widget=Header_Item
   slot {class=CommonVisibilitySwitcherSlot}
  }
  Spacer_HighlightHeight_2 {
   widgetClassPath=&Spacer
   parent=Header_Item
   widget=Spacer_HighlightHeight_2
   properties {size=<1 32>}
   slot {class=VerticalBoxSlot, properties {horizontalAlignment=HAlign_Left, verticalAlignment=VAlign_Top, padding=<ltrb 0 0 0 8>}}
  }
  ItemDescription_Item {
   widgetClassPath=&WBP_MPItemShop_OfferInspection_ItemDescription
   bIsVariable=true
   parent=Header_Item
   widget=ItemDescription_Item
   properties {
    bodyTextRow=3
    compatibilityRow=4
    flowDirectionPreference=LeftToRight
    tagSize=Compact
    bottonSlotPadding=<ltrb 0 8 0 8>
    compatibilityPadding=<ltrb 0 8 0 0>
    gameTags [BR,Juno]
    subtitleFont {
     bForceMonospaced=false
     bMaterialIsStencil=false
     fontMaterial=null
     fontObject=&HeadingNow
     letterSpacing=0
     monospacedWidth=1
     size=15.75
     skewAmount=0
     typefaceFontName=HeadingNow-86BoldItalic
     outlineSettings {
      bApplyOutlineToDropShadows=false
      bMiteredCorners=false
      bSeparateFillAlpha=false
      outlineMaterial=null
      outlineSize=0
      outlineColor=<rgba 0 0 0 1>
     }
    }
    titleFont {
     bForceMonospaced=false
     bMaterialIsStencil=false
     fontMaterial=null
     fontObject=&HeadingNow
     letterSpacing=0
     monospacedWidth=1
     size=30
     skewAmount=0
     typefaceFontName=86 Bold Italic
     outlineSettings {
      bApplyOutlineToDropShadows=false
      bMiteredCorners=false
      bSeparateFillAlpha=false
      outlineMaterial=null
      outlineSize=0
      outlineColor=<rgba 0 0 0 1>
     }
    }
   }
   slot {class=VerticalBoxSlot}
  }
  WBP_ItemShop_XpProgressBar {
   widgetClassPath=/ItemShop/MultiProduct/WBP_ItemShop_XpProgressBar
   bIsVariable=true
   parent=Overlay_2
   widget=WBP_ItemShop_XpProgressBar
   slot {class=OverlaySlot}
  }
 ]
}
