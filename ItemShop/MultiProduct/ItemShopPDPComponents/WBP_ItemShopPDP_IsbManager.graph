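# Item shop product detail page widget that shows contextual notes for the previewed item and hosts an offer-specific custom inspection banner.
# It async-creates the offer's banner class, discards stale loads via a class-path check, and feeds it offer and item view models.
# View model updates drive rating, offer popup (data table by tag), unlock requirement and Rocket League, Sidekick or instrument notes, based on view, variants and item type.
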WidgetBlueprint /ItemShop/MultiProduct/ItemShopPDPComponents/WBP_ItemShopPDP_IsbManager
graphs [
 EventGraph {nodes [
   n1 = VM_CustomBannerUpdated!() -> (OutputDelegate, then, CustomInspectionBanner:User Widget Soft Class Reference) {is=title K2Node_CustomEvent}
   > Set CustomISB(CustomISB:WBP Custom Banner Base Object Reference) -> (Output_Get:WBP Custom Banner Base Object Reference)
   n2 = Get NamedSlot_CustomISB() -> (NamedSlot_CustomISB:Named Slot Object Reference)
   > ClearChildren(self=@n2)
   > SetVisibility(self=@n2, InVisibility=Collapsed)
   > Branch(Condition=IsValidSoftClassReference!(SoftClassReference:"Object Soft Class Reference (by ref)"=@n1.CustomInspectionBanner) -> (ReturnValue:bool))
   > n3 = CreateWidgetAsync!(
    execute:Exec
    UserWidgetSoftClass:User Widget Soft Class Reference=@n1.CustomInspectionBanner
    OwningPlayer:Player Controller Object Reference=GetOwningPlayer()
    bSuspendInputUntilComplete:bool=false
   ) -> (then:Exec, OnComplete:Exec, OnFailure:Exec, UserWidget:User Widget Object Reference) {is=title K2Node_AsyncAction}
   K2Node_Composite_0 = IsExpectedBannerType!(Object:Object Reference=@n3) -> (ReturnValue:bool=~) {is=title K2Node_Composite}
   n4 = Branch(execute=@n3.OnComplete, Condition=@K2Node_Composite_0)
   > n5 = Cast To WBP_CustomBanner_Base!(execute, Object=@n3) -> (then, CastFailed, AsWBP Custom Banner Base:WBP Custom Banner Base Object Reference) {is=title K2Node_DynamicCast}
   > Set CustomISB(CustomISB=@n5)
   > AddChild(self=Get NamedSlot_CustomISB(), Content=Get CustomISB() -> (CustomISB:WBP Custom Banner Base Object Reference))
   > SetVisibility(self=Get NamedSlot_CustomISB(), InVisibility=SelfHitTestInvisible)
   > SetContext!(execute, self:WBP Custom Banner Base Object Reference=Get CustomISB(), Desired Area:E_MultiproductContext Enum=NewEnumerator0) -> (then)
   > SetViewType!(execute, self:WBP Custom Banner Base Object Reference=Get CustomISB(), IsFullDetailsView:bool=true) -> (then)
   > SetOfferViewModel!(
    execute
    self:WBP Custom Banner Base Object Reference=Get CustomISB()
    OfferVM:Fort Item Shop Offer VM Object Reference=Get FortItemShopOfferVM() -> (FortItemShopOfferVM:Fort Item Shop Offer VM Object Reference)
   ) -> (then)
   n6 = Get FortItemShopVM() -> (FortItemShopVM:Fort Item Shop VM Object Reference)
   > SetSingleItem!(
    execute
    self:WBP Custom Banner Base Object Reference=Get CustomISB()
    ItemVM:Fort Item VM Object Reference=Get ItemVM!(self:Item Shop Preview Data Object Reference=Get PreviewedItem!(self:Fort Item Shop VM Object Reference=@n6) -> (PreviewedItem:Item Shop Preview Data Object Reference)) -> (ItemVM:Fort Item VM Object Reference)
   ) -> (then)
   RemoveFromParent(execute=@n4.else, self=@n3)
   n7 = Event OnOfferSet!() -> (
    OutputDelegate
    then
    OfferVM:Fort Item Shop Offer VM Object Reference
    ParentOfferVM:Fort Item Shop Offer VM Object Reference
    OfferGroupVM:Fort Item Shop Offer Group VM Object Reference
   ) {is=title K2Node_Event}
   > Set FortItemShopOfferVM(FortItemShopOfferVM:Fort Item Shop Offer VM Object Reference=@n7.OfferVM) -> (Output_Get:Fort Item Shop Offer VM Object Reference)
   n8 = Event OnViewChanged!() -> (OutputDelegate, then, NewView:EFortItemShopPdpView Enum=OfferInspect) {is=title K2Node_Event}
   > n9 = Set with Broadcast CurrentView!(execute, CurrentView:EFortItemShopPdpView Enum=@n8.NewView) -> (then, Output_Get:EFortItemShopPdpView Enum=OfferInspect)
   n10 = Get PreviewedItem(self=Get FortItemShopVM())
   n11 = NOT Boolean(A=Get bHasVariants!(self:Fort Item VM Object Reference=Get ItemVM(self=@n10)) -> (bHasVariants:bool))
   n12 = "Equal (Enum)"(A:EFortItemShopPdpView Enum=@n9, B:EFortItemShopPdpView Enum=OfferInspect)
   Set Visibility(
    Visibility=Select(Option 0=Collapsed, Option 1=SelfHitTestInvisible, Index=OR Boolean(A=@n12, B=@n11))
    self=Get Border_OfferPopupVisibility() -> (Border_OfferPopupVisibility:Border Object Reference)
   )
  ]}
 IsExpectedBannerType {
  path=EventGraph.K2Node_Composite_0.IsExpectedBannerType
  nodes [
   n1 = IsValid(Object=Get CustomISB())
   n2 = GetSoftClassPath!(Class:Object Class Reference=Get Class(Object=Inputs!() -> (Object:Object Reference)) -> (ReturnValue:Object Class Reference)) -> (ReturnValue_PathString:String)
   n3 = To Soft Class Path!(SoftClassReference:Object Soft Class Reference=Get CustomInspectionBanner!(self:Fort Item Shop Offer VM Object Reference=Get FortItemShopOfferVM()) -> (CustomInspectionBanner:User Widget Soft Class Reference)) -> (ReturnValue_PathString:String)
   Outputs!(ReturnValue:bool=AND Boolean(A="Equal Exactly (String)"(A=@n2, B=@n3), B=NOT Boolean(A=@n1)))
  ]
 }
 GetItemUnlockRequirementText {nodes [
   n1 = GetItemUnlockRequirementText!() -> (then, PreviewData:Item Shop Preview Data Object Reference) {is=title K2Node_FunctionEntry}
   > n2 = Is Valid(InputObject=@n1)
   > n3 = Get!(execute:Exec, self:Item Shop Preview Data Object Reference=@n1) -> (then:Exec, else:Exec, ItemVM:Fort Item VM Object Reference)
   n4 = Get UnlockRequirements!(self:Fort Item Details VM Object Reference=Get ItemDetails!(self:Fort Item VM Object Reference=@n3) -> (ItemDetails:Fort Item Details VM Object Reference)) -> (UnlockRequirements:Text)
   n5 = TextIsEmpty(InText=@n4)
   > n6 = Branch(Condition=AND Boolean(A=Get bIsLocked!(self:Fort Item VM Object Reference=@n3) -> (bIsLocked:bool), B=NOT Boolean(A=@n5)))
   > Return Node!(execute, Return Value:Text=@n4)
   Return Node(execute=[@n6.else, @n3.else, @n2.Is Not Valid])
  ]}
 GetOfferPopupText {nodes [
   n1 = GetOfferPopupText!() -> (then, OfferTag:String) {is=title K2Node_FunctionEntry}
   > n2 = Get Data Table Row OfferPopupDataTable!(
    execute:Exec
    DataTable:Data Table Object Reference=/Game/Athena/UI/Frontend/Data/OfferPopupDataTable
    RowName:Name=String To Name(InString=@n1)
   ) -> (then:Exec, RowNotFound:Exec, ReturnValue:Offer Pop Up Data Structure) {is=title K2Node_GetDataTableRow}
   > Return Node!(
    execute
    OutText:Text=Break Offer Pop Up Data!(OfferPopUpData:"Offer Pop Up Data Structure (by ref)"=@n2) -> (FormatedText:Text) {is=title K2Node_BreakStruct}
   )
   Return Node(execute=@n2.RowNotFound)
  ]}
 GetOfferPopupVisibility {nodes [
   n1 = GetOfferPopupVisibility!() -> (then, OfferTag:String) {is=title K2Node_FunctionEntry}
   > n2 = Get Data Table Row OfferPopupDataTable(DataTable=&OfferPopupDataTable, RowName=String To Name(InString=@n1))
   > Return Node!(execute, OutVisibility:ESlateVisibility Enum=SelfHitTestInvisible)
   Return Node(execute=@n2.RowNotFound, OutVisibility=Collapsed)
  ]}
 GetSelectedItem {nodes [
   n1 = GetSelectedItem!() -> (then, PreviewedItem:Item Shop Preview Data Object Reference, CurrentView:EFortItemShopPdpView Enum=OfferInspect) {is=title K2Node_FunctionEntry}
   > n2 = Is Valid(InputObject=@n1)
   Return Node!(execute=@n2.Is Not Valid, ReturnValue:Fort Item VM Object Reference)
   > Return Node(ReturnValue=Get ItemVM(self=@n1))
  ]}
 GetUnlockRequirementTextToShow {nodes [
   GetUnlockRequirementTextToShow!() -> (
    then
    PreviewData:Item Shop Preview Data Object Reference
    CurrentVariantOverrides:Array of Mcp Variant Channel Info Structures
    CurrentView:EFortItemShopPdpView Enum=OfferInspect
   ) {is=title K2Node_FunctionEntry}
   > Set VariantUnlockReq(VariantUnlockReq:Text=GetVariantUnlockRequirementText!(self) -> (Return Value:Text)) -> (Output_Get:Text)
   > Set ItemUnlockReq(ItemUnlockReq:Text=GetItemUnlockRequirementText!(self, PreviewData:Item Shop Preview Data Object Reference=Get PreviewData() -> (PreviewData:Item Shop Preview Data Object Reference)) -> (Return Value:Text) {is=title K2Node_CallFunction}) -> (Output_Get:Text)
   n1 = Get VariantUnlockReq() -> (VariantUnlockReq:Text)
   n2 = TextIsEmpty(InText=@n1)
   n3 = Get ItemUnlockReq() -> (ItemUnlockReq:Text)
   n4 = Get ItemUnlockReq()
   n5 = TextIsEmpty(InText=@n4)
   n6 = Get VariantUnlockReq()
   > Return Node!(
    execute
    UnlockRequirementToShow:Text=Select!(
     OfferInspect:Text=Select!(Option 0:Text=@n4, Option 1:Text=@n6, Index:bool=@n5) -> (ReturnValue)
     ItemDetails:Text=Select(Option 0=@n1, Option 1=@n3, Index=@n2)
     Index:EFortItemShopPdpView Enum=Get CurrentView() -> (CurrentView:EFortItemShopPdpView Enum=OfferInspect)
    ) -> (ReturnValue:Text)
   )
   Reroute Node(InputPin:Wildcard) -> (OutputPin:Wildcard)
  ]}
 GetVariantUnlockRequirementText {nodes [
   GetVariantUnlockRequirementText() {is=title K2Node_FunctionEntry}
   > n1 = For Each Loop(Array:Array of Fort Variant Set VM Object References=Get VariantSets!(self:Fort Variant Selector VM Object Reference=Get FortVariantSelectorVM() -> (FortVariantSelectorVM:Fort Variant Selector VM Object Reference)) -> (VariantSets:Array of Fort Variant Set VM Object References)) -> (Array Element:Fort Variant Set VM Object Reference)
   > n2 = For Each Loop(Array:Array of Fort Variant Channel VM Object References=Get VariantChannels!(self:Fort Variant Set VM Object Reference=@n1) -> (VariantChannels:Array of Fort Variant Channel VM Object References)) -> (Array Element:Fort Variant Channel VM Object Reference)
   n3 = GetSelectedVariantItem!(self:Fort Variant Channel VM Object Reference=@n2) -> (ReturnValue:Fort Variant Item VM Object Reference)
   > Is Valid(InputObject=@n3)
   n4 = Get UnlockRequirements(self:Fort Variant Item VM Object Reference=@n3)
   n5 = TextIsEmpty(InText=@n4)
   > Branch(Condition=AND Boolean(A=NOT Boolean(A=@n5), B=Get bIsLocked(self:Fort Variant Item VM Object Reference=@n3)))
   > Return Node!(execute, Return Value:Text=@n4)
  ]}
 SetTextJustification {nodes [
   n1 = SetTextJustification!() -> (then, TextJustification:ETextJustify Enum=Left) {is=title K2Node_FunctionEntry}
   > n2 = Set Justification!(
    execute
    Justification:ETextJustify Enum=@n1
    self:Text Layout Widget Object Reference=Get Text_RatingBoardWarning() -> (Text_RatingBoardWarning:Common Rich Text Block Object Reference)
   ) -> (then, Output_Get:ETextJustify Enum=Left)
   > n3 = Set Justification(Justification=@n2, self=Get Text_OfferPopup() -> (Text_OfferPopup:Common Rich Text Block Object Reference)) -> (Output_Get=Left)
   > n4 = Set Justification(Justification=@n3, self=Get Text_UnlockRequirement() -> (Text_UnlockRequirement:Common Rich Text Block Object Reference)) -> (Output_Get=Left)
   > n5 = Set Justification(Justification=@n4, self=Get Text_CrossEntitlement() -> (Text_CrossEntitlement:Common Rich Text Block Object Reference)) -> (Output_Get=Left)
   > n6 = Set Justification(Justification=@n5, self=Get Text_Mimosa() -> (Text_Mimosa:Common Text Object Reference)) -> (Output_Get=Left)
   > n7 = Set Justification(Justification=@n6, self=Get Text_Instrument() -> (Text_Instrument:Common Rich Text Block Object Reference)) -> (Output_Get=Left)
   > Set HorizontalAlignment!(
    execute
    HorizontalAlignment:EHorizontalAlignment Enum=Select!(
     Left:EHorizontalAlignment Enum=HAlign_Left
     Center:EHorizontalAlignment Enum=HAlign_Center
     Right:EHorizontalAlignment Enum=HAlign_Right
     InvariantLeft:EHorizontalAlignment Enum=HAlign_Left
     InvariantRight:EHorizontalAlignment Enum=HAlign_Right
     Index:ETextJustify Enum=@n7
    ) -> (ReturnValue:EHorizontalAlignment Enum)
    self:Vertical Box Slot Object Reference=SlotAsVerticalBoxSlot(Widget=Get NamedSlot_CustomISB())
   ) -> (then, Output_Get:EHorizontalAlignment Enum=HAlign_Fill)
  ]}
 VM_CurrentVariantOverridesUpdated {nodes [
   VM_CurrentVariantOverridesUpdated!() -> (then, CurrentVariantOverrides:Array of Mcp Variant Channel Info Structures) {is=title K2Node_FunctionEntry}
   > n1 = For Each Loop(Array:Array of Fort Variant Set VM Object References=Get VariantSets(self=Get FortVariantSelectorVM())) -> (Array Element:Fort Variant Set VM Object Reference)
   > n2 = For Each Loop(Array:Array of Fort Variant Channel VM Object References=Get VariantChannels(self=@n1)) -> (Array Element:Fort Variant Channel VM Object Reference)
   n3 = Get EquippedVariantItem!(self:Fort Variant Channel VM Object Reference=@n2) -> (EquippedVariantItem:Fort Variant Item VM Object Reference)
   > Is Valid(InputObject=@n3)
   n4 = Get Text_UnlockRequirement()
   > Set Text(Text=Get UnlockRequirements(self=@n3), self:Rich Text Block Object Reference=@n4)
   > SetVisibility(
    self=@n4
    InVisibility=Select!(Option 0:ESlateVisibility Enum=Collapsed, Option 1:ESlateVisibility Enum=HitTestInvisible, Index:bool=Get bIsLocked(self=@n3)) -> (ReturnValue:ESlateVisibility Enum)
   )
   > Return Node!(execute)
  ]}
 VM_SelectedItemUpdated {nodes [
   n1 = VM_SelectedItemUpdated!() -> (then, SelectedItem:Fort Item VM Object Reference) {is=title K2Node_FunctionEntry}
   > Is Valid(InputObject=@n1)
   > n2 = Sequence!(execute) -> (then_0, then_1, then_2:Exec, then_3:Exec)
   K2Node_Composite_1 = Is Vehicle Cosmetic!(self:Fort Item VM Object Reference=Get SelectedItem() -> (SelectedItem:Fort Item VM Object Reference)) -> (ReturnValue:bool=~) {is=title K2Node_Composite}
   Set Visibility(
    execute=@n2.then_2
    Visibility=Select(Option 0=Collapsed, Option 1=SelfHitTestInvisible, Index=@K2Node_Composite_1)
    self=Get HB_CrossEntitlement() -> (HB_CrossEntitlement:Horizontal Box Object Reference)
   )
   K2Node_Composite_2 = Is Companion!(self:Fort Item VM Object Reference=Get SelectedItem()) -> (ReturnValue:bool=~) {is=title K2Node_Composite}
   n3 = "Equal (Enum)"(A=Get CurrentView() -> (CurrentView=OfferInspect), B=OfferInspect)
   Set Visibility(
    execute=@n2.then_1
    Visibility=Select(Option 0=Collapsed, Option 1=SelfHitTestInvisible, Index=AND Boolean(A=@K2Node_Composite_2, B=@n3))
    self=Get Text_Mimosa()
   )
   K2Node_Composite_0 = Is Instrument!(self:Fort Item VM Object Reference=Get SelectedItem()) -> (ReturnValue:bool=~) {is=title K2Node_Composite}
   n4 = Get CurrentView() -> (CurrentView=OfferInspect)
   n5 = Branch(execute=@n2.then_3, Condition=AND Boolean(A=@K2Node_Composite_0, B="Equal (Enum)"(A=@n4, B=OfferInspect)))
   > n6 = HasIgnoredItemOfSameName!(execute, self:Fort Item Shop Offer VM Object Reference=Get FortItemShopOfferVM(), Item:Fort Item VM Object Reference=Get SelectedItem()) -> (then, ReturnValue:bool)
   > Set Visibility(Visibility=Select(Option 0=Collapsed, Option 1=SelfHitTestInvisible, Index=@n6), self=Get Text_Instrument())
   Set Visibility(execute=@n5.else, Visibility=Collapsed, self=Get Text_Instrument())
   n7 = Get!(execute=@n2.then_0) -> (then, else, CustomISB:WBP Custom Banner Base Object Reference)
   > SetSingleItem(self=@n7, ItemVM=Get SelectedItem())
  ]}
 Is Instrument {
  path=VM_SelectedItemUpdated.K2Node_Composite_0.Is Instrument
  nodes [
   n1 = GetItemType!(ItemDefinition:Item Definition Base Object Reference=Get ItemDefinition!(self:Fort Item VM Object Reference=Inputs!() -> (self:Fort Item VM Object Reference)) -> (ItemDefinition:Fort Item Definition Object Reference)) -> (ReturnValue:EFortItemType Enum=WorldItem)
   Outputs(ReturnValue=OR Boolean!(
     A="Equal (Enum)"(A:EFortItemType Enum=@n1, B:EFortItemType Enum=SparksGuitar)
     B="Equal (Enum)"(A=@n1, B=SparksBass)
     C:bool="Equal (Enum)"(A=@n1, B=SparksKeyboard)
     D:bool="Equal (Enum)"(A=@n1, B=SparksMicrophone)
     E:bool="Equal (Enum)"(A=@n1, B=SparksDrums)
    ) -> (ReturnValue))
  ]
 }
 Is Vehicle Cosmetic {
  path=VM_SelectedItemUpdated.K2Node_Composite_1.Is Vehicle Cosmetic
  nodes [
   n1 = Get ItemDefinition(self=Inputs())
   Outputs(ReturnValue="Equal (Enum)"(A=GetItemType(ItemDefinition=@n1) -> (ReturnValue=WorldItem), B=AthenaVehicleCosmetic))
  ]
 }
 Is Companion {
  path=VM_SelectedItemUpdated.K2Node_Composite_2.Is Companion
  nodes [
   n1 = Get ItemDefinition(self=Inputs())
   Outputs(ReturnValue="Equal (Enum)"(A=GetItemType(ItemDefinition=@n1) -> (ReturnValue=WorldItem), B=CosmeticMimosa))
  ]
 }
]
variables [
 CurrentView {type=EFortItemShopPdpView, default=OfferInspect}
 CustomISB {type=WBP_CustomBanner_Base_C, default=null}
]
widgets {
 info {inheritedWidgetCount=0, parentClass=&CommonUserWidget, rootWidgetClass=&VerticalBox, widgetCount=13}
 widgets [
  VerticalBox_All {widgetClassPath=&VerticalBox, widget=VerticalBox_All}
  SizeBox_ItemInfo {
   widgetClassPath=&SizeBox
   bIsVariable=true
   parent=VerticalBox_All
   widget=SizeBox_ItemInfo
   properties {bOverride_WidthOverride=true, widthOverride=604}
   slot {class=VerticalBoxSlot}
  }
  VerticalBox_ItemInfo {widgetClassPath=&VerticalBox, parent=SizeBox_ItemInfo, widget=VerticalBox_ItemInfo, slot {class=SizeBoxSlot}}
  Text_RatingBoardWarning {
   widgetClassPath=&CommonRichTextBlock
   bIsVariable=true
   parent=VerticalBox_ItemInfo
   widget=Text_RatingBoardWarning
   properties {
    autoWrapText=true
    bAutoCollapseWithEmptyText=true
    bOverrideDefaultStyle=true
    mobileTextBlockScale=1.3
    text="[PH] Rating Board Warning"
    textStyleSet=&/DT_ISB_TextStyles
    defaultTextStyleOverride {
     overflowPolicy=Clip
     transformPolicy=None
     colorAndOpacity {colorUseRule=UseColor_Specified, specifiedColor=<rgba 1 1 1 0.75>}
     font {
      bForceMonospaced=false
      bMaterialIsStencil=false
      fontMaterial=null
      fontObject=&HeadingNow
      letterSpacing=10
      monospacedWidth=1
      size=14.25
      skewAmount=0
      typefaceFontName=74 Regular
      outlineSettings {
       bApplyOutlineToDropShadows=false
       bMiteredCorners=false
       bSeparateFillAlpha=false
       outlineMaterial=null
       outlineSize=0
       outlineColor=<rgba 0 0 0 1>
      }
     }
     highlightColor {colorUseRule=UseColor_Specified, specifiedColor=<rgba 0 0 0 1>}
     highlightShape {
      bIsDynamicallyLoaded=false
      drawAs=Image
      imageType=NoImage
      mirroring=NoMirror
      resourceName=None
      resourceObject=null
      tiling=NoTile
      imageSize=<32 32>
      margin=<ltrb 0 0 0 0>
      outlineSettings {
       bUseBrushTransparency=false
       roundingType=HalfHeightRadius
       width=0
       color {colorUseRule=UseColor_Specified, specifiedColor=<rgba 0 0 0 0>}
       cornerRadii=<0 0 0 0>
      }
      tintColor {colorUseRule=UseColor_Specified, specifiedColor=<rgba 1 1 1 1>}
      uVRegion {bIsValid=false, max=<0 0>, min=<0 0>}
     }
     selectedBackgroundColor {colorUseRule=UseColor_Foreground, specifiedColor=<rgba 1 0 1 1>}
     shadowColorAndOpacity=<rgba 0 0 0 1>
     shadowOffset=<0 0>
     strikeBrush {
      bIsDynamicallyLoaded=false
      drawAs=Image
      imageType=NoImage
      mirroring=NoMirror
      resourceName=None
      resourceObject=null
      tiling=NoTile
      imageSize=<32 32>
      margin=<ltrb 0 0 0 0>
      outlineSettings {
       bUseBrushTransparency=false
       roundingType=HalfHeightRadius
       width=0
       color {colorUseRule=UseColor_Specified, specifiedColor=<rgba 0 0 0 0>}
       cornerRadii=<0 0 0 0>
      }
      tintColor {colorUseRule=UseColor_Specified, specifiedColor=<rgba 1 1 1 1>}
      uVRegion {bIsValid=false, max=<0 0>, min=<0 0>}
     }
     underlineBrush {
      bIsDynamicallyLoaded=false
      drawAs=Image
      imageType=NoImage
      mirroring=NoMirror
      resourceName=None
      resourceObject=null
      tiling=NoTile
      imageSize=<32 32>
      margin=<ltrb 0 0 0 0>
      outlineSettings {
       bUseBrushTransparency=false
       roundingType=HalfHeightRadius
       width=0
       color {colorUseRule=UseColor_Specified, specifiedColor=<rgba 0 0 0 0>}
       cornerRadii=<0 0 0 0>
      }
      tintColor {colorUseRule=UseColor_Specified, specifiedColor=<rgba 1 1 1 1>}
      uVRegion {bIsValid=false, max=<0 0>, min=<0 0>}
     }
    }
   }
   slot {class=VerticalBoxSlot, properties {padding=<ltrb 0 0 0 16>}}
  }
  Border_OfferPopupVisibility {
   widgetClassPath=&Border
   bIsVariable=true
   parent=VerticalBox_ItemInfo
   widget=Border_OfferPopupVisibility
   properties {
    background {
     bIsDynamicallyLoaded=false
     drawAs=NoDrawType
     imageType=NoImage
     mirroring=NoMirror
     resourceName=None
     resourceObject=null
     tiling=NoTile
     imageSize=<32 32>
     margin=<ltrb 0 0 0 0>
     outlineSettings {
      bUseBrushTransparency=false
      roundingType=HalfHeightRadius
      width=0
      color {colorUseRule=UseColor_Specified, specifiedColor=<rgba 0 0 0 0>}
      cornerRadii=<0 0 0 0>
     }
     tintColor {colorUseRule=UseColor_Specified, specifiedColor=<rgba 1 1 1 1>}
     uVRegion {bIsValid=false, max=<0 0>, min=<0 0>}
    }
    padding=<ltrb 0 0 0 16>
   }
   slot {class=VerticalBoxSlot}
  }
  Text_OfferPopup {
   widgetClassPath=&CommonRichTextBlock
   bIsVariable=true
   parent=Border_OfferPopupVisibility
   widget=Text_OfferPopup
   properties {
    autoWrapText=true
    bAutoCollapseWithEmptyText=true
    bOverrideDefaultStyle=true
    mobileTextBlockScale=1.3
    text="[PH] Use the <highlight_strong>BUILT-IN EMOTES</> to <highlight_strong>SWITCH STYLES</> in matches!"
    textStyleSet=&DT_ISB_TextStyles
    defaultTextStyleOverride {
     overflowPolicy=Clip
     transformPolicy=None
     colorAndOpacity {colorUseRule=UseColor_Specified, specifiedColor=<rgba 1 1 1 0.75>}
     font {
      bForceMonospaced=false
      bMaterialIsStencil=false
      fontMaterial=null
      fontObject=&HeadingNow
      letterSpacing=10
      monospacedWidth=1
      size=14.25
      skewAmount=0
      typefaceFontName=74 Regular
      outlineSettings {
       bApplyOutlineToDropShadows=false
       bMiteredCorners=false
       bSeparateFillAlpha=false
       outlineMaterial=null
       outlineSize=0
       outlineColor=<rgba 0 0 0 1>
      }
     }
     highlightColor {colorUseRule=UseColor_Specified, specifiedColor=<rgba 0 0 0 1>}
     highlightShape {
      bIsDynamicallyLoaded=false
      drawAs=Image
      imageType=NoImage
      mirroring=NoMirror
      resourceName=None
      resourceObject=null
      tiling=NoTile
      imageSize=<32 32>
      margin=<ltrb 0 0 0 0>
      outlineSettings {
       bUseBrushTransparency=false
       roundingType=HalfHeightRadius
       width=0
       color {colorUseRule=UseColor_Specified, specifiedColor=<rgba 0 0 0 0>}
       cornerRadii=<0 0 0 0>
      }
      tintColor {colorUseRule=UseColor_Specified, specifiedColor=<rgba 1 1 1 1>}
      uVRegion {bIsValid=false, max=<0 0>, min=<0 0>}
     }
     selectedBackgroundColor {colorUseRule=UseColor_Foreground, specifiedColor=<rgba 1 0 1 1>}
     shadowColorAndOpacity=<rgba 0 0 0 1>
     shadowOffset=<0 0>
     strikeBrush {
      bIsDynamicallyLoaded=false
      drawAs=Image
      imageType=NoImage
      mirroring=NoMirror
      resourceName=None
      resourceObject=null
      tiling=NoTile
      imageSize=<32 32>
      margin=<ltrb 0 0 0 0>
      outlineSettings {
       bUseBrushTransparency=false
       roundingType=HalfHeightRadius
       width=0
       color {colorUseRule=UseColor_Specified, specifiedColor=<rgba 0 0 0 0>}
       cornerRadii=<0 0 0 0>
      }
      tintColor {colorUseRule=UseColor_Specified, specifiedColor=<rgba 1 1 1 1>}
      uVRegion {bIsValid=false, max=<0 0>, min=<0 0>}
     }
     underlineBrush {
      bIsDynamicallyLoaded=false
      drawAs=Image
      imageType=NoImage
      mirroring=NoMirror
      resourceName=None
      resourceObject=null
      tiling=NoTile
      imageSize=<32 32>
      margin=<ltrb 0 0 0 0>
      outlineSettings {
       bUseBrushTransparency=false
       roundingType=HalfHeightRadius
       width=0
       color {colorUseRule=UseColor_Specified, specifiedColor=<rgba 0 0 0 0>}
       cornerRadii=<0 0 0 0>
      }
      tintColor {colorUseRule=UseColor_Specified, specifiedColor=<rgba 1 1 1 1>}
      uVRegion {bIsValid=false, max=<0 0>, min=<0 0>}
     }
    }
   }
   slot {class=BorderSlot, properties {padding=<ltrb 0 0 0 16>}}
  }
  Text_UnlockRequirement {
   widgetClassPath=&CommonRichTextBlock
   bIsVariable=true
   parent=VerticalBox_ItemInfo
   widget=Text_UnlockRequirement
   properties {
    autoWrapText=true
    bAutoCollapseWithEmptyText=true
    bOverrideDefaultStyle=true
    mobileTextBlockScale=1.3
    text="[ph] Complete 28 Nevermore Quests (Available with Week 4’s quest drop Feburary 17 2024 9:00AM ET)"
    textStyleSet=&DT_ISB_TextStyles
    defaultTextStyleOverride {
     overflowPolicy=Clip
     transformPolicy=None
     colorAndOpacity {colorUseRule=UseColor_Specified, specifiedColor=<rgba 1 1 1 0.75>}
     font {
      bForceMonospaced=false
      bMaterialIsStencil=false
      fontMaterial=null
      fontObject=&HeadingNow
      letterSpacing=10
      monospacedWidth=1
      size=14.25
      skewAmount=0
      typefaceFontName=74 Regular
      outlineSettings {
       bApplyOutlineToDropShadows=false
       bMiteredCorners=false
       bSeparateFillAlpha=false
       outlineMaterial=null
       outlineSize=0
       outlineColor=<rgba 0 0 0 1>
      }
     }
     highlightColor {colorUseRule=UseColor_Specified, specifiedColor=<rgba 0 0 0 1>}
     highlightShape {
      bIsDynamicallyLoaded=false
      drawAs=Image
      imageType=NoImage
      mirroring=NoMirror
      resourceName=None
      resourceObject=null
      tiling=NoTile
      imageSize=<32 32>
      margin=<ltrb 0 0 0 0>
      outlineSettings {
       bUseBrushTransparency=false
       roundingType=HalfHeightRadius
       width=0
       color {colorUseRule=UseColor_Specified, specifiedColor=<rgba 0 0 0 0>}
       cornerRadii=<0 0 0 0>
      }
      tintColor {colorUseRule=UseColor_Specified, specifiedColor=<rgba 1 1 1 1>}
      uVRegion {bIsValid=false, max=<0 0>, min=<0 0>}
     }
     selectedBackgroundColor {colorUseRule=UseColor_Foreground, specifiedColor=<rgba 1 0 1 1>}
     shadowColorAndOpacity=<rgba 0 0 0 1>
     shadowOffset=<0 0>
     strikeBrush {
      bIsDynamicallyLoaded=false
      drawAs=Image
      imageType=NoImage
      mirroring=NoMirror
      resourceName=None
      resourceObject=null
      tiling=NoTile
      imageSize=<32 32>
      margin=<ltrb 0 0 0 0>
      outlineSettings {
       bUseBrushTransparency=false
       roundingType=HalfHeightRadius
       width=0
       color {colorUseRule=UseColor_Specified, specifiedColor=<rgba 0 0 0 0>}
       cornerRadii=<0 0 0 0>
      }
      tintColor {colorUseRule=UseColor_Specified, specifiedColor=<rgba 1 1 1 1>}
      uVRegion {bIsValid=false, max=<0 0>, min=<0 0>}
     }
     underlineBrush {
      bIsDynamicallyLoaded=false
      drawAs=Image
      imageType=NoImage
      mirroring=NoMirror
      resourceName=None
      resourceObject=null
      tiling=NoTile
      imageSize=<32 32>
      margin=<ltrb 0 0 0 0>
      outlineSettings {
       bUseBrushTransparency=false
       roundingType=HalfHeightRadius
       width=0
       color {colorUseRule=UseColor_Specified, specifiedColor=<rgba 0 0 0 0>}
       cornerRadii=<0 0 0 0>
      }
      tintColor {colorUseRule=UseColor_Specified, specifiedColor=<rgba 1 1 1 1>}
      uVRegion {bIsValid=false, max=<0 0>, min=<0 0>}
     }
    }
   }
   slot {class=VerticalBoxSlot, properties {padding=<ltrb 0 0 0 16>}}
  }
  HB_CrossEntitlement {
   widgetClassPath=&HorizontalBox
   bIsVariable=true
   parent=VerticalBox_ItemInfo
   widget=HB_CrossEntitlement
   slot {class=VerticalBoxSlot, properties {padding=<ltrb 0 0 0 16>}}
  }
  Image_CrossEntitlement {
   widgetClassPath=&Image
   parent=HB_CrossEntitlement
   widget=Image_CrossEntitlement
   properties {brush {
     bIsDynamicallyLoaded=false
     drawAs=Image
     imageType=FullColor
     mirroring=NoMirror
     resourceName=None
     resourceObject=/ItemShop/MultiProduct/Art/T_UI_RocketLeagueIcon
     tiling=NoTile
     imageSize=<64 64>
     margin=<ltrb 0 0 0 0>
     outlineSettings {
      bUseBrushTransparency=false
      roundingType=HalfHeightRadius
      width=0
      color {colorUseRule=UseColor_Specified, specifiedColor=<rgba 0 0 0 0>}
      cornerRadii=<0 0 0 0>
     }
     tintColor {colorUseRule=UseColor_Specified, specifiedColor=<rgba 1 1 1 1>}
     uVRegion {bIsValid=false, max=<0 0>, min=<0 0>}
    }}
   slot {class=HorizontalBoxSlot, properties {horizontalAlignment=HAlign_Left, verticalAlignment=VAlign_Center, padding=<ltrb 0 0 10 0>}}
  }
  Text_CrossEntitlement {
   widgetClassPath=&CommonRichTextBlock
   bIsVariable=true
   parent=HB_CrossEntitlement
   widget=Text_CrossEntitlement
   properties {
    autoWrapText=true
    bOverrideDefaultStyle=true
    mobileTextBlockScale=1.3
    text="Purchasing this item will also make it available in <highlight_strong>Rocket League.</>"
    textStyleSet=&DT_ISB_TextStyles
    defaultTextStyleOverride {
     overflowPolicy=Clip
     transformPolicy=None
     colorAndOpacity {colorUseRule=UseColor_Specified, specifiedColor=<rgba 1 1 1 0.75>}
     font {
      bForceMonospaced=false
      bMaterialIsStencil=false
      fontMaterial=null
      fontObject=&HeadingNow
      letterSpacing=10
      monospacedWidth=1
      size=14.25
      skewAmount=0
      typefaceFontName=74 Regular
      outlineSettings {
       bApplyOutlineToDropShadows=false
       bMiteredCorners=false
       bSeparateFillAlpha=false
       outlineMaterial=null
       outlineSize=0
       outlineColor=<rgba 0 0 0 1>
      }
     }
     highlightColor {colorUseRule=UseColor_Specified, specifiedColor=<rgba 0 0 0 1>}
     highlightShape {
      bIsDynamicallyLoaded=false
      drawAs=Image
      imageType=NoImage
      mirroring=NoMirror
      resourceName=None
      resourceObject=null
      tiling=NoTile
      imageSize=<32 32>
      margin=<ltrb 0 0 0 0>
      outlineSettings {
       bUseBrushTransparency=false
       roundingType=HalfHeightRadius
       width=0
       color {colorUseRule=UseColor_Specified, specifiedColor=<rgba 0 0 0 0>}
       cornerRadii=<0 0 0 0>
      }
      tintColor {colorUseRule=UseColor_Specified, specifiedColor=<rgba 1 1 1 1>}
      uVRegion {bIsValid=false, max=<0 0>, min=<0 0>}
     }
     selectedBackgroundColor {colorUseRule=UseColor_Foreground, specifiedColor=<rgba 1 0 1 1>}
     shadowColorAndOpacity=<rgba 0 0 0 1>
     shadowOffset=<0 0>
     strikeBrush {
      bIsDynamicallyLoaded=false
      drawAs=Image
      imageType=NoImage
      mirroring=NoMirror
      resourceName=None
      resourceObject=null
      tiling=NoTile
      imageSize=<32 32>
      margin=<ltrb 0 0 0 0>
      outlineSettings {
       bUseBrushTransparency=false
       roundingType=HalfHeightRadius
       width=0
       color {colorUseRule=UseColor_Specified, specifiedColor=<rgba 0 0 0 0>}
       cornerRadii=<0 0 0 0>
      }
      tintColor {colorUseRule=UseColor_Specified, specifiedColor=<rgba 1 1 1 1>}
      uVRegion {bIsValid=false, max=<0 0>, min=<0 0>}
     }
     underlineBrush {
      bIsDynamicallyLoaded=false
      drawAs=Image
      imageType=NoImage
      mirroring=NoMirror
      resourceName=None
      resourceObject=null
      tiling=NoTile
      imageSize=<32 32>
      margin=<ltrb 0 0 0 0>
      outlineSettings {
       bUseBrushTransparency=false
       roundingType=HalfHeightRadius
       width=0
       color {colorUseRule=UseColor_Specified, specifiedColor=<rgba 0 0 0 0>}
       cornerRadii=<0 0 0 0>
      }
      tintColor {colorUseRule=UseColor_Specified, specifiedColor=<rgba 1 1 1 1>}
      uVRegion {bIsValid=false, max=<0 0>, min=<0 0>}
     }
    }
   }
   slot {class=HorizontalBoxSlot, properties {verticalAlignment=VAlign_Center, size {sizeRule=Fill, value=1}}}
  }
  Text_Mimosa {
   widgetClassPath=&CommonTextBlock
   bIsVariable=true
   parent=VerticalBox_ItemInfo
   widget=Text_Mimosa
   properties {
    autoWrapText=true
    mobileFontSizeMultiplier=1.3
    text="Sidekicks will react to the world around you. They won't detect any enemies or items though!\r\nIn Battle Royale, you can only see your team's Sidekicks. Visibility varies in other Fortnite experiences."
    colorAndOpacity {colorUseRule=UseColor_Specified, specifiedColor=<rgba 1 1 1 0.75>}
    font {
     bForceMonospaced=false
     bMaterialIsStencil=false
     fontMaterial=null
     fontObject=&HeadingNow
     letterSpacing=10
     monospacedWidth=1
     size=14.25
     skewAmount=0
     typefaceFontName=74 Regular
     outlineSettings {
      bApplyOutlineToDropShadows=false
      bMiteredCorners=false
      bSeparateFillAlpha=false
      outlineMaterial=null
      outlineSize=0
      outlineColor=<rgba 0 0 0 1>
     }
    }
    shadowOffset=<0 0>
   }
   slot {class=VerticalBoxSlot, properties {padding=<ltrb 0 0 0 16>}}
  }
  Text_Instrument {
   widgetClassPath=&CommonRichTextBlock
   bIsVariable=true
   parent=VerticalBox_ItemInfo
   widget=Text_Instrument
   properties {
    autoWrapText=true
    bOverrideDefaultStyle=true
    mobileTextBlockScale=1.3
    text="This <highlight_strong>INSTRUMENT</> is now equippable as a <highlight_strong>PICKAXE</> and <highlight_strong>BACK BLING</> in the Locker."
    textStyleSet=&DT_ISB_TextStyles
    visibility=Collapsed
    defaultTextStyleOverride {
     overflowPolicy=Clip
     transformPolicy=None
     colorAndOpacity {colorUseRule=UseColor_Specified, specifiedColor=<rgba 1 1 1 0.75>}
     font {
      bForceMonospaced=false
      bMaterialIsStencil=false
      fontMaterial=null
      fontObject=&HeadingNow
      letterSpacing=10
      monospacedWidth=1
      size=14.25
      skewAmount=0
      typefaceFontName=74 Regular
      outlineSettings {
       bApplyOutlineToDropShadows=false
       bMiteredCorners=false
       bSeparateFillAlpha=false
       outlineMaterial=null
       outlineSize=0
       outlineColor=<rgba 0 0 0 1>
      }
     }
     highlightColor {colorUseRule=UseColor_Specified, specifiedColor=<rgba 0 0 0 1>}
     highlightShape {
      bIsDynamicallyLoaded=false
      drawAs=Image
      imageType=NoImage
      mirroring=NoMirror
      resourceName=None
      resourceObject=null
      tiling=NoTile
      imageSize=<32 32>
      margin=<ltrb 0 0 0 0>
      outlineSettings {
       bUseBrushTransparency=false
       roundingType=HalfHeightRadius
       width=0
       color {colorUseRule=UseColor_Specified, specifiedColor=<rgba 0 0 0 0>}
       cornerRadii=<0 0 0 0>
      }
      tintColor {colorUseRule=UseColor_Specified, specifiedColor=<rgba 1 1 1 1>}
      uVRegion {bIsValid=false, max=<0 0>, min=<0 0>}
     }
     selectedBackgroundColor {colorUseRule=UseColor_Foreground, specifiedColor=<rgba 1 0 1 1>}
     shadowColorAndOpacity=<rgba 0 0 0 1>
     shadowOffset=<0 0>
     strikeBrush {
      bIsDynamicallyLoaded=false
      drawAs=Image
      imageType=NoImage
      mirroring=NoMirror
      resourceName=None
      resourceObject=null
      tiling=NoTile
      imageSize=<32 32>
      margin=<ltrb 0 0 0 0>
      outlineSettings {
       bUseBrushTransparency=false
       roundingType=HalfHeightRadius
       width=0
       color {colorUseRule=UseColor_Specified, specifiedColor=<rgba 0 0 0 0>}
       cornerRadii=<0 0 0 0>
      }
      tintColor {colorUseRule=UseColor_Specified, specifiedColor=<rgba 1 1 1 1>}
      uVRegion {bIsValid=false, max=<0 0>, min=<0 0>}
     }
     underlineBrush {
      bIsDynamicallyLoaded=false
      drawAs=Image
      imageType=NoImage
      mirroring=NoMirror
      resourceName=None
      resourceObject=null
      tiling=NoTile
      imageSize=<32 32>
      margin=<ltrb 0 0 0 0>
      outlineSettings {
       bUseBrushTransparency=false
       roundingType=HalfHeightRadius
       width=0
       color {colorUseRule=UseColor_Specified, specifiedColor=<rgba 0 0 0 0>}
       cornerRadii=<0 0 0 0>
      }
      tintColor {colorUseRule=UseColor_Specified, specifiedColor=<rgba 1 1 1 1>}
      uVRegion {bIsValid=false, max=<0 0>, min=<0 0>}
     }
    }
   }
   slot {class=VerticalBoxSlot, properties {padding=<ltrb 0 0 0 16>}}
  }
  NamedSlot_CustomISB {
   widgetClassPath=&NamedSlot
   bIsVariable=true
   parent=VerticalBox_All
   widget=NamedSlot_CustomISB
   slot {class=VerticalBoxSlot, properties {horizontalAlignment=HAlign_Left}}
  }
 ]
}
