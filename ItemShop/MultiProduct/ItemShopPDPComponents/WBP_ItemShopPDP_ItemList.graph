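# Item list for the Item Shop PDP, extending WBP_UIKit_ItemList with a one-time ALC coachmark.
# When an offer is set and shop_has_seen_alc is false, it finds the first granted item with a related offer and scrolls to its card.
# After a short delay it activates the coachmark, offsets it horizontally from card geometry and scroll, then saves the seen flag.
# PDP deactivation or switching to ItemDetails clears it.
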WidgetBlueprint /ItemShop/MultiProduct/ItemShopPDPComponents/WBP_ItemShopPDP_ItemList
graphs [
 EventGraph {nodes [
   n1 = CreateALCCoachmark!() -> (OutputDelegate, then, ItemCard:Widget Object Reference) {is=title K2Node_CustomEvent}
   > n2 = Set ALC CoachMark ItemCard(ALC CoachMark ItemCard:Widget Object Reference=@n1.ItemCard) -> (Output_Get:Widget Object Reference)
   > ScrollToCard!(execute, self:WBP UIKit Item List Object Reference, WidgetToFind:Widget Object Reference=@n2, AnimateScroll:bool=false) -> (then)
   > Delay(Duration=0.5)
   > ClearChildren(self=Get Attachment_ALC_Coachmark() -> (Attachment_ALC_Coachmark:Common Visual Attachment Object Reference))
   > n3 = Create Widget!(
    execute:Exec
    Class:User Widget Class Reference=/ItemShop/MultiProduct/WBP_ItemShop_Coachmark_ALC.WBP_ItemShop_Coachmark_ALC_C
    OwningPlayer:Player Controller Object Reference
   ) -> (then:Exec, ReturnValue:WBP Item Shop Coachmark ALC Object Reference)
   > AddChild(self=Get Attachment_ALC_Coachmark(), Content=@n3)
   > n4 = Cast To WBP_UIKit_CoachMark!(execute, Object=@n3) -> (then, CastFailed, AsWBP UIKit Coach Mark:WBP UIKit Coach Mark Object Reference) {is=title K2Node_DynamicCast}
   > n5 = Set ALC Coachmark(ALC Coachmark:WBP UIKit Coach Mark Object Reference=@n4) -> (Output_Get:WBP UIKit Coach Mark Object Reference)
   > ActivateWidget(self=@n5)
   n6 = GetLocalTopLeft!(Geometry:"Geometry Structure (by ref)"=GetCachedGeometry!(self:Widget Object Reference=Get ALC CoachMark ItemCard() -> (ALC CoachMark ItemCard:Widget Object Reference)) -> (ReturnValue:"Geometry Structure (by ref)")) -> (ReturnValue_X:double, ReturnValue_Y:double)
   n7 = GetScrollOffset!(self:Scroll Box Object Reference=Get Scrollbox_Items() -> (Scrollbox_Items:Athena Scroll Box Object Reference)) -> (ReturnValue:float)
   n8 = Get Spacer_StartPadding() -> (Spacer_StartPadding:Spacer Object Reference)
   n9 = IsVisible(self=@n8)
   n10 = Get!(self:Spacer Object Reference=@n8) -> (Size_X:double, Size_Y:double)
   > Set!(
    execute
    Padding_Left:float=float + float!(A=float - float(A=@n6, B=@n7), B=24, C:double=Select(Option 0:double=0, Option 1:double=@n10, Index=@n9) -> (ReturnValue:double)) -> (ReturnValue)
    Padding_Top:float=0
    Padding_Right:float=0
    Padding_Bottom:float=0
    self:Overlay Slot Object Reference=SlotAsOverlaySlot(Widget=Get Attachment_ALC_Coachmark())
   ) -> (then, Output_Get:Margin Structure)
   > SetUserSettingBool!(execute, Name:String=shop_has_seen_alc, bValue:bool=true) -> (then)
   Event OnPdpActivated() {is=title K2Node_Event}
   > Activation(self:WBP UIKit Item List Object Reference)
   Event OnPdpDeactivated() {is=title K2Node_Event}
   > Deactivation(self:WBP UIKit Item List Object Reference)
   > ClearChildren(self=Get Attachment_ALC_Coachmark())
   n11 = Event OnOfferSet!() -> (
    OutputDelegate
    then
    OfferVM:Fort Item Shop Offer VM Object Reference
    ParentOfferVM:Fort Item Shop Offer VM Object Reference
    OfferGroupVM:Fort Item Shop Offer Group VM Object Reference
   ) {is=title K2Node_Event}
   > n12 = Set FortItemShopOfferVM(FortItemShopOfferVM:Fort Item Shop Offer VM Object Reference=@n11.OfferVM) -> (Output_Get:Fort Item Shop Offer VM Object Reference)
   > UpdateALCCoachmark!(
    execute
    self
    GrantedItems:Array of Fort Item VM Object References=GetGrantedItems!(
     self
     GrantedItems:Array of Fort Item VM Object References=Get GrantedItems!(self:Fort Item Shop Offer VM Object Reference=@n12) -> (GrantedItems:Array of Fort Item VM Object References)
     ItemShopVM:Fort Item Shop VM Object Reference=Get FortItemShopVM() -> (FortItemShopVM:Fort Item Shop VM Object Reference)
    ) -> (ReturnValue:Array of Fort Item VM Object References)
   ) -> (then)
   n13 = Event OnViewChanged!() -> (OutputDelegate, then, NewView:EFortItemShopPdpView Enum=OfferInspect) {is=title K2Node_Event}
   > Branch(Condition="Equal (Enum)"(A:EFortItemShopPdpView Enum=@n13.NewView, B:EFortItemShopPdpView Enum=ItemDetails))
   > n14 = Get!(execute:Exec) -> (then:Exec, else:Exec, ALC Coachmark:WBP UIKit Coach Mark Object Reference)
   > DeactivateWidget(self=@n14)
   > ClearChildren(self=Get Attachment_ALC_Coachmark())
  ]}
 GetGrantedItems {nodes [
   n1 = GetGrantedItems!() -> (then, GrantedItems:Array of Fort Item VM Object References, ItemShopVM:Fort Item Shop VM Object Reference) {is=title K2Node_FunctionEntry}
   > Return Node!(
    execute
    ReturnValue:Array of Fort Item VM Object References=Select(
     Option 0:Array of Fort Item VM Object References=Make Array!() -> (Array:Array of Fort Item VM Object References)
     Option 1:Array of Fort Item VM Object References=@n1
     Index=IsValid(Object=@n1.ItemShopVM)
    ) -> (ReturnValue:Array of Fort Item VM Object References)
   )
  ]}
 UpdateALCCoachmark {nodes [
   n1 = UpdateALCCoachmark!() -> (then, GrantedItems:Array of Fort Item VM Object References) {is=title K2Node_FunctionEntry}
   > Set Items(Items:Array of Fort Item VM Object References=@n1) -> (Output_Get:Array of Fort Item VM Object References)
   > n2 = Branch(Condition=GetConfigDataBoolWithDefault!(EntryName:String=shop_has_seen_alc, bDefaultValue:bool=false) -> (ReturnValue:bool))
   n3 = For Each Loop(Exec=@n2.else, Array:Array of Fort Item VM Object References=Get Items() -> (Items:Array of Fort Item VM Object References)) -> (Array Element:Fort Item VM Object Reference)
   > Is Valid(InputObject=FindRelatedOfferForItem!(
     self:Fort Item Shop VM Object Reference=Get FortItemShopVM()
     BaseOffer:Fort Item Shop Offer VM Object Reference=Get FortItemShopOfferVM() -> (FortItemShopOfferVM:Fort Item Shop Offer VM Object Reference)
     Item:Fort Item VM Object Reference=@n3
    ) -> (ReturnValue:Fort Item Shop Offer VM Object Reference))
   > CreateALCCoachmark!(
    execute
    self
    ItemCard:Widget Object Reference=GetWidgetAtIndex!(self:WBP UIKit Item List Object Reference, InIndex:int=@n3.Array Index) -> (OutWIdget:Widget Object Reference)
   ) -> (then) {is=title K2Node_CallFunction}
   > Return Node!(execute)
  ]}
]
variables [
 ALC Coachmark {type=WBP_UIKit_CoachMark_C, default=null}
 ALC CoachMark ItemCard {type=Widget, default=null}
]
widgets {
 info {
  inheritedWidgetCount=0
  namedSlotCount=1
  parentClass=/FortUILibrary/UIKit/Items/WBP_UIKit_ItemList.WBP_UIKit_ItemList_C
  rootWidgetClass=null
  widgetCount=2
 }
 widgets [
  Overlay_50 {widgetClassPath=&Overlay, widget=Overlay_50}
  Attachment_ALC_Coachmark {
   widgetClassPath=&CommonVisualAttachment
   bIsVariable=true
   parent=Overlay_50
   widget=Attachment_ALC_Coachmark
   properties {contentAnchor=<0 1>}
   slot {class=OverlaySlot}
  }
 ]
}
