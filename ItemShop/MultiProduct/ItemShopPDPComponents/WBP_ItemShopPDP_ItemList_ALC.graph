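# Item shop product detail page widget listing an offer's granted items, visible only in ItemDetails view with multiple items.
# A switcher shows companion cards when the first item or nothing is previewed, otherwise a "Not sold separately" panel naming the primary item.
# PopulateItemList puts unique item types first, skips the first item, and caps cards, making the last a non-interactable overflow count.
# Card presses log analytics and broadcast CardPressed.
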WidgetBlueprint /ItemShop/MultiProduct/ItemShopPDPComponents/WBP_ItemShopPDP_ItemList_ALC
graphs [
 ALCVisibilityChanged {nodes [ALCVisibilityChanged!() -> (then, Visibility:ESlateVisibility Enum=Visible) {is=title K2Node_FunctionEntry}]}
 CardPressed {nodes [CardPressed!() -> (then, Button:Widget Object Reference, FortItemVM:Fort Item VM Object Reference) {is=title K2Node_FunctionEntry}]}
 EventGraph {nodes [
   n1 = OnCardPressed!() -> (OutputDelegate, then, ItemCard:WBP UIKit Item Card Base Object Reference) {is=title K2Node_CustomEvent}
   n2 = Get FortItemShopOfferVM() -> (FortItemShopOfferVM:Fort Item Shop Offer VM Object Reference)
   > Fire Event Element Interacted(
    Element=@n1.ItemCard
    AdditionalInfo="Add String (Non-Localized)"!(
     Info:"Flow Analytics Additional Info Structure (by ref)"=Make Flow Analytics Additional Info!() -> (FlowAnalyticsAdditionalInfo:Flow Analytics Additional Info Structure) {is=title K2Node_MakeStruct}
     Key:String=Display Asset Name
     Value:String=GetDisplayAssetName!(self:Fort Item Shop Offer VM Object Reference=@n2) -> (ReturnValue:String)
    ) -> (ReturnValue:Flow Analytics Additional Info Structure)
   )
   > n3 = Cast To WBP_ItemShop_ItemCard_ALC!(execute, Object=@n1.ItemCard) -> (then, CastFailed, AsWBP Item Shop Item Card ALC:WBP Item Shop Item Card ALC Object Reference) {is=title K2Node_DynamicCast}
   > Call Card Pressed!(
    execute:Exec
    self:WBP Item Shop PDP Item List ALC Object Reference
    Button:Widget Object Reference=@n3
    FortItemVM:Fort Item VM Object Reference=Get FortItemVM!(self:WBP Item Shop Item Card ALC Object Reference=@n3) -> (FortItemVM:Fort Item VM Object Reference)
   ) -> (then:Exec) {is=title K2Node_CallDelegate}
   n4 = OnALCVisibilityChanged!() -> (OutputDelegate, then, InVisibility:ESlateVisibility Enum=Visible) {is=title K2Node_CustomEvent}
   > Call ALCVisibility Changed!(execute:Exec, self:WBP Item Shop PDP Item List ALC Object Reference, Visibility:ESlateVisibility Enum=@n4.InVisibility) -> (then:Exec) {is=title K2Node_CallDelegate}
   Event PreConstruct()
   > Bind Event to On Visibility Changed(self:User Widget Object Reference, Delegate=@n4) {is=title K2Node_AddDelegate}
   n5 = Event OnOfferSet!() -> (
    OutputDelegate
    then
    OfferVM:Fort Item Shop Offer VM Object Reference
    ParentOfferVM:Fort Item Shop Offer VM Object Reference
    OfferGroupVM:Fort Item Shop Offer Group VM Object Reference
   ) {is=title K2Node_Event}
   > Set FortItemShopOfferVM(FortItemShopOfferVM:Fort Item Shop Offer VM Object Reference=@n5.OfferVM) -> (Output_Get:Fort Item Shop Offer VM Object Reference)
   n6 = Event OnViewChanged!() -> (OutputDelegate, then, NewView:EFortItemShopPdpView Enum=OfferInspect) {is=title K2Node_Event}
   > Set with Broadcast CurrentView!(execute, CurrentView:EFortItemShopPdpView Enum=@n6.NewView) -> (then, Output_Get:EFortItemShopPdpView Enum=OfferInspect)
  ]}
 GetActiveContentWidget {nodes [
   n1 = GetActiveContentWidget!() -> (then, GrantedItems:Array of Fort Item VM Object References, PreviewedItem:Item Shop Preview Data Object Reference) {is=title K2Node_FunctionEntry}
   > n2 = Is Valid(InputObject=@n1.PreviewedItem)
   Return Node!(execute=@n2.Is Not Valid, ReturnValue:Widget Object Reference=Get Grid_Available() -> (Grid_Available:Grid Panel Object Reference))
   n3 = FindItemByDefinition!(
    ItemArray:Array of Fort Item VM Object References=@n1
    Item:Fort Item VM Object Reference=Get ItemVM!(self:Item Shop Preview Data Object Reference=@n1.PreviewedItem) -> (ItemVM:Fort Item VM Object Reference)
   ) -> (FoundIndex:int=0)
   > Return Node(ReturnValue=Select(
     Option 0:Widget Object Reference=Get Grid_Unavailable() -> (Grid_Unavailable:Grid Panel Object Reference)
     Option 1:Widget Object Reference=Get Grid_Available()
     Index="Equal (Integer)"(A=@n3)
    ) -> (ReturnValue:Widget Object Reference))
  ]}
 GetDesiredFocusTarget {nodes [
   GetDesiredFocusTarget() {is=title K2Node_FunctionEntry}
   n1 = Get EntryBox_Items() -> (EntryBox_Items:Dynamic Entry Box Object Reference)
   > Return Node!(
    execute
    FocusTarget:Widget Object Reference="Get (a copy)"(Array:Array of User Widget Object References=GetAllEntries!(self:Dynamic Entry Box Base Object Reference=@n1) -> (ReturnValue:Array of User Widget Object References)) -> (Output:User Widget Object Reference)
   )
  ]}
 GetItemsToShow {nodes [
   n1 = GetItemsToShow!() -> (then, GrantedItems:Array of Fort Item VM Object References, CurrentView:EFortItemShopPdpView Enum=OfferInspect) {is=title K2Node_FunctionEntry}
   n2 = "Equal (Enum)"(A:EFortItemShopPdpView Enum=@n1.CurrentView, B:EFortItemShopPdpView Enum=ItemDetails)
   n3 = "integer > integer"(A=Length(TargetArray:Array of Fort Item VM Object References=@n1), B=1)
   > Return Node!(
    execute
    ReturnValue:Array of Fort Item VM Object References=Select(
     Option 0:Array of Fort Item VM Object References=Make Array!() -> (Array:Array of Fort Item VM Object References)
     Option 1:Array of Fort Item VM Object References=Get GrantedItems() -> (GrantedItems:Array of Fort Item VM Object References)
     Index=AND Boolean(A=@n3, B=@n2)
    ) -> (ReturnValue:Array of Fort Item VM Object References)
   )
  ]}
 GetPrimaryItemName {nodes [
   n1 = GetPrimaryItemName!() -> (then, GrantedItems:Array of Fort Item VM Object References) {is=title K2Node_FunctionEntry}
   > n2 = Branch(Condition=Is Empty!(TargetArray:Array of Fort Item VM Object References=@n1) -> (ReturnValue:bool))
   > Return Node(ReturnValue:Text)
   n3 = Get ItemDetails!(self:Fort Item VM Object Reference="Get (a copy)"(Array:Array of Fort Item VM Object References=Get GrantedItems()) -> (Output:Fort Item VM Object Reference)) -> (ItemDetails:Fort Item Details VM Object Reference)
   n4 = GetItemStyleText!(self:Fort Item Details VM Object Reference=@n3) -> (ReturnValue:Text)
   n5 = TextIsEmpty(InText=@n4)
   Return Node(
    execute=@n2.else
    ReturnValue=Format Text!(
     Format=Select(Option 0:Text="LOCTEXT(\"{Name} {Category} - {Style}\")", Option 1:Text="LOCTEXT(\"{Name} {Category}\")", Index=@n5) -> (ReturnValue:Text)
     Name:Text=GetDisplayName!(self:Fort Item Details VM Object Reference=@n3) -> (ReturnValue:Text)
     Category:Text=GetItemCategoryName!(self:Fort Item Details VM Object Reference=@n3) -> (ReturnValue:Text)
     Style:Text=@n4
    ) -> (Result)
   )
  ]}
 GetWidgetVisibility {nodes [
   n1 = GetWidgetVisibility!() -> (then, GrantedItems:Array of Fort Item VM Object References, CurrentView:EFortItemShopPdpView Enum=OfferInspect) {is=title K2Node_FunctionEntry}
   n2 = "Equal (Enum)"(A=@n1.CurrentView, B=ItemDetails)
   n3 = "integer > integer"(A=Length(TargetArray=@n1), B=1)
   > Return Node(ReturnValue:ESlateVisibility Enum=Select(Option 0:ESlateVisibility Enum=Collapsed, Option 1:ESlateVisibility Enum=SelfHitTestInvisible, Index=AND Boolean(A=@n3, B=@n2)) -> (ReturnValue:ESlateVisibility Enum))
  ]}
 PopulateItemList {nodes [
   n1 = PopulateItemList!() -> (then, GrantedItems:Array of Fort Item VM Object References) {is=title K2Node_FunctionEntry}
   > Branch(Condition="integer > integer"(A=Length(TargetArray=@n1), B=1))
   > n2 = Sequence()
   > n3 = For Each Loop(Array:Array of Fort Item VM Object References=Get GrantedItems()) -> (Array Element:Fort Item VM Object Reference)
   Append Array!(
    execute=@n3.Completed
    TargetArray:Array of Fort Item VM Object References=Get Items() -> (Items:Array of Fort Item VM Object References)
    SourceArray:Array of Fort Item VM Object References=Get DupeItems() -> (DupeItems:Array of Fort Item VM Object References)
   ) -> (then)
   n4 = GetItemType!(ItemDefinition:Item Definition Base Object Reference=Get ItemDefinition!(self:Fort Item VM Object Reference=@n3) -> (ItemDefinition:Fort Item Definition Object Reference)) -> (ReturnValue:EFortItemType Enum=WorldItem)
   n5 = Branch(
    execute=@n3.LoopBody
    Condition=Contains Item!(
     TargetSet:Set of EFortItemType Enums=Get UniqueTypes() -> (UniqueTypes:Set of EFortItemType Enums=WorldItem)
     ItemToFind:"EFortItemType Enum (by ref)"=@n4
    ) -> (ReturnValue:bool) {is=title K2Node_CallFunction}
   )
   > Add(TargetArray:Array of Fort Item VM Object References=Get DupeItems(), NewItem:"Fort Item VM Object Reference (by ref)"=@n3) {is=title K2Node_CallArrayFunction}
   Add!(
    execute=@n5.else
    TargetSet:Set of EFortItemType Enums=Get UniqueTypes() -> (UniqueTypes=WorldItem)
    NewItem:"EFortItemType Enum (by ref)"=@n4
   ) -> (then) {is=title K2Node_CallFunction}
   > Add(TargetArray=Get Items(), NewItem=@n3) {is=title K2Node_CallArrayFunction}
   Reset!(execute=@n2.then_1, self:Dynamic Entry Box Object Reference=Get EntryBox_Items(), bDeleteWidgets:bool=false) -> (then)
   > n6 = For Each Loop(Array=Get Items())
   > n7 = Set LoopIndex(LoopIndex:int=@n6.Array Index) -> (Output_Get:int=0)
   n8 = Get MaxCards() -> (MaxCards:int=0)
   > Branch(Condition=AND Boolean(A="integer > integer"(A=@n7), B="integer <= integer"(A=@n7, B=@n8)))
   > n9 = Create Entry of Class!(
    execute
    self:Dynamic Entry Box Object Reference=Get EntryBox_Items()
    EntryClass:User Widget Class Reference=/ItemShop/MultiProduct/WBP_ItemShop_ItemCard_ALC.WBP_ItemShop_ItemCard_ALC_C
   ) -> (then, ReturnValue:WBP Item Shop Item Card ALC Object Reference)
   > n10 = Set LoopCard(LoopCard:WBP Item Shop Item Card ALC Object Reference=@n9) -> (Output_Get:WBP Item Shop Item Card ALC Object Reference)
   > SetViewModelByClass!(execute, Widget:User Widget Object Reference=@n10, ViewModel:Notify Field Value Changed Interface=@n6) -> (then)
   n11 = Get LoopIndex() -> (LoopIndex:int=0)
   n12 = Get Items()
   n13 = AND Boolean(
    A="Equal (Integer)"(A=Get MaxCards() -> (MaxCards=0), B=@n11)
    B="integer < integer"(A=@n11, B=Last Index!(TargetArray:Array of Fort Item VM Object References=@n12) -> (ReturnValue=0))
   )
   > SetIsInteractable!(
    execute
    self:WBP Item Shop Item Card ALC Object Reference=Get LoopCard() -> (LoopCard:WBP Item Shop Item Card ALC Object Reference)
    IsInteractable:bool=InvertBool!(bValue:bool=@n13) -> (ReturnValue:bool)
   ) -> (then)
   > n14 = Branch(Condition=@n13)
   n15 = Get Items()
   n16 = Get LoopIndex() -> (LoopIndex=0)
   > SetCount!(
    execute
    self:WBP Item Shop Item Card ALC Object Reference=Get LoopCard()
    Count:int=int - int(A=Length(TargetArray=@n15), B=int + int(A=@n16))
   ) -> (then)
   SetCount(execute=@n14.else, self=Get LoopCard(), Count=0)
   n17 = Create Event()
   n18 = Get LoopCard()
   > Unbind Event from On Card Pressed(self:WBP UIKit Item Card Base Object Reference=@n18, Delegate=@n17) {is=title K2Node_RemoveDelegate}
   > Bind Event to On Card Pressed(self:WBP UIKit Item Card Base Object Reference=@n18, Delegate=@n17) {is=title K2Node_AddDelegate}
  ]}
 PopulatePrimaryItemCard {nodes [
   n1 = PopulatePrimaryItemCard!() -> (then, GrantedItems:Array of Fort Item VM Object References) {is=title K2Node_FunctionEntry}
   > n2 = Branch(Condition=Is Empty(TargetArray=@n1))
   Set FortItemVM!(
    execute=@n2.else
    FortItemVM:Fort Item VM Object Reference="Get (a copy)"(Array=@n1)
    self:WBP Item Shop Item Card ALC Object Reference=Get ItemCard_Unavailable() -> (ItemCard_Unavailable:WBP Item Shop Item Card ALC Object Reference)
   ) -> (then, Output_Get:Fort Item VM Object Reference)
  ]}
]
variables [
 ALCVisibilityChanged {}
 CardPressed {}
 CurrentView {type=EFortItemShopPdpView, default=OfferInspect}
]
widgets {
 info {inheritedWidgetCount=0, parentClass=&UserWidget, rootWidgetClass=&CommonVisibilitySwitcher, widgetCount=9}
 widgets [
  Switcher_Content {widgetClassPath=&CommonVisibilitySwitcher, widget=Switcher_Content}
  Grid_Available {
   widgetClassPath=&GridPanel
   bIsVariable=true
   parent=Switcher_Content
   widget=Grid_Available
   properties {bOverride_Cursor=true, columnFill [0,1]}
   slot {class=CommonVisibilitySwitcherSlot}
  }
  Text_AvailableHeader {
   widgetClassPath=&CommonTextBlock
   parent=Grid_Available
   widget=Text_AvailableHeader
   properties {
    bAutoCollapseWithEmptyText=true
    mobileFontSizeMultiplier=1.2
    text="[PH] This purchase comes with:"
    font {
     bForceMonospaced=false
     bMaterialIsStencil=false
     fontMaterial=null
     fontObject=&HeadingNow
     letterSpacing=0
     monospacedWidth=1
     size=14.25
     skewAmount=0
     typefaceFontName=74 Regular
     outlineSettings {
      bApplyOutlineToDropShadows=false
      bMiteredCorners=false
      bSeparateFillAlpha=false
      outlineMaterial=null
      outlineSize=0
      outlineColor=<rgba 0 0 0 1>
     }
    }
    shadowOffset=<0 0>
   }
   slot {class=GridSlot, properties {verticalAlignment=VAlign_Center}}
  }
  EntryBox_Items {
   widgetClassPath=&DynamicEntryBox
   bIsVariable=true
   parent=Grid_Available
   widget=EntryBox_Items
   properties {entryWidgetClass=&WBP_ItemShop_ItemCard_ALC_C, entrySpacing=<10 0>}
   slot {class=GridSlot, properties {column=1, padding=<ltrb 10 0 0 0>}}
  }
  Grid_Unavailable {
   widgetClassPath=&GridPanel
   bIsVariable=true
   parent=Switcher_Content
   widget=Grid_Unavailable
   slot {class=CommonVisibilitySwitcherSlot}
  }
  ItemCard_Unavailable {
   widgetClassPath=/ItemShop/MultiProduct/WBP_ItemShop_ItemCard_ALC
   bIsVariable=true
   parent=Grid_Unavailable
   widget=ItemCard_Unavailable
   properties {holdData=&CommonUI_HoldData_Base_C, itemCardSize=<74 74>}
   slot {class=GridSlot, properties {rowSpan=3, padding=<ltrb 0 0 10 0>}}
  }
  Text_UnavailableHeader {
   widgetClassPath=&CommonTextBlock
   parent=Grid_Unavailable
   widget=Text_UnavailableHeader
   properties {
    bAutoCollapseWithEmptyText=true
    mobileFontSizeMultiplier=1.16
    text=Not sold separately.
    colorAndOpacity {colorUseRule=UseColor_Specified, specifiedColor=<rgba 1 1 1 0.7>}
    font {
     bForceMonospaced=false
     bMaterialIsStencil=false
     fontMaterial=null
     fontObject=&HeadingNow
     letterSpacing=0
     monospacedWidth=1
     size=13.5
     skewAmount=0
     typefaceFontName=74 Regular
     outlineSettings {
      bApplyOutlineToDropShadows=false
      bMiteredCorners=false
      bSeparateFillAlpha=false
      outlineMaterial=null
      outlineSize=0
      outlineColor=<rgba 0 0 0 1>
     }
    }
    shadowOffset=<0 0>
   }
   slot {class=GridSlot, properties {column=1, verticalAlignment=VAlign_Center}}
  }
  Text_UnavailableSubheader {
   widgetClassPath=&CommonTextBlock
   parent=Grid_Unavailable
   widget=Text_UnavailableSubheader
   properties {
    bAutoCollapseWithEmptyText=true
    mobileFontSizeMultiplier=1.26
    text="Included with the purchase of: "
    font {
     bForceMonospaced=false
     bMaterialIsStencil=false
     fontMaterial=null
     fontObject=&HeadingNow
     letterSpacing=0
     monospacedWidth=1
     size=14.25
     skewAmount=0
     typefaceFontName=74 Regular
     outlineSettings {
      bApplyOutlineToDropShadows=false
      bMiteredCorners=false
      bSeparateFillAlpha=false
      outlineMaterial=null
      outlineSize=0
      outlineColor=<rgba 0 0 0 1>
     }
    }
    shadowOffset=<0 0>
   }
   slot {class=GridSlot, properties {column=1, row=1, verticalAlignment=VAlign_Center}}
  }
  Text_PrimaryItemName {
   widgetClassPath=&CommonTextBlock
   parent=Grid_Unavailable
   widget=Text_PrimaryItemName
   properties {
    bAutoCollapseWithEmptyText=true
    mobileFontSizeMultiplier=1.26
    text="[PH] Rumi Outfit - Fortnite Style"
    font {
     bForceMonospaced=false
     bMaterialIsStencil=false
     fontMaterial=null
     fontObject=&HeadingNow
     letterSpacing=0
     monospacedWidth=1
     size=14.25
     skewAmount=0
     typefaceFontName=75 Medium
     outlineSettings {
      bApplyOutlineToDropShadows=false
      bMiteredCorners=false
      bSeparateFillAlpha=false
      outlineMaterial=null
      outlineSize=0
      outlineColor=<rgba 0 0 0 1>
     }
    }
    shadowOffset=<0 0>
   }
   slot {class=GridSlot, properties {column=1, row=2, verticalAlignment=VAlign_Center}}
  }
 ]
}
