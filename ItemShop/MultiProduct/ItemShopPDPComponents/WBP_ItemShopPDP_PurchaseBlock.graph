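# Purchase block widget for the Item Shop product detail page (PDP), showing price, restrictions and purchase, gift, V-Bucks, undo, history and bundle buttons.
# Button handlers log analytics and raise dispatchers.
# PDP interface events update offer and view state, then forward to price and endorsement components; viewmodel-driven functions choose text, visibility, hold-to-purchase and active content.
# A named slot accepts injected widgets, and layout adapts to compact and touch UI.
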WidgetBlueprint /ItemShop/MultiProduct/ItemShopPDPComponents/WBP_ItemShopPDP_PurchaseBlock
graphs [
 EventGraph {nodes [
   n1 = "On Clicked (Button_PurchaseHistory)"!() -> (OutputDelegate, then, Button:Common Button Base Object Reference) {is=title K2Node_ComponentBoundEvent}
   n2 = Get CurrentOfferVM() -> (CurrentOfferVM:Fort Item Shop Offer VM Object Reference)
   > Fire Event Element Interacted(
    Element=@n1.Button
    AdditionalInfo="Add String (Non-Localized)"!(
     Info:"Flow Analytics Additional Info Structure (by ref)"=Make Flow Analytics Additional Info!() -> (FlowAnalyticsAdditionalInfo:Flow Analytics Additional Info Structure) {is=title K2Node_MakeStruct}
     Key:String=Display Asset Name
     Value:String=GetDisplayAssetName!(self:Fort Item Shop Offer VM Object Reference=@n2) -> (ReturnValue:String)
    ) -> (ReturnValue:Flow Analytics Additional Info Structure)
   )
   > Call On Purchase History Clicked!(execute:Exec, self:WBP Item Shop PDP Purchase Block Object Reference) -> (then:Exec) {is=title K2Node_CallDelegate}
   n3 = "On Clicked (Button_Purchase)"!() -> (OutputDelegate, then, Button:Common Button Base Object Reference) {is=title K2Node_ComponentBoundEvent}
   > n4 = Branch(Condition=GetRequiresHold!(self:Common Button Base Object Reference=@n3.Button) -> (ReturnValue:bool))
   n5 = Get CurrentOfferVM()
   Fire Event Element Interacted(
    execute=[@n11.then, @n4.else]
    Element=Get Button_Purchase() -> (Button_Purchase:WBP UIKit Button Loud Object Reference)
    AdditionalInfo="Add String (Non-Localized)"(Info=Make Flow Analytics Additional Info(), Key=Display Asset Name, Value=GetDisplayAssetName(self=@n5))
   )
   > Call On Purchase Clicked!(execute:Exec, self:WBP Item Shop PDP Purchase Block Object Reference) -> (then:Exec) {is=title K2Node_CallDelegate}
   n6 = "On Clicked (Button_GetVBucks)"!() -> (OutputDelegate, then, Button:Common Button Base Object Reference) {is=title K2Node_ComponentBoundEvent}
   > Fire Event Element Interacted(Element=@n6.Button)
   > Call On Get Vbucks Clicked!(execute:Exec, self:WBP Item Shop PDP Purchase Block Object Reference) -> (then:Exec) {is=title K2Node_CallDelegate}
   n7 = "On Hovered (Button_PurchaseHistory)"!() -> (OutputDelegate, then, Button:Common Button Base Object Reference) {is=title K2Node_ComponentBoundEvent}
   Call On Purchase Block Hovered!(execute:Exec=[@n7.then, @n8.then, @n9.then, @n15.then, @n16.then, @n17.then], self:WBP Item Shop PDP Purchase Block Object Reference) -> (then:Exec) {is=title K2Node_CallDelegate}
   n8 = "On Hovered (Button_UndoPurchase)"!() -> (OutputDelegate, then, Button:Common Button Base Object Reference) {is=title K2Node_ComponentBoundEvent}
   n9 = "On Hovered (Button_GetVBucks)"!() -> (OutputDelegate, then, Button:Common Button Base Object Reference) {is=title K2Node_ComponentBoundEvent}
   "On Hold Completed (Button_UndoPurchase)"() {is=title K2Node_ComponentBoundEvent}
   n10 = Get CurrentOfferVM()
   > Fire Event Element Interacted(
    Element=Get Button_UndoPurchase() -> (Button_UndoPurchase:WBP UIKit Button Regular Object Reference)
    AdditionalInfo="Add String (Non-Localized)"(Info=Make Flow Analytics Additional Info(), Key=Display Asset Name, Value=GetDisplayAssetName(self=@n10))
   )
   > Call On Undo Purchase Clicked!(execute:Exec, self:WBP Item Shop PDP Purchase Block Object Reference) -> (then:Exec) {is=title K2Node_CallDelegate}
   n11 = "On Hold Completed (Button_Purchase)"() {is=title K2Node_ComponentBoundEvent}
   n12 = "On Clicked (Button_GoToItem)"!() -> (OutputDelegate, then, Button:Common Button Base Object Reference) {is=title K2Node_ComponentBoundEvent}
   n13 = Get CurrentOfferVM()
   > Fire Event Element Interacted(
    Element=@n12.Button
    AdditionalInfo="Add String (Non-Localized)"(Info=Make Flow Analytics Additional Info(), Key=Display Asset Name, Value=GetDisplayAssetName(self=@n13))
   )
   n14 = Get CurrentOfferVM()
   > Call On Go to Item Clicked!(
    execute:Exec
    self:WBP Item Shop PDP Purchase Block Object Reference
    FortItemVM:Fort Item VM Object Reference="Get (a copy)"(Array:Array of Fort Item VM Object References=Get GrantedItems!(self:Fort Item Shop Offer VM Object Reference=@n14) -> (GrantedItems:Array of Fort Item VM Object References)) -> (Output:Fort Item VM Object Reference)
   ) -> (then:Exec) {is=title K2Node_CallDelegate}
   n15 = "On Hovered (Button_Purchase)"!() -> (OutputDelegate, then, Button:Common Button Base Object Reference) {is=title K2Node_ComponentBoundEvent}
   n16 = "On Hovered (Button_Gift)"!() -> (OutputDelegate, then, Button:Common Button Base Object Reference) {is=title K2Node_ComponentBoundEvent}
   n17 = "On Hovered (Button_ViewBundle)"!() -> (OutputDelegate, then, Button:Common Button Base Object Reference) {is=title K2Node_ComponentBoundEvent}
   n18 = "On Clicked (Button_ViewBundle)"!() -> (OutputDelegate, then, Button:Common Button Base Object Reference) {is=title K2Node_ComponentBoundEvent}
   n19 = Get ParentOfferVM() -> (ParentOfferVM:Fort Item Shop Offer VM Object Reference)
   > Fire Event Element Interacted(
    Element=@n18.Button
    AdditionalInfo="Add String (Non-Localized)"(Info=Make Flow Analytics Additional Info(), Key=Display Asset Name, Value=GetDisplayAssetName(self=@n19))
   )
   > Call On View Bundle Clicked!(execute:Exec, self:WBP Item Shop PDP Purchase Block Object Reference) -> (then:Exec) {is=title K2Node_CallDelegate}
   n20 = "On Clicked (Button_Gift)"!() -> (OutputDelegate, then, Button:Common Button Base Object Reference) {is=title K2Node_ComponentBoundEvent}
   n21 = Get CurrentOfferVM()
   > Fire Event Element Interacted(
    Element=@n20.Button
    AdditionalInfo="Add String (Non-Localized)"(Info=Make Flow Analytics Additional Info(), Key=Display Asset Name, Value=GetDisplayAssetName(self=@n21))
   )
   > Call On Gift Clicked!(execute:Exec, self:WBP Item Shop PDP Purchase Block Object Reference) -> (then:Exec) {is=title K2Node_CallDelegate}
   n22 = OnPdpDeactivated_Implementation() {is=title K2Node_CustomEvent}
   n23 = OnPTTCanChange() {is=title K2Node_CustomEvent}
   n24 = Unbind Event from On Can PTTChange(
    execute=@n22.then
    self:Athena HUDContext Object Reference=Get AthenaHUDContext!() -> (ReturnValue:Athena HUDContext Object Reference) {is=title K2Node_GetBlueprintContextEx}
    Delegate=@n23
   ) {is=title K2Node_RemoveDelegate}
   n25 = OnInputMethodChanged!() -> (OutputDelegate, then, bNewInputType:ECommonInputType Enum) {is=title K2Node_CustomEvent}
   Unbind Event from On Input Method Changed(execute=@n24.then, self=Get CommonInputSubsystem(), Delegate=@n25)
   n26 = OnFocusAreaChanged_Implementation!() -> (OutputDelegate, then, IsFocusingItems:bool=~) {is=title K2Node_CustomEvent}
   > Set with Broadcast Is Focusing Items!(execute, Is Focusing Items:bool=@n26.IsFocusingItems) -> (then, Output_Get:bool)
   OnPdpActivated_Implementation() {is=title K2Node_CustomEvent}
   > Bind Event to On Can PTTChange(self:Athena HUDContext Object Reference=Get AthenaHUDContext(), Delegate=@n23) {is=title K2Node_AddDelegate}
   > Bind Event to On Input Method Changed(self=Get CommonInputSubsystem(), Delegate=@n25)
   > SetTouchStyle()
   K2Node_Event_1 = OnInteractionEnabledChanged_Implementation!() -> (OutputDelegate, then, ShouldEnable:bool=~) {is=title K2Node_CustomEvent}
   > SetIsInteractionEnabled!(
    execute
    self:Common Button Base Object Reference=[
     Get Button_UndoPurchase()
     Get Button_PurchaseHistory() -> (Button_PurchaseHistory:WBP UIKit Button Regular Object Reference)
     Get Button_ViewBundle() -> (Button_ViewBundle:WBP UIKit Button Regular Object Reference)
     Get Button_Purchase()
     Get Button_Gift() -> (Button_Gift:WBP UIKit Button Regular Object Reference)
     Get Button_GetVBucks() -> (Button_GetVBucks:WBP UIKit Button Loud Object Reference)
    ]
    bInIsInteractionEnabled:bool=@K2Node_Event_1.ShouldEnable
   ) -> (then)
   K2Node_Event_2 = OnViewChanged_Implementation!() -> (OutputDelegate, then, NewView:EFortItemShopPdpView Enum=OfferInspect) {is=title K2Node_CustomEvent}
   > Set with Broadcast CurrentView!(execute, CurrentView:EFortItemShopPdpView Enum=@K2Node_Event_2.NewView) -> (then, Output_Get:EFortItemShopPdpView Enum=OfferInspect)
   Event OnPdpActivated() {is=title K2Node_Event}
   > OnPdpActivated_Implementation() {is=title K2Node_CallFunction}
   > n27 = For Each Loop(Array:Array of BPI Item Shop PDP Interfaces=Get Components() -> (Components:Array of BPI Item Shop PDP Interfaces)) -> (Array Element:BPI Item Shop PDP Interface)
   > OnPdpActivated(self:BPI Item Shop PDP Interface=@n27)
   Event PreConstruct()
   > SetupComponents()
   > SetCompactStyle()
   > SetPurchaseButtonStyle!(
    execute
    Purchase Button:Fort CTAButton Object Reference=Get Button_Purchase()
    GetVbucks Button:Fort CTAButton Object Reference=Get Button_GetVBucks()
   ) -> (then)
   n28 = Event OnInteractionEnabledChanged!() -> (OutputDelegate, then, ShouldEnable:bool) {is=title K2Node_Event}
   > OnInteractionEnabledChanged_Implementation!(execute, self, ShouldEnable:bool=@n28.ShouldEnable) -> (then) {is=title K2Node_CallFunction}
   > n29 = For Each Loop(Array=Get Components())
   > OnInteractionEnabledChanged!(execute, self:BPI Item Shop PDP Interface=@n29, ShouldEnable:bool=@n28.ShouldEnable) -> (then)
   n30 = Event OnFocusAreaChanged!() -> (OutputDelegate, then, IsFocusingItems:bool) {is=title K2Node_Event}
   > OnFocusAreaChanged_Implementation!(execute, self, IsFocusingItems:bool=@n30.IsFocusingItems) -> (then) {is=title K2Node_CallFunction}
   > n31 = For Each Loop(Array=Get Components())
   > OnFocusAreaChanged!(execute, self:BPI Item Shop PDP Interface=@n31, IsFocusingItems:bool=@n30.IsFocusingItems) -> (then)
   n32 = Event OnViewChanged!() -> (OutputDelegate, then, NewView:EFortItemShopPdpView Enum=OfferInspect) {is=title K2Node_Event}
   > OnViewChanged_Implementation!(execute, self, NewView:EFortItemShopPdpView Enum=@n32.NewView) -> (then) {is=title K2Node_CallFunction}
   > n33 = For Each Loop(Array=Get Components())
   > OnViewChanged!(execute, self:BPI Item Shop PDP Interface=@n33, NewView:EFortItemShopPdpView Enum=@n32.NewView) -> (then)
   n34 = Event OnOfferSet!() -> (
    OutputDelegate
    then
    OfferVM:Fort Item Shop Offer VM Object Reference
    ParentOfferVM:Fort Item Shop Offer VM Object Reference
    OfferGroupVM:Fort Item Shop Offer Group VM Object Reference
   ) {is=title K2Node_Event}
   > OnOfferSet_Implementation!(
    execute
    self
    OfferVM:Fort Item Shop Offer VM Object Reference=@n34.OfferVM
    ParentOfferVM:Fort Item Shop Offer VM Object Reference=@n34.ParentOfferVM
    OfferGroupVM:Fort Item Shop Offer Group VM Object Reference=@n34.OfferGroupVM
   ) -> (then)
   > n35 = For Each Loop(Array=Get Components())
   > OnOfferSet!(
    execute
    self:BPI Item Shop PDP Interface=@n35
    OfferVM:Fort Item Shop Offer VM Object Reference=@n34.OfferVM
    ParentOfferVM:Fort Item Shop Offer VM Object Reference=@n34.ParentOfferVM
    OfferGroupVM:Fort Item Shop Offer Group VM Object Reference=@n34.OfferGroupVM
   ) -> (then)
   K2Node_Event_5 = OnOfferSet_Implementation!() -> (
    OutputDelegate
    then
    OfferVM:Fort Item Shop Offer VM Object Reference
    ParentOfferVM:Fort Item Shop Offer VM Object Reference
    OfferGroupVM:Fort Item Shop Offer Group VM Object Reference
   ) {is=title K2Node_CustomEvent}
   > Set CurrentOfferVM(CurrentOfferVM:Fort Item Shop Offer VM Object Reference) -> (Output_Get:Fort Item Shop Offer VM Object Reference)
   > Set ParentOfferVM(ParentOfferVM:Fort Item Shop Offer VM Object Reference) -> (Output_Get:Fort Item Shop Offer VM Object Reference)
   > Set FortItemShopOfferGroupVM(FortItemShopOfferGroupVM:Fort Item Shop Offer Group VM Object Reference) -> (Output_Get:Fort Item Shop Offer Group VM Object Reference)
   > Set CurrentOfferVM(CurrentOfferVM=@K2Node_Event_5.OfferVM)
   > Set ParentOfferVM(ParentOfferVM=@K2Node_Event_5.ParentOfferVM)
   > Set FortItemShopOfferGroupVM(FortItemShopOfferGroupVM=@K2Node_Event_5.OfferGroupVM)
   > UpdateButtons()
   n36 = Event InjectCustomWidget!() -> (OutputDelegate, then, CustomWidget:Widget Object Reference) {is=title K2Node_Event}
   n37 = Get Slot_InjectedButton() -> (Slot_InjectedButton:Named Slot Object Reference)
   > ClearChildren(self=@n37)
   > SetContent!(execute, self:Content Widget Object Reference=@n37, Content:Widget Object Reference=@n36.CustomWidget) -> (then, ReturnValue:Panel Slot Object Reference)
   Event ClearCustomInjection() {is=title K2Node_Event}
   > ClearChildren(self=Get Slot_InjectedButton())
   Event OnPdpDeactivated() {is=title K2Node_Event}
   > OnPdpDeactivated_Implementation() {is=title K2Node_CallFunction}
   > n38 = For Each Loop(Array=Get Components())
   > OnPdpDeactivated(self:BPI Item Shop PDP Interface=@n38)
   SetTouchStyle(execute=@n25.then)
   SetTouchStyle(execute=@n23.then)
  ]}
 GetActiveContentWidget {nodes [
   n1 = GetActiveContentWidget!() -> (
    then
    PreviewedItem:Item Shop Preview Data Object Reference
    GrantedItems:Array of Fort Item VM Object References
    CurrentView:EFortItemShopPdpView Enum=OfferInspect
   ) {is=title K2Node_FunctionEntry}
   n2 = IsValid(Object=@n1)
   n3 = Length(TargetArray:Array of Fort Item VM Object References=@n1.GrantedItems)
   > n4 = Branch(Condition=OR Boolean!(
     A=NOT Boolean(A=@n2)
     B="integer <= integer"(A=@n3, B=1)
     C:bool="Not Equal (Enum)"!(A:EFortItemShopPdpView Enum=@n1.CurrentView, B:EFortItemShopPdpView Enum=ItemDetails) -> (ReturnValue:bool)
    ) -> (ReturnValue))
   > Return Node!(execute, ReturnValue:Widget Object Reference=Get Grid_PurchaseBlock() -> (Grid_PurchaseBlock:Grid Panel Object Reference))
   n5 = "Get (a copy)"(Array=Get GrantedItems!() -> (GrantedItems))
   n6 = Get ItemVM!(self:Item Shop Preview Data Object Reference=Get PreviewedItem() -> (PreviewedItem:Item Shop Preview Data Object Reference)) -> (ItemVM:Fort Item VM Object Reference)
   Return Node(
    execute=@n4.else
    ReturnValue=Select(
     Option 0:Widget Object Reference=Get Button_GoToItem() -> (Button_GoToItem:WBP UIKit Button Regular Object Reference)
     Option 1:Widget Object Reference=Get Grid_PurchaseBlock()
     Index=CompareItemDefinitions!(ItemA:Fort Item VM Object Reference=@n5, ItemB:Fort Item VM Object Reference=@n6) -> (ItemDefsAreEqual:bool)
    ) -> (ReturnValue:Widget Object Reference)
   )
  ]}
 GetButtonLockSwitcherWidget {nodes [
   n1 = GetButtonLockSwitcherWidget!() -> (then, CurrentProgress:int, TotalProgress:int) {is=title K2Node_FunctionEntry}
   n2 = "integer > integer"(A=@n1.TotalProgress)
   n3 = "integer < integer"(A=@n1, B=@n1.TotalProgress)
   > Return Node(ReturnValue=Select(
     Option 0=Get GridPanel_Buttons() -> (GridPanel_Buttons:Grid Panel Object Reference)
     Option 1=Get Button_LockedPurchase() -> (Button_LockedPurchase:WBP Item Shop Locked Purchase Button Object Reference)
     Index=AND Boolean(A=@n3, B=@n2)
    ))
  ]}
 GetDesiredFocusTarget {nodes [
   GetDesiredFocusTarget() {is=title K2Node_FunctionEntry}
   n1 = GetActiveWidgetIndex!(self:Common Visibility Switcher Object Reference=Get Switcher_Content() -> (Switcher_Content:Common Visibility Switcher Object Reference)) -> (ReturnValue:int=0)
   n2 = IsVisible(self=Get Button_Purchase())
   n3 = Get Button_Gift()
   n4 = IsVisible(self=Get Button_UndoPurchase())
   n5 = Get Button_UndoPurchase()
   n6 = Select(
    Option 0=Select(Option 0=Select(Option 1=@n5, Index=@n4), Option 1=Get Button_Gift(), Index=IsVisible(self=@n3))
    Option 1=Get Button_GetVBucks()
    Index=IsVisible(self=Get Button_GetVBucks())
   )
   n7 = Get Button_Purchase()
   > Return Node!(
    execute
    FocusTarget:Widget Object Reference=Select(Option 0=Get Button_GoToItem(), Option 1=Select(Option 0=@n6, Option 1=@n7, Index=@n2), Index="integer < integer"(A=@n1, B=1))
   )
  ]}
 GetGiftButtonVisibility {nodes [
   n1 = GetGiftButtonVisibility!() -> (then, CanGift:"Boolean (by ref)") {is=title K2Node_FunctionEntry}
   > Return Node!(
    execute
    ReturnValue:ESlateVisibility Enum=Select(Option 0:ESlateVisibility Enum=Collapsed, Option 1:ESlateVisibility Enum=SelfHitTestInvisible, Index=@n1) -> (ReturnValue:ESlateVisibility Enum)
   )
  ]}
 GetIncludedItemsText {nodes [
   n1 = GetIncludedItemsText!() -> (
    then
    PreviewedItem:Item Shop Preview Data Object Reference
    GrantedItems:Array of Fort Item VM Object References
    CurrentView:EFortItemShopPdpView Enum=OfferInspect
   ) {is=title K2Node_FunctionEntry}
   n2 = IsValid(Object=@n1)
   n3 = Length(TargetArray=@n1.GrantedItems)
   > n4 = Branch(Condition=OR Boolean(A=NOT Boolean(A=@n2), B="integer <= integer"(A=@n3, B=1), C="Not Equal (Enum)"(A=@n1.CurrentView, B=ItemDetails)))
   > Return Node(ReturnValue:Text)
   n5 = For Each Loop(Exec=@n4.else, Array:Array of Fort Item VM Object References=Get GrantedItems()) -> (Array Element:Fort Item VM Object Reference)
   n6 = Get Item Name and Category!(Target:Fort Item VM Object Reference=@n5) -> (Result:Text) {is=title K2Node_MacroInstance}
   n7 = Get ItemsText() -> (ItemsText:Text)
   > Set ItemsText(ItemsText:Text=Select(
     Option 0:Text=Format Text!(Format="LOCTEXT(\"{Items}, {NextItem}\")", Items:Text=@n7, NextItem:Text=@n6) -> (Result)
     Option 1:Text=@n6
     Index="Equal (Integer)"(A=@n5.Array Index)
    ) -> (ReturnValue:Text)) -> (Output_Get:Text)
   n8 = Get PreviewedItem()
   n9 = Get GrantedItems()
   n10 = "Equal (Object)"(
    A=Get ItemDefinition!(self:Fort Item VM Object Reference="Get (a copy)"(Array=@n9)) -> (ItemDefinition:Fort Item Definition Object Reference)
    B=Get ItemDefinition(self=Get ItemVM(self=@n8))
   )
   Return Node(
    execute=@n5.Completed
    ReturnValue=Format Text!(
     Format=Select(
      Option 0="LOCTEXT(\"Not avilable separately. Items included with this purchase: {Items}\")"
      Option 1="LOCTEXT(\"Items included with this purchase: {Items}\")"
      Index=@n10
     )
     Items=Get ItemsText()
    ) -> (Result)
   )
  ]}
 GetItemForRestriction {nodes [
   n1 = GetItemForRestriction!() -> (then, PreviewedItem:Item Shop Preview Data Object Reference, IsFocusingItems:bool=~) {is=title K2Node_FunctionEntry}
   > n2 = Branch(Condition=AND Boolean(A=IsValid(Object=@n1), B=@n1.IsFocusingItems))
   > Return Node(ReturnValue:Fort Item VM Object Reference=Get ItemVM(self=Get PreviewedItem()))
   Return Node(execute=@n2.else)
  ]}
 GetItemNameAndCategory {nodes [
   n1 = Get ItemDetails!(self:Fort Item VM Object Reference=Inputs!() -> (Target:Fort Item VM Object Reference)) -> (ItemDetails:Fort Item Details VM Object Reference)
   n2 = GetItemStyleText!(self:Fort Item Details VM Object Reference=@n1) -> (ReturnValue:Text)
   n3 = TextIsEmpty(InText=@n2)
   Outputs!(Result:Text=Format Text!(
     Format=Select(
      Option 0="LOCTEXT(\"<text fontface=\\\"75 Medium\\\">{Name} {Category} {Style}</>\")"
      Option 1="LOCTEXT(\"<text fontface=\\\"75 Medium\\\">{Name} {Category}</>\")"
      Index=@n3
     )
     Name:Text=GetDisplayName!(self:Fort Item Details VM Object Reference=@n1) -> (ReturnValue:Text)
     Category:Text=GetItemCategoryName!(self:Fort Item Details VM Object Reference=@n1) -> (ReturnValue:Text)
     Style:Text=@n2
    ) -> (Result))
  ]}
 GetPriceHeaderText {nodes [
   GetPriceHeaderText!() -> (
    then
    GrantedItems:Array of Fort Item VM Object References
    OfferType:ECatalogOfferType Enum=StaticPrice
    IsOwned:bool=~
    CurrentView:EFortItemShopPdpView Enum=OfferInspect
   ) {is=title K2Node_FunctionEntry}
   n1 = Get CurrentView() -> (CurrentView:EFortItemShopPdpView Enum=OfferInspect)
   n2 = Get OfferType() -> (OfferType:ECatalogOfferType Enum=StaticPrice)
   > n3 = Branch(Condition=OR Boolean(
     A="Not Equal (Enum)"(A:ECatalogOfferType Enum=@n2, B:ECatalogOfferType Enum=DynamicBundle)
     B=Get IsOwned() -> (IsOwned:bool)
     C="Equal (Enum)"(A:EFortItemShopPdpView Enum=@n1, B:EFortItemShopPdpView Enum=ItemDetails)
    ))
   > Return Node!(execute, Text:Text)
   n4 = Get CurrentOfferVM()
   n5 = For Each Loop!(
    Exec=@n3.else
    Array:Array of Bundled Item Info Structures=GetDynamicBundleItems!(self:Fort Store Front Offer Info Object Reference=Get OfferInfo!(self:Fort Item Shop Offer VM Object Reference=@n4) -> (OfferInfo:Athena Item Shop Offer Info Object Reference)) -> (Items:Array of Bundled Item Info Structures)
   ) -> (LoopBody, Array Element_TemplateId:String, Array Element_Quantity:int, Array Element_bOwned:bool=~, Array Index, Completed)
   Return Node(execute=@n5.Completed, Text="LOCTEXT(\"The price for included items is:\")")
   > Branch(Condition=@n5.Array Element_bOwned)
   > Return Node(Text="LOCTEXT(\"The price for unowned items is:\")")
  ]}
 GetPurchaseButtonHold {nodes [
   n1 = GetPurchaseButtonHold!() -> (then, OfferVM:Fort Item Shop Offer VM Object Reference, Currency:EFortPriceCurrency Enum=Mtx) {is=title K2Node_FunctionEntry}
   > n2 = ShouldShowCpcDialog!(execute, Context:Widget Object Reference=Self-Reference(), OfferVM:Fort Item Shop Offer VM Object Reference=@n1) -> (then, ReturnValue:bool)
   n3 = OR Boolean!(
    A=@n2
    B=GetConfigDataBoolWithDefault!(EntryName:String=shop_confirm_purchase_through_dialog, bDefaultValue:bool=false) -> (ReturnValue:bool)
   ) -> (ReturnValue)
   > Return Node!(
    execute
    ReturnValue:bool=AND Boolean(A=NOT Boolean(A=@n3), B="Equal (Enum)"(A:EFortPriceCurrency Enum=@n1.Currency, B:EFortPriceCurrency Enum=Mtx))
   )
  ]}
 GetPurchaseButtonText {nodes [
   n1 = GetPurchaseButtonText!() -> (then, GrantedItems:Array of Fort Item VM Object References, CurrentView:EFortItemShopPdpView Enum=OfferInspect, NumOwnedStacks:int) {is=title K2Node_FunctionEntry}
   n2 = "Equal (Enum)"(A:EFortItemShopPdpView Enum=@n1.CurrentView, B:EFortItemShopPdpView Enum=ItemDetails)
   n3 = "Equal (Integer)"(A=Length(TargetArray=@n1), B=1)
   > n4 = Branch(Condition=AND Boolean(A=OR Boolean(A=@n3, B=@n2), B="integer > integer"(A=@n1.NumOwnedStacks)))
   > Return Node(ReturnValue:Text="LOCTEXT(\"Purchase More\")")
   Return Node(execute=@n4.else, ReturnValue="LOCTEXT(\"Purchase\")")
  ]}
 GetPurchaseButtonVisibility {nodes [
   n1 = GetPurchaseButtonVisibility!() -> (then, CanPurchaseOffer:EOfferPurchaseError Enum=NoError) {is=title K2Node_FunctionEntry}
   > Return Node(ReturnValue:ESlateVisibility Enum=Select(
     Option 0:ESlateVisibility Enum=Collapsed
     Option 1:ESlateVisibility Enum=SelfHitTestInvisible
     Index="Equal (Enum)"(A:EOfferPurchaseError Enum=@n1, B:EOfferPurchaseError Enum=NoError)
    ) -> (ReturnValue:ESlateVisibility Enum))
  ]}
 GetPurchaseHistoryButtonVisibility {nodes [
   n1 = GetPurchaseHistoryButtonVisibility!() -> (then, GrantedItems:Array of Fort Item VM Object References) {is=title K2Node_FunctionEntry}
   > n2 = Branch(Condition=Is Empty!(TargetArray:Array of Fort Item VM Object References=@n1) -> (ReturnValue:bool))
   > Return Node(ReturnValue=Collapsed)
   n3 = Get GrantedItems()
   n4 = Offer Contains Single Stackable Item!(GrantedItems:Array of Fort Item VM Object References=@n3) -> (ReturnValue:bool=~) {is=title K2Node_MacroInstance}
   n5 = Owns Any Stacks!(FortItemVM:Fort Item VM Object Reference="Get (a copy)"(Array=@n3)) -> (ReturnValue:bool=~) {is=title K2Node_MacroInstance}
   Return Node(execute=@n2.else, ReturnValue=Select(Option 0=Collapsed, Option 1=SelfHitTestInvisible, Index=AND Boolean(A=@n4, B=@n5)))
  ]}
 GetSelectedItem {nodes [
   n1 = GetSelectedItem!() -> (then, CurrentView:EFortItemShopPdpView Enum=OfferInspect, PreviewedItem:Item Shop Preview Data Object Reference) {is=title K2Node_FunctionEntry}
   > n2 = Get!(execute:Exec) -> (then:Exec, else:Exec, PreviewedItem:Item Shop Preview Data Object Reference)
   > Return Node(ReturnValue:Fort Item VM Object Reference=Get ItemVM(self=@n2))
   Return Node(
    execute=@n2.else
    ReturnValue=Select!(
     OfferInspect:Fort Item VM Object Reference
     ItemDetails:Fort Item VM Object Reference=Get FortItemVM() -> (FortItemVM:Fort Item VM Object Reference)
     Index:EFortItemShopPdpView Enum=@n1
    ) -> (ReturnValue:Fort Item VM Object Reference)
   )
  ]}
 GetUndoPurchaseButtonVisibility {nodes [
   n1 = GetUndoPurchaseButtonVisibility!() -> (then, CanUndoPurchase:bool=~, GrantedItems:Array of Fort Item VM Object References) {is=title K2Node_FunctionEntry}
   n2 = GetPurchaseHistoryButtonVisibility!(self, GrantedItems:Array of Fort Item VM Object References=@n1.GrantedItems) -> (ReturnValue:ESlateVisibility Enum=Visible) {is=title K2Node_CallFunction}
   > n3 = Branch(Condition=AND Boolean(A=@n1, B="Equal (Enum)"(A:ESlateVisibility Enum=@n2, B:ESlateVisibility Enum=Collapsed)))
   > Return Node(ReturnValue:ESlateVisibility Enum=SelfHitTestInvisible)
   Return Node(execute=@n3.else, ReturnValue=Collapsed)
  ]}
 GetVBucksButtonVisibility {nodes [
   n1 = GetVBucksButtonVisibility!() -> (then, Currency:EFortPriceCurrency Enum=Mtx, CanPurchaseOfferResult:EOfferPurchaseError Enum=NoError) {is=title K2Node_FunctionEntry}
   n2 = "Equal (Enum)"(A:EOfferPurchaseError Enum=@n1.CanPurchaseOfferResult, B:EOfferPurchaseError Enum=CannotAffordItem)
   n3 = "Equal (Enum)"(A:EFortPriceCurrency Enum=Get Currency() -> (Currency:EFortPriceCurrency Enum=Mtx), B:EFortPriceCurrency Enum=Mtx)
   > Return Node(ReturnValue=Select!(Option 0:ESlateVisibility Enum=Collapsed, Option 1:ESlateVisibility Enum=SelfHitTestInvisible, Index:bool=AND Boolean(A=@n2, B=@n3)) -> (ReturnValue:ESlateVisibility Enum))
  ]}
 GetViewBundleButtonText {nodes [
   n1 = GetViewBundleButtonText!() -> (then, GrantedItems:Array of Fort Item VM Object References) {is=title K2Node_FunctionEntry}
   > n2 = Branch(Condition=Is Empty(TargetArray=@n1))
   > Return Node(ReturnValue:Text)
   Return Node(
    execute=@n2.else
    ReturnValue=Select(
     Option 0:Text="LOCTABLE(\"/ItemShop/MultiProduct/Data/ST_ItemShop\", \"ViewBundleBtnText_Default\")"
     Option 1:Text="LOCTABLE(\"&ST_ItemShop\", \"ViewBundleBtnText_Compact\")"
     Index=Offer Contains Single Stackable Item(GrantedItems=@n1) -> (ReturnValue=~)
    ) -> (ReturnValue:Text)
   )
  ]}
 GetViewBundleButtonVisibility {nodes [
   n1 = GetViewBundleButtonVisibility!() -> (
    then
    InOfferVM:Fort Item Shop Offer VM Object Reference
    InParentOfferVM:Fort Item Shop Offer VM Object Reference
    InOfferGroupVM:Fort Item Shop Offer Group VM Object Reference
    InCurrentView:EFortItemShopPdpView Enum=OfferInspect
    CanUndoPurchase:bool=~
    GrantedItems:Array of Fort Item VM Object References
   ) {is=title K2Node_FunctionEntry}
   > n2 = Get!(execute) -> (then, else, InOfferVM:Fort Item Shop Offer VM Object Reference)
   > n3 = Get!(execute) -> (then, else, InParentOfferVM:Fort Item Shop Offer VM Object Reference)
   > n4 = Get!(execute) -> (then, else, InOfferGroupVM:Fort Item Shop Offer Group VM Object Reference)
   n5 = NOT Boolean(A=@n1.CanUndoPurchase)
   n6 = "Equal (Enum)"(
    A:ESlateVisibility Enum=GetPurchaseHistoryButtonVisibility(GrantedItems=@n1.GrantedItems) -> (ReturnValue=Visible)
    B:ESlateVisibility Enum=Collapsed
   )
   n7 = "Equal (Enum)"(A:EFortItemShopPdpView Enum=@n1.InCurrentView, B:EFortItemShopPdpView Enum=OfferInspect)
   n8 = "Not Equal (Object)"(
    A:Object Reference=Get InOfferVM() -> (InOfferVM:Fort Item Shop Offer VM Object Reference)
    B:Object Reference=Get InParentOfferVM() -> (InParentOfferVM:Fort Item Shop Offer VM Object Reference)
   ) -> (ReturnValue:bool=~) {is=title K2Node_PromotableOperator}
   > Return Node(ReturnValue:ESlateVisibility Enum=Select(
     Option 0:ESlateVisibility Enum=Collapsed
     Option 1:ESlateVisibility Enum=SelfHitTestInvisible
     Index=AND Boolean!(A=@n8, B=@n7, C:bool=@n5, D:bool=@n6) -> (ReturnValue)
    ) -> (ReturnValue:ESlateVisibility Enum))
   Return Node(execute=[@n4.else, @n3.else, @n2.else], ReturnValue=Collapsed)
  ]}
 GetViewBundleHeaderText {nodes [
   n1 = GetViewBundleHeaderText!() -> (then, BundleName:Text, GrantedItems:Array of Fort Item VM Object References) {is=title K2Node_FunctionEntry}
   > Return Node(ReturnValue:Text=Select(
     Option 0:Text=Format Text!(Format="LOCTEXT(\"Part of the <text color=\\\"FFFFFF\\\" fontface=\\\"75 medium\\\">{BundleName}</>\")", BundleName:Text=@n1) -> (Result)
     Option 1:Text
     Index=Offer Contains Single Stackable Item(GrantedItems=@n1.GrantedItems) -> (ReturnValue=~)
    ) -> (ReturnValue:Text))
  ]}
 IsItemStackable {nodes [
   n1 = Get ItemDefinition(self=Inputs!() -> (FortItemVM:Fort Item VM Object Reference))
   Outputs!(ReturnValue:bool="integer > integer"(A=GetMaxNumStacks!(self:Fort Item Definition Object Reference=@n1) -> (ReturnValue:int=0), B=1))
  ]}
 IsSingleItemOffer {nodes [
   n1 = Inputs!() -> (GrantedItems:Array of Fort Item VM Object References)
   Outputs(ReturnValue="Equal (Integer)"(A=Length(TargetArray=@n1), B=1))
  ]}
 OfferContainsSingleStackableItem {nodes [
   n1 = Inputs()
   n2 = "Get (a copy)"(Array=@n1)
   Outputs(ReturnValue=AND Boolean!(
     A=Is Single Item Offer!(GrantedItems:Array of Fort Item VM Object References=@n1) -> (ReturnValue:bool=~) {is=title K2Node_MacroInstance}
     B=Is Item Stackable!(FortItemVM:Fort Item VM Object Reference=@n2) -> (ReturnValue:bool=~) {is=title K2Node_MacroInstance}
    ) -> (ReturnValue))
  ]}
 OnGetVbucksClicked {nodes [OnGetVbucksClicked() {is=title K2Node_FunctionEntry}]}
 OnGiftClicked {nodes [OnGiftClicked() {is=title K2Node_FunctionEntry}]}
 OnGoToItemClicked {nodes [OnGoToItemClicked!() -> (then, FortItemVM:Fort Item VM Object Reference) {is=title K2Node_FunctionEntry}]}
 OnPurchaseBlockHovered {nodes [OnPurchaseBlockHovered() {is=title K2Node_FunctionEntry}]}
 OnPurchaseClicked {nodes [OnPurchaseClicked() {is=title K2Node_FunctionEntry}]}
 OnPurchaseHistoryClicked {nodes [OnPurchaseHistoryClicked() {is=title K2Node_FunctionEntry}]}
 OnUndoPurchaseClicked {nodes [OnUndoPurchaseClicked() {is=title K2Node_FunctionEntry}]}
 OnViewBundleClicked {nodes [OnViewBundleClicked() {is=title K2Node_FunctionEntry}]}
 OwnsAnyStacks {nodes [
   n1 = Inputs!() -> (FortItemVM:Fort Item VM Object Reference)
   Outputs(ReturnValue="integer > integer"(A=GetOwnedStacksNum!(self:Fort Item VM Object Reference=@n1) -> (ReturnValue:int=0)))
  ]}
 SetCompactStyle {nodes [
   SetCompactStyle() {is=title K2Node_FunctionEntry}
   > Branch(Condition=OR Boolean(A=IsUsingCompactUI(), B=GetSmallTextScalingMultiplier!() -> (ScalingMultiplier:float, IsEnabledOnDevice:bool).IsEnabledOnDevice))
   > n1 = Sequence!(execute) -> (then_0, then_1, then_2:Exec)
   SetTouchStyle(execute=@n1.then_2)
   SetScrollingEnabled!(
    execute=@n1.then_0
    self:Common Rich Text Block Object Reference=Get RichText_Restriction() -> (RichText_Restriction:Common Rich Text Block Object Reference)
    bInIsScrollingEnabled:bool=true
   ) -> (then)
   Set ButtonSize!(execute=@n1.then_1, ButtonSize:E_UI_ButtonSize Enum=NewEnumerator2, self:WBP UIKit Button CTA Base Object Reference=Get Button_Purchase()) -> (then, Output_Get:E_UI_ButtonSize Enum=NewEnumerator0)
   > Set ButtonSize(ButtonSize=NewEnumerator2, self=Get Button_GetVBucks()) -> (Output_Get=NewEnumerator0)
   > Set ButtonSize(ButtonSize=NewEnumerator2, self=Get Button_UndoPurchase()) -> (Output_Get=NewEnumerator0)
   > Set ButtonSize(ButtonSize=NewEnumerator2, self=Get Button_Gift()) -> (Output_Get=NewEnumerator0)
  ]}
 SetTouchStyle {nodes [
   SetTouchStyle() {is=title K2Node_FunctionEntry}
   > Branch(Condition=IsUsingCompactUI())
   n1 = Get CommonInputSubsystem()
   > Set IsUsingTouch(IsUsingTouch:bool="Equal (Enum)"(A:ECommonInputType Enum=GetCurrentInputType(self=@n1), B:ECommonInputType Enum=Touch)) -> (Output_Get:bool)
   > n2 = Sequence!(execute) -> (then_0, then_1, then_2, then_3:Exec, then_4:Exec)
   n3 = Get Button_Purchase()
   n4 = Set bTriggerHapticsOnPressed!(
    execute=@n2.then_1
    bTriggerHapticsOnPressed:bool=Get bRequiresHold!(self:Common Button Base Object Reference=@n3) -> (bRequiresHold:bool)
    self:UIKit Modular Button Object Reference=@n3
   ) -> (then, Output_Get:bool)
   > Set bTriggerHapticsOnActionCommited!(execute, bTriggerHapticsOnActionCommited:bool=@n4, self:UIKit Modular Button Object Reference=@n3) -> (then, Output_Get:bool)
   Set ButtonWidthOverride(
    execute=@n2.then_2
    ButtonWidthOverride:float=SelectFloat(
     A=600
     B=800
     bPickA=CompareToScreenAspectRatio!(Comparison:EAspectRatioComparison Enum=LessThan, AspectRatio:EAspectRatio Enum=Ratio_16_10) -> (ReturnValue:bool)
    )
   ) -> (Output_Get:float)
   n5 = Get ButtonWidthOverride() -> (ButtonWidthOverride:float)
   SetWidthOverride(execute=@n2.then_3, self=Get SizeBox_Buttons() -> (SizeBox_Buttons:Size Box Object Reference), InWidthOverride=@n5)
   > n6 = For Each Loop!(
    Exec
    Array:Array of Common Text Object References=Make Array!(
     "[0]":Common Text Object Reference=Get Text_Legal_UndoPurchase() -> (Text_Legal_UndoPurchase:Common Text Object Reference)
     "[1]":Common Text Object Reference=Get Text_OfferLegal() -> (Text_OfferLegal:Common Text Object Reference)
     "[2]":Common Text Object Reference=Get Text_OutDate() -> (Text_OutDate:Common Text Object Reference)
     "[3]":Common Text Object Reference=Get Text_RefundDisclaimer() -> (Text_RefundDisclaimer:Common Text Object Reference)
    ) -> (Array:Array of Common Text Object References)
   ) -> (LoopBody, Array Element:Common Text Object Reference, Array Index, Completed)
   > SetWrapTextWidth!(execute, self:Common Text Object Reference=@n6, InWrapTextAt:int=Truncate(A=@n5)) -> (then)
   > SetScrollingEnabled(self:Common Text Object Reference=@n6, bInIsScrollingEnabled=false)
   > Set Grid Slot Padding!(execute:Exec, Widget:Widget Object Reference=@n6, Padding_Left:float=0, Padding_Top:float=8, Padding_Right:float=0, Padding_Bottom:float=0) -> (then:Exec) {is=title K2Node_MacroInstance}
   Set Grid Slot Padding(
    execute=@n2.then_4
    Widget=Get Component_Price() -> (Component_Price:WBP Item Shop PDP Price Object Reference)
    Padding_Left=0
    Padding_Top=0
    Padding_Right=0
    Padding_Bottom=12
   )
   > Set Grid Slot Padding(
    Widget=Get VerticalBox_ViewBundle() -> (VerticalBox_ViewBundle:Vertical Box Object Reference)
    Padding_Left=0
    Padding_Top=16
    Padding_Right=0
    Padding_Bottom=0
   )
   n7 = Get IsUsingTouch() -> (IsUsingTouch:bool)
   n8 = Get HorizontalBox_Buttons() -> (HorizontalBox_Buttons:Horizontal Box Object Reference)
   SetGridSlotColumnAndRow!(
    execute=@n2.then_0
    Widget:Widget Object Reference=@n8
    Row:int=SelectInt(A=5, B=0, bPickA=@n7)
    Row Span:int=0
    Column:int=0
    Column Span:int=0
   ) -> (then)
   > SetGridSlotPadding!(
    execute
    Widget:Widget Object Reference=@n8
    Padding_Left:float=0
    Padding_Top:float=SelectFloat(A=8, B=0, bPickA=@n7)
    Padding_Right:float=0
    Padding_Bottom:float=0
   ) -> (then)
   n9 = SelectInt(A=8, B=3, bPickA=@n7)
   > SetGridSlotColumnAndRow(Widget=Get SizeBox_Buttons(), Row=@n9, Row Span=0, Column=0, Column Span=0)
   > SetGridSlotColumnAndRow(Widget=Get SizeBox_Restrictions() -> (SizeBox_Restrictions:Size Box Object Reference), Row=@n9, Row Span=0, Column=1, Column Span=0)
  ]}
 SetupComponents {nodes [
   SetupComponents() {is=title K2Node_FunctionEntry}
   > Set Components(Components:Array of BPI Item Shop PDP Interfaces=Make Array!(
     "[0]":BPI Item Shop PDP Interface=Get Component_Price()
     "[1]":BPI Item Shop PDP Interface=Get Component_Button_Endorse() -> (Component_Button_Endorse:WBP Item Shop PDP Endorsement Object Reference)
    ) -> (Array:Array of BPI Item Shop PDP Interfaces)) -> (Output_Get:Array of BPI Item Shop PDP Interfaces)
  ]}
 UpdateButtons {nodes [
   UpdateButtons() {is=title K2Node_FunctionEntry}
   n1 = Get CurrentOfferVM()
   n2 = Get OfferInfo(self=@n1)
   > Set Text(
    Text=GetPriceHeaderText!(
     self
     GrantedItems:Array of Fort Item VM Object References=Get GrantedItems!(self:Fort Item Shop Offer VM Object Reference=@n1) -> (GrantedItems)
     OfferType:ECatalogOfferType Enum=GetOfferType!(self:Fort Store Front Offer Info Object Reference=@n2) -> (ReturnValue:ECatalogOfferType Enum=StaticPrice)
     IsOwned:bool=Get bOwned!(self:Fort Item Shop Offer VM Object Reference=@n1) -> (bOwned:bool)
     CurrentView:EFortItemShopPdpView Enum=Get CurrentView() -> (CurrentView=OfferInspect)
    ) -> (Text:Text) {is=title K2Node_CallFunction}
    self=Get Text_PriceHeader() -> (Text_PriceHeader:Common Text Object Reference)
   )
   n3 = Get CurrentOfferVM()
   > Set Visibility(
    Visibility=GetViewBundleButtonVisibility!(
     self
     InOfferVM:Fort Item Shop Offer VM Object Reference=Get CurrentOfferVM()
     InParentOfferVM:Fort Item Shop Offer VM Object Reference=Get ParentOfferVM()
     InOfferGroupVM:Fort Item Shop Offer Group VM Object Reference=Get FortItemShopOfferGroupVM() -> (FortItemShopOfferGroupVM:Fort Item Shop Offer Group VM Object Reference)
     InCurrentView:EFortItemShopPdpView Enum=Get CurrentView() -> (CurrentView=OfferInspect)
     CanUndoPurchase:bool=Get bCanUndoPurchase!(self:Fort Item Shop Offer VM Object Reference=@n3) -> (bCanUndoPurchase:bool)
     GrantedItems:Array of Fort Item VM Object References=Get GrantedItems(self=@n3)
    ) -> (ReturnValue:ESlateVisibility Enum=Visible) {is=title K2Node_CallFunction}
    self=Get VerticalBox_ViewBundle()
   )
  ]}
 VM_ActiveContentChanged {nodes [
   n1 = VM_ActiveContentChanged!() -> (then, ActiveWidget:Widget Object Reference) {is=title K2Node_FunctionEntry}
   n2 = Get Switcher_Content()
   > n3 = Branch(Condition="Equal (Object)"(A=@n1, B=GetActiveWidget!(self:Common Visibility Switcher Object Reference=@n2) -> (ReturnValue:Widget Object Reference)))
   SetActiveWidget(execute=@n3.else, self=Get Switcher_Content(), Widget=Get ActiveWidget() -> (ActiveWidget:Widget Object Reference))
   > RefreshFocusIfLeafmostDescendant!(execute) -> (then, ReturnValue:bool)
  ]}
 VM_ItemForRestrictionChanged {nodes [
   n1 = VM_ItemForRestrictionChanged!() -> (then, ItemForRestriction:Fort Item VM Object Reference) {is=title K2Node_FunctionEntry}
   > UpdateLegalText!(execute, self:Fort Item Shop Offer VM Object Reference=Get CurrentOfferVM(), FocusedItem:Fort Item VM Object Reference=@n1) -> (then)
  ]}
 VM_RestrictionTextChanged {nodes [
   VM_RestrictionTextChanged!() -> (then, RestrictionText:"Text (by ref)") {is=title K2Node_FunctionEntry}
   n1 = Get RestrictionText() -> (RestrictionText:"Text (by ref)")
   > n2 = Set Visibility(
    Visibility=Select(Option 0:ESlateVisibility Enum=SelfHitTestInvisible, Option 1:ESlateVisibility Enum=Collapsed, Index=TextIsEmpty(InText=@n1)) -> (ReturnValue:ESlateVisibility Enum)
    self=Get SizeBox_Restrictions()
   )
   > Branch(Condition="Not Equal (Enum)"(A:ESlateVisibility Enum=@n2, B:ESlateVisibility Enum=Collapsed))
   > Set AutoWrapText!(execute, AutoWrapText:bool=false, self:Text Layout Widget Object Reference=Get RichText_Restriction()) -> (then, Output_Get:bool)
   n3 = Get RichText_Restriction()
   > Set Text(Text=Get RestrictionText(), self:Rich Text Block Object Reference=@n3)
   > ForceLayoutPrepass(self:Widget Object Reference=@n3)
   n4 = GetDesiredSize!(self:Widget Object Reference=@n3) -> (ReturnValue_X:double, ReturnValue_Y:double)
   n5 = Get HeightOverride!(self:Size Box Object Reference=Get SizeBox_Restrictions()) -> (HeightOverride:float)
   > Set WidthOverride!(
    execute
    WidthOverride:float=SelectFloat(A=1000, B=656, bPickA="float > float"(A=@n4.ReturnValue_Y, B=@n5))
    self:Size Box Object Reference=Get SizeBox_Restrictions()
   ) -> (then, Output_Get:float)
   > Set AutoWrapText(AutoWrapText=true, self=Get RichText_Restriction())
  ]}
]
variables [
 Components {type="array<object>", default []}
 CurrentView {type=EFortItemShopPdpView, default=OfferInspect}
 Is Focusing Items {type=boolean}
 OnGetVbucksClicked {}
 OnGiftClicked {}
 OnGoToItemClicked {}
 OnPurchaseBlockHovered {}
 OnPurchaseClicked {}
 OnPurchaseHistoryClicked {}
 OnUndoPurchaseClicked {}
 OnViewBundleClicked {}
]
widgets {
 info {inheritedWidgetCount=0, parentClass=&UserWidget, rootWidgetClass=&GridPanel, widgetCount=33}
 widgets [
  GridPanel_1 {widgetClassPath=&GridPanel, widget=GridPanel_1}
  Text_IncludedItems {
   widgetClassPath=&CommonRichTextBlock
   bIsVariable=true
   parent=GridPanel_1
   widget=Text_IncludedItems
   properties {
    bOverrideDefaultStyle=true
    mobileTextBlockScale=1.1
    text="[PH] Items included with this purchase: <text fontface=\"75 Medium\">Rumi Outfit Fortnite Style, Rumi Outfit Lego Style, Rumi's Ramyeon Backbling</>"
    visibility=Collapsed
    wrapTextAt=600
    defaultTextStyleOverride {
     overflowPolicy=Clip
     transformPolicy=None
     colorAndOpacity {colorUseRule=UseColor_Specified, specifiedColor=<rgba 1 1 1 1>}
     font {
      bForceMonospaced=false
      bMaterialIsStencil=false
      fontMaterial=null
      fontObject=&HeadingNow
      letterSpacing=0
      monospacedWidth=1
      size=14.25
      skewAmount=0
      typefaceFontName=74 Regular
      outlineSettings {
       bApplyOutlineToDropShadows=false
       bMiteredCorners=false
       bSeparateFillAlpha=false
       outlineMaterial=null
       outlineSize=0
       outlineColor=<rgba 0 0 0 1>
      }
     }
     highlightColor {colorUseRule=UseColor_Specified, specifiedColor=<rgba 0 0 0 1>}
     highlightShape {
      bIsDynamicallyLoaded=false
      drawAs=Image
      imageType=NoImage
      mirroring=NoMirror
      resourceName=None
      resourceObject=null
      tiling=NoTile
      imageSize=<32 32>
      margin=<ltrb 0 0 0 0>
      outlineSettings {
       bUseBrushTransparency=false
       roundingType=HalfHeightRadius
       width=0
       color {colorUseRule=UseColor_Specified, specifiedColor=<rgba 0 0 0 0>}
       cornerRadii=<0 0 0 0>
      }
      tintColor {colorUseRule=UseColor_Specified, specifiedColor=<rgba 1 1 1 1>}
      uVRegion {bIsValid=false, max=<0 0>, min=<0 0>}
     }
     selectedBackgroundColor {colorUseRule=UseColor_Foreground, specifiedColor=<rgba 1 0 1 1>}
     shadowColorAndOpacity=<rgba 0 0 0 1>
     shadowOffset=<0 0>
     strikeBrush {
      bIsDynamicallyLoaded=false
      drawAs=Image
      imageType=NoImage
      mirroring=NoMirror
      resourceName=None
      resourceObject=null
      tiling=NoTile
      imageSize=<32 32>
      margin=<ltrb 0 0 0 0>
      outlineSettings {
       bUseBrushTransparency=false
       roundingType=HalfHeightRadius
       width=0
       color {colorUseRule=UseColor_Specified, specifiedColor=<rgba 0 0 0 0>}
       cornerRadii=<0 0 0 0>
      }
      tintColor {colorUseRule=UseColor_Specified, specifiedColor=<rgba 1 1 1 1>}
      uVRegion {bIsValid=false, max=<0 0>, min=<0 0>}
     }
     underlineBrush {
      bIsDynamicallyLoaded=false
      drawAs=Image
      imageType=NoImage
      mirroring=NoMirror
      resourceName=None
      resourceObject=null
      tiling=NoTile
      imageSize=<32 32>
      margin=<ltrb 0 0 0 0>
      outlineSettings {
       bUseBrushTransparency=false
       roundingType=HalfHeightRadius
       width=0
       color {colorUseRule=UseColor_Specified, specifiedColor=<rgba 0 0 0 0>}
       cornerRadii=<0 0 0 0>
      }
      tintColor {colorUseRule=UseColor_Specified, specifiedColor=<rgba 1 1 1 1>}
      uVRegion {bIsValid=false, max=<0 0>, min=<0 0>}
     }
    }
   }
   slot {class=GridSlot, properties {horizontalAlignment=HAlign_Left, padding=<ltrb 0 0 0 16>}}
  }
  Switcher_Content {
   widgetClassPath=&CommonVisibilitySwitcher
   bIsVariable=true
   parent=GridPanel_1
   widget=Switcher_Content
   slot {class=GridSlot, properties {horizontalAlignment=HAlign_Left, row=1}}
  }
  Grid_PurchaseBlock {
   widgetClassPath=&GridPanel
   bIsVariable=true
   parent=Switcher_Content
   widget=Grid_PurchaseBlock
   slot {class=CommonVisibilitySwitcherSlot}
  }
  Text_PriceHeader {
   widgetClassPath=&CommonTextBlock
   bIsVariable=true
   parent=Grid_PurchaseBlock
   widget=Text_PriceHeader
   properties {
    bAutoCollapseWithEmptyText=true
    mobileFontSizeMultiplier=1.26
    text="[PH] The price for included items is:"
    font {
     bForceMonospaced=false
     bMaterialIsStencil=false
     fontMaterial=null
     fontObject=&HeadingNow
     letterSpacing=0
     monospacedWidth=1
     size=14.25
     skewAmount=0
     typefaceFontName=74 Regular
     outlineSettings {
      bApplyOutlineToDropShadows=false
      bMiteredCorners=false
      bSeparateFillAlpha=false
      outlineMaterial=null
      outlineSize=0
      outlineColor=<rgba 0 0 0 1>
     }
    }
    shadowOffset=<0 0>
   }
   slot {class=GridSlot, properties {row=1, padding=<ltrb 0 0 0 8>}}
  }
  Component_Price {
   widgetClassPath=&/WBP_ItemShopPDP_Price
   bIsVariable=true
   parent=Grid_PurchaseBlock
   widget=Component_Price
   properties {realPriceLegalText_Optional=&self:WidgetTree.Text_RealPriceLegal}
   slot {class=GridSlot, properties {row=2, padding=<ltrb 0 0 0 24>}}
  }
  SizeBox_Buttons {
   widgetClassPath=&SizeBox
   bIsVariable=true
   parent=Grid_PurchaseBlock
   widget=SizeBox_Buttons
   properties {bOverride_WidthOverride=true, maxDesiredWidth=840, minDesiredWidth=604, widthOverride=600}
   slot {class=GridSlot, properties {layer=1, row=3, verticalAlignment=VAlign_Center}}
  }
  Switcher_ButtonLock {
   widgetClassPath=&CommonVisibilitySwitcher
   bIsVariable=true
   parent=SizeBox_Buttons
   widget=Switcher_ButtonLock
   slot {class=SizeBoxSlot}
  }
  GridPanel_Buttons {
   widgetClassPath=&GridPanel
   bIsVariable=true
   parent=Switcher_ButtonLock
   widget=GridPanel_Buttons
   properties {bOverride_Cursor=true, columnFill [1]}
   slot {class=CommonVisibilitySwitcherSlot}
  }
  HorizontalBox_Buttons {widgetClassPath=&HorizontalBox, bIsVariable=true, parent=GridPanel_Buttons, widget=HorizontalBox_Buttons, slot {class=GridSlot}}
  Button_Purchase {
   widgetClassPath=/FortUILibrary/UIKit/Buttons/WBP_UIKit_Button_Loud
   bIsVariable=true
   parent=HorizontalBox_Buttons
   widget=Button_Purchase
   properties {
    bRequiresHold=true
    holdData=/FortUILibrary/Input/PurchaseInteraction_HoldData.PurchaseInteraction_HoldData_C
    minHeight=64
    minWidth=300
    text=Purchase
   }
   slot {class=HorizontalBoxSlot, properties {padding=<ltrb 0 0 24 0>, size {sizeRule=Fill, value=0.65}}}
  }
  Button_GetVBucks {
   widgetClassPath=&WBP_UIKit_Button_Loud
   bIsVariable=true
   parent=HorizontalBox_Buttons
   widget=Button_GetVBucks
   properties {bShouldUseFallbackDefaultInputAction=false, minHeight=64, minWidth=300, text=Get V-Bucks}
   slot {class=HorizontalBoxSlot, properties {padding=<ltrb 0 0 24 0>, size {sizeRule=Fill, value=0.65}}}
  }
  Button_UndoPurchase {
   widgetClassPath=&WBP_UIKit_Button_Regular
   bIsVariable=true
   parent=HorizontalBox_Buttons
   widget=Button_UndoPurchase
   properties {
    bRequiresHold=true
    holdData=&PurchaseInteraction_HoldData_C
    minHeight=56
    minWidth=300
    text=Cancel purchase
   }
   slot {class=HorizontalBoxSlot, properties {padding=<ltrb 0 0 24 0>, size {sizeRule=Fill, value=0.65}}}
  }
  Button_Gift {
   widgetClassPath=&WBP_UIKit_Button_Regular
   bIsVariable=true
   parent=HorizontalBox_Buttons
   widget=Button_Gift
   properties {
    bShouldUseFallbackDefaultInputAction=false
    icon=/FortUILibrary/Textures/Icons/T_UI_Icon_UIKit_Gift
    iconBeforeText=false
    iconMaterial=/FortUILibrary/UIKit/Buttons/MaterialInstances/CTAButtons/MI_UIKit_IconQuiet
    minHeight=56
    minWidth=300
    text=""
   }
   slot {class=HorizontalBoxSlot, properties {size {sizeRule=Automatic, value=0.35}}}
  }
  Component_Button_Endorse {
   widgetClassPath=&/WBP_ItemShopPDP_Endorsement
   bIsVariable=true
   parent=HorizontalBox_Buttons
   widget=Component_Button_Endorse
   properties {bIsEnabled=false}
   slot {class=HorizontalBoxSlot, properties {padding=<ltrb 24 0 0 0>}}
  }
  Text_Legal_UndoPurchase {
   widgetClassPath=&CommonTextBlock
   bIsVariable=true
   parent=GridPanel_Buttons
   widget=Text_Legal_UndoPurchase
   properties {
    autoWrapText=true
    lineHeightPercentage=1.15
    mobileFontSizeMultiplier=1.4
    text=Cancel Purchase is available until you equip or use your purchase during gameplay. Only items purchased within the last 24 hours are eligible for Cancel Purchase.
    colorAndOpacity {colorUseRule=UseColor_Specified, specifiedColor=<rgba 1 1 1 0.65>}
    font {
     bForceMonospaced=false
     bMaterialIsStencil=false
     fontMaterial=null
     fontObject=&HeadingNow
     letterSpacing=0
     monospacedWidth=1
     size=11.25
     skewAmount=0
     typefaceFontName=74 Regular
     outlineSettings {
      bApplyOutlineToDropShadows=false
      bMiteredCorners=false
      bSeparateFillAlpha=false
      outlineMaterial=null
      outlineSize=0
      outlineColor=<rgba 0 0 0 1>
     }
    }
    shadowOffset=<0 0>
   }
   slot {class=GridSlot, properties {row=1, padding=<ltrb 0 16 0 0>}}
  }
  VerticalBox_ViewBundle {
   widgetClassPath=&VerticalBox
   bIsVariable=true
   parent=GridPanel_Buttons
   widget=VerticalBox_ViewBundle
   slot {class=GridSlot, properties {row=2, padding=<ltrb 0 16 0 8>}}
  }
  Button_ViewBundle {
   widgetClassPath=&WBP_UIKit_Button_Regular
   bIsVariable=true
   parent=VerticalBox_ViewBundle
   widget=Button_ViewBundle
   properties {bShouldUseFallbackDefaultInputAction=false, buttonSize=Small, text=View Bundle}
   slot {class=VerticalBoxSlot}
  }
  RichText_ViewBundleHeader {
   widgetClassPath=&CommonRichTextBlock
   bIsVariable=true
   parent=VerticalBox_ViewBundle
   widget=RichText_ViewBundleHeader
   properties {
    bAutoCollapseWithEmptyText=true
    bOverrideDefaultStyle=true
    justification=Center
    mobileTextBlockScale=1.1
    text="[PH] Part of the <text color=\"FFFFFF\" fontface=\"75 medium\">{BundleName}</>"
    wrapTextAt=640
    defaultTextStyleOverride {
     overflowPolicy=Clip
     transformPolicy=None
     colorAndOpacity {colorUseRule=UseColor_Specified, specifiedColor=<rgba 1 1 1 1>}
     font {
      bForceMonospaced=false
      bMaterialIsStencil=false
      fontMaterial=null
      fontObject=&HeadingNow
      letterSpacing=0
      monospacedWidth=1
      size=14.25
      skewAmount=0
      typefaceFontName=74 Regular
      outlineSettings {
       bApplyOutlineToDropShadows=false
       bMiteredCorners=false
       bSeparateFillAlpha=false
       outlineMaterial=null
       outlineSize=0
       outlineColor=<rgba 0 0 0 1>
      }
     }
     highlightColor {colorUseRule=UseColor_Specified, specifiedColor=<rgba 0 0 0 1>}
     highlightShape {
      bIsDynamicallyLoaded=false
      drawAs=Image
      imageType=NoImage
      mirroring=NoMirror
      resourceName=None
      resourceObject=null
      tiling=NoTile
      imageSize=<32 32>
      margin=<ltrb 0 0 0 0>
      outlineSettings {
       bUseBrushTransparency=false
       roundingType=HalfHeightRadius
       width=0
       color {colorUseRule=UseColor_Specified, specifiedColor=<rgba 0 0 0 0>}
       cornerRadii=<0 0 0 0>
      }
      tintColor {colorUseRule=UseColor_Specified, specifiedColor=<rgba 1 1 1 1>}
      uVRegion {bIsValid=false, max=<0 0>, min=<0 0>}
     }
     selectedBackgroundColor {colorUseRule=UseColor_Foreground, specifiedColor=<rgba 1 0 1 1>}
     shadowColorAndOpacity=<rgba 0 0 0 1>
     shadowOffset=<0 0>
     strikeBrush {
      bIsDynamicallyLoaded=false
      drawAs=Image
      imageType=NoImage
      mirroring=NoMirror
      resourceName=None
      resourceObject=null
      tiling=NoTile
      imageSize=<32 32>
      margin=<ltrb 0 0 0 0>
      outlineSettings {
       bUseBrushTransparency=false
       roundingType=HalfHeightRadius
       width=0
       color {colorUseRule=UseColor_Specified, specifiedColor=<rgba 0 0 0 0>}
       cornerRadii=<0 0 0 0>
      }
      tintColor {colorUseRule=UseColor_Specified, specifiedColor=<rgba 1 1 1 1>}
      uVRegion {bIsValid=false, max=<0 0>, min=<0 0>}
     }
     underlineBrush {
      bIsDynamicallyLoaded=false
      drawAs=Image
      imageType=NoImage
      mirroring=NoMirror
      resourceName=None
      resourceObject=null
      tiling=NoTile
      imageSize=<32 32>
      margin=<ltrb 0 0 0 0>
      outlineSettings {
       bUseBrushTransparency=false
       roundingType=HalfHeightRadius
       width=0
       color {colorUseRule=UseColor_Specified, specifiedColor=<rgba 0 0 0 0>}
       cornerRadii=<0 0 0 0>
      }
      tintColor {colorUseRule=UseColor_Specified, specifiedColor=<rgba 1 1 1 1>}
      uVRegion {bIsValid=false, max=<0 0>, min=<0 0>}
     }
    }
   }
   slot {class=VerticalBoxSlot, properties {padding=<ltrb 0 0 0 10>}}
  }
  Button_PurchaseHistory {
   widgetClassPath=&WBP_UIKit_Button_Regular
   bIsVariable=true
   parent=GridPanel_Buttons
   widget=Button_PurchaseHistory
   properties {bShouldUseFallbackDefaultInputAction=false, buttonSize=Small, text=Cancel / Return, visibility=Collapsed}
   slot {class=GridSlot, properties {row=3, padding=<ltrb 0 16 0 0>}}
  }
  Slot_InjectedButton {
   widgetClassPath=&NamedSlot
   bIsVariable=true
   parent=GridPanel_Buttons
   widget=Slot_InjectedButton
   slot {class=GridSlot, properties {row=4, padding=<ltrb 0 16 0 0>}}
  }
  Button_LockedPurchase {
   widgetClassPath=/ItemShop/MultiProduct/WBP_ItemShop_LockedPurchaseButton
   bIsVariable=true
   parent=Switcher_ButtonLock
   widget=Button_LockedPurchase
   slot {class=CommonVisibilitySwitcherSlot}
  }
  SizeBox_Restrictions {
   widgetClassPath=&SizeBox
   bIsVariable=true
   parent=Grid_PurchaseBlock
   widget=SizeBox_Restrictions
   properties {bOverride_HeightOverride=true, bOverride_WidthOverride=true, heightOverride=46, maxDesiredWidth=656, widthOverride=656}
   slot {class=GridSlot, properties {column=1, horizontalAlignment=HAlign_Left, row=3, padding=<ltrb 24 0 0 0>}}
  }
  Overlay_Restrictions {
   widgetClassPath=&Overlay
   parent=SizeBox_Restrictions
   widget=Overlay_Restrictions
   properties {flowDirectionPreference=LeftToRight}
   slot {class=SizeBoxSlot}
  }
  Image_Separator {
   widgetClassPath=&Image
   parent=Overlay_Restrictions
   widget=Image_Separator
   properties {
    visibility=SelfHitTestInvisible
    brush {
     bIsDynamicallyLoaded=false
     drawAs=Image
     imageType=FullColor
     mirroring=NoMirror
     resourceName=None
     resourceObject=null
     tiling=NoTile
     imageSize=<2 32>
     margin=<ltrb 0 0 0 0>
     outlineSettings {
      bUseBrushTransparency=false
      roundingType=HalfHeightRadius
      width=0
      color {colorUseRule=UseColor_Specified, specifiedColor=<rgba 0 0 0 0>}
      cornerRadii=<0 0 0 0>
     }
     tintColor {colorUseRule=UseColor_Specified, specifiedColor=<rgba 1 1 1 1>}
     uVRegion {bIsValid=false, max=<0 0>, min=<0 0>}
    }
    colorAndOpacity=<rgba 1 1 1 0.25>
   }
   slot {class=OverlaySlot, properties {verticalAlignment=VAlign_Fill}}
  }
  RestrictionWarning_Bottom {
   widgetClassPath=&HorizontalBox
   parent=Overlay_Restrictions
   widget=RestrictionWarning_Bottom
   properties {flowDirectionPreference=Culture}
   slot {class=OverlaySlot, properties {horizontalAlignment=HAlign_Fill, verticalAlignment=VAlign_Fill, padding=<ltrb 16 0 0 0>}}
  }
  Image_Warning {
   widgetClassPath=&Image
   parent=RestrictionWarning_Bottom
   widget=Image_Warning
   properties {
    brush {
     bIsDynamicallyLoaded=false
     drawAs=Image
     imageType=FullColor
     mirroring=NoMirror
     resourceName=None
     resourceObject=/Game/Athena/UI/Frontend/CosmeticGating/Materials/MI_UI_CosmeticGate_WarningIcon
     tiling=NoTile
     imageSize=<24 24>
     margin=<ltrb 0 0 0 0>
     outlineSettings {
      bUseBrushTransparency=false
      roundingType=HalfHeightRadius
      width=0
      color {colorUseRule=UseColor_Specified, specifiedColor=<rgba 0 0 0 0>}
      cornerRadii=<0 0 0 0>
     }
     tintColor {colorUseRule=UseColor_Specified, specifiedColor=<rgba 1 1 1 1>}
     uVRegion {bIsValid=false, max=<0 0>, min=<0 0>}
    }
    colorAndOpacity=<rgba 1 1 1 0.6>
   }
   slot {class=HorizontalBoxSlot, properties {verticalAlignment=VAlign_Center}}
  }
  RichText_Restriction {
   widgetClassPath=&CommonRichTextBlock
   bIsVariable=true
   parent=RestrictionWarning_Bottom
   widget=RichText_Restriction
   properties {
    autoWrapText=true
    bIsScrollingEnabled=false
    bOverrideDefaultStyle=true
    mobileTextBlockScale=1.6
    scrollOrientation=Orient_Vertical
    scrollStyle=&TextScrollStyle-Base_C
    text="[PH] Items with the <img id=\"WarningIcon\"/> symbol can only be used in experiences rated Teen."
    decoratorClasses [/Game/Athena/UI/Frontend/CosmeticGating/Data/RichTextDecorator_CosmeticGate.RichTextDecorator_CosmeticGate_C]
    defaultTextStyleOverride {
     overflowPolicy=Clip
     transformPolicy=None
     colorAndOpacity {colorUseRule=UseColor_Specified, specifiedColor=<rgba 1 1 1 0.65>}
     font {
      bForceMonospaced=false
      bMaterialIsStencil=false
      fontMaterial=null
      fontObject=&HeadingNow
      letterSpacing=10
      monospacedWidth=1
      size=11.25
      skewAmount=0
      typefaceFontName=74 Regular
      outlineSettings {
       bApplyOutlineToDropShadows=false
       bMiteredCorners=false
       bSeparateFillAlpha=false
       outlineMaterial=null
       outlineSize=0
       outlineColor=<rgba 0 0 0 1>
      }
     }
     highlightColor {colorUseRule=UseColor_Specified, specifiedColor=<rgba 0 0 0 1>}
     highlightShape {
      bIsDynamicallyLoaded=false
      drawAs=Image
      imageType=NoImage
      mirroring=NoMirror
      resourceName=None
      resourceObject=null
      tiling=NoTile
      imageSize=<32 32>
      margin=<ltrb 0 0 0 0>
      outlineSettings {
       bUseBrushTransparency=false
       roundingType=HalfHeightRadius
       width=0
       color {colorUseRule=UseColor_Specified, specifiedColor=<rgba 0 0 0 0>}
       cornerRadii=<0 0 0 0>
      }
      tintColor {colorUseRule=UseColor_Specified, specifiedColor=<rgba 1 1 1 1>}
      uVRegion {bIsValid=false, max=<0 0>, min=<0 0>}
     }
     selectedBackgroundColor {colorUseRule=UseColor_Foreground, specifiedColor=<rgba 1 0 1 1>}
     shadowColorAndOpacity=<rgba 0 0 0 1>
     shadowOffset=<0 0>
     strikeBrush {
      bIsDynamicallyLoaded=false
      drawAs=Image
      imageType=NoImage
      mirroring=NoMirror
      resourceName=None
      resourceObject=null
      tiling=NoTile
      imageSize=<32 32>
      margin=<ltrb 0 0 0 0>
      outlineSettings {
       bUseBrushTransparency=false
       roundingType=HalfHeightRadius
       width=0
       color {colorUseRule=UseColor_Specified, specifiedColor=<rgba 0 0 0 0>}
       cornerRadii=<0 0 0 0>
      }
      tintColor {colorUseRule=UseColor_Specified, specifiedColor=<rgba 1 1 1 1>}
      uVRegion {bIsValid=false, max=<0 0>, min=<0 0>}
     }
     underlineBrush {
      bIsDynamicallyLoaded=false
      drawAs=Image
      imageType=NoImage
      mirroring=NoMirror
      resourceName=None
      resourceObject=null
      tiling=NoTile
      imageSize=<32 32>
      margin=<ltrb 0 0 0 0>
      outlineSettings {
       bUseBrushTransparency=false
       roundingType=HalfHeightRadius
       width=0
       color {colorUseRule=UseColor_Specified, specifiedColor=<rgba 0 0 0 0>}
       cornerRadii=<0 0 0 0>
      }
      tintColor {colorUseRule=UseColor_Specified, specifiedColor=<rgba 1 1 1 1>}
      uVRegion {bIsValid=false, max=<0 0>, min=<0 0>}
     }
    }
   }
   slot {class=HorizontalBoxSlot, properties {verticalAlignment=VAlign_Center, padding=<ltrb 8 0 0 0>, size {sizeRule=Fill, value=1}}}
  }
  Text_OutDate {
   widgetClassPath=&CommonTextBlock
   bIsVariable=true
   parent=Grid_PurchaseBlock
   widget=Text_OutDate
   properties {
    bAutoCollapseWithEmptyText=true
    lineHeightPercentage=1.15
    mobileFontSizeMultiplier=1.4
    text="[PH] This item will be removed from the Item Shop on 01/01/24 at 6:00 PM your local time (GMT-4)."
    wrapTextAt=600
    colorAndOpacity {colorUseRule=UseColor_Specified, specifiedColor=<rgba 1 1 1 0.65>}
    font {
     bForceMonospaced=false
     bMaterialIsStencil=false
     fontMaterial=null
     fontObject=&HeadingNow
     letterSpacing=0
     monospacedWidth=1
     size=11.25
     skewAmount=0
     typefaceFontName=74 Regular
     outlineSettings {
      bApplyOutlineToDropShadows=false
      bMiteredCorners=false
      bSeparateFillAlpha=false
      outlineMaterial=null
      outlineSize=0
      outlineColor=<rgba 0 0 0 1>
     }
    }
    shadowOffset=<0 0>
   }
   slot {class=GridSlot, properties {row=4, padding=<ltrb 0 8 0 0>}}
  }
  Text_RefundDisclaimer {
   widgetClassPath=&CommonTextBlock
   bIsVariable=true
   parent=Grid_PurchaseBlock
   widget=Text_RefundDisclaimer
   properties {
    lineHeightPercentage=1.15
    mobileFontSizeMultiplier=1.4
    text=This purchase is not eligible for refund.
    visibility=Collapsed
    wrapTextAt=600
    colorAndOpacity {colorUseRule=UseColor_Specified, specifiedColor=<rgba 1 1 1 0.65>}
    font {
     bForceMonospaced=false
     bMaterialIsStencil=false
     fontMaterial=null
     fontObject=&HeadingNow
     letterSpacing=0
     monospacedWidth=1
     size=11.25
     skewAmount=0
     typefaceFontName=74 Regular
     outlineSettings {
      bApplyOutlineToDropShadows=false
      bMiteredCorners=false
      bSeparateFillAlpha=false
      outlineMaterial=null
      outlineSize=0
      outlineColor=<rgba 0 0 0 1>
     }
    }
    shadowOffset=<0 0>
   }
   slot {class=GridSlot, properties {row=5, padding=<ltrb 0 8 0 0>}}
  }
  Text_OfferLegal {
   widgetClassPath=&CommonTextBlock
   bIsVariable=true
   parent=Grid_PurchaseBlock
   widget=Text_OfferLegal
   properties {
    bAutoCollapseWithEmptyText=true
    lineHeightPercentage=1.15
    mobileFontSizeMultiplier=1.4
    text="[PH] Kicks are usable with most Outfits. For exceptions and other info see fn.gg/KicksUsability"
    wrapTextAt=600
    colorAndOpacity {colorUseRule=UseColor_Specified, specifiedColor=<rgba 1 1 1 0.65>}
    font {
     bForceMonospaced=false
     bMaterialIsStencil=false
     fontMaterial=null
     fontObject=&HeadingNow
     letterSpacing=0
     monospacedWidth=1
     size=11.25
     skewAmount=0
     typefaceFontName=74 Regular
     outlineSettings {
      bApplyOutlineToDropShadows=false
      bMiteredCorners=false
      bSeparateFillAlpha=false
      outlineMaterial=null
      outlineSize=0
      outlineColor=<rgba 0 0 0 1>
     }
    }
    shadowOffset=<0 0>
   }
   slot {class=GridSlot, properties {row=6, padding=<ltrb 0 8 0 0>}}
  }
  Text_RealPriceLegal {
   widgetClassPath=&CommonTextBlock
   bIsVariable=true
   parent=Grid_PurchaseBlock
   widget=Text_RealPriceLegal
   properties {
    bAutoCollapseWithEmptyText=true
    mobileFontSizeMultiplier=1.4
    text="[PH] *To buy exactly 1,000 V-Bucks, you’d pay €8.99."
    wrapTextAt=600
    colorAndOpacity {colorUseRule=UseColor_Specified, specifiedColor=<rgba 1 1 1 0.65>}
    font {
     bForceMonospaced=false
     bMaterialIsStencil=false
     fontMaterial=null
     fontObject=&HeadingNow
     letterSpacing=0
     monospacedWidth=1
     size=11.25
     skewAmount=0
     typefaceFontName=74 Regular
     outlineSettings {
      bApplyOutlineToDropShadows=false
      bMiteredCorners=false
      bSeparateFillAlpha=false
      outlineMaterial=null
      outlineSize=0
      outlineColor=<rgba 0 0 0 1>
     }
    }
    shadowOffset=<0 0>
   }
   slot {class=GridSlot, properties {row=7, padding=<ltrb 0 8 0 0>}}
  }
  Button_GoToItem {
   widgetClassPath=&WBP_UIKit_Button_Regular
   bIsVariable=true
   parent=Switcher_Content
   widget=Button_GoToItem
   properties {buttonSize=Small, minWidth=500, minWidthOverride=500, text=View Item}
   slot {class=CommonVisibilitySwitcherSlot, properties {padding=<ltrb 0 8 0 24>}}
  }
 ]
}
