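# Translucent UI material for an item shop multi-product backplate, drawing a faint rounded rectangle with a more opaque outline.
# The MF_UI_CoreUI_RoundedMask function builds shape and outline masks from a pixel-based corner radius and outline thickness.
# Named reroutes carry the masks into a Lerp between fill and outline alpha levels, clipped by a Min, alongside a constant near-white color.
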Material /ItemShop/MultiProduct/Art/M_UI_BackplateWithOutline
graph {
 properties {
  bAutomaticallySetUsageInEditor=false
  bUsedWithStaticMesh=false
  blendMode=BLEND_Translucent
  materialDomain=MD_UI
  usedShadingModels=Default Lit
 }
 nodes [
  Constant3Vector() {constant=<rgba 0.9216 0.9301 0.9301 0.102>}
  Min(A=NamedRerouteUsage(), B=LinearInterpolate(Alpha=NamedRerouteUsage()) {constA=0.1, constB=0.4})
  n1 = MaterialFunctionCall(Corner Radius=Constant4Vector() {constant=<rgba 16 16 16 16>}.A, Radius in Pixels=Constant() {r=1}, Outline Thickness=Constant() {r=3}) {materialFunction=&MF_UI_CoreUI_RoundedMask}
  NamedRerouteDeclaration(None=@n1.Shape Mask) {name=ShapeMask, nodeColor=<rgba 1 0.8203 0 1>}
  NamedRerouteDeclaration(None=@n1.Outline Mask) {name=OutlineMask, nodeColor=<rgba 0.5 0 1 1>}
 ]
}
