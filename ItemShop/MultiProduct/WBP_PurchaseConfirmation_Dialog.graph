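# Item shop purchase confirmation dialog: an activatable widget wrapping the UIKit dialog with confirmation text and a matchmaking-ban notice.
# PreConstruct supplies a DialogVM_PurchaseConfirmation as dialog data.
# While active, button clicks 0 and 1 fire PurchaseConfirmed and PurchaseDismissed, then deactivate it.
# On activation, SetABTestingStyle left-aligns title, body and buttons for compact UI when the config flag is on, outside undo cooldown, with ban notice hidden.
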WidgetBlueprint /ItemShop/MultiProduct/WBP_PurchaseConfirmation_Dialog
graphs [
 EventGraph {nodes [
   n1 = OnButtonClicked!() -> (OutputDelegate, then, ButtonIndex:int) {is=title K2Node_CustomEvent}
   > n2 = Switch on Int!(execute:Exec, Selection:int=@n1.ButtonIndex) -> (Default:Exec, 0:Exec, 1:Exec)
   n3 = Call Purchase Confirmed!(execute:Exec=@n2.0, self:WBP Purchase Confirmation Dialog Object Reference) -> (then:Exec) {is=title K2Node_CallDelegate}
   DeactivateWidget(execute=[@n3.then, @n4.then])
   n4 = Call Purchase Dismissed!(execute:Exec=@n2.1, self:WBP Purchase Confirmation Dialog Object Reference) -> (then:Exec) {is=title K2Node_CallDelegate}
   Event PreConstruct()
   > n5 = Construct Object from Class(Class=/ItemShop/MultiProduct/Data/DialogVM_PurchaseConfirmation.DialogVM_PurchaseConfirmation_C) -> (ReturnValue:Dialog VM Purchase Confirmation Object Reference)
   > SetDialogData!(
    execute
    self:WBP UIKit Dialog Internal Object Reference=Get Dialog_Base() -> (Dialog_Base:WBP UIKit Dialog Internal Object Reference)
    Data:UI Kit Dialog Viewmodel Object Reference=@n5
   ) -> (then)
   Event On Activated()
   > SetABTestingStyle()
   > Bind Event to Dialog Button Clicked Dispatcher(self:WBP UIKit Dialog Internal Object Reference=Get Dialog_Base(), Delegate=@n1) {is=title K2Node_AddDelegate}
   Event On Deactivated()
   > Unbind all Events from Dialog Button Clicked Dispatcher!(execute:Exec, self:WBP UIKit Dialog Internal Object Reference=Get Dialog_Base()) -> (then:Exec) {is=title K2Node_ClearDelegate}
  ]}
 GetVisibilityIfTextEmpty {nodes [
   n1 = GetVisibilityIfTextEmpty!() -> (then, InText:Text) {is=title K2Node_FunctionEntry}
   > Return Node!(execute, OutVisibility:ESlateVisibility Enum=Select(Option 0=SelfHitTestInvisible, Option 1=Collapsed, Index=TextIsEmpty(InText=@n1)))
  ]}
 PurchaseConfirmed {nodes [PurchaseConfirmed() {is=title K2Node_FunctionEntry}]}
 PurchaseDismissed {nodes [PurchaseDismissed() {is=title K2Node_FunctionEntry}]}
 SetABTestingStyle {nodes [
   SetABTestingStyle() {is=title K2Node_FunctionEntry}
   n1 = IsVisible(self=Get Border_BanInfo() -> (Border_BanInfo:Border Object Reference))
   n2 = Get bIsInUndoCooldown!(self:Fort Item Shop Offer VM Object Reference=Get FortItemShopOfferVM() -> (FortItemShopOfferVM:Fort Item Shop Offer VM Object Reference)) -> (bIsInUndoCooldown:bool)
   > Branch(Condition=AND Boolean!(
     A=IsUsingCompactUI()
     B=NOT Boolean(A=@n2)
     C:bool=GetConfigDataBoolWithDefault!(EntryName:String=shop_confirm_purchase_through_dialog, bDefaultValue:bool=false) -> (ReturnValue:bool)
     D:bool=NOT Boolean(A=@n1)
    ) -> (ReturnValue))
   > Set Justification!(
    execute
    Justification:ETextJustify Enum=Left
    self:Text Layout Widget Object Reference=Get Text_ConfirmationInfo() -> (Text_ConfirmationInfo:Common Rich Text Block Object Reference)
   ) -> (then, Output_Get:ETextJustify Enum=Left)
   n3 = Get Dialog_Base()
   n4 = Get!(self:WBP UIKit Dialog Internal Object Reference=@n3) -> (
    TitleProperties_Placement_67_#1:S UI Placement Structure
    TitleProperties_Font_45_#2:Slate Font Info Structure
    TitleProperties_Color_65_#3:Slate Color Structure
    TitleProperties_TransformPolicy_49_#4:ETextTransformPolicy Enum
    TitleProperties_Justification_52_#5:ETextJustify Enum
    TitleProperties_LineHeightPercentage_55_#6:double=~
    TitleProperties_AutoWrapText_58_#7:bool=~
    TitleProperties_WrapTextAt_68_#8:int
    TitleProperties_MobileFontSizeMultiplier_73_#9:double=~
   )
   > SetTitleProperties!(
    execute
    self:WBP UIKit Dialog Internal Object Reference=@n3
    Data_Placement_67_#1:S UI Placement Structure=@n4
    Data_Font_45_#2:Slate Font Info Structure=@n4.TitleProperties_Font_45_#2
    Data_Color_65_#3:Slate Color Structure=@n4.TitleProperties_Color_65_#3
    Data_TransformPolicy_49_#4:ETextTransformPolicy Enum=@n4.TitleProperties_TransformPolicy_49_#4
    Data_Justification_52_#5:ETextJustify Enum=Left
    Data_LineHeightPercentage_55_#6:double=@n4.TitleProperties_LineHeightPercentage_55_#6
    Data_AutoWrapText_58_#7:bool=@n4.TitleProperties_AutoWrapText_58_#7
    Data_WrapTextAt_68_#8:int=@n4.TitleProperties_WrapTextAt_68_#8
    Data_MobileFontSizeMultiplier_73_#9:double=@n4.TitleProperties_MobileFontSizeMultiplier_73_#9
   ) -> (then)
   n5 = Get Dialog_Base()
   n6 = Get!(self=@n5) -> (
    Buttons_GroupPlacement_71_#1_HorizontalAlignment_4_#1:EHorizontalAlignment Enum
    Buttons_GroupPlacement_71_#1_VerticalAlignment_5_#10:EVerticalAlignment Enum
    Buttons_GroupPlacement_71_#1_Padding_8_#11:Margin Structure
    Buttons_GroupPlacement_71_#1_Grid_20_#12:S UI Grid Cell Structure
    Buttons_SpaceBetweenButtons_62_#8:double=~
    Buttons_EachButtonSize_90_#13:E_UI_ButtonSize Enum
    Buttons_EachButtonMinWidth_77_#14:int
    Buttons_EachButtonHorizontalFill_87_#15:bool=~
    Buttons_CTAButtonsTypes_89_#16:Array of E_UI_ButtonVisualType Enums
    Buttons_FocuseButton_85_#17:int
   )
   > SetPlacement!(
    execute
    self:WBP UIKit Dialog Internal Object Reference=@n5
    Widget:Widget Object Reference=Get ButtonsContainer!(self:WBP UIKit Dialog Internal Object Reference=@n5) -> (ButtonsContainer:Horizontal Box Object Reference)
    Placement:S UI Placement Structure=Make S UI Placement!(
     HorizontalAlignment_4_#1:EHorizontalAlignment Enum=HAlign_Left
     VerticalAlignment_5_#10:EVerticalAlignment Enum=@n6.Buttons_GroupPlacement_71_#1_VerticalAlignment_5_#10
     Padding_8_#11:Margin Structure=@n6.Buttons_GroupPlacement_71_#1_Padding_8_#11
     Grid_20_#12:S UI Grid Cell Structure=@n6.Buttons_GroupPlacement_71_#1_Grid_20_#12
    ) -> (S_UI_Placement:S UI Placement Structure) {is=title K2Node_MakeStruct}
   ) -> (then)
  ]}
]
variables [
 PurchaseConfirmed {}
 PurchaseDismissed {}
]
widgets {
 info {
  inheritedWidgetCount=0
  namedSlotCount=1
  parentClass=&CommonActivatableWidget
  rootWidgetClass=/FortUILibrary/Systems/Dialog/WBP_UIKit_Dialog_Internal.WBP_UIKit_Dialog_Internal_C
  widgetCount=6
 }
 widgets [
  Dialog_Base {
   widgetClassPath=/FortUILibrary/Systems/Dialog/WBP_UIKit_Dialog_Internal
   bIsVariable=true
   widget=Dialog_Base
   properties {
    designTimePreview_DialogVM=&DialogVM_PurchaseConfirmation_C
    buttons {
     each Button Horizontal Fill=false
     each Button Min Width=0
     each Button Size=Medium
     focuse Button=0
     space Between Buttons=30
     cTA Buttons Types [Loud,Regular]
     group Placement {
      horizontal Alignment=HAlign_Center
      vertical Alignment=VAlign_Bottom
      grid {column=1, column Span=0, layer=0, row=2, row Span=0, nudge=<0 0>}
      padding=<ltrb 80 64 80 48>
     }
    }
    dialogSIze {
     height Override {enable=false, value=800}
     max Aspect Ratio {enable=false, value=0}
     max Desired Height {enable=false, value=0}
     max Desired Width {enable=false, value=0}
     min Aspect Ratio {enable=false, value=0}
     min Desired Height {enable=false, value=0}
     min Desired Width {enable=false, value=0}
     width Override {enable=true, value=1100}
    }
    titleProperties {
     auto Wrap Text=true
     justification=Center
     line Height Percentage=1
     mobile Font Size Multiplier=1.3
     transform Policy=ToUpper
     wrap Text At=0
     color {colorUseRule=UseColor_Specified, specifiedColor=<rgba 1 1 1 1>}
     font {
      bForceMonospaced=false
      bMaterialIsStencil=false
      fontMaterial=null
      fontObject=&HeadingNow
      letterSpacing=0
      monospacedWidth=1
      size=25.5
      skewAmount=0
      typefaceFontName=86 Bold Italic
      outlineSettings {
       bApplyOutlineToDropShadows=false
       bMiteredCorners=false
       bSeparateFillAlpha=false
       outlineMaterial=null
       outlineSize=0
       outlineColor=<rgba 0 0 0 1>
      }
     }
     placement {
      horizontal Alignment=HAlign_Fill
      vertical Alignment=VAlign_Top
      grid {column=1, column Span=0, layer=0, row=0, row Span=0, nudge=<0 0>}
      padding=<ltrb 80 80 80 32>
     }
    }
   }
  }
  Overlay_0 {widgetClassPath=&Overlay, namedSlotHost=Dialog_Base, widget=Overlay_0}
  VerticalBox_49 {
   widgetClassPath=&VerticalBox
   parent=Overlay_0
   widget=VerticalBox_49
   slot {class=OverlaySlot, properties {horizontalAlignment=HAlign_Fill, padding=<ltrb 80 0 80 0>}}
  }
  Border_BanInfo {
   widgetClassPath=&Border
   bIsVariable=true
   parent=VerticalBox_49
   widget=Border_BanInfo
   properties {
    background {
     bIsDynamicallyLoaded=false
     drawAs=Image
     imageType=FullColor
     mirroring=NoMirror
     resourceName=None
     resourceObject=/ItemShop/MultiProduct/Art/M_UI_BackplateWithOutline
     tiling=NoTile
     imageSize=<32 32>
     margin=<ltrb 0 0 0 0>
     outlineSettings {
      bUseBrushTransparency=false
      roundingType=HalfHeightRadius
      width=0
      color {colorUseRule=UseColor_Specified, specifiedColor=<rgba 0 0 0 0>}
      cornerRadii=<0 0 0 0>
     }
     tintColor {colorUseRule=UseColor_Specified, specifiedColor=<rgba 1 1 1 1>}
     uVRegion {bIsValid=false, max=<0 0>, min=<0 0>}
    }
    padding=<ltrb 16 16 16 16>
   }
   slot {class=VerticalBoxSlot, properties {padding=<ltrb 0 0 0 24>}}
  }
  Text_BanInfo {
   widgetClassPath=&CommonTextBlock
   parent=Border_BanInfo
   widget=Text_BanInfo
   properties {
    autoWrapText=true
    bAutoCollapseWithEmptyText=true
    bIsScrollingEnabled=false
    justification=Center
    text="[PH] Your account has an active matchmaking ban. You won't be able to use this item in-game until your ban is lifted on Jan 1, 2000. Would you like to proceed?"
    font {
     bForceMonospaced=false
     bMaterialIsStencil=false
     fontMaterial=null
     fontObject=&HeadingNow
     letterSpacing=0
     monospacedWidth=1
     size=24
     skewAmount=0
     typefaceFontName=86 Bold
     outlineSettings {
      bApplyOutlineToDropShadows=false
      bMiteredCorners=false
      bSeparateFillAlpha=false
      outlineMaterial=null
      outlineSize=0
      outlineColor=<rgba 0 0 0 1>
     }
    }
    shadowOffset=<0 0>
   }
   slot {class=BorderSlot, properties {padding=<ltrb 16 16 16 16>}}
  }
  Text_ConfirmationInfo {
   widgetClassPath=&CommonRichTextBlock
   bIsVariable=true
   parent=VerticalBox_49
   widget=Text_ConfirmationInfo
   properties {
    applyLineHeightToBottomLine=false
    autoWrapText=true
    bAutoCollapseWithEmptyText=true
    bOverrideDefaultStyle=true
    justification=Center
    lineHeightPercentage=1.25
    text="[PH] Please confirm this purchase. Note that Cancel Purchase will not be available for this transaction. You can still return this purchase normally if you have remaining Return Tickets. For more information see <text color=\"F7FF19\">fn.gg/cancelpurchase</>"
    decoratorClasses [/FortniteCoreUI/Frontend/StoreMain/RichTextDecorator_VBucksStore.RichTextDecorator_VBucksStore_C]
    defaultTextStyleOverride {
     overflowPolicy=Clip
     transformPolicy=None
     colorAndOpacity {colorUseRule=UseColor_Specified, specifiedColor=<rgba 1 1 1 1>}
     font {
      bForceMonospaced=false
      bMaterialIsStencil=false
      fontMaterial=null
      fontObject=&HeadingNow
      letterSpacing=10
      monospacedWidth=1
      size=18
      skewAmount=0
      typefaceFontName=74 Regular
      outlineSettings {
       bApplyOutlineToDropShadows=false
       bMiteredCorners=false
       bSeparateFillAlpha=false
       outlineMaterial=null
       outlineSize=0
       outlineColor=<rgba 0 0 0 1>
      }
     }
     highlightColor {colorUseRule=UseColor_Specified, specifiedColor=<rgba 0 0 0 1>}
     highlightShape {
      bIsDynamicallyLoaded=false
      drawAs=Image
      imageType=NoImage
      mirroring=NoMirror
      resourceName=None
      resourceObject=null
      tiling=NoTile
      imageSize=<32 32>
      margin=<ltrb 0 0 0 0>
      outlineSettings {
       bUseBrushTransparency=false
       roundingType=HalfHeightRadius
       width=0
       color {colorUseRule=UseColor_Specified, specifiedColor=<rgba 0 0 0 0>}
       cornerRadii=<0 0 0 0>
      }
      tintColor {colorUseRule=UseColor_Specified, specifiedColor=<rgba 1 1 1 1>}
      uVRegion {bIsValid=false, max=<0 0>, min=<0 0>}
     }
     selectedBackgroundColor {colorUseRule=UseColor_Foreground, specifiedColor=<rgba 1 0 1 1>}
     shadowColorAndOpacity=<rgba 0 0 0 1>
     shadowOffset=<0 0>
     strikeBrush {
      bIsDynamicallyLoaded=false
      drawAs=Image
      imageType=NoImage
      mirroring=NoMirror
      resourceName=None
      resourceObject=null
      tiling=NoTile
      imageSize=<32 32>
      margin=<ltrb 0 0 0 0>
      outlineSettings {
       bUseBrushTransparency=false
       roundingType=HalfHeightRadius
       width=0
       color {colorUseRule=UseColor_Specified, specifiedColor=<rgba 0 0 0 0>}
       cornerRadii=<0 0 0 0>
      }
      tintColor {colorUseRule=UseColor_Specified, specifiedColor=<rgba 1 1 1 1>}
      uVRegion {bIsValid=false, max=<0 0>, min=<0 0>}
     }
     underlineBrush {
      bIsDynamicallyLoaded=false
      drawAs=Image
      imageType=NoImage
      mirroring=NoMirror
      resourceName=None
      resourceObject=null
      tiling=NoTile
      imageSize=<32 32>
      margin=<ltrb 0 0 0 0>
      outlineSettings {
       bUseBrushTransparency=false
       roundingType=HalfHeightRadius
       width=0
       color {colorUseRule=UseColor_Specified, specifiedColor=<rgba 0 0 0 0>}
       cornerRadii=<0 0 0 0>
      }
      tintColor {colorUseRule=UseColor_Specified, specifiedColor=<rgba 1 1 1 1>}
      uVRegion {bIsValid=false, max=<0 0>, min=<0 0>}
     }
    }
   }
   slot {class=VerticalBoxSlot, properties {size {sizeRule=Fill, value=1}}}
  }
 ]
}
