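# Item Shop PDP widget showing an offer for inspection, purchase, gifting, equipping, undo and preview.
# ChangeView switches between OfferInspect and ItemDetails, updating preview, purchase block and focus.
# Item list, purchase block and preview widget events drive item selection, async About and bundle screen pushes, and download tracking; button clicks fire analytics.
# Next and previous arrows swap the offer viewmodel between short slide-and-fade animations of the left info panel.
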WidgetBlueprint /ItemShop/MultiProduct/WBP_ItemShop_PDP
animations [
 OfferOutToPrevious {
  binding_count=1
  length_seconds=0.1
  object=&self:OfferOutToPrevious
  sequence {bindings [VerticalBox_LeftSideInfo {
     id=#1
     possessed_class=VerticalBox
     tracks [
      {
       type=MovieSceneFloatTrack
       display_name=Render Opacity
       sections [{type=MovieSceneFloatSection, end_frame=2, start_frame=0, channels [None {type=MovieSceneScriptingFloatChannel, keys [[0,1],[1.9997,0]]}]}]
      }
      {
       type=MovieScene2DTransformTrack
       sections [{
         type=MovieScene2DTransformSection
         end_frame=2
         start_frame=0
         channels [
          Translation.X {type=MovieSceneScriptingFloatChannel, keys [[0,0],[1.9997,300]]}
          Translation.Y {type=MovieSceneScriptingFloatChannel, const [0,1]}
          Angle {type=MovieSceneScriptingFloatChannel, const [0,1]}
          Scale.X {type=MovieSceneScriptingFloatChannel, const [1,1]}
          Scale.Y {type=MovieSceneScriptingFloatChannel, const [1,1]}
          Shear.X {type=MovieSceneScriptingFloatChannel, const [0,1]}
          Shear.Y {type=MovieSceneScriptingFloatChannel, const [0,1]}
         ]
        }]
      }
     ]
    }]}
 }
 OfferInPrevious {
  binding_count=1
  length_seconds=0.1
  object=&self:OfferInPrevious
  sequence {bindings [VerticalBox_LeftSideInfo {
     id=#2
     possessed_class=VerticalBox
     tracks [
      {
       type=MovieSceneFloatTrack
       display_name=Render Opacity
       sections [{type=MovieSceneFloatSection, end_frame=2, start_frame=0, channels [None {type=MovieSceneScriptingFloatChannel, keys [[0,0],[2,1]]}]}]
      }
      {
       type=MovieScene2DTransformTrack
       sections [{
         type=MovieScene2DTransformSection
         end_frame=2
         start_frame=0
         channels [
          Translation.X {type=MovieSceneScriptingFloatChannel, keys [[0,-100],[2,0]]}
          Translation.Y {type=MovieSceneScriptingFloatChannel, const [0,1]}
          Angle {type=MovieSceneScriptingFloatChannel, const [0,1]}
          Scale.X {type=MovieSceneScriptingFloatChannel, const [1,1]}
          Scale.Y {type=MovieSceneScriptingFloatChannel, const [1,1]}
          Shear.X {type=MovieSceneScriptingFloatChannel, const [0,1]}
          Shear.Y {type=MovieSceneScriptingFloatChannel, const [0,1]}
         ]
        }]
      }
     ]
    }]}
 }
 OfferOutToNext {
  binding_count=1
  length_seconds=0.1
  object=&self:OfferOutToNext
  sequence {bindings [VerticalBox_LeftSideInfo {
     id=#3
     possessed_class=VerticalBox
     tracks [
      {
       type=MovieSceneFloatTrack
       display_name=Render Opacity
       sections [{type=MovieSceneFloatSection, end_frame=2, start_frame=0, channels [None {type=MovieSceneScriptingFloatChannel, keys [[0,1],[1.9997,0]]}]}]
      }
      {
       type=MovieScene2DTransformTrack
       sections [{
         type=MovieScene2DTransformSection
         end_frame=2
         start_frame=0
         channels [
          Translation.X {type=MovieSceneScriptingFloatChannel, keys [[0,0],[1.9997,-300]]}
          Translation.Y {type=MovieSceneScriptingFloatChannel, const [0,1]}
          Angle {type=MovieSceneScriptingFloatChannel, const [0,1]}
          Scale.X {type=MovieSceneScriptingFloatChannel, const [1,1]}
          Scale.Y {type=MovieSceneScriptingFloatChannel, const [1,1]}
          Shear.X {type=MovieSceneScriptingFloatChannel, const [0,1]}
          Shear.Y {type=MovieSceneScriptingFloatChannel, const [0,1]}
         ]
        }]
      }
     ]
    }]}
 }
 OfferInNext {
  binding_count=1
  length_seconds=0.1
  object=&self:OfferInNext
  sequence {bindings [VerticalBox_LeftSideInfo {
     id=#4
     possessed_class=VerticalBox
     tracks [
      {
       type=MovieSceneFloatTrack
       display_name=Render Opacity
       sections [{type=MovieSceneFloatSection, end_frame=2, start_frame=0, channels [None {type=MovieSceneScriptingFloatChannel, keys [[0,0],[2,1]]}]}]
      }
      {
       type=MovieScene2DTransformTrack
       sections [{
         type=MovieScene2DTransformSection
         end_frame=2
         start_frame=0
         channels [
          Translation.X {type=MovieSceneScriptingFloatChannel, keys [[0,100],[2,0]]}
          Translation.Y {type=MovieSceneScriptingFloatChannel, const [0,1]}
          Angle {type=MovieSceneScriptingFloatChannel, const [0,1]}
          Scale.X {type=MovieSceneScriptingFloatChannel, const [1,1]}
          Scale.Y {type=MovieSceneScriptingFloatChannel, const [1,1]}
          Shear.X {type=MovieSceneScriptingFloatChannel, const [0,1]}
          Shear.Y {type=MovieSceneScriptingFloatChannel, const [0,1]}
         ]
        }]
      }
     ]
    }]}
 }
 Anim_Equipped {
  binding_count=1
  length_seconds=0.85
  object=&self:Anim_Equipped
  sequence {bindings [EquippedText {
     id=#5
     possessed_class=CommonTextBlock
     tracks [
      {
       type=MovieSceneFloatTrack
       display_name=Render Opacity
       sections [{type=MovieSceneFloatSection, end_frame=15, start_frame=0, channels [None {type=MovieSceneScriptingFloatChannel, keys [[0,0],[1,1],[15,0]]}]}]
      }
      {
       type=MovieSceneColorTrack
       display_name=Color and Opacity
       sections [{
         type=MovieSceneColorSection
         end_frame=17
         start_frame=0
         channels [
          Color.R {type=MovieSceneScriptingFloatChannel, keys [[0,1],[17,0]]}
          Color.G {type=MovieSceneScriptingFloatChannel, keys [[0,1],[17,0.6654]]}
          Color.B {type=MovieSceneScriptingFloatChannel, const [1,2]}
          Color.A {type=MovieSceneScriptingFloatChannel, const [1,2]}
         ]
        }]
      }
      {
       type=MovieSceneByteTrack
       display_name=Visibility
       sections [{type=MovieSceneByteSection, end_frame=15, start_frame=0, channels [None {type=MovieSceneScriptingByteChannel, keys [[0,4],[15,2]]}]}]
      }
      {
       type=MovieScene2DTransformTrack
       sections [{
         type=MovieScene2DTransformSection
         end_frame=17
         start_frame=0
         channels [Translation.Y {type=MovieSceneScriptingFloatChannel, keys [[0,0],[15,-150],[17,-160]]}]
        }]
      }
     ]
    }]}
 }
 Anim_HideISBs {
  binding_count=1
  length_seconds=0.35
  object=&self:Anim_HideISBs
  sequence {
   bindings [Component_IsbManager {
     id=#6
     display_name=WBP_ItemShopPDP_IsbManager
     possessed_class=WBP_ItemShopPDP_IsbManager_C
     tracks [
      {
       type=MovieSceneFloatTrack
       display_name=Render Opacity
       sections [{type=MovieSceneFloatSection, end_frame=21, start_frame=0, channels [None {type=MovieSceneScriptingFloatChannel, keys [[0,1],[21,0]]}]}]
      }
      {
       type=MovieScene2DTransformTrack
       sections [{
         type=MovieScene2DTransformSection
         end_frame=21
         start_frame=0
         channels [
          Translation.X {type=MovieSceneScriptingFloatChannel, keys [[0,0],[21,32]]}
          Translation.Y {type=MovieSceneScriptingFloatChannel, keys [[0,0],[21,64]]}
          Scale.X {type=MovieSceneScriptingFloatChannel, keys [[0,1],[21,1.2]]}
          Scale.Y {type=MovieSceneScriptingFloatChannel, keys [[0,1],[21,1.2]]}
         ]
        }]
      }
     ]
    }]
   display_rate {numerator=60}
  }
 }
 Anim_ShowISBs {
  binding_count=1
  length_seconds=0.35
  object=&self:Anim_ShowISBs
  sequence {
   bindings [Component_IsbManager {
     id=#7
     display_name=WBP_ItemShopPDP_IsbManager
     possessed_class=WBP_ItemShopPDP_IsbManager_C
     tracks [
      {
       type=MovieSceneFloatTrack
       display_name=Render Opacity
       sections [{type=MovieSceneFloatSection, end_frame=21, start_frame=0, channels [None {type=MovieSceneScriptingFloatChannel, keys [[0,0],[21,1]]}]}]
      }
      {
       type=MovieScene2DTransformTrack
       sections [{
         type=MovieScene2DTransformSection
         end_frame=21
         start_frame=0
         channels [
          Translation.X {type=MovieSceneScriptingFloatChannel, const [0,2]}
          Translation.Y {type=MovieSceneScriptingFloatChannel, const [0,2]}
          Scale.X {type=MovieSceneScriptingFloatChannel, const [1,2]}
          Scale.Y {type=MovieSceneScriptingFloatChannel, const [1,2]}
         ]
        }]
      }
     ]
    }]
   display_rate {numerator=60}
  }
 }
]
graphs [
 AttemptPurchase {nodes [
   AttemptPurchase() {is=title K2Node_FunctionEntry}
   > PurchaseStep_ConfirmationOrMMBan()
  ]}
 BP_GetDesiredFocusTarget {nodes [
   Get Desired Focus Target()
   > n1 = GetDesiredFocusTarget!(
    execute:Exec
    self:Object Reference=Get Component_PurchaseBlock() -> (Component_PurchaseBlock:WBP Item Shop PDP Purchase Block Object Reference)
   ) -> (then:Exec, FocusTarget:Widget Object Reference) {is=title K2Node_Message}
   > Set PurchaseBlockTarget(PurchaseBlockTarget:Widget Object Reference=@n1) -> (Output_Get:Widget Object Reference)
   > n2 = Switch on EFortItemShopPdpView!(execute, Selection:EFortItemShopPdpView Enum=Get CurrentView() -> (CurrentView:EFortItemShopPdpView Enum=OfferInspect)) -> (OfferInspect:Exec, ItemDetails:Exec) {is=title K2Node_SwitchEnum}
   n3 = Get WBP_UIKit_VariantSelector() -> (WBP_UIKit_VariantSelector:WBP UIKit Variant Selector Object Reference)
   n4 = Branch(execute=@n2.ItemDetails, Condition=IsVisible(self=@n3))
   > Return Node!(
    execute
    ReturnValue:Widget Object Reference=GetDesiredFocusTarget!(self:WBP UIKit Variant Selector Object Reference=@n3) -> (ReturnValue:Widget Object Reference) {is=title K2Node_CallFunction}
   )
   n5 = Branch(execute=@n4.else, Condition=IsVisible(self=Get Component_PurchaseBlock()))
   n6 = Get PurchaseBlockTarget() -> (PurchaseBlockTarget:Widget Object Reference)
   > n7 = Is Valid(InputObject=@n6)
   Return Node(execute=[@n7.Is Not Valid, @n5.else], ReturnValue=Self-Reference())
   > Return Node(ReturnValue=@n6)
   n8 = Get CachedPreviewedItemIndex() -> (CachedPreviewedItemIndex:int=0)
   n9 = Get Component_ItemList() -> (Component_ItemList:WBP Item Shop PDP Item List Object Reference)
   n10 = Get PurchaseBlockTarget()
   n11 = IsValid(Object=@n10)
   n12 = GetWidgetAtIndex!(self:WBP UIKit Item List Object Reference=Get Component_ItemList(), InIndex:int=0) -> (OutWIdget:Widget Object Reference)
   Return Node(
    execute=@n2.OfferInspect
    ReturnValue=Select(
     Option 0:Widget Object Reference=Select(Option 0=@n12, Option 1=@n10, Index=@n11)
     Option 1:Widget Object Reference=GetWidgetAtIndex(self=@n9, InIndex=@n8)
     Index="integer >= integer"(A=@n8)
    ) -> (ReturnValue:Widget Object Reference)
   )
  ]}
 BP_GetDesiredInputConfig {nodes [
   Get Desired Input Config() {is=title K2Node_FunctionEntry}
   > Return Node(ReturnValue:UIInput Config Structure=Make UIInput Config!(
     bIgnoreMoveInput:bool=False
     bIgnoreLookInput:bool=False
     InputMode:ECommonInputMode Enum=All
     MouseCaptureMode:EMouseCaptureMode Enum=CaptureDuringMouseDown
     MouseLockMode:EMouseLockMode Enum=DoNotLock
     bHideCursorDuringViewportCapture:bool=True
    ) -> (UIInputConfig:UIInput Config Structure) {is=title K2Node_MakeStruct})
  ]}
 BuildAnalyticsInfo {nodes [
   BuildAnalyticsInfo!() -> (then, OptionalAdditionalInfoMap:Map of Strings to Strings) {is=title K2Node_FunctionEntry}
   > Set AnalyticsInfo(AnalyticsInfo:Flow Analytics Additional Info Structure=Make Flow Analytics Additional Info!() -> (FlowAnalyticsAdditionalInfo:Flow Analytics Additional Info Structure) {is=title K2Node_MakeStruct}) -> (Output_Get:Flow Analytics Additional Info Structure)
   > n1 = "For Each Loop (Map)"!(
    execute:Exec
    MapPin:Map of Strings to Strings=Get OptionalAdditionalInfoMap() -> (OptionalAdditionalInfoMap:Map of Strings to Strings)
    BreakPin:Exec
   ) -> (then:Exec, KeyPin:String, ValuePin:String, CompletedPin:Exec)
   > Set AnalyticsInfo(AnalyticsInfo="Add String (Non-Localized)"!(
     Info:"Flow Analytics Additional Info Structure (by ref)"=Get AnalyticsInfo() -> (AnalyticsInfo:Flow Analytics Additional Info Structure)
     Key:String=@n1
     Value:String=@n1.ValuePin
    ) -> (ReturnValue:Flow Analytics Additional Info Structure))
   n2 = Get FortItemShopOfferVM() -> (FortItemShopOfferVM:Fort Item Shop Offer VM Object Reference)
   n3 = Branch(
    execute=@n1.CompletedPin
    Condition="Equal (Enum)"(
     A:EFortPriceCurrency Enum=GetCurrencyType!(self:Fort Base Offer VM Object Reference=@n2) -> (ReturnValue:EFortPriceCurrency Enum=Mtx)
     B:EFortPriceCurrency Enum=RealMoney
    )
   )
   n4 = GetDisplayName(Object=Self-Reference())
   n5 = Get FortMcpContext!() -> (ReturnValue:Fort Mcp Context Object Reference) {is=title K2Node_GetBlueprintContextEx}
   > Return Node!(
    execute
    AdditionalInfoStruct:Flow Analytics Additional Info Structure="Add String (Non-Localized)"(
     Info=Get AnalyticsInfo()
     Key=PurchaseUXTraceId
     Value=GetPurchaseUXTraceId!(self:Fort Mcp Context Object Reference=@n5, DebugSourceContext:String=@n4) -> (ReturnValue:String)
    )
   )
   Return Node(execute=@n3.else, AdditionalInfoStruct=Get AnalyticsInfo())
  ]}
 CanInspectOfferVM {nodes [
   n1 = CanInspectOfferVM!() -> (then, InOfferVM:Fort Item Shop Offer VM Object Reference) {is=title K2Node_FunctionEntry}
   > n2 = Is Valid(InputObject=@n1)
   > n3 = Get!(execute:Exec, self:Fort Item Shop Offer VM Object Reference=Get InOfferVM() -> (InOfferVM:Fort Item Shop Offer VM Object Reference)) -> (then:Exec, else:Exec, OfferInfo:Athena Item Shop Offer Info Object Reference)
   n4 = Get GrantedItems!(self:Fort Item Shop Offer VM Object Reference=@n1) -> (GrantedItems:Array of Fort Item VM Object References)
   > n5 = Branch(Condition=AND Boolean(
     A=IsPurchasingAvailable!(self:Fort Store Front Offer Info Object Reference=@n3) -> (ReturnValue:bool)
     B=Is Not Empty!(TargetArray:Array of Fort Item VM Object References=@n4) -> (ReturnValue:bool)
    ))
   n6 = Get OfferType!(self:Fort Item Shop Offer VM Object Reference=@n1) -> (OfferType:EFortItemShopOfferType Enum=None)
   n7 = "Equal (Enum)"(A:EFortItemShopOfferType Enum=@n6, B:EFortItemShopOfferType Enum=Special)
   n8 = "Equal (Enum)"(A=@n6, B=None)
   n9 = "Equal (Enum)"(A=@n6, B=BattlePass)
   n10 = "Equal (Enum)"(A=@n6, B=ProgressiveCosmetic)
   n11 = "Equal (Enum)"(A=@n6, B=Crew)
   > Return Node!(execute, ReturnValue:bool=NOT Boolean(A=OR Boolean!(A=@n8, B=@n9, C:bool=@n10, D:bool=@n11, E:bool=@n7) -> (ReturnValue)))
   Return Node(execute=[@n5.else, @n3.else, @n2.Is Not Valid], ReturnValue=false)
  ]}
 CanShowPrimaryPriceSwitcher {nodes [
   n1 = CanShowPrimaryPriceSwitcher!() -> (then, IsOwned:bool=~, IsGiftable:bool=~) {is=title K2Node_FunctionEntry}
   > Return Node!(execute, CanShow:bool=OR Boolean!(A=NOT Boolean(A=@n1), B=AND Boolean(A=@n1, B=@n1.IsGiftable)) -> (ReturnValue))
  ]}
 ChangeFocusArea {nodes [
   n1 = ChangeFocusArea!() -> (then, IsFocusingItem:bool=~) {is=title K2Node_FunctionEntry}
   > OnFocusAreaChanged!(execute:Exec, self:Object Reference=Self-Reference(), IsFocusingItems:bool=@n1) -> (then:Exec) {is=title K2Node_Message}
   > n2 = Set with Broadcast IsFocusingItems!(execute, IsFocusingItems:bool=Get IsFocusingItem() -> (IsFocusingItem:bool)) -> (then, Output_Get:bool)
   > n3 = Branch(Condition=@n2)
   TryRequestOverridePreview!(execute=@n3.else, self) -> (then, Success:bool)
  ]}
 ChangeView {nodes [
   n1 = ChangeView!() -> (then, NewView:EFortItemShopPdpView Enum=OfferInspect, DeactivateOnBack:bool=~) {is=title K2Node_FunctionEntry}
   > Set CurrentView(CurrentView:EFortItemShopPdpView Enum=@n1) -> (Output_Get:EFortItemShopPdpView Enum=OfferInspect)
   > Set ShouldDeactivateOnBack(ShouldDeactivateOnBack:bool=Get DeactivateOnBack() -> (DeactivateOnBack:bool)) -> (Output_Get:bool)
   > Set Visibility(
    Visibility=Select!(
     OfferInspect:ESlateVisibility Enum=Collapsed
     ItemDetails:ESlateVisibility Enum=SelfHitTestInvisible
     Index:EFortItemShopPdpView Enum=Get NewView() -> (NewView:EFortItemShopPdpView Enum=OfferInspect)
    ) -> (ReturnValue:ESlateVisibility Enum)
    self=Get Border_CameraButton() -> (Border_CameraButton:Border Object Reference)
   )
   > n2 = Get Previewed Item Preview Data!(execute:Exec) -> (Is Valid, Is Not Valid, PreviewedItem:Item Shop Preview Data Object Reference) {is=title K2Node_MacroInstance}
   > n3 = Branch(Condition=HasLoadoutOverride!(self:Item Shop Preview Data Object Reference=@n2) -> (ReturnValue:bool))
   n4 = Set CachedPreviewedItem(
    execute=@n3.else
    CachedPreviewedItem:Fort Item VM Object Reference=Get ItemVM!(self:Item Shop Preview Data Object Reference=@n2) -> (ItemVM:Fort Item VM Object Reference)
   ) -> (Output_Get:Fort Item VM Object Reference)
   n5 = Get CachedPreviewedItem() -> (CachedPreviewedItem:Fort Item VM Object Reference)
   n6 = Get GrantedItems(self=Get FortItemShopOfferVM())
   Set CachedPreviewedItemIndex(
    execute=[@n4.then, @n2.Is Not Valid]
    CachedPreviewedItemIndex:int="Max (Integer)"(
     A=FindItemByDefinition!(ItemArray:Array of Fort Item VM Object References=@n6, Item:Fort Item VM Object Reference=@n5) -> (FoundIndex:int=0)
     B=0
    )
   ) -> (Output_Get:int=0)
   > n7 = Branch(Condition=Get HasVariants() -> (HasVariants:bool))
   > n8 = Set PreviewedItem!(
    execute
    PreviewedItem:Item Shop Preview Data Object Reference
    self:Fort Item Shop VM Object Reference=Get FortItemShopVM() -> (FortItemShopVM:Fort Item Shop VM Object Reference)
   ) -> (then, Output_Get:Item Shop Preview Data Object Reference)
   SetActiveWidget(
    execute=[@n7.else, @n8.then]
    self=Get Switcher_ScreenView() -> (Switcher_ScreenView:Common Visibility Switcher Object Reference)
    Widget=Select(
     OfferInspect:Widget Object Reference=Get View_OfferInspect() -> (View_OfferInspect:Overlay Object Reference)
     ItemDetails:Widget Object Reference=Get View_ItemDetails() -> (View_ItemDetails:Vertical Box Object Reference)
     Index=Get NewView() -> (NewView=OfferInspect)
    ) -> (ReturnValue:Widget Object Reference)
   )
   > Set Visibility(
    Visibility=Select(
     OfferInspect:ESlateVisibility Enum=SelfHitTestInvisible
     ItemDetails:ESlateVisibility Enum=Collapsed
     Index=Get NewView() -> (NewView=OfferInspect)
    ) -> (ReturnValue:ESlateVisibility Enum)
    self=Get HorizontalBox_NavigationArrows() -> (HorizontalBox_NavigationArrows:Horizontal Box Object Reference)
   )
   > UpdateCustomSticker!(execute, self, SelectedItem:Fort Item VM Object Reference=Get CachedPreviewedItem()) -> (then)
   > UpdateAboutItemButton()
   > PreviewItemAtIndex!(execute, self, Index:int=Get CachedPreviewedItemIndex() -> (CachedPreviewedItemIndex=0)) -> (then)
   > UpdatePurchaseBlock()
   > n9 = Branch(Condition="Equal (Enum)"(A:EFortItemShopPdpView Enum=Get NewView() -> (NewView=OfferInspect), B:EFortItemShopPdpView Enum=ItemDetails))
   > n10 = SetItemDetailsView()
   OnViewChanged!(
    execute:Exec=[@n10.then, @n9.else]
    self:Object Reference=Self-Reference()
    NewView:EFortItemShopPdpView Enum=Get NewView() -> (NewView=OfferInspect)
   ) -> (then:Exec) {is=title K2Node_Message}
   > RequestRefreshFocus(self:Common Activatable Widget Object Reference)
  ]}
 ClearPreviewActionButtons {nodes [
   ClearPreviewActionButtons() {is=title K2Node_FunctionEntry}
   > n1 = To Array!(
    execute
    A:Set of Common Button Base Object References=Get PreviewActionButtons!(self:Fort Item Preview Screen Object Reference=Get FortItemPreviewWidget() -> (FortItemPreviewWidget:Fort Item Preview Widget Object Reference)) -> (PreviewActionButtons:Set of Common Button Base Object References)
   ) -> (then, Result:Array of Common Button Base Object References)
   > n2 = For Each Loop(Array:Array of Common Button Base Object References=@n1) -> (Array Element:Common Button Base Object Reference)
   > SetTriggeringInputAction!(
    execute
    self:Common Button Base Object Reference=@n2
    InputActionRow_DataTable:Data Table Object Reference
    InputActionRow_RowName:Name=None
   ) -> (then)
  ]}
 ClosePDP {nodes [
   ClosePDP() {is=title K2Node_FunctionEntry}
   > DeactivateWidget()
   > Set ShouldDeactivateOnBack(ShouldDeactivateOnBack=false)
  ]}
 CustomNavigationToFirstItem {nodes [
   CustomNavigationToFirstItem!() -> (then, Navigation:EUINavigation Enum) {is=title K2Node_FunctionEntry}
   n1 = Get WBP_UIKit_VariantSelector()
   n2 = Get Component_ItemList()
   > Return Node!(
    execute
    ReturnValue:Widget Object Reference=Select(
     OfferInspect:Widget Object Reference=GetWidgetAtIndex(self=@n2, InIndex=0)
     ItemDetails:Widget Object Reference=GetDesiredFocusTarget(self=@n1)
     Index=Get CurrentView() -> (CurrentView=OfferInspect)
    ) -> (ReturnValue:Widget Object Reference)
   )
  ]}
 DoesOwnAnyStacks {nodes [
   n1 = DoesOwnAnyStacks!() -> (then, ItemVm:Fort Item VM Object Reference) {is=title K2Node_FunctionEntry}
   > Return Node(ReturnValue:bool="integer > integer"(A=GetOwnedStacksNum!(self:Fort Item VM Object Reference=@n1) -> (ReturnValue:int=0)))
  ]}
 EvaluateRmtStorefront {nodes [
   EvaluateRmtStorefront() {is=title K2Node_FunctionEntry}
   n1 = Get Offer to Purchase!() -> (OfferToPurchase:Fort Item Shop Offer VM Object Reference) {is=title K2Node_MacroInstance}
   > n2 = Branch(Condition="Equal (Enum)"(A:EFortPriceCurrency Enum=GetCurrencyType(self=@n1) -> (ReturnValue=Mtx), B:EFortPriceCurrency Enum=RealMoney))
   n3 = Get Offer to Purchase()
   > n4 = Branch(Condition=AND Boolean(
     A=HasMultiplePrices!(self:Fort Item Shop Offer VM Object Reference=@n3) -> (ReturnValue:bool)
     B=IsPurchaseChoiceRequestedEventBound!(self) -> (ReturnValue:bool)
    ))
   > BroadcastPurchaseChoiceRequestedEvent!(
    execute
    self:Purchase Choice Requester Interface Interface
    Requester:"Purchase Choice Requester Interface Interface (by ref)"=Self-Reference()
   ) -> (then)
   > Return Node!(execute, HasMultipleStorefronts:bool=true)
   Return Node(execute=[@n4.else, @n2.else], HasMultipleStorefronts=false)
  ]}
 EventGraph {nodes [
   n1 = "On Clicked (Button_Back)"!() -> (OutputDelegate, then, Button:Common Button Base Object Reference) {is=title K2Node_ComponentBoundEvent}
   > Fire Event Element Interacted(
    Element=@n1.Button
    AdditionalInfo=BuildAnalyticsInfo!(OptionalAdditionalInfoMap:Map of Strings to Strings=Make Map!(Key 0:String, Value 0:String) -> (Map:Map of Strings to Strings) {is=title K2Node_MakeMap}) -> (AdditionalInfoStruct:Flow Analytics Additional Info Structure) {is=title K2Node_CallFunction}
   )
   > n2 = Branch(Condition="Equal (Enum)"(A:EFortItemShopPdpView Enum=Get CurrentView() -> (CurrentView=OfferInspect), B:EFortItemShopPdpView Enum=ItemDetails))
   > n3 = Branch(Condition=Get ShouldDeactivateOnBack() -> (ShouldDeactivateOnBack:bool))
   > n4 = DeactivateWidget()
   ChangeView!(execute=[@n4.then, @n3.else], self, NewView:EFortItemShopPdpView Enum=OfferInspect, DeactivateOnBack:bool=false) -> (then) {is=title K2Node_CallFunction}
   > UpdateNavigationArrows()
   ClosePDP(execute=@n2.else) {is=title K2Node_CallFunction}
   n5 = "On Go to Item Clicked (Component_PurchaseBlock)"!() -> (OutputDelegate, then, FortItemVM:Fort Item VM Object Reference) {is=title K2Node_ComponentBoundEvent}
   > SelectItem!(execute, self, ItemVm:Fort Item VM Object Reference=@n5.FortItemVM) -> (then)
   > ChangeView(NewView=ItemDetails, DeactivateOnBack=false)
   "On Clicked (Button_Equip)"!() -> (OutputDelegate, then, Button:Common Button Base Object Reference) {is=title K2Node_ComponentBoundEvent}
   > SetIsInteractionEnabled!(
    execute
    self:Common Button Base Object Reference=Get Button_Equip() -> (Button_Equip:WBP UIKit Button Quiet Object Reference)
    bInIsInteractionEnabled:bool=false
   ) -> (then)
   > n6 = SetVisibility(self=Get Throbber_Equip() -> (Throbber_Equip:WBP UIKit Throbber Object Reference), InVisibility=HitTestInvisible)
   n7 = OnEquipItemCompleted!() -> (OutputDelegate, then, bItemEquipped:bool=~) {is=title K2Node_CustomEvent}
   n8 = Get CurrentView() -> (CurrentView=OfferInspect)
   n9 = Get FortItemShopVM()
   EquipItem!(
    execute=@n6.then
    self:Fort MPItem Shop Offer Inspection Screen Object Reference
    SelectedItem:Fort Item VM Object Reference=Get ItemVM(self=Get PreviewedItem!(self:Fort Item Shop VM Object Reference=@n9) -> (PreviewedItem:Item Shop Preview Data Object Reference))
    OptionalVariantSelector:Fort Variant Selector VM Object Reference=Select!(
     Option 0:Fort Variant Selector VM Object Reference
     Option 1:Fort Variant Selector VM Object Reference=Get FortVariantSelectorVM() -> (FortVariantSelectorVM:Fort Variant Selector VM Object Reference)
     Index:bool="Equal (Enum)"(A=@n8, B=ItemDetails)
    ) -> (ReturnValue:Fort Variant Selector VM Object Reference)
    OnCompleted:"Delegate (by ref)"=@n7
   ) -> (then)
   n10 = "On Next Item Requested (FortItemPreviewWidget)"!() -> (OutputDelegate, then, Target Cosmetic Query:Gameplay Tag Query Structure, Item Definition:Fort Account Item Definition Object Reference) {is=title K2Node_ComponentBoundEvent}
   > n11 = FindNextCosmeticItem!(
    execute
    self:Fort Item Shop Offer VM Object Reference=Get FortItemShopOfferVM()
    TargetCosmeticQuery:"Gameplay Tag Query Structure (by ref)"=@n10.Target Cosmetic Query
    CurrentItemDefinition:Fort Account Item Definition Object Reference=@n10.Item Definition
   ) -> (then, ReturnValue:Fort Item VM Object Reference)
   > Branch(Condition=IsValid(Object=@n11))
   > n12 = CreatePreviewDataFromItemVM!(execute, Context:Widget Object Reference, ItemVM:Fort Item VM Object Reference=@n11) -> (then, ReturnValue:Item Shop Preview Data Object Reference)
   > Set PreviewedItem(PreviewedItem=@n12, self=Get FortItemShopVM())
   "On Clicked (Button_NextOffer)"!() -> (OutputDelegate, then, Button:Common Button Base Object Reference) {is=title K2Node_ComponentBoundEvent}
   > n13 = GetNextOffer!(execute, self) -> (then, OfferVM:Fort Item Shop Offer VM Object Reference, Found:bool)
   > Branch(Condition=@n13.Found)
   n14 = GetDisplayAssetName!(self:Fort Item Shop Offer VM Object Reference=@n13) -> (ReturnValue:String)
   > Fire Event Element Interacted(
    Element=Get Button_NextOffer() -> (Button_NextOffer:WBP UIKit Button Generic Object Reference)
    AdditionalInfo=BuildAnalyticsInfo(OptionalAdditionalInfoMap=Make Map(Key 0=Display Asset Name, Value 0=@n14))
   )
   > n15 = "Play Animation with Finished event (legacy)"!(
    execute:Exec
    Widget:User Widget Object Reference=Self-Reference()
    InAnimation:Widget Animation Object Reference=Get OfferOutToNext() -> (OfferOutToNext:Widget Animation Object Reference)
    StartAtTime:float=0
    NumLoopsToPlay:int=1
    PlayMode:EUMGSequencePlayMode Enum
    PlaybackSpeed:float=1
   ) -> (then:Exec, Result:UMGSequence Player Object Reference, Finished:Exec) {is=title K2Node_PlayAnimation}
   SetFortItemShopOfferVM!(execute=@n15.Finished, self, Viewmodel:Fort Item Shop Offer VM Object Reference=@n13) -> (then)
   > n16 = PlayAnimationForward(InAnimation=Get OfferInNext() -> (OfferInNext:Widget Animation Object Reference))
   SetOffer(execute=[@n16.then, @n37.then])
   n17 = "On Clicked (Button_AboutItem)"!() -> (OutputDelegate, then, Button:Common Button Base Object Reference) {is=title K2Node_ComponentBoundEvent}
   n18 = Get!(execute=[@n101.then, @n17.then]) -> (then, else, FortItemShopVM:Fort Item Shop VM Object Reference)
   > n19 = Get!(execute, self:Fort Item Shop VM Object Reference=@n18) -> (then, else, PreviewedItem:Item Shop Preview Data Object Reference)
   > n20 = Get!(execute, self:Item Shop Preview Data Object Reference=@n19) -> (then, else, ItemVM:Fort Item VM Object Reference)
   n21 = TryGetSoftCustomActivatableWidgetClass!(
    self:Fort Locker Category Config Object Reference=GetItemCategory!(self:Fort Item VM Object Reference=@n20) -> (ReturnValue:Fort Locker Category Config Object Reference)
    ViewRequest:Gameplay Tag Structure=(TagName="UI.ItemShop.AboutItemDialogClass")
   ) -> (WidgetClass:Common Activatable Widget Soft Class Reference, ReturnValue:bool)
   > Branch(Condition=@n21.ReturnValue)
   > SetAllInteractionsEnabled!(execute, self, ShouldEnable:bool=false) -> (then)
   > n22 = CreateWidgetAsync!(
    execute:Exec
    UserWidgetSoftClass:User Widget Soft Class Reference=@n21
    OwningPlayer:Player Controller Object Reference=GetOwningPlayer()
    bSuspendInputUntilComplete:bool=true
   ) -> (then:Exec, OnComplete:Exec, OnFailure:Exec, UserWidget:User Widget Object Reference) {is=title K2Node_AsyncAction}
   SetViewModelByClass!(execute=@n22.OnComplete, Widget:User Widget Object Reference=@n22, ViewModel:Notify Field Value Changed Interface=@n20) -> (then)
   > TryPushWidgetAsModalPanel!(execute, PushWidget:User Widget Object Reference=@n22) -> (then)
   > SetHasSeenAboutItemModal!(
    execute
    self:Fort Game User Settings Object Reference=GetFortGameUserSettings!() -> (ReturnValue:Fort Game User Settings Object Reference)
    ItemType:EFortItemType Enum=GetItemType!(ItemDefinition:Item Definition Base Object Reference=Get ItemDefinition!(self:Fort Item VM Object Reference=@n20) -> (ItemDefinition:Fort Item Definition Object Reference)) -> (ReturnValue:EFortItemType Enum=WorldItem)
    bHasSeen:bool=true
   ) -> (then)
   > SetAllInteractionsEnabled(ShouldEnable=true)
   n23 = "On Item 2d Download Started (FortItemPreviewWidget)"!() -> (OutputDelegate, then, ImageUrl:String) {is=title K2Node_ComponentBoundEvent}
   > Get Previewed Item VM!(execute:Exec) -> (Is Valid, Is Not Valid, ItemVM:Fort Item VM Object Reference) {is=title K2Node_MacroInstance}
   > RegisterImageDownloadStarted!(execute, self:Fort Item Shop Offer VM Object Reference=Get FortItemShopOfferVM(), ImageUrl:String=@n23.ImageUrl) -> (then)
   n24 = "On Item 2d Download Completed (FortItemPreviewWidget)"!() -> (OutputDelegate, then, ImageUrl:String, Result:bool) {is=title K2Node_ComponentBoundEvent}
   > Get Previewed Item VM()
   > RegisterImageDownloadCompleted!(execute, self:Fort Item Shop Offer VM Object Reference=Get FortItemShopOfferVM(), ImageUrl:String=@n24.ImageUrl, bSuccess:bool=@n24.Result) -> (then)
   n25 = "On Item 3d Download Completed (FortItemPreviewWidget)"!() -> (OutputDelegate, then, Item:Fort Item VM Object Reference, Result:bool) {is=title K2Node_ComponentBoundEvent}
   > n26 = Get Previewed Item Definition!(execute:Exec) -> (Is Valid, Is Not Valid, ItemDefinition:Fort Item Definition Object Reference) {is=title K2Node_MacroInstance}
   > n27 = RegisterItemDownloadCompleted!(
    execute
    self:Fort Item Shop Offer VM Object Reference=Get FortItemShopOfferVM()
    Item:Fort Item Definition Object Reference=@n26
    bSuccess:bool=@n25.Result
   ) -> (then)
   Branch(execute=[@n27.then, @n26.Is Not Valid], Condition=AND Boolean(A=Get HasVariantsToCycle() -> (HasVariantsToCycle:bool), B=@n25.Result))
   > Set Visibility(Visibility=SelfHitTestInvisible, self=Get InfoBox_VariantInfo() -> (InfoBox_VariantInfo:WBP UIKit Info Box Object Reference))
   "On Download Cancelled (FortItemPreviewWidget)"() {is=title K2Node_ComponentBoundEvent}
   > n28 = Get Previewed Item Definition()
   > RegisterItemDownloadCompleted(self=Get FortItemShopOfferVM(), Item=@n28, bSuccess=false)
   "On Item 3d Download Started (FortItemPreviewWidget)"!() -> (OutputDelegate, then, Item:Fort Item VM Object Reference) {is=title K2Node_ComponentBoundEvent}
   > n29 = Get Previewed Item Definition()
   > n30 = RegisterItemDownloadStarted!(execute, self:Fort Item Shop Offer VM Object Reference=Get FortItemShopOfferVM(), Item:Fort Item Definition Object Reference=@n29) -> (then)
   Set Visibility(execute=[@n30.then, @n29.Is Not Valid], Visibility=Collapsed, self=Get InfoBox_VariantInfo())
   n31 = "On Preview Actions Updated Event (FortItemPreviewWidget)"!() -> (OutputDelegate, then, PreviewActions:Array of Data Table Row Handle Structures) {is=title K2Node_ComponentBoundEvent}
   > RefreshContextMenuPreviewActions!(execute, self, PreviewActions:Array of Data Table Row Handle Structures=@n31.PreviewActions) -> (then)
   n32 = "On Preview Action Button Updated Event (FortItemPreviewWidget)"!() -> (OutputDelegate, then, Row Display Name:Text, Button:Common Button Base Object Reference) {is=title K2Node_ComponentBoundEvent}
   > n33 = Cast To FortCTAButton!(execute, Object=@n32.Button) -> (then, CastFailed, AsFort CTAButton:Fort CTAButton Object Reference) {is=title K2Node_DynamicCast}
   > SetText!(execute, self:Fort CTAButton Object Reference=@n33, ButtonText:"Text (by ref)"=@n32.Row Display Name) -> (then)
   "On Clicked (Button_PreviousOffer)"!() -> (OutputDelegate, then, Button:Common Button Base Object Reference) {is=title K2Node_ComponentBoundEvent}
   > n34 = GetPreviousOffer!(execute, self) -> (then, OfferVM:Fort Item Shop Offer VM Object Reference, Found:bool)
   > Branch(Condition=@n34.Found)
   n35 = GetDisplayAssetName(self=@n34)
   > Fire Event Element Interacted(
    Element=Get Button_PreviousOffer() -> (Button_PreviousOffer:WBP UIKit Button Generic Object Reference)
    AdditionalInfo=BuildAnalyticsInfo(OptionalAdditionalInfoMap=Make Map(Key 0=Display Asset Name, Value 0=@n35))
   )
   > n36 = "Play Animation with Finished event (legacy)"(
    Widget=Self-Reference()
    InAnimation=Get OfferOutToPrevious() -> (OfferOutToPrevious:Widget Animation Object Reference)
    StartAtTime=0
    NumLoopsToPlay=1
    PlaybackSpeed=1
   )
   SetFortItemShopOfferVM(execute=@n36.Finished, Viewmodel=@n34)
   > n37 = PlayAnimationForward(InAnimation=Get OfferInPrevious() -> (OfferInPrevious:Widget Animation Object Reference))
   "On Purchase Clicked (Component_PurchaseBlock)"() {is=title K2Node_ComponentBoundEvent}
   > OnPurchaseInitiated()
   "On Get Vbucks Clicked (Component_PurchaseBlock)"() {is=title K2Node_ComponentBoundEvent}
   > RequestOpenMtxStore!(execute, self:Fort MPItem Shop Offer Inspection Screen Object Reference, Offer:Fort Base Offer VM Object Reference=Get Offer to Purchase()) -> (then)
   "On Gift Clicked (Component_PurchaseBlock)"() {is=title K2Node_ComponentBoundEvent}
   > ShowGiftingScreen!(
    execute
    self:Fort MPItem Shop Offer Inspection Screen Object Reference
    OfferVM:Fort Item Shop Offer VM Object Reference=Get Offer to Purchase()
   ) -> (then)
   "On Undo Purchase Clicked (Component_PurchaseBlock)"() {is=title K2Node_ComponentBoundEvent}
   > n38 = SetLoadingOverlayVisibility!(execute, self, IsVisible:bool=true) -> (then)
   n39 = OnUndoPurchase!() -> (OutputDelegate, then, bSuccess:bool=~, RefundedItemTemplateIds:Array of Strings, PurchaseId:String, MtxReturned:int) {is=title K2Node_CustomEvent}
   UndoPurchase!(
    execute=@n38.then
    self:Fort Item Shop Offer VM Object Reference=Get Offer to Purchase()
    OnUndoPurchaseCompleteCallback:"Delegate (by ref)"=@n39
   ) -> (then)
   "On Purchase History Clicked (Component_PurchaseBlock)"() {is=title K2Node_ComponentBoundEvent}
   > RequestOpenPurchaseHistoryScreen!(
    execute
    self:Fort MPItem Shop Offer Inspection Screen Object Reference
    OfferToFilter:Fort Item Shop Offer VM Object Reference=Get Offer to Purchase()
   ) -> (then)
   "On View Bundle Clicked (Component_PurchaseBlock)"() {is=title K2Node_ComponentBoundEvent}
   > SetAllInteractionsEnabled(ShouldEnable=false)
   > n40 = PushPrimaryContentClassAsync!(execute:Exec, WidgetClass:Common Activatable Widget Class Reference=&/WBP_ItemShop_PDP.WBP_ItemShop_PDP_C) -> (then:Exec, OnPrimaryContentPushed:Exec, ActivatableWidget:Common Activatable Widget Object Reference) {is=title K2Node_AsyncAction}
   n41 = Cast To WBP_ItemShop_PDP!(execute=@n40.OnPrimaryContentPushed, Object=@n40) -> (then, CastFailed, AsWBP Item Shop PDP:WBP Item Shop PDP Object Reference) {is=title K2Node_DynamicCast}
   > InitializePDP!(
    execute
    self=@n41
    OfferVM:Fort Item Shop Offer VM Object Reference=Get ParentOfferVM() -> (ParentOfferVM:Fort Item Shop Offer VM Object Reference)
    OfferGroupVM:Fort Item Shop Offer Group VM Object Reference
   ) -> (then)
   > SetAllInteractionsEnabled(ShouldEnable=true)
   "On Purchase Block Hovered (Component_PurchaseBlock)"() {is=title K2Node_ComponentBoundEvent}
   n42 = Get CurrentView() -> (CurrentView=OfferInspect)
   > Branch(Condition=AND Boolean(A=Get IsFocusingItems() -> (IsFocusingItems:bool), B="Equal (Enum)"(A=@n42, B=OfferInspect)))
   > ChangeFocusArea!(execute, self, IsFocusingItem:bool=false) -> (then) {is=title K2Node_CallFunction}
   n43 = "Item Hovered (Component_ItemList)"!() -> (OutputDelegate, then, Button:Common Button Base Object Reference, ItemIndex:int=0, ListItem:Object Reference) {is=title K2Node_ComponentBoundEvent}
   > n44 = Cast To FortItemVM!(execute, Object=@n43.ListItem) -> (then, CastFailed, AsFort Item VM:Fort Item VM Object Reference) {is=title K2Node_DynamicCast}
   > SelectItem(ItemVm=@n44)
   n45 = "Item Clicked (Component_ItemList)"!() -> (OutputDelegate, then, Button:Common Button Base Object Reference, ItemIndex:int=0, ListItem:Object Reference) {is=title K2Node_ComponentBoundEvent}
   > n46 = Cast To FortItemVM(Object=@n45.ListItem)
   n47 = GetSelected!(self:Common Button Base Object Reference=@n45.Button) -> (ReturnValue:bool)
   n48 = GetOwningPlayer()
   > n49 = Branch(Condition=AND Boolean(A=IsUsingTouch(PlayerController=@n48), B=NOT Boolean(A=@n47)))
   > SelectItem(ItemVm=@n46)
   ChangeView(execute=@n49.else, NewView=ItemDetails, DeactivateOnBack=false)
   n50 = GetItemTemplateId!(self:Fort Item VM Object Reference=@n46) -> (ReturnValue:String)
   > "Fire Event Element Interacted (Non-Localized)"!(
    execute
    FortPC:Fort Player Controller Object Reference=Cast To FortPlayerController!(Object=GetOwningPlayer()) -> (AsFort Player Controller, bSuccess:bool=~)
    ElementName:String=Button_ItemDetails
    AdditionalInfo:Flow Analytics Additional Info Structure=BuildAnalyticsInfo(OptionalAdditionalInfoMap=Make Map(Key 0=ItemId, Value 0=@n50))
   ) -> (then)
   n51 = "Card Pressed (Component_ItemList_ALC)"!() -> (OutputDelegate, then, Button:Widget Object Reference, FortItemVM:Fort Item VM Object Reference) {is=title K2Node_ComponentBoundEvent}
   > SelectItem(ItemVm=@n51.FortItemVM)
   > ChangeView(NewView=ItemDetails, DeactivateOnBack=false)
   n52 = "ALCVisibility Changed (Component_ItemList_ALC)"!() -> (OutputDelegate, then, Visibility:ESlateVisibility Enum=Visible) {is=title K2Node_ComponentBoundEvent}
   > n53 = Switch on ESlateVisibility!(execute, Selection:ESlateVisibility Enum=@n52.Visibility) -> (Visible:Exec, Collapsed:Exec, Hidden:Exec, HitTestInvisible:Exec, SelfHitTestInvisible:Exec) {is=title K2Node_SwitchEnum}
   SetNavigationRuleCustomBoundary!(
    execute=[@n53.Collapsed, @n53.Hidden]
    self:Widget Object Reference=Get Switcher_ItemDetails() -> (Switcher_ItemDetails:Common Visibility Switcher Object Reference)
    Direction:EUINavigation Enum=Down
    InCustomDelegate:Delegate=Create Event()
   ) -> (then)
   > SetNavigationRuleCustomBoundary(self=Get PurchaseBlock_ItemSlot() -> (PurchaseBlock_ItemSlot:Named Slot Object Reference), Direction=Up, InCustomDelegate=Create Event())
   n54 = Create Event()
   SetNavigationRuleCustomBoundary(
    execute=[@n53.HitTestInvisible, @n53.SelfHitTestInvisible, @n53.Visible]
    self=Get Switcher_ItemDetails()
    Direction=Down
    InCustomDelegate=@n54
   )
   > SetNavigationRuleCustomBoundary(self=Get PurchaseBlock_ItemSlot(), Direction=Up, InCustomDelegate=@n54)
   "On Clicked (Button_Options)"!() -> (OutputDelegate, then, Button:Common Button Base Object Reference) {is=title K2Node_ComponentBoundEvent}
   > ToggleContextMenu(self:WBP UIKit Context Menu Anchor Button Base Object Reference=Get ContextMenuAnchorButton_Options() -> (ContextMenuAnchorButton_Options:WBP UIKit Context Menu Anchor Button Object Reference))
   n55 = OnPurchaseOfferRequested!() -> (OutputDelegate, then, PriceIndex:int) {is=title K2Node_CustomEvent}
   n56 = OnPurchaseComplete!() -> (OutputDelegate, then, bSuccess:bool=~, PurchasedItems:Array of Purchased Item Info Structures, OfferId:String) {is=title K2Node_CustomEvent}
   n57 = PurchaseOffer!(
    execute=@n55.then
    self:Fort Item Shop Offer VM Object Reference=Get Offer to Purchase()
    PriceIndex:int=@n55.PriceIndex
    OnPurchaseCompleteCallback:"Delegate (by ref)"=@n56
   ) -> (then, ReturnValue:EOfferPurchaseError Enum=NoError)
   > Branch(Condition="Not Equal (Enum)"!(A:EOfferPurchaseError Enum=@n57, B:EOfferPurchaseError Enum=NoError) -> (ReturnValue:bool))
   > SetLoadingOverlayVisibility(IsVisible=true)
   OnLoadingError() {is=title K2Node_CustomEvent}
   > n58 = Branch(Condition=Get bIsActive() -> (bIsActive:bool))
   n59 = OnLoadingErrorButtonClicked!() -> (OutputDelegate, then, ButtonIndex:int) {is=title K2Node_CustomEvent}
   n60 = Push Simple UIKit Modal Dialog!(
    execute=@n58.then
    Context:User Widget Object Reference=Self-Reference()
    VM Class:UI Kit Dialog Viewmodel Class Reference=/ItemShop/MultiProduct/Data/DialogVM_ShopLoadingError.DialogVM_ShopLoadingError_C
    OnDialogButtonClicked:"Delegate (by ref)"=@n59
   ) -> (then, DialogActivatableWidget:WBP UIKit Dialog Object Reference, DialogVM:UI Kit Dialog Viewmodel Object Reference)
   n61 = OnCosmeticGateModalConfirmed() {is=title K2Node_CustomEvent}
   > AttemptPurchase(self:Fort MPItem Shop Offer Inspection Screen Object Reference) {is=title K2Node_CallFunction}
   n62 = OnCustomItemDetailsReactivated() {is=title K2Node_CustomEvent}
   > OnInteractionEnabledChanged!(execute, self:BPI Item Shop PDP Interface=Get Component_PurchaseBlock(), ShouldEnable:bool=true) -> (then)
   ClearSlottedItemDetails() {is=title K2Node_CustomEvent}
   n63 = GetContent!(self:Content Widget Object Reference=Get NamedSlot_CustomDetails() -> (NamedSlot_CustomDetails:Named Slot Object Reference)) -> (ReturnValue:Widget Object Reference)
   > ClearPurchaseBlockInjection!(execute:Exec, self:Object Reference=@n63) -> (then:Exec) {is=title K2Node_Message}
   > ClearALCListInjection!(execute:Exec, self:Object Reference=@n63) -> (then:Exec) {is=title K2Node_Message}
   > ClearChildren(self=Get NamedSlot_CustomDetails())
   n64 = ShowPurchaseOverlay!() -> (OutputDelegate, then, InPurchasedItems:Array of Strings, bInShowAsRefund:bool=~) {is=title K2Node_CustomEvent}
   > n65 = Is Valid(InputObject=Get PurchaseOverlayWidget() -> (PurchaseOverlayWidget:Athena Item Shop Offer Purchased Overlay Object Reference))
   n66 = Create Widget!(
    execute:Exec=@n65.Is Not Valid
    Class:User Widget Class Reference=/ItemShop/Athena/AthenaItemShopOfferPurchasedOverlay.AthenaItemShopOfferPurchasedOverlay_C
    OwningPlayer:Player Controller Object Reference=GetOwningPlayer()
   ) -> (then:Exec, ReturnValue:Athena Item Shop Offer Purchased Overlay Object Reference)
   > n67 = Set PurchaseOverlayWidget(PurchaseOverlayWidget:Athena Item Shop Offer Purchased Overlay Object Reference=@n66) -> (Output_Get:Athena Item Shop Offer Purchased Overlay Object Reference)
   > n68 = Bind Event to On Widget Deactivated(self:Common Activatable Widget Object Reference=@n67, Delegate=Create Event()) {is=title K2Node_AddDelegate}
   Setup!(
    execute=[@n68.then, @n65.Is Valid]
    self:Athena Item Shop Purchase Overlay Modal Object Reference=Get PurchaseOverlayWidget()
    InQuantityPurchased:int=1
    InPurchasedItemTemplateIds:Array of Strings=@n64.InPurchasedItems
    bInShowAsRefund:bool=@n64.bInShowAsRefund
   ) -> (then)
   > SetOffer!(
    execute
    self:Athena Item Shop Offer Details Widget Base Object Reference=Get PurchaseOverlayWidget()
    InOfferData:Fort Store Front Offer Info Object Reference=Get OfferInfo!(self:Fort Item Shop Offer VM Object Reference=Get Offer to Purchase()) -> (OfferInfo:Athena Item Shop Offer Info Object Reference)
   ) -> (then)
   > TryPushWidgetAsModalPanel(PushWidget=Get PurchaseOverlayWidget())
   n69 = OnInputMethodChanged!() -> (OutputDelegate, then, bNewInputType:ECommonInputType Enum) {is=title K2Node_CustomEvent}
   > RequestRefreshFocus()
   OnCpcModalRequested() {is=title K2Node_CustomEvent}
   > n70 = Construct Object from Class(Class=/FortniteCoreUI/Frontend/RightOfWithdrawal/DialogVM_RightOfWithdrawal.DialogVM_RightOfWithdrawal_C) -> (ReturnValue:Dialog VM Right Of Withdrawal Object Reference)
   > n71 = Create Widget!(
    execute
    Class=/FortniteCoreUI/Frontend/RightOfWithdrawal/WBP_RightOfWithdrawal_Dialog.WBP_RightOfWithdrawal_Dialog_C
    OwningPlayer=GetOwningPlayer()
    DialogVM_RightOfWithdrawal:Dialog VM Right Of Withdrawal Object Reference=@n70
    BodyText:Text="LOCTABLE(\"/ItemShop/MultiProduct/Data/ST_ItemShop\", \"CPC_DefaultText\")"
   ) -> (then, ReturnValue:WBP Right Of Withdrawal Dialog Object Reference)
   n72 = OnCpcDialogDismissed!() -> (OutputDelegate, then, Accepted:bool=~) {is=title K2Node_CustomEvent}
   Unbind Event from On Dialog Dismissed(execute=@n71.then, self:WBP Right Of Withdrawal Dialog Object Reference=@n71, Delegate=@n72) {is=title K2Node_RemoveDelegate}
   > Bind Event to On Dialog Dismissed(self:WBP Right Of Withdrawal Dialog Object Reference=@n71, Delegate=@n72) {is=title K2Node_AddDelegate}
   > TryPushWidgetAsModalPanel(PushWidget=@n71)
   OnShowNextVariant() {is=title K2Node_CustomEvent}
   > IncreaseVariantIndex()
   > ShowVariantPreview()
   n73 = HandlePreviewActionPressed!() -> (OutputDelegate, then, InteractionNameID:Name) {is=title K2Node_CustomEvent}
   > PreviewActionPressed!(execute, self:Fort Item Preview Screen Object Reference=Get FortItemPreviewWidget(), ActionName:Name=@n73.InteractionNameID) -> (then)
   n74 = HandlePreviewActionReleased!() -> (OutputDelegate, then, InteractionNameID:Name) {is=title K2Node_CustomEvent}
   > PreviewActionReleased!(execute, self:Fort Item Preview Screen Object Reference=Get FortItemPreviewWidget(), ActionName:Name=@n74.InteractionNameID) -> (then)
   n75 = HandlePreviewAction!() -> (OutputDelegate, then, InteractionNameID:Name) {is=title K2Node_CustomEvent}
   > PreviewAction!(execute, self:Fort Item Preview Screen Object Reference=Get FortItemPreviewWidget(), ActionName:Name=@n75.InteractionNameID) -> (then)
   n76 = On3DAssetsLoaded!() -> (OutputDelegate, then, bPreloadSucceeded:bool=~, Error:"Fort Download Error Structure (by ref)") {is=title K2Node_CustomEvent}
   > n77 = Set LoadState3D(LoadState3D:E_ItemPreview_LoadState Enum=Select(Option 0:E_ItemPreview_LoadState Enum=NewEnumerator1, Option 1:E_ItemPreview_LoadState Enum=NewEnumerator2, Index=@n76.bPreloadSucceeded) -> (ReturnValue:E_ItemPreview_LoadState Enum)) -> (Output_Get:E_ItemPreview_LoadState Enum=NewEnumerator3)
   > n78 = Switch on E_ItemPreview_LoadState!(execute, Selection:E_ItemPreview_LoadState Enum=@n77) -> (NewEnumerator3:Exec, NewEnumerator0:Exec, NewEnumerator2:Exec, NewEnumerator1:Exec) {is=title K2Node_SwitchEnum}
   n79 = Get FortItemShopOfferVM()
   n80 = Branch(
    execute=@n78.NewEnumerator1
    Condition="Not Equal (Enum)"(A:EFortPriceCurrency Enum=GetCurrencyType(self=@n79) -> (ReturnValue=Mtx), B:EFortPriceCurrency Enum=RealMoney)
   )
   n81 = Branch(
    execute=[@n80.then, @n84.NewEnumerator1]
    Condition="Equal (Enum)"(
     A:E_ItemPreview_LoadState Enum=Get LoadState2D() -> (LoadState2D:E_ItemPreview_LoadState Enum=NewEnumerator3)
     B:E_ItemPreview_LoadState Enum=Get LoadState3D() -> (LoadState3D:E_ItemPreview_LoadState Enum=NewEnumerator3)
    )
   )
   OnLoadingError(execute=[@n81.then, @n80.else]) {is=title K2Node_CallFunction}
   n82 = On2dAssetsLoaded!() -> (OutputDelegate, then, bPreloadSucceeded:bool=~, Error:"Fort Download Error Structure (by ref)") {is=title K2Node_CustomEvent}
   > n83 = Set LoadState2D(LoadState2D:E_ItemPreview_LoadState Enum=Select(Option 0=NewEnumerator1, Option 1=NewEnumerator2, Index=@n82.bPreloadSucceeded)) -> (Output_Get:E_ItemPreview_LoadState Enum=NewEnumerator3)
   > n84 = Switch on E_ItemPreview_LoadState(Selection=@n83)
   PurchaseStep_AllConfirmed() {is=title K2Node_CustomEvent}
   n85 = Get Offer to Purchase()
   > Branch(Condition="Equal (Enum)"(
     A:EOfferPurchaseError Enum=CanPurchaseOffer!(self:Fort Item Shop Offer VM Object Reference=@n85) -> (ReturnValue:EOfferPurchaseError Enum=NoError)
     B:EOfferPurchaseError Enum=NoError
    ))
   > OnPurchaseOfferRequested!(execute, self, PriceIndex:int=0) -> (then) {is=title K2Node_CallFunction}
   OnPurchaseInitiated() {is=title K2Node_CustomEvent}
   > SetAllInteractionsEnabled(ShouldEnable=false)
   > PurchaseStep_CosmeticGate()
   n86 = CreateAboutItemCoachmark!() -> (OutputDelegate, then, Class:Common Activatable Widget Class Reference) {is=title K2Node_CustomEvent}
   > Delay(Duration=0.5)
   > n87 = Create Widget!(execute, Class=@n86.Class, OwningPlayer) -> (then, ReturnValue:Common Activatable Widget Object Reference)
   > AddChild(
    self=Get VisualAttachment_AboutItemCoachmark() -> (VisualAttachment_AboutItemCoachmark:Common Visual Attachment Object Reference)
    Content=@n87
   )
   > n88 = Cast To WBP_UIKit_CoachMark!(execute, Object=@n87) -> (then, CastFailed, AsWBP UIKit Coach Mark:WBP UIKit Coach Mark Object Reference) {is=title K2Node_DynamicCast}
   > n89 = Set AboutItemCoachmark(AboutItemCoachmark:WBP UIKit Coach Mark Object Reference=@n88) -> (Output_Get:WBP UIKit Coach Mark Object Reference)
   n90 = Create Event()
   > Unbind Event from Button Dismiss Clicked(self:WBP UIKit Coach Mark Base Object Reference=@n89, Delegate=@n90) {is=title K2Node_RemoveDelegate}
   > Bind Event to Button Dismiss Clicked(self:WBP UIKit Coach Mark Base Object Reference=@n89, Delegate=@n90) {is=title K2Node_AddDelegate}
   > ActivateWidget(self=@n89)
   PurchaseStep_CosmeticGate() {is=title K2Node_CustomEvent}
   > n91 = Branch(Condition=Should Show Cosmetic Gate Modal!() -> (ShouldShow:bool=~) {is=title K2Node_MacroInstance})
   > OnCosmeticGateModalRequested()
   PurchaseStep_ConfirmationOrMMBan(execute=@n91.else)
   SetOffer() {is=title K2Node_CustomEvent}
   > SendOfferInspectedAnalytics(self:Fort Item Shop Offer VM Object Reference=Get FortItemShopOfferVM())
   > n92 = FindParentOffer!(
    execute
    self:Fort Item Shop VM Object Reference=Get FortItemShopVM()
    ChildOffer:Fort Item Shop Offer VM Object Reference=Get FortItemShopOfferVM()
   ) -> (then, ReturnValue:Fort Item Shop Offer VM Object Reference)
   > Set ParentOfferVM(ParentOfferVM:Fort Item Shop Offer VM Object Reference=@n92) -> (Output_Get:Fort Item Shop Offer VM Object Reference)
   > OnOfferSet!(
    execute
    self:BPI Item Shop PDP Interface
    OfferVM:Fort Item Shop Offer VM Object Reference=Get FortItemShopOfferVM()
    ParentOfferVM:Fort Item Shop Offer VM Object Reference=Get ParentOfferVM()
    OfferGroupVM:Fort Item Shop Offer Group VM Object Reference=Get FortItemShopOfferGroupVM() -> (FortItemShopOfferGroupVM:Fort Item Shop Offer Group VM Object Reference)
   ) -> (then)
   > PreloadGrantedItemPreviews!(
    execute
    self:Fort Item Shop Offer VM Object Reference=Get FortItemShopOfferVM()
    bAllow2dPreviews:bool=IsCosmetic2dPreviewEnabled!(self:Fort Item Preview Screen Object Reference=Get FortItemPreviewWidget()) -> (ReturnValue:bool)
    On2dAssetsPreloaded:"Delegate (by ref)"=Create Event()
    On3dAssetsPreloaded:"Delegate (by ref)"=Create Event()
   ) -> (then)
   > UpdateOwnershipInfo()
   > UpdateButtons()
   > UpdateNavigationArrows()
   > UpdateFeaturedOfferID()
   > RequestRefreshFocus()
   > ClearCachedPreviewContextTags(self:Fort Item Preview Screen Object Reference=Get FortItemPreviewWidget())
   > n93 = Branch(Condition="Equal (Integer)"(A=Get CachedPreviewedItemIndex() -> (CachedPreviewedItemIndex=0), B=-1))
   > n94 = TryRequestOverridePreview()
   > n95 = Branch(Condition=@n94)
   n96 = GetOwningPlayer()
   n97 = Get CurrentView() -> (CurrentView=OfferInspect)
   ChangeFocusArea(
    execute=[@n99.then, @n95.then, @n100.then]
    IsFocusingItem=OR Boolean(A="Equal (Enum)"(A:EFortItemShopPdpView Enum=@n97, B:EFortItemShopPdpView Enum=ItemDetails), B=IsUsingTouch(PlayerController=@n96))
   )
   n98 = Get Anim_ShowISBs() -> (Anim_ShowISBs:Widget Animation Object Reference)
   > PlayAnimation(InAnimation=@n98, StartAtTime=GetEndTime(self=@n98))
   n99 = PreviewItemAtIndex(execute=@n95.else, Index=0)
   PreviewItemAtIndex(execute=@n93.else, Index=Get CachedPreviewedItemIndex() -> (CachedPreviewedItemIndex=0))
   > n100 = UpdatePurchaseBlock()
   OnAboutItemCoachmarkDismissed() {is=title K2Node_CustomEvent}
   > n101 = DeactivateWidget(self=Get AboutItemCoachmark() -> (AboutItemCoachmark:WBP UIKit Coach Mark Object Reference))
   PurchaseStep_ConfirmationOrMMBan() {is=title K2Node_CustomEvent}
   n102 = Get Offer to Purchase()
   n103 = GetPurchaseConfirmationMessage_MatchmakingBan!(self:Fort Item Shop Offer VM Object Reference=@n102) -> (ReturnValue:Text)
   n104 = GetPurchaseConfirmationMessage_Generic!(self:Fort Item Shop Offer VM Object Reference=@n102) -> (ReturnValue:Text)
   > n105 = Branch(Condition=NAND Boolean(A=TextIsEmpty(InText=@n104), B=TextIsEmpty(InText=@n103)) {is=title K2Node_CommutativeAssociativeBinaryOperator})
   > On Purchase Confirmation Message Requested()
   PurchaseStep_Cpc(execute=@n105.else)
   SetItemDetailsView() {is=title K2Node_CustomEvent}
   > n106 = Get Previewed Item Preview Data()
   > n107 = Get(self=@n106)
   > n108 = Branch(Condition=RequiresCustomization!(self:Fort Item VM Object Reference=@n107) -> (ReturnValue:bool))
   > n109 = Branch(Condition=GetIsBulkCosmeticInstalled!(self:Fort Download System VM Object Reference=Get FortDownloadSystemVM() -> (FortDownloadSystemVM:Fort Download System VM Object Reference)) -> (ReturnValue:bool))
   n110 = TryGetSoftCustomActivatableWidgetClass(self=GetItemCategory(self=@n107), ViewRequest=(TagName="UI.ItemShop.ItemDetails"))
   n111 = Branch(execute=[@n109.then, @n108.else], Condition=@n110.ReturnValue)
   > OpenPrimaryItemDetails!(execute, self, UserWidgetSoftClass:User Widget Soft Class Reference=@n110, FortItemVM:Fort Item VM Object Reference=@n107) -> (then)
   n112 = HasLoadoutOverride(self=@n106)
   n113 = GetCustomItemDetailsFromItem!(ItemVM:Fort Item VM Object Reference=@n107) -> (CustomItemDetails:User Widget Soft Class Reference, IsValid:bool)
   n114 = Branch(execute=@n111.else, Condition=AND Boolean(A=@n113.IsValid, B=NOT Boolean(A=@n112)))
   > OpenSlottedItemDetails!(execute, self, WidgetSoftClass:User Widget Soft Class Reference=@n113, ItemVM:Fort Item VM Object Reference=@n107) -> (then)
   OpenDefaultItemDetails(execute=[@n114.else, @n107.else, @n106.Is Not Valid])
   PurchaseStep_Cpc() {is=title K2Node_CustomEvent}
   > n115 = ShouldShowCpcDialog!(execute, Context:Widget Object Reference=Self-Reference(), OfferVM:Fort Item Shop Offer VM Object Reference=Get Offer to Purchase()) -> (then, ReturnValue:bool)
   > n116 = Branch(Condition=@n115)
   > OnCpcModalRequested() {is=title K2Node_CallFunction}
   PurchaseStep_RmtStorefront(execute=@n116.else)
   PurchaseStep_RmtStorefront() {is=title K2Node_CustomEvent}
   > n117 = EvaluateRmtStorefront!(execute, self) -> (then, HasMultipleStorefronts:bool) {is=title K2Node_CallFunction}
   > n118 = Branch(Condition=@n117)
   PurchaseStep_AllConfirmed(execute=@n118.else) {is=title K2Node_CallFunction}
   n119 = OnCustomDetailsDeactivated() {is=title K2Node_CustomEvent}
   > Branch(Condition=IsActivated!(self:Common Activatable Widget Object Reference) -> (ReturnValue:bool))
   > Set PrimaryItemDetails(PrimaryItemDetails:Common Activatable Widget Object Reference) -> (Output_Get:Common Activatable Widget Object Reference)
   > ChangeView(NewView=OfferInspect, DeactivateOnBack=false)
   > SetAllInteractionsEnabled(ShouldEnable=true)
   n120 = OpenPrimaryItemDetails!() -> (OutputDelegate, then, UserWidgetSoftClass:User Widget Soft Class Reference, FortItemVM:Fort Item VM Object Reference) {is=title K2Node_CustomEvent}
   > SetAllInteractionsEnabled(ShouldEnable=false)
   > n121 = CreateWidgetAsync(UserWidgetSoftClass=@n120.UserWidgetSoftClass, OwningPlayer=GetOwningPlayer(), bSuspendInputUntilComplete=true)
   n122 = Cast To CommonActivatableWidget!(execute=@n121.OnComplete, Object=@n121) -> (then, CastFailed, AsCommon Activatable Widget:Common Activatable Widget Object Reference) {is=title K2Node_DynamicCast}
   > n123 = Set PrimaryItemDetails(PrimaryItemDetails=@n122)
   > Assign On Widget Deactivated!(execute:Exec, self:Common Activatable Widget Object Reference=@n123, Delegate:"Delegate (by ref)"=@n119) -> (then:Exec) {is=title K2Node_AssignDelegate}
   > Bind Event to On Widget Activated(self:Common Activatable Widget Object Reference=@n123, Delegate=@n62) {is=title K2Node_AddDelegate}
   > SetViewModelByClass(Widget=@n123, ViewModel=@n120.FortItemVM)
   > n124 = Sequence()
   > n125 = Branch(Condition="Equal (Object)"(A=Get Offer to Purchase(), B=Get FortItemShopOfferVM()))
   n126 = Get PrimaryItemDetails() -> (PrimaryItemDetails:Common Activatable Widget Object Reference)
   InjectPurchaseBlock!(execute:Exec=@n125.else, self:Object Reference=@n126, PurchaseBlock:Widget Object Reference=Get Component_PurchaseBlock()) -> (then:Exec) {is=title K2Node_Message}
   > InjectALCList!(
    execute:Exec
    self:Object Reference=@n126
    ItemListALC:Widget Object Reference=Get Component_ItemList_ALC() -> (Component_ItemList_ALC:WBP Item Shop PDP Item List ALC Object Reference)
   ) -> (then:Exec) {is=title K2Node_Message}
   n127 = Get Component_IsbManager() -> (Component_IsbManager:WBP Item Shop PDP Isb Manager Object Reference)
   > InjectISBs!(execute:Exec, self:Object Reference=@n126, ISBs:Widget Object Reference=@n127) -> (then:Exec) {is=title K2Node_Message}
   > SetTextJustification!(execute, self:WBP Item Shop PDP Isb Manager Object Reference=@n127, TextJustification:ETextJustify Enum=Left) -> (then)
   TryPushWidgetAsActivatablePanel!(
    execute=@n124.then_1
    PushWidget:User Widget Object Reference=Get PrimaryItemDetails()
    Header:EActivatePanelOption Enum=Hide
    Footer:EActivatePanelOption Enum=Hide
    ChatWidget:EActivatePanelOption Enum=Hide
   ) -> (then)
   n128 = OpenSlottedItemDetails!() -> (OutputDelegate, then, WidgetSoftClass:User Widget Soft Class Reference, ItemVM:Fort Item VM Object Reference) {is=title K2Node_CustomEvent}
   > SetAllInteractionsEnabled(ShouldEnable=false)
   > n129 = CreateWidgetAsync(UserWidgetSoftClass=@n128.WidgetSoftClass, OwningPlayer=GetOwningPlayer(), bSuspendInputUntilComplete=false)
   AddChild(execute=@n129.OnComplete, self=Get NamedSlot_CustomDetails(), Content=@n129)
   > SetActiveWidget(self=Get Switcher_ItemDetails(), Widget=Get NamedSlot_CustomDetails())
   > SetViewModelByClass(
    Widget=@n129
    ViewModel=Get ItemDetails!(self:Fort Item VM Object Reference=@n128.ItemVM) -> (ItemDetails:Fort Item Details VM Object Reference)
   )
   > SetAllInteractionsEnabled(ShouldEnable=true)
   > n130 = Branch(Condition="Equal (Object)"(A=Get Offer to Purchase(), B=Get FortItemShopOfferVM()))
   InjectPurchaseBlock(execute=@n130.else, self=@n129, PurchaseBlock=Get Component_PurchaseBlock())
   > InjectALCList(self=@n129, ItemListALC=Get Component_ItemList_ALC())
   OpenDefaultItemDetails() {is=title K2Node_CustomEvent}
   > SetActiveWidget(self=Get Switcher_ItemDetails(), Widget=Get ItemDetails_Default() -> (ItemDetails_Default:Vertical Box Object Reference))
   > Set Visibility(
    Visibility=Select(Option 0:ESlateVisibility Enum=Collapsed, Option 1:ESlateVisibility Enum=SelfHitTestInvisible, Index=Get HasVariants()) -> (ReturnValue:ESlateVisibility Enum)
    self=Get WBP_UIKit_VariantSelector()
   )
   n131 = Get Offer to Purchase()
   n132 = Get FortItemShopOfferVM()
   > n133 = Branch(Condition=InvertBool!(bValue:bool="Equal (Object)"(A=@n131, B=@n132)) -> (ReturnValue:bool))
   > SetContent!(execute, self:Content Widget Object Reference=Get PurchaseBlock_ItemSlot(), Content:Widget Object Reference=Get Component_PurchaseBlock()) -> (then, ReturnValue:Panel Slot Object Reference)
   > Set!(
    execute
    Padding_Left:float=0
    Padding_Top:float=16
    Padding_Right:float=0
    Padding_Bottom:float=0
    self:User Widget Object Reference=Get Component_PurchaseBlock()
   ) -> (then, Output_Get:Margin Structure)
   > n134 = Branch(Condition=Get IsIsbOnRightSide() -> (IsIsbOnRightSide:bool))
   SetContent(
    execute=[@n135.then, @n134.then]
    self=Get ItemListALC_ItemSlot() -> (ItemListALC_ItemSlot:Named Slot Object Reference)
    Content=Get Component_ItemList_ALC()
   )
   > SetVisibility(self=Get ItemListALC_ItemSlot(), InVisibility=SelfHitTestInvisible)
   n135 = SetContent(execute=@n134.else, self=Get IsbSlot_ALC() -> (IsbSlot_ALC:Named Slot Object Reference), Content=Get Component_IsbManager())
   SetVisibility(execute=@n133.else, self=Get ItemListALC_ItemSlot(), InVisibility=Collapsed)
   n136 = OnBackgroundDataLoaded!() -> (OutputDelegate, then, LoadedBackground:Object Reference) {is=title K2Node_CustomEvent}
   > UpdateBackground!(
    execute
    self:Fort Item Preview Background Object Reference=Get FortItemPreviewBackground() -> (FortItemPreviewBackground:Fort Item Preview Background Object Reference)
    LoadedBackground:Object Reference=@n136.LoadedBackground
    TransitionDirection:EVaultWorldTransitionDirection Enum=None
    PrimaryColor:Linear Color Structure=(R=0,G=0,B=0,A=0)
    SecondaryColor:Linear Color Structure=(R=0,G=0,B=0,A=0)
   ) -> (then)
   OnPurchaseCanceled() {is=title K2Node_CustomEvent}
   > UpdateButtons()
   On Purchase Confirmation Message Requested() {is=title K2Node_CustomEvent}
   > n137 = Construct Object from Class(Class=/ItemShop/MultiProduct/Data/DialogVM_PurchaseConfirmation.DialogVM_PurchaseConfirmation_C) -> (ReturnValue:Dialog VM Purchase Confirmation Object Reference)
   > n138 = Create Widget!(
    execute
    Class=&/WBP_PurchaseConfirmation_Dialog.WBP_PurchaseConfirmation_Dialog_C
    OwningPlayer=GetOwningPlayer()
    DialogVM_PurchaseConfirmation:Dialog VM Purchase Confirmation Object Reference=@n137
    FortItemShopOfferVM:Fort Item Shop Offer VM Object Reference=Get Offer to Purchase()
   ) -> (then, ReturnValue:WBP Purchase Confirmation Dialog Object Reference)
   n139 = OnPurchaseConfirmed() {is=title K2Node_CustomEvent}
   Unbind Event from Purchase Confirmed(execute=@n138.then, self:WBP Purchase Confirmation Dialog Object Reference=@n138, Delegate=@n139) {is=title K2Node_RemoveDelegate}
   > Bind Event to Purchase Confirmed(self:WBP Purchase Confirmation Dialog Object Reference=@n138, Delegate=@n139) {is=title K2Node_AddDelegate}
   n140 = Create Event()
   > Unbind Event from Purchase Dismissed(self:WBP Purchase Confirmation Dialog Object Reference=@n138, Delegate=@n140) {is=title K2Node_RemoveDelegate}
   > Bind Event to Purchase Dismissed(self:WBP Purchase Confirmation Dialog Object Reference=@n138, Delegate=@n140) {is=title K2Node_AddDelegate}
   > TryPushWidgetAsModalPanel(PushWidget=@n138)
   OnCosmeticGateModalRequested() {is=title K2Node_CustomEvent}
   > n141 = Branch(Condition=Should Show Pin Modal!() -> (ShouldShow:bool=~) {is=title K2Node_MacroInstance})
   n142 = Get Offer to Purchase()
   > RequestPurchasePinModal!(execute, self:Fort MPItem Shop Offer Inspection Screen Object Reference, OfferVM:Fort Item Shop Offer VM Object Reference=@n142) -> (then)
   > SetAllInteractionsEnabled(ShouldEnable=true)
   n143 = Create Widget!(
    execute=@n141.else
    Class=/FortniteCoreUI/Frontend/CosmeticGating/WBP_CosmeticGate_ShopModal.WBP_CosmeticGate_ShopModal_C
    OwningPlayer
    FortItemShopOfferVM=@n142
   ) -> (then, ReturnValue:WBP Cosmetic Gate Shop Modal Object Reference)
   > Bind Event to On Dismissed(self:WBP Cosmetic Gate Shop Modal Object Reference=@n143, Delegate=Create Event()) {is=title K2Node_AddDelegate}
   > Bind Event to On Purchase Clicked(self:WBP Cosmetic Gate Shop Modal Object Reference=@n143, Delegate=@n61) {is=title K2Node_AddDelegate}
   > TryPushWidgetAsModalPanel(PushWidget=@n143)
   n144 = Event OnOfferSet!() -> (
    OutputDelegate
    then
    OfferVM:Fort Item Shop Offer VM Object Reference
    ParentOfferVM:Fort Item Shop Offer VM Object Reference
    OfferGroupVM:Fort Item Shop Offer Group VM Object Reference
   ) {is=title K2Node_Event}
   > n145 = For Each Loop(Array:Array of BPI Item Shop PDP Interfaces=Get Components() -> (Components:Array of BPI Item Shop PDP Interfaces)) -> (Array Element:BPI Item Shop PDP Interface)
   > OnOfferSet(self=@n145, OfferVM=@n144.OfferVM, ParentOfferVM=@n144.ParentOfferVM, OfferGroupVM=@n144.OfferGroupVM)
   Event On Activated()
   > SetTextChatEnabled!(execute, self, bSetActive:bool=false) -> (then)
   > n146 = UpdateVaultWorldBackground!(
    execute
    self:WBP Item Shop Background Gradient Object Reference=Get WBP_ItemShop_BackgroundGradient() -> (WBP_ItemShop_BackgroundGradient:WBP Item Shop Background Gradient Object Reference)
    IsActive:bool=Get bIsActive()
    Fort Item Shop VM:Fort Item Shop VM Object Reference=Get FortItemShopVM()
    OnLoadedCallback:"Delegate (by ref)"=Create Event()
   ) -> (then, ToUpdateBackground:bool)
   > n147 = Branch(Condition=@n146)
   > n148 = UpdateBackground(self=Get FortItemPreviewBackground(), TransitionDirection=None, PrimaryColor=(R=0,G=0,B=0,A=0), SecondaryColor=(R=0,G=0,B=0,A=0))
   Set PreviewedItem(execute=[@n148.then, @n147.else], self=Get FortItemShopVM())
   > n149 = Branch(Condition="Not Equal (Enum)"(A:EFortItemShopPdpView Enum=Get CurrentView() -> (CurrentView=OfferInspect), B:EFortItemShopPdpView Enum=ItemDetails))
   > n150 = Set CachedPreviewedItemIndex(CachedPreviewedItemIndex=-1) -> (Output_Get=0)
   n151 = Get FortItemPreviewWidget()
   ActivateWidget(execute=[@n150.then, @n149.else], self=@n151)
   > ClearPreviewItem(self:Fort Item Preview Widget Object Reference=@n151)
   n152 = Get Anim_Equipped() -> (Anim_Equipped:Widget Animation Object Reference)
   > PlayAnimation(InAnimation=@n152, StartAtTime=GetEndTime(self=@n152), PlayMode=Reverse)
   > SetOffer!(execute, self:Self Object Reference) -> (then) {is=title K2Node_CallFunction}
   > SetScalarParameterValue!(
    execute
    Collection:Material Parameter Collection Object Reference=/Game/Athena/UI/Frontend/CosmeticGating/Materials/MPC_CosmeticGate
    ParameterName:Name=UseAlternativeWarningPresentation
    ParameterValue:float=1
   ) -> (then) {is=title K2Node_CallMaterialParameterCollectionFunction}
   > OnPdpActivated(self:BPI Item Shop PDP Interface)
   > Bind Event to On Input Method Changed(self=Get CommonInputSubsystem(), Delegate=@n69)
   n153 = Event RequestChangeView!() -> (OutputDelegate, then, InNewView:EFortItemShopPdpView Enum=OfferInspect, InOverrideBack:bool) {is=title K2Node_Event}
   > ChangeView(NewView=@n153.InNewView, DeactivateOnBack=@n153.InOverrideBack)
   Event On Deactivated()
   > OnPdpDeactivated(self:BPI Item Shop PDP Interface)
   > Unbind Event from On Input Method Changed(self=Get CommonInputSubsystem(), Delegate=@n69)
   > Set PreviewedItem(self=Get FortItemShopVM())
   > Set LoadState2D(LoadState2D=NewEnumerator0) -> (Output_Get=NewEnumerator3)
   > Set LoadState3D(LoadState3D=NewEnumerator0) -> (Output_Get=NewEnumerator3)
   > SetTextChatEnabled(bSetActive=true)
   > SetScalarParameterValue(
    Collection=&MPC_CosmeticGate
    ParameterName=UseAlternativeWarningPresentation
    ParameterValue=0
   )
   > n154 = Get!(execute) -> (then, else, FortItemShopOfferVM:Fort Item Shop Offer VM Object Reference)
   > ReleasePreloadedGrantedItemPreviews(self:Fort Item Shop Offer VM Object Reference=@n154)
   > DeactivateWidget(self=Get FortItemPreviewWidget())
   > FlushAnalyticsImpressionsForOffer(self:Fort Item Shop Offer VM Object Reference=@n154)
   > ClearUnsavedPreCustomizations!(execute) -> (then)
   Event CancelPurchase() {is=title K2Node_Event}
   > OnPurchaseCanceled() {is=title K2Node_CallFunction}
   n155 = Event StartPurchase!() -> (OutputDelegate, then, PriceIndex:int=0) {is=title K2Node_Event}
   > OnPurchaseOfferRequested(PriceIndex=@n155.PriceIndex)
   Event OnPdpActivated() {is=title K2Node_Event}
   > n156 = For Each Loop(Array=Get Components())
   > OnPdpActivated(self=@n156)
   Event PreConstruct()
   > SetupComponents()
   > PaginationButtons_UpdateOutline()
   > SetCompactStyle()
   > SetRTLAlignment()
   n157 = Event OnViewChanged!() -> (OutputDelegate, then, NewView:EFortItemShopPdpView Enum=OfferInspect) {is=title K2Node_Event}
   > n158 = For Each Loop(Array=Get Components())
   > OnViewChanged!(execute, self:BPI Item Shop PDP Interface=@n158, NewView:EFortItemShopPdpView Enum=@n157.NewView) -> (then) {is=title K2Node_CallFunction}
   n159 = Event OnHandlePreviewItem!() -> (OutputDelegate, then, bUseWidePreview:bool) {is=title K2Node_Event}
   n160 = Get SizeBox_CameraFrame() -> (SizeBox_CameraFrame:Size Box Object Reference)
   > SetGridSlotPadding!(
    execute
    Widget:Widget Object Reference=@n160
    Padding_Left:float=SelectFloat(A=106, B=0, bPickA=@n159.bUseWidePreview)
    Padding_Top:float=100
    Padding_Right:float=0
    Padding_Bottom:float=100
   ) -> (then)
   > SetWidthOverride(self=@n160, InWidthOverride=SelectFloat(A=1200, B=1160, bPickA=@n159.bUseWidePreview))
   n161 = Event OnInteractionEnabledChanged!() -> (OutputDelegate, then, ShouldEnable:bool) {is=title K2Node_Event}
   > n162 = For Each Loop(Array=Get Components())
   > OnInteractionEnabledChanged(self=@n162, ShouldEnable=@n161.ShouldEnable)
   n163 = Event OnFocusAreaChanged!() -> (OutputDelegate, then, IsFocusingItems:bool) {is=title K2Node_Event}
   > n164 = For Each Loop(Array=Get Components())
   > OnFocusAreaChanged!(execute, self:BPI Item Shop PDP Interface=@n164, IsFocusingItems:bool=@n163.IsFocusingItems) -> (then) {is=title K2Node_CallFunction}
   Event OnPdpDeactivated() {is=title K2Node_Event}
   > n165 = For Each Loop(Array=Get Components())
   > OnPdpDeactivated(self=@n165)
   n166 = Event OnViewModelsSet!() -> (OutputDelegate, then, OfferVM:Fort Item Shop Offer VM Object Reference, OfferGroupVM:Fort Item Shop Offer Group VM Object Reference) {is=title K2Node_Event}
   > InitializePDP(OfferVM=@n166.OfferVM, OfferGroupVM=@n166.OfferGroupVM)
   PushGiftBoxScreenFromPurchaseInfo!(
    execute:Exec=@n170.then
    InPurchasedItems:Array of Purchased Item Info Structures=@n56.PurchasedItems
    bRemoveDuplicatesAndVariantInfo:bool=true
   ) -> (then:Exec, OnPrimaryContentPushed:Exec, ActivatableWidget:Common Activatable Widget Object Reference) {is=title K2Node_AsyncAction}
   PushGiftBoxScreenFromPurchaseInfo(execute=@n182.else, InPurchasedItems=@n56.PurchasedItems, bRemoveDuplicatesAndVariantInfo=true)
   Call On Load Error!(execute:Exec=@n175.then, self:Fort MPItem Shop Offer Inspection Screen Object Reference) -> (then:Exec) {is=title K2Node_CallDelegate}
   PlayAnimationForward(execute=@n171.then, InAnimation=Get Anim_Equipped())
   PurchaseStep_Cpc(execute=@n139.then) {is=title K2Node_CallFunction}
   OnPurchaseCanceled(execute=@n183.else)
   PurchaseStep_RmtStorefront(execute=@n183.then) {is=title K2Node_CallFunction}
   DeactivateWidget(execute=@n180.then)
   n167 = GetItemTemplateIDs!(execute=@n182.then, self, ItemInfo:Array of Purchased Item Info Structures=@n56.PurchasedItems) -> (then, ItemTemplateIDs:Array of Strings)
   > ShowPurchaseOverlay!(execute, self, InPurchasedItems:Array of Strings=@n167, bInShowAsRefund:bool=false) -> (then) {is=title K2Node_CallFunction}
   ShowPurchaseOverlay(execute=@n173.then, InPurchasedItems=@n39.RefundedItemTemplateIds, bInShowAsRefund=true)
   Reroute Node(InputPin=@n172.else)
   Reroute Node(InputPin=@n178.then)
   n168 = Length(TargetArray:Array of Strings=@n39.RefundedItemTemplateIds)
   n169 = Is Empty!(TargetArray:Array of Purchased Item Info Structures=@n56.PurchasedItems) -> (ReturnValue:bool)
   Call On End Purchase!(execute:Exec=@n176.then, self:WBP Item Shop PDP Object Reference, PurchasedItems:Array of Purchased Item Info Structures=@n56.PurchasedItems) -> (then:Exec) {is=title K2Node_CallDelegate}
   > n170 = Branch(Condition=ShouldShowGiftBoxScreenForRMTs!() -> (ReturnValue:bool))
   UpdateEquipButtonVisibility(execute=@n7.then)
   > SetVisibility(self=Get Throbber_Equip(), InVisibility=Hidden)
   > n171 = Branch(Condition=@n7.bItemEquipped)
   SetLoadingOverlayVisibility(execute=@n39.then, IsVisible=false)
   > n172 = Branch(Condition=AND Boolean(A=@n39.bSuccess, B="integer > integer"(A=@n168)))
   > n173 = UpdateOwnershipInfo()
   n174 = GetCurrencyType(self=Get FortItemShopOfferVM()) -> (ReturnValue=Mtx)
   ClosePDP(execute=@n59.then)
   > n175 = Unbind Event from Dialog Button Clicked Dispatcher(
    self:WBP UIKit Dialog Internal Object Reference=Get Dialog_Internal!(self:WBP UIKit Dialog Object Reference=@n60) -> (Dialog_Internal:WBP UIKit Dialog Internal Object Reference)
    Delegate=@n59
   ) {is=title K2Node_RemoveDelegate}
   n176 = DeactivateWidget(execute=@n181.then)
   n177 = ShouldShowOfferPurchasedOverlay!() -> (ReturnValue:bool)
   n178 = SetAllInteractionsEnabled(execute=@n179.else, ShouldEnable=true)
   SetLoadingOverlayVisibility(execute=@n56.then, IsVisible=false)
   > n179 = Branch(Condition=@n56.bSuccess)
   > UpdateOwnershipInfo()
   > n180 = Branch(Condition=@n169)
   n181 = Branch(execute=@n180.else, Condition="Equal (Enum)"(A:EFortPriceCurrency Enum=@n174, B:EFortPriceCurrency Enum=RealMoney))
   n182 = Branch(execute=@n181.else, Condition=@n177)
   n183 = Branch(execute=@n72.then, Condition=@n72.Accepted)
  ]}
 GetBundleOwnedItemsCount {nodes [
   GetBundleOwnedItemsCount() {is=title K2Node_FunctionEntry}
   n1 = GetDynamicBundleItems!(self:Fort Store Front Offer Info Object Reference=Get OfferInfo(self=Get FortItemShopOfferVM())) -> (Items:Array of Bundled Item Info Structures)
   > n2 = For Each Loop!(Exec, Array:Array of Bundled Item Info Structures=@n1) -> (LoopBody, Array Element_TemplateId:String, Array Element_Quantity:int, Array Element_bOwned:bool=~, Array Index, Completed)
   > Branch(Condition=@n2.Array Element_bOwned)
   > Increment Int(Value=Get OwnedBundleItems() -> (OwnedBundleItems:int=0))
   n3 = Length(TargetArray:Array of Bundled Item Info Structures=@n1)
   n4 = Get OwnedBundleItems() -> (OwnedBundleItems=0)
   Return Node!(execute=@n2.Completed, OwnedItems:int=@n4, UnownedItems:int=int - int(A=@n3, B=@n4), TotalItems:int=@n3)
  ]}
 GetCustomItemDetailsFromItem {nodes [
   n1 = GetCustomItemDetailsFromItem!() -> (then, ItemVM:Fort Item VM Object Reference) {is=title K2Node_FunctionEntry}
   > n2 = Is Valid(InputObject=@n1)
   n3 = GetItemCategory(self=@n1)
   > n4 = Is Valid(InputObject=@n3)
   n5 = Get CustomItemDetails!(self:Fort Locker Category Config Object Reference=@n3) -> (CustomItemDetails:User Widget Soft Class Reference)
   > Return Node!(
    execute
    CustomItemDetails:User Widget Soft Class Reference=@n5
    IsValid:bool=IsValidSoftClassReference!(SoftClassReference:"Object Soft Class Reference (by ref)"=@n5) -> (ReturnValue:bool)
   )
   Return Node(execute=[@n4.Is Not Valid, @n2.Is Not Valid], IsValid=false)
  ]}
 GetIsOfferRestricted {nodes [
   n1 = GetIsOfferRestricted!() -> (then, OfferVM:Fort Item Shop Offer VM Object Reference) {is=title K2Node_FunctionEntry}
   > n2 = For Each Loop!(Exec, Array:Array of Fort Item VM Object References=Get GrantedItems(self=@n1)) -> (LoopBody, Array Element:Fort Item VM Object Reference, Array Index, Completed)
   n3 = ConvertCosmeticItemDefinitionsToInterfaces!(
    execute=@n2.Completed
    CosmeticItemDefinitions:Array of Athena Cosmetic Item Definition Object References=Get ItemDefs() -> (ItemDefs:Array of Athena Cosmetic Item Definition Object References)
   ) -> (then, OutCosmeticItemGatingInterfaces:Array of Cosmetic Item Gating Interface Interfaces)
   > n4 = Set CosmeticGatingInterfaces(CosmeticGatingInterfaces:Array of Cosmetic Item Gating Interface Interfaces=@n3) -> (Output_Get:Array of Cosmetic Item Gating Interface Interfaces)
   > n5 = ItemsHaveVocalRatingRestriction!(
    execute
    LocalPlayer:Local Player Object Reference=GetOwningLocalPlayer!(self:Widget Object Reference) -> (ReturnValue:Local Player Object Reference)
    CosmeticItems:Array of Cosmetic Item Gating Interface Interfaces=@n4
   ) -> (then, ReturnValue:bool)
   > Set OfferContainsVocalRestriction(OfferContainsVocalRestriction:bool=@n5) -> (Output_Get:bool)
   > n6 = IsPlayerEligibleToUseRestrictedItems!(
    execute
    LocalPlayer:Local Player Object Reference=GetOwningLocalPlayer()
    CosmeticItems:Array of Cosmetic Item Gating Interface Interfaces=Get CosmeticGatingInterfaces() -> (CosmeticGatingInterfaces:Array of Cosmetic Item Gating Interface Interfaces)
   ) -> (then, ReturnValue:bool)
   > n7 = ItemsHaveRatingRestrictions!(
    execute
    LocalPlayer:Local Player Object Reference=GetOwningLocalPlayer()
    CosmeticItems:Array of Cosmetic Item Gating Interface Interfaces=Get CosmeticGatingInterfaces()
   ) -> (then, ReturnValue:bool)
   n8 = Get OfferContainsVocalRestriction() -> (OfferContainsVocalRestriction:bool)
   n9 = Get HasLoopFoundUnrestrictedOutfit() -> (HasLoopFoundUnrestrictedOutfit:bool)
   > Return Node!(execute, OfferContainsRestriction:bool=@n7, OfferRequiresPin:bool=NOT Boolean(A=OR Boolean!(A=@n6, B=@n9, C:bool=@n8) -> (ReturnValue)))
   Set LoopItem(execute=@n2.LoopBody, LoopItem:Fort Item VM Object Reference=@n2) -> (Output_Get:Fort Item VM Object Reference)
   > n10 = Cast To AthenaItemVM!(execute, Object=Get LoopItem() -> (LoopItem:Fort Item VM Object Reference)) -> (then, CastFailed, AsAthena Item VM:Athena Item VM Object Reference) {is=title K2Node_DynamicCast}
   > n11 = Cast To AthenaCosmeticItemDefinition!(execute, Object=Get ItemDefinition(self=@n10)) -> (then, CastFailed, AsAthena Cosmetic Item Definition:Athena Cosmetic Item Definition Object Reference) {is=title K2Node_DynamicCast}
   > Add(
    TargetArray:Array of Athena Cosmetic Item Definition Object References=Get ItemDefs()
    NewItem:"Athena Cosmetic Item Definition Object Reference (by ref)"=@n11
   ) {is=title K2Node_CallArrayFunction}
   > Branch(Condition=NOT Boolean(A=Get HasLoopFoundUnrestrictedOutfit()))
   > n12 = HasRatingRestrictions!(
    execute
    LocalPlayer:Local Player Object Reference=GetOwningLocalPlayer()
    CosmeticItemGatingInterface:"Cosmetic Item Gating Interface Interface (by ref)"=@n11
   ) -> (then, ReturnValue:bool)
   n13 = GetItemType(ItemDefinition=@n11) -> (ReturnValue=WorldItem)
   > Branch(Condition=AND Boolean(A=NOT Boolean(A=@n12), B="Equal (Enum)"(A:EFortItemType Enum=@n13, B:EFortItemType Enum=AthenaCharacter)))
   > Set HasLoopFoundUnrestrictedOutfit(HasLoopFoundUnrestrictedOutfit:bool=true) -> (Output_Get:bool)
  ]}
 GetItemTemplateIDs {nodes [
   n1 = GetItemTemplateIDs!() -> (then, ItemInfo:Array of Purchased Item Info Structures) {is=title K2Node_FunctionEntry}
   > n2 = For Each Loop(Array:Array of Purchased Item Info Structures=@n1) -> (Array Element:Purchased Item Info Structure)
   n3 = Break Purchased Item Info!(PurchasedItemInfo:"Purchased Item Info Structure (by ref)"=@n2) -> (Item:Fort Item Object Reference, Quantity:int) {is=title K2Node_BreakStruct}
   n4 = Get TemplateIDs() -> (TemplateIDs:Array of Strings)
   > Add(
    TargetArray:Array of Strings=@n4
    NewItem:"String (by ref)"=GetTemplateId!(self:Mcp Item Definition Base Object Reference=GetItemDefinition!(self:Fort Item Object Reference=@n3) -> (ReturnValue:Fort Item Definition Object Reference)) -> (ReturnValue:String)
   )
   Return Node!(execute=@n2.Completed, ItemTemplateIDs:Array of Strings=@n4)
  ]}
 GetNextOffer {nodes [
   GetNextOffer() {is=title K2Node_FunctionEntry}
   n1 = Get FortItemShopOfferVM()
   > n2 = GetOfferAtIndex!(execute, self, Index:int=int + int(A=GetOfferIndex!(OfferVM:Object Reference=@n1) -> (Array Index:int=0))) -> (then, OfferVM:Fort Item Shop Offer VM Object Reference, Found:bool)
   > Return Node!(execute, OfferVM:Fort Item Shop Offer VM Object Reference=@n2, Found:bool=@n2.Found)
  ]}
 GetOfferAtIndex {nodes [
   GetOfferAtIndex!() -> (then, Index:int) {is=title K2Node_FunctionEntry}
   n1 = Get FortItemShopOfferGroupVM()
   > n2 = Branch(Condition=Is Valid Index!(
     TargetArray:Array of Fort Item Shop Offer VM Object References=Get Offers!(self:Fort Item Shop Offer Group VM Object Reference=@n1) -> (Offers:Array of Fort Item Shop Offer VM Object References)
     IndexToTest:int=Get Index() -> (Index:int=0)
    ) -> (ReturnValue:bool))
   n3 = Get FortItemShopOfferGroupVM()
   > Return Node(
    OfferVM="Get (a copy)"(Array:Array of Fort Item Shop Offer VM Object References=Get Offers(self=@n3), Dimension 1=Get Index() -> (Index=0)) -> (Output:Fort Item Shop Offer VM Object Reference)
    Found=true
   )
   Return Node(execute=@n2.else, Found=false)
  ]}
 GetOfferIndex {nodes [
   GetOfferIndex!() -> (then, OfferVM:Object Reference) {is=title K2Node_FunctionEntry}
   > n1 = For Each Loop(Array:Array of Fort Item Shop Offer VM Object References=Get Offers(self=Get FortItemShopOfferGroupVM())) -> (Array Element:Fort Item Shop Offer VM Object Reference)
   Return Node!(execute=@n1.Completed, Array Index:int=-1)
   > Branch(Condition="Equal (Object)"(A=@n1, B=Get OfferVM() -> (OfferVM:Object Reference)))
   > Return Node(Array Index=@n1.Array Index)
  ]}
 GetOfferToPurchase {nodes [
   Inputs!()
   Outputs!(OfferToPurchase:Fort Item Shop Offer VM Object Reference=Get CurrentOfferVM!(self:WBP Item Shop PDP Purchase Block Object Reference=Get Component_PurchaseBlock()) -> (CurrentOfferVM:Fort Item Shop Offer VM Object Reference))
  ]}
 GetPreviewedItemDefinition {nodes [
   Inputs!() -> (execute:Exec)
   > n1 = Get Previewed Item VM()
   > n2 = Get!(execute, self:Fort Item VM Object Reference=@n1) -> (then, else, ItemDefinition:Fort Item Definition Object Reference)
   > Outputs!(Is Valid:Exec, Is Not Valid:Exec=[@n2.else, @n1.Is Not Valid], ItemDefinition:Fort Item Definition Object Reference=@n2)
  ]}
 GetPreviewedItemPreviewData {nodes [
   Inputs()
   > n1 = Get!(execute) -> (then, else, FortItemShopVM:Fort Item Shop VM Object Reference)
   > n2 = Get!(execute, self:Fort Item Shop VM Object Reference=@n1) -> (then, else, PreviewedItem:Item Shop Preview Data Object Reference)
   > Outputs!(Is Valid, Is Not Valid=[@n2.else, @n1.else], PreviewedItem:Item Shop Preview Data Object Reference=@n2)
  ]}
 GetPreviewedItemVM {nodes [
   Inputs()
   > n1 = Get Previewed Item Preview Data()
   > n2 = Get!(execute, self:Item Shop Preview Data Object Reference=@n1) -> (then, else, ItemVM:Fort Item VM Object Reference)
   > Outputs!(Is Valid, Is Not Valid=[@n2.else, @n1.Is Not Valid], ItemVM:Fort Item VM Object Reference=@n2)
  ]}
 GetPreviousOffer {nodes [
   GetPreviousOffer() {is=title K2Node_FunctionEntry}
   n1 = Get FortItemShopOfferVM()
   > n2 = GetOfferAtIndex(Index=int - int(A=GetOfferIndex(OfferVM=@n1) -> (Array Index=0) {is=title K2Node_CallFunction})) {is=title K2Node_CallFunction}
   > Return Node!(execute, OfferVM:Fort Item Shop Offer VM Object Reference=@n2, Found:bool=@n2.Found)
  ]}
 GetPrimaryPriceSwitcherVisibility {nodes [
   n1 = GetPrimaryPriceSwitcherVisibility!() -> (then, IsOfferOwned:bool=~, IsOfferGiftable:bool=~) {is=title K2Node_FunctionEntry}
   > Return Node!(
    execute
    Visibility:ESlateVisibility Enum=Select(
     Option 0=Collapsed
     Option 1=SelfHitTestInvisible
     Index=CanShowPrimaryPriceSwitcher!(self, IsOwned:bool=@n1, IsGiftable:bool=@n1.IsOfferGiftable) -> (CanShow:bool) {is=title K2Node_CallFunction}
    )
   )
  ]}
 GetPurchaseChoiceOfferInfo {nodes [
   GetPurchaseChoiceOfferInfo() {is=title K2Node_FunctionEntry}
   > Return Node!(execute, ReturnValue:Fort Store Front Offer Info Object Reference)
  ]}
 GetPurchaseChoiceOwningPlayer {nodes [
   GetPurchaseChoiceOwningPlayer() {is=title K2Node_FunctionEntry}
   > Return Node(ReturnValue:Fort Player Controller Object Reference)
  ]}
 GetSecondaryPriceSwitcherText {nodes [
   n1 = GetSecondaryPriceSwitcherText!() -> (then, IsOfferOwned:bool=~, SelectedItemVM:Fort Item VM Object Reference) {is=title K2Node_FunctionEntry}
   > n2 = Branch(Condition=@n1)
   > Return Node!(execute, Text:Text="LOCTEXT(\"OWNED\")")
   n3 = Get!(execute=@n2.else) -> (then, else, SelectedItemVM:Fort Item VM Object Reference)
   n4 = Get SelectedItemVM() -> (SelectedItemVM:Fort Item VM Object Reference)
   n5 = GetOwnedStacksNum(self=@n4) -> (ReturnValue=0)
   > Return Node(Text=SelectText(
     A=Format Text!(Format="LOCTEXT(\"OWNED: {numOwned}\")", numOwned:int=@n5) -> (Result)
     bPickA=IsItemStackable!(self, ItemVm:Fort Item VM Object Reference=@n4) -> (ReturnValue:bool)
    ))
   Return Node(execute=@n3.else)
  ]}
 GetSecondaryPriceSwitcherVisibility {nodes [
   n1 = GetSecondaryPriceSwitcherVisibility!() -> (
    then
    IsOfferOwned:bool=~
    IsFocusingItems:bool=~
    SelectedItemVM:Fort Item VM Object Reference
    GrantedItems:Array of Fort Item VM Object References
   ) {is=title K2Node_FunctionEntry}
   > n2 = Branch(Condition=@n1)
   > Return Node!(execute, Visibility:ESlateVisibility Enum=SelfHitTestInvisible)
   n3 = Branch(execute=@n2.else, Condition="Equal (Integer)"(A=Length(TargetArray:Array of Fort Item VM Object References=@n1.GrantedItems), B=1))
   > n4 = Set ItemToEvaluate(ItemToEvaluate:Fort Item VM Object Reference="Get (a copy)"(Array:Array of Fort Item VM Object References=Get GrantedItems!() -> (GrantedItems)) -> (Output:Fort Item VM Object Reference)) -> (Output_Get:Fort Item VM Object Reference)
   n5 = Get ItemToEvaluate() -> (ItemToEvaluate:Fort Item VM Object Reference)
   n6 = IsItemStackable(ItemVm=@n5)
   n7 = DoesOwnAnyStacks!(self, ItemVm:Fort Item VM Object Reference=@n5) -> (ReturnValue:bool) {is=title K2Node_CallFunction}
   Return Node(execute=[@n10.then, @n4.then], Visibility=Select(Option 0=Collapsed, Option 1=SelfHitTestInvisible, Index=AND Boolean(A=@n7, B=@n6)))
   n8 = Branch(execute=@n3.else, Condition=Get IsFocusingItems())
   > n9 = Get()
   > n10 = Set ItemToEvaluate(ItemToEvaluate=@n9)
   Return Node(execute=[@n9.else, @n8.else], Visibility=Collapsed)
  ]}
 IncreaseVariantIndex {nodes [
   IncreaseVariantIndex() {is=title K2Node_FunctionEntry}
   n1 = Get CurrentVariant() -> (CurrentVariant:int=0)
   n2 = Get TotalVariants() -> (TotalVariants:int=0)
   > Set CurrentVariant(CurrentVariant:int="Wrap (Integer)"!(Value:int=int + int(A=@n1), Min:int=-1, Max:int=int - int(A=@n2)) -> (ReturnValue:int=0)) -> (Output_Get:int=0)
  ]}
 InitializePDP {nodes [
   n1 = InitializePDP!() -> (then, OfferVM:Fort Item Shop Offer VM Object Reference, OfferGroupVM:Fort Item Shop Offer Group VM Object Reference) {is=title K2Node_FunctionEntry}
   > SetFortItemShopOfferVM(Viewmodel=@n1)
   > SetFortItemShopOfferGroupVM!(
    execute
    self
    Viewmodel:Fort Item Shop Offer Group VM Object Reference=Get OfferGroupVM() -> (OfferGroupVM:Fort Item Shop Offer Group VM Object Reference)
   ) -> (then)
  ]}
 IsItemStackable {nodes [
   n1 = IsItemStackable!() -> (then, ItemVm:Fort Item VM Object Reference) {is=title K2Node_FunctionEntry}
   > n2 = Is Valid(InputObject=@n1)
   Return Node!(execute=@n2.Is Not Valid, ReturnValue:bool=false)
   n3 = Get ItemDefinition(self=@n1)
   > Return Node(ReturnValue="integer > integer"(A=GetMaxNumStacks!(self:Fort Item Definition Object Reference=@n3) -> (ReturnValue:int=0), B=1))
  ]}
 IsPurchaseChoiceRequestedEventBound {nodes [
   IsPurchaseChoiceRequestedEventBound() {is=title K2Node_FunctionEntry}
   > Return Node(ReturnValue=false)
  ]}
 IsShoesItem {nodes [
   n1 = IsShoesItem!() -> (then, Item:Fort Item VM Object Reference) {is=title K2Node_FunctionEntry}
   > Return Node(ReturnValue=Contains(SearchIn=GetItemTemplateId(self=@n1), Substring=CosmeticShoes))
  ]}
 IsSingleItemItemOfffer {nodes [
   IsSingleItemItemOfffer() {is=title K2Node_FunctionEntry}
   n1 = Get GrantedItems!(self:Fort Item Shop Offer VM Object Reference=Get FortItemShopOfferVM()) -> (GrantedItems)
   > Return Node(ReturnValue="Equal (Integer)"(A=Length(TargetArray=@n1), B=1))
  ]}
 IsSingleItemStackableOffer {nodes [
   IsSingleItemStackableOffer() {is=title K2Node_FunctionEntry}
   > n1 = Get!(execute) -> (then, else, FortItemShopOfferVM:Fort Item Shop Offer VM Object Reference)
   > n2 = Branch(Condition=IsSingleItemItemOfffer!(self) -> (ReturnValue:bool) {is=title K2Node_CallFunction})
   n3 = Get GrantedItems(self=Get FortItemShopOfferVM())
   > Return Node!(execute, IsSingleItemStackableOffer:bool=IsItemStackable(ItemVm="Get (a copy)"(Array=@n3)) {is=title K2Node_CallFunction})
   Return Node(execute=[@n2.else, @n1.else], IsSingleItemStackableOffer=false)
  ]}
 NavigateToALCList {nodes [
   NavigateToALCList!() -> (then, Navigation:EUINavigation Enum) {is=title K2Node_FunctionEntry}
   > n1 = GetDesiredFocusTarget!(execute:Exec, self:WBP Item Shop PDP Item List ALC Object Reference=Get Component_ItemList_ALC()) -> (then:Exec, FocusTarget:Widget Object Reference)
   > Return Node!(execute, ReturnValue:Widget Object Reference=@n1)
  ]}
 NavigateToPurchaseBlock {nodes [
   NavigateToPurchaseBlock!() -> (then, Navigation:EUINavigation Enum) {is=title K2Node_FunctionEntry}
   > n1 = GetDesiredFocusTarget(self:WBP Item Shop PDP Purchase Block Object Reference=Get Component_PurchaseBlock())
   > Return Node(ReturnValue=@n1)
  ]}
 OnCurrentVariantOverrides {nodes [
   n1 = OnCurrentVariantOverrides!() -> (then, CurrentVariantOverrides:Array of Mcp Variant Channel Info Structures) {is=title K2Node_FunctionEntry}
   > Branch(Condition="Equal (Enum)"(A:EFortItemShopPdpView Enum=Get CurrentView() -> (CurrentView=OfferInspect), B:EFortItemShopPdpView Enum=ItemDetails))
   > SetVariantOverrides!(execute, self:Fort MPItem Shop Offer Inspection Screen Object Reference, InVariantOverrides:Array of Mcp Variant Channel Info Structures=@n1) -> (then)
   > UpdateEquipButtonVisibility()
  ]}
 OnItemsUpdated {nodes [
   OnItemsUpdated!() -> (then, Items:Array of Fort Item VM Object References) {is=title K2Node_FunctionEntry}
   > n1 = GetIsOfferRestricted!(execute, self, OfferVM:Fort Item Shop Offer VM Object Reference=Get FortItemShopOfferVM()) -> (then, OfferContainsRestriction:bool, OfferRequiresPin:bool) {is=title K2Node_CallFunction}
   > Set IsOfferRestricted(IsOfferRestricted:bool=@n1) -> (Output_Get:bool)
   > Set OfferRequiresPin(OfferRequiresPin:bool=@n1.OfferRequiresPin) -> (Output_Get:bool)
   > SetTouchMimosaLayoutOverride()
   > Set Visibility(
    Visibility=Select(Option 0=Collapsed, Option 1=SelfHitTestInvisible, Index=Get IsOfferRestricted() -> (IsOfferRestricted:bool))
    self=Get RestrictionWarning() -> (RestrictionWarning:Horizontal Box Object Reference)
   )
  ]}
 OnPreviewedItemChanged {nodes [
   n1 = OnPreviewedItemChanged!() -> (then, Previewed Item:Item Shop Preview Data Object Reference) {is=title K2Node_FunctionEntry}
   > RequestItemPreview!(
    execute
    self:Fort MPItem Shop Offer Inspection Screen Object Reference
    InPreviewedItem:Item Shop Preview Data Object Reference=@n1
    bInKeepInputRotation:bool="Equal (Enum)"(A=Get CurrentView() -> (CurrentView=OfferInspect), B=ItemDetails)
   ) -> (then)
   > UpdateEquipButtonVisibility()
   Get!(execute) -> (then, else, Previewed Item:Item Shop Preview Data Object Reference)
   > UpdateCustomSticker(SelectedItem=Get ItemVM(self=Get Previewed Item() -> (Previewed Item:Item Shop Preview Data Object Reference)))
   > UpdateAboutItemButton()
  ]}
 PaginationButtons_UpdateOutline {nodes [
   PaginationButtons_UpdateOutline() {is=title K2Node_FunctionEntry}
   > n1 = For Each Loop(Array:Array of WBP UIKit Button Generic Object References=Make Array!(
     "[0]":WBP UIKit Button Generic Object Reference=Get Button_NextOffer()
     "[1]":WBP UIKit Button Generic Object Reference=Get Button_PreviousOffer()
    ) -> (Array:Array of WBP UIKit Button Generic Object References)) -> (Array Element:WBP UIKit Button Generic Object Reference)
   > n2 = Cast To WBP_UIKit_Block_BackgroundWithOutline!(
    execute
    Object=Get Block_Background!(self:WBP UIKit Button Generic Base Object Reference=@n1) -> (Block_Background:WBP UIKit Block Background Base Object Reference)
   ) -> (then, CastFailed, AsWBP UIKit Block Background with Outline:WBP UIKit Block Background With Outline Object Reference) {is=title K2Node_DynamicCast}
   > SetCornerRadiuses!(
    execute
    self:WBP UIKit Block Background With Outline Object Reference=@n2
    TopLeft:float=0.325
    TopRight:float=0.325
    BottomLeft:float=0.325
    BottomRight:float=0.325
    RadiusInPixels:bool=false
   ) -> (then)
  ]}
 PreviewItemAtIndex {nodes [
   n1 = PreviewItemAtIndex!() -> (then, Index:int) {is=title K2Node_FunctionEntry}
   n2 = Get GrantedItems(self=Get FortItemShopOfferVM())
   > n3 = Branch(Condition=Is Valid Index(TargetArray:Array of Fort Item VM Object References=@n2, IndexToTest=@n1))
   > PreviewItemFromVM!(execute, self, ItemVM:Fort Item VM Object Reference="Get (a copy)"(Array=@n2, Dimension 1=@n1)) -> (then)
   PreviewItemFromVM(execute=@n3.else)
  ]}
 PreviewItemFromVM {nodes [
   n1 = PreviewItemFromVM!() -> (then, ItemVM:Fort Item VM Object Reference) {is=title K2Node_FunctionEntry}
   > n2 = Is Valid(InputObject=@n1)
   > n3 = Set RequestedItem(RequestedItem:Fort Item VM Object Reference=@n1) -> (Output_Get:Fort Item VM Object Reference)
   n4 = Get PreviewedItem(self=Get FortItemShopVM())
   n5 = IsValid(Object=@n4)
   n6 = Get ItemVM(self=@n4)
   > Branch(Condition="Not Equal (Object)"(
     A:Object Reference=@n3
     B:Object Reference=Select(Option 0:Object Reference, Option 1:Object Reference=@n6, Index=@n5) -> (ReturnValue:Object Reference)
    ) -> (ReturnValue:bool=~) {is=title K2Node_PromotableOperator})
   > n7 = CreatePreviewDataFromItemVM(ItemVM=Get RequestedItem() -> (RequestedItem:Fort Item VM Object Reference))
   n8 = Get FortItemShopVM()
   > Set PreviewedItem(PreviewedItem=@n7, self=@n8)
   > Set FortItemVM!(execute, FortItemVM:Fort Item VM Object Reference, self:WBP UIKit Variant Selector Object Reference=Get WBP_UIKit_VariantSelector()) -> (then, Output_Get:Fort Item VM Object Reference)
   > Set FortItemVM(FortItemVM=Get RequestedItem(), self=Get WBP_UIKit_VariantSelector())
   > UpdateCustomSticker(SelectedItem=Get RequestedItem())
   > n9 = Branch(Condition="Equal (Enum)"(A=Get CurrentView() -> (CurrentView=OfferInspect), B=OfferInspect))
   > PreviewVariants()
   > UpdateAboutItemButton()
   ResetVariants(execute=@n9.else)
   Set PreviewedItem(execute=@n2.Is Not Valid, self=@n8)
  ]}
 PreviewVariants {nodes [
   PreviewVariants() {is=title K2Node_FunctionEntry}
   > ResetVariants()
   > Branch(Condition=Get HasVariants())
   n1 = Get PreviewedItem(self=Get FortItemShopVM())
   > n2 = Branch(Condition=NOT Boolean(A=HasLoadoutOverride(self=@n1)))
   > n3 = Cast To AthenaCosmeticItemDefinition(Object=Get ItemDefinition(self=Get ItemVM(self=@n1)))
   > n4 = Set VariantPreviews(VariantPreviews:Array of Fort Cosmetic Variant Preview Structures=GetOrGenerateItemVariantPreviews!(self:Athena Cosmetic Item Definition Object Reference=@n3, PlayerController:Fort Player Controller Object Reference) -> (ReturnValue:Array of Fort Cosmetic Variant Preview Structures)) -> (Output_Get:Array of Fort Cosmetic Variant Preview Structures)
   > Insert!(
    execute
    TargetArray:Array of Fort Cosmetic Variant Preview Structures=@n4
    NewItem:"Fort Cosmetic Variant Preview Structure (by ref)"
    Index:int=0
   ) -> (then) {is=title K2Node_CallArrayFunction}
   > n5 = Set TotalVariants(TotalVariants:int=Length(TargetArray:Array of Fort Cosmetic Variant Preview Structures=Get VariantPreviews() -> (VariantPreviews:Array of Fort Cosmetic Variant Preview Structures))) -> (Output_Get:int=0)
   > n6 = Branch(Condition="integer > integer"(A=@n5, B=1))
   > Set HasVariantsToCycle(HasVariantsToCycle:bool=true) -> (Output_Get:bool)
   > ShowVariantPreview()
   Set HasVariantsToCycle(execute=[@n6.else, @n3.CastFailed, @n2.else], HasVariantsToCycle=false)
  ]}
 RefreshContextMenuPreviewActions {nodes [
   RefreshContextMenuPreviewActions!() -> (then, PreviewActions:Array of Data Table Row Handle Structures) {is=title K2Node_FunctionEntry}
   > Branch(Condition=ShouldEnablePreviewActionsContextButton!() -> (ReturnValue:bool))
   > n1 = Sequence!(execute) -> (then_0, then_1, then_2:Exec)
   > n2 = Is Valid(InputObject=Get InteractionContext!(self:WBP UIKit Context Menu Anchor Button Base Object Reference=Get ContextMenuAnchorButton_Options()) -> (InteractionContext:UIKit Context Menu Action Context Object Reference))
   n3 = Construct Object from Class(execute=@n2.Is Not Valid, Class=/Script/FortniteUI.UIKitContextMenuActionContext) -> (ReturnValue:UIKit Context Menu Action Context Object Reference)
   > n4 = Set InteractionContext!(
    execute
    InteractionContext:UIKit Context Menu Action Context Object Reference=@n3
    self:WBP UIKit Context Menu Anchor Button Base Object Reference=Get ContextMenuAnchorButton_Options()
   ) -> (then, Output_Get:UIKit Context Menu Action Context Object Reference)
   RemoveAllInteractions(
    execute=[@n2.Is Valid, @n4.then]
    self:UIKit Context Menu Action Context Object Reference=Get InteractionContext(self=Get ContextMenuAnchorButton_Options())
   )
   n5 = Length(TargetArray:Array of Data Table Row Handle Structures=Get PreviewActions() -> (PreviewActions:Array of Data Table Row Handle Structures))
   SetVisibility(
    execute=@n1.then_2
    self=Get Grid_Panel_Options_Button() -> (Grid_Panel_Options_Button:Grid Panel Object Reference)
    InVisibility=Select(Option 0:ESlateVisibility Enum=Collapsed, Option 1:ESlateVisibility Enum=Visible, Index="integer > integer"(A=@n5)) -> (ReturnValue:ESlateVisibility Enum)
   )
   n6 = For Each Loop(Exec=@n1.then_1, Array:Array of Data Table Row Handle Structures=Get PreviewActions()) -> (Array Element:Data Table Row Handle Structure)
   n7 = Get ContextMenuAnchorButton_Options()
   SetContext!(
    execute=@n6.Completed
    self:WBP UIKit Context Menu Anchor Button Base Object Reference=@n7
    InteractionContext_Data:UIKit Context Menu Action Context Object Reference=Get InteractionContext(self=@n7)
   ) -> (then)
   n8 = Break Data Table Row Handle!(DataTableRowHandle:"Data Table Row Handle Structure (by ref)"=@n6) -> (DataTable:Data Table Object Reference, RowName:Name) {is=title K2Node_BreakStruct}
   n9 = Get Data Table Row!(execute:Exec=@n6.LoopBody, DataTable:Data Table Object Reference=@n8, RowName:Name=@n8.RowName) -> (then:Exec, RowNotFound:Exec, ReturnValue:Common Input Action Data Base Structure)
   > AddInteraction!(
    execute
    self:UIKit Context Menu Action Context Object Reference=Get InteractionContext(self=Get ContextMenuAnchorButton_Options())
    InActionInfo_DisplayName:Text=Break Common Input Action Data Base!(CommonInputActionDataBase:"Common Input Action Data Base Structure (by ref)"=@n9) -> (DisplayName:Text, HoldDisplayName:Text, NavBarPriority:int) {is=title K2Node_BreakStruct}
    InActionInfo_IconTexture:Texture 2D Soft Object Reference
    InActionInfo_IconMaterial:Material Instance Soft Object Reference
    InActionInfo_SortPriority:int=1000
    InActionInfo_bIsAvailable:bool=True
    InActionInfo_bIsEnabled:bool=True
    InActionInfo_bIsSelected:bool=False
    InActionInfo_bAddDivider:bool=False
    InActionInfo_bPositionDividerAfter:bool=False
    InActionInfo_bCloseMenuOnSelect:bool=False
    InActionInfo_InteractionNameID:Name=@n8.RowName
    InActionInfo_Interaction_DataTable:Data Table Object Reference=@n8
    InActionInfo_Interaction_RowName:Name=@n8.RowName
    InActionInfo_PressedDelegate:Delegate=Create Event()
    InActionInfo_ReleasedDelegate:Delegate=Create Event()
    InInteractionDelegate:"Delegate (by ref)"=Create Event()
   ) -> (then)
   > ClearPreviewActionButtons() {is=title K2Node_CallFunction}
  ]}
 ResetVariants {nodes [
   ResetVariants() {is=title K2Node_FunctionEntry}
   > Clear and Invalidate Timer by Handle(Handle=Get VariantTimerHandle() -> (VariantTimerHandle:Timer Handle Structure))
   > Clear(TargetArray:Array of Fort Cosmetic Variant Preview Structures=Get VariantPreviews())
   > Set CurrentVariant(CurrentVariant=0) -> (Output_Get=0)
   > Set TotalVariants(TotalVariants=0) -> (Output_Get=0)
   > Set HasVariantsToCycle(HasVariantsToCycle=false)
   > UpdateVariantInfo()
   > Set Visibility(Visibility=Collapsed, self=Get InfoBox_VariantInfo())
  ]}
 SelectItem {nodes [
   n1 = SelectItem!() -> (then, ItemVm:Fort Item VM Object Reference) {is=title K2Node_FunctionEntry}
   > PreviewItemFromVM(ItemVM=@n1) {is=title K2Node_CallFunction}
   > n2 = Branch(Condition=Get IsFocusingItems())
   > n3 = Branch(Condition=IsUsingTouch(PlayerController=GetOwningPlayer()))
   ChangeFocusArea(execute=[@n3.then, @n2.else], IsFocusingItem=true)
  ]}
 SetAllInteractionsEnabled {nodes [
   n1 = SetAllInteractionsEnabled!() -> (then, ShouldEnable:bool=~) {is=title K2Node_FunctionEntry}
   > SetIsInteractionEnabled(
    self=[
     Get Button_Equip()
     Get Button_Back() -> (Button_Back:WBP UIKit Button Quiet Object Reference)
     Get Button_NextOffer()
     Get Button_PreviousOffer()
    ]
    bInIsInteractionEnabled=@n1
   )
   > OnInteractionEnabledChanged(self:Self Object Reference, ShouldEnable=@n1)
  ]}
 SetCompactStyle {nodes [
   SetCompactStyle() {is=title K2Node_FunctionEntry}
   > Branch(Condition=OR Boolean!(A=IsUsingCompactUI(), B=GetSmallTextScalingMultiplier!() -> (ScalingMultiplier:float, IsEnabledOnDevice:bool).IsEnabledOnDevice) -> (ReturnValue))
   > SetPadding!(
    execute
    self:Overlay Slot Object Reference=SlotAsOverlaySlot(Widget=Get VerticalBox_LeftSideInfo() -> (VerticalBox_LeftSideInfo:Vertical Box Object Reference))
    InPadding_Left:float=48
    InPadding_Top:float=24
    InPadding_Right:float=48
    InPadding_Bottom:float=24
   ) -> (then)
   > SetPadding(
    self=SlotAsOverlaySlot(Widget=Get WBP_ActionBarContainer() -> (WBP_ActionBarContainer:WBP Action Bar Container Object Reference))
    InPadding_Left=0
    InPadding_Top=0
    InPadding_Right=48
    InPadding_Bottom=24
   )
   > SetPadding(
    self=SlotAsOverlaySlot(Widget=Get WBP_PlayerInfo() -> (WBP_PlayerInfo:WBP Player Info Object Reference))
    InPadding_Left=0
    InPadding_Top=24
    InPadding_Right=48
    InPadding_Bottom=0
   )
   > SetPadding(
    self=SlotAsOverlaySlot(Widget=Get HorizontalBox_NavigationArrows())
    InPadding_Left=0
    InPadding_Top=0
    InPadding_Right=0
    InPadding_Bottom=132
   )
   > n1 = Sequence!(execute) -> (then_0, then_1)
   > n2 = Branch(Condition=IsTabletAspectRatio!() -> (ReturnValue:bool))
   > SetPadding(
    self=SlotAsOverlaySlot(Widget=Get EquippedText() -> (EquippedText:Common Text Object Reference))
    InPadding_Left=0
    InPadding_Top=0
    InPadding_Right=0
    InPadding_Bottom=350
   )
   SetMinDesiredHeight!(
    execute=@n2.else
    self:Size Box Object Reference=Get SizeBox_Details() -> (SizeBox_Details:Size Box Object Reference)
    InMinDesiredHeight:float=205
   ) -> (then)
   > SetContent(self=Get IsbSlot_Mobile() -> (IsbSlot_Mobile:Named Slot Object Reference), Content=Get Component_IsbManager())
   > Set IsIsbOnRightSide(IsIsbOnRightSide:bool=true) -> (Output_Get:bool)
   SetSidesToPad!(
    execute=@n1.then_1
    self:Safe Zone Object Reference=Get SafeZone_All() -> (SafeZone_All:Safe Zone Object Reference)
    InPadLeft:bool=true
    InPadRight:bool=true
    InPadTop:bool=true
    InPadBottom:bool=true
   ) -> (then)
  ]}
 SetLoadingOverlayVisibility {nodes [
   n1 = SetLoadingOverlayVisibility!() -> (then, IsVisible:bool=~) {is=title K2Node_FunctionEntry}
   n2 = Get Loading_Overlay() -> (Loading_Overlay:Overlay Object Reference)
   n3 = Select(Option 0=Collapsed, Option 1=Visible, Index=@n1)
   > Branch(Condition="Not Equal (Enum)"(
     A:ESlateVisibility Enum=@n3
     B:ESlateVisibility Enum=GetVisibility!(self:Widget Object Reference=@n2) -> (ReturnValue:ESlateVisibility Enum=Visible)
    ))
   > SetVisibility(self=Get Loading_Overlay(), InVisibility=@n3)
   > n4 = Branch(Condition=Get IsVisible() -> (IsVisible:bool))
   n5 = Get FortItemPreviewWidget()
   > n6 = Set CurrentItemViewSettings(CurrentItemViewSettings:Fort Item View Settings Structure=GetSettings!(self:Fort Item View Object Reference=Get ItemViewObject!(self:Fort Item Preview Screen Object Reference=@n5) -> (ItemViewObject:Fort Item View Object Reference)) -> (ReturnValue:Fort Item View Settings Structure)) -> (Output_Get:Fort Item View Settings Structure)
   n7 = Break Fort Item View Settings!(FortItemViewSettings:"Fort Item View Settings Structure (by ref)"=Get CurrentItemViewSettings() -> (CurrentItemViewSettings:Fort Item View Settings Structure)) -> (
    UsesPlacementActor:bool=~
    UsesFixedCamera:bool=~
    SupportsZooming:bool=~
    RestrictZoomToWidget:bool=~
    DefaultZoomLevel:float=~
    ZoomRange:Float Range Structure
    RotationMode:EFortItemViewRotationMode Enum
    CameraRotationOffset:Rotator=~
    MeshBoundsCenterOffset:Vector=~
   ) {is=title K2Node_BreakStruct}
   SetSettings!(
    execute=[@n6.then, @n4.else]
    self:Fort Item View Object Reference=Get ItemViewObject(self=Get FortItemPreviewWidget())
    Settings:"Fort Item View Settings Structure (by ref)"=Select(
     Option 0:"Fort Item View Settings Structure (by ref)"=Get CurrentItemViewSettings()
     Option 1:"Fort Item View Settings Structure (by ref)"=Make Fort Item View Settings!(
      UsesPlacementActor:bool=@n7
      UsesFixedCamera:bool=@n7.UsesFixedCamera
      SupportsZooming:bool=False
      RestrictZoomToWidget:bool=@n7.RestrictZoomToWidget
      DefaultZoomLevel:float=@n7.DefaultZoomLevel
      ZoomRange:Float Range Structure=@n7.ZoomRange
      RotationMode:EFortItemViewRotationMode Enum=None
      CameraRotationOffset:Rotator=@n7.CameraRotationOffset
      MeshBoundsCenterOffset:Vector=@n7.MeshBoundsCenterOffset
     ) -> (FortItemViewSettings:Fort Item View Settings Structure) {is=title K2Node_MakeStruct}
     Index=Get IsVisible()
    ) -> (ReturnValue:"Fort Item View Settings Structure (by ref)")
   ) -> (then)
  ]}
 SetRTLAlignment {nodes [
   SetRTLAlignment() {is=title K2Node_FunctionEntry}
   > Branch(Condition=IsCultureRightToLeft!(Culture:String=GetCurrentCulture!() -> (ReturnValue:String)) -> (ReturnValue:bool))
   > SetWidthOverride(self=Get SizeBox_Details(), InWidthOverride=640)
  ]}
 SetTextChatEnabled {nodes [
   n1 = SetTextChatEnabled!() -> (then, bSetActive:bool=~) {is=title K2Node_FunctionEntry}
   > Broadcast Event!(
    execute
    self:Gameplay Event Router Component Object Reference=GetGlobalEventRouter()
    EventData:BPS Set Text Chat Active State Structure=Make BPS Set Text Chat Active State!(bSetActive_5_#8:bool=@n1) -> (BPS_SetTextChatActiveState:BPS Set Text Chat Active State Structure) {is=title K2Node_MakeStruct}
    EventContext:Object Reference=GetOwningPlayer()
   ) -> (then, ReturnValue:bool)
  ]}
 SetTouchMimosaLayoutOverride {nodes [
   SetTouchMimosaLayoutOverride() {is=title K2Node_FunctionEntry}
   > Branch(Condition=Get IsIsbOnRightSide())
   n1 = "Get (a copy)"(Array=Get GrantedItems(self=Get FortItemShopOfferVM()))
   n2 = "Equal (Enum)"(
    A:EFortItemType Enum=GetItemType(ItemDefinition=Get ItemDefinition(self=@n1)) -> (ReturnValue=WorldItem)
    B:EFortItemType Enum=CosmeticMimosa
   )
   > SetTextJustification(
    self=Get Component_IsbManager()
    TextJustification=Select(Option 0:ETextJustify Enum=Left, Option 1:ETextJustify Enum=InvariantRight, Index=@n2) -> (ReturnValue:ETextJustify Enum)
   )
   > Set!(
    execute
    Nudge_X:double=Select(Option 0:double=0, Option 1:double=-300, Index=@n2) -> (ReturnValue:double)
    Nudge_Y:double=0
    self:Grid Slot Object Reference=SlotAsGridSlot(Widget=Get SizeBox_CameraFrame())
   ) -> (then, Output_Get:Vector 2D Structure)
  ]}
 SetupComponents {nodes [
   SetupComponents() {is=title K2Node_FunctionEntry}
   > Set Components(Components:Array of BPI Item Shop PDP Interfaces=Make Array!(
     "[0]":BPI Item Shop PDP Interface=Get WBP_ItemShopPDP_Description() -> (WBP_ItemShopPDP_Description:WBP Item Shop PDP Description Object Reference)
     "[1]":BPI Item Shop PDP Interface=Get Component_IsbManager()
     "[2]":BPI Item Shop PDP Interface=Get Component_PurchaseBlock()
     "[3]":BPI Item Shop PDP Interface=Get Component_ItemList()
     "[4]":BPI Item Shop PDP Interface=Get Component_ItemList_ALC()
    ) -> (Array:Array of BPI Item Shop PDP Interfaces)) -> (Output_Get:Array of BPI Item Shop PDP Interfaces)
  ]}
 ShouldShowCosmeticGateModal {nodes [
   Inputs!()
   n1 = Get Offer to Purchase()
   n2 = Get FortItemShopOfferVM()
   Outputs!(ShouldShow:bool=Select(
     Option 0:bool=Get IsALCOfferRestricted() -> (IsALCOfferRestricted:bool)
     Option 1:bool=Get IsOfferRestricted()
     Index="Equal (Object)"(A=@n1, B=@n2)
    ) -> (ReturnValue:bool))
  ]}
 ShouldShowGiftBoxScreenForRMTs {nodes [
   ShouldShowGiftBoxScreenForRMTs() {is=title K2Node_FunctionEntry}
   n1 = Get FortGlobalUIContext() -> (ReturnValue:Fort Global UIContext Object Reference) {is=title K2Node_GetSubsystem}
   > Return Node(ReturnValue:bool=OR Boolean(
     A=IsInSubGame!(SubGame:ESubGame Enum=Campaign) -> (ReturnValue:bool)
     B=IsInZone!(self:Fort Global UIContext Object Reference=@n1) -> (ReturnValue:bool)
    ))
  ]}
 ShouldShowOfferPurchasedOverlay {nodes [
   ShouldShowOfferPurchasedOverlay() {is=title K2Node_FunctionEntry}
   n1 = Get bUseItemPresentationScreenOnItemPurchased!(self:Fort Runtime Options Object Reference=GetRuntimeOptions()) -> (bUseItemPresentationScreenOnItemPurchased:bool)
   n2 = Get OfferInfo(self=Get FortItemShopOfferVM())
   n3 = IsBattlePassLevelBundle!(self:Fort Store Front Offer Info Object Reference=@n2) -> (ReturnValue:bool)
   n4 = GetPriceCurrencyType!(self:Fort Store Front Offer Info Object Reference=@n2, PriceIndex:int=0) -> (ReturnValue:EStoreCurrencyType Enum=RealMoney)
   > Return Node(ReturnValue=AND Boolean!(A="Equal (Enum)"(A:EStoreCurrencyType Enum=@n4, B:EStoreCurrencyType Enum=MtxCurrency), B=NOT Boolean(A=@n3), C:bool=NOT Boolean(A=@n1)) -> (ReturnValue))
  ]}
 ShouldShowPinModal {nodes [
   Inputs()
   n1 = Get Offer to Purchase()
   n2 = Get FortItemShopOfferVM()
   Outputs(ShouldShow=Select(
     Option 0=Get ALCOfferRequiresPin() -> (ALCOfferRequiresPin:bool)
     Option 1=Get OfferRequiresPin() -> (OfferRequiresPin:bool)
     Index="Equal (Object)"(A=@n1, B=@n2)
    ))
  ]}
 ShowVariantPreview {nodes [
   ShowVariantPreview() {is=title K2Node_FunctionEntry}
   > Clear and Invalidate Timer by Handle(Handle=Get VariantTimerHandle())
   > SetVariantOverrides(InVariantOverrides="Get (a ref)"!(Array:Array of Fort Cosmetic Variant Preview Structures=Get VariantPreviews(), Dimension 1=Get CurrentVariant() -> (CurrentVariant=0)) -> (
     Output_UnlockCondition:Text
     Output_PreviewTime:float=~
     Output_VariantOptions:Array of Mcp Variant Channel Info Structures
     Output_AdditionalItems:Array of Fort Cosmetic Variant Preview Element Structures
    ).Output_VariantOptions)
   n1 = Get TotalVariants() -> (TotalVariants=0)
   n2 = "Get (a ref)"(
    Array=Get VariantPreviews()
    Dimension 1="Clamp (Integer)"(Value=Get CurrentVariant() -> (CurrentVariant=0), Min=1, Max=int - int(A=@n1)) -> (ReturnValue=0)
   ) -> (Output_PreviewTime=~)
   > n3 = Set Timer by Event(
    Delegate=Create Event()
    Time=SelectFloat(
     A=@n2.Output_PreviewTime
     B=Get DefaultTimeBetweenVariants() -> (DefaultTimeBetweenVariants:float)
     bPickA="float > float"(A=@n2.Output_PreviewTime)
    )
    bLooping=false
   )
   > Set VariantTimerHandle(VariantTimerHandle:Timer Handle Structure=@n3) -> (Output_Get:Timer Handle Structure)
   > UpdateVariantInfo()
  ]}
 TryRequestOverridePreview {nodes [
   TryRequestOverridePreview() {is=title K2Node_FunctionEntry}
   n1 = Get OverridePreviews!(self:Fort Item Shop Offer VM Object Reference=Get FortItemShopOfferVM()) -> (OverridePreviews:Array of Item Shop Preview Data Object References)
   > n2 = Branch(Condition=Is Empty(TargetArray:Array of Item Shop Preview Data Object References=@n1))
   > Return Node!(execute, Success:bool=false)
   ResetVariants(execute=@n2.else) {is=title K2Node_CallFunction}
   > Set PreviewedItem(
    PreviewedItem="Get (a copy)"(Array:Array of Item Shop Preview Data Object References=@n1) -> (Output:Item Shop Preview Data Object Reference)
    self=Get FortItemShopVM()
   )
   > Return Node(Success=true)
  ]}
 UpdateAboutItemButton {nodes [
   UpdateAboutItemButton() {is=title K2Node_FunctionEntry}
   > n1 = Get Previewed Item VM()
   > n2 = Set ItemVM(ItemVM:Fort Item VM Object Reference=@n1) -> (Output_Get:Fort Item VM Object Reference)
   > Set Category(Category:Fort Locker Category Config Object Reference=GetItemCategory(self=@n2)) -> (Output_Get:Fort Locker Category Config Object Reference)
   > n3 = Branch(Condition=TryGetSoftCustomActivatableWidgetClass(self=Get Category() -> (Category:Fort Locker Category Config Object Reference), ViewRequest=(TagName="UI.ItemShop.AboutItemDialogClass")).ReturnValue)
   > Set Visibility(Visibility=SelfHitTestInvisible, self=Get Overlay_AboutItem() -> (Overlay_AboutItem:Overlay Object Reference))
   n4 = GetItemType(ItemDefinition=Get ItemDefinition(self=Get ItemVM!() -> (ItemVM))) -> (ReturnValue=WorldItem)
   > n5 = Branch(Condition=OR Boolean(A="Equal (Enum)"(A:EFortItemType Enum=@n4, B:EFortItemType Enum=JunoBuildingSet), B="Equal (Enum)"(A=@n4, B=JunoBuildingProp)))
   > n6 = Set Text(
    Text="LOCTEXT(\"ABOUT LEGO® KITS\")"
    self:Fort CTAButton Object Reference=Get Button_AboutItem() -> (Button_AboutItem:WBP UIKit Button Quiet Object Reference)
   )
   n7 = TryGetSoftCustomActivatableWidgetClass(self=Get Category(), ViewRequest=(TagName="UI.ItemShop.AboutItemCoachMark"))
   n8 = Get ItemVM()
   n9 = GetHasSeenAboutItemModal!(
    self:Fort Game User Settings Object Reference=GetFortGameUserSettings()
    ItemType:EFortItemType Enum=GetItemType(ItemDefinition=Get ItemDefinition(self=@n8)) -> (ReturnValue=WorldItem)
   ) -> (ReturnValue:bool)
   n10 = Branch(execute=[@n14.then, @n6.then], Condition=AND Boolean!(A=NOT Boolean(A=@n9), B=@n7.ReturnValue) -> (ReturnValue))
   > ClearChildren(self=Get VisualAttachment_AboutItemCoachmark())
   > n11 = LoadClassAsset_Blocking(AssetClass=@n7)
   > n12 = Cast To CommonActivatableWidget Class!(execute:Exec, Class:Object Class Reference=@n11) -> (then:Exec, CastFailed:Exec, AsCommon Activatable Widget:Common Activatable Widget Class Reference) {is=title K2Node_ClassDynamicCast}
   > CreateAboutItemCoachmark!(execute, self, Class:Common Activatable Widget Class Reference=@n12) -> (then) {is=title K2Node_CallFunction}
   ClearChildren(execute=[@n10.else, @n15.then], self=Get VisualAttachment_AboutItemCoachmark())
   n13 = Get SlotInfo!(self:Fort Locker Category Config Object Reference=Get Category()) -> (SlotInfo:Fort Locker Category Slot Info Structure)
   n14 = Set Text(
    execute=@n5.else
    Text=Format Text!(
     Format="LOCTEXT(\"About {ItemCategoryPluralName}\")"
     ItemCategoryPluralName:Text=Break Fort Locker Category Slot Info!(FortLockerCategorySlotInfo:"Fort Locker Category Slot Info Structure (by ref)"=@n13) -> (CategoryNamePlural:Text, CategorySoundOverrides:Map of Gameplay Tag Structures to Sound Base Soft Object References) {is=title K2Node_BreakStruct}
    ) -> (Result)
    self=Get Button_AboutItem()
   )
   n15 = Set Visibility(execute=[@n3.else, @n1.Is Not Valid], Visibility=Collapsed, self=Get Overlay_AboutItem())
  ]}
 UpdateButtons {nodes [
   UpdateButtons() {is=title K2Node_FunctionEntry}
   n1 = Get FortItemShopOfferVM()
   > Set bTogglePreviewActionsEnabled!(
    execute
    bTogglePreviewActionsEnabled:bool=NOT Boolean(A=Get bOwned!(self:Fort Item Shop Offer VM Object Reference=@n1) -> (bOwned:bool))
    self:Fort Item Preview Screen Object Reference=Get FortItemPreviewWidget()
   ) -> (then, Output_Get:bool)
   > UpdateEquipButtonVisibility()
   > SetAllInteractionsEnabled(ShouldEnable=true) {is=title K2Node_CallFunction}
  ]}
 UpdateCustomSticker {nodes [
   UpdateCustomSticker!() -> (then, SelectedItem:Fort Item VM Object Reference) {is=title K2Node_FunctionEntry}
   > ClearChildren(self=Get Slot_CustomSticker() -> (Slot_CustomSticker:Named Slot Object Reference))
   > n1 = Get!(execute) -> (then, else, SelectedItem:Fort Item VM Object Reference)
   n2 = TryGetSoftCustomWidgetClass!(
    self:Fort Locker Category Config Object Reference=GetItemCategory(self=@n1)
    ViewRequest:Gameplay Tag Structure=(TagName="UI.ItemShop.Sticker")
   ) -> (WidgetClass:Common User Widget Soft Class Reference, ReturnValue:bool)
   > n3 = LoadClassAsset_Blocking(AssetClass=@n2)
   > n4 = Cast To CommonUserWidget Class!(execute:Exec, Class:Object Class Reference=@n3) -> (then:Exec, CastFailed:Exec, AsCommon User Widget:Common User Widget Class Reference) {is=title K2Node_ClassDynamicCast}
   > n5 = Set NewStickerClass(NewStickerClass:Common User Widget Class Reference=@n4) -> (Output_Get:Common User Widget Class Reference)
   > Branch(Condition=AND Boolean(A=IsValidClass!(Class:Object Class Reference=@n5) -> (ReturnValue:bool), B=@n2.ReturnValue))
   > n6 = Create Widget!(execute, Class=Get NewStickerClass() -> (NewStickerClass:Common User Widget Class Reference), OwningPlayer) -> (then, ReturnValue:Common User Widget Object Reference)
   > AddChild(self=Get Slot_CustomSticker(), Content=@n6)
   > n7 = Cast To WBP_CustomBanner_Base!(execute, Object=@n6) -> (then, CastFailed, AsWBP Custom Banner Base:WBP Custom Banner Base Object Reference) {is=title K2Node_DynamicCast}
   > SetOfferViewModel!(execute, self:WBP Custom Banner Base Object Reference=@n7, OfferVM:Fort Item Shop Offer VM Object Reference=Get FortItemShopOfferVM()) -> (then)
   > SetSingleItem!(
    execute
    self:WBP Custom Banner Base Object Reference=@n7
    ItemVM:Fort Item VM Object Reference=Get SelectedItem() -> (SelectedItem:Fort Item VM Object Reference)
   ) -> (then)
  ]}
 UpdateEquipButtonVisibility {nodes [
   UpdateEquipButtonVisibility() {is=title K2Node_FunctionEntry}
   > n1 = Get Previewed Item VM()
   n2 = Get CurrentView() -> (CurrentView=OfferInspect)
   > n3 = CanShowEquipItemButton!(
    execute
    self:Fort MPItem Shop Offer Inspection Screen Object Reference
    SelectedItem:Fort Item VM Object Reference=@n1
    OptionalVariantSelector:Fort Variant Selector VM Object Reference=Select(
     Option 0:Fort Variant Selector VM Object Reference
     Option 1:Fort Variant Selector VM Object Reference=Get FortVariantSelectorVM()
     Index="Equal (Enum)"(A:EFortItemShopPdpView Enum=@n2, B:EFortItemShopPdpView Enum=ItemDetails)
    ) -> (ReturnValue:Fort Variant Selector VM Object Reference)
   ) -> (then, ReturnValue:bool)
   > n4 = Get!(execute) -> (then, else, FortItemShopOfferVM:Fort Item Shop Offer VM Object Reference)
   n5 = AND Boolean!(A=@n3, B=Get bOwned(self=Get FortItemShopOfferVM()), C:bool=NOT Boolean(A=IsItemStackable(ItemVm=@n1))) -> (ReturnValue)
   n6 = Get Button_Equip()
   > SetVisibility(
    self=@n6
    InVisibility=Select(Option 0:ESlateVisibility Enum=Collapsed, Option 1:ESlateVisibility Enum=SelfHitTestInvisible, Index=@n5) -> (ReturnValue:ESlateVisibility Enum)
   )
   > SetIsInteractionEnabled(self=@n6, bInIsInteractionEnabled=@n5)
   SetVisibility(execute=[@n4.else, @n1.Is Not Valid], self=Get Button_Equip(), InVisibility=Collapsed)
  ]}
 UpdateFeaturedOfferID {nodes [
   UpdateFeaturedOfferID() {is=title K2Node_FunctionEntry}
   > Set FeaturedOfferId!(
    execute
    FeaturedOfferId:String=Get OfferId!(self:Fort Item Shop Offer VM Object Reference=Get FortItemShopOfferVM()) -> (OfferId:String)
    self:Fort Item Shop VM Object Reference=Get FortItemShopVM()
   ) -> (then, Output_Get:String)
  ]}
 UpdateNavigationArrows {nodes [
   UpdateNavigationArrows() {is=title K2Node_FunctionEntry}
   > n1 = Is Valid(InputObject=Get FortItemShopOfferGroupVM())
   > n2 = Branch(Condition="Equal (Enum)"(A=Get CurrentView() -> (CurrentView=OfferInspect), B=OfferInspect))
   n3 = Get HorizontalBox_NavigationArrows()
   > Set Visibility(Visibility=SelfHitTestInvisible, self=@n3)
   > n4 = Sequence()
   > n5 = GetPreviousOffer() {is=title K2Node_CallFunction}
   n6 = CanInspectOfferVM!(self, InOfferVM:Fort Item Shop Offer VM Object Reference=@n5) -> (ReturnValue:bool) {is=title K2Node_CallFunction}
   > SetIsInteractionEnabled(self=Get Button_PreviousOffer(), bInIsInteractionEnabled=@n6)
   > Set RenderOpacity!(execute, RenderOpacity:float=SelectFloat(A=1, B=0.4, bPickA=@n6), self:Widget Object Reference=Get Button_PreviousOffer()) -> (then, Output_Get:float)
   n7 = GetNextOffer(execute=@n4.then_1) {is=title K2Node_CallFunction}
   n8 = CanInspectOfferVM(InOfferVM=@n7)
   > SetIsInteractionEnabled(self=Get Button_NextOffer(), bInIsInteractionEnabled=@n8)
   > Set RenderOpacity(RenderOpacity=SelectFloat(A=1, B=0.4, bPickA=@n8), self=Get Button_NextOffer())
   Set Visibility(execute=[@n2.else, @n1.Is Not Valid], Visibility=Collapsed, self=@n3)
  ]}
 UpdateOwnershipInfo {nodes [
   UpdateOwnershipInfo() {is=title K2Node_FunctionEntry}
   n1 = Get Offer to Purchase()
   > UpdateGrantedItemsIsOwned(self:Fort Item Shop Offer VM Object Reference=@n1)
   > n2 = Branch(Condition="Equal (Object)"(A=@n1, B=Get FortItemShopOfferVM()))
   UpdatePresentations(execute=[@n3.then, @n2.then], self:WBP UIKit Item List Object Reference=Get Component_ItemList())
   n3 = UpdateGrantedItemsIsOwned(execute=@n2.else, self=Get FortItemShopOfferVM())
  ]}
 UpdatePurchaseBlock {nodes [
   UpdatePurchaseBlock() {is=title K2Node_FunctionEntry}
   > n1 = Branch(Condition="Equal (Enum)"(A=Get CurrentView() -> (CurrentView=OfferInspect), B=ItemDetails))
   > n2 = Get Previewed Item VM()
   > n3 = Set ALC Offer(ALC Offer:Fort Item Shop Offer VM Object Reference=FindRelatedOfferForItem!(
     self:Fort Item Shop VM Object Reference=Get FortItemShopVM()
     BaseOffer:Fort Item Shop Offer VM Object Reference=Get FortItemShopOfferVM()
     Item:Fort Item VM Object Reference=@n2
    ) -> (ReturnValue:Fort Item Shop Offer VM Object Reference)) -> (Output_Get:Fort Item Shop Offer VM Object Reference)
   > Is Valid(InputObject=@n3)
   > OnOfferSet(self=[Get Component_PurchaseBlock(), Get Component_IsbManager(), Get Component_ItemList_ALC()], OfferVM=@n3)
   > n4 = GetIsOfferRestricted(OfferVM=@n3)
   > Set IsALCOfferRestricted(IsALCOfferRestricted:bool=@n4) -> (Output_Get:bool)
   > Set ALCOfferRequiresPin(ALCOfferRequiresPin:bool=@n4.OfferRequiresPin) -> (Output_Get:bool)
   OnOfferSet(
    execute=@n1.else
    self=[Get Component_PurchaseBlock(), Get Component_IsbManager()]
    OfferVM=Get FortItemShopOfferVM()
    ParentOfferVM=Get ParentOfferVM()
    OfferGroupVM=Get FortItemShopOfferGroupVM()
   )
   > ClearSlottedItemDetails() {is=title K2Node_CallFunction}
   > SetContent(self=Get PurchaseBlock_OfferSlot() -> (PurchaseBlock_OfferSlot:Named Slot Object Reference), Content=Get Component_PurchaseBlock())
   > Set!(
    execute
    Padding_Left:float=0
    Padding_Top:float=0
    Padding_Right:float=0
    Padding_Bottom:float=0
    self:User Widget Object Reference=Get Component_PurchaseBlock()
   ) -> (then, Output_Get:Margin Structure)
   n5 = Get IsIsbOnRightSide()
   n6 = Get Component_IsbManager()
   > SetContent(
    self=Select(
     Option 0:Content Widget Object Reference=Get IsbSlot_Offer() -> (IsbSlot_Offer:Named Slot Object Reference)
     Option 1:Content Widget Object Reference=Get IsbSlot_Mobile()
     Index=@n5
    ) -> (ReturnValue:Content Widget Object Reference)
    Content=@n6
   )
   > SetTextJustification(
    self=@n6
    TextJustification=Select(Option 0:ETextJustify Enum=Left, Option 1:ETextJustify Enum=Right, Index=@n5) -> (ReturnValue:ETextJustify Enum)
   )
  ]}
 UpdateVariantInfo {nodes [
   UpdateVariantInfo() {is=title K2Node_FunctionEntry}
   n1 = Get CurrentVariant() -> (CurrentVariant=0)
   > SetText!(
    execute
    self:WBP UIKit Info Box Object Reference=Get InfoBox_VariantInfo()
    InText:Text=Format Text!(Format="LOCTEXT(\"{0} of {1}\")", 0:int=int + int(A=@n1), 1:int=Get TotalVariants() -> (TotalVariants=0)) -> (Result)
   ) -> (then)
  ]}
 VM_VariantSetsUpdated {nodes [
   n1 = VM_VariantSetsUpdated!() -> (then, VariantSetVMs:Array of Fort Variant Set VM Object References) {is=title K2Node_FunctionEntry}
   > n2 = For Each Loop(Array:Array of Fort Variant Set VM Object References=@n1) -> (Array Element:Fort Variant Set VM Object Reference)
   Set HasVariants(execute=@n2.Completed, HasVariants:bool=false) -> (Output_Get:bool)
   n3 = For Each Loop(
    Exec=@n2.LoopBody
    Array:Array of Fort Variant Channel VM Object References=Get VariantChannels!(self:Fort Variant Set VM Object Reference=@n2) -> (VariantChannels:Array of Fort Variant Channel VM Object References)
   ) -> (Array Element:Fort Variant Channel VM Object Reference)
   n4 = Get VariantItems!(self:Fort Variant Channel VM Object Reference=@n3) -> (VariantItems:Array of Fort Variant Item VM Object References)
   > Branch(Condition="integer > integer"(A=Length(TargetArray:Array of Fort Variant Item VM Object References=@n4)))
   > Set HasVariants(HasVariants=true)
   > Return Node!(execute)
  ]}
]
variables [
 AboutItemCoachmark {type=WBP_UIKit_CoachMark_C, default=null}
 ALCOfferRequiresPin {type=boolean}
 CachedPreviewedItem {type=FortItemVM, default=null}
 CachedPreviewedItemIndex {type=integer, default=-1}
 Components {type="array<object>", default []}
 CurrentItemViewSettings {
  type=FortItemViewSettings
  default {
   defaultZoomLevel=0
   restrictZoomToWidget=false
   rotationMode=None
   supportsZooming=false
   usesFixedCamera=false
   usesPlacementActor=false
   cameraRotationOffset=<rot 0 0 0>
   meshBoundsCenterOffset=<0 0 0>
   zoomRange {lowerBound {type=Inclusive, value=0}, upperBound {type=Exclusive, value=1}}
  }
 }
 CurrentVariant {type=integer}
 CurrentView {type=EFortItemShopPdpView, default=OfferInspect}
 CustomISB {type=WBP_CustomBanner_Base_C, default=null}
 DefaultTimeBetweenVariants {type=number, default=2}
 HasVariants {type=boolean}
 HasVariantsToCycle {type=boolean}
 IsALCOfferRestricted {type=boolean}
 IsFocusingItems {type=boolean}
 IsIsbOnRightSide {type=boolean}
 IsOfferRestricted {type=boolean}
 LoadState2D {type=E_ItemPreview_LoadState, default=Unset}
 LoadState3D {type=E_ItemPreview_LoadState, default=Unset}
 OfferRequiresPin {type=boolean}
 ParentOfferVM {type=FortItemShopOfferVM, default=null}
 PrimaryItemDetails {type=CommonActivatableWidget, default=null}
 PurchaseOverlayWidget {type=AthenaItemShopOfferPurchasedOverlay_C, default=null}
 ShouldDeactivateOnBack {type=boolean}
 TotalVariants {type=integer}
 VariantPreviews {type="array<FortCosmeticVariantPreview>", default []}
 VariantTimerHandle {type=TimerHandle}
]
widgets {
 info {
  inheritedWidgetCount=1
  namedSlotCount=1
  parentClass=/Script/FortniteUI.FortMPItemShopOfferInspectionScreen
  rootWidgetClass=&Overlay
  widgetCount=65
 }
 widgets [
  Overlay_3 {widgetClassPath=&Overlay, widget=Overlay_3}
  FortItemPreviewBackground {
   widgetClassPath=/Script/FortniteUI.FortItemPreviewBackground
   bIsVariable=true
   parent=Overlay_3
   widget=FortItemPreviewBackground
   properties {previewSceneChangerTag {tagName=UI.ItemPreview.SceneChanger.OfferInspection}}
   slot {class=OverlaySlot, properties {horizontalAlignment=HAlign_Fill, verticalAlignment=VAlign_Fill}}
  }
  WBP_ItemShop_BackgroundGradient {
   widgetClassPath=&/WBP_ItemShop_BackgroundGradient
   bIsVariable=true
   parent=Overlay_3
   widget=WBP_ItemShop_BackgroundGradient
   slot {class=OverlaySlot}
  }
  Image_BackgroundGradient {
   widgetClassPath=&Image
   bIsVariable=true
   parent=Overlay_3
   widget=Image_BackgroundGradient
   properties {
    visibility=SelfHitTestInvisible
    brush {
     bIsDynamicallyLoaded=false
     drawAs=Image
     imageType=FullColor
     mirroring=NoMirror
     resourceName=None
     resourceObject=/FortUILibrary/UIKit/ModulateBackground/Materials/MI_UI_ModulateBackground
     tiling=NoTile
     imageSize=<800 32>
     margin=<ltrb 0 0 0 0>
     outlineSettings {
      bUseBrushTransparency=false
      roundingType=HalfHeightRadius
      width=0
      color {colorUseRule=UseColor_Specified, specifiedColor=<rgba 0 0 0 0>}
      cornerRadii=<0 0 0 0>
     }
     tintColor {colorUseRule=UseColor_Specified, specifiedColor=<rgba 1 1 1 1>}
     uVRegion {bIsValid=false, max=<0 0>, min=<0 0>}
    }
   }
   slot {class=OverlaySlot, properties {horizontalAlignment=HAlign_Fill, verticalAlignment=VAlign_Fill}}
  }
  WBP_CaptureForPostBufferUpdate {
   widgetClassPath=&WBP_CaptureForPostBufferUpdate
   parent=Overlay_3
   slot {class=OverlaySlot, properties {horizontalAlignment=HAlign_Fill, verticalAlignment=VAlign_Fill}}
  }
  SafeZone_All {
   widgetClassPath=&SafeZone
   bIsVariable=true
   parent=Overlay_3
   widget=SafeZone_All
   properties {padBottom=false, padTop=false}
   slot {class=OverlaySlot, properties {horizontalAlignment=HAlign_Fill, verticalAlignment=VAlign_Fill}}
  }
  Overlay_All {widgetClassPath=&Overlay, parent=SafeZone_All, widget=Overlay_All, slot {class=SafeZoneSlot}}
  GridPanel_CameraFrame {
   widgetClassPath=&GridPanel
   parent=Overlay_All
   widget=GridPanel_CameraFrame
   properties {columnFill [0.5,1], rowFill [1]}
   slot {class=OverlaySlot, properties {horizontalAlignment=HAlign_Fill, verticalAlignment=VAlign_Fill}}
  }
  SizeBox_CameraFrame {
   widgetClassPath=&SizeBox
   bIsVariable=true
   parent=GridPanel_CameraFrame
   widget=SizeBox_CameraFrame
   properties {bOverride_HeightOverride=true, bOverride_WidthOverride=true, heightOverride=1000, widthOverride=1160}
   slot {class=GridSlot, properties {column=1, horizontalAlignment=HAlign_Center, verticalAlignment=VAlign_Center, padding=<ltrb 0 100 0 100>}}
  }
  Overlay_1 {widgetClassPath=&Overlay, parent=SizeBox_CameraFrame, widget=Overlay_1, slot {class=SizeBoxSlot}}
  FortItemPreviewWidget {
   widgetClassPath=/FortniteCoreUI/Frontend/ItemPreview/WBP_ItemPreview
   bInherited=true
   bIsVariable=true
   parent=Overlay_1
   widget=FortItemPreviewWidget
   properties {
    bAllowPawnOverride=true
    bCanDisplayUserAccessories=false
    bRequireExclusiveLoadoutSatisfied=false
    bSupportsActivationFocus=false
    button_Camera=&self:WidgetTree.Button_Camera
    itemViewObject=&self:WidgetTree.FortItemPreviewWidget.ItemViewObject
    infoPadding=<ltrb 60 20 40 60>
    previewActionButtons [
     &self:WidgetTree.PreviewActionButton1
     &self:WidgetTree.PreviewActionButton2
     &self:WidgetTree.PreviewActionButton3
    ]
    previewSceneChangerTag {tagName=UI.ItemPreview.SceneChanger.OfferInspection}
   }
   slot {class=OverlaySlot, properties {horizontalAlignment=HAlign_Fill, verticalAlignment=VAlign_Fill}}
  }
  Throbber_Equip {
   widgetClassPath=/FortUILibrary/UIKit/Throbber/WBP_UIKit_Throbber
   bIsVariable=true
   parent=Overlay_1
   widget=Throbber_Equip
   properties {throbberSize=Medium, visibility=Hidden}
   slot {class=OverlaySlot, properties {horizontalAlignment=HAlign_Center, verticalAlignment=VAlign_Center}}
  }
  EquippedText {
   widgetClassPath=&CommonTextBlock
   bIsVariable=true
   parent=Overlay_1
   widget=EquippedText
   properties {
    text=EQUIPPED
    visibility=Hidden
    colorAndOpacity {colorUseRule=UseColor_Specified, specifiedColor=<rgba 0 0.6654 1 1>}
    font {
     bForceMonospaced=false
     bMaterialIsStencil=false
     fontMaterial=null
     fontObject=&HeadingNow
     letterSpacing=0
     monospacedWidth=1
     size=39
     skewAmount=0
     typefaceFontName=75 Medium
     outlineSettings {
      bApplyOutlineToDropShadows=false
      bMiteredCorners=false
      bSeparateFillAlpha=false
      outlineMaterial=null
      outlineSize=0
      outlineColor=<rgba 0 0 0 1>
     }
    }
    shadowOffset=<0 0>
   }
   slot {class=OverlaySlot, properties {horizontalAlignment=HAlign_Center, verticalAlignment=VAlign_Bottom, padding=<ltrb 0 0 0 250>}}
  }
  VerticalBox_LeftSideInfo {
   widgetClassPath=&VerticalBox
   bIsVariable=true
   parent=Overlay_All
   widget=VerticalBox_LeftSideInfo
   slot {class=OverlaySlot, properties {horizontalAlignment=HAlign_Fill, verticalAlignment=VAlign_Fill, padding=<ltrb 80 40 80 24>}}
  }
  SizeBox_Details {
   widgetClassPath=&SizeBox
   bIsVariable=true
   parent=VerticalBox_LeftSideInfo
   widget=SizeBox_Details
   properties {bOverride_WidthOverride=true, heightOverride=308, maxDesiredWidth=800, minDesiredHeight=312, widthOverride=840}
   slot {class=VerticalBoxSlot, properties {horizontalAlignment=HAlign_Left}}
  }
  WBP_ItemShopPDP_Description {
   widgetClassPath=/ItemShop/MultiProduct/ItemShopPDPComponents/WBP_ItemShopPDP_Description
   bIsVariable=true
   parent=SizeBox_Details
   widget=WBP_ItemShopPDP_Description
   properties {flowDirectionPreference=Culture}
   slot {class=SizeBoxSlot}
  }
  IsbSlot_ALC {
   widgetClassPath=&NamedSlot
   bIsVariable=true
   parent=VerticalBox_LeftSideInfo
   widget=IsbSlot_ALC
   slot {class=VerticalBoxSlot, properties {horizontalAlignment=HAlign_Left, padding=<ltrb 0 8 0 0>}}
  }
  RestrictionWarning {
   widgetClassPath=&HorizontalBox
   bIsVariable=true
   parent=VerticalBox_LeftSideInfo
   widget=RestrictionWarning
   slot {class=VerticalBoxSlot, properties {horizontalAlignment=HAlign_Left, padding=<ltrb 0 8 0 0>}}
  }
  Image_Warning_1 {
   widgetClassPath=&Image
   parent=RestrictionWarning
   widget=Image_Warning_1
   properties {
    visibility=SelfHitTestInvisible
    brush {
     bIsDynamicallyLoaded=false
     drawAs=Image
     imageType=FullColor
     mirroring=NoMirror
     resourceName=None
     resourceObject=/Game/Athena/UI/Frontend/CosmeticGating/Materials/MI_UI_CosmeticGate_WarningIcon
     tiling=NoTile
     imageSize=<24 23>
     margin=<ltrb 0 0 0 0>
     outlineSettings {
      bUseBrushTransparency=false
      roundingType=HalfHeightRadius
      width=0
      color {colorUseRule=UseColor_Specified, specifiedColor=<rgba 0 0 0 0>}
      cornerRadii=<0 0 0 0>
     }
     tintColor {colorUseRule=UseColor_Specified, specifiedColor=<rgba 1 1 1 1>}
     uVRegion {bIsValid=false, max=<0 0>, min=<0 0>}
    }
    colorAndOpacity=<rgba 1 1 1 0.6>
   }
   slot {class=HorizontalBoxSlot, properties {verticalAlignment=VAlign_Center}}
  }
  CommonTextBlock_91 {
   widgetClassPath=&CommonTextBlock
   parent=RestrictionWarning
   widget=CommonTextBlock_91
   properties {
    mobileFontSizeMultiplier=1.6
    text=Restrictions apply
    colorAndOpacity {colorUseRule=UseColor_Specified, specifiedColor=<rgba 1 1 1 0.65>}
    font {
     bForceMonospaced=false
     bMaterialIsStencil=false
     fontMaterial=null
     fontObject=&HeadingNow
     letterSpacing=10
     monospacedWidth=1
     size=11.25
     skewAmount=0
     typefaceFontName=74 Regular
     outlineSettings {
      bApplyOutlineToDropShadows=false
      bMiteredCorners=false
      bSeparateFillAlpha=false
      outlineMaterial=null
      outlineSize=0
      outlineColor=<rgba 0 0 0 1>
     }
    }
    shadowOffset=<0 0>
   }
   slot {class=HorizontalBoxSlot, properties {verticalAlignment=VAlign_Center, padding=<ltrb 8 0 0 0>}}
  }
  Switcher_ScreenView {
   widgetClassPath=&CommonVisibilitySwitcher
   bIsVariable=true
   parent=VerticalBox_LeftSideInfo
   widget=Switcher_ScreenView
   slot {class=VerticalBoxSlot, properties {size {sizeRule=Fill, value=1}}}
  }
  View_OfferInspect {
   widgetClassPath=&Overlay
   bIsVariable=true
   parent=Switcher_ScreenView
   widget=View_OfferInspect
   slot {class=CommonVisibilitySwitcherSlot}
  }
  VerticalBox_1 {
   widgetClassPath=&VerticalBox
   parent=View_OfferInspect
   widget=VerticalBox_1
   slot {class=OverlaySlot, properties {horizontalAlignment=HAlign_Fill, verticalAlignment=VAlign_Bottom}}
  }
  IsbSlot_Offer {
   widgetClassPath=&NamedSlot
   bIsVariable=true
   parent=VerticalBox_1
   widget=IsbSlot_Offer
   slot {class=VerticalBoxSlot, properties {horizontalAlignment=HAlign_Left}}
  }
  Component_IsbManager {
   widgetClassPath=/ItemShop/MultiProduct/ItemShopPDPComponents/WBP_ItemShopPDP_IsbManager
   bIsVariable=true
   parent=IsbSlot_Offer
   widget=Component_IsbManager
   slot {class=PanelSlot}
  }
  Component_ItemList {
   widgetClassPath=/ItemShop/MultiProduct/ItemShopPDPComponents/WBP_ItemShopPDP_ItemList
   bIsVariable=true
   parent=VerticalBox_1
   widget=Component_ItemList
   properties {navigation=&self:WidgetTree.Component_ItemList.WidgetNavigation_0}
   slot {class=VerticalBoxSlot, properties {horizontalAlignment=HAlign_Left, padding=<ltrb -24 0 0 24>}}
  }
  PurchaseBlock_OfferSlot {
   widgetClassPath=&NamedSlot
   bIsVariable=true
   parent=VerticalBox_1
   widget=PurchaseBlock_OfferSlot
   properties {navigation=&self:WidgetTree.PurchaseBlock_OfferSlot.WidgetNavigation_0}
   slot {class=VerticalBoxSlot}
  }
  Component_PurchaseBlock {
   widgetClassPath=/ItemShop/MultiProduct/ItemShopPDPComponents/WBP_ItemShopPDP_PurchaseBlock
   bIsVariable=true
   parent=PurchaseBlock_OfferSlot
   widget=Component_PurchaseBlock
   slot {class=PanelSlot}
  }
  View_ItemDetails {
   widgetClassPath=&VerticalBox
   bIsVariable=true
   parent=Switcher_ScreenView
   widget=View_ItemDetails
   slot {class=CommonVisibilitySwitcherSlot}
  }
  SizeBox_ItemDetails {
   widgetClassPath=&SizeBox
   parent=View_ItemDetails
   widget=SizeBox_ItemDetails
   properties {bOverride_WidthOverride=true, flowDirectionPreference=Culture, widthOverride=640}
   slot {class=VerticalBoxSlot, properties {horizontalAlignment=HAlign_Left, size {sizeRule=Fill, value=1}}}
  }
  Switcher_ItemDetails {
   widgetClassPath=&CommonVisibilitySwitcher
   bIsVariable=true
   parent=SizeBox_ItemDetails
   widget=Switcher_ItemDetails
   properties {navigation=&self:WidgetTree.Switcher_ItemDetails.WidgetNavigation_0}
   slot {class=SizeBoxSlot}
  }
  ItemDetails_Default {
   widgetClassPath=&VerticalBox
   bIsVariable=true
   parent=Switcher_ItemDetails
   widget=ItemDetails_Default
   slot {class=CommonVisibilitySwitcherSlot}
  }
  WBP_UIKit_VariantSelector {
   widgetClassPath=/FortUILibrary/UIKit/VariantSelector/WBP_UIKit_VariantSelector
   bIsVariable=true
   parent=ItemDetails_Default
   widget=WBP_UIKit_VariantSelector
   properties {displayUnownedVariants=true, showAgeGateWarnings=true, showUnownedVariantsAsLocked=false}
   slot {class=VerticalBoxSlot, properties {size {sizeRule=Fill, value=1}}}
  }
  NamedSlot_CustomDetails {
   widgetClassPath=&NamedSlot
   bIsVariable=true
   parent=Switcher_ItemDetails
   widget=NamedSlot_CustomDetails
   slot {class=CommonVisibilitySwitcherSlot, properties {padding=<ltrb 0 16 0 0>}}
  }
  ItemListALC_ItemSlot {
   widgetClassPath=&NamedSlot
   bIsVariable=true
   parent=View_ItemDetails
   widget=ItemListALC_ItemSlot
   properties {navigation=&self:WidgetTree.ItemListALC_ItemSlot.WidgetNavigation_0}
   slot {class=VerticalBoxSlot, properties {horizontalAlignment=HAlign_Left, verticalAlignment=VAlign_Top}}
  }
  Component_ItemList_ALC {
   widgetClassPath=/ItemShop/MultiProduct/ItemShopPDPComponents/WBP_ItemShopPDP_ItemList_ALC
   bIsVariable=true
   parent=ItemListALC_ItemSlot
   widget=Component_ItemList_ALC
   properties {navigation=&self:WidgetTree.Component_ItemList_ALC.WidgetNavigation_1}
   slot {class=PanelSlot}
  }
  PurchaseBlock_ItemSlot {
   widgetClassPath=&NamedSlot
   bIsVariable=true
   parent=View_ItemDetails
   widget=PurchaseBlock_ItemSlot
   properties {navigation=&self:WidgetTree.PurchaseBlock_ItemSlot.WidgetNavigation_0}
   slot {class=VerticalBoxSlot}
  }
  CommonHardwareVisibilityBorder_0 {
   widgetClassPath=&CommonHardwareVisibilityBorder
   parent=Overlay_All
   widget=CommonHardwareVisibilityBorder_0
   properties {
    renderOpacity=0
    visibilityQuery {
     autoDescription=" ANY( Input.Gamepad )"
     tokenStreamVersion=0
     userDescription=""
     queryTokenStream [0,1,1,1,0]
     tagDictionary [{tagName=Input.Gamepad}]
    }
   }
   slot {class=OverlaySlot, properties {horizontalAlignment=HAlign_Center, verticalAlignment=VAlign_Bottom, padding=<ltrb 0 0 0 24>}}
  }
  HorizontalBox_NavigationArrows {
   widgetClassPath=&HorizontalBox
   bIsVariable=true
   parent=CommonHardwareVisibilityBorder_0
   widget=HorizontalBox_NavigationArrows
   slot {class=BorderSlot, properties {padding=<ltrb 0 0 0 0>}}
  }
  Button_PreviousOffer {
   widgetClassPath=&WBP_UIKit_Button_Generic
   bIsVariable=true
   parent=HorizontalBox_NavigationArrows
   widget=Button_PreviousOffer
   properties {
    bIsFocusable=false
    iconMaterial=/ItemShop/MultiProduct/Materials/Buttons/MI_UI_Icon_Button_Previous
    minHeight=48
    mobileMinHeight=80
    outlineMaterial=/FortUILibrary/UIKit/Buttons/MaterialInstances/TileOutline/MI_UIKit_Outline
    shouldTextImpactLayout=false
    text=""
    useTextBlock=false
    buttonPadding=<ltrb 8 0 8 0>
    promptNudge=<-50 0>
    triggeringInputAction {dataTable=&InputActionDataTable, rowName=GenericLeftTrigger}
   }
   slot {class=HorizontalBoxSlot}
  }
  Spacer_102 {
   widgetClassPath=&Spacer
   parent=HorizontalBox_NavigationArrows
   widget=Spacer_102
   properties {size=<32 1>}
   slot {class=HorizontalBoxSlot}
  }
  Button_NextOffer {
   widgetClassPath=&WBP_UIKit_Button_Generic
   bIsVariable=true
   parent=HorizontalBox_NavigationArrows
   widget=Button_NextOffer
   properties {
    bIsFocusable=false
    holdData=null
    iconMaterial=/ItemShop/MultiProduct/Materials/Buttons/MI_UI_Icon_Button_Next
    minHeight=48
    mobileMinHeight=80
    outlineMaterial=&MI_UIKit_Outline
    promptHorizontalAlignment=HAlign_Right
    shouldTextImpactLayout=false
    text=""
    useTextBlock=false
    buttonPadding=<ltrb 8 0 8 0>
    promptNudge=<50 0>
    triggeringInputAction {dataTable=&InputActionDataTable, rowName=GenericRightTrigger}
   }
   slot {class=HorizontalBoxSlot}
  }
  InfoBox_VariantInfo {
   widgetClassPath=/FortUILibrary/UIKit/InfoBox/WBP_UIKit_InfoBox
   bIsVariable=true
   parent=Overlay_All
   widget=InfoBox_VariantInfo
   properties {autoWrapText=false, previewText=1 of 2, showIcon=false, textJustification=Right}
   slot {class=OverlaySlot, properties {horizontalAlignment=HAlign_Right, padding=<ltrb 80 160 80 24>}}
  }
  IsbSlot_Mobile {
   widgetClassPath=&NamedSlot
   bIsVariable=true
   parent=Overlay_All
   widget=IsbSlot_Mobile
   slot {class=OverlaySlot, properties {horizontalAlignment=HAlign_Right, verticalAlignment=VAlign_Bottom, padding=<ltrb 0 0 48 120>}}
  }
  WBP_ActionBarContainer {
   widgetClassPath=/Game/UI/Frontend/BottomBar/WBP_ActionBarContainer
   bIsVariable=true
   parent=Overlay_All
   widget=WBP_ActionBarContainer
   slot {class=OverlaySlot, properties {horizontalAlignment=HAlign_Right, verticalAlignment=VAlign_Bottom, padding=<ltrb 0 0 80 24>}}
  }
  HorizontalBox_ButtonPrompts {widgetClassPath=&HorizontalBox, namedSlotHost=WBP_ActionBarContainer, widget=HorizontalBox_ButtonPrompts}
  Button_Equip {
   widgetClassPath=/FortUILibrary/UIKit/Buttons/WBP_UIKit_Button_Quiet
   bIsVariable=true
   parent=HorizontalBox_ButtonPrompts
   widget=Button_Equip
   properties {
    bHideInputActionWithKeyboard=true
    bIsFocusable=false
    minWidth=48
    text=Equip
    triggeringInputAction {dataTable=&InputActionDataTable, rowName=GenericFaceButtonLeft}
   }
   slot {class=HorizontalBoxSlot, properties {padding=<ltrb 0 0 24 0>}}
  }
  Overlay_AboutItem {
   widgetClassPath=&Overlay
   bIsVariable=true
   parent=HorizontalBox_ButtonPrompts
   widget=Overlay_AboutItem
   slot {class=HorizontalBoxSlot, properties {padding=<ltrb 0 0 24 0>}}
  }
  Button_AboutItem {
   widgetClassPath=&WBP_UIKit_Button_Quiet
   bIsVariable=true
   parent=Overlay_AboutItem
   widget=Button_AboutItem
   properties {
    bHideInputActionWithKeyboard=true
    bIsFocusable=false
    minWidth=48
    text="[PH] About Item"
    triggeringInputAction {dataTable=&InputActionDataTable, rowName=GenericFaceButtonTop}
   }
   slot {class=OverlaySlot}
  }
  VisualAttachment_AboutItemCoachmark {
   widgetClassPath=&CommonVisualAttachment
   bIsVariable=true
   parent=Overlay_AboutItem
   widget=VisualAttachment_AboutItemCoachmark
   properties {contentAnchor=<0.88 1.05>}
   slot {class=OverlaySlot, properties {horizontalAlignment=HAlign_Center}}
  }
  PreviewActionButton1 {
   widgetClassPath=&WBP_UIKit_Button_Quiet
   bIsVariable=true
   parent=HorizontalBox_ButtonPrompts
   widget=PreviewActionButton1
   properties {
    bAutoUpdateFromInputAction=true
    bIsFocusable=false
    minWidth=48
    text=""
    triggeringInputAction {dataTable=&InputActionDataTable, rowName=GenericFaceButtonLeft}
   }
   slot {class=HorizontalBoxSlot, properties {padding=<ltrb 0 0 24 0>}}
  }
  PreviewActionButton2 {
   widgetClassPath=&WBP_UIKit_Button_Quiet
   bIsVariable=true
   parent=HorizontalBox_ButtonPrompts
   widget=PreviewActionButton2
   properties {
    bAutoUpdateFromInputAction=true
    bIsFocusable=false
    minWidth=48
    text=""
    triggeringInputAction {dataTable=&InputActionDataTable, rowName=GenericFaceButtonLeft}
   }
   slot {class=HorizontalBoxSlot, properties {padding=<ltrb 0 0 24 0>}}
  }
  PreviewActionButton3 {
   widgetClassPath=&WBP_UIKit_Button_Quiet
   bIsVariable=true
   parent=HorizontalBox_ButtonPrompts
   widget=PreviewActionButton3
   properties {
    bAutoUpdateFromInputAction=true
    bIsFocusable=false
    minWidth=48
    text=""
    triggeringInputAction {dataTable=&InputActionDataTable, rowName=GenericFaceButtonLeft}
   }
   slot {class=HorizontalBoxSlot, properties {padding=<ltrb 0 0 24 0>}}
  }
  Grid_Panel_Options_Button {
   widgetClassPath=&GridPanel
   bIsVariable=true
   parent=HorizontalBox_ButtonPrompts
   widget=Grid_Panel_Options_Button
   properties {visibility=Collapsed}
   slot {class=HorizontalBoxSlot, properties {padding=<ltrb 0 0 24 0>}}
  }
  Button_Options {
   widgetClassPath=&WBP_UIKit_Button_Quiet
   bIsVariable=true
   parent=Grid_Panel_Options_Button
   widget=Button_Options
   properties {
    bHideInputActionWithKeyboard=true
    bIsFocusable=false
    minWidth=48
    text=Options
    triggeringInputAction {dataTable=&InputActionDataTable, rowName=GenericLeftStickClick}
   }
   slot {class=GridSlot}
  }
  ContextMenuAnchorButton_Options {
   widgetClassPath=/FortUILibrary/UIKit/ContextMenu/WBP_UIKit_ContextMenuAnchorButton
   bIsVariable=true
   parent=Grid_Panel_Options_Button
   widget=ContextMenuAnchorButton_Options
   properties {
    buttonClass=/Game/UI/Frontend/Lobby/Nameplates/WBP_EmptyButton_ContextMenu.WBP_EmptyButton_ContextMenu_C
    closeMenuOnContextClicked=true
    menuPlacement=MenuPlacement_AboveAnchor
   }
   slot {class=GridSlot, properties {horizontalAlignment=HAlign_Left, verticalAlignment=VAlign_Top, nudge=<0 -32>}}
  }
  Border_CameraButton {
   widgetClassPath=&Border
   bIsVariable=true
   parent=HorizontalBox_ButtonPrompts
   widget=Border_CameraButton
   properties {
    visibility=Collapsed
    background {
     bIsDynamicallyLoaded=false
     drawAs=NoDrawType
     imageType=NoImage
     mirroring=NoMirror
     resourceName=None
     resourceObject=null
     tiling=NoTile
     imageSize=<32 32>
     margin=<ltrb 0 0 0 0>
     outlineSettings {
      bUseBrushTransparency=false
      roundingType=HalfHeightRadius
      width=0
      color {colorUseRule=UseColor_Specified, specifiedColor=<rgba 0 0 0 0>}
      cornerRadii=<0 0 0 0>
     }
     tintColor {colorUseRule=UseColor_Specified, specifiedColor=<rgba 1 1 1 1>}
     uVRegion {bIsValid=false, max=<0 0>, min=<0 0>}
    }
    padding=<ltrb 0 0 24 0>
   }
   slot {class=HorizontalBoxSlot}
  }
  Button_Camera {
   widgetClassPath=&WBP_UIKit_Button_Quiet
   bIsVariable=true
   parent=Border_CameraButton
   widget=Button_Camera
   properties {bIsFocusable=false, text=Camera, triggeringInputAction {dataTable=&InputActionDataTable, rowName=UnifiedRotateItem}}
   slot {class=BorderSlot, properties {padding=<ltrb 0 0 24 0>}}
  }
  Button_Back {
   widgetClassPath=&WBP_UIKit_Button_Quiet
   bIsVariable=true
   parent=HorizontalBox_ButtonPrompts
   widget=Button_Back
   properties {bIsFocusable=false, text=Back, triggeringInputAction {dataTable=&InputActionDataTable, rowName=GenericBack}}
   slot {class=HorizontalBoxSlot}
  }
  WBP_AthenaBottomBar_WithNotifications {
   widgetClassPath=/Game/UI/Frontend/BottomBar/WBP_AthenaBottomBar_WithNotifications
   bIsVariable=true
   parent=HorizontalBox_ButtonPrompts
   widget=WBP_AthenaBottomBar_WithNotifications
   slot {class=HorizontalBoxSlot, properties {padding=<ltrb 24 0 0 0>}}
  }
  WBP_PlayerInfo {
   widgetClassPath=/Game/UI/Frontend/ItemShop/MultiProduct/WBP_PlayerInfo
   bIsVariable=true
   parent=Overlay_All
   widget=WBP_PlayerInfo
   slot {class=OverlaySlot, properties {horizontalAlignment=HAlign_Right, padding=<ltrb 80 40 80 24>}}
  }
  Loading_Overlay {
   widgetClassPath=&Overlay
   bIsVariable=true
   parent=Overlay_All
   widget=Loading_Overlay
   properties {visibility=Collapsed}
   slot {class=OverlaySlot, properties {horizontalAlignment=HAlign_Fill, verticalAlignment=VAlign_Fill}}
  }
  WBP_UIKit_Scrim {
   widgetClassPath=/FortUILibrary/Systems/Dialog/WBP_UIKit_Scrim
   parent=Loading_Overlay
   widget=WBP_UIKit_Scrim
   properties {blur=10}
   slot {class=OverlaySlot, properties {horizontalAlignment=HAlign_Fill, verticalAlignment=VAlign_Fill}}
  }
  Loading_Throbber {
   widgetClassPath=&WBP_UIKit_Throbber
   parent=Loading_Overlay
   widget=Loading_Throbber
   slot {class=OverlaySlot, properties {horizontalAlignment=HAlign_Center, verticalAlignment=VAlign_Center}}
  }
  Slot_CustomSticker {
   widgetClassPath=&NamedSlot
   bIsVariable=true
   parent=Overlay_All
   widget=Slot_CustomSticker
   slot {class=OverlaySlot, properties {horizontalAlignment=HAlign_Right, padding=<ltrb 0 200 100 0>}}
  }
 ]
}
