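# Multi-product Item Shop list screen showing an offer group as a tile list beside a details panel and item preview.
# View model updates fill the list, header and dynamic background.
# Selection previews the offer's first granted item, and hovering details buttons deactivates non-selected rows.
# The Details action async-loads and pushes the PDP; activation also toggles text chat off, and pre-construct applies mobile layout by aspect ratio.
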WidgetBlueprint /ItemShop/MultiProduct/ListView/WBP_MPItemShop_ListView
animations [
 OnOpen {
  binding_count=1
  length_seconds=0.35
  object=&self:OnOpen
  sequence {
   bindings [WBP_MPItemShop_ListView_Screen_C_69 {
     id=#1
     possessed_class=WBP_MPItemShop_ListView_C
     tracks [
      {
       type=MovieSceneFloatTrack
       display_name=Render Opacity
       sections [{type=MovieSceneFloatSection, end_frame=21, start_frame=0, channels [None {type=MovieSceneScriptingFloatChannel, keys [[6,0],[21,1]]}]}]
      }
      {
       type=MovieScene2DTransformTrack
       sections [{
         type=MovieScene2DTransformSection
         end_frame=21
         start_frame=6
         channels [
          Angle {type=MovieSceneScriptingFloatChannel, const [0,1]}
          Scale.X {type=MovieSceneScriptingFloatChannel, keys [[6,0.98],[21,1]]}
          Scale.Y {type=MovieSceneScriptingFloatChannel, keys [[6,0.98],[21,1]]}
          Shear.X {type=MovieSceneScriptingFloatChannel, const [0,1]}
          Shear.Y {type=MovieSceneScriptingFloatChannel, const [0,1]}
         ]
        }]
      }
     ]
    }]
   display_rate {numerator=60}
  }
 }
 OnClose {
  binding_count=1
  length_seconds=0.15
  object=&self:OnClose
  sequence {
   bindings [WBP_MPItemShop_ListView_Screen_C_83 {
     id=#2
     possessed_class=WBP_MPItemShop_ListView_C
     tracks [
      {
       type=MovieSceneFloatTrack
       display_name=Render Opacity
       sections [{type=MovieSceneFloatSection, end_frame=8, start_frame=0, channels [None {type=MovieSceneScriptingFloatChannel, keys [[0,1],[8.992,0]]}]}]
      }
      {
       type=MovieScene2DTransformTrack
       sections [{
         type=MovieScene2DTransformSection
         end_frame=9
         start_frame=0
         channels [
          Scale.X {type=MovieSceneScriptingFloatChannel, thinned_from=3, keys [[0,1],[9,0.98]]}
          Scale.Y {type=MovieSceneScriptingFloatChannel, keys [[0,1],[9,0.98]]}
         ]
        }]
      }
     ]
    }]
   display_rate {numerator=60}
  }
 }
 ShowDetailsPanel {
  binding_count=1
  length_seconds=0.1
  object=&self:ShowDetailsPanel
  sequence {
   bindings [ListView_Details {
     id=#3
     possessed_class=WBP_MPItemShop_ListView_Details_C
     tracks [
      {
       type=MovieSceneFloatTrack
       display_name=Render Opacity
       sections [{type=MovieSceneFloatSection, end_frame=6, start_frame=0, channels [None {type=MovieSceneScriptingFloatChannel, keys [[0,0],[6,1]]}]}]
      }
      {
       type=MovieSceneByteTrack
       display_name=Visibility
       sections [{type=MovieSceneByteSection, end_frame=0, start_frame=0, channels [None {type=MovieSceneScriptingByteChannel, const [4,1]}]}]
      }
     ]
    }]
   display_rate {numerator=60}
  }
 }
]
graphs [
 BP_GetDesiredFocusTarget {nodes [
   Get Desired Focus Target()
   > n1 = Get!(execute:Exec) -> (then:Exec, else:Exec, FortItemShopVM:Fort Item Shop VM Object Reference)
   n2 = GetOfferFromId!(
    self:Fort Item Shop VM Object Reference=@n1
    OfferId:String=Get FeaturedOfferId!(self:Fort Item Shop VM Object Reference=@n1) -> (FeaturedOfferId:String)
    bSearchFromCatalog:bool=false
   ) -> (ReturnValue:Fort Item Shop Offer VM Object Reference)
   n3 = GetIndexForItem!(self:List View Object Reference=Get ItemList() -> (ItemList:Fort Tile View Object Reference), Item:Object Reference=@n2) -> (ReturnValue:int=0)
   > n4 = Branch(Condition=AND Boolean(A="Not Equal (Integer)"(A=@n3, B=-1), B=IsValid(Object=@n2)))
   > n5 = Set Selected Item!(execute, self:List View Object Reference=Get ItemList(), Item:Object Reference=@n2) -> (then) {is=title K2Node_CallFunction}
   Return Node!(execute=[@n5.then, @n4.else, @n1.else], ReturnValue:Widget Object Reference=Get ItemList())
  ]}
 EventGraph {nodes [
   n1 = "On Clicked (DecoyButton_Back)"!() -> (OutputDelegate, then, Button:Common Button Base Object Reference) {is=title K2Node_ComponentBoundEvent}
   > Fire Event Element Interacted(Element=@n1.Button)
   > n2 = Branch(Condition=HasAnyUserFocusedDescendants!(self:Widget Object Reference=Get ListView_Details() -> (ListView_Details:WBP MPItem Shop List View Details Object Reference)) -> (ReturnValue:bool))
   > SetFocus(self=Get ItemList())
   CloseListView(execute=@n2.else)
   n3 = "On Item Clicked (ItemList)"!() -> (OutputDelegate, then, Item:Object Reference) {is=title K2Node_ComponentBoundEvent}
   n4 = Get CommonInputSubsystem()
   > n5 = Branch(Condition="Equal (Enum)"(A=GetCurrentInputType(self=@n4), B=Gamepad))
   > SetFocus(self=GetDesiredFocusTarget!(self:WBP MPItem Shop List View Details Object Reference=Get ListView_Details()) -> (ReturnValue:Widget Object Reference))
   Navigate To Item!(execute=@n5.else, self:List View Object Reference=Get ItemList(), Item:Object Reference=@n3.Item) -> (then)
   n6 = "On Clicked (DecoyButton_Details)"!() -> (OutputDelegate, then, Button:Common Button Base Object Reference) {is=title K2Node_ComponentBoundEvent}
   > Fire Event Element Interacted(Element=@n6.Button)
   > n7 = Cast To FortItemShopOfferVM!(
    execute
    Object=Get Selected Item!(self:List View Object Reference=Get ItemList()) -> (ReturnValue:Object Reference) {is=title K2Node_CallFunction}
   ) -> (then, CastFailed, AsFort Item Shop Offer VM:Fort Item Shop Offer VM Object Reference) {is=title K2Node_DynamicCast}
   > OpenOfferInspection!(execute, self, OfferVM:Fort Item Shop Offer VM Object Reference=@n7) -> (then)
   n8 = "On Details Buttons Hovered Changed (ListView_Details)"!() -> (OutputDelegate, then, IsHovered:bool) {is=title K2Node_ComponentBoundEvent}
   > SetListActiveState!(execute, self, SetActive:bool=NOT Boolean(A=@n8.IsHovered)) -> (then)
   n9 = "On Item Selection Changed (ItemList)"!() -> (OutputDelegate, then, Item:Object Reference, bIsSelected:bool) {is=title K2Node_ComponentBoundEvent}
   > Branch(Condition=@n9.bIsSelected)
   > n10 = Cast To FortItemShopOfferVM(Object=@n9.Item)
   > "Fire Event Element Interacted (Non-Localized)"!(
    execute
    FortPC:Fort Player Controller Object Reference=Cast To FortPlayerController!(Object=GetOwningPlayer()) -> (AsFort Player Controller, bSuccess:bool=~)
    ElementName:String=Button_ItemSelected
    AdditionalInfo:Flow Analytics Additional Info Structure="Add String (Non-Localized)"!(
     Info:"Flow Analytics Additional Info Structure (by ref)"=Make Flow Analytics Additional Info!() -> (FlowAnalyticsAdditionalInfo:Flow Analytics Additional Info Structure) {is=title K2Node_MakeStruct}
     Key:String=Offer
     Value:String=GetDisplayAssetName!(self:Fort Item Shop Offer VM Object Reference=@n10) -> (ReturnValue:String)
    ) -> (ReturnValue:Flow Analytics Additional Info Structure)
   ) -> (then)
   > SetListActiveState(SetActive=true)
   > PreviewOffer!(execute, self, OfferVM:Fort Item Shop Offer VM Object Reference=@n10) -> (then)
   n11 = LoadBackground!() -> (OutputDelegate, then, Layout:Fort Item Shop Layout Structure) {is=title K2Node_CustomEvent}
   > n12 = Branch(Condition=IsItemShopDynamicBackgroundEnabled!(self:Fort MPItem Shop Offer Group List Object Reference) -> (ReturnValue:bool))
   n13 = Get FortItemShopVM() -> (FortItemShopVM:Fort Item Shop VM Object Reference)
   n14 = OnBackgroundLoaded!() -> (OutputDelegate, then, LoadedBackground:Object Reference) {is=title K2Node_CustomEvent}
   n15 = LoadBackground!(
    execute=@n12.then
    self:Fort Item Shop VM Object Reference=@n13
    Background:"Fort Item Shop CMSBackground Structure (by ref)"=GetPreferredBackgroundForLayout!(self:Fort Item Shop VM Object Reference=@n13, InLayout:"Fort Item Shop Layout Structure (by ref)"=@n11.Layout) -> (ReturnValue:Fort Item Shop CMSBackground Structure)
    OnLoadedCallback:"Delegate (by ref)"=@n14
   ) -> (then, ReturnValue:bool) {is=title K2Node_CallFunction}
   > n16 = Branch(Condition=@n15)
   UpdateBackground!(
    execute=[@n16.else, @n12.else]
    self:Fort MPItem Shop Offer Group List Object Reference
    LoadedBackground:Object Reference
    bPlayForward:bool=false
    PrimaryColor:Linear Color Structure=(R=0,G=0,B=0,A=0)
    SecondaryColor:Linear Color Structure=(R=0,G=0,B=0,A=0)
   ) -> (then)
   n17 = PreviewOffer!() -> (OutputDelegate, then, OfferVM:Fort Item Shop Offer VM Object Reference) {is=title K2Node_CustomEvent}
   n18 = Get GrantedItems!(self:Fort Item Shop Offer VM Object Reference=@n17.OfferVM) -> (GrantedItems:Array of Fort Item VM Object References)
   > Branch(Condition=Is Valid Index!(TargetArray:Array of Fort Item VM Object References=@n18, IndexToTest:int) -> (ReturnValue:bool))
   n19 = "Get (a copy)"(Array:Array of Fort Item VM Object References=@n18) -> (Output:Fort Item VM Object Reference)
   > RequestItemPreview!(execute, self:Fort MPItem Shop Offer Group List Object Reference, InPreviewedItem:Fort Item VM Object Reference=@n19) -> (then)
   > UpdateAboutItemButton!(
    execute
    self:WBP Item Shop PDP About Item Object Reference=Get WBP_ItemShopPDP_AboutItem() -> (WBP_ItemShopPDP_AboutItem:WBP Item Shop PDP About Item Object Reference)
    Item VM:Fort Item VM Object Reference=@n19
   ) -> (then)
   n20 = SetHeaderTextFromLayout!() -> (OutputDelegate, then, Layout:Fort Item Shop Layout Structure) {is=title K2Node_CustomEvent}
   > n21 = Get()
   > n22 = For Each Loop(Array:Array of Fort Item Shop Section VM Object References=Get SectionsWithOffers!(self:Fort Item Shop VM Object Reference=@n21) -> (SectionsWithOffers:Array of Fort Item Shop Section VM Object References)) -> (Array Element:Fort Item Shop Section VM Object Reference)
   > Branch(Condition="Equal Exactly (String)"(
     A=Get SectionId!(self:Fort Item Shop Section VM Object Reference=@n22) -> (SectionId:String)
     B=Break Fort Item Shop Layout!(FortItemShopLayout:"Fort Item Shop Layout Structure (by ref)"=@n20.Layout) -> (SectionId:String, SubsectionId:String) {is=title K2Node_BreakStruct}
    ))
   > Set Text(
    Text=Get DisplayName!(self:Fort Item Shop Section VM Object Reference=@n22) -> (DisplayName:Text)
    self=Get Text_Header() -> (Text_Header:Common Text Object Reference)
   )
   CloseListView() {is=title K2Node_CustomEvent}
   > n23 = "Play Animation with Finished event (legacy)"!(
    execute:Exec
    Widget:User Widget Object Reference=Self-Reference()
    InAnimation:Widget Animation Object Reference=Get OnClose() -> (OnClose:Widget Animation Object Reference)
    StartAtTime:float=0
    NumLoopsToPlay:int=1
    PlayMode:EUMGSequencePlayMode Enum
    PlaybackSpeed:float=1
   ) -> (then:Exec, Result:UMGSequence Player Object Reference, Finished:Exec) {is=title K2Node_PlayAnimation}
   DeactivateWidget(execute=@n23.Finished)
   n24 = OpenOfferInspection!() -> (OutputDelegate, then, OfferVM:Fort Item Shop Offer VM Object Reference) {is=title K2Node_CustomEvent}
   > n25 = Async Load Class Asset!(execute:Exec, AssetClass:Object Soft Class Reference=/ItemShop/MultiProduct/WBP_ItemShop_PDP.WBP_ItemShop_PDP_C) -> (then:Exec, Completed:Exec, Class:Object Class Reference) {is=title K2Node_LoadAssetClass}
   n26 = Cast To CommonActivatableWidget Class!(execute:Exec=@n25.Completed, Class:Object Class Reference=@n25) -> (then:Exec, CastFailed:Exec, AsCommon Activatable Widget:Common Activatable Widget Class Reference) {is=title K2Node_ClassDynamicCast}
   > n27 = PushPrimaryContentClassAsync!(execute:Exec, WidgetClass:Common Activatable Widget Class Reference=@n26) -> (then:Exec, OnPrimaryContentPushed:Exec, ActivatableWidget:Common Activatable Widget Object Reference) {is=title K2Node_AsyncAction}
   n28 = Cast To FortMPItemShopOfferInspectionScreen!(execute=@n27.OnPrimaryContentPushed, Object=@n27) -> (then, CastFailed, AsFort MPItem Shop Offer Inspection Screen:Fort MPItem Shop Offer Inspection Screen Object Reference) {is=title K2Node_DynamicCast}
   > SetViewModels!(
    execute
    self:Fort MPItem Shop Offer Inspection Screen Object Reference=@n28
    OfferVM:Fort Item Shop Offer VM Object Reference=@n24.OfferVM
    OfferGroupVM:Fort Item Shop Offer Group VM Object Reference
   ) -> (then)
   > RequestChangeView!(
    execute
    self:Fort MPItem Shop Offer Inspection Screen Object Reference=@n28
    InNewView:EFortItemShopPdpView Enum=ItemDetails
    InOverrideBack:bool=true
   ) -> (then)
   n29 = SetListActiveState!() -> (OutputDelegate, then, SetActive:bool=~) {is=title K2Node_CustomEvent}
   > n30 = Branch(Condition="Equal (Boolean)"(A:bool=@n29.SetActive, B:bool=Get IsListActive() -> (IsListActive:bool)) -> (ReturnValue:bool=~) {is=title K2Node_PromotableOperator})
   Set IsListActive(execute=@n30.else, IsListActive:bool=@n29.SetActive) -> (Output_Get:bool)
   > n31 = For Each Loop(Array:Array of User Widget Object References=GetDisplayedEntryWidgets!(self:List View Base Object Reference=Get ItemList()) -> (ReturnValue:Array of User Widget Object References)) -> (Array Element:User Widget Object Reference)
   > n32 = Cast To WBP_ItemShopTile_ListView!(execute, Object=@n31) -> (then, CastFailed, AsWBP Item Shop Tile List View:WBP Item Shop Tile List View Object Reference) {is=title K2Node_DynamicCast}
   > n33 = GetOfferVM!(execute, self:WBP Item Shop Tile Base Object Reference=@n32) -> (then, OfferVM:Fort Item Shop Offer VM Object Reference)
   n34 = Get ItemList()
   n35 = OR Boolean(A="Equal (Object)"(A=@n33, B=Get Selected Item(self=@n34)), B=Get IsListActive())
   n36 = Get RowState!(self:WBP Item Shop Tile List View Object Reference=@n32) -> (RowState:E_ItemShop_RowState Enum=NewEnumerator4)
   > n37 = Branch(Condition="Equal (Boolean)"(A="Equal (Enum)"(A:E_ItemShop_RowState Enum=@n36, B:E_ItemShop_RowState Enum=NewEnumerator0), B=@n35) -> (ReturnValue=~))
   SetRowState!(
    execute=@n37.else
    self:BPI Item Shop Tile Interface=@n32
    InRowState:E_ItemShop_RowState Enum=Select(Option 0:E_ItemShop_RowState Enum=NewEnumerator4, Option 1:E_ItemShop_RowState Enum=NewEnumerator0, Index=@n35) -> (ReturnValue:E_ItemShop_RowState Enum)
   ) -> (then)
   Event OnHandlePreviewItem() {is=title K2Node_Event}
   > n38 = Cast To FortItemShopOfferVM(Object=Get Selected Item(self=Get ItemList()))
   n39 = Get ListView_Details()
   > Set FortItemShopOfferVM!(execute, FortItemShopOfferVM:Fort Item Shop Offer VM Object Reference=@n38, self:WBP MPItem Shop List View Details Object Reference=@n39) -> (then, Output_Get:Fort Item Shop Offer VM Object Reference)
   > n40 = Branch(Condition=IsVisible(self=@n39))
   PlayAnimation(execute=@n40.else, InAnimation=Get ShowDetailsPanel() -> (ShowDetailsPanel:Widget Animation Object Reference))
   Event On Activated()
   > Fire Event Sub View Opened!(execute, View:Widget Object Reference=Self-Reference(), AdditionalInfo:Flow Analytics Additional Info Structure, bIsOverlay:bool=false) -> (then)
   > PlayAnimation(InAnimation=Get OnOpen() -> (OnOpen:Widget Animation Object Reference))
   > SetTextChatEnabled!(execute, self, bSetActive:bool=false) -> (then)
   > UpdateVaultWorldBackground!(
    execute
    self:WBP Item Shop Background Gradient Object Reference=Get WBP_ItemShop_BackgroundGradient() -> (WBP_ItemShop_BackgroundGradient:WBP Item Shop Background Gradient Object Reference)
    IsActive:bool=true
    Fort Item Shop VM:Fort Item Shop VM Object Reference=Get FortItemShopVM()
    OnLoadedCallback:"Delegate (by ref)"
   ) -> (then, ToUpdateBackground:bool)
   > n41 = Cast To FortItemShopOfferVM(Object=Get Selected Item(self=Get ItemList()))
   > PreviewOffer(OfferVM=@n41) {is=title K2Node_CallFunction}
   Event PreConstruct()
   > SetMobileStyle() {is=title K2Node_CallFunction}
   Event On Deactivated()
   > Set FeaturedOfferId!(execute, FeaturedOfferId:String, self:Fort Item Shop VM Object Reference=Get FortItemShopVM()) -> (then, Output_Get:String)
   > SetTextChatEnabled(bSetActive=true)
   UpdateBackground(
    execute=@n14.then
    LoadedBackground=@n14.LoadedBackground
    bPlayForward=false
    PrimaryColor=(R=0,G=0,B=0,A=0)
    SecondaryColor=(R=0,G=0,B=0,A=0)
   )
  ]}
 NavRightFromList {nodes [
   NavRightFromList!() -> (then, Navigation:EUINavigation Enum) {is=title K2Node_FunctionEntry}
   > Return Node(ReturnValue=GetDesiredFocusTarget(self=Get ListView_Details()))
  ]}
 SetMobileStyle {nodes [
   SetMobileStyle() {is=title K2Node_FunctionEntry}
   > Branch(Condition=IsMobileGame())
   > SetPadding!(
    execute
    self:Overlay Slot Object Reference=SlotAsOverlaySlot(Widget=Get SafeZone() -> (SafeZone:Safe Zone Object Reference))
    InPadding_Left:float=48
    InPadding_Top:float=24
    InPadding_Right:float=48
    InPadding_Bottom:float=24
   ) -> (then)
   > SetPadding(
    self:Vertical Box Slot Object Reference=SlotAsVerticalBoxSlot(Widget=Get ItemList())
    InPadding_Left=0
    InPadding_Top=0
    InPadding_Right=0
    InPadding_Bottom=-24
   )
   > SetPadding(
    self:Horizontal Box Slot Object Reference=SlotAsHorizontalBoxSlot!(Widget:Widget Object Reference=Get VerticalBox_List() -> (VerticalBox_List:Vertical Box Object Reference)) -> (ReturnValue:Horizontal Box Slot Object Reference)
    InPadding_Left=0
    InPadding_Top=0
    InPadding_Right=0
    InPadding_Bottom=0
   )
   > SetPadding(
    self:Overlay Slot Object Reference=SlotAsOverlaySlot(Widget=Get WBP_PlayerInfo() -> (WBP_PlayerInfo:WBP Player Info Object Reference))
    InPadding_Left=0
    InPadding_Top=0
    InPadding_Right=0
    InPadding_Bottom=0
   )
   > SetPadding(
    self:Horizontal Box Slot Object Reference=SlotAsHorizontalBoxSlot(Widget=Get Overlay_Details() -> (Overlay_Details:Overlay Object Reference))
    InPadding_Left=24
    InPadding_Top=128
    InPadding_Right=0
    InPadding_Bottom=96
   )
   > SetHorizontalAlignment(
    self=SlotAsOverlaySlot(Widget=Get HorizontalBox_Main() -> (HorizontalBox_Main:Horizontal Box Object Reference))
    InHorizontalAlignment=HAlign_Fill
   )
   n1 = GetViewportSize!() -> (ReturnValue_X:double, ReturnValue_Y:double)
   > n2 = Branch(Condition="float >= float"(A=float / float(A=@n1, B=@n1.ReturnValue_Y), B=1.34))
   n3 = SlotAsOverlaySlot(Widget=Get ListView_Details())
   > SetPadding(self:Overlay Slot Object Reference=@n3, InPadding_Left=0, InPadding_Top=-48, InPadding_Right=0, InPadding_Bottom=0)
   n4 = SlotAsOverlaySlot(Widget=Get Widget_CameraFrame() -> (Widget_CameraFrame:Widget Object Reference))
   > SetPadding(self=@n4, InPadding_Left=100, InPadding_Top=100, InPadding_Right=100, InPadding_Bottom=250)
   > n5 = SetEntryWidth!(execute, self:Tile View Object Reference=Get ItemList(), NewWidth:float=680) -> (then)
   SetPadding(
    execute=[@n6.then, @n5.then]
    self:Horizontal Box Slot Object Reference=SlotAsHorizontalBoxSlot(Widget=Get SizeBox_0() -> (SizeBox_0:Size Box Object Reference))
    InPadding_Left=15
    InPadding_Top=140
    InPadding_Right=15
    InPadding_Bottom=100
   )
   SetPadding(execute=@n2.else, self:Overlay Slot Object Reference=@n3, InPadding_Left=0, InPadding_Top=0, InPadding_Right=0, InPadding_Bottom=0)
   > SetPadding(self=@n4, InPadding_Left=132, InPadding_Top=155, InPadding_Right=132, InPadding_Bottom=470)
   > n6 = SetEntryWidth(self=Get ItemList(), NewWidth=750)
  ]}
 SetTextChatEnabled {nodes [
   n1 = SetTextChatEnabled!() -> (then, bSetActive:bool=~) {is=title K2Node_FunctionEntry}
   > Broadcast Event!(
    execute
    self:Gameplay Event Router Component Object Reference=GetGlobalEventRouter()
    EventData:BPS Set Text Chat Active State Structure=Make BPS Set Text Chat Active State!(bSetActive_5_#4:bool=@n1) -> (BPS_SetTextChatActiveState:BPS Set Text Chat Active State Structure) {is=title K2Node_MakeStruct}
    EventContext:Object Reference=GetOwningPlayer()
   ) -> (then, ReturnValue:bool)
  ]}
 VM_FeaturedLayoutUpdated {nodes [
   n1 = VM_FeaturedLayoutUpdated!() -> (then, FeaturedLayout:Fort Item Shop Layout Structure) {is=title K2Node_FunctionEntry}
   > LoadBackground!(execute, self:Self Object Reference, Layout:Fort Item Shop Layout Structure=@n1) -> (then)
  ]}
 VM_OffersUpdated {nodes [
   VM_OffersUpdated!() -> (then, Offers:Array of Fort Item Shop Offer VM Object References) {is=title K2Node_FunctionEntry}
   > ClearListItems(self=Get ItemList())
   > Set List Items!(
    execute
    self:List View Object Reference=Get ItemList()
    InListItems:Array of Object References=Get Offers() -> (Offers:Array of Fort Item Shop Offer VM Object References)
   ) -> (then) {is=title K2Node_CallFunction}
  ]}
 VM_OwningLayoutUpdated {nodes [
   n1 = VM_OwningLayoutUpdated!() -> (then, OwningLayout:Fort Item Shop Layout Structure) {is=title K2Node_FunctionEntry}
   > SetHeaderTextFromLayout!(execute, self, Layout:Fort Item Shop Layout Structure=@n1) -> (then) {is=title K2Node_CallFunction}
  ]}
]
variables [
 Fort Item Preview Background {type=boolean}
 IsListActive {type=boolean, default=true}
]
widgets {
 info {inheritedWidgetCount=6, parentClass=/Script/FortniteUI.FortMPItemShopOfferGroupList, rootWidgetClass=&Overlay, widgetCount=22}
 widgets [
  Overlay_Main {widgetClassPath=&Overlay, widget=Overlay_Main}
  WBP_ItemShop_BackgroundGradient {
   widgetClassPath=/ItemShop/MultiProduct/WBP_ItemShop_BackgroundGradient
   bIsVariable=true
   parent=Overlay_Main
   widget=WBP_ItemShop_BackgroundGradient
   slot {class=OverlaySlot, properties {horizontalAlignment=HAlign_Fill, verticalAlignment=VAlign_Fill}}
  }
  WBP_CaptureForPostBufferUpdate {
   widgetClassPath=&WBP_CaptureForPostBufferUpdate
   parent=Overlay_Main
   slot {class=OverlaySlot, properties {horizontalAlignment=HAlign_Fill, verticalAlignment=VAlign_Fill}}
  }
  SafeZone {
   widgetClassPath=&SafeZone
   bIsVariable=true
   parent=Overlay_Main
   widget=SafeZone
   slot {class=OverlaySlot, properties {horizontalAlignment=HAlign_Fill, verticalAlignment=VAlign_Fill}}
  }
  Overlay_Content {widgetClassPath=&Overlay, parent=SafeZone, widget=Overlay_Content, slot {class=SafeZoneSlot}}
  HorizontalBox_Main {
   widgetClassPath=&HorizontalBox
   bIsVariable=true
   parent=Overlay_Content
   widget=HorizontalBox_Main
   slot {class=OverlaySlot, properties {horizontalAlignment=HAlign_Center, verticalAlignment=VAlign_Fill}}
  }
  VerticalBox_List {
   widgetClassPath=&VerticalBox
   bIsVariable=true
   parent=HorizontalBox_Main
   widget=VerticalBox_List
   slot {class=HorizontalBoxSlot, properties {padding=<ltrb 86 86 0 0>}}
  }
  Text_Header {
   widgetClassPath=&CommonTextBlock
   bIsVariable=true
   parent=VerticalBox_List
   widget=Text_Header
   properties {
    text=All Jam Tracks
    textTransformPolicy=ToUpper
    font {
     bForceMonospaced=false
     bMaterialIsStencil=false
     fontMaterial=null
     fontObject=&HeadingNow
     letterSpacing=-20
     monospacedWidth=1
     size=32.25
     skewAmount=0
     typefaceFontName=86 Bold Italic
     outlineSettings {
      bApplyOutlineToDropShadows=false
      bMiteredCorners=false
      bSeparateFillAlpha=false
      outlineMaterial=null
      outlineSize=0
      outlineColor=<rgba 0 0 0 1>
     }
    }
    shadowOffset=<0 0>
    strikeBrush {
     bIsDynamicallyLoaded=false
     drawAs=NoDrawType
     imageType=NoImage
     mirroring=NoMirror
     resourceName=None
     resourceObject=null
     tiling=NoTile
     imageSize=<32 32>
     margin=<ltrb 0 0 0 0>
     outlineSettings {
      bUseBrushTransparency=false
      roundingType=HalfHeightRadius
      width=0
      color {colorUseRule=UseColor_Specified, specifiedColor=<rgba 0 0 0 0>}
      cornerRadii=<0 0 0 0>
     }
     tintColor {colorUseRule=UseColor_Specified, specifiedColor=<rgba 1 1 1 1>}
     uVRegion {bIsValid=false, max=<0 0>, min=<0 0>}
    }
   }
   slot {class=VerticalBoxSlot, properties {verticalAlignment=VAlign_Top, padding=<ltrb 12 0 0 8>}}
  }
  ItemList {
   widgetClassPath=/Script/FortniteUI.FortTileView
   bIsVariable=true
   parent=VerticalBox_List
   widget=ItemList
   properties {
    bEntrySizeIncludesEntrySpacing=false
    entryHeight=114
    entryWidgetClass=&/WBP_ItemShopTile_ListView.WBP_ItemShopTile_ListView_C
    entryWidth=840
    horizontalEntrySpacing=24
    navigation=&self:WidgetTree.ItemList.WidgetNavigation_0
    numDesignerPreviewEntries=10
    tileAlignment=LeftAligned
    verticalEntrySpacing=24
    scrollBarStyle {
     thickness=12
     draggedThumbImage {
      bIsDynamicallyLoaded=false
      drawAs=RoundedBox
      imageType=FullColor
      mirroring=NoMirror
      resourceName=../../../Engine/Content/Slate/Common/Scrollbar_Thumb.png
      resourceObject=null
      tiling=NoTile
      imageSize=<12 12>
      margin=<ltrb 0.25 0.25 0.25 0.25>
      outlineSettings {
       bUseBrushTransparency=false
       roundingType=FixedRadius
       width=2
       color {colorUseRule=UseColor_Specified, specifiedColor=<rgba 0 0 0 0>}
       cornerRadii=<6 6 6 6>
      }
      tintColor {colorUseRule=UseColor_Specified, specifiedColor=<rgba 1 1 1 1>}
      uVRegion {bIsValid=false, max=<0 0>, min=<0 0>}
     }
     horizontalBackgroundImage {
      bIsDynamicallyLoaded=false
      drawAs=Box
      imageType=FullColor
      mirroring=NoMirror
      resourceName=None
      resourceObject=/FortUILibrary/UIKit/ScrollBar/MaterialInstances/MI_UIKit_ScrollbarBackground
      tiling=NoTile
      imageSize=<0 0>
      margin=<ltrb 0.5 0.5 0.5 0.5>
      outlineSettings {
       bUseBrushTransparency=false
       roundingType=HalfHeightRadius
       width=0
       color {colorUseRule=UseColor_Specified, specifiedColor=<rgba 0 0 0 0>}
       cornerRadii=<0 0 0 0>
      }
      tintColor {colorUseRule=UseColor_Specified, specifiedColor=<rgba 1 1 1 1>}
      uVRegion {bIsValid=false, max=<0 0>, min=<0 0>}
     }
     horizontalBottomSlotImage {
      bIsDynamicallyLoaded=false
      drawAs=NoDrawType
      imageType=FullColor
      mirroring=NoMirror
      resourceName=../../../Engine/Content/Slate/Common/Scrollbar_Background_Horizontal.png
      resourceObject=null
      tiling=NoTile
      imageSize=<8 8>
      margin=<ltrb 0 0 0 0>
      outlineSettings {
       bUseBrushTransparency=false
       roundingType=HalfHeightRadius
       width=0
       color {colorUseRule=UseColor_Specified, specifiedColor=<rgba 0 0 0 0>}
       cornerRadii=<0 0 0 0>
      }
      tintColor {colorUseRule=UseColor_Specified, specifiedColor=<rgba 1 1 1 1>}
      uVRegion {bIsValid=false, max=<0 0>, min=<0 0>}
     }
     horizontalTopSlotImage {
      bIsDynamicallyLoaded=false
      drawAs=NoDrawType
      imageType=FullColor
      mirroring=NoMirror
      resourceName=../../../Engine/Content/Slate/Common/Scrollbar_Background_Horizontal.png
      resourceObject=null
      tiling=NoTile
      imageSize=<8 8>
      margin=<ltrb 0 0 0 0>
      outlineSettings {
       bUseBrushTransparency=false
       roundingType=HalfHeightRadius
       width=0
       color {colorUseRule=UseColor_Specified, specifiedColor=<rgba 0 0 0 0>}
       cornerRadii=<0 0 0 0>
      }
      tintColor {colorUseRule=UseColor_Specified, specifiedColor=<rgba 1 1 1 1>}
      uVRegion {bIsValid=false, max=<0 0>, min=<0 0>}
     }
     hoveredThumbImage {
      bIsDynamicallyLoaded=false
      drawAs=RoundedBox
      imageType=FullColor
      mirroring=NoMirror
      resourceName=../../../Engine/Content/Slate/Common/Scrollbar_Thumb.png
      resourceObject=null
      tiling=NoTile
      imageSize=<12 12>
      margin=<ltrb 0.25 0.25 0.25 0.25>
      outlineSettings {
       bUseBrushTransparency=false
       roundingType=FixedRadius
       width=2
       color {colorUseRule=UseColor_Specified, specifiedColor=<rgba 0 0 0 0>}
       cornerRadii=<6 6 6 6>
      }
      tintColor {colorUseRule=UseColor_Specified, specifiedColor=<rgba 1 1 1 1>}
      uVRegion {bIsValid=false, max=<0 0>, min=<0 0>}
     }
     normalThumbImage {
      bIsDynamicallyLoaded=false
      drawAs=RoundedBox
      imageType=FullColor
      mirroring=NoMirror
      resourceName=../../../Engine/Content/Slate/Common/Scrollbar_Thumb.png
      resourceObject=null
      tiling=NoTile
      imageSize=<0 0>
      margin=<ltrb 0.25 0.25 0.25 0.25>
      outlineSettings {
       bUseBrushTransparency=false
       roundingType=FixedRadius
       width=2
       color {colorUseRule=UseColor_Specified, specifiedColor=<rgba 0 0 0 0>}
       cornerRadii=<6 6 6 6>
      }
      tintColor {colorUseRule=UseColor_Specified, specifiedColor=<rgba 1 1 1 0.75>}
      uVRegion {bIsValid=false, max=<0 0>, min=<0 0>}
     }
     verticalBackgroundImage {
      bIsDynamicallyLoaded=false
      drawAs=Box
      imageType=FullColor
      mirroring=NoMirror
      resourceName=None
      resourceObject=&MI_UIKit_ScrollbarBackground
      tiling=NoTile
      imageSize=<12 12>
      margin=<ltrb 0.5 0.5 0.5 0.5>
      outlineSettings {
       bUseBrushTransparency=false
       roundingType=HalfHeightRadius
       width=0
       color {colorUseRule=UseColor_Specified, specifiedColor=<rgba 0 0 0 0>}
       cornerRadii=<0 0 0 0>
      }
      tintColor {colorUseRule=UseColor_Specified, specifiedColor=<rgba 1 1 1 1>}
      uVRegion {bIsValid=false, max=<0 0>, min=<0 0>}
     }
     verticalBottomSlotImage {
      bIsDynamicallyLoaded=false
      drawAs=NoDrawType
      imageType=FullColor
      mirroring=NoMirror
      resourceName=../../../Engine/Content/Slate/Common/Scrollbar_Background_Vertical.png
      resourceObject=null
      tiling=NoTile
      imageSize=<8 8>
      margin=<ltrb 0 0 0 0>
      outlineSettings {
       bUseBrushTransparency=false
       roundingType=HalfHeightRadius
       width=0
       color {colorUseRule=UseColor_Specified, specifiedColor=<rgba 0 0 0 0>}
       cornerRadii=<0 0 0 0>
      }
      tintColor {colorUseRule=UseColor_Specified, specifiedColor=<rgba 1 1 1 1>}
      uVRegion {bIsValid=false, max=<0 0>, min=<0 0>}
     }
     verticalTopSlotImage {
      bIsDynamicallyLoaded=false
      drawAs=NoDrawType
      imageType=FullColor
      mirroring=NoMirror
      resourceName=../../../Engine/Content/Slate/Common/Scrollbar_Background_Vertical.png
      resourceObject=null
      tiling=NoTile
      imageSize=<8 8>
      margin=<ltrb 0 0 0 0>
      outlineSettings {
       bUseBrushTransparency=false
       roundingType=HalfHeightRadius
       width=0
       color {colorUseRule=UseColor_Specified, specifiedColor=<rgba 0 0 0 0>}
       cornerRadii=<0 0 0 0>
      }
      tintColor {colorUseRule=UseColor_Specified, specifiedColor=<rgba 1 1 1 1>}
      uVRegion {bIsValid=false, max=<0 0>, min=<0 0>}
     }
    }
   }
   slot {class=VerticalBoxSlot, properties {size {sizeRule=Fill, value=1}}}
  }
  SizeBox_0 {
   widgetClassPath=&SizeBox
   bIsVariable=true
   parent=HorizontalBox_Main
   widget=SizeBox_0
   properties {widthOverride=800}
   slot {class=HorizontalBoxSlot, properties {padding=<ltrb 86 140 86 100>, size {sizeRule=Fill, value=1}}}
  }
  Overlay_Details {widgetClassPath=&Overlay, bIsVariable=true, parent=SizeBox_0, widget=Overlay_Details, slot {class=SizeBoxSlot}}
  ListView_Details {
   widgetClassPath=&/WBP_MPItemShop_ListView_Details
   bIsVariable=true
   parent=Overlay_Details
   widget=ListView_Details
   properties {visibility=Hidden}
   slot {class=OverlaySlot, properties {horizontalAlignment=HAlign_Fill, verticalAlignment=VAlign_Fill}}
  }
  Widget_CameraFrame {
   widgetClassPath=&Overlay
   bInherited=true
   bIsVariable=true
   parent=Overlay_Details
   widget=Widget_CameraFrame
   slot {class=OverlaySlot, properties {horizontalAlignment=HAlign_Fill, verticalAlignment=VAlign_Fill, padding=<ltrb 132 100 132 264>}}
  }
  DecoyButton_Back {
   widgetClassPath=/Game/UI/Frontend/BottomBar/WBP_BottomBarDecoyButton
   bIsVariable=true
   parent=Overlay_Content
   widget=DecoyButton_Back
   properties {triggeringInputAction {dataTable=&InputActionDataTable, rowName=Back}}
   slot {class=OverlaySlot}
  }
  DecoyButton_Details {
   widgetClassPath=&WBP_BottomBarDecoyButton
   bIsVariable=true
   parent=Overlay_Content
   widget=DecoyButton_Details
   properties {triggeringInputAction {dataTable=/FortUILibrary/Input/ItemShopInputActionDataTable, rowName=ListViewDetails}}
   slot {class=OverlaySlot}
  }
  WBP_ItemShopPDP_AboutItem {
   widgetClassPath=/ItemShop/MultiProduct/ItemShopPDPComponents/WBP_ItemShopPDP_AboutItem
   bIsVariable=true
   parent=Overlay_Content
   widget=WBP_ItemShopPDP_AboutItem
   properties {visibility=Collapsed}
   slot {class=OverlaySlot}
  }
  WBP_PlayerInfo {
   widgetClassPath=/Game/UI/Frontend/ItemShop/MultiProduct/WBP_PlayerInfo
   bIsVariable=true
   parent=Overlay_Content
   widget=WBP_PlayerInfo
   slot {class=OverlaySlot, properties {horizontalAlignment=HAlign_Right, padding=<ltrb 80 40 80 24>}}
  }
  Button_PreviewAction {widgetClassPath=&CommonButtonBase, bInherited=true, widget=null}
  Button_PreviewAction1 {widgetClassPath=&CommonButtonBase, bInherited=true, widget=null}
  Button_Camera {widgetClassPath=&CommonButtonBase, bInherited=true, widget=null}
  ItemDownloadIndicator {widgetClassPath=/Script/FortniteUI.FortItemDownloadIndicator, bInherited=true, widget=null}
  Text_DebugPreviewDetails {widgetClassPath=&CommonTextBlock, bInherited=true, widget=null}
 ]
}
