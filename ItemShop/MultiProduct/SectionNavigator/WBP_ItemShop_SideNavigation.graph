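# Item Shop side navigation widget that lists categories and nested sections as UIKit tabs and drives the featured layout.
# Section VM updates rebuild the tabs, grouping sections into category VMs.
# Tab selection sets the featured layout, while layout changes reselect the matching tab, with an ignore flag preventing feedback.
# It also handles collapsing, focus delegates, a gamepad open prompt and interaction analytics, flagging the middle-most button when the scrollbar shows.
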WidgetBlueprint /ItemShop/MultiProduct/SectionNavigator/WBP_ItemShop_SideNavigation
graphs [
 AnalyticsSectionInteracted {nodes [
   n1 = AnalyticsSectionInteracted!() -> (then, NewFeaturedLayout:"Fort Item Shop Layout Structure (by ref)") {is=title K2Node_FunctionEntry}
   > n2 = Cast To FortPlayerController(Object=GetOwningPlayer())
   > "Fire Event Element Interacted (Non-Localized)"!(
    execute
    FortPC:Fort Player Controller Object Reference=@n2
    ElementName:String=Break Fort Item Shop Layout!(FortItemShopLayout:"Fort Item Shop Layout Structure (by ref)"=@n1) -> (SectionId:String, SubsectionId:String) {is=title K2Node_BreakStruct}
    AdditionalInfo:Flow Analytics Additional Info Structure
   ) -> (then)
  ]}
 EventGraph {nodes [
   n1 = "On Clicked (Button_OpenNav_PromptBelowNav)"!() -> (OutputDelegate, then, Button:Common Button Base Object Reference) {is=title K2Node_ComponentBoundEvent}
   > Call On Section Navigator Focused!(execute:Exec, self:WBP Item Shop Side Navigation Object Reference) -> (then:Exec) {is=title K2Node_CallDelegate}
   > FocusSelectedSideNavButton()
   > Fire Event Element Interacted(Element=@n1.Button)
   n2 = "On Side Navigation Selection Changed (WBP_UIKit_SideNavigation)"!() -> (OutputDelegate, then, TabID:Name=None, InnerTabID:Name=None, TabViewModel:MVVM Base Viewmodel Object Reference) {is=title K2Node_ComponentBoundEvent}
   n3 = GetSideNavigationTabViewModel!(
    self:WBP UIKit Side Navigation Base Object Reference=Get WBP_UIKit_SideNavigation() -> (WBP_UIKit_SideNavigation:WBP UIKit Side Navigation Object Reference)
    TabID:Name=@n2.TabID
    InnerTabID:Name=@n2.InnerTabID
   ) -> (TabViewModel:MVVM Base Viewmodel Object Reference)
   > n4 = Cast To ItemShopCategoryVM!(execute, Object=@n3) -> (then, CastFailed, AsItem Shop Category VM:Item Shop Category VM Object Reference) {is=title K2Node_DynamicCast}
   n5 = Get AllOfferGroupsWithOffers!(self:Fort Item Shop Section VM Object Reference="Get (a copy)"(Array:Array of Fort Item Shop Section VM Object References=Get SectionVMs!(self:Item Shop Category VM Object Reference=@n4) -> (SectionVMs:Array of Fort Item Shop Section VM Object References)) -> (Output:Fort Item Shop Section VM Object Reference)) -> (AllOfferGroupsWithOffers:Array of Fort Item Shop Offer Group VM Object References)
   > n6 = Set Featured Layout!(
    execute:Exec
    FortItemShopLayout:"Fort Item Shop Layout Structure (by ref)"=Get OwningLayout!(self:Fort Item Shop Offer Group VM Object Reference="Get (a copy)"(Array:Array of Fort Item Shop Offer Group VM Object References=@n5) -> (Output:Fort Item Shop Offer Group VM Object Reference)) -> (OwningLayout:Fort Item Shop Layout Structure)
    IgnoreCallback:bool=true
   ) -> (then:Exec) {is=title K2Node_MacroInstance}
   n7 = GetCurrentInputType(self=Get CommonInputSubsystem())
   Branch(execute=[@n10.then, @n6.then], Condition=AND Boolean(A=Get IsInFocusPath() -> (IsInFocusPath:bool), B="Equal (Enum)"(A=@n7, B=Gamepad)))
   > AnalyticsSectionInteracted!(
    execute
    self
    NewFeaturedLayout:"Fort Item Shop Layout Structure (by ref)"=Get FeaturedLayout!(self:Fort Item Shop VM Object Reference=Get FortItemShopVM() -> (FortItemShopVM:Fort Item Shop VM Object Reference)) -> (FeaturedLayout:Fort Item Shop Layout Structure)
   ) -> (then) {is=title K2Node_CallFunction}
   n8 = Cast To FortItemShopSectionVM!(execute=@n4.CastFailed, Object=@n3) -> (then, CastFailed, AsFort Item Shop Section VM:Fort Item Shop Section VM Object Reference) {is=title K2Node_DynamicCast}
   n9 = Get AllOfferGroupsWithOffers(self=@n8)
   > n10 = Set Featured Layout(FortItemShopLayout=Get OwningLayout(self="Get (a copy)"(Array=@n9)), IgnoreCallback=true)
   n11 = "On Side Navigation Tab Clicked (WBP_UIKit_SideNavigation)"!() -> (OutputDelegate, then, TabID:Name=None, IsTabContainerExpanded:bool, TabViewModel:MVVM Base Viewmodel Object Reference) {is=title K2Node_ComponentBoundEvent}
   > n12 = Cast To ItemShopCategoryVM(Object=@n11.TabViewModel)
   n13 = GetCurrentInputType(self=Get CommonInputSubsystem())
   > n14 = Branch(Condition=AND Boolean(A="Not Equal (Enum)"!(A:ECommonInputType Enum=@n13, B:ECommonInputType Enum=Gamepad) -> (ReturnValue:bool), B=@n11.IsTabContainerExpanded))
   n15 = "Get (a copy)"(Array:Array of Fort Item Shop Section VM Object References=Get SectionVMs(self=@n12)) -> (Output:Fort Item Shop Section VM Object Reference)
   n16 = Get!(self:Fort Item Shop Offer Group VM Object Reference="Get (a copy)"(Array:Array of Fort Item Shop Offer Group VM Object References=Get AllOfferGroupsWithOffers(self=@n15)) -> (Output:Fort Item Shop Offer Group VM Object Reference)) -> (OwningLayout_SectionId:String, OwningLayout_SubsectionId:String)
   > n17 = AnalyticsSectionInteracted!(execute, self, NewFeaturedLayout_SectionId:String=@n16, NewFeaturedLayout_SubsectionId:String=@n16.OwningLayout_SubsectionId) -> (then)
   n18 = Get WBP_UIKit_SideNavigation()
   n19 = Branch(
    execute=[@n17.then, @n14.else]
    Condition=AND Boolean(
     A=Is Selected Section!(SectionVM:Fort Item Shop Section VM Object Reference=@n15) -> (IsCurrentlySelectedID:bool=~) {is=title K2Node_MacroInstance}
     B=IsTabContainerEmpty!(self:WBP UIKit Side Navigation Base Object Reference=@n18, TabID:Name=@n11.TabID) -> (IsEmpty:bool)
    )
   )
   Call On Section Navigator Unfocused!(execute:Exec=[@n26.then, @n19.then], self:WBP Item Shop Side Navigation Object Reference) -> (then:Exec) {is=title K2Node_CallDelegate}
   n20 = "On Side Navigation Inner Tab Clicked (WBP_UIKit_SideNavigation)"!() -> (OutputDelegate, then, TabID:Name=None, InnerTabID:Name=None, TabViewModel:MVVM Base Viewmodel Object Reference) {is=title K2Node_ComponentBoundEvent}
   > n21 = Cast To FortItemShopSectionVM(Object=@n20.TabViewModel)
   n22 = Get CommonInputSubsystem()
   > n23 = Branch(Condition="Not Equal (Enum)"(A=GetCurrentInputType(self=@n22), B=Gamepad))
   n24 = Get(self="Get (a copy)"(Array=Get AllOfferGroupsWithOffers(self=@n21)))
   > n25 = AnalyticsSectionInteracted(NewFeaturedLayout_SectionId=@n24, NewFeaturedLayout_SubsectionId=@n24.OwningLayout_SubsectionId)
   n26 = Branch(execute=[@n25.then, @n23.else], Condition=Is Selected Section(SectionVM=@n21) -> (IsCurrentlySelectedID=~))
   n27 = OnScrollbarVisibilityChanged!() -> (OutputDelegate, then, NewVisibility:ESlateVisibility Enum=Visible) {is=title K2Node_CustomEvent}
   n28 = Make Margin(Left=0, Top=0, Right=24, Bottom=0)
   > SetGridSlotPadding!(
    execute
    Widget:Widget Object Reference=Get ScrollBox_Tabs!(self:WBP UIKit Tab Group Base Object Reference=Get WBP_UIKit_SideNavigation()) -> (ScrollBox_Tabs:Scroll Box Object Reference)
    Padding:Margin Structure=Select!(
     Visible:Margin Structure=@n28
     Collapsed:Margin Structure
     Hidden:Margin Structure=@n28
     HitTestInvisible:Margin Structure=@n28
     SelfHitTestInvisible:Margin Structure=@n28
     Index:ESlateVisibility Enum=@n27.NewVisibility
    ) -> (ReturnValue:Margin Structure)
   ) -> (then)
   > Set IsScrollbarVisible(IsScrollbarVisible:bool=Select(
     Visible:bool=true
     Collapsed:bool=false
     Hidden:bool=false
     HitTestInvisible:bool=true
     SelfHitTestInvisible:bool=true
     Index=@n27.NewVisibility
    ) -> (ReturnValue:bool)) -> (Output_Get:bool)
   Event OnAddedToFocusPath!() -> (OutputDelegate, then, InFocusEvent:Focus Event Structure) {is=title K2Node_Event}
   > n29 = Set IsInFocusPath(IsInFocusPath:bool=true) -> (Output_Get:bool)
   Branch(
    execute=[@n32.then, @n29.then]
    Condition=Get SideNavigationIsCollapsed!(self:WBP UIKit Side Navigation Base Object Reference=Get WBP_UIKit_SideNavigation()) -> (SideNavigationIsCollapsed:bool)
   )
   > Call On Section Navigator Focused()
   Event OnRemovedFromFocusPath!() -> (OutputDelegate, then, InFocusEvent:Focus Event Structure) {is=title K2Node_Event}
   > Set IsInFocusPath(IsInFocusPath=false)
   Event Construct()
   > n30 = Cast To FortGameState_InGame!(execute, Object=GetGameState!() -> (ReturnValue:Game State Base Object Reference)) -> (then, CastFailed, AsFort Game State in Game:Fort Game State In Game Object Reference) {is=title K2Node_DynamicCast}
   SetTriggeringInputAction!(
    execute=[@n30.CastFailed, @n30.then]
    self:Common Button Base Object Reference=Get Button_OpenNav_PromptBelowNav() -> (Button_OpenNav_PromptBelowNav:WBP UIKit Button Generic Object Reference)
    InputActionRow_DataTable:Data Table Object Reference=&InputActionDataTable
    InputActionRow_RowName:Name=Select!(Option 0:Name=GenericBack, Option 1:Name=GenericRightStickClick, Index:bool=false) -> (ReturnValue:Name)
   ) -> (then)
   > SetScrollbarPadding!(
    execute
    self:Scroll Box Object Reference=Get ScrollBox_Tabs(self=Get WBP_UIKit_SideNavigation())
    NewScrollbarPadding_Left:float=0
    NewScrollbarPadding_Top:float=12
    NewScrollbarPadding_Right:float=-12
    NewScrollbarPadding_Bottom:float=12
   ) -> (then)
   n31 = Get ScrollBox_Tabs(self=Get WBP_UIKit_SideNavigation())
   > Unbind Event from On Scroll Bar Visibility Changed(self:Scroll Box Object Reference=@n31, Delegate=@n27) {is=title K2Node_RemoveDelegate}
   > Bind Event to On Scroll Bar Visibility Changed(self:Scroll Box Object Reference=@n31, Delegate=@n27) {is=title K2Node_AddDelegate}
   Event OnMouseEnter()
   > n32 = Branch(Condition=Get bCanNavigateToSideNav!(self:Fort Item Shop Selection VM Object Reference=Get FortItemShopSelectionVM() -> (FortItemShopSelectionVM:Fort Item Shop Selection VM Object Reference)) -> (bCanNavigateToSideNav:bool))
  ]}
 FocusSelectedSideNavButton {nodes [
   FocusSelectedSideNavButton() {is=title K2Node_FunctionEntry}
   > SetFocus(self=GetSideNavigationSelectedButtonBase!(self:WBP UIKit Side Navigation Base Object Reference=Get WBP_UIKit_SideNavigation()) -> (Button:Common Button Base Object Reference))
  ]}
 GetAbsoluteCentreY {nodes [
   n1 = GetCachedGeometry!(self:Widget Object Reference=Inputs!() -> (Widget:Widget Object Reference)) -> (ReturnValue:"Geometry Structure (by ref)")
   n2 = GetAbsoluteSize!(Geometry:"Geometry Structure (by ref)"=@n1) -> (ReturnValue:Vector 2D Structure)
   Outputs!(CentreY:double=vector2d + vector2d!(
     A:Vector 2D Structure=GetAbsolutePosition!(Geometry:"Geometry Structure (by ref)"=@n1) -> (ReturnValue:Vector 2D Structure)
     B:Vector 2D Structure=vector2d * vector2d(A:Vector 2D Structure=@n2, B:Vector 2D Structure=(X=0.5,Y=0.5)) -> (ReturnValue:Vector 2D Structure) {is=title K2Node_PromotableOperator}
    ) -> (ReturnValue_X:double, ReturnValue_Y:double) {is=title K2Node_PromotableOperator}.ReturnValue_Y)
  ]}
 GetIsCollapsed {nodes [
   GetIsCollapsed() {is=title K2Node_FunctionEntry}
   > Return Node!(execute, IsCollapsed:bool=Get SideNavigationIsCollapsed(self=Get WBP_UIKit_SideNavigation()))
  ]}
 GetMiddleMostChild {nodes [
   n1 = GetMiddleMostChild!() -> (then, Child Widgets:Array of Widget Object References) {is=title K2Node_FunctionEntry}
   > Set ContainerCenterY(ContainerCenterY:double=Get Absolute Centre Y!(Widget:Widget Object Reference=Self-Reference()) -> (CentreY:double=~) {is=title K2Node_MacroInstance}) -> (Output_Get:double)
   > Set MinVerticalDistance(MinVerticalDistance:double=999999) -> (Output_Get:double)
   > n2 = For Each Loop(Array:Array of Widget Object References=@n1) -> (Array Element:Widget Object Reference)
   Return Node!(execute=@n2.Completed, MiddleWidget:Widget Object Reference=Get ClosestWidget() -> (ClosestWidget:Widget Object Reference))
   > Cast To BPI_UIKit_Collapsible!(execute, Object=@n2) -> (then, CastFailed, AsBPI UIKit Collapsible:BPI UIKit Collapsible Interface) {is=title K2Node_DynamicCast}
   n3 = "Absolute (Float)"(A=float - float(A=Get Absolute Centre Y(Widget=@n2) -> (CentreY=~), B=Get ContainerCenterY() -> (ContainerCenterY:double)))
   > Branch(Condition="float < float"(A=@n3, B=Get MinVerticalDistance() -> (MinVerticalDistance:double)))
   > Set MinVerticalDistance(MinVerticalDistance=@n3)
   > Set ClosestWidget(ClosestWidget:Widget Object Reference=@n2) -> (Output_Get:Widget Object Reference)
  ]}
 GetSectionButton {nodes [
   GetSectionButton!() -> (then, SectionID:String) {is=title K2Node_FunctionEntry}
   > n1 = Get!(execute:Exec) -> (then:Exec, else:Exec, FortItemShopVM:Fort Item Shop VM Object Reference)
   n2 = GetSectionFromLayoutStruct!(
    InLayoutStruct_SectionId:String=Get SectionID() -> (SectionID:String)
    InLayoutStruct_SubsectionId:String
    InItemShopVM:Fort Item Shop VM Object Reference=@n1
   ) -> (OutSectionVM:Fort Item Shop Section VM Object Reference)
   > n3 = Is Valid(InputObject=@n2)
   Return Node!(execute=@n3.Is Not Valid, Button:Common Button Base Object Reference)
   n4 = Get Tab IDs from VM!(SectionVM:Fort Item Shop Section VM Object Reference=@n2) -> (TabID:Name, InnerTabID:Name) {is=title K2Node_MacroInstance}
   n5 = Get WBP_UIKit_SideNavigation()
   n6 = GetInnerTabButton!(self:WBP UIKit Side Navigation Base Object Reference=@n5, TabID:Name=@n4, InnerTabID:Name=@n4.InnerTabID) -> (TabButton:Common Button Base Object Reference)
   > Return Node(Button=Select(
     Option 0:Common Button Base Object Reference=GetTabButtonBaseByID!(self:Common Tab List Widget Base Object Reference=@n5, TabNameID:Name=@n4) -> (ReturnValue:Common Button Base Object Reference)
     Option 1:Common Button Base Object Reference=@n6
     Index=IsValid(Object=@n6)
    ) -> (ReturnValue:Common Button Base Object Reference))
   Return Node(execute=@n1.else)
  ]}
 GetTabIDsFromVM {nodes [
   n1 = Inputs!() -> (SectionVM:Fort Item Shop Section VM Object Reference)
   n2 = GetCategory!(self:Fort Item Shop Section VM Object Reference=@n1) -> (ReturnValue:Text)
   n3 = TextIsEmpty(InText=@n2)
   n4 = "To String (Text)"(InText=@n2)
   n5 = String To Name(InString=Get SectionId!(self:Fort Item Shop Section VM Object Reference=@n1) -> (SectionId:String))
   Outputs!(
    TabID:Name=SelectName!(A:Name=String To Name(InString=@n4), B:Name=@n5, bPickA:bool=NOT Boolean(A=@n3)) -> (ReturnValue:Name=None)
    InnerTabID:Name=@n5
   )
  ]}
 IsSelectedSection {nodes [
   n1 = Inputs()
   n2 = Get OwningLayout(self="Get (a copy)"(Array=Get AllOfferGroupsWithOffers(self=@n1)))
   n3 = Get FortItemShopVM()
   Outputs!(IsCurrentlySelectedID:bool="Equal Exactly (String)"(
     A=Break Fort Item Shop Layout(FortItemShopLayout=@n2)
     B=Get!(self:Fort Item Shop VM Object Reference=@n3) -> (FeaturedLayout_SectionId:String, FeaturedLayout_SubsectionId:String)
    ))
  ]}
 OnSectionNavigatorFocused {nodes [OnSectionNavigatorFocused() {is=title K2Node_FunctionEntry}]}
 OnSectionNavigatorUnfocused {nodes [OnSectionNavigatorUnfocused() {is=title K2Node_FunctionEntry}]}
 Reset {nodes [
   Reset() {is=title K2Node_FunctionEntry}
   > ClearSideNavigationTabs(self:WBP UIKit Side Navigation Base Object Reference=Get WBP_UIKit_SideNavigation())
   > Set FortItemShopVM(FortItemShopVM:Fort Item Shop VM Object Reference) -> (Output_Get:Fort Item Shop VM Object Reference)
   > Set IgnoreNextLayoutUpdate(IgnoreNextLayoutUpdate:bool=false) -> (Output_Get:bool)
   > Clear(TargetArray:Array of Fort Item Shop Section VM Object References=Get SectionsVMs() -> (SectionsVMs:Array of Fort Item Shop Section VM Object References))
  ]}
 SetFeaturedLayout {nodes [
   n1 = Inputs!() -> (execute:Exec, FortItemShopLayout:"Fort Item Shop Layout Structure (by ref)", IgnoreCallback:bool=~)
   n2 = Break Fort Item Shop Layout(FortItemShopLayout=@n1)
   n3 = Get(self=Get FortItemShopVM())
   n4 = "Equal Exactly (String)"(A=@n3, B=@n2)
   n5 = "Equal Exactly (String)"(A=@n3.FeaturedLayout_SubsectionId, B=@n2.SubsectionId)
   > n6 = Branch(Condition=OR Boolean(A=AND Boolean(A=@n4, B=@n5), B=NOT Boolean(A=@n1.IgnoreCallback)))
   Set FeaturedLayout!(
    execute=[@n6.then, @n7.then]
    FeaturedLayout:Fort Item Shop Layout Structure=@n1
    self:Fort Item Shop VM Object Reference=Get FortItemShopVM()
   ) -> (then, Output_Get:Fort Item Shop Layout Structure)
   > Outputs!(then:Exec)
   n7 = Set IgnoreNextLayoutUpdate(execute=@n6.else, IgnoreNextLayoutUpdate=true)
  ]}
 SetIsCollapsed {nodes [
   n1 = SetIsCollapsed!() -> (then, ShouldCollapse:bool=~, IsImmediate:bool=~) {is=title K2Node_FunctionEntry}
   n2 = GetAllChildren(self=Get TabsContainer!(self:WBP UIKit Tab Group Base Object Reference=Get WBP_UIKit_SideNavigation()) -> (TabsContainer:Grid Panel Object Reference))
   > n3 = For Each Loop(Array=@n2)
   n4 = Branch(execute=@n3.Completed, Condition=Get IsScrollbarVisible() -> (IsScrollbarVisible:bool))
   > n5 = GetMiddleMostChild!(execute, self, Child Widgets:Array of Widget Object References=@n2) -> (then, MiddleWidget:Widget Object Reference) {is=title K2Node_CallFunction}
   > n6 = Cast To WBP_UIKit_SideNavigation_Button!(execute, Object=@n5) -> (then, CastFailed, AsWBP UIKit Side Navigation Button:WBP UIKit Side Navigation Button Object Reference) {is=title K2Node_DynamicCast}
   > n7 = Set IsMiddleWidget!(execute, IsMiddleWidget:bool=true, self:WBP UIKit Side Navigation Button Object Reference=@n6) -> (then, Output_Get:bool)
   SetIsCollapsed!(
    execute=[@n7.then, @n6.CastFailed, @n4.else]
    self:WBP UIKit Side Navigation Base Object Reference=Get WBP_UIKit_SideNavigation()
    IsCollapsed:bool=@n1
    SkipAnimation:bool=@n1.IsImmediate
   ) -> (then) {is=title K2Node_CallFunction}
   > SetVisibility(
    self=Get Button_OpenNav_PromptBelowNav()
    InVisibility=Select(Option 0:ESlateVisibility Enum=Collapsed, Option 1:ESlateVisibility Enum=Visible, Index=Get ShouldCollapse() -> (ShouldCollapse:bool)) -> (ReturnValue:ESlateVisibility Enum)
   )
   n8 = Cast To WBP_UIKit_SideNavigation_Button(execute=@n3.LoopBody, Object=@n3)
   > Set IsMiddleWidget(IsMiddleWidget=false, self=@n8)
   > Set IsScrollbarVisibleOnAnimationStart!(execute, IsScrollbarVisibleOnAnimationStart:bool=Get IsScrollbarVisible(), self:WBP UIKit Side Navigation Button Object Reference=@n8) -> (then, Output_Get:bool)
  ]}
 VM_OnFeaturedLayoutUpdated {nodes [
   VM_OnFeaturedLayoutUpdated!() -> (then, NewLayout:Fort Item Shop Layout Structure) {is=title K2Node_FunctionEntry}
   > n1 = Branch(Condition=Get IgnoreNextLayoutUpdate() -> (IgnoreNextLayoutUpdate:bool))
   > Set IgnoreNextLayoutUpdate(IgnoreNextLayoutUpdate=false)
   n2 = Get!(execute:Exec=@n1.else) -> (then:Exec, else:Exec, FortItemShopVM:Fort Item Shop VM Object Reference)
   > Branch(Condition="Equal (Enum)"(
     A:EFortItemShopLoadingState Enum=Get LoadingState!(self:Fort Item Shop VM Object Reference=@n2) -> (LoadingState:EFortItemShopLoadingState Enum=Uninitialized)
     B:EFortItemShopLoadingState Enum=Ready
    ))
   n3 = GetSectionFromLayoutStruct!(
    InLayoutStruct:Fort Item Shop Layout Structure=Get NewLayout() -> (NewLayout:Fort Item Shop Layout Structure)
    InItemShopVM=Get FortItemShopVM()
   ) -> (OutSectionVM)
   > Is Valid(InputObject=@n3)
   n4 = Get Tab IDs from VM(SectionVM=@n3)
   > SelectSideNavigationTab!(
    execute
    self:WBP UIKit Side Navigation Base Object Reference=Get WBP_UIKit_SideNavigation()
    TabID:Name=@n4
    InnerTabID:Name=@n4.InnerTabID
    bSuppressClickFeedback:bool=true
   ) -> (then, Success:bool)
   > Set CachedLayout(CachedLayout:Fort Item Shop Layout Structure=Get NewLayout()) -> (Output_Get:Fort Item Shop Layout Structure)
  ]}
 VM_SectionVMsUpdated {nodes [
   n1 = VM_SectionVMsUpdated!() -> (then, SectionViewModels:Array of Fort Item Shop Section VM Object References) {is=title K2Node_FunctionEntry}
   > Set SectionsVMs(SectionsVMs:Array of Fort Item Shop Section VM Object References=@n1) -> (Output_Get:Array of Fort Item Shop Section VM Object References)
   > ClearSideNavigationTabs(self=Get WBP_UIKit_SideNavigation())
   > n2 = For Each Loop(Array:Array of Fort Item Shop Section VM Object References=Get SectionsVMs()) -> (Array Element:Fort Item Shop Section VM Object Reference)
   > Is Valid(InputObject=@n2)
   > Branch(Condition=IsVisible(self:Fort Item Shop Section VM Object Reference=@n2))
   > Set LoopSectionVM(LoopSectionVM:Fort Item Shop Section VM Object Reference=@n2) -> (Output_Get:Fort Item Shop Section VM Object Reference)
   n3 = Get LoopSectionVM() -> (LoopSectionVM:Fort Item Shop Section VM Object Reference)
   > Set IsCategoryEmpty(IsCategoryEmpty:bool=TextIsEmpty(InText=GetCategory(self=@n3))) -> (Output_Get:bool)
   > n4 = Get!(execute) -> (then, else, CurrentCategory:Item Shop Category VM Object Reference)
   n5 = Get IsCategoryEmpty() -> (IsCategoryEmpty:bool)
   n6 = Get Tab IDs from VM(SectionVM=Get LoopSectionVM())
   n7 = Get CategoryID!(self:Item Shop Category VM Object Reference=@n4) -> (CategoryID:Name=None)
   > n8 = Branch(Condition=AND Boolean(A=NOT Boolean(A=@n5), B="Equal (Name)"(A=@n7, B=@n6)))
   n9 = Get Tab IDs from VM(SectionVM=Get LoopSectionVM())
   AddSideNavigationInnerTab!(
    execute=[@n13.else, @n8.then]
    self:WBP UIKit Side Navigation Base Object Reference=Get WBP_UIKit_SideNavigation()
    TabViewModel:MVVM Base Viewmodel Object Reference=Get LoopSectionVM()
    TabID:Name=@n9
    InnerTabID:Name=@n9.InnerTabID
    Index:int=-1
   ) -> (then)
   > Add(
    TargetArray:Array of Fort Item Shop Section VM Object References=Get SectionVMs(self=Get CurrentCategory() -> (CurrentCategory:Item Shop Category VM Object Reference))
    NewItem:"Fort Item Shop Section VM Object Reference (by ref)"=Get LoopSectionVM()
   ) {is=title K2Node_CallArrayFunction}
   n10 = Get LoopSectionVM()
   n11 = Construct Object from Class!(
    execute=[@n8.else, @n4.else]
    Class=&/ItemShopCategoryVM.ItemShopCategoryVM_C
    self
    CategoryID:Name=Get Tab IDs from VM(SectionVM=@n10)
    DisplayName:Text=Select(
     Option 0:Text=GetCategory(self=@n10)
     Option 1:Text=GetDisplayName!(self:Fort Item Shop Section VM Object Reference=@n10) -> (ReturnValue:Text)
     Index=Get IsCategoryEmpty()
    ) -> (ReturnValue:Text)
    SectionVMs:Array of Fort Item Shop Section VM Object References=Make Array("[0]":Fort Item Shop Section VM Object Reference=@n10) -> (Array:Array of Fort Item Shop Section VM Object References)
   ) -> (then, ReturnValue:Item Shop Category VM Object Reference)
   > n12 = Set CurrentCategory(CurrentCategory:Item Shop Category VM Object Reference=@n11) -> (Output_Get:Item Shop Category VM Object Reference)
   > AddSideNavigationTab!(
    execute
    self:WBP UIKit Side Navigation Base Object Reference=Get WBP_UIKit_SideNavigation()
    TabViewModel:MVVM Base Viewmodel Object Reference=@n12
    TabID:Name=Get CategoryID(self=@n12) -> (CategoryID=None)
    Index:int=-1
   ) -> (then)
   > n13 = Branch(Condition=Get IsCategoryEmpty())
   n14 = GetOwningPlayer()
   SetIsCollapsed(
    execute=@n2.Completed
    self=Get WBP_UIKit_SideNavigation()
    IsCollapsed=Select(
     Option 0:bool=true
     Option 1:bool=false
     Index=IsUsingInputType!(InputType:ECommonInputType Enum=Touch, PlayerController:Player Controller Object Reference=@n14) -> (ReturnValue:bool)
    ) -> (ReturnValue:bool)
    SkipAnimation=true
   )
   > VM_OnFeaturedLayoutUpdated!(execute, self, NewLayout:Fort Item Shop Layout Structure=Get CachedLayout() -> (CachedLayout:Fort Item Shop Layout Structure)) -> (then) {is=title K2Node_CallFunction}
  ]}
]
variables [
 CachedLayout {type=FortItemShopLayout, default {sectionId="", subsectionId=""}}
 IgnoreNextLayoutUpdate {type=boolean}
 IsInFocusPath {type=boolean}
 IsScrollbarVisible {type=boolean}
 OnSectionNavigatorFocused {}
 OnSectionNavigatorUnfocused {}
 SectionsVMs {type="array<FortItemShopSectionVM>", default []}
]
widgets {
 info {inheritedWidgetCount=0, parentClass=&UserWidget, rootWidgetClass=&VerticalBox, widgetCount=7}
 widgets [
  VerticalBox_0 {widgetClassPath=&VerticalBox, widget=VerticalBox_0}
  SizeBox_0 {
   widgetClassPath=&SizeBox
   parent=VerticalBox_0
   widget=SizeBox_0
   properties {bOverride_MaxDesiredHeight=true, heightOverride=300, maxDesiredHeight=752}
   slot {class=VerticalBoxSlot, properties {verticalAlignment=VAlign_Top}}
  }
  Overlay_0 {widgetClassPath=&Overlay, parent=SizeBox_0, widget=Overlay_0, slot {class=SizeBoxSlot}}
  WBP_CaptureForPostBufferUpdate_132 {
   widgetClassPath=&WBP_CaptureForPostBufferUpdate
   bIsVariable=true
   parent=Overlay_0
   widget=WBP_CaptureForPostBufferUpdate_132
   slot {class=OverlaySlot, properties {horizontalAlignment=HAlign_Fill, verticalAlignment=VAlign_Fill}}
  }
  WBP_UIKit_SideNavigation {
   widgetClassPath=/FortUILibrary/UIKit/Tabs/WBP_UIKit_SideNavigation
   bIsVariable=true
   parent=Overlay_0
   widget=WBP_UIKit_SideNavigation
   properties {
    buttonClass=&/WBP_ItemShop_SideNavigation_Category_Button.WBP_ItemShop_SideNavigation_Category_Button_C
    innerTabButtonClassOverride=&/WBP_ItemShop_SideNavigation_Section_Button.WBP_ItemShop_SideNavigation_Section_Button_C
    scrollBarVisibility=SelfHitTestInvisible
    tabGroupHorizontalAlignment=HAlign_Fill
    innerTabGroupPadding=<ltrb 8 8 8 8>
   }
   slot {class=OverlaySlot, properties {horizontalAlignment=HAlign_Fill, verticalAlignment=VAlign_Fill}}
  }
  CommonHardwareVisibilityBorder_0 {
   widgetClassPath=&CommonHardwareVisibilityBorder
   parent=VerticalBox_0
   widget=CommonHardwareVisibilityBorder_0
   properties {visibilityQuery {
     autoDescription=" ANY( Input.Gamepad )"
     tokenStreamVersion=0
     userDescription=""
     queryTokenStream [0,1,1,1,0]
     tagDictionary [{tagName=Input.Gamepad}]
    }}
   slot {class=VerticalBoxSlot, properties {horizontalAlignment=HAlign_Center, verticalAlignment=VAlign_Top, padding=<ltrb 0 8 0 0>}}
  }
  Button_OpenNav_PromptBelowNav {
   widgetClassPath=&WBP_UIKit_Button_Generic
   bIsVariable=true
   parent=CommonHardwareVisibilityBorder_0
   widget=Button_OpenNav_PromptBelowNav
   properties {
    bHideInputActionWithKeyboard=true
    bIsFocusable=false
    minHeight=40
    promptHorizontalAlignment=HAlign_Center
    useBackgroundBlock=false
    useIconBlock=false
    useTextBlock=false
    visibility=Collapsed
    buttonPadding=<ltrb 0 0 0 0>
    promptNudge=<0 0>
    triggeringInputAction {dataTable=&InputActionDataTable, rowName=GenericBack}
   }
   slot {class=BorderSlot, properties {padding=<ltrb 0 0 0 0>}}
  }
 ]
}
