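# Item Shop "About Pcb" informational dialog, a CommonActivatableWidget wrapping the UIKit internal dialog.
# On activation it sets rich text decorators and the scroll target, fires a sub-view-opened analytics event, and raises button minimum height on mobile.
# VM_TempItemsUpdated concatenates granted item names into a bulleted list and formats it into the viewmodel's DialogInfoText.
# Deactivation always reports a Button_GotIt interaction.
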WidgetBlueprint /ItemShop/MultiProduct/WBP_AboutPcb_Dialog
graphs [
 EventGraph {nodes [
   Event On Activated()
   > SetDecorators!(
    execute
    self:Rich Text Block Object Reference=Get ContentRichText!(self:WBP UIKit Dialog Internal Object Reference=Get Dialog_Internal() -> (Dialog_Internal:WBP UIKit Dialog Internal Object Reference)) -> (ContentRichText:Common Rich Text Block Object Reference)
    InDecoratorClasses:Array of Rich Text Block Decorator Class References=Make Array("[0]":Rich Text Block Decorator Class Reference=/FortniteCoreUI/Foundation/RichTextData/CommonRichTextDecorators.CommonRichTextDecorators_C) -> (Array:Array of Rich Text Block Decorator Class References)
   ) -> (then)
   > SetScrollWidget!(
    execute
    ToScroll:Widget Object Reference=Get ScrollBox_Text!(self:WBP UIKit Dialog Internal Object Reference=Get Dialog_Internal()) -> (ScrollBox_Text:Scroll Box Object Reference)
   ) -> (then)
   > Fire Event Sub View Opened!(execute, View:Widget Object Reference=Self-Reference(), AdditionalInfo:Flow Analytics Additional Info Structure, bIsOverlay:bool=true) -> (then)
   > SetMobileStyle() {is=title K2Node_CallFunction}
   Event On Deactivated()
   > FireEventElementInteractedByName!(execute, Element Name:String=Button_GotIt, WidgetCallingFrom:Widget Object Reference=Self-Reference()) -> (then, Success:bool)
  ]}
 SetMobileStyle {nodes [
   SetMobileStyle() {is=title K2Node_FunctionEntry}
   > Branch(Condition=IsMobileGame())
   > n1 = For Each Loop(Array:Array of WBP UIKit Button CTA Base Object References=Get ButtonWidgets!(self:WBP UIKit Dialog Internal Object Reference=Get Dialog_Internal()) -> (ButtonWidgets:Array of WBP UIKit Button CTA Base Object References)) -> (Array Element:WBP UIKit Button CTA Base Object Reference)
   > Set MobileMinHeight!(execute, MobileMinHeight:int=80, self:WBP UIKit Button CTA Base Object Reference=@n1) -> (then, Output_Get:int=0)
  ]}
 VM_TempItemsUpdated {nodes [
   n1 = VM_TempItemsUpdated!() -> (then, InGrantedItems:Array of Texts) {is=title K2Node_FunctionEntry}
   > n2 = For Each Loop(Array:Array of Texts=@n1) -> (Array Element:Text)
   Set GrantedItemsText(
    execute=@n2.Completed
    GrantedItemsText:Text=TextTrimPrecedingAndTrailing!(InText:"Text (by ref)"=Get GrantedItemsText() -> (GrantedItemsText:Text)) -> (ReturnValue:Text)
   ) -> (Output_Get:Text)
   n3 = Get DialogVM_PcbShop() -> (DialogVM_PcbShop:Dialog VM About Pcb Object Reference)
   > Set DialogInfoText!(
    execute
    DialogInfoText:Text=Format Text!(
     Format=Get DialogInfoText!(self:UI Kit Dialog Viewmodel Object Reference=@n3) -> (DialogInfoText:Text)
     GrantedItems:Text=Get GrantedItemsText()
    ) -> (Result)
    self:UI Kit Dialog Viewmodel Object Reference=Get DialogVM_PcbShop()
   ) -> (then, Output_Get:Text)
   Set GrantedItemsText(
    execute=@n2.LoopBody
    GrantedItemsText=Format Text!(Format="INVTEXT(\"{ExistingText}\\r\\n• {CurrentCategoryText}\")", ExistingText:Text=Get GrantedItemsText(), CurrentCategoryText:Text=@n2) -> (Result)
   )
  ]}
]
variables []
widgets {
 info {
  inheritedWidgetCount=0
  parentClass=&CommonActivatableWidget
  rootWidgetClass=/FortUILibrary/Systems/Dialog/WBP_UIKit_Dialog_Internal.WBP_UIKit_Dialog_Internal_C
  widgetCount=1
 }
 widgets [Dialog_Internal {
   widgetClassPath=/FortUILibrary/Systems/Dialog/WBP_UIKit_Dialog_Internal
   bIsVariable=true
   widget=Dialog_Internal
   properties {
    designTimePreview_DialogVM=/ItemShop/MultiProduct/Data/DialogVM_AboutPcb.DialogVM_AboutPcb_C
    useText=Common Rich Text
    buttons {
     each Button Horizontal Fill=false
     each Button Min Width=0
     each Button Size=Medium
     focuse Button=0
     space Between Buttons=30
     cTA Buttons Types [Loud,Regular]
     group Placement {
      horizontal Alignment=HAlign_Fill
      vertical Alignment=VAlign_Bottom
      grid {column=1, column Span=0, layer=0, row=2, row Span=0, nudge=<0 0>}
      padding=<ltrb 80 64 80 48>
     }
    }
    dialogSIze {
     height Override {enable=false, value=800}
     max Aspect Ratio {enable=false, value=0}
     max Desired Height {enable=true, value=800}
     max Desired Width {enable=false, value=0}
     min Aspect Ratio {enable=false, value=0}
     min Desired Height {enable=false, value=0}
     min Desired Width {enable=false, value=0}
     width Override {enable=true, value=1000}
    }
    textProperties {
     auto Wrap Text=true
     justification=Left
     line Height Percentage=1.2
     mobile Font Size Multiplier=1.3
     transform Policy=None
     wrap Text At=0
     color {colorUseRule=UseColor_Specified, specifiedColor=<rgba 1 1 1 0.749>}
     font {
      bForceMonospaced=false
      bMaterialIsStencil=false
      fontMaterial=null
      fontObject=&HeadingNow
      letterSpacing=0
      monospacedWidth=1
      size=15.75
      skewAmount=0
      typefaceFontName=74 Regular
      outlineSettings {
       bApplyOutlineToDropShadows=false
       bMiteredCorners=false
       bSeparateFillAlpha=false
       outlineMaterial=null
       outlineSize=0
       outlineColor=<rgba 0 0 0 1>
      }
     }
     placement {
      horizontal Alignment=HAlign_Fill
      vertical Alignment=VAlign_Top
      grid {column=1, column Span=0, layer=0, row=1, row Span=0, nudge=<0 0>}
      padding=<ltrb 80 0 80 0>
     }
    }
   }
  }]
}
