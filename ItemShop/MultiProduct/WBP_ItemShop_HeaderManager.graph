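# Item Shop widget that manages the current, next and filter section headers plus a subtitle, animating them as the shop scrolls between rows.
# Scroll functions choose directional swipe-and-fade animations or header transitions, and scrolling to the billboard hides all headers.
# A stand-in transition text replaces the real header during size changes and current/next handoffs, driven by an animated alpha with sequencer-event lerp setup.
# PreConstruct selects BR, STW or mobile paddings.
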WidgetBlueprint /ItemShop/MultiProduct/WBP_ItemShop_HeaderManager
animations [
 CurrentHeaderSwipeInUp {
  binding_count=1
  length_seconds=0.3
  object=&self:CurrentHeaderSwipeInUp
  sequence {
   bindings [Text_CurrentHeader {
     id=#1
     possessed_class=CommonTextBlock
     tracks [
      {
       type=MovieScene2DTransformTrack
       sections [{
         type=MovieScene2DTransformSection
         end_frame=36
         start_frame=0
         channels [Translation.Y {type=MovieSceneScriptingFloatChannel, keys [[0,250],[36,0]]}]
        }]
      }
      {
       type=MovieSceneFloatTrack
       display_name=Render Opacity
       sections [{type=MovieSceneFloatSection, end_frame=36, start_frame=0, channels [None {type=MovieSceneScriptingFloatChannel, keys [[15,0],[36,1]]}]}]
      }
     ]
    }]
   display_rate {numerator=120}
  }
 }
 CurrentHeaderSwipeInDown {
  binding_count=1
  length_seconds=0.3
  object=&self:CurrentHeaderSwipeInDown
  sequence {
   bindings [Text_CurrentHeader {
     id=#2
     display_name=Text_CurrentSection
     possessed_class=CommonTextBlock
     tracks [
      {
       type=MovieScene2DTransformTrack
       sections [{
         type=MovieScene2DTransformSection
         end_frame=35
         start_frame=0
         channels [Translation.Y {type=MovieSceneScriptingFloatChannel, keys [[0,-250],[35.998,0]]}]
        }]
      }
      {
       type=MovieSceneFloatTrack
       display_name=Render Opacity
       sections [{
         type=MovieSceneFloatSection
         end_frame=35
         start_frame=0
         channels [None {type=MovieSceneScriptingFloatChannel, keys [[15.428,0],[35.998,1]]}]
        }]
      }
     ]
    }]
   display_rate {numerator=120}
  }
 }
 CurrentHeaderSwipeOutUp {
  binding_count=1
  length_seconds=0.3
  object=&self:CurrentHeaderSwipeOutUp
  sequence {
   bindings [Text_CurrentHeader {
     id=#1
     possessed_class=CommonTextBlock
     tracks [
      {
       type=MovieScene2DTransformTrack
       sections [{
         type=MovieScene2DTransformSection
         end_frame=36
         start_frame=0
         channels [Translation.Y {type=MovieSceneScriptingFloatChannel, keys [[0,0],[17,-129.7498],[36,-250]]}]
        }]
      }
      {
       type=MovieSceneFloatTrack
       display_name=Render Opacity
       sections [{type=MovieSceneFloatSection, end_frame=36, start_frame=0, channels [None {type=MovieSceneScriptingFloatChannel, keys [[0,1],[17,0]]}]}]
      }
     ]
    }]
   display_rate {numerator=120}
  }
 }
 CurrentHeaderSwipeOutDown {
  binding_count=1
  length_seconds=0.3
  object=&self:CurrentHeaderSwipeOutDown
  sequence {
   bindings [Text_CurrentHeader {
     id=#1
     possessed_class=CommonTextBlock
     tracks [
      {
       type=MovieScene2DTransformTrack
       sections [{
         type=MovieScene2DTransformSection
         end_frame=36
         start_frame=0
         channels [Translation.Y {type=MovieSceneScriptingFloatChannel, keys [[0,0],[36,250]]}]
        }]
      }
      {
       type=MovieSceneFloatTrack
       display_name=Render Opacity
       sections [{type=MovieSceneFloatSection, end_frame=36, start_frame=0, channels [None {type=MovieSceneScriptingFloatChannel, keys [[0,1],[17,0]]}]}]
      }
     ]
    }]
   display_rate {numerator=120}
  }
 }
 NextHeaderSwipeInUp {
  binding_count=1
  length_seconds=0.4
  object=&self:NextHeaderSwipeInUp
  sequence {
   bindings [Text_NextHeader {
     id=#3
     possessed_class=CommonTextBlock
     tracks [
      {
       type=MovieScene2DTransformTrack
       sections [{
         type=MovieScene2DTransformSection
         end_frame=48
         start_frame=0
         channels [Translation.Y {type=MovieSceneScriptingFloatChannel, keys [[0,250],[48,0]]}]
        }]
      }
      {
       type=MovieSceneFloatTrack
       display_name=Render Opacity
       sections [{type=MovieSceneFloatSection, end_frame=48, start_frame=0, channels [None {type=MovieSceneScriptingFloatChannel, keys [[0,0],[48,1]]}]}]
      }
     ]
    }]
   display_rate {numerator=120}
  }
 }
 NextHeaderSwipeInDown {
  binding_count=1
  length_seconds=0.4
  object=&self:NextHeaderSwipeInDown
  sequence {
   bindings [Text_NextHeader {
     id=#3
     possessed_class=CommonTextBlock
     tracks [
      {
       type=MovieScene2DTransformTrack
       sections [{
         type=MovieScene2DTransformSection
         end_frame=48
         start_frame=0
         channels [Translation.Y {type=MovieSceneScriptingFloatChannel, keys [[0,-250],[48,0]]}]
        }]
      }
      {
       type=MovieSceneFloatTrack
       display_name=Render Opacity
       sections [{type=MovieSceneFloatSection, end_frame=48, start_frame=0, channels [None {type=MovieSceneScriptingFloatChannel, keys [[0,0],[48,1]]}]}]
      }
     ]
    }]
   display_rate {numerator=120}
  }
 }
 NextHeaderSwipeOutUp {
  binding_count=1
  length_seconds=0.3
  object=&self:NextHeaderSwipeOutUp
  sequence {
   bindings [Text_NextHeader {
     id=#3
     possessed_class=CommonTextBlock
     tracks [
      {
       type=MovieScene2DTransformTrack
       sections [{
         type=MovieScene2DTransformSection
         end_frame=36
         start_frame=0
         channels [Translation.Y {type=MovieSceneScriptingFloatChannel, keys [[0,0],[36,-250]]}]
        }]
      }
      {
       type=MovieSceneFloatTrack
       display_name=Render Opacity
       sections [{type=MovieSceneFloatSection, end_frame=36, start_frame=0, channels [None {type=MovieSceneScriptingFloatChannel, keys [[0,1],[36,0]]}]}]
      }
     ]
    }]
   display_rate {numerator=120}
  }
 }
 NextHeaderSwipeOutDown {
  binding_count=1
  length_seconds=0.3
  object=&self:NextHeaderSwipeOutDown
  sequence {
   bindings [Text_NextHeader {
     id=#3
     possessed_class=CommonTextBlock
     tracks [
      {
       type=MovieScene2DTransformTrack
       sections [{
         type=MovieScene2DTransformSection
         end_frame=36
         start_frame=0
         channels [Translation.Y {type=MovieSceneScriptingFloatChannel, keys [[0,0],[36,250]]}]
        }]
      }
      {
       type=MovieSceneFloatTrack
       display_name=Render Opacity
       sections [{type=MovieSceneFloatSection, end_frame=36, start_frame=0, channels [None {type=MovieSceneScriptingFloatChannel, keys [[0,1],[36,0]]}]}]
      }
     ]
    }]
   display_rate {numerator=120}
  }
 }
 TransitionNextToCurrent {
  binding_count=1
  length_seconds=0.4584
  object=&self:TransitionNextToCurrent
  sequence {
   bindings [WBP_ItemShop_HeaderManager_C_132 {
     id=#4
     possessed_class=WBP_ItemShop_HeaderManager_C
     tracks [
      {
       type=MovieSceneDoubleTrack
       display_name=Transition Anim Alpha
       sections [{type=MovieSceneDoubleSection, end_frame=55, start_frame=0, channels [None {type=MovieSceneScriptingDoubleChannel, keys [[0,0],[55,1]]}]}]
      }
      {
       type=MovieSceneEventTrack
       sections [{
         type=MovieSceneEventTriggerSection
         channels [None {type=MovieSceneScriptingEventChannel, const ["<Struct 'MovieSceneEvent' (0x4689441a280) {}>", 1]}]
        }]
      }
     ]
    }]
   display_rate {numerator=120}
  }
 }
 TransitionCurrentToNext {
  binding_count=1
  length_seconds=0.4584
  object=&self:TransitionCurrentToNext
  sequence {
   bindings [WBP_ItemShop_HeaderManager_C_132 {
     id=#4
     possessed_class=WBP_ItemShop_HeaderManager_C
     tracks [
      {
       type=MovieSceneDoubleTrack
       display_name=Transition Anim Alpha
       sections [{type=MovieSceneDoubleSection, end_frame=55, start_frame=0, channels [None {type=MovieSceneScriptingDoubleChannel, keys [[0,0],[55,1]]}]}]
      }
      {
       type=MovieSceneEventTrack
       sections [{
         type=MovieSceneEventTriggerSection
         channels [None {type=MovieSceneScriptingEventChannel, const ["<Struct 'MovieSceneEvent' (0x4689441a820) {}>", 1]}]
        }]
      }
     ]
    }]
   display_rate {numerator=120}
  }
 }
 ShrinkCurrentHeader {
  binding_count=1
  length_seconds=0.3
  object=&self:ShrinkCurrentHeader
  sequence {
   bindings [WBP_ItemShop_HeaderManager_C_132 {
     id=#4
     possessed_class=WBP_ItemShop_HeaderManager_C
     tracks [
      {
       type=MovieSceneDoubleTrack
       display_name=Transition Anim Alpha
       sections [{type=MovieSceneDoubleSection, end_frame=36, start_frame=0, channels [None {type=MovieSceneScriptingDoubleChannel, keys [[0,0],[36,1]]}]}]
      }
      {
       type=MovieSceneEventTrack
       sections [{
         type=MovieSceneEventTriggerSection
         channels [None {type=MovieSceneScriptingEventChannel, const ["<Struct 'MovieSceneEvent' (0x4689441a6e0) {}>", 1]}]
        }]
      }
     ]
    }]
   display_rate {numerator=120}
  }
 }
 EnlargeCurrentHeader {
  binding_count=1
  length_seconds=0.3
  object=&self:EnlargeCurrentHeader
  sequence {
   bindings [WBP_ItemShop_HeaderManager_C_132 {
     id=#4
     possessed_class=WBP_ItemShop_HeaderManager_C
     tracks [
      {
       type=MovieSceneDoubleTrack
       display_name=Transition Anim Alpha
       sections [{type=MovieSceneDoubleSection, end_frame=36, start_frame=0, channels [None {type=MovieSceneScriptingDoubleChannel, keys [[0,0],[36,1]]}]}]
      }
      {
       type=MovieSceneEventTrack
       sections [{
         type=MovieSceneEventTriggerSection
         channels [None {type=MovieSceneScriptingEventChannel, const ["<Struct 'MovieSceneEvent' (0x4689441a640) {}>", 1]}]
        }]
      }
     ]
    }]
   display_rate {numerator=120}
  }
 }
 FilterHeaderSwipeOutUp {
  binding_count=1
  length_seconds=0.3
  object=&self:FilterHeaderSwipeOutUp
  sequence {
   bindings [Text_FilterHeader {
     id=#5
     possessed_class=CommonTextBlock
     tracks [
      {
       type=MovieSceneFloatTrack
       display_name=Render Opacity
       sections [{type=MovieSceneFloatSection, end_frame=18, start_frame=0, channels [None {type=MovieSceneScriptingFloatChannel, keys [[0,1],[8.5,0]]}]}]
      }
      {
       type=MovieScene2DTransformTrack
       sections [{
         type=MovieScene2DTransformSection
         end_frame=18
         start_frame=0
         channels [Translation.Y {type=MovieSceneScriptingFloatChannel, keys [[0,0],[8.5,-129.7498],[18,-250]]}]
        }]
      }
     ]
    }]
   display_rate {numerator=60}
  }
 }
 FilterHeaderSwipeInDown {
  binding_count=1
  length_seconds=0.3
  object=&self:FilterHeaderSwipeInDown
  sequence {
   bindings [Text_FilterHeader {
     id=#6
     possessed_class=CommonTextBlock
     tracks [
      {
       type=MovieSceneFloatTrack
       display_name=Render Opacity
       sections [{
         type=MovieSceneFloatSection
         end_frame=17
         start_frame=0
         channels [None {type=MovieSceneScriptingFloatChannel, keys [[7.714,0],[17.999,1]]}]
        }]
      }
      {
       type=MovieScene2DTransformTrack
       sections [{
         type=MovieScene2DTransformSection
         end_frame=17
         start_frame=0
         channels [Translation.Y {type=MovieSceneScriptingFloatChannel, keys [[0,-250],[17.999,0]]}]
        }]
      }
     ]
    }]
   display_rate {numerator=60}
  }
 }
 SubheaderSwipeInUp {
  binding_count=2
  length_seconds=0.3
  object=&self:SubheaderSwipeInUp
  sequence {
   bindings [Text_CurrentSubheader {
     id=#7
     possessed_class=CommonTextBlock
     tracks [
      {
       type=MovieSceneFloatTrack
       display_name=Render Opacity
       sections [{type=MovieSceneFloatSection, end_frame=36, start_frame=0, channels [None {type=MovieSceneScriptingFloatChannel, keys [[18,0],[36,1]]}]}]
      }
      {
       type=MovieScene2DTransformTrack
       sections [{
         type=MovieScene2DTransformSection
         end_frame=36
         start_frame=0
         channels [Translation.Y {type=MovieSceneScriptingFloatChannel, keys [[18,105.7487],[36,0]]}]
        }]
      }
     ]
    }]
   display_rate {numerator=120}
  }
 }
 SubheaderSwipeInDown {
  binding_count=1
  length_seconds=0.3
  object=&self:SubheaderSwipeInDown
  sequence {
   bindings [Text_CurrentSubheader {
     id=#8
     possessed_class=CommonTextBlock
     tracks [
      {
       type=MovieSceneFloatTrack
       display_name=Render Opacity
       sections [{type=MovieSceneFloatSection, end_frame=18, start_frame=0, channels [None {type=MovieSceneScriptingFloatChannel, keys [[9,0],[18,1]]}]}]
      }
      {
       type=MovieScene2DTransformTrack
       sections [{
         type=MovieScene2DTransformSection
         end_frame=18
         start_frame=0
         channels [Translation.Y {type=MovieSceneScriptingFloatChannel, keys [[9,-111.9703],[18,0]]}]
        }]
      }
     ]
    }]
   display_rate {numerator=60}
  }
 }
 SubheaderSwipeOutUp {
  binding_count=1
  length_seconds=0.3
  object=&self:SubheaderSwipeOutUp
  sequence {bindings [Text_CurrentSubheader {
     id=#9
     possessed_class=CommonTextBlock
     tracks [
      {
       type=MovieSceneFloatTrack
       display_name=Render Opacity
       sections [{type=MovieSceneFloatSection, end_frame=6, start_frame=0, channels [None {type=MovieSceneScriptingFloatChannel, keys [[0,1],[3,0]]}]}]
      }
      {
       type=MovieScene2DTransformTrack
       sections [{
         type=MovieScene2DTransformSection
         end_frame=6
         start_frame=0
         channels [Translation.Y {type=MovieSceneScriptingFloatChannel, default=0, keys [[0,0],[3,-129.7498]]}]
        }]
      }
     ]
    }]}
 }
 SubheaderSwipeOutDown {
  binding_count=1
  length_seconds=0.3
  object=&self:SubheaderSwipeOutDown
  sequence {bindings [Text_CurrentSubheader {
     id=#10
     possessed_class=CommonTextBlock
     tracks [
      {
       type=MovieSceneFloatTrack
       display_name=Render Opacity
       sections [{type=MovieSceneFloatSection, end_frame=6, start_frame=0, channels [None {type=MovieSceneScriptingFloatChannel, keys [[0,1],[3,0]]}]}]
      }
      {
       type=MovieScene2DTransformTrack
       sections [{
         type=MovieScene2DTransformSection
         end_frame=6
         start_frame=0
         channels [Translation.Y {type=MovieSceneScriptingFloatChannel, keys [[0,0],[3,161.4482]]}]
        }]
      }
     ]
    }]}
 }
]
graphs [
 AnimateScrollDown {nodes [
   AnimateScrollDown!() -> (then, ShouldShowLargeCurrentHeader:bool=~, ShouldShowNextHeader:bool=~, ShouldShowLargeNextHeader:bool=~, ShouldShowFilterHeader:bool=~) {is=title K2Node_FunctionEntry}
   > n1 = Sequence!(execute) -> (then_0, then_1, then_2:Exec, then_3:Exec)
   > n2 = Branch(Condition=Get IsCurrentHeaderLarge() -> (IsCurrentHeaderLarge:bool))
   > n3 = Branch(Condition=Get ShouldShowLargeCurrentHeader() -> (ShouldShowLargeCurrentHeader:bool))
   > SwipeHeaderDirectional!(
    execute
    self
    TargetHeader:Common Text Object Reference=Get Text_CurrentSubheader() -> (Text_CurrentSubheader:Common Text Object Reference)
    SwipeIn:bool=true
    AnimationDirection:EScrollDirection Enum=Scroll_Up
   ) -> (then)
   > n4 = Set Text(
    Text=GetSubtitle!(self:Fort Item Shop Section VM Object Reference=Get CurrentRowSection() -> (CurrentRowSection:Fort Item Shop Section VM Object Reference)) -> (ReturnValue:Text)
    self=Get Text_CurrentSubheader()
   )
   > Call On Subheader Visibility Changed!(execute:Exec, self:WBP Item Shop Header Manager Object Reference, IsVisible:bool=NOT Boolean(A=TextIsEmpty(InText=@n4))) -> (then:Exec) {is=title K2Node_CallDelegate}
   SwipeHeaderDirectional(execute=@n3.else, TargetHeader=Get Text_CurrentSubheader(), SwipeIn=false, AnimationDirection=Scroll_Up)
   > Call On Subheader Visibility Changed()
   Branch(execute=@n2.else, Condition=Get ShouldShowLargeCurrentHeader())
   > SwipeHeaderDirectional(TargetHeader=Get Text_CurrentSubheader(), SwipeIn=true, AnimationDirection=Scroll_Up)
   > n5 = Set Text(Text=GetSubtitle(self=Get CurrentRowSection()), self=Get Text_CurrentSubheader())
   > Call On Subheader Visibility Changed(IsVisible=NOT Boolean(A=TextIsEmpty(InText=@n5)))
   n6 = Branch(execute=@n1.then_1, Condition=Get IsShowingCurrentHeader() -> (IsShowingCurrentHeader:bool))
   > n7 = Branch(Condition=Get ShouldShowLargeCurrentHeader())
   > SwipeHeaderDirectional(TargetHeader=Get Text_CurrentHeader() -> (Text_CurrentHeader:Common Text Object Reference), SwipeIn=false, AnimationDirection=Scroll_Up)
   > TransitionNextHeaderToCurrent()
   Set Header to Section Name!(
    exec=@n7.else
    TargetHeader:Text Object Reference=Get Text_CurrentHeader()
    Section:Fort Item Shop Section VM Object Reference=Get CurrentRowSection()
   ) -> (then:Exec) {is=title K2Node_MacroInstance}
   > Branch(Condition=Get IsCurrentHeaderLarge())
   > TransitionCurrentHeaderSize!(execute, self, ShowLarge:bool=false) -> (then)
   Set Header to Section Name(exec=@n6.else, TargetHeader=Get Text_CurrentHeader(), Section=Get CurrentRowSection())
   > SetCurrentHeaderSize!(execute, self, ShowLarge:bool=Get ShouldShowLargeCurrentHeader()) -> (then)
   > SwipeHeaderDirectional(TargetHeader=Get Text_CurrentHeader(), SwipeIn=true, AnimationDirection=Scroll_Up)
   n8 = Branch(execute=@n1.then_2, Condition=Get ShouldShowNextHeader() -> (ShouldShowNextHeader:bool))
   > Set Header to Section Name(
    TargetHeader=Get Text_NextHeader() -> (Text_NextHeader:Common Text Object Reference)
    Section=Get NextRowSection() -> (NextRowSection:Fort Item Shop Section VM Object Reference)
   )
   > SetNextHeaderSize!(execute, self, ShowLarge:bool=Get ShouldShowLargeNextHeader() -> (ShouldShowLargeNextHeader:bool)) -> (then)
   > SwipeHeaderDirectional(TargetHeader=Get Text_NextHeader(), SwipeIn=true, AnimationDirection=Scroll_Up)
   n9 = Get ShouldShowLargeCurrentHeader()
   n10 = Branch(execute=@n8.else, Condition=AND Boolean(A=Get IsShowingNextHeader() -> (IsShowingNextHeader:bool), B=NOT Boolean(A=@n9)))
   > StopTransitionAnimation!(
    execute
    self
    InAnimation:Widget Animation Object Reference=Get TransitionCurrentToNext() -> (TransitionCurrentToNext:Widget Animation Object Reference)
   ) -> (then)
   > SwipeHeaderDirectional(TargetHeader=Get Text_NextHeader(), SwipeIn=false, AnimationDirection=Scroll_Up)
   Reroute Node(InputPin=@n10.else)
   Branch(execute=@n1.then_3, Condition=Get IsShowingFilterHeader() -> (IsShowingFilterHeader:bool))
   > SwipeHeaderDirectional(TargetHeader=Get Text_FilterHeader() -> (Text_FilterHeader:Common Text Object Reference), SwipeIn=false, AnimationDirection=Scroll_Up)
  ]}
 AnimateScrollToBillboard {nodes [
   AnimateScrollToBillboard() {is=title K2Node_FunctionEntry}
   > n1 = Switch on EFortItemShopScrollDirection!(
    execute
    Selection:EFortItemShopScrollDirection Enum=Get LastScrolledDirection!(self:Fort Item Shop Selection VM Object Reference=Get FortItemShopSelectionVM() -> (FortItemShopSelectionVM:Fort Item Shop Selection VM Object Reference)) -> (LastScrolledDirection:EFortItemShopScrollDirection Enum=NoScroll)
   ) -> (NoScroll:Exec, ScrolledDown:Exec, ScrolledUp:Exec) {is=title K2Node_SwitchEnum}
   > SetVisibility(self=[Get Text_CurrentSubheader(), Get Text_NextHeader(), Get Text_CurrentHeader(), Get Text_FilterHeader()], InVisibility=Collapsed)
   > Set IsShowingCurrentHeader(IsShowingCurrentHeader:bool=false) -> (Output_Get:bool)
   > Set IsShowingNextHeader(IsShowingNextHeader:bool=false) -> (Output_Get:bool)
   > Set IsShowingFilterHeader(IsShowingFilterHeader:bool=false) -> (Output_Get:bool)
   > Call On Subheader Visibility Changed()
   n2 = Get FortItemShopSelectionVM()
   Set AnimationDirection(
    execute=[@n1.ScrolledDown, @n1.ScrolledUp]
    AnimationDirection:EScrollDirection Enum=Select!(
     NoScroll:EScrollDirection Enum=Scroll_Down
     ScrolledDown:EScrollDirection Enum=Scroll_Up
     ScrolledUp:EScrollDirection Enum=Scroll_Down
     Index:EFortItemShopScrollDirection Enum=Get LastScrolledDirection(self=@n2) -> (LastScrolledDirection=NoScroll)
    ) -> (ReturnValue:EScrollDirection Enum)
   ) -> (Output_Get:EScrollDirection Enum=Scroll_Down)
   > n3 = Branch(Condition=AND Boolean(A=Get IsCurrentHeaderLarge(), B=Get IsShowingCurrentHeader()))
   > SwipeHeaderDirectional(
    TargetHeader=Get Text_CurrentSubheader()
    SwipeIn=false
    AnimationDirection=Get AnimationDirection() -> (AnimationDirection:EScrollDirection Enum=Scroll_Down)
   )
   > n4 = Call On Subheader Visibility Changed()
   n5 = Branch(execute=[@n4.then, @n3.else], Condition=Get IsShowingCurrentHeader())
   > StopTransitionAnimation(InAnimation=Get TransitionNextToCurrent() -> (TransitionNextToCurrent:Widget Animation Object Reference))
   > n6 = SwipeHeaderDirectional(TargetHeader=Get Text_CurrentHeader(), SwipeIn=false, AnimationDirection=Get AnimationDirection() -> (AnimationDirection=Scroll_Down))
   n7 = Branch(execute=[@n6.then, @n5.else], Condition=Get IsShowingNextHeader())
   > StopTransitionAnimation(InAnimation=Get TransitionCurrentToNext())
   > n8 = SwipeHeaderDirectional(TargetHeader=Get Text_NextHeader(), SwipeIn=false, AnimationDirection=Get AnimationDirection() -> (AnimationDirection=Scroll_Down))
   Branch(execute=[@n8.then, @n7.else], Condition=Get IsShowingFilterHeader())
   > SwipeHeaderDirectional(TargetHeader=Get Text_FilterHeader(), SwipeIn=false, AnimationDirection=Get AnimationDirection() -> (AnimationDirection=Scroll_Down))
  ]}
 AnimateScrollUp {nodes [
   AnimateScrollUp!() -> (then, ShouldShowLargeCurrentHeader:bool=~, ShouldShowNextHeader:bool=~, ShouldShowLargeNextHeader:bool=~, ShouldShowFilterHeader:bool=~) {is=title K2Node_FunctionEntry}
   > n1 = Sequence!(execute) -> (then_0, then_1, then_2, then_3, then_4:Exec)
   > n2 = Branch(Condition=Get IsCurrentHeaderLarge())
   > n3 = Branch(Condition=Get ShouldShowLargeCurrentHeader())
   > SwipeHeaderDirectional(TargetHeader=Get Text_CurrentSubheader(), SwipeIn=true, AnimationDirection=Scroll_Down)
   > n4 = Set Text(Text=GetSubtitle(self=Get CurrentRowSection()), self=Get Text_CurrentSubheader())
   > Call On Subheader Visibility Changed(IsVisible=NOT Boolean(A=TextIsEmpty(InText=@n4)))
   SwipeHeaderDirectional(execute=@n3.else, TargetHeader=Get Text_CurrentSubheader(), SwipeIn=false, AnimationDirection=Scroll_Down)
   > Call On Subheader Visibility Changed()
   Branch(execute=@n2.else, Condition=Get ShouldShowLargeCurrentHeader())
   > SwipeHeaderDirectional(TargetHeader=Get Text_CurrentSubheader(), SwipeIn=true, AnimationDirection=Scroll_Down)
   > n5 = Set Text(Text=GetSubtitle(self=Get CurrentRowSection()), self=Get Text_CurrentSubheader())
   > Call On Subheader Visibility Changed(IsVisible=NOT Boolean(A=TextIsEmpty(InText=@n5)))
   n6 = Branch(execute=@n1.then_2, Condition=Get ShouldShowNextHeader())
   > n7 = Branch(Condition=AND Boolean(A=Get IsShowingCurrentHeader(), B=Get IsCurrentHeaderLarge()))
   > TransitionCurrentHeaderToNext!(execute, self, ShouldShowLargeNextHeader:bool=Get ShouldShowLargeNextHeader()) -> (then)
   Set Header to Section Name(exec=@n7.else, TargetHeader=Get Text_NextHeader(), Section=Get NextRowSection())
   > SetNextHeaderSize(ShowLarge=Get ShouldShowLargeNextHeader())
   > SwipeHeaderDirectional(TargetHeader=Get Text_NextHeader(), SwipeIn=true, AnimationDirection=Scroll_Down)
   StopTransitionAnimation(execute=@n6.else, InAnimation=Get TransitionCurrentToNext())
   Branch(execute=@n1.then_4, Condition=Get ShouldShowFilterHeader() -> (ShouldShowFilterHeader:bool))
   > SwipeHeaderDirectional(TargetHeader=Get Text_FilterHeader(), SwipeIn=true, AnimationDirection=Scroll_Down)
   n8 = Branch(execute=@n1.then_1, Condition=Get IsShowingNextHeader())
   > SwipeHeaderDirectional(TargetHeader=Get Text_NextHeader(), SwipeIn=false, AnimationDirection=Scroll_Down)
   Reroute Node(InputPin=@n8.else)
   Set Header to Section Name(exec=@n1.then_3, TargetHeader=Get Text_CurrentHeader(), Section=Get CurrentRowSection())
   n9 = Get IsCurrentHeaderLarge()
   > n10 = Branch(Condition=AND Boolean(A=Get IsShowingCurrentHeader(), B=NOT Boolean(A=@n9)))
   > Branch(Condition=Get ShouldShowLargeCurrentHeader())
   > TransitionCurrentHeaderSize(ShowLarge=true)
   SetCurrentHeaderSize(execute=@n10.else, ShowLarge=Get ShouldShowLargeCurrentHeader())
   > SwipeHeaderDirectional(TargetHeader=Get Text_CurrentHeader(), SwipeIn=true, AnimationDirection=Scroll_Down)
  ]}
 EventGraph {nodes [
   n1 = TransitionCurrentHeaderSize!() -> (OutputDelegate, then, ShowLarge:bool=~) {is=title K2Node_CustomEvent}
   > SetCurrentHeaderSize(ShowLarge=@n1.ShowLarge)
   > StopTransitionAnimation(InAnimation=Select!(
     Option 0:Widget Animation Object Reference=Get EnlargeCurrentHeader() -> (EnlargeCurrentHeader:Widget Animation Object Reference)
     Option 1:Widget Animation Object Reference=Get ShrinkCurrentHeader() -> (ShrinkCurrentHeader:Widget Animation Object Reference)
     Index:bool=@n1.ShowLarge
    ) -> (ReturnValue:Widget Animation Object Reference))
   > StopTransitionAnimation(InAnimation=Get TransitionCurrentToNext())
   > StopTransitionAnimation(InAnimation=Get TransitionNextToCurrent())
   > StopHeaderSwipeAnimations!(execute, self, TargetHeader:Object Reference=Get Text_CurrentHeader()) -> (then)
   > Set Visibility(Visibility=Collapsed, self=Get Text_CurrentHeader())
   > Set Header to Section Name(TargetHeader=Get Text_TransitionHeader() -> (Text_TransitionHeader:Common Text Object Reference), Section=Get CurrentRowSection())
   > n2 = Set Visibility(Visibility=SelfHitTestInvisible, self=Get Text_TransitionHeader())
   n3 = OnGrowShrinkAnimFinished() {is=title K2Node_CustomEvent}
   Play Animation with Finished Delegate!(
    execute:Exec=@n2.then
    Animation:Widget Animation Object Reference=Select(Option 0=Get ShrinkCurrentHeader(), Option 1=Get EnlargeCurrentHeader(), Index=Get IsCurrentHeaderLarge())
    OnFinishedDelegate:Delegate=@n3
   ) -> (then:Exec) {is=title K2Node_MacroInstance}
   n4 = OnCurrentToNextAnimFinished() {is=title K2Node_CustomEvent}
   n5 = Get TransitionCurrentToNext()
   > UnbindAllFromAnimationFinished!(execute, self:User Widget Object Reference, Animation:Widget Animation Object Reference=@n5) -> (then)
   > Set Visibility(Visibility=Collapsed, self=Get Text_TransitionHeader())
   > Set Header to Section Name(TargetHeader=Get Text_NextHeader(), Section=Get NextRowSection())
   > StopHeaderSwipeAnimations(TargetHeader=Get Text_NextHeader())
   > SetNextHeaderSize(ShowLarge=Get IsNextHeaderLarge() -> (IsNextHeaderLarge:bool))
   > Set Visibility(Visibility=SelfHitTestInvisible, self=Get Text_NextHeader())
   n6 = TransitionCurrentHeaderToNext!() -> (OutputDelegate, then, ShouldShowLargeNextHeader:bool=~) {is=title K2Node_CustomEvent}
   > StopTransitionAnimation(InAnimation=Get EnlargeCurrentHeader())
   > StopTransitionAnimation(InAnimation=Get ShrinkCurrentHeader())
   > StopTransitionAnimation(InAnimation=Get TransitionNextToCurrent())
   > Set IsNextHeaderLarge(IsNextHeaderLarge:bool=@n6.ShouldShowLargeNextHeader) -> (Output_Get:bool)
   > Set IsShowingNextHeader(IsShowingNextHeader=true)
   > Set Visibility(Visibility=Collapsed, self=Get Text_CurrentHeader())
   > Set Header to Section Name(TargetHeader=Get Text_TransitionHeader(), Section=Get NextRowSection())
   > Set Visibility(Visibility=SelfHitTestInvisible, self=Get Text_TransitionHeader())
   > Play Animation with Finished Delegate(Animation=@n5, OnFinishedDelegate=@n4)
   TransitionNextHeaderToCurrent() {is=title K2Node_CustomEvent}
   > StopTransitionAnimation(InAnimation=Get EnlargeCurrentHeader())
   > StopTransitionAnimation(InAnimation=Get ShrinkCurrentHeader())
   > StopTransitionAnimation(InAnimation=Get TransitionCurrentToNext())
   > Set IsShowingCurrentHeader(IsShowingCurrentHeader=true)
   > Set IsCurrentHeaderLarge(IsCurrentHeaderLarge:bool=true) -> (Output_Get:bool)
   > Set IsShowingNextHeader(IsShowingNextHeader=false)
   > Set Visibility(Visibility=Collapsed, self=Get Text_NextHeader())
   > Set Header to Section Name(TargetHeader=Get Text_TransitionHeader(), Section=Get CurrentRowSection())
   > n7 = Set Visibility(Visibility=SelfHitTestInvisible, self=Get Text_TransitionHeader())
   n8 = OnNextToCurrentAnimFinished() {is=title K2Node_CustomEvent}
   n9 = Get TransitionNextToCurrent()
   Play Animation with Finished Delegate(execute=@n7.then, Animation=@n9, OnFinishedDelegate=@n8)
   n10 = Event PreConstruct()
   > K2Node_Composite_0 = Setup Active Paddings!(execute) -> (then:Exec) {is=title K2Node_Composite}
   > Set Padding!(
    execute
    Padding:Margin Structure=Get ActiveLargeCurrentHeaderPadding() -> (ActiveLargeCurrentHeaderPadding:Margin Structure)
    self:Overlay Slot Object Reference=SlotAsOverlaySlot(Widget=Get Text_CurrentHeader())
   ) -> (then, Output_Get:Margin Structure)
   > Set Font Size!(
    execute:Exec
    Target:Text Object Reference=Get Text_CurrentHeader()
    FontSize:float=Get LargeCurrentHeaderFontSize() -> (LargeCurrentHeaderFontSize:double)
   ) -> (then:Exec) {is=title K2Node_MacroInstance}
   > Set!(
    execute
    ColorAndOpacity_SpecifiedColor:Linear Color Structure=Get LargeCurrentHeaderColor() -> (LargeCurrentHeaderColor:Linear Color Structure)
    ColorAndOpacity_ColorUseRule:ESlateColorStylingMode Enum
    self:Text Object Reference=Get Text_CurrentHeader()
   ) -> (then, Output_Get:Slate Color Structure)
   > Set Padding(
    Padding=Get ActiveSectionNextHeaderPadding() -> (ActiveSectionNextHeaderPadding:Margin Structure)
    self=SlotAsOverlaySlot(Widget=Get Text_NextHeader())
   )
   > Set Font Size(Target=Get Text_NextHeader(), FontSize=Get SectionNextHeaderFontSize() -> (SectionNextHeaderFontSize:double))
   > Set(ColorAndOpacity_SpecifiedColor=Get SectionNextHeaderColor() -> (SectionNextHeaderColor:Linear Color Structure), self=Get Text_NextHeader())
   > Set Padding(
    Padding=Get ActiveFilterHeaderPadding() -> (ActiveFilterHeaderPadding:Margin Structure)
    self=SlotAsOverlaySlot(Widget=Get Text_FilterHeader())
   )
   > n11 = Branch(Condition=@n10.IsDesignTime)
   > Set Padding(
    Padding=Get ActiveSmallCurrentHeaderPadding() -> (ActiveSmallCurrentHeaderPadding:Margin Structure)
    self=SlotAsOverlaySlot(Widget=Get Text_SmallCurrentHeader_PREVIEW_ONLY() -> (Text_SmallCurrentHeader_PREVIEW_ONLY:Common Text Object Reference))
   )
   > Set Font Size(Target=Get Text_SmallCurrentHeader_PREVIEW_ONLY(), FontSize=Get SmallCurrentHeaderFontSize() -> (SmallCurrentHeaderFontSize:double))
   > Set(
    ColorAndOpacity_SpecifiedColor=Get SmallCurrentHeaderColor() -> (SmallCurrentHeaderColor:Linear Color Structure)
    self=Get Text_SmallCurrentHeader_PREVIEW_ONLY()
   )
   > Set Padding(
    Padding=Get ActiveSubsectionNextHeaderPadding() -> (ActiveSubsectionNextHeaderPadding:Margin Structure)
    self=SlotAsOverlaySlot(Widget=Get Text_SubsectionNextHeader_PREVIEW_ONLY() -> (Text_SubsectionNextHeader_PREVIEW_ONLY:Common Text Object Reference))
   )
   > Set Font Size(Target=Get Text_SubsectionNextHeader_PREVIEW_ONLY(), FontSize=Get SubsectionNextHeaderFontSize() -> (SubsectionNextHeaderFontSize:double))
   > Set(
    ColorAndOpacity_SpecifiedColor=Get SubsectionNextHeaderColor() -> (SubsectionNextHeaderColor:Linear Color Structure)
    self=Get Text_SubsectionNextHeader_PREVIEW_ONLY()
   )
   RemoveFromParent(execute=@n11.else, self=[Get Text_SmallCurrentHeader_PREVIEW_ONLY(), Get Text_SubsectionNextHeader_PREVIEW_ONLY()])
   Event On Widget Styled() {is=title K2Node_Event}
   > Set Font Size(Target=Get Text_TransitionHeader(), FontSize=Get LargeCurrentHeaderFontSize())
   Set Visibility(execute=@n13.then, Visibility=SelfHitTestInvisible, self=Get Text_CurrentHeader())
   Set Visibility(execute=@n12.then, Visibility=SelfHitTestInvisible, self=Get Text_CurrentHeader())
   UnbindAllFromAnimationFinished(execute=@n8.then, Animation=@n9)
   > Set Visibility(Visibility=Collapsed, self=Get Text_TransitionHeader())
   > Set Header to Section Name(TargetHeader=Get Text_CurrentHeader(), Section=Get CurrentRowSection())
   > StopHeaderSwipeAnimations(TargetHeader=Get Text_CurrentHeader())
   > n12 = SetCurrentHeaderSize(ShowLarge=true)
   UnbindAllFromAnimationFinished(execute=@n14.then, Animation=Get EnlargeCurrentHeader())
   > Set Visibility(Visibility=Collapsed, self=Get Text_TransitionHeader())
   > n13 = ResetAnimState!(execute, self, TargetHeader:Common Text Object Reference=Get Text_CurrentHeader()) -> (then)
   n14 = UnbindAllFromAnimationFinished(execute=@n3.then, Animation=Get ShrinkCurrentHeader())
  ]}
 Setup Active Paddings {
  path=EventGraph.K2Node_Composite_0.Setup Active Paddings
  nodes [
   Inputs()
   > Set ActiveLargeCurrentHeaderPadding(ActiveLargeCurrentHeaderPadding:Margin Structure=Get Padding!(
     BR Padding:Margin Structure=Get LargeCurrentHeaderPadding() -> (LargeCurrentHeaderPadding:Margin Structure)
     STW Padding:Margin Structure=Get LargeCurrentHeaderPaddingSTW() -> (LargeCurrentHeaderPaddingSTW:Margin Structure)
     Mobile Padding:Margin Structure=Get LargeCurrentHeaderPaddingMobile() -> (LargeCurrentHeaderPaddingMobile:Margin Structure)
    ) -> (ReturnValue:Margin Structure) {is=title K2Node_MacroInstance}) -> (Output_Get:Margin Structure)
   > Set ActiveSmallCurrentHeaderPadding(ActiveSmallCurrentHeaderPadding:Margin Structure=Get Padding(
     BR Padding=Get SmallCurrentHeaderPadding() -> (SmallCurrentHeaderPadding:Margin Structure)
     STW Padding=Get SmallCurrentHeaderPaddingSTW() -> (SmallCurrentHeaderPaddingSTW:Margin Structure)
     Mobile Padding=Get SmallCurrentHeaderPaddingMobile() -> (SmallCurrentHeaderPaddingMobile:Margin Structure)
    )) -> (Output_Get:Margin Structure)
   > Set ActiveSectionNextHeaderPadding(ActiveSectionNextHeaderPadding:Margin Structure=Get Padding(
     BR Padding=Get SectionNextHeaderPadding() -> (SectionNextHeaderPadding:Margin Structure)
     STW Padding=Get SectionNextHeaderPaddingSTW() -> (SectionNextHeaderPaddingSTW:Margin Structure)
     Mobile Padding=Get SectionNextHeaderPaddingMobile() -> (SectionNextHeaderPaddingMobile:Margin Structure)
    )) -> (Output_Get:Margin Structure)
   > Set ActiveSubsectionNextHeaderPadding(ActiveSubsectionNextHeaderPadding:Margin Structure=Get Padding(
     BR Padding=Get SubsectionNextHeaderPadding() -> (SubsectionNextHeaderPadding:Margin Structure)
     STW Padding=Get SubsectionNextHeaderPaddingSTW() -> (SubsectionNextHeaderPaddingSTW:Margin Structure)
     Mobile Padding=Get SubsectionNextHeaderPaddingMobile() -> (SubsectionNextHeaderPaddingMobile:Margin Structure)
    )) -> (Output_Get:Margin Structure)
   > Set ActiveFilterHeaderPadding(ActiveFilterHeaderPadding:Margin Structure=Get Padding(
     BR Padding=Get FilterHeaderPadding() -> (FilterHeaderPadding:Margin Structure)
     STW Padding=Get FilterHeaderPaddingSTW() -> (FilterHeaderPaddingSTW:Margin Structure)
     Mobile Padding=Get FilterHeaderPaddingMobile() -> (FilterHeaderPaddingMobile:Margin Structure)
    )) -> (Output_Get:Margin Structure)
   > Outputs!(then:Exec)
  ]
 }
 GetDisplayTypeFromRowIndex {nodes [
   GetDisplayTypeFromRowIndex!() -> (then, RowIndex:int) {is=title K2Node_FunctionEntry}
   n1 = Get FortItemShopVM() -> (FortItemShopVM:Fort Item Shop VM Object Reference)
   > n2 = Branch(Condition=Is Valid Index!(
     TargetArray:Array of Fort Item Shop Offer Group VM Object References=Get AllOfferGroupsWithOffers!(self:Fort Item Shop VM Object Reference=@n1) -> (AllOfferGroupsWithOffers:Array of Fort Item Shop Offer Group VM Object References)
     IndexToTest:int=Get RowIndex() -> (RowIndex:int=0)
    ) -> (ReturnValue:bool))
   n3 = Get RowIndex() -> (RowIndex=0)
   n4 = Get AllOfferGroupsWithOffers(self=Get FortItemShopVM())
   > Return Node!(
    execute
    Display Type:EFortItemShopOfferDisplayType Enum=GetDisplayType!(self:Fort Item Shop Offer Group VM Object Reference="Get (a copy)"(Array:Array of Fort Item Shop Offer Group VM Object References=@n4, Dimension 1=@n3) -> (Output:Fort Item Shop Offer Group VM Object Reference)) -> (ReturnValue:EFortItemShopOfferDisplayType Enum=TileGrid)
   )
   Return Node(execute=@n2.else, Display Type=TileGrid)
  ]}
 GetFilterHeaderText {nodes [
   n1 = GetFilterHeaderText!() -> (then, IsGamemodeFilterActive:bool=~) {is=title K2Node_FunctionEntry}
   > n2 = FindCosmeticConfig!(execute, ProductTag:Gameplay Tag Structure=GetCurrentContextProductTag!() -> (ReturnValue:Gameplay Tag Structure)) -> (then, ReturnValue:Fort Product Cosmetic Config Object Reference)
   n3 = Get ProductName!(self:Fort Product Cosmetic Config Object Reference=@n2) -> (ProductName:Text)
   > Return Node!(
    execute
    Text:Text=Select(
     Option 0:Text="LOCTEXT(\"Filters Applied\")"
     Option 1:Text=Format Text!(Format="LOCTEXT(\"Items for: {Label}\")", Label:Text=@n3) -> (Result)
     Index=@n1
    ) -> (ReturnValue:Text)
   )
  ]}
 GetPadding {nodes [
   n1 = IsInSubGame!(SubGame:ESubGame Enum=Campaign) -> (ReturnValue:bool)
   n2 = Inputs!() -> (BR Padding:Margin Structure, STW Padding:Margin Structure, Mobile Padding:Margin Structure)
   Outputs!(ReturnValue:Margin Structure=Select(
     Option 0:Margin Structure=Select(Option 0=@n2, Option 1=@n2.STW Padding, Index=@n1)
     Option 1:Margin Structure=@n2.Mobile Padding
     Index=IsMobileGame()
    ) -> (ReturnValue:Margin Structure))
  ]}
 GetSubsectionNextHeaderPadding {nodes [
   Inputs!()
   n1 = Break Margin!(Margin:"Margin Structure (by ref)"=Get ActiveSubsectionNextHeaderPadding()) -> (Left:float=~, Top:float=~, Right:float=~, Bottom:float=~) {is=title K2Node_BreakStruct}
   n2 = Select(
    Option 0:double=0
    Option 1:double=80
    Index="Equal (Enum)"(
     A:EFortItemShopOfferDisplayType Enum=Get CurrentRowDisplayType() -> (CurrentRowDisplayType:EFortItemShopOfferDisplayType Enum=TileGrid)
     B:EFortItemShopOfferDisplayType Enum=ExpandableList
    )
   ) -> (ReturnValue:double)
   n3 = Select(Option 0=102, Option 1=0, Index=IsItemPreviewEnabled!(self:Fort Item Shop VM Object Reference=Get FortItemShopVM()) -> (ReturnValue:bool))
   Outputs(ReturnValue=Make Margin(Left=@n1, Top=float + float!(A=@n1.Top, B=@n3, C:double=@n2) -> (ReturnValue), Right=@n1.Right, Bottom=@n1.Bottom))
  ]}
 OnSubheaderVisibilityChanged {nodes [OnSubheaderVisibilityChanged!() -> (then, IsVisible:bool=~) {is=title K2Node_FunctionEntry}]}
 PlayAnimationWithFinishedDelegate {nodes [
   n1 = Inputs!() -> (execute:Exec, Animation:Widget Animation Object Reference, OnFinishedDelegate:Delegate)
   > UnbindAllFromAnimationFinished(Animation=@n1)
   > BindToAnimationFinished(Animation=@n1, Delegate=@n1.OnFinishedDelegate)
   > PlayAnimation(InAnimation=@n1)
   > Outputs!(then:Exec)
  ]}
 ResetAnimState {nodes [
   n1 = ResetAnimState!() -> (then, TargetHeader:Common Text Object Reference) {is=title K2Node_FunctionEntry}
   > Set RenderTransform!(execute, RenderTransform:Widget Transform Structure, self:Widget Object Reference=@n1) -> (then, Output_Get:Widget Transform Structure)
   > Set RenderOpacity!(execute, RenderOpacity:float=1, self:Widget Object Reference=@n1) -> (then, Output_Get:float)
  ]}
 Sequencer Events {nodes [
   Sequence_SetupEnlargeCurrentLerp!() -> (OutputDelegate, then, WBP_ItemShop_HeaderManager_C_132:WBP Item Shop Header Manager Object Reference) {is=title K2Node_CustomEvent}
   n1 = Get!() -> (
    ActiveSmallCurrentHeaderPadding_Left:float=~
    ActiveSmallCurrentHeaderPadding_Top:float=~
    ActiveSmallCurrentHeaderPadding_Right:float=~
    ActiveSmallCurrentHeaderPadding_Bottom:float=~
   )
   > Set!(execute, TransitionStartPos_X:double=@n1, TransitionStartPos_Y:double=@n1.ActiveSmallCurrentHeaderPadding_Top) -> (then, Output_Get:Vector 2D Structure)
   > Set TransitionStartFontScale(TransitionStartFontScale:double=float / float(A=Get SmallCurrentHeaderFontSize(), B=Get LargeCurrentHeaderFontSize())) -> (Output_Get:double)
   > Set TransitionStartColor(TransitionStartColor:Linear Color Structure=Get SmallCurrentHeaderColor()) -> (Output_Get:Linear Color Structure)
   n2 = Get!() -> (
    ActiveLargeCurrentHeaderPadding_Left:float=~
    ActiveLargeCurrentHeaderPadding_Top:float=~
    ActiveLargeCurrentHeaderPadding_Right:float=~
    ActiveLargeCurrentHeaderPadding_Bottom:float=~
   )
   > Set!(execute, TransitionEndPos_X:double=@n2, TransitionEndPos_Y:double=@n2.ActiveLargeCurrentHeaderPadding_Top) -> (then, Output_Get)
   > Set TransitionEndFontScale(TransitionEndFontScale:double=1) -> (Output_Get:double)
   > Set TransitionEndColor(TransitionEndColor:Linear Color Structure=Get LargeCurrentHeaderColor()) -> (Output_Get:Linear Color Structure)
   Sequence_SetupNextToCurrentLerp!() -> (OutputDelegate, then, WBP_ItemShop_HeaderManager_C_132:WBP Item Shop Header Manager Object Reference) {is=title K2Node_CustomEvent}
   n3 = Get Text_NextHeader()
   n4 = Get!(self:Overlay Slot Object Reference=SlotAsOverlaySlot(Widget=@n3)) -> (Padding_Left:float=~, Padding_Top:float=~, Padding_Right:float=~, Padding_Bottom:float=~)
   > Set!(execute, TransitionStartPos_X:double=@n4, TransitionStartPos_Y:double=@n4.Padding_Top) -> (then, Output_Get)
   > Set TransitionStartFontScale(TransitionStartFontScale=float / float(
     A=Get!(self:Text Object Reference=@n3) -> (
      Font_FontObject:Object Reference
      Font_FontMaterial:Object Reference
      Font_OutlineSettings:Font Outline Settings Structure
      Font_TypefaceFontName:Name
      Font_Size:float=~
      Font_LetterSpacing:int
      Font_SkewAmount:float=~
      Font_bForceMonospaced:bool=~
      Font_bMaterialIsStencil:bool=~
      Font_MonospacedWidth:float=~
     ).Font_Size
     B=Get LargeCurrentHeaderFontSize()
    ))
   > Set TransitionStartColor(TransitionStartColor=Get!(self=Get Text_NextHeader()) -> (ColorAndOpacity_SpecifiedColor:Linear Color Structure, ColorAndOpacity_ColorUseRule:ESlateColorStylingMode Enum))
   n5 = Get!() -> (
    ActiveLargeCurrentHeaderPadding_Left:float=~
    ActiveLargeCurrentHeaderPadding_Top:float=~
    ActiveLargeCurrentHeaderPadding_Right:float=~
    ActiveLargeCurrentHeaderPadding_Bottom:float=~
   )
   > Set!(execute, TransitionEndPos_X:double=@n5, TransitionEndPos_Y:double=@n5.ActiveLargeCurrentHeaderPadding_Top) -> (then, Output_Get)
   > Set TransitionEndFontScale(TransitionEndFontScale=1)
   > Set TransitionEndColor(TransitionEndColor=Get LargeCurrentHeaderColor())
   Sequence_SetupCurrentToNextLerp!() -> (OutputDelegate, then, WBP_ItemShop_HeaderManager_C_132:WBP Item Shop Header Manager Object Reference) {is=title K2Node_CustomEvent}
   n6 = Get() -> (
    ActiveLargeCurrentHeaderPadding_Left=~
    ActiveLargeCurrentHeaderPadding_Top=~
    ActiveLargeCurrentHeaderPadding_Right=~
    ActiveLargeCurrentHeaderPadding_Bottom=~
   )
   > Set!(execute, TransitionStartPos_X:double=@n6, TransitionStartPos_Y:double=@n6.ActiveLargeCurrentHeaderPadding_Top) -> (then, Output_Get)
   > Set TransitionStartFontScale(TransitionStartFontScale=1)
   > Set TransitionStartColor(TransitionStartColor=Get LargeCurrentHeaderColor())
   n7 = Select!(
    Option 0:Margin Structure=Get Subsection Next Header Padding!() -> (ReturnValue:Margin Structure) {is=title K2Node_MacroInstance}
    Option 1:Margin Structure=Get ActiveSectionNextHeaderPadding()
    Index=Get IsNextHeaderLarge()
   ) -> (ReturnValue_Left:float=~, ReturnValue_Top:float=~, ReturnValue_Right:float=~, ReturnValue_Bottom:float=~)
   > Set!(execute, TransitionEndPos_X:double=@n7, TransitionEndPos_Y:double=@n7.ReturnValue_Top) -> (then, Output_Get)
   n8 = Get SectionNextHeaderFontSize()
   n9 = Get IsNextHeaderLarge()
   n10 = Get SubsectionNextHeaderFontSize()
   > Set TransitionEndFontScale(TransitionEndFontScale=float / float(A=Select!(Option 0:double=@n10, Option 1:double=@n8, Index=@n9) -> (ReturnValue:double), B=Get LargeCurrentHeaderFontSize()))
   > Set TransitionEndColor(TransitionEndColor=Select(
     Option 0:Linear Color Structure=Get SubsectionNextHeaderColor()
     Option 1:Linear Color Structure=Get SectionNextHeaderColor()
     Index=Get IsNextHeaderLarge()
    ) -> (ReturnValue:Linear Color Structure))
   Sequence_SetupShrinkCurrentLerp!() -> (OutputDelegate, then, WBP_ItemShop_HeaderManager_C_132:WBP Item Shop Header Manager Object Reference) {is=title K2Node_CustomEvent}
   n11 = Get() -> (
    ActiveLargeCurrentHeaderPadding_Left=~
    ActiveLargeCurrentHeaderPadding_Top=~
    ActiveLargeCurrentHeaderPadding_Right=~
    ActiveLargeCurrentHeaderPadding_Bottom=~
   )
   > Set!(execute, TransitionStartPos_X:double=@n11, TransitionStartPos_Y:double=@n11.ActiveLargeCurrentHeaderPadding_Top) -> (then, Output_Get)
   > Set TransitionStartFontScale(TransitionStartFontScale=1)
   > Set TransitionStartColor(TransitionStartColor=Get LargeCurrentHeaderColor())
   n12 = Get!() -> (
    ActiveSmallCurrentHeaderPadding_Left:float=~
    ActiveSmallCurrentHeaderPadding_Top:float=~
    ActiveSmallCurrentHeaderPadding_Right:float=~
    ActiveSmallCurrentHeaderPadding_Bottom:float=~
   )
   > Set!(execute, TransitionEndPos_X:double=@n12, TransitionEndPos_Y:double=@n12.ActiveSmallCurrentHeaderPadding_Top) -> (then, Output_Get)
   > Set TransitionEndFontScale(TransitionEndFontScale=float / float(A=Get SmallCurrentHeaderFontSize(), B=Get LargeCurrentHeaderFontSize()))
   > Set TransitionEndColor(TransitionEndColor=Get SmallCurrentHeaderColor())
  ]}
 SetCurrentHeaderSize {nodes [
   n1 = SetCurrentHeaderSize!() -> (then, ShowLarge:bool=~) {is=title K2Node_FunctionEntry}
   > Set IsCurrentHeaderLarge(IsCurrentHeaderLarge=@n1)
   > Set Font Size(
    Target=Get Text_CurrentHeader()
    FontSize=Select(Option 0:float=Get SmallCurrentHeaderFontSize(), Option 1:float=Get LargeCurrentHeaderFontSize(), Index=Get IsCurrentHeaderLarge()) -> (ReturnValue:float)
   )
   > Set Padding(
    Padding=Select(
     Option 0:Margin Structure=Get ActiveSmallCurrentHeaderPadding()
     Option 1:Margin Structure=Get ActiveLargeCurrentHeaderPadding()
     Index=Get IsCurrentHeaderLarge()
    ) -> (ReturnValue:Margin Structure)
    self=SlotAsOverlaySlot(Widget=Get Text_CurrentHeader())
   )
   > Set!(
    execute
    ColorAndOpacity_SpecifiedColor:Linear Color Structure=Select(
     Option 0:Linear Color Structure=Get SmallCurrentHeaderColor()
     Option 1:Linear Color Structure=Get LargeCurrentHeaderColor()
     Index=Get IsCurrentHeaderLarge()
    ) -> (ReturnValue:Linear Color Structure)
    ColorAndOpacity_ColorUseRule:ESlateColorStylingMode Enum
    self:Text Object Reference=Get Text_CurrentHeader()
   ) -> (then, Output_Get:Slate Color Structure)
   > ResetAnimState(TargetHeader=Get Text_CurrentHeader()) {is=title K2Node_CallFunction}
  ]}
 SetFontSize {nodes [
   n1 = Inputs!() -> (execute, Target:Text Object Reference, FontSize:float=~)
   n2 = Get!(self:Text Object Reference=@n1) -> (
    Font_FontObject:Object Reference
    Font_FontMaterial:Object Reference
    Font_OutlineSettings:Font Outline Settings Structure
    Font_TypefaceFontName:Name
    Font_Size:float=~
    Font_LetterSpacing:int
    Font_SkewAmount:float=~
    Font_bForceMonospaced:bool=~
    Font_bMaterialIsStencil:bool=~
    Font_MonospacedWidth:float=~
   )
   > Set!(
    execute
    Font_FontObject:Object Reference=@n2
    Font_FontMaterial:Object Reference=@n2.Font_FontMaterial
    Font_OutlineSettings:Font Outline Settings Structure=@n2.Font_OutlineSettings
    Font_TypefaceFontName:Name=@n2.Font_TypefaceFontName
    Font_Size:float=float * float(A=@n1.FontSize, B=0.75) {is=title K2Node_PromotableOperator}
    Font_LetterSpacing:int=@n2.Font_LetterSpacing
    Font_SkewAmount:float=@n2.Font_SkewAmount
    Font_bForceMonospaced:bool=@n2.Font_bForceMonospaced
    Font_bMaterialIsStencil:bool=false
    Font_MonospacedWidth:float=@n2.Font_MonospacedWidth
    self=@n1
   ) -> (then, Output_Get:Slate Font Info Structure)
   > Outputs()
  ]}
 SetHeaderToSectionName {nodes [
   n1 = Inputs!() -> (exec:Exec, TargetHeader:Text Object Reference, Section:Fort Item Shop Section VM Object Reference)
   > n2 = Is Valid(InputObject=@n1.Section)
   n3 = Set Text(execute=@n2.Is Not Valid, self=@n1)
   Outputs(then=[@n4.then, @n3.then])
   n4 = Set Text(execute=@n2.Is Valid, Text=Get DisplayName!(self:Fort Item Shop Section VM Object Reference=@n1.Section) -> (DisplayName:Text), self=@n1)
  ]}
 SetNextHeaderSize {nodes [
   n1 = SetNextHeaderSize!() -> (then, ShowLarge:bool=~) {is=title K2Node_FunctionEntry}
   > Set IsNextHeaderLarge(IsNextHeaderLarge=@n1)
   > Set Font Size(
    Target=Get Text_NextHeader()
    FontSize=Select(Option 0:float=Get SubsectionNextHeaderFontSize(), Option 1:float=Get SectionNextHeaderFontSize(), Index=Get IsNextHeaderLarge()) -> (ReturnValue:float)
   )
   > Set Padding(
    Padding=Select(
     Option 0:Margin Structure=Get Subsection Next Header Padding()
     Option 1:Margin Structure=Get ActiveSectionNextHeaderPadding()
     Index=Get IsNextHeaderLarge()
    ) -> (ReturnValue:Margin Structure)
    self=SlotAsOverlaySlot(Widget=Get Text_NextHeader())
   )
   > Set!(
    execute
    ColorAndOpacity_SpecifiedColor:Linear Color Structure=Select(
     Option 0:Linear Color Structure=Get SubsectionNextHeaderColor()
     Option 1:Linear Color Structure=Get SectionNextHeaderColor()
     Index=Get IsNextHeaderLarge()
    ) -> (ReturnValue:Linear Color Structure)
    ColorAndOpacity_ColorUseRule:ESlateColorStylingMode Enum
    self=Get Text_NextHeader()
   ) -> (then, Output_Get:Slate Color Structure)
   > ResetAnimState(TargetHeader=Get Text_NextHeader())
  ]}
 SetTransitionAnimAlpha {nodes [
   n1 = SetTransitionAnimAlpha!() -> (then, Value:double=~) {is=title K2Node_FunctionEntry}
   n2 = Get TransitionStartFontScale() -> (TransitionStartFontScale:double)
   n3 = Get TransitionEndFontScale() -> (TransitionEndFontScale:double)
   > Set!(
    execute
    RenderTransform_Translation:Vector 2D Structure=Lerp Vector 2D!(
     A:Vector 2D Structure=Get TransitionStartPos() -> (TransitionStartPos:Vector 2D Structure)
     B:Vector 2D Structure=Get TransitionEndPos() -> (TransitionEndPos:Vector 2D Structure)
     Alpha:double=@n1
    ) -> (ReturnValue:Vector 2D Structure)
    RenderTransform_Scale:Vector 2D Structure=To Vector2D!(InDouble:double=Lerp(A=@n2, B=@n3, Alpha=@n1)) -> (ReturnValue:Vector 2D Structure)
    RenderTransform_Shear:Vector 2D Structure=(X=0,Y=0)
    RenderTransform_Angle:float=0
    self:Widget Object Reference=Get Text_TransitionHeader()
   ) -> (then, Output_Get:Widget Transform Structure)
   > Set!(
    execute
    ColorAndOpacity_SpecifiedColor:Linear Color Structure="Lerp (LinearColor)"(
     A=Get TransitionStartColor() -> (TransitionStartColor:Linear Color Structure)
     B=Get TransitionEndColor() -> (TransitionEndColor:Linear Color Structure)
     Alpha=Get Value() -> (Value:double)
    )
    ColorAndOpacity_ColorUseRule:ESlateColorStylingMode Enum
    self:Text Object Reference=Get Text_TransitionHeader()
   ) -> (then, Output_Get:Slate Color Structure)
  ]}
 StopHeaderSwipeAnimations {nodes [
   StopHeaderSwipeAnimations!() -> (then, TargetHeader:Object Reference) {is=title K2Node_FunctionEntry}
   > n1 = Branch(Condition="Equal (Object)"(A=Get Text_CurrentHeader(), B=Get TargetHeader() -> (TargetHeader:Object Reference)))
   > StopAnimation(InAnimation=Get CurrentHeaderSwipeInUp() -> (CurrentHeaderSwipeInUp:Widget Animation Object Reference))
   > StopAnimation(InAnimation=Get CurrentHeaderSwipeInDown() -> (CurrentHeaderSwipeInDown:Widget Animation Object Reference))
   > StopAnimation(InAnimation=Get CurrentHeaderSwipeOutUp() -> (CurrentHeaderSwipeOutUp:Widget Animation Object Reference))
   > StopAnimation(InAnimation=Get CurrentHeaderSwipeOutDown() -> (CurrentHeaderSwipeOutDown:Widget Animation Object Reference))
   > ResetAnimState(TargetHeader=Get Text_CurrentHeader())
   n2 = Branch(execute=@n1.else, Condition="Equal (Object)"(A=Get Text_NextHeader(), B=Get TargetHeader()))
   > StopAnimation(InAnimation=Get NextHeaderSwipeInUp() -> (NextHeaderSwipeInUp:Widget Animation Object Reference))
   > StopAnimation(InAnimation=Get NextHeaderSwipeInDown() -> (NextHeaderSwipeInDown:Widget Animation Object Reference))
   > StopAnimation(InAnimation=Get NextHeaderSwipeOutUp() -> (NextHeaderSwipeOutUp:Widget Animation Object Reference))
   > StopAnimation(InAnimation=Get NextHeaderSwipeOutDown() -> (NextHeaderSwipeOutDown:Widget Animation Object Reference))
   > ResetAnimState(TargetHeader=Get Text_NextHeader())
   n3 = Branch(execute=@n2.else, Condition="Equal (Object)"(A=Get Text_FilterHeader(), B=Get TargetHeader()))
   > StopAnimation(InAnimation=Get FilterHeaderSwipeInDown() -> (FilterHeaderSwipeInDown:Widget Animation Object Reference))
   > StopAnimation(InAnimation=Get FilterHeaderSwipeOutUp() -> (FilterHeaderSwipeOutUp:Widget Animation Object Reference))
   > ResetAnimState(TargetHeader=Get Text_FilterHeader())
   Branch(execute=@n3.else, Condition="Equal (Object)"(A=Get Text_CurrentSubheader(), B=Get TargetHeader()))
   > StopAnimation(InAnimation=Get SubheaderSwipeInUp() -> (SubheaderSwipeInUp:Widget Animation Object Reference))
   > StopAnimation(InAnimation=Get SubheaderSwipeInDown() -> (SubheaderSwipeInDown:Widget Animation Object Reference))
   > StopAnimation(InAnimation=Get SubheaderSwipeOutUp() -> (SubheaderSwipeOutUp:Widget Animation Object Reference))
   > StopAnimation(InAnimation=Get SubheaderSwipeOutDown() -> (SubheaderSwipeOutDown:Widget Animation Object Reference))
   > ResetAnimState(TargetHeader=Get Text_CurrentSubheader())
  ]}
 StopTransitionAnimation {nodes [
   n1 = StopTransitionAnimation!() -> (then, InAnimation:Widget Animation Object Reference) {is=title K2Node_FunctionEntry}
   > Branch(Condition=IsAnimationPlaying(InAnimation=@n1))
   > n2 = Branch(Condition="Equal (Object)"(A=Get InAnimation() -> (InAnimation:Widget Animation Object Reference), B=Get TransitionNextToCurrent()))
   > n3 = Set Header to Section Name(TargetHeader=Get Text_CurrentHeader(), Section=Get CurrentRowSection())
   UnbindAllFromAnimationFinished(execute=[@n5.then, @n3.then, @n10.then], Animation=Get InAnimation())
   > StopAnimation(InAnimation=Get InAnimation())
   > Set Visibility(Visibility=Collapsed, self=Get Text_TransitionHeader())
   n4 = Branch(execute=@n2.else, Condition="Equal (Object)"(A=Get InAnimation(), B=Get TransitionCurrentToNext()))
   > n5 = Set Header to Section Name(TargetHeader=Get Text_NextHeader(), Section=Get NextRowSection())
   n6 = Get InAnimation()
   n7 = Get EnlargeCurrentHeader()
   n8 = Get InAnimation()
   n9 = Get ShrinkCurrentHeader()
   Branch(execute=@n4.else, Condition=OR Boolean(A="Equal (Object)"(A=@n8, B=@n9), B="Equal (Object)"(A=@n6, B=@n7)))
   > n10 = Set Visibility(Visibility=SelfHitTestInvisible, self=Get Text_CurrentHeader())
  ]}
 SwipeHeaderDirectional {nodes [
   SwipeHeaderDirectional!() -> (then, TargetHeader:Common Text Object Reference, SwipeIn:bool=~, AnimationDirection:EScrollDirection Enum=Scroll_Down) {is=title K2Node_FunctionEntry}
   n1 = Get TargetHeader() -> (TargetHeader:Common Text Object Reference)
   n2 = Get Text_NextHeader()
   n3 = Get Text_CurrentSubheader()
   n4 = Get Text_CurrentHeader()
   n5 = Get Text_FilterHeader()
   > n6 = Branch(Condition=AND Boolean!(
     A="Not Equal (Object)"(A:Object Reference=@n4, B:Object Reference=@n1) -> (ReturnValue:bool=~) {is=title K2Node_PromotableOperator}
     B="Not Equal (Object)"(A=@n1, B=@n2) -> (ReturnValue=~)
     C:bool="Not Equal (Object)"(A=@n1, B=@n5) -> (ReturnValue=~)
     D:bool="Not Equal (Object)"(A=@n1, B=@n3) -> (ReturnValue=~)
    ) -> (ReturnValue))
   > Return Node!(execute)
   StopHeaderSwipeAnimations(execute=@n6.else, TargetHeader=Get TargetHeader()) {is=title K2Node_CallFunction}
   > Set Visibility(Visibility=SelfHitTestInvisible, self=Get TargetHeader())
   n7 = Get TargetHeader()
   n8 = Get Text_FilterHeader()
   n9 = "Equal (Object)"(A=Get Text_CurrentSubheader(), B=Get TargetHeader())
   n10 = Get AnimationDirection() -> (AnimationDirection=Scroll_Down)
   n11 = Get CurrentHeaderSwipeOutDown()
   n12 = Get CurrentHeaderSwipeOutUp()
   n13 = Get CurrentHeaderSwipeInUp()
   n14 = Get CurrentHeaderSwipeInDown()
   n15 = Get NextHeaderSwipeInDown()
   n16 = Get NextHeaderSwipeInUp()
   n17 = Get NextHeaderSwipeOutDown()
   n18 = Get NextHeaderSwipeOutUp()
   n19 = Select(
    Option 0:Widget Animation Object Reference=Select(
     Option 0=Select!(Scroll_Down:Widget Animation Object Reference=@n11, Scroll_Up:Widget Animation Object Reference=@n12, Index:EScrollDirection Enum=@n10) -> (ReturnValue)
     Option 1=Select(Scroll_Down=@n14, Scroll_Up=@n13, Index=@n10)
     Index=Get SwipeIn() -> (SwipeIn:bool)
    )
    Option 1:Widget Animation Object Reference=Select!(
     Option 0:Widget Animation Object Reference=Select!(Scroll_Down:Widget Animation Object Reference=@n17, Scroll_Up:Widget Animation Object Reference=@n18, Index:EScrollDirection Enum=@n10) -> (ReturnValue)
     Option 1:Widget Animation Object Reference=Select(Scroll_Down=@n15, Scroll_Up=@n16, Index=@n10)
     Index:bool=Get SwipeIn()
    ) -> (ReturnValue)
    Index="Equal (Object)"(A=Get Text_NextHeader(), B=Get TargetHeader())
   ) -> (ReturnValue:Widget Animation Object Reference)
   n20 = Get AnimationDirection() -> (AnimationDirection=Scroll_Down)
   n21 = Select!(
    Option 0:Widget Animation Object Reference=Select!(
     Scroll_Down:Widget Animation Object Reference=Get SubheaderSwipeOutDown()
     Scroll_Up:Widget Animation Object Reference=Get SubheaderSwipeOutUp()
     Index:EScrollDirection Enum=@n20
    ) -> (ReturnValue)
    Option 1:Widget Animation Object Reference=Select(Scroll_Down=Get SubheaderSwipeInDown(), Scroll_Up=Get SubheaderSwipeInUp(), Index=@n20)
    Index:bool=Get SwipeIn()
   ) -> (ReturnValue)
   n22 = Get FilterHeaderSwipeOutUp()
   n23 = Get SwipeIn()
   n24 = Get FilterHeaderSwipeInDown()
   > PlayAnimationForward(InAnimation=Select!(
     Option 0:Widget Animation Object Reference=Select(Option 0=@n19, Option 1=@n21, Index=@n9)
     Option 1:Widget Animation Object Reference=Select(Option 0=@n22, Option 1=@n24, Index=@n23)
     Index:bool="Equal (Object)"(A=@n8, B=@n7)
    ) -> (ReturnValue))
   > n25 = Branch(Condition="Equal (Object)"(A=Get Text_NextHeader(), B=Get TargetHeader()))
   > Set IsShowingNextHeader(IsShowingNextHeader=Get SwipeIn())
   n26 = Branch(execute=@n25.else, Condition="Equal (Object)"(A=Get Text_CurrentHeader(), B=Get TargetHeader()))
   > Set IsShowingCurrentHeader(IsShowingCurrentHeader=Get SwipeIn())
   Set IsShowingFilterHeader(execute=@n26.else, IsShowingFilterHeader=Get SwipeIn())
  ]}
 UpdateCachedSections {nodes [
   UpdateCachedSections() {is=title K2Node_FunctionEntry}
   > n1 = Sequence()
   n2 = Get FortItemShopVM()
   n3 = Get FortItemShopSelectionVM()
   > Set CurrentRowIndex(CurrentRowIndex:int=Find Item!(
     TargetArray:Array of Fort Item Shop Offer Group VM Object References=Get AllOfferGroupsWithOffers(self=@n2)
     ItemToFind:"Fort Item Shop Offer Group VM Object Reference (by ref)"=Get SelectedOfferGroup!(self:Fort Item Shop Selection VM Object Reference=@n3) -> (SelectedOfferGroup:Fort Item Shop Offer Group VM Object Reference)
    ) -> (ReturnValue)) -> (Output_Get:int=0)
   n4 = Get SelectedOfferGroup(self=Get FortItemShopSelectionVM())
   > Set CurrentRowSection(CurrentRowSection:Fort Item Shop Section VM Object Reference=GetSection!(
     self:Fort Item Shop VM Object Reference=Get FortItemShopVM()
     SectionId:String=Get!(self:Fort Item Shop Offer Group VM Object Reference=@n4) -> (OwningLayout_SectionId:String, OwningLayout_SubsectionId:String)
     bSearchFromSectionsWithOffers:bool=true
    ) -> (ReturnValue:Fort Item Shop Section VM Object Reference)) -> (Output_Get:Fort Item Shop Section VM Object Reference)
   n5 = int + int(A=Get CurrentRowIndex() -> (CurrentRowIndex:int=0))
   n6 = Get FortItemShopVM()
   n7 = Branch(execute=@n1.then_1, Condition=Is Valid Index(TargetArray=Get AllOfferGroupsWithOffers(self=@n6), IndexToTest=@n5))
   n8 = "Get (a copy)"(Array=Get AllOfferGroupsWithOffers(self=Get FortItemShopVM()), Dimension 1=@n5)
   > Set NextRowSection(NextRowSection:Fort Item Shop Section VM Object Reference=GetSection(self=Get FortItemShopVM(), SectionId=Get(self=@n8), bSearchFromSectionsWithOffers=true)) -> (Output_Get:Fort Item Shop Section VM Object Reference)
   Set NextRowSection(execute=@n7.else)
   Set CurrentRowDisplayType(
    execute=@n1.then_2
    CurrentRowDisplayType:EFortItemShopOfferDisplayType Enum=GetDisplayTypeFromRowIndex!(self, RowIndex:int=Get CurrentRowIndex() -> (CurrentRowIndex=0)) -> (Display Type:EFortItemShopOfferDisplayType Enum=TileGrid) {is=title K2Node_CallFunction}
   ) -> (Output_Get:EFortItemShopOfferDisplayType Enum=TileGrid)
   n9 = Get CurrentRowIndex() -> (CurrentRowIndex=0)
   > Set PrevRowDisplayType(PrevRowDisplayType:EFortItemShopOfferDisplayType Enum=GetDisplayTypeFromRowIndex(RowIndex=int - int(A=@n9)) -> (Display Type=TileGrid)) -> (Output_Get:EFortItemShopOfferDisplayType Enum=TileGrid)
   n10 = Get CurrentRowIndex() -> (CurrentRowIndex=0)
   > Set NextRowDisplayType(NextRowDisplayType:EFortItemShopOfferDisplayType Enum=GetDisplayTypeFromRowIndex(RowIndex=int + int(A=@n10)) -> (Display Type=TileGrid)) -> (Output_Get:EFortItemShopOfferDisplayType Enum=TileGrid)
  ]}
 UpdateHeaders {nodes [
   UpdateHeaders() {is=title K2Node_FunctionEntry}
   > n1 = Get!(execute:Exec) -> (then:Exec, else:Exec, FortItemShopSelectionVM:Fort Item Shop Selection VM Object Reference)
   > n2 = Get!(execute, self:Fort Item Shop Selection VM Object Reference=@n1) -> (then, else, SelectedOfferGroup:Fort Item Shop Offer Group VM Object Reference)
   > UpdateCachedSections() {is=title K2Node_CallFunction}
   n3 = Get SelectedOfferGroup(self=Get FortItemShopSelectionVM())
   > n4 = Branch(Condition="Equal (Enum)"(A=GetDisplayType(self=@n3) -> (ReturnValue=TileGrid), B=Billboard))
   > AnimateScrollToBillboard() {is=title K2Node_CallFunction}
   n5 = Sequence!(execute=@n4.else) -> (then_0, then_1, then_2, then_3)
   n6 = Switch on EFortItemShopScrollDirection(execute=@n5.then_3, Selection=Get LastScrolledDirection(self=Get FortItemShopSelectionVM()) -> (LastScrolledDirection=NoScroll))
   n7 = Get ShouldShowFilterHeader()
   n8 = Get ShouldShowLargeCurrentHeader()
   n9 = Get ShouldShowNextHeader()
   n10 = Get ShouldShowLargeNextHeader()
   > UpdateHeadersImmediate!(
    execute
    self
    ShouldShowLargeCurrentHeader:bool=@n8
    ShouldShowNextHeader:bool=@n9
    ShouldShowLargeNextHeader:bool=@n10
    ShouldShowFilterHeader:bool=@n7
   ) -> (then)
   AnimateScrollDown!(
    execute=@n6.ScrolledDown
    self
    ShouldShowLargeCurrentHeader:bool=@n8
    ShouldShowNextHeader:bool=@n9
    ShouldShowLargeNextHeader:bool=@n10
    ShouldShowFilterHeader:bool=@n7
   ) -> (then) {is=title K2Node_CallFunction}
   AnimateScrollUp!(
    execute=@n6.ScrolledUp
    self
    ShouldShowLargeCurrentHeader:bool=@n8
    ShouldShowNextHeader:bool=@n9
    ShouldShowLargeNextHeader:bool=@n10
    ShouldShowFilterHeader:bool=@n7
   ) -> (then) {is=title K2Node_CallFunction}
   n11 = IsFirstOfferGroupInSection!(
    self:Fort Item Shop VM Object Reference=Get FortItemShopVM()
    OfferGroup:Fort Item Shop Offer Group VM Object Reference=Get SelectedOfferGroup(self=Get FortItemShopSelectionVM())
   ) -> (ReturnValue:bool)
   n12 = Get!(self=Get FortItemShopSelectionVM()) -> (
    SelectedRow_OfferGroup:Fort Item Shop Offer Group VM Object Reference
    SelectedRow_Offers:Array of Fort Item Shop Offer VM Object References
    SelectedRow_bIsFirstRowInOfferGroup:bool=~
    SelectedRow_bIsLastRowInOfferGroup:bool=~
   )
   n13 = Get PrevRowDisplayType() -> (PrevRowDisplayType:EFortItemShopOfferDisplayType Enum=TileGrid)
   Set ShouldShowLargeCurrentHeader(
    execute=@n5.then_0
    ShouldShowLargeCurrentHeader:bool=OR Boolean(A="Equal (Enum)"(A=@n13, B=Billboard), B=AND Boolean!(A=@n11, B=@n12.SelectedRow_bIsFirstRowInOfferGroup) -> (ReturnValue))
   ) -> (Output_Get:bool)
   n14 = Get FortItemShopVM()
   n15 = Get SelectedOfferGroup(self=Get FortItemShopSelectionVM())
   n16 = Get NextRowSection()
   n17 = Get(self=Get FortItemShopSelectionVM()) -> (SelectedRow_bIsFirstRowInOfferGroup=~, SelectedRow_bIsLastRowInOfferGroup=~)
   Set ShouldShowNextHeader(
    execute=@n5.then_1
    ShouldShowNextHeader:bool=AND Boolean!(
     A=IsValid(Object=@n16)
     B=IsLastOfferGroupInSection!(self:Fort Item Shop VM Object Reference=@n14, OfferGroup:Fort Item Shop Offer Group VM Object Reference=@n15) -> (ReturnValue:bool)
     C:bool=@n17.SelectedRow_bIsLastRowInOfferGroup
    ) -> (ReturnValue)
   ) -> (Output_Get:bool)
   n18 = IsLastOfferGroupInCategory!(self:Fort Item Shop VM Object Reference=@n14, OfferGroup:Fort Item Shop Offer Group VM Object Reference=@n15) -> (ReturnValue:bool)
   n19 = Get NextRowDisplayType() -> (NextRowDisplayType:EFortItemShopOfferDisplayType Enum=TileGrid)
   > Set ShouldShowLargeNextHeader(ShouldShowLargeNextHeader:bool=OR Boolean(A="Equal (Enum)"(A=@n19, B=Billboard), B=AND Boolean!(A=@n18, B=@n17.SelectedRow_bIsLastRowInOfferGroup) -> (ReturnValue))) -> (Output_Get:bool)
   n20 = "Get (a copy)"(Array=Get AllOfferGroupsWithOffers(self=Get FortItemShopVM()))
   n21 = Get SelectedOfferGroup(self=Get FortItemShopSelectionVM())
   n22 = Get FilterVM!(self:Fort Item Shop VM Object Reference=Get FortItemShopVM()) -> (FilterVM:Fort Item Shop Filter VM Object Reference)
   Set ShouldShowFilterHeader(
    execute=@n5.then_2
    ShouldShowFilterHeader:bool=AND Boolean!(
     A=@n17.SelectedRow_bIsFirstRowInOfferGroup
     B=Get bIsActive!(self:Fort Item Shop Filter VM Object Reference=@n22) -> (bIsActive:bool)
     C:bool="Equal (Object)"(A=@n21, B=@n20)
    ) -> (ReturnValue)
   ) -> (Output_Get:bool)
   SetVisibility(
    execute=[@n2.else, @n1.else]
    self=[Get Text_CurrentHeader(), Get Text_NextHeader(), Get Text_FilterHeader(), Get Text_TransitionHeader()]
    InVisibility=Collapsed
   )
   > Set IsShowingNextHeader(IsShowingNextHeader=false)
   > Set IsShowingCurrentHeader(IsShowingCurrentHeader=false)
   > Set IsShowingFilterHeader(IsShowingFilterHeader=false)
  ]}
 UpdateHeadersImmediate {nodes [
   UpdateHeadersImmediate!() -> (then, ShouldShowLargeCurrentHeader:bool=~, ShouldShowNextHeader:bool=~, ShouldShowLargeNextHeader:bool=~, ShouldShowFilterHeader:bool=~) {is=title K2Node_FunctionEntry}
   > Set Header to Section Name(TargetHeader=Get Text_CurrentHeader(), Section=Get CurrentRowSection())
   > SetCurrentHeaderSize(ShowLarge=Get ShouldShowLargeCurrentHeader()) {is=title K2Node_CallFunction}
   > Set Visibility(Visibility=SelfHitTestInvisible, self=Get Text_CurrentHeader())
   > Set IsShowingCurrentHeader(IsShowingCurrentHeader=true)
   > Set Header to Section Name(TargetHeader=Get Text_NextHeader(), Section=Get NextRowSection())
   > ResetAnimState(TargetHeader=Get Text_NextHeader())
   > SetNextHeaderSize(ShowLarge=Get ShouldShowLargeNextHeader()) {is=title K2Node_CallFunction}
   n1 = Get ShouldShowNextHeader()
   > Set Visibility(
    Visibility=Select(Option 0:ESlateVisibility Enum=Collapsed, Option 1:ESlateVisibility Enum=SelfHitTestInvisible, Index=@n1) -> (ReturnValue:ESlateVisibility Enum)
    self=Get Text_NextHeader()
   )
   > Set IsShowingNextHeader(IsShowingNextHeader=@n1)
   > ResetAnimState(TargetHeader=Get Text_FilterHeader())
   > Set Visibility(Visibility=Select(Option 0=Collapsed, Option 1=SelfHitTestInvisible, Index=Get ShouldShowFilterHeader()), self=Get Text_FilterHeader())
   > Set IsShowingFilterHeader(IsShowingFilterHeader=Get ShouldShowFilterHeader())
   > ResetAnimState(TargetHeader=Get Text_CurrentSubheader())
   > n2 = Set Text(Text=GetSubtitle(self=Get CurrentRowSection()), self=Get Text_CurrentSubheader())
   n3 = Get ShouldShowLargeCurrentHeader()
   > Set Visibility(Visibility=Select(Option 0=Collapsed, Option 1=SelfHitTestInvisible, Index=@n3), self=Get Text_CurrentSubheader())
   n4 = TextIsEmpty(InText=@n2)
   > Call On Subheader Visibility Changed(IsVisible=AND Boolean!(A=@n3, B=NOT Boolean(A=@n4)) -> (ReturnValue))
  ]}
 VM_SelectedRowUpdated {nodes [
   VM_SelectedRowUpdated!() -> (then, SelectedRow:Fort Item Shop Row Structure) {is=title K2Node_FunctionEntry}
   > UpdateHeaders() {is=title K2Node_CallFunction}
  ]}
]
variables [
 ActiveFilterHeaderPadding {type=Margin, default=<ltrb 0 0 0 0>}
 ActiveLargeCurrentHeaderPadding {type=Margin, default=<ltrb 0 0 0 0>}
 ActiveSectionNextHeaderPadding {type=Margin, default=<ltrb 0 0 0 0>}
 ActiveSmallCurrentHeaderPadding {type=Margin, default=<ltrb 0 0 0 0>}
 ActiveSubsectionNextHeaderPadding {type=Margin, default=<ltrb 0 0 0 0>}
 CurrentRowDisplayType {type=EFortItemShopOfferDisplayType, default=TileGrid}
 CurrentRowSection {type=FortItemShopSectionVM, default=null}
 FilterHeaderPadding {type=Margin, default=<ltrb 0 129 0 0>}
 FilterHeaderPaddingMobile {type=Margin, default=<ltrb 0 129 0 0>}
 FilterHeaderPaddingSTW {type=Margin, default=<ltrb 0 129 0 0>}
 IsCurrentHeaderLarge {type=boolean}
 IsNextHeaderLarge {type=boolean}
 IsShowingCurrentHeader {type=boolean}
 IsShowingFilterHeader {type=boolean}
 IsShowingNextHeader {type=boolean}
 LargeCurrentHeaderColor {type=LinearColor, default=<rgba 1 1 1 1>}
 LargeCurrentHeaderFontSize {type=number, default=43}
 LargeCurrentHeaderPadding {type=Margin, default=<ltrb 0 181 0 0>}
 LargeCurrentHeaderPaddingMobile {type=Margin, default=<ltrb 0 152 0 0>}
 LargeCurrentHeaderPaddingSTW {type=Margin, default=<ltrb 0 51 0 0>}
 NextRowDisplayType {type=EFortItemShopOfferDisplayType, default=TileGrid}
 NextRowSection {type=FortItemShopSectionVM, default=null}
 OnSubheaderVisibilityChanged {}
 PrevRowDisplayType {type=EFortItemShopOfferDisplayType, default=TileGrid}
 SectionNextHeaderColor {type=LinearColor, default=<rgba 1 1 1 0.25>}
 SectionNextHeaderFontSize {type=number, default=27}
 SectionNextHeaderPadding {type=Margin, default=<ltrb 0 762 0 0>}
 SectionNextHeaderPaddingMobile {type=Margin, default=<ltrb 0 832 0 0>}
 SectionNextHeaderPaddingSTW {type=Margin, default=<ltrb 0 782 0 0>}
 SmallCurrentHeaderColor {type=LinearColor, default=<rgba 1 1 1 1>}
 SmallCurrentHeaderFontSize {type=number, default=27}
 SmallCurrentHeaderPadding {type=Margin, default=<ltrb 0 121 0 0>}
 SmallCurrentHeaderPaddingMobile {type=Margin, default=<ltrb 0 152 0 0>}
 SmallCurrentHeaderPaddingSTW {type=Margin, default=<ltrb 0 21 0 0>}
 SubsectionNextHeaderColor {type=LinearColor, default=<rgba 1 1 1 0.25>}
 SubsectionNextHeaderFontSize {type=number, default=27}
 SubsectionNextHeaderPadding {type=Margin, default=<ltrb 0 762 0 0>}
 SubsectionNextHeaderPaddingMobile {type=Margin, default=<ltrb 0 762 0 0>}
 SubsectionNextHeaderPaddingSTW {type=Margin, default=<ltrb 0 582 0 0>}
 TransitionAnimAlpha {type=number}
 TransitionEndColor {type=LinearColor, default=<rgba 0 0 0 1>}
 TransitionEndFontScale {type=number, default=1}
 TransitionEndPos {type=Vector2D, default=<0 0>}
 TransitionStartColor {type=LinearColor, default=<rgba 0 0 0 1>}
 TransitionStartFontScale {type=number, default=0.628}
 TransitionStartPos {type=Vector2D, default=<0 0>}
]
widgets {
 info {inheritedWidgetCount=0, parentClass=&UserWidget, rootWidgetClass=&Overlay, widgetCount=8}
 widgets [
  Overlay_13 {widgetClassPath=&Overlay, widget=Overlay_13}
  Text_FilterHeader {
   widgetClassPath=&CommonTextBlock
   bIsVariable=true
   parent=Overlay_13
   widget=Text_FilterHeader
   properties {
    mobileFontSizeMultiplier=1.2
    text=Filters Applied
    textTransformPolicy=ToUpper
    visibility=Collapsed
    font {
     bForceMonospaced=false
     bMaterialIsStencil=false
     fontMaterial=null
     fontObject=&HeadingNow
     letterSpacing=0
     monospacedWidth=1
     size=15.75
     skewAmount=0
     typefaceFontName=75 Medium
     outlineSettings {
      bApplyOutlineToDropShadows=false
      bMiteredCorners=false
      bSeparateFillAlpha=false
      outlineMaterial=null
      outlineSize=0
      outlineColor=<rgba 0 0 0 1>
     }
    }
    renderTransformPivot=<0 0>
   }
   slot {class=OverlaySlot, properties {padding=<ltrb 0 129 0 0>}}
  }
  Text_SmallCurrentHeader_PREVIEW_ONLY {
   widgetClassPath=&CommonTextBlock
   bIsVariable=true
   parent=Overlay_13
   widget=Text_SmallCurrentHeader_PREVIEW_ONLY
   properties {
    mobileFontSizeMultiplier=1.2
    text=Small Current Header
    textTransformPolicy=ToUpper
    visibility=Collapsed
    font {
     bForceMonospaced=false
     bMaterialIsStencil=false
     fontMaterial=null
     fontObject=&HeadingNow
     letterSpacing=0
     monospacedWidth=1
     size=20.25
     skewAmount=0
     typefaceFontName=86 Bold Italic
     outlineSettings {
      bApplyOutlineToDropShadows=false
      bMiteredCorners=false
      bSeparateFillAlpha=false
      outlineMaterial=null
      outlineSize=0
      outlineColor=<rgba 0 0 0 1>
     }
    }
    renderTransformPivot=<0 0>
   }
   slot {class=OverlaySlot, properties {padding=<ltrb 0 121 0 0>}}
  }
  Text_CurrentHeader {
   widgetClassPath=&CommonTextBlock
   bIsVariable=true
   parent=Overlay_13
   widget=Text_CurrentHeader
   properties {
    mobileFontSizeMultiplier=1.2
    text=Large Current Header
    textTransformPolicy=ToUpper
    font {
     bForceMonospaced=false
     bMaterialIsStencil=false
     fontMaterial=null
     fontObject=&HeadingNow
     letterSpacing=0
     monospacedWidth=1
     size=32.25
     skewAmount=0
     typefaceFontName=86 Bold Italic
     outlineSettings {
      bApplyOutlineToDropShadows=false
      bMiteredCorners=false
      bSeparateFillAlpha=false
      outlineMaterial=null
      outlineSize=0
      outlineColor=<rgba 0 0 0 1>
     }
    }
    renderTransformPivot=<0 0>
   }
   slot {class=OverlaySlot, properties {padding=<ltrb 0 164 0 0>}}
  }
  Text_CurrentSubheader {
   widgetClassPath=&CommonTextBlock
   bIsVariable=true
   parent=Overlay_13
   widget=Text_CurrentSubheader
   properties {
    mobileFontSizeMultiplier=1.2
    text=Subheader for the current section
    colorAndOpacity {colorUseRule=UseColor_Specified, specifiedColor=<rgba 0.6 0.6 0.6 1>}
    font {
     bForceMonospaced=false
     bMaterialIsStencil=false
     fontMaterial=null
     fontObject=&HeadingNow
     letterSpacing=0
     monospacedWidth=1
     size=18
     skewAmount=0
     typefaceFontName=86 Bold Italic
     outlineSettings {
      bApplyOutlineToDropShadows=false
      bMiteredCorners=false
      bSeparateFillAlpha=false
      outlineMaterial=null
      outlineSize=0
      outlineColor=<rgba 0 0 0 1>
     }
    }
    renderTransformPivot=<0 0>
   }
   slot {class=OverlaySlot, properties {padding=<ltrb 0 230 0 0>}}
  }
  Text_SubsectionNextHeader_PREVIEW_ONLY {
   widgetClassPath=&CommonTextBlock
   bIsVariable=true
   parent=Overlay_13
   widget=Text_SubsectionNextHeader_PREVIEW_ONLY
   properties {
    mobileFontSizeMultiplier=1.2
    text=Small Next Header
    textTransformPolicy=ToUpper
    visibility=Collapsed
    font {
     bForceMonospaced=false
     bMaterialIsStencil=false
     fontMaterial=null
     fontObject=&HeadingNow
     letterSpacing=0
     monospacedWidth=1
     size=20.25
     skewAmount=0
     typefaceFontName=86 Bold Italic
     outlineSettings {
      bApplyOutlineToDropShadows=false
      bMiteredCorners=false
      bSeparateFillAlpha=false
      outlineMaterial=null
      outlineSize=0
      outlineColor=<rgba 0 0 0 1>
     }
    }
    renderTransformPivot=<0 0>
   }
   slot {class=OverlaySlot, properties {padding=<ltrb 0 751 0 0>}}
  }
  Text_NextHeader {
   widgetClassPath=&CommonTextBlock
   bIsVariable=true
   parent=Overlay_13
   widget=Text_NextHeader
   properties {
    mobileFontSizeMultiplier=1.2
    text=Large Next Header
    textTransformPolicy=ToUpper
    font {
     bForceMonospaced=false
     bMaterialIsStencil=false
     fontMaterial=null
     fontObject=&HeadingNow
     letterSpacing=0
     monospacedWidth=1
     size=20.25
     skewAmount=0
     typefaceFontName=86 Bold Italic
     outlineSettings {
      bApplyOutlineToDropShadows=false
      bMiteredCorners=false
      bSeparateFillAlpha=false
      outlineMaterial=null
      outlineSize=0
      outlineColor=<rgba 0 0 0 1>
     }
    }
    renderTransformPivot=<0 1>
   }
   slot {class=OverlaySlot, properties {padding=<ltrb 0 851 0 0>}}
  }
  Text_TransitionHeader {
   widgetClassPath=&CommonTextBlock
   bIsVariable=true
   parent=Overlay_13
   widget=Text_TransitionHeader
   properties {
    mobileFontSizeMultiplier=1.2
    text=Transition Header
    textTransformPolicy=ToUpper
    visibility=Collapsed
    font {
     bForceMonospaced=false
     bMaterialIsStencil=false
     fontMaterial=null
     fontObject=&HeadingNow
     letterSpacing=0
     monospacedWidth=1
     size=32.25
     skewAmount=0
     typefaceFontName=86 Bold Italic
     outlineSettings {
      bApplyOutlineToDropShadows=false
      bMiteredCorners=false
      bSeparateFillAlpha=false
      outlineMaterial=null
      outlineSize=0
      outlineColor=<rgba 0 0 0 1>
     }
    }
    renderTransformPivot=<0 0>
   }
   slot {class=OverlaySlot}
  }
 ]
}
