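# Skeleton placeholder row for the multi-product item shop, apparently shown while shop content loads.
# A horizontal box holds three textureless dark-gray images, one wide and two narrower, spaced like product tiles.
# On initialization it plays the Pulse animation with zero loop count, so the row's render opacity fades between dim and brighter endlessly.
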WidgetBlueprint /ItemShop/MultiProduct/WBP_MPItemShop_FakeLoadingRow
animations [Pulse {
  binding_count=1
  length_seconds=2
  object=&self:Pulse
  sequence {bindings [HBox_Row {
     id=#1
     possessed_class=HorizontalBox
     tracks [{
       type=MovieSceneFloatTrack
       display_name=Render Opacity
       sections [{type=MovieSceneFloatSection, channels [None {type=MovieSceneScriptingFloatChannel, keys [[0,0.3],[20,0.7],[40,0.3]]}]}]
      }]
    }]}
 }]
graphs [EventGraph {nodes [
   Event OnInitialized()
   > PlayAnimation(InAnimation=Get Pulse() -> (Pulse:Widget Animation Object Reference), NumLoopsToPlay=0)
  ]}]
variables []
widgets {
 info {inheritedWidgetCount=0, parentClass=&UserWidget, rootWidgetClass=&HorizontalBox, widgetCount=4}
 widgets [
  HBox_Row {widgetClassPath=&HorizontalBox, widget=HBox_Row}
  Image_Fake1 {
   widgetClassPath=&Image
   parent=HBox_Row
   widget=Image_Fake1
   properties {brush {
     bIsDynamicallyLoaded=false
     drawAs=Image
     imageType=NoImage
     mirroring=NoMirror
     resourceName=None
     resourceObject=null
     tiling=NoTile
     imageSize=<561.6 468>
     margin=<ltrb 0 0 0 0>
     outlineSettings {
      bUseBrushTransparency=false
      roundingType=HalfHeightRadius
      width=0
      color {colorUseRule=UseColor_Specified, specifiedColor=<rgba 0 0 0 0>}
      cornerRadii=<0 0 0 0>
     }
     tintColor {colorUseRule=UseColor_Specified, specifiedColor=<rgba 0.1562 0.1562 0.1562 1>}
     uVRegion {bIsValid=false, max=<0 0>, min=<0 0>}
    }}
   slot {class=HorizontalBoxSlot}
  }
  Image_Fake2 {
   widgetClassPath=&Image
   parent=HBox_Row
   widget=Image_Fake2
   properties {brush {
     bIsDynamicallyLoaded=false
     drawAs=Image
     imageType=NoImage
     mirroring=NoMirror
     resourceName=None
     resourceObject=null
     tiling=NoTile
     imageSize=<270.4 468>
     margin=<ltrb 0 0 0 0>
     outlineSettings {
      bUseBrushTransparency=false
      roundingType=HalfHeightRadius
      width=0
      color {colorUseRule=UseColor_Specified, specifiedColor=<rgba 0 0 0 0>}
      cornerRadii=<0 0 0 0>
     }
     tintColor {colorUseRule=UseColor_Specified, specifiedColor=<rgba 0.1562 0.1562 0.1562 1>}
     uVRegion {bIsValid=false, max=<0 0>, min=<0 0>}
    }}
   slot {class=HorizontalBoxSlot, properties {padding=<ltrb 16 0 0 0>}}
  }
  Image_Fake3 {
   widgetClassPath=&Image
   parent=HBox_Row
   widget=Image_Fake3
   properties {brush {
     bIsDynamicallyLoaded=false
     drawAs=Image
     imageType=NoImage
     mirroring=NoMirror
     resourceName=None
     resourceObject=null
     tiling=NoTile
     imageSize=<270.4 468>
     margin=<ltrb 0 0 0 0>
     outlineSettings {
      bUseBrushTransparency=false
      roundingType=HalfHeightRadius
      width=0
      color {colorUseRule=UseColor_Specified, specifiedColor=<rgba 0 0 0 0>}
      cornerRadii=<0 0 0 0>
     }
     tintColor {colorUseRule=UseColor_Specified, specifiedColor=<rgba 0.1562 0.1562 0.1562 1>}
     uVRegion {bIsValid=false, max=<0 0>, min=<0 0>}
    }}
   slot {class=HorizontalBoxSlot, properties {padding=<ltrb 16 0 0 0>}}
  }
 ]
}
