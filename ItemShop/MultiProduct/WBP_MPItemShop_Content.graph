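# Item shop content widget that shows offer rows in a scroll box grid and handles row-by-row scrolling and navigation.
# ScrollToIndex scrolls to a row, sets row states, syncs the featured layout to the shop view model and adjusts clipping and bottom fade.
# Row wrappers report focus and list-view requests through rebound delegates.
# The focused offer is restored on activation, and a section-to-row map drives jumps when the featured layout changes.
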WidgetBlueprint /ItemShop/MultiProduct/WBP_MPItemShop_Content
graphs [
 BP_GetDesiredFocusTarget {nodes [
   Get Desired Focus Target()
   > n1 = GetCurrentRowWidget!(execute, self) -> (then, RowWidget:WBP MPItem Shop Row Wrapper Object Reference)
   > Return Node!(execute, ReturnValue:Widget Object Reference=@n1)
  ]}
 EventGraph {nodes [
   "On Hovered (Button_Hitbox)"() {is=title K2Node_ComponentBoundEvent}
   > n1 = Branch(Condition=Get IsContentFocused() -> (IsContentFocused:bool))
   Call On Content Focused!(execute:Exec=@n1.else, self:WBP MPItem Shop Content Object Reference) -> (then:Exec) {is=title K2Node_CallDelegate}
   n2 = "On Subheader Visibility Changed (WBP_ItemShop_HeaderManager)"!() -> (OutputDelegate, then, IsVisible:bool) {is=title K2Node_ComponentBoundEvent}
   > Set Border Padding!(execute:Exec, IsSubheaderVisible:bool=@n2.IsVisible) -> (then:Exec) {is=title K2Node_MacroInstance}
   n3 = OnOfferTileFocused!() -> (OutputDelegate, then, OfferId:String) {is=title K2Node_CustomEvent}
   > n4 = Set FocusedOfferId(FocusedOfferId:String=@n3.OfferId) -> (Output_Get:String)
   > Call On Offer Focused!(execute:Exec, self:WBP MPItem Shop Content Object Reference, OfferId:String=@n4) -> (then:Exec) {is=title K2Node_CallDelegate}
   > Set FocusedLayout(FocusedLayout:String=Get!(self:Fort Item Shop VM Object Reference=Get FortItemShopVM() -> (FortItemShopVM:Fort Item Shop VM Object Reference)) -> (FeaturedLayout_SectionId:String, FeaturedLayout_SubsectionId:String)) -> (Output_Get:String)
   n5 = OnRowFocused!() -> (OutputDelegate, then, RowWidget:User Widget Object Reference) {is=title K2Node_CustomEvent}
   > ScrollToIndex!(
    execute
    self
    DestinationIndex:int=GetRowIndexFromWidget!(self, RowWidget:Widget Object Reference=@n5.RowWidget) -> (RowIndex:int=0)
    bRefreshAll:bool=false
   ) -> (then)
   > n6 = Branch(Condition=Get IsContentFocused())
   Call On Content Focused(execute=@n6.else)
   OnSoftDeactivation() {is=title K2Node_CustomEvent}
   > n7 = GetCurrentRowWidget()
   > Is Valid(InputObject=@n7)
   > I_SetRowState!(execute, self:I MPItem Shop Row Interface=@n7, InRowState:E_ItemShop_RowState Enum=NewEnumerator4) -> (then)
   n8 = OnSoftActivation!() -> (OutputDelegate, then, bRowsDirty:bool=~) {is=title K2Node_CustomEvent}
   > n9 = Set bRowsDirty(bRowsDirty:bool=@n8.bRowsDirty) -> (Output_Get:bool)
   n10 = Get FortItemShopVM()
   n11 = Branch(
    execute=[@n14.then, @n9.then]
    Condition=IsEmpty(InString=Get FeaturedOfferId!(self:Fort Item Shop VM Object Reference=@n10) -> (FeaturedOfferId:String))
   )
   ScrollToFocusedRowIndex!(execute=[@n12.then, @n11.then], self, bForceRefreshAll:bool=false) -> (then)
   > Set bRowsDirty(bRowsDirty=false)
   > UpdateRowFocusedOffer!(execute, self, RowIndex:int=GetFocusedRowIndex!(self) -> (RowIndex:int=0)) -> (then)
   > UpdateHeaders(self:WBP Item Shop Header Manager Object Reference=Get WBP_ItemShop_HeaderManager() -> (WBP_ItemShop_HeaderManager:WBP Item Shop Header Manager Object Reference))
   n12 = UpdateFeaturedOffer(execute=@n11.else)
   n13 = Handle_OnRowListViewRequested!() -> (OutputDelegate, then, OfferGroupVM:Fort Item Shop Offer Group VM Object Reference) {is=title K2Node_CustomEvent}
   > Call On List View Requested!(execute:Exec, self:WBP MPItem Shop Content Object Reference, OfferGroupVM:Fort Item Shop Offer Group VM Object Reference=@n13.OfferGroupVM) -> (then:Exec) {is=title K2Node_CallDelegate}
   Event On Activated()
   > n14 = Set bRowsDirty(bRowsDirty=true)
   Event On Deactivated()
   > n15 = GetCurrentRowWidget()
   > n16 = Is Valid(InputObject=@n15)
   > n17 = I_SetRowState(self=@n15, InRowState=NewEnumerator4)
   ClearRows(execute=[@n17.then, @n16.Is Not Valid], self:Fort MPItem Shop Content Object Reference)
   > Set FeaturedOfferId!(
    execute
    FeaturedOfferId:String=Get FocusedOfferId() -> (FocusedOfferId:String)
    self:Fort Item Shop VM Object Reference=Get FortItemShopVM()
   ) -> (then, Output_Get:String)
   n18 = Event OnScrollComplete!() -> (OutputDelegate, then, DestinationIndex:int=0) {is=title K2Node_Event}
   > NotifyScrollComplete!(execute, self, DestinationIndex:int=@n18.DestinationIndex) -> (then)
   > Call On Scroll Finished!(execute:Exec, self:WBP MPItem Shop Content Object Reference) -> (then:Exec) {is=title K2Node_CallDelegate}
   Event OnAddedToFocusPath!() -> (OutputDelegate, then, InFocusEvent:Focus Event Structure) {is=title K2Node_Event}
   > Call On Content Focused()
   n19 = Event OnResetRow!() -> (OutputDelegate, then, RowWidget:User Widget Object Reference) {is=title K2Node_Event}
   > n20 = Cast To WBP_MPItemShop_RowWrapper!(execute, Object=@n19.RowWidget) -> (then, CastFailed, AsWBP MPItem Shop Row Wrapper:WBP MPItem Shop Row Wrapper Object Reference) {is=title K2Node_DynamicCast}
   > I_ResetRow(self:WBP MPItem Shop Row Wrapper Object Reference=@n20)
   Event OnLoadedRowWidgetsUpdated!() -> (OutputDelegate, then, DestinationIndex:int=0) {is=title K2Node_Event}
   > UpdateFakeLoadingRowsVisibility()
   n21 = Event OnScrollInput!() -> (OutputDelegate, then, ScrollAmount:float) {is=title K2Node_Event}
   > HandleScroll!(execute, self, OffsetAmount:"Float (double-precision) (by ref)"=@n21.ScrollAmount) -> (then)
   Event Construct()
   > n22 = Sequence()
   > SetMobileStyle() {is=title K2Node_CallFunction}
   > Branch(Condition=IsInSubGame!(SubGame:ESubGame Enum=Campaign) -> (ReturnValue:bool))
   > Set members in Margin!(
    execute:Exec
    StructRef:"Margin Structure (by ref)"=Get DefaultBorderOverlaySlotPadding() -> (DefaultBorderOverlaySlotPadding:Margin Structure)
    Top:float=120
   ) -> (then:Exec, StructOut:"Margin Structure (by ref)") {is=title K2Node_SetFieldsInStruct}
   > Set members in Margin(StructRef=Get DefaultBorderPadding() -> (DefaultBorderPadding:Margin Structure), Top=-120)
   Set Border Padding(execute=@n22.then_1)
  ]}
 GetCurrentRowWidget {nodes [
   GetCurrentRowWidget() {is=title K2Node_FunctionEntry}
   > n1 = For Each Loop(Array:Array of Widget Object References=GetAllChildren(self=Get GridPanel_Content() -> (GridPanel_Content:Grid Panel Object Reference))) -> (Array Element:Widget Object Reference)
   > Branch(Condition="Equal (Integer)"(A=GetRowIndexFromWidget(RowWidget=@n1) -> (RowIndex=0), B=Get CurrentRowIndex() -> (CurrentRowIndex:int=0)))
   > n2 = Cast To WBP_MPItemShop_RowWrapper(Object=@n1)
   > Return Node!(execute, RowWidget:WBP MPItem Shop Row Wrapper Object Reference=@n2)
  ]}
 GetFocusedRowIndex {nodes [
   GetFocusedRowIndex() {is=title K2Node_FunctionEntry}
   n1 = Get FortItemShopVM()
   n2 = Get FortItemShopSelectionVM() -> (FortItemShopSelectionVM:Fort Item Shop Selection VM Object Reference)
   > n3 = Branch(Condition=AND Boolean(A=IsValid(Object=@n1), B=IsValid(Object=@n2)))
   n4 = Get FortItemShopSelectionVM()
   n5 = GetRowFromOffer!(
    self:Fort Item Shop Selection VM Object Reference=@n4
    Offer:Fort Item Shop Offer VM Object Reference=GetOfferFromId!(self:Fort Item Shop VM Object Reference=Get FortItemShopVM(), OfferId:String=Get FocusedOfferId(), bSearchFromCatalog:bool=false) -> (ReturnValue:Fort Item Shop Offer VM Object Reference)
    PreferredSection:String=Get FocusedLayout() -> (FocusedLayout:String)
   ) -> (ReturnValue:Fort Item Shop Row Structure)
   > Return Node!(
    execute
    RowIndex:int="Max (Integer)"(
     A=GetIndexForRow!(self:Fort Item Shop Selection VM Object Reference=@n4, Row:"Fort Item Shop Row Structure (by ref)"=@n5) -> (ReturnValue:int=0)
     B=0
    )
   )
   Return Node(execute=@n3.else, RowIndex=0)
  ]}
 GetRowIndexFromWidget {nodes [
   n1 = GetRowIndexFromWidget!() -> (then, RowWidget:Widget Object Reference) {is=title K2Node_FunctionEntry}
   n2 = SlotAsGridSlot(Widget=@n1)
   > Return Node(RowIndex=int + int(A=Get Row!(self:Grid Slot Object Reference=@n2) -> (Row:int=0), B=Get FirstRowIndex() -> (FirstRowIndex:int=0)))
  ]}
 GetRowWidgetByRowIndex {nodes [
   GetRowWidgetByRowIndex!() -> (then, RowIndex:int) {is=title K2Node_FunctionEntry}
   > n1 = For Each Loop(Array=GetAllChildren(self=Get GridPanel_Content()))
   Return Node!(execute=@n1.Completed, RowWidget:WBP MPItem Shop Row Wrapper Object Reference)
   > Branch(Condition="Equal (Integer)"(A=GetRowIndexFromWidget(RowWidget=@n1) -> (RowIndex=0) {is=title K2Node_CallFunction}, B=Get RowIndex() -> (RowIndex:int=0)))
   > n2 = Cast To WBP_MPItemShop_RowWrapper(Object=@n1)
   > Return Node(RowWidget=@n2)
  ]}
 GetScrollDestination {nodes [
   n1 = GetScrollDestination!() -> (then, TargetWidget:Widget Object Reference) {is=title K2Node_FunctionEntry}
   > n2 = Cast To WBP_MPItemShop_RowWrapper(Object=@n1)
   > n3 = I_GetDesiredScrollDestination!(execute, self:WBP MPItem Shop Row Wrapper Object Reference=@n2) -> (then, Destination:EDescendantScrollDestination Enum=IntoView)
   > Return Node!(execute, ReturnValue:EDescendantScrollDestination Enum=@n3)
   Return Node(execute=@n2.CastFailed, ReturnValue=TopOrLeft)
  ]}
 GetScrollPadding {nodes [
   n1 = GetScrollPadding!() -> (then, TargetWidget:Widget Object Reference) {is=title K2Node_FunctionEntry}
   > n2 = Cast To WBP_MPItemShop_RowWrapper(Object=@n1)
   > n3 = I_GetDesiredScrollPadding!(execute, self:WBP MPItem Shop Row Wrapper Object Reference=@n2) -> (then, ScrollPadding:double)
   > Return Node(ReturnValue:float=@n3)
   Return Node(execute=@n2.CastFailed, ReturnValue=0)
  ]}
 GetScrollTarget {nodes [
   n1 = GetScrollTarget!() -> (then, TargetWidget:Widget Object Reference) {is=title K2Node_FunctionEntry}
   > n2 = Cast To WBP_MPItemShop_RowWrapper(Object=@n1)
   > n3 = I_GetScrollTarget!(execute, self:WBP MPItem Shop Row Wrapper Object Reference=@n2) -> (then, ScrollTarget:Widget Object Reference)
   > n4 = Is Valid(InputObject=@n3)
   Return Node(execute=[@n4.Is Not Valid, @n2.CastFailed], ReturnValue:Widget Object Reference=Get TargetWidget() -> (TargetWidget:Widget Object Reference))
   > Return Node(ReturnValue=@n3)
  ]}
 GetShopSelectionVM {nodes [
   GetShopSelectionVM() {is=title K2Node_FunctionEntry}
   > Return Node(ReturnValue:Fort Item Shop Selection VM Object Reference=Get FortItemShopSelectionVM())
  ]}
 HandleScroll {nodes [
   HandleScroll!() -> (then, OffsetAmount:"Float (double-precision) (by ref)") {is=title K2Node_FunctionEntry}
   > n1 = Branch(Condition=IsUsingTouch(PlayerController=GetOwningPlayer()))
   > n2 = Sequence()
   > Set BeforeScrollRowIndex(BeforeScrollRowIndex:int=Get CurrentRowIndex() -> (CurrentRowIndex=0)) -> (Output_Get:int=0)
   n3 = Get OffsetAmount() -> (OffsetAmount:"Float (double-precision) (by ref)"=0)
   > Branch(Condition="float >= float"(A="Absolute (Float)"(A=@n3), B=Get TouchScrollThreshold() -> (TouchScrollThreshold:float)))
   > n4 = Branch(Condition="float > float"(A=Get OffsetAmount() -> (OffsetAmount=0)))
   > ScrollUp()
   ScrollDown(execute=@n4.else)
   Branch(
    execute=@n2.then_1
    Condition="Equal (Integer)"(A=Get CurrentRowIndex() -> (CurrentRowIndex=0), B=Get BeforeScrollRowIndex() -> (BeforeScrollRowIndex:int=0))
   )
   > EndInertialScrolling(self:Scroll Box Object Reference=Get ScrollBox_Content() -> (ScrollBox_Content:Athena Scroll Box Object Reference))
   > ScrollToCenterCurrentRow!(execute, self:Fort MPItem Shop Content Object Reference, bAnimateScroll:bool=true) -> (then)
   n5 = Branch(execute=@n1.else, Condition="float > float"(A=Get OffsetAmount() -> (OffsetAmount=0)))
   > ScrollDown()
   ScrollUp(execute=@n5.else)
  ]}
 IsQuickNavCollapsible {nodes [
   IsQuickNavCollapsible() {is=title K2Node_FunctionEntry}
   n1 = IsUsingTouch(PlayerController=GetOwningPlayer())
   n2 = Get FortItemShopVM()
   > Return Node(ReturnValue:bool=OR Boolean(A=NOT Boolean(A=@n1), B=IsItemPreviewEnabled!(self:Fort Item Shop VM Object Reference=@n2) -> (ReturnValue:bool)))
  ]}
 "Layout !=" {nodes [
   n1 = Inputs!() -> (Lhs:Fort Item Shop Layout Structure, Rhs:"Fort Item Shop Layout Structure (by ref)")
   n2 = Make Array("[0]":Fort Item Shop Layout Structure=@n1) -> (Array:Array of Fort Item Shop Layout Structures)
   Outputs!(bNotEqual:bool=NOT Boolean(A=Contains Item(TargetArray:Array of Fort Item Shop Layout Structures=@n2, ItemToFind:"Fort Item Shop Layout Structure (by ref)"=@n1.Rhs)))
  ]}
 NotifyScrollComplete {nodes [
   NotifyScrollComplete!() -> (then, DestinationIndex:int) {is=title K2Node_FunctionEntry}
   > n1 = For Each Loop(Array=GetAllChildren(self=Get GridPanel_Content()))
   > n2 = Cast To WBP_MPItemShop_RowWrapper(Object=@n1)
   > I_OnScrollComplete!(execute, self:I MPItem Shop Row Interface=@n2, DestinationIndex:int=Get DestinationIndex() -> (DestinationIndex:int=0)) -> (then)
  ]}
 OnContentFocused {nodes [OnContentFocused() {is=title K2Node_FunctionEntry}]}
 OnFeaturedLayoutUpdated {nodes [
   OnFeaturedLayoutUpdated!() -> (then, FeaturedLayout:Fort Item Shop Layout Structure) {is=title K2Node_FunctionEntry}
   > UpdateSectionIndexMap()
   > n1 = Branch(Condition=Get bIgnoreNextFeaturedLayoutUpdate() -> (bIgnoreNextFeaturedLayoutUpdate:bool))
   > Set bIgnoreNextFeaturedLayoutUpdate(bIgnoreNextFeaturedLayoutUpdate:bool=false) -> (Output_Get:bool)
   n2 = Find(
    TargetMap:Map of Strings to Integers=Get SectionIndexMap() -> (SectionIndexMap:Map of Strings to Integers)
    Key=Get!() -> (FeaturedLayout_SectionId, FeaturedLayout_SubsectionId)
   ) -> (Value:int)
   Branch(execute=@n1.else, Condition=@n2.ReturnValue)
   > ScrollToIndex(DestinationIndex=@n2, bRefreshAll=false)
  ]}
 OnInitializeRow {nodes [
   OnInitializeRow!() -> (then, RowWidget:User Widget Object Reference, RowIndex:int=0) {is=title K2Node_FunctionEntry}
   > n1 = Get!(execute:Exec) -> (then:Exec, else:Exec, FortItemShopSelectionVM:Fort Item Shop Selection VM Object Reference)
   > Branch(Condition="In Range (Integer)"!(
     Value:int=Get RowIndex() -> (RowIndex=0)
     Min:int=0
     Max:int=GetNumRows!(self:Fort Item Shop Selection VM Object Reference=@n1) -> (ReturnValue:int=0)
     InclusiveMin:bool=true
     InclusiveMax:bool=false
    ) -> (ReturnValue:bool))
   > n2 = Cast To WBP_MPItemShop_RowWrapper(Object=Get RowWidget() -> (RowWidget:User Widget Object Reference))
   > n3 = Set RowWrapperRef(RowWrapperRef:WBP MPItem Shop Row Wrapper Object Reference=@n2) -> (Output_Get:WBP MPItem Shop Row Wrapper Object Reference)
   > I_InitializeRow!(
    execute
    self:I MPItem Shop Row Interface=@n3
    FortItemShopRow:Fort Item Shop Row Structure=GetRowAtIndex!(self:Fort Item Shop Selection VM Object Reference=Get FortItemShopSelectionVM(), RowIndex:int=Get RowIndex() -> (RowIndex=0)) -> (ReturnValue:Fort Item Shop Row Structure)
    ContentHasFocus:bool=Get IsContentFocused()
   ) -> (then)
   n4 = Create Event()
   n5 = Get RowWrapperRef() -> (RowWrapperRef:WBP MPItem Shop Row Wrapper Object Reference)
   > Unbind Event from On Row Focused(self:WBP MPItem Shop Row Wrapper Object Reference=@n5, Delegate=@n4) {is=title K2Node_RemoveDelegate}
   > Bind Event to On Row Focused(self:WBP MPItem Shop Row Wrapper Object Reference=@n5, Delegate=@n4) {is=title K2Node_AddDelegate}
   n6 = Create Event()
   n7 = Get RowWrapperRef()
   > Unbind Event from On Offer Focused(self:WBP MPItem Shop Row Wrapper Object Reference=@n7, Delegate=@n6) {is=title K2Node_RemoveDelegate}
   > Bind Event to On Offer Focused(self:WBP MPItem Shop Row Wrapper Object Reference=@n7, Delegate=@n6) {is=title K2Node_AddDelegate}
   n8 = Create Event()
   n9 = Get RowWrapperRef()
   > Unbind Event from On List View Requested(self:WBP MPItem Shop Row Wrapper Object Reference=@n9, Delegate=@n8) {is=title K2Node_RemoveDelegate}
   > Bind Event to On List View Requested(self:WBP MPItem Shop Row Wrapper Object Reference=@n9, Delegate=@n8) {is=title K2Node_AddDelegate}
  ]}
 OnListViewRequested {nodes [OnListViewRequested!() -> (then, OfferGroupVM:Fort Item Shop Offer Group VM Object Reference) {is=title K2Node_FunctionEntry}]}
 OnOfferFocused {nodes [OnOfferFocused!() -> (then, OfferId:String) {is=title K2Node_FunctionEntry}]}
 OnRowNavigatedFrom {nodes [
   n1 = OnRowNavigatedFrom!() -> (then, Direction:EUINavigation Enum, FocusedTilePosition:Vector 2D Structure) {is=title K2Node_FunctionEntry}
   n2 = Get CurrentRowIndex() -> (CurrentRowIndex=0)
   > n3 = GetRowWidgetByRowIndex!(
    execute
    self
    RowIndex:int=Select!(
     Left:int=0
     Right:int=0
     Up:int=int - int(A=@n2)
     Down:int=int + int(A=@n2)
     Next:int=0
     Previous:int=0
     Invalid:int=0
     Index:EUINavigation Enum=@n1
    ) -> (ReturnValue:int)
   ) -> (then, RowWidget:WBP MPItem Shop Row Wrapper Object Reference) {is=title K2Node_CallFunction}
   > Is Valid(InputObject=@n3)
   > n4 = I_GetClosestFocusTarget!(
    execute
    self:WBP MPItem Shop Row Wrapper Object Reference=@n3
    TargetLocation:Vector 2D Structure=Get FocusedTilePosition() -> (FocusedTilePosition:Vector 2D Structure)
   ) -> (then, Target:Widget Object Reference)
   > SetFocus(self=@n4)
  ]}
 OnScrollFinished {nodes [OnScrollFinished() {is=title K2Node_FunctionEntry}]}
 OnScrollStarted {nodes [OnScrollStarted() {is=title K2Node_FunctionEntry}]}
 RequestScroll {nodes [
   n1 = RequestScroll!() -> (then, Direction:EScrollDirection Enum=Scroll_Down) {is=title K2Node_FunctionEntry}
   > n2 = Switch on EScrollDirection!(execute, Selection:EScrollDirection Enum=@n1) -> (Scroll_Down:Exec, Scroll_Up:Exec) {is=title K2Node_SwitchEnum}
   > ScrollDown()
   ScrollUp(execute=@n2.Scroll_Up)
  ]}
 ScrollDown {nodes [
   ScrollDown() {is=title K2Node_FunctionEntry}
   n1 = int + int(A=Get CurrentRowIndex() -> (CurrentRowIndex=0))
   n2 = GetNumRows(self=Get FortItemShopSelectionVM()) -> (ReturnValue=0)
   > Branch(Condition=AND Boolean(A="integer < integer"(A=@n1, B=@n2), B=true))
   > ScrollToIndex(DestinationIndex=int + int(A=Get CurrentRowIndex() -> (CurrentRowIndex=0)), bRefreshAll=false)
   Reroute Node(InputPin:bool=HasLoadedRowsForAnimatedScroll!(self:Fort MPItem Shop Content Object Reference, DestinationIndex:int=@n1, NumRows:int=@n2) -> (ReturnValue:bool)) -> (OutputPin:bool)
  ]}
 ScrollToFocusedRowIndex {nodes [
   ScrollToFocusedRowIndex!() -> (then, bForceRefreshAll:bool=~) {is=title K2Node_FunctionEntry}
   > ScrollToIndex(
    DestinationIndex=GetFocusedRowIndex() -> (RowIndex=0) {is=title K2Node_CallFunction}
    bRefreshAll=OR Boolean(A=Get bRowsDirty() -> (bRowsDirty:bool), B=Get bForceRefreshAll() -> (bForceRefreshAll:bool))
   )
  ]}
 ScrollToIndex {nodes [
   ScrollToIndex!() -> (then, DestinationIndex:int, bRefreshAll:bool=~) {is=title K2Node_FunctionEntry}
   > n1 = Get!(execute) -> (then, else, FortItemShopVM:Fort Item Shop VM Object Reference)
   > Branch(Condition="Equal (Enum)"(
     A:EFortItemShopLoadingState Enum=Get LoadingState!(self:Fort Item Shop VM Object Reference=@n1) -> (LoadingState:EFortItemShopLoadingState Enum=Uninitialized)
     B:EFortItemShopLoadingState Enum=Ready
    ))
   > Call On Scroll Started!(execute:Exec, self:WBP MPItem Shop Content Object Reference) -> (then:Exec) {is=title K2Node_CallDelegate}
   > n2 = Set CurrentRowIndex(CurrentRowIndex:int=Get DestinationIndex() -> (DestinationIndex=0)) -> (Output_Get:int=0)
   > ScrollToRowIndex!(
    execute
    self:Fort MPItem Shop Content Object Reference
    DestinationIndex:int=@n2
    NumRows:int=GetNumRows(self=Get FortItemShopSelectionVM()) -> (ReturnValue=0)
    bRefreshAll:bool=Get bRefreshAll() -> (bRefreshAll:bool)
   ) -> (then)
   > SetRowStates()
   n3 = GetRowAtIndex!(self=Get FortItemShopSelectionVM(), RowIndex=Get CurrentRowIndex() -> (CurrentRowIndex=0)) -> (
    ReturnValue_OfferGroup:Fort Item Shop Offer Group VM Object Reference
    ReturnValue_Offers:Array of Fort Item Shop Offer VM Object References
    ReturnValue_bIsFirstRowInOfferGroup:bool=~
    ReturnValue_bIsLastRowInOfferGroup:bool=~
   )
   > Is Valid(InputObject=@n3)
   > Update Featured Layout!(execute, self, Target:Fort Item Shop Offer Group VM Object Reference=@n3) -> (then)
   > Set Clipping!(
    execute
    Clipping:EWidgetClipping Enum=Select!(
     TileGrid:EWidgetClipping Enum=ClipToBounds
     ExpandableList:EWidgetClipping Enum=ClipToBounds
     Billboard:EWidgetClipping Enum=Inherit
     ScrollableList:EWidgetClipping Enum=Inherit
     Index:EFortItemShopOfferDisplayType Enum=GetDisplayType!(self:Fort Item Shop Offer Group VM Object Reference=@n3) -> (ReturnValue:EFortItemShopOfferDisplayType Enum=TileGrid)
    ) -> (ReturnValue:EWidgetClipping Enum)
    self:Widget Object Reference=Get Border() -> (Border:Border Object Reference)
   ) -> (then, Output_Get:EWidgetClipping Enum=Inherit)
   > UpdateBottomFadeVisibility()
  ]}
 ScrollUp {nodes [
   ScrollUp() {is=title K2Node_FunctionEntry}
   n1 = int - int(A=Get CurrentRowIndex() -> (CurrentRowIndex=0))
   > Branch(Condition=AND Boolean(A="integer >= integer"(A=@n1), B=true))
   > ScrollToIndex(DestinationIndex=int - int(A=Get CurrentRowIndex() -> (CurrentRowIndex=0)), bRefreshAll=false) {is=title K2Node_CallFunction}
   n2 = Get FortItemShopSelectionVM()
   Reroute Node(InputPin=HasLoadedRowsForAnimatedScroll(DestinationIndex=@n1, NumRows=GetNumRows(self=@n2) -> (ReturnValue=0)))
  ]}
 SetBorderPadding {nodes [
   n1 = Inputs!() -> (execute:Exec, IsSubheaderVisible:bool=~)
   n2 = Get DefaultBorderPadding()
   n3 = Break Margin!(Margin:"Margin Structure (by ref)"=@n2) -> (Left:float=~, Top:float=~, Right:float=~, Bottom:float=~) {is=title K2Node_BreakStruct}
   n4 = Get WBP_ItemShop_HeaderManager()
   n5 = float + float(
    A=@n3.Top
    B=GetDesiredSize!(self:Widget Object Reference=Get Text_CurrentSubheader!(self:WBP Item Shop Header Manager Object Reference=@n4) -> (Text_CurrentSubheader:Common Text Object Reference)) -> (ReturnValue_X:double, ReturnValue_Y:double).ReturnValue_Y
   )
   n6 = Get Border()
   > SetPadding!(
    execute
    self:Border Object Reference=@n6
    InPadding:Margin Structure=Select!(Option 0:Margin Structure=@n2, Option 1:Margin Structure=Make Margin(Left=@n3, Top=@n5, Right=@n3.Right, Bottom=@n3.Bottom), Index:bool=@n1) -> (ReturnValue:Margin Structure)
   ) -> (then)
   > SetPadding(self:Overlay Slot Object Reference=SlotAsOverlaySlot(Widget=@n6), InPadding=Get DefaultBorderOverlaySlotPadding())
   > Outputs!(then:Exec)
  ]}
 SetContentFocusedState {nodes [
   n1 = SetContentFocusedState!() -> (then, IsFocused:bool=~) {is=title K2Node_FunctionEntry}
   > Set IsContentFocused(IsContentFocused:bool=@n1) -> (Output_Get:bool)
   > n2 = Get()
   n3 = NOT Boolean(A=Get IsContentFocused())
   n4 = IsQuickNavCollapsible!() -> (ReturnValue:bool) {is=title K2Node_CallFunction}
   n5 = "Equal (Enum)"(A:ECommonInputType Enum=GetCurrentInputType(self=Get CommonInputSubsystem()), B:ECommonInputType Enum=MouseAndKeyboard)
   > n6 = SetVisibility(
    self=Get Button_Hitbox() -> (Button_Hitbox:Button Object Reference)
    InVisibility=Select(
     Option 0:ESlateVisibility Enum=Collapsed
     Option 1:ESlateVisibility Enum=Visible
     Index=AND Boolean!(A=@n3, B=@n5, C:bool=@n4) -> (ReturnValue)
    ) -> (ReturnValue:ESlateVisibility Enum)
   )
   n7 = For Each Loop(Exec=[@n6.then, @n2.else], Array=GetAllChildren(self=Get GridPanel_Content()))
   > n8 = Cast To WBP_MPItemShop_RowWrapper(Object=@n7)
   > I_SetParentHasFocusState!(execute, self:WBP MPItem Shop Row Wrapper Object Reference=@n8, HasFocus:bool=Get IsContentFocused()) -> (then)
  ]}
 SetMobileStyle {nodes [
   SetMobileStyle() {is=title K2Node_FunctionEntry}
   > Branch(Condition=IsUsingCompactUI())
   > n1 = Set!(
    execute
    DefaultBorderOverlaySlotPadding_Left:float=-16
    DefaultBorderOverlaySlotPadding_Top:float=200
    DefaultBorderOverlaySlotPadding_Right:float=-16
    DefaultBorderOverlaySlotPadding_Bottom:float=0
   ) -> (then, Output_Get:Margin Structure)
   n2 = Get Border()
   > SetPadding(self=SlotAsOverlaySlot(Widget=@n2), InPadding=@n1)
   > n3 = Set!(
    execute
    DefaultBorderPadding_Left:float=16
    DefaultBorderPadding_Top:float=-160
    DefaultBorderPadding_Right:float=16
    DefaultBorderPadding_Bottom:float=0
   ) -> (then, Output_Get)
   > SetPadding(self:Border Object Reference=@n2, InPadding=@n3)
  ]}
 SetRowStates {nodes [
   SetRowStates() {is=title K2Node_FunctionEntry}
   > n1 = GetCurrentRowWidget() {is=title K2Node_CallFunction}
   > n2 = Set CurrentRow(CurrentRow:WBP MPItem Shop Row Wrapper Object Reference=@n1) -> (Output_Get:WBP MPItem Shop Row Wrapper Object Reference)
   > Set SelectedRow!(
    execute
    SelectedRow:Fort Item Shop Row Structure=Get FortItemShopRow!(self:WBP MPItem Shop Row Wrapper Object Reference=@n2) -> (FortItemShopRow:Fort Item Shop Row Structure)
    self:Fort Item Shop Selection VM Object Reference=Get FortItemShopSelectionVM()
   ) -> (then, Output_Get:Fort Item Shop Row Structure)
   > n3 = For Each Loop(Array=GetAllChildren(self=Get GridPanel_Content()))
   n4 = Get!(execute=@n3.Completed) -> (then, else, CurrentRow:WBP MPItem Shop Row Wrapper Object Reference)
   > n5 = I_GetFocusTargetFromOfferID!(execute, self:WBP MPItem Shop Row Wrapper Object Reference=@n4, OfferID:String=Get FocusedOfferId()) -> (then, FocusTarget:Widget Object Reference)
   > n6 = Is Valid(InputObject=@n5)
   n7 = I_GetDesiredFocusTarget!(
    execute=@n6.Is Not Valid
    self:WBP MPItem Shop Row Wrapper Object Reference=Get CurrentRow() -> (CurrentRow:WBP MPItem Shop Row Wrapper Object Reference)
   ) -> (then, TileWidget:Widget Object Reference)
   > n8 = Cast To BPI_ItemShopTile!(execute, Object=@n7) -> (then, CastFailed, AsBPI Item Shop Tile:BPI Item Shop Tile Interface) {is=title K2Node_DynamicCast}
   > n9 = GetOfferVM!(execute, self:BPI Item Shop Tile Interface=@n8) -> (then, OfferVM:Fort Item Shop Offer VM Object Reference)
   > OnOfferTileFocused!(execute, self, OfferId:String=Get OfferId!(self:Fort Item Shop Offer VM Object Reference=@n9) -> (OfferId:String)) -> (then) {is=title K2Node_CallFunction}
   UpdateRowFocusedOffer(execute=@n6.Is Valid, RowIndex=GetFocusedRowIndex() -> (RowIndex=0))
   n10 = Cast To WBP_MPItemShop_RowWrapper(execute=@n3.LoopBody, Object=@n3)
   > Set LoopRow(LoopRow:WBP MPItem Shop Row Wrapper Object Reference=@n10) -> (Output_Get:WBP MPItem Shop Row Wrapper Object Reference)
   n11 = "Equal (Object)"(A=Get LoopRow() -> (LoopRow:WBP MPItem Shop Row Wrapper Object Reference), B=Get CurrentRow())
   n12 = Get CurrentRow()
   n13 = Get LoopRow()
   n14 = "Equal Exactly (String)"(
    A=Get!(self:Fort Item Shop Offer Group VM Object Reference=Get!(self:WBP MPItem Shop Row Wrapper Object Reference=@n13) -> (
      FortItemShopRow_OfferGroup:Fort Item Shop Offer Group VM Object Reference
      FortItemShopRow_Offers:Array of Fort Item Shop Offer VM Object References
      FortItemShopRow_bIsFirstRowInOfferGroup:bool=~
      FortItemShopRow_bIsLastRowInOfferGroup:bool=~
     )) -> (OwningLayout_SectionId:String, OwningLayout_SubsectionId:String)
    B=Get!(self:Fort Item Shop Offer Group VM Object Reference=Get!(self:WBP MPItem Shop Row Wrapper Object Reference=@n12) -> (
      FortItemShopRow_OfferGroup:Fort Item Shop Offer Group VM Object Reference
      FortItemShopRow_Offers:Array of Fort Item Shop Offer VM Object References
      FortItemShopRow_bIsFirstRowInOfferGroup:bool=~
      FortItemShopRow_bIsLastRowInOfferGroup:bool=~
     )) -> (OwningLayout_SectionId:String, OwningLayout_SubsectionId:String)
   )
   n15 = Get LoopRow()
   n16 = Select(
    Option 0:E_ItemShop_RowState Enum=NewEnumerator2
    Option 1:E_ItemShop_RowState Enum=NewEnumerator1
    Index="integer < integer"(A=GetRowIndexFromWidget(RowWidget=@n15) -> (RowIndex=0), B=Get CurrentRowIndex() -> (CurrentRowIndex=0))
   ) -> (ReturnValue:E_ItemShop_RowState Enum)
   > I_SetRowState(
    self=Get LoopRow()
    InRowState=Select(Option 0=Select(Option 0=NewEnumerator3, Option 1=@n16, Index=@n14), Option 1=NewEnumerator0, Index=@n11)
   )
   n17 = Create Event()
   > Unbind Event from On Row Navigated From(self:WBP MPItem Shop Row Wrapper Object Reference=Get LoopRow(), Delegate=@n17) {is=title K2Node_RemoveDelegate}
   > Branch(Condition=@n11)
   > Bind Event to On Row Navigated From(self:WBP MPItem Shop Row Wrapper Object Reference=Get LoopRow(), Delegate=@n17) {is=title K2Node_AddDelegate}
  ]}
 Update Featured Layout {nodes [
   n1 = Update Featured Layout!() -> (then, Target:Fort Item Shop Offer Group VM Object Reference) {is=title K2Node_FunctionEntry}
   > Get!(execute:Exec) -> (then:Exec, else:Exec, FortItemShopVM:Fort Item Shop VM Object Reference)
   n2 = Get OwningLayout!(self:Fort Item Shop Offer Group VM Object Reference=@n1) -> (OwningLayout:Fort Item Shop Layout Structure)
   n3 = Get FortItemShopVM()
   > Branch(Condition="Layout !="!(
     Lhs:Fort Item Shop Layout Structure=@n2
     Rhs:"Fort Item Shop Layout Structure (by ref)"=Get FeaturedLayout!(self:Fort Item Shop VM Object Reference=@n3) -> (FeaturedLayout:Fort Item Shop Layout Structure)
    ) -> (bNotEqual:bool=~) {is=title K2Node_MacroInstance})
   > Set bIgnoreNextFeaturedLayoutUpdate(bIgnoreNextFeaturedLayoutUpdate=true)
   > Set FeaturedLayout!(execute, FeaturedLayout:Fort Item Shop Layout Structure=@n2, self:Fort Item Shop VM Object Reference=Get FortItemShopVM()) -> (then, Output_Get:Fort Item Shop Layout Structure)
  ]}
 UpdateBottomFadeVisibility {nodes [
   UpdateBottomFadeVisibility() {is=title K2Node_FunctionEntry}
   n1 = GetRowAtIndex(self=Get FortItemShopSelectionVM(), RowIndex=Get CurrentRowIndex() -> (CurrentRowIndex=0)) -> (ReturnValue_bIsFirstRowInOfferGroup=~, ReturnValue_bIsLastRowInOfferGroup=~)
   n2 = NOT Boolean(A="Equal (Enum)"(A:EFortItemShopOfferDisplayType Enum=GetDisplayType(self=@n1) -> (ReturnValue=TileGrid), B:EFortItemShopOfferDisplayType Enum=Billboard))
   n3 = "integer < integer"(A=int + int(A=Get CurrentRowIndex() -> (CurrentRowIndex=0)), B=GetNumRows(self=Get FortItemShopSelectionVM()) -> (ReturnValue=0))
   > Set Visibility(
    Visibility=Select(
     Option 0:ESlateVisibility Enum=Collapsed
     Option 1:ESlateVisibility Enum=SelfHitTestInvisible
     Index=AND Boolean!(A=@n3, B=@n2) -> (ReturnValue)
    ) -> (ReturnValue:ESlateVisibility Enum)
    self=Get Image_Fade_Bottom() -> (Image_Fade_Bottom:Image Object Reference)
   )
  ]}
 UpdateFakeLoadingRowsVisibility {nodes [
   UpdateFakeLoadingRowsVisibility() {is=title K2Node_FunctionEntry}
   > SetVisibility(
    self=[
     Get FakeLoadingRow_Bottom() -> (FakeLoadingRow_Bottom:WBP MPItem Shop Fake Loading Row Object Reference)
     Get FakeLoadingRow_Top() -> (FakeLoadingRow_Top:WBP MPItem Shop Fake Loading Row Object Reference)
    ]
    InVisibility=SelfHitTestInvisible
   )
   > n1 = For Each Loop(Array=GetAllChildren(self=Get GridPanel_Content()))
   > Set RowIndex(RowIndex:int=GetRowIndexFromWidget(RowWidget=@n1) -> (RowIndex=0)) -> (Output_Get:int=0)
   > n2 = Sequence()
   n3 = Get!(execute=@n2.then_1) -> (then, else, FortItemShopSelectionVM:Fort Item Shop Selection VM Object Reference)
   n4 = GetNumRows(self=@n3) -> (ReturnValue=0)
   > Branch(Condition="Equal (Integer)"(A=int - int(A=@n4), B=Get RowIndex() -> (RowIndex=0)))
   > SetVisibility(self=Get FakeLoadingRow_Bottom(), InVisibility=Collapsed)
   Branch(execute=@n2.then_0, Condition="Equal (Integer)"(A=Get RowIndex() -> (RowIndex=0)))
   > SetVisibility(self=Get FakeLoadingRow_Top(), InVisibility=Collapsed)
  ]}
 UpdateFeaturedOffer {nodes [
   UpdateFeaturedOffer() {is=title K2Node_FunctionEntry}
   n1 = Get FortItemShopVM()
   > OnOfferTileFocused(OfferId=Get FeaturedOfferId(self=@n1))
   > Set FeaturedOfferId(self=@n1)
  ]}
 UpdateRowFocusedOffer {nodes [
   n1 = UpdateRowFocusedOffer!() -> (then, RowIndex:int) {is=title K2Node_FunctionEntry}
   > n2 = GetRowWidgetByRowIndex(RowIndex=@n1)
   > Is Valid(InputObject=@n2)
   > I_SetFocusedOfferID!(execute, self:I MPItem Shop Row Interface=@n2, OfferID:String=Get FocusedOfferId()) -> (then)
  ]}
 UpdateSectionIndexMap {nodes [
   UpdateSectionIndexMap() {is=title K2Node_FunctionEntry}
   > Clear!(execute, TargetMap:Map of Strings to Integers=Get SectionIndexMap()) -> (then) {is=title K2Node_CallFunction}
   > n1 = Get!(execute) -> (then, else, FortItemShopVM:Fort Item Shop VM Object Reference)
   > n2 = For Each Loop(Array:Array of Fort Item Shop Section VM Object References=Get SectionsWithOffers!(self:Fort Item Shop VM Object Reference=@n1) -> (SectionsWithOffers:Array of Fort Item Shop Section VM Object References)) -> (Array Element:Fort Item Shop Section VM Object Reference)
   > Branch(Condition=IsVisible(self:Fort Item Shop Section VM Object Reference=@n2))
   > n3 = For Each Loop with Break!(
    Exec:Exec
    Array:Array of Fort Item Shop Offer Group VM Object References=Get AllOfferGroupsWithOffers!(self:Fort Item Shop Section VM Object Reference=@n2) -> (AllOfferGroupsWithOffers:Array of Fort Item Shop Offer Group VM Object References)
    Break:Exec=@n7.then
   ) -> (LoopBody:Exec, Array Element:Fort Item Shop Offer Group VM Object Reference, Array Index:int, Completed:Exec)
   > Branch(Condition=IsVisible(self:Fort Item Shop Offer Group VM Object Reference=@n3))
   n4 = "Get (a copy)"(Array:Array of Fort Item Shop Offer VM Object References=Get Offers!(self:Fort Item Shop Offer Group VM Object Reference=@n3) -> (Offers:Array of Fort Item Shop Offer VM Object References)) -> (Output:Fort Item Shop Offer VM Object Reference)
   n5 = Get FortItemShopSelectionVM()
   n6 = Get SectionId!(self:Fort Item Shop Section VM Object Reference=@n2) -> (SectionId:String)
   > n7 = Add!(
    execute
    TargetMap:Map of Strings to Integers=Get SectionIndexMap()
    Key:"String (by ref)"=@n6
    Value:"Integer (by ref)"=GetIndexForRow(self=@n5, Row=GetRowFromOffer(self=@n5, Offer=@n4, PreferredSection=@n6)) -> (ReturnValue=0)
   ) -> (then) {is=title K2Node_CallFunction}
  ]}
 VM_NumRowsUpdated {nodes [
   VM_NumRowsUpdated!() -> (then, NumRows:int) {is=title K2Node_FunctionEntry}
   > UpdateSectionIndexMap() {is=title K2Node_CallFunction}
   > n1 = Branch(Condition=IsActivated!(self:Common Activatable Widget Object Reference) -> (ReturnValue:bool))
   > ScrollToIndex(DestinationIndex="Max (Integer)"(A=GetFocusedRowIndex() -> (RowIndex=0), B=0), bRefreshAll=true)
   Set bRowsDirty(execute=@n1.else, bRowsDirty=true)
  ]}
]
variables [
 bIgnoreNextFeaturedLayoutUpdate {type=boolean}
 bRowsDirty {type=boolean}
 ContentWhoosh_RetractSound {type=SoundBase, default=null}
 CurrentRowIndex {type=integer, default=-1}
 DefaultBorderOverlaySlotPadding {type=Margin, default=<ltrb -16 173 -16 0>}
 DefaultBorderPadding {type=Margin, default=<ltrb 16 -197 16 0>}
 FocusedLayout {type=string, default=""}
 FocusedOfferId {type=string, default=""}
 IsContentFocused {type=boolean}
 OnContentFocused {}
 OnListViewRequested {}
 OnOfferFocused {}
 OnScrollFinished {}
 OnScrollStarted {}
 RowWidthWith3DPreview {type=number, default=1096}
 RowWidthWithout3DPreview {type=number, default=1351}
 RowWidthWithWide3DPreview {type=number, default=953}
 SectionIndexMap {type="map<integer>", default {}}
]
widgets {
 info {inheritedWidgetCount=2, parentClass=/Script/FortniteUI.FortMPItemShopContent, rootWidgetClass=&Overlay, widgetCount=13}
 widgets [
  Overlay_0 {widgetClassPath=&Overlay, widget=Overlay_0}
  Border {
   widgetClassPath=&Border
   bIsVariable=true
   parent=Overlay_0
   widget=Border
   properties {
    clipping=ClipToBounds
    visibility=SelfHitTestInvisible
    background {
     bIsDynamicallyLoaded=false
     drawAs=NoDrawType
     imageType=NoImage
     mirroring=NoMirror
     resourceName=None
     resourceObject=null
     tiling=NoTile
     imageSize=<32 32>
     margin=<ltrb 0 0 0 0>
     outlineSettings {
      bUseBrushTransparency=false
      roundingType=HalfHeightRadius
      width=0
      color {colorUseRule=UseColor_Specified, specifiedColor=<rgba 0 0 0 0>}
      cornerRadii=<0 0 0 0>
     }
     tintColor {colorUseRule=UseColor_Specified, specifiedColor=<rgba 1 1 1 1>}
     uVRegion {bIsValid=false, max=<0 0>, min=<0 0>}
    }
    padding=<ltrb 16 -197 16 0>
   }
   slot {class=OverlaySlot, properties {horizontalAlignment=HAlign_Fill, padding=<ltrb -16 173 -16 0>}}
  }
  ScrollBox_Content {
   widgetClassPath=/Script/FortniteUI.AthenaScrollBox
   bInherited=true
   bIsVariable=true
   parent=Border
   widget=ScrollBox_Content
   properties {
    bAllowRightClickDragScrolling=false
    bAnimateWheelScrolling=false
    bEnableFixedPageOffset=true
    clipping=Inherit
    navigationDestination=Center
    scrollAnimationInterpolationSpeed=10
    scrollBarVisibility=Collapsed
    visibility=SelfHitTestInvisible
    widgetStyle {
     barThickness=9
     bottomShadowBrush {
      bIsDynamicallyLoaded=false
      drawAs=NoDrawType
      imageType=FullColor
      mirroring=NoMirror
      resourceName=../../../Engine/Content/Slate/Common/ScrollBoxShadowBottom.png
      resourceObject=null
      tiling=NoTile
      imageSize=<0 0>
      margin=<ltrb 0.5 0 0.5 1>
      outlineSettings {
       bUseBrushTransparency=false
       roundingType=HalfHeightRadius
       width=0
       color {colorUseRule=UseColor_Specified, specifiedColor=<rgba 0 0 0 0>}
       cornerRadii=<0 0 0 0>
      }
      tintColor {colorUseRule=UseColor_Specified, specifiedColor=<rgba 1 1 1 0>}
      uVRegion {bIsValid=false, max=<0 0>, min=<0 0>}
     }
     horizontalScrolledContentPadding=<ltrb 0 0 1 0>
     leftShadowBrush {
      bIsDynamicallyLoaded=false
      drawAs=Box
      imageType=FullColor
      mirroring=NoMirror
      resourceName=../../../Engine/Content/Slate/Common/ScrollBoxShadowLeft.png
      resourceObject=null
      tiling=NoTile
      imageSize=<8 16>
      margin=<ltrb 1 0.5 0 0.5>
      outlineSettings {
       bUseBrushTransparency=false
       roundingType=HalfHeightRadius
       width=0
       color {colorUseRule=UseColor_Specified, specifiedColor=<rgba 0 0 0 0>}
       cornerRadii=<0 0 0 0>
      }
      tintColor {colorUseRule=UseColor_Specified, specifiedColor=<rgba 1 1 1 1>}
      uVRegion {bIsValid=false, max=<0 0>, min=<0 0>}
     }
     rightShadowBrush {
      bIsDynamicallyLoaded=false
      drawAs=Box
      imageType=FullColor
      mirroring=NoMirror
      resourceName=../../../Engine/Content/Slate/Common/ScrollBoxShadowRight.png
      resourceObject=null
      tiling=NoTile
      imageSize=<8 16>
      margin=<ltrb 0 0.5 1 0.5>
      outlineSettings {
       bUseBrushTransparency=false
       roundingType=HalfHeightRadius
       width=0
       color {colorUseRule=UseColor_Specified, specifiedColor=<rgba 0 0 0 0>}
       cornerRadii=<0 0 0 0>
      }
      tintColor {colorUseRule=UseColor_Specified, specifiedColor=<rgba 1 1 1 1>}
      uVRegion {bIsValid=false, max=<0 0>, min=<0 0>}
     }
     topShadowBrush {
      bIsDynamicallyLoaded=false
      drawAs=NoDrawType
      imageType=FullColor
      mirroring=NoMirror
      resourceName=../../../Engine/Content/Slate/Common/ScrollBoxShadowTop.png
      resourceObject=null
      tiling=NoTile
      imageSize=<0 0>
      margin=<ltrb 0.5 1 0.5 0>
      outlineSettings {
       bUseBrushTransparency=false
       roundingType=HalfHeightRadius
       width=0
       color {colorUseRule=UseColor_Specified, specifiedColor=<rgba 0 0 0 0>}
       cornerRadii=<0 0 0 0>
      }
      tintColor {colorUseRule=UseColor_Specified, specifiedColor=<rgba 1 1 1 0>}
      uVRegion {bIsValid=false, max=<0 0>, min=<0 0>}
     }
     verticalScrolledContentPadding=<ltrb 0 0 0 1>
    }
   }
   slot {class=BorderSlot, properties {padding=<ltrb 16 -197 16 0>}}
  }
  Spacer {widgetClassPath=&Spacer, parent=ScrollBox_Content, widget=Spacer, properties {size=<1 1080>}, slot {class=ScrollBoxSlot}}
  FakeLoadingRow_Top {
   widgetClassPath=&/WBP_MPItemShop_FakeLoadingRow
   bIsVariable=true
   parent=ScrollBox_Content
   widget=FakeLoadingRow_Top
   slot {class=ScrollBoxSlot, properties {padding=<ltrb 0 0 0 250>}}
  }
  GridPanel_Content {
   widgetClassPath=&GridPanel
   bInherited=true
   bIsVariable=true
   parent=ScrollBox_Content
   widget=GridPanel_Content
   properties {columnFill [1]}
   slot {class=ScrollBoxSlot}
  }
  FakeLoadingRow_Bottom {
   widgetClassPath=&WBP_MPItemShop_FakeLoadingRow
   bIsVariable=true
   parent=ScrollBox_Content
   widget=FakeLoadingRow_Bottom
   slot {class=ScrollBoxSlot, properties {padding=<ltrb 0 250 0 0>}}
  }
  Overlay_1 {
   widgetClassPath=&Overlay
   parent=ScrollBox_Content
   widget=Overlay_1
   slot {class=ScrollBoxSlot, properties {verticalAlignment=VAlign_Top}}
  }
  Spacer_73 {widgetClassPath=&Spacer, parent=Overlay_1, widget=Spacer_73, properties {size=<1 1080>}, slot {class=OverlaySlot}}
  Text_LegalText {
   widgetClassPath=&CommonTextBlock
   bIsVariable=true
   parent=Overlay_1
   widget=Text_LegalText
   properties {
    bAutoCollapseWithEmptyText=true
    text="[PH] *Savings refer to the lowest price offered in the past 30 days on Fortnite Item Shop."
    wrapTextAt=1096
    font {
     bForceMonospaced=false
     bMaterialIsStencil=false
     fontMaterial=null
     fontObject=&HeadingNow
     letterSpacing=0
     monospacedWidth=1
     size=13.5
     skewAmount=0
     typefaceFontName=74 Regular
     outlineSettings {
      bApplyOutlineToDropShadows=false
      bMiteredCorners=false
      bSeparateFillAlpha=false
      outlineMaterial=null
      outlineSize=0
      outlineColor=<rgba 0 0 0 1>
     }
    }
    shadowOffset=<0 0>
   }
   slot {class=OverlaySlot, properties {horizontalAlignment=HAlign_Fill}}
  }
  Button_Hitbox {
   widgetClassPath=&Button
   parent=Overlay_0
   widget=Button_Hitbox
   properties {
    isFocusable=false
    renderOpacity=0
    visibility=Collapsed
    widgetStyle {
     clickedSlateSound {resourceObject=null}
     disabled {
      bIsDynamicallyLoaded=false
      drawAs=NoDrawType
      imageType=NoImage
      mirroring=NoMirror
      resourceName=None
      resourceObject=null
      tiling=NoTile
      imageSize=<0 0>
      margin=<ltrb 0 0 0 0>
      outlineSettings {
       bUseBrushTransparency=false
       roundingType=HalfHeightRadius
       width=0
       color {colorUseRule=UseColor_Specified, specifiedColor=<rgba 0 0 0 0>}
       cornerRadii=<0 0 0 0>
      }
      tintColor {colorUseRule=UseColor_Specified, specifiedColor=<rgba 1 1 1 1>}
      uVRegion {bIsValid=false, max=<0 0>, min=<0 0>}
     }
     disabledForeground {colorUseRule=UseColor_Specified, specifiedColor=<rgba 0.5271 0.5271 0.5271 1>}
     hovered {
      bIsDynamicallyLoaded=false
      drawAs=RoundedBox
      imageType=NoImage
      mirroring=NoMirror
      resourceName=None
      resourceObject=null
      tiling=NoTile
      imageSize=<32 32>
      margin=<ltrb 0 0 0 0>
      outlineSettings {
       bUseBrushTransparency=true
       roundingType=FixedRadius
       width=1
       color {colorUseRule=UseColor_Specified, specifiedColor=<rgba 0.7243 0.7243 0.7243 1>}
       cornerRadii=<4 4 4 4>
      }
      tintColor {colorUseRule=UseColor_Specified, specifiedColor=<rgba 0.7243 0.7243 0.7243 1>}
      uVRegion {bIsValid=false, max=<0 0>, min=<0 0>}
     }
     hoveredForeground {colorUseRule=UseColor_Specified, specifiedColor=<rgba 0 1 1 1>}
     hoveredSlateSound {resourceObject=null}
     normal {
      bIsDynamicallyLoaded=false
      drawAs=RoundedBox
      imageType=NoImage
      mirroring=NoMirror
      resourceName=None
      resourceObject=null
      tiling=NoTile
      imageSize=<32 32>
      margin=<ltrb 0 0 0 0>
      outlineSettings {
       bUseBrushTransparency=true
       roundingType=FixedRadius
       width=1
       color {colorUseRule=UseColor_Specified, specifiedColor=<rgba 0.6951 0.6951 0.6951 1>}
       cornerRadii=<4 4 4 4>
      }
      tintColor {colorUseRule=UseColor_Specified, specifiedColor=<rgba 0.4955 0.4955 0.4955 1>}
      uVRegion {bIsValid=false, max=<0 0>, min=<0 0>}
     }
     normalForeground {colorUseRule=UseColor_Specified, specifiedColor=<rgba 0 1 1 1>}
     normalPadding=<ltrb 0 0 0 0>
     pressed {
      bIsDynamicallyLoaded=false
      drawAs=RoundedBox
      imageType=NoImage
      mirroring=NoMirror
      resourceName=None
      resourceObject=null
      tiling=NoTile
      imageSize=<32 32>
      margin=<ltrb 0 0 0 0>
      outlineSettings {
       bUseBrushTransparency=true
       roundingType=FixedRadius
       width=1
       color {colorUseRule=UseColor_Specified, specifiedColor=<rgba 0.7243 0.7243 0.7243 1>}
       cornerRadii=<4 4 4 4>
      }
      tintColor {colorUseRule=UseColor_Specified, specifiedColor=<rgba 0.3843 0.3843 0.3843 1>}
      uVRegion {bIsValid=false, max=<0 0>, min=<0 0>}
     }
     pressedForeground {colorUseRule=UseColor_Specified, specifiedColor=<rgba 0 1 1 1>}
     pressedPadding=<ltrb 0 0 0 0>
     pressedSlateSound {resourceObject=null}
    }
   }
   slot {class=OverlaySlot, properties {horizontalAlignment=HAlign_Fill, verticalAlignment=VAlign_Fill}}
  }
  WBP_ItemShop_HeaderManager {
   widgetClassPath=&/WBP_ItemShop_HeaderManager
   bIsVariable=true
   parent=Overlay_0
   widget=WBP_ItemShop_HeaderManager
   slot {class=OverlaySlot, properties {verticalAlignment=VAlign_Fill}}
  }
  Image_Fade_Bottom {
   widgetClassPath=&Image
   bIsVariable=true
   parent=Overlay_0
   widget=Image_Fade_Bottom
   properties {
    visibility=SelfHitTestInvisible
    brush {
     bIsDynamicallyLoaded=false
     drawAs=Image
     imageType=FullColor
     mirroring=NoMirror
     resourceName=None
     resourceObject=/ItemShop/MultiProduct/Art/MI_UI_ListFade_Bottom
     tiling=NoTile
     imageSize=<32 100>
     margin=<ltrb 0 0 0 0>
     outlineSettings {
      bUseBrushTransparency=false
      roundingType=HalfHeightRadius
      width=0
      color {colorUseRule=UseColor_Specified, specifiedColor=<rgba 0 0 0 0>}
      cornerRadii=<0 0 0 0>
     }
     tintColor {colorUseRule=UseColor_Specified, specifiedColor=<rgba 1 1 1 1>}
     uVRegion {bIsValid=false, max=<0 0>, min=<0 0>}
    }
   }
   slot {class=OverlaySlot, properties {horizontalAlignment=HAlign_Fill, verticalAlignment=VAlign_Bottom}}
  }
 ]
}
