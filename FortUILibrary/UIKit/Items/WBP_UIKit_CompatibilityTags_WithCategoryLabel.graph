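# UI Kit widget that shows a set of compatibility tags under an uppercase label listing one or more category names.
# SetCompatibilityTags stores the gameplay tags and passes them to the child CompatibilityTags widget.
# MatchesCompatibilityTags tests order-independent equality with another tag list by comparing set size and membership.
# AddCategory queues names, and FinalizeCategoryLabel joins them with commas, removing the first from the array as it goes.
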WidgetBlueprint /FortUILibrary/UIKit/Items/WBP_UIKit_CompatibilityTags_WithCategoryLabel
graphs [
 AddCategory {nodes [
   n1 = AddCategory!() -> (then, CategoryName:Text) {is=title K2Node_FunctionEntry}
   > Add(TargetArray:Array of Texts=Get CategoryNames() -> (CategoryNames:Array of Texts), NewItem:"Text (by ref)"=@n1) {is=title K2Node_CallArrayFunction}
  ]}
 EventGraph {nodes [
   n1 = Event PreConstruct()
   > Parent: PreConstruct(IsDesignTime=@n1.IsDesignTime)
   Event Construct()
   Event Tick()
  ]}
 FinalizeCategoryLabel {nodes [
   FinalizeCategoryLabel() {is=title K2Node_FunctionEntry}
   > Set Text(
    Text="Get (a copy)"(Array:Array of Texts=Get CategoryNames()) -> (Output:Text)
    self=Get CategoryLabel() -> (CategoryLabel:Common Text Object Reference)
   )
   > Remove Index!(execute, TargetArray:Array of Texts=Get CategoryNames(), IndexToRemove:int=0) -> (then)
   > n1 = For Each Loop(Array:Array of Texts=Get CategoryNames()) -> (Array Element:Text)
   n2 = Get CategoryLabel()
   > Set Text(
    Text=Format Text!(
     Format="LOCTEXT(\"{CurrentCategories}, {NewCategory}\")"
     CurrentCategories:Text=Get Text(self:Text Object Reference=@n2)
     NewCategory:Text=@n1
    ) -> (Result)
    self=Get CategoryLabel()
   )
  ]}
 MatchesCompatibilityTags {nodes [
   n1 = MatchesCompatibilityTags!() -> (then, OtherCompatibilityTags:Array of Gameplay Tag Structures) {is=title K2Node_FunctionEntry}
   n2 = Get OtherCompatibilityTagsAsSet() -> (OtherCompatibilityTagsAsSet:Set of Gameplay Tag Structures)
   > Clear!(execute, TargetSet:Set of Gameplay Tag Structures=@n2) -> (then) {is=title K2Node_CallFunction}
   > Add Items!(execute, TargetSet:Set of Gameplay Tag Structures=@n2, NewItems:Array of Gameplay Tag Structures=@n1) -> (then)
   n3 = Get Compatibility Tags() -> (Compatibility Tags:Set of Gameplay Tag Structures)
   n4 = Get OtherCompatibilityTagsAsSet()
   > n5 = Branch(Condition="Equal (Integer)"(
     A=Length!(TargetSet:Set of Gameplay Tag Structures=@n4) -> (ReturnValue:int=0) {is=title K2Node_CallFunction}
     B=Length(TargetSet=@n3) -> (ReturnValue=0)
    ))
   > n6 = "For Each Loop (Set)"!(execute:Exec, SetPin:Set of Gameplay Tag Structures=Get OtherCompatibilityTagsAsSet(), BreakPin:Exec) -> (then:Exec, ValuePin:Gameplay Tag Structure, CompletedPin:Exec) {is=title K2Node_SetForEach}
   > n7 = Branch(Condition=Contains Item!(TargetSet:Set of Gameplay Tag Structures=Get Compatibility Tags(), ItemToFind:"Gameplay Tag Structure (by ref)"=@n6) -> (ReturnValue:bool) {is=title K2Node_CallFunction})
   Return Node!(execute=@n7.else, isMatching:bool=false)
   Return Node(execute=@n6.CompletedPin, isMatching=true)
   Return Node(execute=@n5.else, isMatching=false)
  ]}
 SetCompatibilityTags {nodes [
   n1 = SetCompatibilityTags!() -> (then, CompatibilityTags:Array of Gameplay Tag Structures) {is=title K2Node_FunctionEntry}
   n2 = Get Compatibility Tags()
   > Clear(TargetSet=@n2)
   > Add Items(TargetSet=@n2, NewItems=@n1)
   > SetCompatibilityTags!(
    execute
    self:WBP UIKit Compatibility Tags Object Reference=Get CompatibilityTagWidgets() -> (CompatibilityTagWidgets:WBP UIKit Compatibility Tags Object Reference)
    Product Tags:Set of Gameplay Tag Structures=Get Compatibility Tags()
   ) -> (then) {is=title K2Node_CallFunction}
  ]}
]
variables [
 CategoryNames {type="array<string>", default []}
 Compatibility Tags {type="array<GameplayTag>", default []}
]
widgets {
 info {inheritedWidgetCount=0, parentClass=&UserWidget, rootWidgetClass=&VerticalBox, widgetCount=3}
 widgets [
  VerticalBox_29 {widgetClassPath=&VerticalBox, widget=VerticalBox_29}
  CategoryLabel {
   widgetClassPath=&CommonTextBlock
   bIsVariable=true
   parent=VerticalBox_29
   widget=CategoryLabel
   properties {
    autoWrapText=true
    mobileFontSizeMultiplier=1.14
    text=Tag Category
    textTransformPolicy=ToUpper
    colorAndOpacity {colorUseRule=UseColor_Specified, specifiedColor=<rgba 1 1 1 0.75>}
    font {
     bForceMonospaced=false
     bMaterialIsStencil=false
     fontMaterial=null
     fontObject=&HeadingNow
     letterSpacing=0
     monospacedWidth=1
     size=15.75
     skewAmount=0
     typefaceFontName=74 Regular
     outlineSettings {
      bApplyOutlineToDropShadows=false
      bMiteredCorners=false
      bSeparateFillAlpha=false
      outlineMaterial=null
      outlineSize=0
      outlineColor=<rgba 0 0 0 1>
     }
    }
    shadowOffset=<0 0>
   }
   slot {class=VerticalBoxSlot, properties {padding=<ltrb 0 25 0 0>}}
  }
  CompatibilityTagWidgets {
   widgetClassPath=&/WBP_UIKit_CompatibilityTags
   bIsVariable=true
   parent=VerticalBox_29
   widget=CompatibilityTagWidgets
   properties {hideItemcard=true}
   slot {class=VerticalBoxSlot}
  }
 ]
}
