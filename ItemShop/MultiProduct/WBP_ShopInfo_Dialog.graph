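# Activatable Item Shop info dialog that explains cosmetic compatibility by category and records when the player has seen it.
# PreConstruct builds tag widgets from category compatibility data, merging categories with matching displayed tags, then appends the footer inside the dialog's scroll box.
# For unseen visits, a visible scrollbar hides "Got It" behind a scroll-to-end button until the end is reached.
# A viewmodel callback formats the time zone into the info text.
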WidgetBlueprint /ItemShop/MultiProduct/WBP_ShopInfo_Dialog
graphs [
 AddCompatibilityInfo {nodes [
   AddCompatibilityInfo() {is=title K2Node_FunctionEntry}
   > n1 = GetCategoryCompatibleProducts!(execute) -> (then, ReturnValue:Array of Fort Category Compatibility Info Structures)
   > n2 = For Each Loop(Array:Array of Fort Category Compatibility Info Structures=@n1) -> (Array Element:Fort Category Compatibility Info Structure)
   > n3 = Sequence!(execute) -> (then_0, then_1, then_2:Exec)
   Set CategoryAddedToExistingWidget(execute=@n3.then_1, CategoryAddedToExistingWidget:bool=false) -> (Output_Get:bool)
   > n4 = For Each Loop with Break!(
    Exec:Exec
    Array:Array of WBP UIKit Compatibility Tags With Category Label Object References=Get CompatibilityWidgets() -> (CompatibilityWidgets:Array of WBP UIKit Compatibility Tags With Category Label Object References)
    Break:Exec=@n6.then
   ) -> (LoopBody:Exec, Array Element:WBP UIKit Compatibility Tags With Category Label Object Reference, Array Index:int, Completed:Exec)
   > n5 = MatchesCompatibilityTags!(
    execute
    self:WBP UIKit Compatibility Tags With Category Label Object Reference=@n4
    OtherCompatibilityTags:Array of Gameplay Tag Structures=Get DisplayedCompatibilityTags() -> (DisplayedCompatibilityTags:Array of Gameplay Tag Structures)
   ) -> (then, isMatching:bool)
   > Branch(Condition=@n5)
   > AddCategory!(
    execute
    self:WBP UIKit Compatibility Tags With Category Label Object Reference=@n4
    CategoryName:Text=Get CurrentDisplayName() -> (CurrentDisplayName:Text)
   ) -> (then)
   > n6 = Set CategoryAddedToExistingWidget(CategoryAddedToExistingWidget=true)
   n7 = Break Fort Category Compatibility Info!(FortCategoryCompatibilityInfo:"Fort Category Compatibility Info Structure (by ref)"=@n2) -> (CategoryDisplayName:Text, CategoryDisplayNamePlural:Text, CompatibleProducts:Gameplay Tag Container Structure) {is=title K2Node_BreakStruct}
   Set CurrentDisplayName(execute=@n3.then_0, CurrentDisplayName:Text=@n7) -> (Output_Get:Text)
   > Clear(TargetArray:Array of Gameplay Tag Structures=Get DisplayedCompatibilityTags())
   > Clear(TargetArray:Array of Texts=Get CurrentProductNames() -> (CurrentProductNames:Array of Texts))
   > n8 = For Each Loop(Array:Array of Gameplay Tag Structures=BreakGameplayTagContainer!(GameplayTagContainer:"Gameplay Tag Container Structure (by ref)"=@n7.CompatibleProducts) -> (GameplayTags:Array of Gameplay Tag Structures)) -> (Array Element:Gameplay Tag Structure)
   > Branch(Condition="Not Equal (GameplayTag)"(A:Gameplay Tag Structure=@n8, B:Gameplay Tag Structure=(TagName="Product")) -> (ReturnValue:bool=~) {is=title K2Node_PromotableOperator})
   > n9 = FindCosmeticConfig!(execute, ProductTag:Gameplay Tag Structure=@n8) -> (then, ReturnValue:Fort Product Cosmetic Config Object Reference)
   > Is Valid(InputObject=@n9)
   n10 = Get ProductName!(self:Fort Product Cosmetic Config Object Reference=@n9) -> (ProductName:Text)
   n11 = TextIsEmpty(InText=@n10)
   > Branch(Condition=AND Boolean(
     A=Get bDisplayCompatibilityTag!(self:Fort Product Cosmetic Config Object Reference=@n9) -> (bDisplayCompatibilityTag:bool)
     B=NOT Boolean(A=@n11)
    ))
   n12 = Get CurrentProductNames()
   > n13 = Add Unique(TargetArray:Array of Texts=@n12, NewItem:"Text (by ref)"=@n10)
   > Branch(Condition=Is Valid Index!(TargetArray:Array of Texts=@n12, IndexToTest:int=@n13) -> (ReturnValue:bool))
   > Add Unique(TargetArray:Array of Gameplay Tag Structures=Get DisplayedCompatibilityTags(), NewItem:"Gameplay Tag Structure (by ref)"=@n8)
   n14 = Branch(execute=@n3.then_2, Condition=Get CategoryAddedToExistingWidget() -> (CategoryAddedToExistingWidget:bool))
   Branch(
    execute=@n14.else
    Condition=Is Not Empty!(TargetArray:Array of Gameplay Tag Structures=Get DisplayedCompatibilityTags()) -> (ReturnValue:bool)
   )
   > n15 = Create Widget!(
    execute:Exec
    Class:User Widget Class Reference=/FortUILibrary/UIKit/Items/WBP_UIKit_CompatibilityTags_WithCategoryLabel.WBP_UIKit_CompatibilityTags_WithCategoryLabel_C
    OwningPlayer:Player Controller Object Reference=GetOwningPlayer()
   ) -> (then:Exec, ReturnValue:WBP UIKit Compatibility Tags With Category Label Object Reference)
   > AddChild(
    self=Get ScrollBox_Text!(self:WBP UIKit Dialog Internal Object Reference=Get Dialog_Internal() -> (Dialog_Internal:WBP UIKit Dialog Internal Object Reference)) -> (ScrollBox_Text:Scroll Box Object Reference)
    Content=@n15
   )
   > Add(
    TargetArray:Array of WBP UIKit Compatibility Tags With Category Label Object References=Get CompatibilityWidgets()
    NewItem:"WBP UIKit Compatibility Tags With Category Label Object Reference (by ref)"=@n15
   ) {is=title K2Node_CallArrayFunction}
   > AddCategory(self=@n15, CategoryName=Get CurrentDisplayName())
   > SetCompatibilityTags!(
    execute
    self:WBP UIKit Compatibility Tags With Category Label Object Reference=@n15
    CompatibilityTags:Array of Gameplay Tag Structures=Get DisplayedCompatibilityTags()
   ) -> (then)
   n16 = For Each Loop(Exec=@n2.Completed, Array:Array of WBP UIKit Compatibility Tags With Category Label Object References=Get CompatibilityWidgets()) -> (Array Element:WBP UIKit Compatibility Tags With Category Label Object Reference)
   > FinalizeCategoryLabel(self:WBP UIKit Compatibility Tags With Category Label Object Reference=@n16)
  ]}
 EventGraph {nodes [
   n1 = OnButtonClicked!() -> (OutputDelegate, then, ButtonIndex:int) {is=title K2Node_CustomEvent}
   > n2 = Switch on Int!(execute:Exec, Selection:int=@n1.ButtonIndex) -> (Default:Exec, 0:Exec, 1:Exec)
   ScrollToEnd(execute=@n2.0, self:Scroll Box Object Reference=Get ScrollBox_Text(self=Get Dialog_Internal()))
   > SetAllowDismiss!(execute, self, InAllowDismiss:bool=true) -> (then)
   FireEventElementInteractedByName!(execute=@n2.1, Element Name:String=Button_GotIt, WidgetCallingFrom:Widget Object Reference=Self-Reference()) -> (then, Success:bool)
   > SetHasSeenAboutShopModal!(
    execute
    self:Fort Item Shop VM Object Reference=Get FortItemShopVM() -> (FortItemShopVM:Fort Item Shop VM Object Reference)
    bHasSeen:bool=true
   ) -> (then)
   > DeactivateWidget()
   n3 = OnScrollbarVisibilityChanged!() -> (OutputDelegate, then, NewVisibility:ESlateVisibility Enum=Visible) {is=title K2Node_CustomEvent}
   > SetAllowDismiss(InAllowDismiss=Select!(
     Visible:bool=false
     Collapsed:bool=true
     Hidden:bool=true
     HitTestInvisible:bool=false
     SelfHitTestInvisible:bool=false
     Index:ESlateVisibility Enum=@n3.NewVisibility
    ) -> (ReturnValue:bool))
   n4 = OnUserScrolled!() -> (OutputDelegate, then, CurrentOffset:float=~) {is=title K2Node_CustomEvent}
   n5 = Get ScrollBox_Text(self=Get Dialog_Internal())
   > Branch(Condition="Nearly Equal (Float)"(A=@n4.CurrentOffset, B=GetScrollOffsetOfEnd!(self:Scroll Box Object Reference=@n5) -> (ReturnValue:float), ErrorTolerance=0))
   > SetAllowDismiss(InAllowDismiss=true)
   n6 = Event PreConstruct()
   > n7 = Branch(Condition=@n6.IsDesignTime)
   > n8 = Create Widget(
    Class=&WBP_UIKit_CompatibilityTags_WithCategoryLabel_C
    OwningPlayer=GetOwningPlayer()
   )
   > n9 = AddChild(self=Get CategoryLabelContainer() -> (CategoryLabelContainer:Vertical Box Object Reference), Content=@n8)
   RemoveFromParent(execute=[@n10.then, @n9.then], self=Get ContentAddedAfterDialogInfo() -> (ContentAddedAfterDialogInfo:Vertical Box Object Reference))
   > AddChild(self=Get ScrollBox_Text(self=Get Dialog_Internal()), Content=Get ContentAddedAfterDialogInfo())
   AddCompatibilityInfo(execute=@n7.else) {is=title K2Node_CallFunction}
   > n10 = RemoveFromParent(self=Get CategoryLabelContainer())
   Event On Activated()
   > Bind Event to Dialog Button Clicked Dispatcher(self:WBP UIKit Dialog Internal Object Reference=Get Dialog_Internal(), Delegate=@n1) {is=title K2Node_AddDelegate}
   > Fire Event Sub View Opened!(execute, View:Widget Object Reference=Self-Reference(), AdditionalInfo:Flow Analytics Additional Info Structure, bIsOverlay:bool=true) -> (then)
   > SetScrollWidget!(execute, ToScroll:Widget Object Reference=Get ScrollBox_Text(self=Get Dialog_Internal())) -> (then)
   > n11 = Branch(Condition=GetHasSeenAboutShopModal!(self:Fort Item Shop VM Object Reference=Get FortItemShopVM()) -> (ReturnValue:bool))
   > SetAllowDismiss(InAllowDismiss=true)
   Bind Event to On Scroll Bar Visibility Changed(execute=@n11.else, self:Scroll Box Object Reference=Get ScrollBox_Text(self=Get Dialog_Internal()), Delegate=@n3) {is=title K2Node_AddDelegate}
   > Bind Event to On User Scrolled(self:Scroll Box Object Reference=Get ScrollBox_Text(self=Get Dialog_Internal()), Delegate=@n4) {is=title K2Node_AddDelegate}
   Event On Deactivated()
   > Unbind Event from Dialog Button Clicked Dispatcher(self:WBP UIKit Dialog Internal Object Reference=Get Dialog_Internal(), Delegate=@n1) {is=title K2Node_RemoveDelegate}
   > Unbind Event from On Scroll Bar Visibility Changed(self:Scroll Box Object Reference=Get ScrollBox_Text(self=Get Dialog_Internal()), Delegate=@n3) {is=title K2Node_RemoveDelegate}
   > Unbind Event from On User Scrolled(self:Scroll Box Object Reference=Get ScrollBox_Text(self=Get Dialog_Internal()), Delegate=@n4) {is=title K2Node_RemoveDelegate}
  ]}
 SetAllowDismiss {nodes [
   SetAllowDismiss!() -> (then, InAllowDismiss:bool=~) {is=title K2Node_FunctionEntry}
   > SetButtonVisibility!(
    execute
    self:WBP UIKit Dialog Internal Object Reference=Get Dialog_Internal()
    ButtonIndex:int=0
    Visibility:ESlateVisibility Enum=Select!(
     Option 0:ESlateVisibility Enum=SelfHitTestInvisible
     Option 1:ESlateVisibility Enum=Collapsed
     Index:bool=Get InAllowDismiss() -> (InAllowDismiss:bool)
    ) -> (ReturnValue:ESlateVisibility Enum)
   ) -> (then)
   > SetButtonVisibility(self=Get Dialog_Internal(), ButtonIndex=1, Visibility=Select(Option 0=Collapsed, Option 1=SelfHitTestInvisible, Index=Get InAllowDismiss()))
  ]}
 VM_ShopRefreshTimeUpdated {nodes [
   n1 = VM_ShopRefreshTimeUpdated!() -> (then, InShopVM:Text) {is=title K2Node_FunctionEntry}
   n2 = Get DialogVM_ShopInfo() -> (DialogVM_ShopInfo:Dialog VM Shop Info Object Reference)
   > Set DialogInfoText!(
    execute
    DialogInfoText:Text=Format Text!(Format=Get DialogInfoText!(self:UI Kit Dialog Viewmodel Object Reference=@n2) -> (DialogInfoText:Text), TimeZone:Text=@n1) -> (Result)
    self:UI Kit Dialog Viewmodel Object Reference=Get DialogVM_ShopInfo()
   ) -> (then, Output_Get:Text)
  ]}
]
variables [CompatibilityWidgets {type="array<WBP_UIKit_CompatibilityTags_WithCategoryLabel_C>", default []}]
widgets {
 info {
  inheritedWidgetCount=0
  namedSlotCount=1
  parentClass=&CommonActivatableWidget
  rootWidgetClass=/FortUILibrary/Systems/Dialog/WBP_UIKit_Dialog_Internal.WBP_UIKit_Dialog_Internal_C
  widgetCount=4
 }
 widgets [
  Dialog_Internal {
   widgetClassPath=/FortUILibrary/Systems/Dialog/WBP_UIKit_Dialog_Internal
   bIsVariable=true
   widget=Dialog_Internal
   properties {
    designTimePreview_DialogVM=/ItemShop/MultiProduct/Data/DialogVM_ShopInfo.DialogVM_ShopInfo_C
    useImage=true
    useText=Common Rich Text
    backgroundProperties {
     material=/FortUILibrary/UIKit/Dialogs/MaterialInstances/MI_UI_DialogBackground
     image Size=<32 32>
     placement {
      horizontal Alignment=HAlign_Fill
      vertical Alignment=VAlign_Fill
      grid {column=0, column Span=2, layer=1, row=0, row Span=0, nudge=<0 0>}
      padding=<ltrb 0 0 0 0>
     }
    }
    buttons {
     each Button Horizontal Fill=false
     each Button Min Width=200
     each Button Size=Medium
     focuse Button=0
     space Between Buttons=30
     cTA Buttons Types [Quiet,Loud]
     group Placement {
      horizontal Alignment=HAlign_Fill
      vertical Alignment=VAlign_Bottom
      grid {column=1, column Span=0, layer=0, row=2, row Span=0, nudge=<0 0>}
      padding=<ltrb 80 64 80 48>
     }
    }
    dialogSIze {
     height Override {enable=true, value=824}
     max Aspect Ratio {enable=false, value=0}
     max Desired Height {enable=true, value=800}
     max Desired Width {enable=false, value=0}
     min Aspect Ratio {enable=false, value=0}
     min Desired Height {enable=false, value=0}
     min Desired Width {enable=false, value=0}
     width Override {enable=true, value=1500}
    }
    footerTextProperties {
     auto Wrap Text=true
     justification=Left
     line Height Percentage=1.2
     mobile Font Size Multiplier=1.3
     transform Policy=None
     wrap Text At=0
     color {colorUseRule=UseColor_Specified, specifiedColor=<rgba 1 1 1 0.75>}
     font {
      bForceMonospaced=false
      bMaterialIsStencil=false
      fontMaterial=null
      fontObject=&HeadingNow
      letterSpacing=0
      monospacedWidth=1
      size=15.75
      skewAmount=0
      typefaceFontName=74 Regular
      outlineSettings {
       bApplyOutlineToDropShadows=false
       bMiteredCorners=false
       bSeparateFillAlpha=false
       outlineMaterial=null
       outlineSize=0
       outlineColor=<rgba 0 0 0 1>
      }
     }
     placement {
      horizontal Alignment=HAlign_Left
      vertical Alignment=VAlign_Top
      grid {column=1, column Span=0, layer=0, row=2, row Span=0, nudge=<0 0>}
      padding=<ltrb 80 0 80 0>
     }
    }
    imageProperties {
     material=/FortUILibrary/Systems/Dialog/MaterialInstances/MI_UI_DialogImage
     image Size=<512 512>
     placement {
      horizontal Alignment=HAlign_Fill
      vertical Alignment=VAlign_Fill
      grid {column=1, column Span=0, layer=20, row=0, row Span=1, nudge=<0 0>}
      padding=<ltrb 0 0 0 0>
     }
    }
    textProperties {
     auto Wrap Text=true
     justification=Left
     line Height Percentage=1.2
     mobile Font Size Multiplier=1.3
     transform Policy=None
     wrap Text At=0
     color {colorUseRule=UseColor_Specified, specifiedColor=<rgba 1 1 1 0.75>}
     font {
      bForceMonospaced=false
      bMaterialIsStencil=false
      fontMaterial=null
      fontObject=&HeadingNow
      letterSpacing=0
      monospacedWidth=1
      size=15.75
      skewAmount=0
      typefaceFontName=74 Regular
      outlineSettings {
       bApplyOutlineToDropShadows=false
       bMiteredCorners=false
       bSeparateFillAlpha=false
       outlineMaterial=null
       outlineSize=0
       outlineColor=<rgba 0 0 0 1>
      }
     }
     placement {
      horizontal Alignment=HAlign_Fill
      vertical Alignment=VAlign_Top
      grid {column=1, column Span=0, layer=0, row=1, row Span=0, nudge=<0 0>}
      padding=<ltrb 80 0 80 0>
     }
    }
   }
  }
  ContentAddedAfterDialogInfo {widgetClassPath=&VerticalBox, bIsVariable=true, namedSlotHost=Dialog_Internal, widget=ContentAddedAfterDialogInfo}
  CategoryLabelContainer {
   widgetClassPath=&VerticalBox
   bIsVariable=true
   parent=ContentAddedAfterDialogInfo
   widget=CategoryLabelContainer
   properties {bOverride_Cursor=true}
   slot {class=VerticalBoxSlot, properties {padding=<ltrb 0 20 0 0>}}
  }
  FooterText {
   widgetClassPath=&CommonTextBlock
   bIsVariable=true
   parent=ContentAddedAfterDialogInfo
   widget=FooterText
   properties {
    autoWrapText=true
    mobileFontSizeMultiplier=1.14
    text="Learn more about item and cosmetic variations and limitations at fn.gg/cosmetics\r\n\r\nThat’s it, happy shopping!"
    colorAndOpacity {colorUseRule=UseColor_Specified, specifiedColor=<rgba 1 1 1 0.75>}
    font {
     bForceMonospaced=false
     bMaterialIsStencil=false
     fontMaterial=null
     fontObject=&HeadingNow
     letterSpacing=0
     monospacedWidth=1
     size=15.75
     skewAmount=0
     typefaceFontName=74 Regular
     outlineSettings {
      bApplyOutlineToDropShadows=false
      bMiteredCorners=false
      bSeparateFillAlpha=false
      outlineMaterial=null
      outlineSize=0
      outlineColor=<rgba 0 0 0 1>
     }
    }
    shadowOffset=<0 0>
   }
   slot {class=VerticalBoxSlot, properties {padding=<ltrb 0 25 0 0>}}
  }
 ]
}
