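# Multi-product Item Shop screen widget that hosts the scrolling offer content, section navigator and item preview.
# Activation splits into soft and full paths; full activation starts catalog monitoring and may restore the cached featured layout.
# Content-ready hands the shop viewmodel to child widgets, sets focus and triggers first-time interstitials.
# Child events drive preview, banner, quick-scroll and background updates, and open inspection, list-view, filter and modal screens.
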WidgetBlueprint /ItemShop/MultiProduct/WBP_MPItemShop_Screen
animations [
 CollapseSpacerCameraFrame {
  binding_count=1
  length_seconds=0.23
  object=&self:CollapseSpacerCameraFrame
  sequence {bindings [WBP_MPItemShop_Screen_C_24 {
     id=#1
     possessed_class=WBP_MPItemShop_Screen_C
     tracks [{
       type=MovieSceneDoubleTrack
       display_name=Camera Frame Collapse Animation Value
       sections [{type=MovieSceneDoubleSection, end_frame=4, start_frame=0, channels [None {type=MovieSceneScriptingDoubleChannel, keys [[0,0],[4.6,1]]}]}]
      }]
    }]}
 }
 FadeOut {
  binding_count=1
  length_seconds=0.2
  object=&self:FadeOut
  sequence {bindings [WBP_MPItemShop_Screen_C_6 {
     id=#2
     possessed_class=WBP_MPItemShop_Screen_C
     tracks [{
       type=MovieSceneFloatTrack
       display_name=Render Opacity
       sections [{type=MovieSceneFloatSection, end_frame=4, start_frame=0, channels [None {type=MovieSceneScriptingFloatChannel, keys [[0,1],[4,0]]}]}]
      }]
    }]}
 }
 SongInfoReveal {
  binding_count=1
  length_seconds=0.2
  object=&self:SongInfoReveal
  sequence {bindings [Border_SongInfo {
     id=#3
     possessed_class=Border
     tracks [
      {
       type=MovieSceneFloatTrack
       display_name=Render Opacity
       sections [{type=MovieSceneFloatSection, end_frame=4, start_frame=0, channels [None {type=MovieSceneScriptingFloatChannel, keys [[0,0],[4,1]]}]}]
      }
      {
       type=MovieScene2DTransformTrack
       sections [{
         type=MovieScene2DTransformSection
         end_frame=4
         start_frame=0
         channels [Translation.Y {type=MovieSceneScriptingFloatChannel, keys [[0,-25],[4,0]]}]
        }]
      }
     ]
    }]}
 }
 ArrowEaseIn {
  binding_count=1
  length_seconds=0.75
  object=&self:ArrowEaseIn
  sequence {bindings [ArrowButton {
     id=#4
     possessed_class=WBP_UIKit_ViewMore_ArrowButton_C
     tracks [
      {
       type=MovieSceneByteTrack
       display_name=Visibility
       sections [{type=MovieSceneByteSection, end_frame=0, start_frame=0, channels [None {type=MovieSceneScriptingByteChannel, const [0,1]}]}]
      }
      {
       type=MovieScene2DTransformTrack
       sections [{
         type=MovieScene2DTransformSection
         end_frame=15
         start_frame=0
         channels [Translation.Y {type=MovieSceneScriptingFloatChannel, keys [[0,75],[7,75],[15,0]]}]
        }]
      }
      {
       type=MovieSceneFloatTrack
       display_name=Render Opacity
       sections [{type=MovieSceneFloatSection, end_frame=15, start_frame=0, channels [None {type=MovieSceneScriptingFloatChannel, keys [[0,0],[7,0],[15,1]]}]}]
      }
     ]
    }]}
 }
 ArrowEaseOut {
  binding_count=1
  length_seconds=0.35
  object=&self:ArrowEaseOut
  sequence {bindings [ArrowButton {
     id=#5
     possessed_class=WBP_UIKit_ViewMore_ArrowButton_C
     tracks [
      {
       type=MovieSceneFloatTrack
       display_name=Render Opacity
       sections [{type=MovieSceneFloatSection, end_frame=7, start_frame=0, channels [None {type=MovieSceneScriptingFloatChannel, keys [[0,1],[7,0]]}]}]
      }
      {
       type=MovieScene2DTransformTrack
       sections [{
         type=MovieScene2DTransformSection
         end_frame=7
         start_frame=0
         channels [Translation.Y {type=MovieSceneScriptingFloatChannel, keys [[0,0],[7,75]]}]
        }]
      }
      {
       type=MovieSceneByteTrack
       display_name=Visibility
       sections [{type=MovieSceneByteSection, end_frame=7, start_frame=7, channels [None {type=MovieSceneScriptingByteChannel, const [1,1]}]}]
      }
     ]
    }]}
 }
]
graphs [
 BP_GetDesiredFocusTarget {nodes [
   Get Desired Focus Target()
   > n1 = Branch(Condition=Get IsNavigatingToOffer() -> (IsNavigatingToOffer:bool))
   n2 = GetCurrentRowWidget!(
    execute=[@n8.then, @n1.then, @n7.then]
    self:WBP MPItem Shop Content Object Reference=Get WBP_MPItemShop_Content() -> (WBP_MPItemShop_Content:WBP MPItem Shop Content Object Reference)
   ) -> (then, RowWidget:WBP MPItem Shop Row Wrapper Object Reference)
   > n3 = Is Valid(InputObject=@n2)
   Return Node!(
    execute=[@n8.else, @n7.else, @n3.Is Not Valid]
    ReturnValue:Widget Object Reference=GetSectionNavigatorFocusTarget!(self) -> (ReturnValue:Widget Object Reference)
   )
   > n4 = I_GetDesiredFocusTarget!(execute, self:WBP MPItem Shop Row Wrapper Object Reference=@n2) -> (then, TileWidget:Widget Object Reference)
   > Return Node(ReturnValue=@n4)
   n5 = Branch(execute=@n1.else, Condition=Get IsSoftActivation() -> (IsSoftActivation:bool))
   n6 = IsItemPreviewEnabled!(self:Fort Item Shop VM Object Reference=Get FortItemShopVM() -> (FortItemShopVM:Fort Item Shop VM Object Reference)) -> (ReturnValue:bool)
   > n7 = Branch(Condition=OR Boolean(A=IsSectionNavigatorCollapsed!(self) -> (IsCollapsed:bool), B=NOT Boolean(A=@n6)))
   n8 = Branch(execute=@n5.else, Condition=Get bShouldQuickNavStartCollapsed() -> (bShouldQuickNavStartCollapsed:bool))
  ]}
 BP_GetDesiredInputConfig {nodes [
   Get Desired Input Config() {is=title K2Node_FunctionEntry}
   > Return Node(ReturnValue:UIInput Config Structure=Make UIInput Config!(
     bIgnoreMoveInput:bool=False
     bIgnoreLookInput:bool=False
     InputMode:ECommonInputMode Enum=All
     MouseCaptureMode:EMouseCaptureMode Enum=CaptureDuringMouseDown
     MouseLockMode:EMouseLockMode Enum=DoNotLock
     bHideCursorDuringViewportCapture:bool=True
    ) -> (UIInputConfig:UIInput Config Structure) {is=title K2Node_MakeStruct})
  ]}
 BindToContentScrollEvents {nodes [
   BindToContentScrollEvents() {is=title K2Node_FunctionEntry}
   n1 = Create Event()
   n2 = Get WBP_MPItemShop_Content()
   > Unbind Event from On Scroll Started(self:WBP MPItem Shop Content Object Reference=@n2, Delegate=@n1) {is=title K2Node_RemoveDelegate}
   > Bind Event to On Scroll Started(self:WBP MPItem Shop Content Object Reference=@n2, Delegate=@n1) {is=title K2Node_AddDelegate}
   n3 = Create Event()
   n4 = Get WBP_MPItemShop_Content()
   > Unbind Event from On Scroll Finished(self:WBP MPItem Shop Content Object Reference=@n4, Delegate=@n3) {is=title K2Node_RemoveDelegate}
   > Bind Event to On Scroll Finished(self:WBP MPItem Shop Content Object Reference=@n4, Delegate=@n3) {is=title K2Node_AddDelegate}
  ]}
 CanShowFtueModal {nodes [
   CanShowFtueModal() {is=title K2Node_FunctionEntry}
   n1 = GetHasSeenAboutShopModal!(self:Fort Item Shop VM Object Reference=Get FortItemShopVM()) -> (ReturnValue:bool)
   n2 = Get HasShownAboutShopModal() -> (HasShownAboutShopModal:bool)
   > Return Node(ReturnValue:bool=AND Boolean(A=NOT Boolean(A=@n2), B=NOT Boolean(A=@n1)))
  ]}
 CancelRequestMarkAllOfferGroupsAsItemShopVisited {nodes [
   CancelRequestMarkAllOfferGroupsAsItemShopVisited() {is=title K2Node_FunctionEntry}
   > Clear and Invalidate Timer by Handle(Handle=Get MarkAllOfferGroupsAsItemShopVisitedHandler() -> (MarkAllOfferGroupsAsItemShopVisitedHandler:Timer Handle Structure))
  ]}
 CreateNavigationBindings {nodes [
   CreateNavigationBindings() {is=title K2Node_FunctionEntry}
   > CreateSectionNavigatorBindings()
   n1 = Get WBP_MPItemShop_Content()
   > Unbind all Events from On Content Focused!(execute:Exec, self:WBP MPItem Shop Content Object Reference=@n1) -> (then:Exec) {is=title K2Node_ClearDelegate}
   > Bind Event to On Content Focused(self:WBP MPItem Shop Content Object Reference=@n1, Delegate=Create Event()) {is=title K2Node_AddDelegate}
  ]}
 CreateSectionNavigatorBindings {nodes [
   CreateSectionNavigatorBindings() {is=title K2Node_FunctionEntry}
   n1 = Get WBP_ItemShop_SectionNavigator_new() -> (WBP_ItemShop_SectionNavigator_new:WBP Item Shop Side Navigation Object Reference)
   > Unbind all Events from On Section Navigator Focused!(execute:Exec, self:WBP Item Shop Side Navigation Object Reference=@n1) -> (then:Exec) {is=title K2Node_ClearDelegate}
   > Unbind all Events from On Section Navigator Unfocused!(execute:Exec, self:WBP Item Shop Side Navigation Object Reference=@n1) -> (then:Exec) {is=title K2Node_ClearDelegate}
   > Bind Event to On Section Navigator Focused(self:WBP Item Shop Side Navigation Object Reference=@n1, Delegate=Create Event()) {is=title K2Node_AddDelegate}
   > Bind Event to On Section Navigator Unfocused(self:WBP Item Shop Side Navigation Object Reference=@n1, Delegate=Create Event()) {is=title K2Node_AddDelegate}
  ]}
 EventGraph {nodes [
   n1 = "On Clicked (Button_AboutShop)"!() -> (OutputDelegate, then, Button:Common Button Base Object Reference) {is=title K2Node_ComponentBoundEvent}
   > Fire Event Element Interacted(Element=@n1.Button)
   > SetAllInteractionsEnabled!(execute, self, bInIsInteractionEnabled:bool=false) -> (then)
   > SetupAndPushInfoModal()
   n2 = "On Clicked (Button_SAC)"!() -> (OutputDelegate, then, Button:Common Button Base Object Reference) {is=title K2Node_ComponentBoundEvent}
   > Fire Event Element Interacted(Element=@n2.Button)
   > SetAllInteractionsEnabled(bInIsInteractionEnabled=false)
   > n3 = Create Widget!(
    execute:Exec
    Class:User Widget Class Reference=&/WBP_SupportACreator_Dialog.WBP_SupportACreator_Dialog_C
    OwningPlayer:Player Controller Object Reference
    bSkipAutoPopulate:bool=False
    SourceName:String="<Undefined>"
   ) -> (then:Exec, ReturnValue:WBP Support ACreator Dialog Object Reference)
   > PushModal!(execute, self, ActivatableWidget:Common Activatable Widget Object Reference=@n3) -> (then)
   "On Clicked (ArrowButton)"!() -> (OutputDelegate, then, Button:Common Button Base Object Reference) {is=title K2Node_ComponentBoundEvent}
   > RequestScroll!(execute, self:WBP MPItem Shop Content Object Reference=Get WBP_MPItemShop_Content(), Direction:EScrollDirection Enum=Scroll_Down) -> (then)
   "On Hovered (Button_Filters)"!() -> (OutputDelegate, then, Button:Common Button Base Object Reference) {is=title K2Node_ComponentBoundEvent}
   > Handle Quicknav or Content Focus!(execute, self, ShouldFocusSectionNavigator:bool=true, IsImmediate:bool=false) -> (then)
   n4 = "On Item 2d Download Started (FortItemPreviewWidget)"!() -> (OutputDelegate, then, ImageUrl:String) {is=title K2Node_ComponentBoundEvent}
   > RegisterImageDownloadStarted!(
    execute
    self:Fort Item Shop Selection VM Object Reference=Get FortItemShopSelectionVM() -> (FortItemShopSelectionVM:Fort Item Shop Selection VM Object Reference)
    ImageUrl:String=@n4.ImageUrl
   ) -> (then)
   "On Clicked (Button_Filters)"!() -> (OutputDelegate, then, Button:Common Button Base Object Reference) {is=title K2Node_ComponentBoundEvent}
   > n5 = Create Widget!(execute, Class=/ItemShop/MultiProduct/Filter/WBP_ShopFilter_Dialog.WBP_ShopFilter_Dialog_C, OwningPlayer) -> (then, ReturnValue:WBP Shop Filter Dialog Object Reference)
   > PushModal(ActivatableWidget=@n5)
   "On Preview Item Shown (FortItemPreviewWidget)"!() -> (OutputDelegate, then, ItemDefinition:Fort Account Item Definition Object Reference) {is=title K2Node_ComponentBoundEvent}
   > n6 = Sequence()
   > Delay(Duration=0.25)
   > n7 = UpdateSongInfoSizing!(execute, self) -> (then, IsValid:bool)
   n8 = Get FortItemShopVM()
   > UpdateSongInfoVisibility!(execute, self, ShouldShow:bool=AND Boolean!(A=@n7, B=IsSectionNavigatorCollapsed(), C:bool=IsItemPreviewEnabled(self=@n8)) -> (ReturnValue)) -> (then)
   OnItemPreviewIsReady(execute=@n6.then_1)
   n9 = "Quick Scroll State Changed (QuickScrollDetector)"!() -> (OutputDelegate, then, IsQuickScrolling:bool) {is=title K2Node_ComponentBoundEvent}
   > n10 = Set Is Quick Scrolling(Is Quick Scrolling:bool=@n9.IsQuickScrolling) -> (Output_Get:bool)
   > SetQuickScrollState!(execute, self:Fort MPItem Shop Screen Object Reference, bInIsQuickScrolling:bool=@n10) -> (then)
   n11 = "On Item 2d Download Completed (FortItemPreviewWidget)"!() -> (OutputDelegate, then, ImageUrl:String, Result:bool) {is=title K2Node_ComponentBoundEvent}
   > RegisterImageDownloadCompleted!(
    execute
    self:Fort Item Shop Selection VM Object Reference=Get FortItemShopSelectionVM()
    ImageUrl:String=@n11.ImageUrl
    bSuccess:bool=@n11.Result
   ) -> (then)
   n12 = "On Item 3d Download Started (FortItemPreviewWidget)"!() -> (OutputDelegate, then, Item:Fort Item VM Object Reference) {is=title K2Node_ComponentBoundEvent}
   > RegisterItemDownloadStarted!(execute, self:Fort Item Shop Selection VM Object Reference=Get FortItemShopSelectionVM(), Item:Fort Item VM Object Reference=@n12.Item) -> (then)
   n13 = "On Item 3d Download Completed (FortItemPreviewWidget)"!() -> (OutputDelegate, then, Item:Fort Item VM Object Reference, Result:bool) {is=title K2Node_ComponentBoundEvent}
   > RegisterItemDownloadCompleted!(
    execute
    self:Fort Item Shop Selection VM Object Reference=Get FortItemShopSelectionVM()
    Item:Fort Item VM Object Reference=@n13.Item
    bSuccess:bool=@n13.Result
   ) -> (then)
   n14 = "On Clicked (Button_BackToGame)"!() -> (OutputDelegate, then, Button:Common Button Base Object Reference) {is=title K2Node_ComponentBoundEvent}
   n15 = Branch(execute=[@n14.then, @n20.then], Condition=IsInHestiaBeauty!() -> (ReturnValue:bool))
   > EnterCameraMode(self:Fort HUDContext Object Reference=Get FortHUDContext!() -> (ReturnValue:Fort HUDContext Object Reference) {is=title K2Node_GetBlueprintContextEx})
   HandleBackToGameRequested(execute=@n15.else, self:Fort MPItem Shop Screen Object Reference)
   n16 = "On Offer Focused (WBP_MPItemShop_Content)"!() -> (OutputDelegate, then, OfferId:String) {is=title K2Node_ComponentBoundEvent}
   > n17 = Set FocusedOfferId(FocusedOfferId:String=@n16.OfferId) -> (Output_Get:String)
   > HandleBanner!(
    execute
    self
    OfferVM:Fort Item Shop Offer VM Object Reference=GetOfferFromId!(self:Fort Item Shop VM Object Reference=Get FortItemShopVM(), OfferId:String=@n17, bSearchFromCatalog:bool=false) -> (ReturnValue:Fort Item Shop Offer VM Object Reference)
   ) -> (then)
   "On Clicked (Button_ClearFilters)"!() -> (OutputDelegate, then, Button:Common Button Base Object Reference) {is=title K2Node_ComponentBoundEvent}
   > ClearFilters(self:Fort Item Shop Filter VM Object Reference=Get FilterVM!(self:Fort Item Shop VM Object Reference=Get FortItemShopVM()) -> (FilterVM:Fort Item Shop Filter VM Object Reference))
   > Clear(TargetArray:Array of Fort Item Shop Offer Filter VM Object References=Get CachedAppliedFilters() -> (CachedAppliedFilters:Array of Fort Item Shop Offer Filter VM Object References))
   > Clear(TargetArray:Array of Fort Option Item VM Object References=Get CachedAppliedOptions() -> (CachedAppliedOptions:Array of Fort Option Item VM Object References))
   > UpdateContentOrdering(self:Fort Item Shop VM Object Reference=Get FortItemShopVM())
   > Set FeaturedLayout!(execute, FeaturedLayout:Fort Item Shop Layout Structure, self:Fort Item Shop VM Object Reference=Get FortItemShopVM()) -> (then, Output_Get:Fort Item Shop Layout Structure)
   > Set ShouldResetFeaturedLayout(ShouldResetFeaturedLayout:bool=true) -> (Output_Get:bool)
   > OnShopContentReady()
   n18 = "On Clicked (Button_ShowPreviewTouch)"!() -> (OutputDelegate, then, Button:Common Button Base Object Reference) {is=title K2Node_ComponentBoundEvent}
   > Fire Event Element Interacted(Element=@n18.Button)
   > Handle Quicknav or Content Focus(ShouldFocusSectionNavigator=false, IsImmediate=false)
   n19 = "On Clicked (Button_OpenQuickNavTouch)"!() -> (OutputDelegate, then, Button:Common Button Base Object Reference) {is=title K2Node_ComponentBoundEvent}
   > Fire Event Element Interacted(Element=@n19.Button)
   > Handle Quicknav or Content Focus(ShouldFocusSectionNavigator=true, IsImmediate=false)
   n20 = "On Clicked (Button_BackToGame_NoOffers)"!() -> (OutputDelegate, then, Button:Common Button Base Object Reference) {is=title K2Node_ComponentBoundEvent}
   n21 = "On List View Requested (WBP_MPItemShop_Content)"!() -> (OutputDelegate, then, OfferGroupVM:Fort Item Shop Offer Group VM Object Reference) {is=title K2Node_ComponentBoundEvent}
   > Do Once(Reset=@n85.Finished)
   > n22 = "Play Animation with Finished event (legacy)"!(
    execute:Exec
    Widget:User Widget Object Reference=Self-Reference()
    InAnimation:Widget Animation Object Reference=Get FadeOut() -> (FadeOut:Widget Animation Object Reference)
    StartAtTime:float=0
    NumLoopsToPlay:int=1
    PlayMode:EUMGSequencePlayMode Enum
    PlaybackSpeed:float=1
   ) -> (then:Exec, Result:UMGSequence Player Object Reference, Finished:Exec) {is=title K2Node_PlayAnimation}
   n23 = Sequence(execute=@n22.Finished)
   > Set IsSoftActivation(IsSoftActivation:bool=true) -> (Output_Get:bool)
   > Set IsShowingOfferInspect(IsShowingOfferInspect:bool=true) -> (Output_Get:bool)
   > Set IsNavigatingToOffer(IsNavigatingToOffer:bool=false) -> (Output_Get:bool)
   n24 = PushPrimaryContentClassAsync!(
    execute:Exec=@n23.then_1
    WidgetClass:Common Activatable Widget Class Reference=/ItemShop/MultiProduct/ListView/WBP_MPItemShop_ListView.WBP_MPItemShop_ListView_C
   ) -> (then:Exec, OnPrimaryContentPushed:Exec, ActivatableWidget:Common Activatable Widget Object Reference) {is=title K2Node_AsyncAction}
   n25 = Cast To WBP_MPItemShop_ListView!(execute=@n24.OnPrimaryContentPushed, Object=@n24) -> (then, CastFailed, AsWBP MPItem Shop List View:WBP MPItem Shop List View Object Reference) {is=title K2Node_DynamicCast}
   > n26 = Set FortItemShopOfferGroupVM!(
    execute
    FortItemShopOfferGroupVM:Fort Item Shop Offer Group VM Object Reference=@n21.OfferGroupVM
    self:WBP MPItem Shop List View Object Reference=@n25
   ) -> (then, Output_Get:Fort Item Shop Offer Group VM Object Reference)
   n27 = OnListViewDeactivated() {is=title K2Node_CustomEvent}
   Bind Event to On Widget Deactivated(execute=@n26.then, self:Common Activatable Widget Object Reference=@n25, Delegate=@n27) {is=title K2Node_AddDelegate}
   n28 = "On Clicked (Button_PCB)"!() -> (OutputDelegate, then, Button:Common Button Base Object Reference) {is=title K2Node_ComponentBoundEvent}
   > Fire Event Element Interacted(Element=@n28.Button)
   > SetAllInteractionsEnabled(bInIsInteractionEnabled=false)
   > SetupAndPushPcbModal()
   OnContentScrollFinished() {is=title K2Node_CustomEvent}
   n29 = Get QuickScrollDetector() -> (QuickScrollDetector:WBP UIKit Quick Scroll Detector Object Reference)
   > OnScrollFinished(self:WBP UIKit Quick Scroll Detector Object Reference=@n29)
   OnSectionNavigatorFocused() {is=title K2Node_CustomEvent}
   n30 = IsUsingTouch!() -> (ReturnValue)
   > n31 = Branch(Condition=@n30)
   > Set IsSoftActivation(IsSoftActivation=false)
   Handle Quicknav or Content Focus(execute=@n31.else, ShouldFocusSectionNavigator=true, IsImmediate=false)
   > SpawnSound2D!(
    execute
    Sound:Sound Base Object Reference=Get NavRevealWhooshSound() -> (NavRevealWhooshSound:Sound Base Object Reference)
    VolumeMultiplier:float=1
    PitchMultiplier:float=1
    StartTime:float=0
    ConcurrencySettings:Sound Concurrency Object Reference
    bPersistAcrossLevelTransition:bool=false
    bAutoDestroy:bool=true
   ) -> (then, ReturnValue:Audio Component Object Reference)
   OnContentUpdating() {is=title K2Node_CustomEvent}
   > Branch(Condition=AND Boolean!(
     A=Get HasCompletedFullActivation() -> (HasCompletedFullActivation:bool)
     B=IsActivated!(self:Common Activatable Widget Object Reference) -> (ReturnValue:bool)
    ) -> (ReturnValue))
   > Set CachedFeaturedLayout(CachedFeaturedLayout:Fort Item Shop Layout Structure) -> (Output_Get:Fort Item Shop Layout Structure)
   n32 = LoadAndUpdateBackground!() -> (OutputDelegate, then, ImmediateTransition:bool=~) {is=title K2Node_CustomEvent}
   > Branch(Condition=Get bIsActive() -> (bIsActive:bool))
   > n33 = Branch(Condition=IsItemShopDynamicBackgroundEnabled!(self:Fort Item Shop VM Object Reference=Get FortItemShopVM()) -> (ReturnValue:bool))
   n34 = Get FortItemShopVM()
   n35 = OnBackgroundDataLoaded!() -> (OutputDelegate, then, LoadedBackground:Object Reference) {is=title K2Node_CustomEvent}
   n36 = LoadBackground!(
    execute=@n33.then
    self:Fort Item Shop VM Object Reference=Get FortItemShopVM()
    Background:"Fort Item Shop CMSBackground Structure (by ref)"=GetPreferredBackgroundForLayout!(
     self:Fort Item Shop VM Object Reference=@n34
     InLayout:"Fort Item Shop Layout Structure (by ref)"=Get FeaturedLayout!(self:Fort Item Shop VM Object Reference=@n34) -> (FeaturedLayout:Fort Item Shop Layout Structure)
    ) -> (ReturnValue:Fort Item Shop CMSBackground Structure)
    OnLoadedCallback:"Delegate (by ref)"=@n35
   ) -> (then, ReturnValue:bool)
   > n37 = Branch(Condition=@n36)
   n38 = Get FortItemShopSelectionVM()
   n39 = Select(
    Option 0:EVaultWorldTransitionDirection Enum=Select!(
     NoScroll:EVaultWorldTransitionDirection Enum=None
     ScrolledDown:EVaultWorldTransitionDirection Enum=Forward
     ScrolledUp:EVaultWorldTransitionDirection Enum=Backward
     Index:EFortItemShopScrollDirection Enum=Get LastScrolledDirection!(self:Fort Item Shop Selection VM Object Reference=@n38) -> (LastScrolledDirection:EFortItemShopScrollDirection Enum=NoScroll)
    ) -> (ReturnValue)
    Option 1:EVaultWorldTransitionDirection Enum=None
    Index=@n32.ImmediateTransition
   ) -> (ReturnValue:EVaultWorldTransitionDirection Enum)
   UpdateBackground!(
    execute=[@n37.else, @n33.else]
    self:Fort MPItem Shop Screen Object Reference
    LoadedBackground:Object Reference
    TransitionDirection:EVaultWorldTransitionDirection Enum=@n39
    PrimaryColor:Linear Color Structure=(R=0,G=0,B=0,A=0)
    SecondaryColor:Linear Color Structure=(R=0,G=0,B=0,A=0)
   ) -> (then)
   OnContentFocused() {is=title K2Node_CustomEvent}
   > n40 = Branch(Condition=@n30)
   > Set IsSoftActivation(IsSoftActivation=false)
   Handle Quicknav or Content Focus(execute=@n40.else, ShouldFocusSectionNavigator=false, IsImmediate=false)
   > SpawnSound2D(
    Sound=Get NavHideWhooshSound() -> (NavHideWhooshSound:Sound Base Object Reference)
    VolumeMultiplier=1
    PitchMultiplier=1
    StartTime=0
    bPersistAcrossLevelTransition=false
    bAutoDestroy=true
   )
   n41 = OnSetFeaturedOfferId!() -> (OutputDelegate, then, FeaturedOfferId:String) {is=title K2Node_CustomEvent}
   n42 = Get IsShowingOfferInspect() -> (IsShowingOfferInspect:bool)
   n43 = IsEmpty(InString=@n41.FeaturedOfferId)
   n44 = Get IsNavigatingToOffer()
   > Branch(Condition=AND Boolean!(A=Get bIsActive(), B=NOT Boolean(A=@n43), C:bool=NOT Boolean(A=@n44), D:bool=NOT Boolean(A=@n42)) -> (ReturnValue))
   > Set RequestedOfferGroupVM(RequestedOfferGroupVM:Fort Item Shop Offer Group VM Object Reference=GetOfferGroupWithOffer!(self:Fort Item Shop VM Object Reference=Get FortItemShopVM(), OfferId:String=@n41.FeaturedOfferId) -> (ReturnValue:Fort Item Shop Offer Group VM Object Reference)) -> (Output_Get:Fort Item Shop Offer Group VM Object Reference)
   n45 = Get FortItemShopVM()
   > n46 = Set RequestedOfferVM(RequestedOfferVM:Fort Item Shop Offer VM Object Reference=GetOfferFromId(self=@n45, OfferId=@n41.FeaturedOfferId, bSearchFromCatalog=true)) -> (Output_Get:Fort Item Shop Offer VM Object Reference)
   > n47 = Is Valid(InputObject=@n46)
   n48 = Set RequestedOfferVM(execute=@n47.Is Not Valid, RequestedOfferVM=GetOfferFromId(self=@n45, OfferId=@n41.FeaturedOfferId, bSearchFromCatalog=false))
   > n49 = Is Valid(InputObject=@n48)
   PrintString!(
    execute=@n49.Is Not Valid
    InString="To String (Text)"(InText=Format Text!(Format="LOCTEXT(\"Couldn\\'t find offer VM for offer id: {OfferId}\")", OfferId:String=@n41.FeaturedOfferId) -> (Result))
    bPrintToScreen
    bPrintToLog
    TextColor=(R=1,G=0,B=0,A=1)
    Duration
    Key
   ) -> (then)
   Set IsSoftActivation(execute=[@n47.Is Valid, @n49.Is Valid], IsSoftActivation=true)
   > Set IsShowingOfferInspect(IsShowingOfferInspect=true)
   > Set IsNavigatingToOffer(IsNavigatingToOffer=false)
   > RequestInspectOffer!(
    execute
    self:Fort MPItem Shop Screen Object Reference
    OfferVM:Fort Item Shop Offer VM Object Reference=Get RequestedOfferVM() -> (RequestedOfferVM:Fort Item Shop Offer VM Object Reference)
    OwningOfferGroupVM:Fort Item Shop Offer Group VM Object Reference=Get RequestedOfferGroupVM() -> (RequestedOfferGroupVM:Fort Item Shop Offer Group VM Object Reference)
   ) -> (then)
   OnSectionNavigatorUnfocused() {is=title K2Node_CustomEvent}
   > n50 = GetCurrentRowWidget(self=Get WBP_MPItemShop_Content())
   > n51 = I_GetDesiredFocusTarget(self=@n50)
   > SetFocus(self=@n51)
   OnItemPreviewIsReady() {is=title K2Node_CustomEvent}
   > Set IsItemPreviewReady(IsItemPreviewReady:bool=true) -> (Output_Get:bool)
   n52 = Get IsShowingOfferInspect()
   n53 = Get FortItemShopSelectionVM()
   n54 = Length(TargetArray:Array of Item Shop Preview Data Object References=Get ItemsToPreview!(self:Fort Item Shop Selection VM Object Reference=@n53) -> (ItemsToPreview:Array of Item Shop Preview Data Object References))
   > Branch(Condition=AND Boolean!(
     A="integer > integer"(A=@n54, B=1)
     B=Get bIsItemPreviewCycleEnabled!(self:Fort Item Shop Selection VM Object Reference=@n53) -> (bIsItemPreviewCycleEnabled:bool)
     C=NOT Boolean(A=@n52)
    ) -> (ReturnValue))
   > Clear and Invalidate Timer by Handle(Handle=Get ItemPreviewTimer() -> (ItemPreviewTimer:Timer Handle Structure))
   > n55 = Set Timer by Event(Delegate=Create Event(), Time=Get DelayBetweenItemPreviews() -> (DelayBetweenItemPreviews:double), bLooping=false)
   > Set ItemPreviewTimer(ItemPreviewTimer:Timer Handle Structure=@n55) -> (Output_Get:Timer Handle Structure)
   OnContentScrollStarted() {is=title K2Node_CustomEvent}
   > OnScrollStarted(self:WBP UIKit Quick Scroll Detector Object Reference=@n29)
   OnShopContentReady() {is=title K2Node_CustomEvent}
   > Handle MOTD Request()
   > n56 = Branch(Condition=Get ShouldResetFeaturedLayout() -> (ShouldResetFeaturedLayout:bool))
   > Set FeaturedLayout!(execute, FeaturedLayout=GetFirstSectionLayout!(self) -> (Owning Layout:Fort Item Shop Layout Structure), self=Get FortItemShopVM()) -> (then, Output_Get_SectionId:String, Output_Get_SubsectionId:String)
   > n57 = Set ShouldResetFeaturedLayout(ShouldResetFeaturedLayout=false)
   SetSectionNavigatorFortItemShopVM(execute=[@n66.then, @n57.then])
   > SetFortItemShopVM!(
    execute
    self:WBP MPItem Shop Content Object Reference=Get WBP_MPItemShop_Content()
    Viewmodel:Fort Item Shop VM Object Reference=Get FortItemShopVM()
   ) -> (then)
   > n58 = Branch(Condition=IsActivated(self=Get WBP_MPItemShop_Content()))
   > n59 = OnSoftActivation!(execute, self:WBP MPItem Shop Content Object Reference=Get WBP_MPItemShop_Content(), bRowsDirty:bool=true) -> (then)
   Handle Quicknav or Content Focus(execute=[@n59.then, @n65.then], ShouldFocusSectionNavigator=NOT Boolean(A=Get bShouldQuickNavStartCollapsed()), IsImmediate=true)
   > RequestRefreshFocus(self:Common Activatable Widget Object Reference)
   > ActivateWidget(self=Get QuickScrollDetector())
   > n60 = Branch(Condition=IsActivated())
   > n61 = Set HasCompletedFullActivation(HasCompletedFullActivation:bool=true) -> (Output_Get:bool)
   SetAllInteractionsEnabled(execute=[@n60.else, @n61.then], bInIsInteractionEnabled=true)
   > UpdateArrowButton()
   > n62 = Branch(Condition=CanShowFtueModal!() -> (ReturnValue:bool) {is=title K2Node_CallFunction})
   > n63 = HandleAboutShopFTUE()
   HandlePcbShopFTUE(execute=[@n64.then, @n63.then])
   n64 = RequestEnterShopInterstitials!(execute=@n62.else, self:Fort Item Shop VM Object Reference=Get FortItemShopVM()) -> (then, ReturnValue:bool)
   n65 = ActivateWidget(execute=@n58.else, self=Get WBP_MPItemShop_Content())
   Set FeaturedLayout!(
    execute=@n56.else
    FeaturedLayout=Get CachedFeaturedLayout() -> (CachedFeaturedLayout:Fort Item Shop Layout Structure)
    self=Get FortItemShopVM()
   ) -> (then, Output_Get:Fort Item Shop Layout Structure)
   > n66 = Set CachedFeaturedLayout()
   Event On Activated()
   > ActivateWidget(self=Get FortItemPreviewWidget() -> (FortItemPreviewWidget:Fort Item Preview Widget Object Reference))
   > n67 = Branch(Condition=Get IsSoftActivation())
   > n68 = SoftActivate()
   Branch(execute=[@n68.then, @n69.then], Condition=IsUsingTouch())
   > Update Preview Quick Nav Buttons Visibility()
   n69 = FullActivate(execute=@n67.else)
   Event On Deactivated()
   n70 = Get FortItemShopVM()
   > FlushImpressions!(
    execute
    self:Fort Item Shop Selection VM Object Reference=Get FortItemShopSelectionVM()
    ImpressionsEventAdditionalInfo:"Flow Analytics Additional Info Structure (by ref)"=GetLastFilterContextAsAdditionalInfo!(self:Fort Item Shop Filter VM Object Reference=Get FilterVM(self=@n70)) -> (ReturnValue:Flow Analytics Additional Info Structure)
   ) -> (then)
   > SetCurrentPreviewItem!(execute, self:Fort Item Preview Widget Object Reference=Get FortItemPreviewWidget(), ItemVM:Fort Item VM Object Reference) -> (then)
   > DeactivateWidget(self=Get FortItemPreviewWidget())
   > n71 = Branch(Condition=Get IsSoftActivation())
   > SoftDeactivate()
   FullDeactivate(execute=@n71.else)
   Event Construct()
   > n72 = Sequence!(execute) -> (then_0, then_1, then_2:Exec, then_3:Exec)
   SetNavigationRules(execute=@n72.then_1)
   > BindToContentScrollEvents() {is=title K2Node_CallFunction}
   SetCompactUIStyle(execute=@n72.then_2) {is=title K2Node_CallFunction}
   > Set Gamepad Action Bar Style() {is=title K2Node_CallFunction}
   > SetSTWStyle()
   > SetButtonStyling()
   SetTouchForwardTarget!(
    execute=@n72.then_3
    self:Athena Scroll Box Object Reference=Get ScrollBox_Content() -> (ScrollBox_Content:Athena Scroll Box Object Reference)
    InTarget:Athena Scroll Box Object Reference=Get ScrollBox_Content!(self:Fort MPItem Shop Content Object Reference=Get WBP_MPItemShop_Content()) -> (ScrollBox_Content)
   ) -> (then)
   n73 = StartListeningToEvent!(
    execute:Exec=@n72.then_0
    Target:Gameplay Event Router Component Object Reference=GetGlobalEventRouter()
    EventType:Script Struct Object Reference=/Script/FortniteGame.GameplayMessageDualBladeMenuVisible
    Context:Object Reference=&self
    EventBubblingRule:EEventBubblingRule Enum
    None:%=Get GameplayMessageRouter() -> (ReturnValue:Gameplay Message Router Object Reference) {is=title K2Node_GetSubsystem}
   ) -> (then:Exec, AsyncTaskProxy:Async Action Start Listening To Event Object Reference, OnEventReceived:Exec, Payload_bBladeMenuVisible:bool=~)
   > n74 = StartListeningToEvent!(
    execute
    Target=GetGlobalEventRouter()
    EventType=/FortniteCoreUI/UI/Social/Chat/TextChat/BPS_TextChatActiveStateChange
    Context=GetOwningPlayer()
    EventBubblingRule
   ) -> (then, AsyncTaskProxy, OnEventReceived, Payload:BPS Text Chat Active State Change Structure)
   > Set ChatEventListener(ChatEventListener:Async Action Start Listening To Event Object Reference=@n74) -> (Output_Get:Async Action Start Listening To Event Object Reference)
   n75 = Set IsChatOpen(
    execute=@n74.OnEventReceived
    IsChatOpen:bool=Break BPS Text Chat Active State Change!(BPS_TextChatActiveStateChange:"BPS Text Chat Active State Change Structure (by ref)"=@n74.Payload) -> (bActive_4_#6:bool=~) {is=title K2Node_BreakStruct}
   ) -> (Output_Get:bool)
   > SetButtonPromptsVisibility!(execute, self, ShouldShow:bool=NOT Boolean(A=@n75)) -> (then)
   n76 = Sequence!(execute=@n73.OnEventReceived) -> (then_0, then_1)
   > Branch(Condition=AND Boolean!(A=@n73.Payload_bBladeMenuVisible, B=IsActivated()) -> (ReturnValue))
   > Set IsSoftActivation(IsSoftActivation=true)
   n77 = Get IsChatOpen() -> (IsChatOpen:bool)
   SetButtonPromptsVisibility(execute=@n76.then_1, ShouldShow=NOT Boolean(A=OR Boolean(A=@n73.Payload_bBladeMenuVisible, B=@n77)))
   n78 = Event OnNavigateToOffer!() -> (OutputDelegate, then, OfferTrackingId:String, bShowOfferDetails:bool) {is=title K2Node_Event}
   > Set MOTD_TrackingID(MOTD_TrackingID:String=@n78.OfferTrackingId) -> (Output_Get:String)
   > Set IsNavigatingToOffer(IsNavigatingToOffer=NOT Boolean(A=@n78.bShowOfferDetails))
   n79 = Event OnOfferInspectionScreenShown!() -> (OutputDelegate, then, OfferInspectionScreenWidget:Common Activatable Widget Object Reference) {is=title K2Node_Event}
   > n80 = Cast To FortMPItemShopOfferInspectionScreen!(execute, Object=@n79.OfferInspectionScreenWidget) -> (then, CastFailed, AsFort MPItem Shop Offer Inspection Screen:Fort MPItem Shop Offer Inspection Screen Object Reference) {is=title K2Node_DynamicCast}
   > SetViewModels!(
    execute
    self:Fort MPItem Shop Offer Inspection Screen Object Reference=@n80
    OfferVM:Fort Item Shop Offer VM Object Reference=Get RequestedOfferVM()
    OfferGroupVM:Fort Item Shop Offer Group VM Object Reference=Get RequestedOfferGroupVM()
   ) -> (then)
   Event OnInputMethodChanged() {is=title K2Node_Event}
   > Set IsSoftActivation(IsSoftActivation=true)
   > Set Gamepad Action Bar Style()
   > SetButtonStyling()
   > ScrollToFocusedRowIndex!(execute, self:WBP MPItem Shop Content Object Reference=Get WBP_MPItemShop_Content(), bForceRefreshAll:bool=true) -> (then)
   > Handle Quicknav or Content Focus(ShouldFocusSectionNavigator=NOT Boolean(A=IsSectionNavigatorCollapsed()), IsImmediate=true)
   n81 = GetCurrentInputType(self=Get CommonInputSubsystem())
   > SetSize!(
    execute
    self:Spacer Object Reference=Get Spacer_Filter_Buttons() -> (Spacer_Filter_Buttons:Spacer Object Reference)
    InSize:Vector 2D Structure=Select!(Option 0:Vector 2D Structure=(X=1,Y=256), Option 1:Vector 2D Structure=(X=440,Y=256), Index:bool="Equal (Enum)"(A=@n81, B=Touch)) -> (ReturnValue:Vector 2D Structure)
   ) -> (then)
   > n82 = Branch(Condition=IsItemPreviewEnabled(self=Get FortItemShopVM()))
   PreviewSingleItem!(execute=@n82.else, self, ItemToPreview:Item Shop Preview Data Object Reference) -> (then)
   n83 = Event OnScrollInput!() -> (OutputDelegate, then, ScrollAmount:float) {is=title K2Node_Event}
   > HandleScroll!(
    execute
    self:WBP MPItem Shop Content Object Reference=Get WBP_MPItemShop_Content()
    OffsetAmount:"Float (double-precision) (by ref)"=@n83.ScrollAmount
   ) -> (then)
   Event OnHandlePreviewItem() {is=title K2Node_Event}
   > RefreshWidgetCameraFrame()
   Event Destruct()
   > StopListeningToEvent(self:Async Action Start Listening To Event Object Reference=Get ChatEventListener() -> (ChatEventListener:Async Action Start Listening To Event Object Reference))
   > Set ChatEventListener()
   UpdateBackground(
    execute=@n84.then
    LoadedBackground=@n35.LoadedBackground
    TransitionDirection=@n39
    PrimaryColor=(R=0,G=0,B=0,A=0)
    SecondaryColor=(R=0,G=0,B=0,A=0)
   )
   OnCatalogRefreshing() {is=title K2Node_CustomEvent}
   n84 = Branch(execute=@n35.then, Condition=Get bIsActive())
   n85 = "Play Animation with Finished event (legacy)"(execute=@n27.then, Widget=Self-Reference(), InAnimation=Get FadeOut(), StartAtTime=0, NumLoopsToPlay=1, PlayMode=Reverse, PlaybackSpeed=1)
  ]}
 FocusContent {nodes [
   FocusContent!() -> (then, Navigation:EUINavigation Enum) {is=title K2Node_FunctionEntry}
   > n1 = Cast To WBP_MPItemShop_RowWrapper!(
    execute
    Object=GetDesiredFocusTarget!(self:Common Activatable Widget Object Reference=Get WBP_MPItemShop_Content()) -> (ReturnValue:Widget Object Reference)
   ) -> (then, CastFailed, AsWBP MPItem Shop Row Wrapper:WBP MPItem Shop Row Wrapper Object Reference) {is=title K2Node_DynamicCast}
   > n2 = I_GetDesiredFocusTarget(self=@n1)
   > Return Node(ReturnValue:Widget Object Reference=@n2)
   Return Node(execute=@n1.CastFailed)
  ]}
 FocusFilterButton {nodes [
   FocusFilterButton!() -> (then, Navigation:EUINavigation Enum) {is=title K2Node_FunctionEntry}
   > Return Node(ReturnValue=Get Button_Filters() -> (Button_Filters:WBP UIKit Button Collapsible Object Reference))
  ]}
 FocusFirstOfferInCurrentRow {nodes [
   FocusFirstOfferInCurrentRow() {is=title K2Node_FunctionEntry}
   n1 = GetRowAtIndex!(
    self:Fort Item Shop Selection VM Object Reference=Get FortItemShopSelectionVM()
    RowIndex:int=Get CurrentRowIndex!(self:WBP MPItem Shop Content Object Reference=Get WBP_MPItemShop_Content()) -> (CurrentRowIndex:int=0)
   ) -> (
    ReturnValue_OfferGroup:Fort Item Shop Offer Group VM Object Reference
    ReturnValue_Offers:Array of Fort Item Shop Offer VM Object References
    ReturnValue_bIsFirstRowInOfferGroup:bool=~
    ReturnValue_bIsLastRowInOfferGroup:bool=~
   )
   > Branch(Condition=Is Valid Index!(TargetArray:Array of Fort Item Shop Offer VM Object References=@n1.ReturnValue_Offers, IndexToTest:int) -> (ReturnValue:bool))
   > Set FocusedOfferId(FocusedOfferId=Get OfferId!(self:Fort Item Shop Offer VM Object Reference="Get (a copy)"(Array:Array of Fort Item Shop Offer VM Object References=@n1.ReturnValue_Offers) -> (Output:Fort Item Shop Offer VM Object Reference)) -> (OfferId:String))
  ]}
 FocusQuickNav {nodes [
   FocusQuickNav!() -> (then, Navigation:EUINavigation Enum) {is=title K2Node_FunctionEntry}
   > Return Node(ReturnValue=GetSectionNavigatorFocusTarget())
  ]}
 FullActivate {nodes [
   FullActivate() {is=title K2Node_FunctionEntry}
   > SetScalarParameterValue!(
    execute
    Collection:Material Parameter Collection Object Reference=/Game/Environments/FrontEnd/Blueprints/ItemPreview/VaultWorldTransitions/MaterialCollectionParameters/MPC_UI_Master_VaultWorld_Settings
    ParameterName:Name=Is Horizontal Orientation
    ParameterValue:float=0
   ) -> (then) {is=title K2Node_CallMaterialParameterCollectionFunction}
   > UpdateInGameButtons()
   > CreateNavigationBindings() {is=title K2Node_CallFunction}
   > UpdateHasSeenItemShop()
   > StartMonitoringForUpdates!(execute, self:Fort Mcp Catalog VM Object Reference=Get FortItemShopVM(), RequestingObject:Object Reference) -> (then)
   > UpdateSubscriptionData(self:Fort Subscription VM Object Reference=Get FortSubscriptionVM() -> (FortSubscriptionVM:Fort Subscription VM Object Reference))
   > RequestMarkAllOfferGroupsAsItemShopVisited()
   > Activate(self:WBP Billboard Background Manager Object Reference=Get WBP_Billboard_Background_Manager() -> (WBP_Billboard_Background_Manager:WBP Billboard Background Manager Object Reference))
   > n1 = Branch(Condition=HasContextChanged!(self) -> (ReturnValue:bool))
   > UpdateContentOrdering(self=Get FortItemShopVM())
   > ClearFilters(self=Get FilterVM(self=Get FortItemShopVM()))
   > Clear(TargetArray:Array of Fort Item Shop Offer Filter VM Object References=Get CachedAppliedFilters())
   > n2 = Clear(TargetArray:Array of Fort Option Item VM Object References=Get CachedAppliedOptions())
   ClearItemShopBang(execute=[@n2.then, @n1.else], self:Fort Item Shop VM Object Reference=Get FortItemShopVM())
   > Set IsShowingOfferInspect(IsShowingOfferInspect=false)
   > n3 = Branch(Condition=HasContextChanged())
   Set FeaturedOfferId!(execute=[@n3.then, @n5.else], FeaturedOfferId:String, self:Fort Item Shop VM Object Reference=Get FortItemShopVM()) -> (then, Output_Get:String)
   > Set ShouldResetFeaturedLayout(ShouldResetFeaturedLayout=true)
   n4 = Get!() -> (CachedFeaturedLayout_SectionId:String, CachedFeaturedLayout_SubsectionId:String)
   n5 = Branch(execute=@n3.else, Condition=NOT Boolean(A=IsEmpty(InString=@n4)))
   > Set ShouldResetFeaturedLayout(ShouldResetFeaturedLayout=false)
  ]}
 FullDeactivate {nodes [
   FullDeactivate() {is=title K2Node_FunctionEntry}
   > StopMonitoringForUpdates!(execute, self:Fort Mcp Catalog VM Object Reference=Get FortItemShopVM(), RequestingObject:Object Reference) -> (then)
   > DeactivateWidget(self=Get QuickScrollDetector())
   > DeactivateWidget(self=Get WBP_MPItemShop_Content())
   > CancelRequestMarkAllOfferGroupsAsItemShopVisited() {is=title K2Node_CallFunction}
   > Set CachedFeaturedLayout(CachedFeaturedLayout=Get FeaturedLayout(self=Get FortItemShopVM()))
   > Set CachedItemShopProductTag(CachedItemShopProductTag:Gameplay Tag Structure=GetItemShopContext!(self:Fort Item Shop VM Object Reference=Get FortItemShopVM()) -> (ReturnValue:Gameplay Tag Structure)) -> (Output_Get:Gameplay Tag Structure)
   > Set FeaturedLayout(self=Get FortItemShopVM())
   > Set HasCompletedFullActivation(HasCompletedFullActivation=false)
   > Reset(self:WBP Item Shop Side Navigation Object Reference=Get WBP_ItemShop_SectionNavigator_new())
   > SetFortItemShopVM(self=Get WBP_MPItemShop_Content())
   > Set IsSoftActivation(IsSoftActivation=false)
   > Set MOTD_TrackingID()
   > SetAllInteractionsEnabled(bInIsInteractionEnabled=false)
   > SaveOfferGroupSeenLevels(self:Fort Item Shop VM Object Reference=Get FortItemShopVM())
   > n1 = Branch(Condition=Get bShouldQuickNavStartCollapsed())
   > n2 = Handle Quicknav or Content Focus(ShouldFocusSectionNavigator=false, IsImmediate=true)
   PreviewSingleItem(execute=[@n2.then, @n1.else])
   > Set HasShownAboutShopModal(HasShownAboutShopModal:bool=false) -> (Output_Get:bool)
  ]}
 GetFirstSectionLayout {nodes [
   GetFirstSectionLayout() {is=title K2Node_FunctionEntry}
   > n1 = Get!(execute:Exec) -> (then:Exec, else:Exec, FortItemShopVM:Fort Item Shop VM Object Reference)
   > n2 = For Each Loop(Array:Array of Fort Item Shop Offer Group VM Object References=Get AllOfferGroupsWithOffers!(self:Fort Item Shop VM Object Reference=@n1) -> (AllOfferGroupsWithOffers:Array of Fort Item Shop Offer Group VM Object References)) -> (Array Element:Fort Item Shop Offer Group VM Object Reference)
   Return Node!(execute=@n2.Completed, Owning Layout:Fort Item Shop Layout Structure)
   > Branch(Condition=IsVisible(self:Fort Item Shop Offer Group VM Object Reference=@n2))
   > Return Node(Owning Layout=Get OwningLayout!(self:Fort Item Shop Offer Group VM Object Reference=@n2) -> (OwningLayout:Fort Item Shop Layout Structure))
  ]}
 GetItemShopVM {nodes [
   GetItemShopVM() {is=title K2Node_FunctionEntry}
   > Return Node!(execute, ReturnValue:Fort Item Shop VM Object Reference=Get FortItemShopVM())
  ]}
 GetPrimaryContentSetup {nodes [
   GetPrimaryContentSetup() {is=title K2Node_FunctionEntry}
   > Return Node!(
    execute
    ReturnValue_bShowBottomBar:bool=false
    ReturnValue_bShowChatWidget:bool=false
    ReturnValue_bEnableNewTextChat:bool=False
    ReturnValue_bShowTopBar:bool=True
    ReturnValue_bTabSupportsInterface:bool=true
   )
  ]}
 GetSectionNavigatorFocusTarget {nodes [
   GetSectionNavigatorFocusTarget() {is=title K2Node_FunctionEntry}
   > n1 = Is Valid(InputObject=Get FortItemShopVM())
   Return Node!(execute=@n1.Is Not Valid, ReturnValue:Widget Object Reference)
   n2 = Get FortItemShopVM()
   > Return Node(ReturnValue=GetSectionButton!(
     self:WBP Item Shop Side Navigation Object Reference=Get WBP_ItemShop_SectionNavigator_new()
     SectionID:String=Get!(self:Fort Item Shop VM Object Reference=@n2) -> (FeaturedLayout_SectionId:String, FeaturedLayout_SubsectionId:String)
    ) -> (Button:Common Button Base Object Reference))
  ]}
 GetSongInfoText {nodes [
   n1 = GetSongInfoText!() -> (then, PreviewedItem:Item Shop Preview Data Object Reference) {is=title K2Node_FunctionEntry}
   > n2 = Is Valid(InputObject=@n1)
   n3 = Get ItemVM!(self:Item Shop Preview Data Object Reference=@n1) -> (ItemVM:Fort Item VM Object Reference)
   > n4 = Is Valid(InputObject=@n3)
   Return Node!(execute=[@n4.Is Not Valid, @n2.Is Not Valid], FormattedText:Text)
   n5 = Get ItemDetails!(self:Fort Item VM Object Reference=@n3) -> (ItemDetails:Fort Item Details VM Object Reference)
   > Return Node(FormattedText=Format Text!(
     Format="LOCTEXT(\"{Artist}\\r\\n<text color=\\\"FFFFFF\\\" fontface=\\\"75 Medium\\\">{Song}</>\")"
     Artist:Text=GetItemShortDescription!(self:Fort Item Details VM Object Reference=@n5) -> (ReturnValue:Text)
     Song:Text=GetDisplayName!(self:Fort Item Details VM Object Reference=@n5) -> (ReturnValue:Text)
    ) -> (Result))
  ]}
 GetSongInfoVisibility {nodes [
   n1 = GetSongInfoVisibility!() -> (then, PreviewedItem:Item Shop Preview Data Object Reference) {is=title K2Node_FunctionEntry}
   > n2 = Is Valid(InputObject=@n1)
   n3 = Get ItemVM(self=@n1)
   > n4 = Is Valid(InputObject=@n3)
   Return Node!(execute=[@n4.Is Not Valid, @n2.Is Not Valid], Visibility:ESlateVisibility Enum=Collapsed)
   n5 = GetItemType!(self:Fort Item Definition Object Reference=Get ItemDefinition!(self:Fort Item VM Object Reference=@n3) -> (ItemDefinition:Fort Item Definition Object Reference)) -> (ReturnValue:EFortItemType Enum=WorldItem)
   > Return Node(Visibility=Select(
     Option 0:ESlateVisibility Enum=Collapsed
     Option 1:ESlateVisibility Enum=SelfHitTestInvisible
     Index="Equal (Enum)"(A:EFortItemType Enum=@n5, B:EFortItemType Enum=SparksSong)
    ) -> (ReturnValue:ESlateVisibility Enum))
  ]}
 GetSupportCreatorText {nodes [
   n1 = GetSupportCreatorText!() -> (then, AffiliateName:String) {is=title K2Node_FunctionEntry}
   > Return Node!(
    execute
    TimerText:Text=Select(
     Option 0:Text=Format Text!(Format="LOCTEXT(\"Support-a-Creator: {name}\")", name:String=@n1) -> (Result)
     Option 1:Text="LOCTEXT(\"Support-a-Creator: NONE\")"
     Index=IsEmpty(InString=@n1)
    ) -> (ReturnValue:Text)
   )
  ]}
 Handle MOTD Request {nodes [
   Handle MOTD Request() {is=title K2Node_FunctionEntry}
   > n1 = Branch(Condition=IsEmpty(InString=Get MOTD_TrackingID() -> (MOTD_TrackingID:String)))
   > Return Node!(execute)
   n2 = FindOfferByDeepLinkId!(self:Fort Item Shop VM Object Reference=Get FortItemShopVM(), DeepLinkId:String=Get MOTD_TrackingID()) -> (OutOwningLayout:Fort Item Shop Layout Structure, ReturnValue:Fort Item Shop Offer VM Object Reference)
   n3 = Is Valid(exec=@n1.else, InputObject=@n2.ReturnValue)
   Return Node(execute=@n3.Is Not Valid)
   > Set CachedFeaturedLayout(CachedFeaturedLayout=@n2)
   n4 = Get FortItemShopVM()
   > Set FeaturedOfferId(self=@n4)
   > Set FeaturedOfferId(FeaturedOfferId=Get OfferId(self=@n2.ReturnValue), self=@n4)
   > Set MOTD_TrackingID()
   > Return Node()
  ]}
 Handle Quicknav or Content Focus {nodes [
   Handle Quicknav or Content Focus!() -> (then, ShouldFocusSectionNavigator:bool=~, IsImmediate:bool=~) {is=title K2Node_FunctionEntry}
   > n1 = Branch(Condition=Get ShouldFocusSectionNavigator() -> (ShouldFocusSectionNavigator:bool))
   > SetSectionNavigatorCollapsed!(execute, self, ShouldCollapse:bool=false, IsImmediate:bool=Get IsImmediate() -> (IsImmediate:bool)) -> (then)
   > SetContentFocusedState!(execute, self:WBP MPItem Shop Content Object Reference=Get WBP_MPItemShop_Content(), IsFocused:bool=false) -> (then)
   > UpdateItemPreview!(execute, self, ShouldShow:bool=false) -> (then)
   > UpdateBannerVisibility!(execute, self, ShouldShow:bool=false) -> (then)
   > n2 = UpdateSongInfoVisibility(ShouldShow=false)
   n3 = Sequence(execute=[@n2.then, @n4.then])
   > Branch(Condition=IsUsingTouch())
   > Update Preview Quick Nav Buttons Visibility()
   Branch(execute=@n3.then_1, Condition=IsActivated())
   > Set IsSoftActivation(IsSoftActivation=false)
   SetSectionNavigatorCollapsed(execute=@n1.else, ShouldCollapse=IsItemPreviewEnabled(self=Get FortItemShopVM()), IsImmediate=Get IsImmediate())
   > SetContentFocusedState(self=Get WBP_MPItemShop_Content(), IsFocused=true)
   > UpdateItemPreview(ShouldShow=IsItemPreviewEnabled(self=Get FortItemShopVM()))
   > UpdateBannerVisibility(ShouldShow=true)
   > n4 = Set IsNavigatingToOffer(IsNavigatingToOffer=false)
  ]}
 HandleAboutShopFTUE {nodes [
   HandleAboutShopFTUE() {is=title K2Node_FunctionEntry}
   > Set HasShownAboutShopModal(HasShownAboutShopModal=true)
   > SetupAndPushInfoModal()
  ]}
 HandleBanner {nodes [
   ClearChildren(self=Get NamedSlot_Banner() -> (NamedSlot_Banner:Named Slot Object Reference))
   n1 = Get CompatibleModes!(self:Fort Item Shop Offer VM Object Reference=Get OfferVM() -> (OfferVM:Fort Item Shop Offer VM Object Reference)) -> (CompatibleModes:Set of Gameplay Tag Structures)
   > Branch(Condition="integer <= integer"(A=Length!(TargetSet:Set of Gameplay Tag Structures=@n1) -> (ReturnValue:int=0) {is=title K2Node_CallFunction}, B=1))
   > n2 = For Each Loop with Break!(
    Exec:Exec
    Array:Array of Fort Item VM Object References=Get GrantedItems!(self:Fort Item Shop Offer VM Object Reference=Get OfferVM()) -> (GrantedItems:Array of Fort Item VM Object References)
    Break:Exec=@n6.then
   ) -> (LoopBody:Exec, Array Element:Fort Item VM Object Reference, Array Index:int, Completed:Exec)
   Branch(
    execute=@n2.Completed
    Condition=IsValidClass!(Class:Object Class Reference=Get CustomBanner() -> (CustomBanner:User Widget Class Reference)) -> (ReturnValue:bool)
   )
   > n3 = Create Widget(Class=Get CustomBanner(), OwningPlayer=GetOwningPlayer()) -> (ReturnValue:User Widget Object Reference)
   > n4 = Cast To WBP_CustomBanner_Base!(execute, Object=@n3) -> (then, CastFailed, AsWBP Custom Banner Base:WBP Custom Banner Base Object Reference) {is=title K2Node_DynamicCast}
   > SetOfferViewModel!(execute, self:WBP Custom Banner Base Object Reference=@n4, OfferVM:Fort Item Shop Offer VM Object Reference=Get OfferVM()) -> (then)
   > SetViewType!(execute, self:WBP Custom Banner Base Object Reference=@n4, IsFullDetailsView:bool=false) -> (then)
   > AddChild(self=Get NamedSlot_Banner(), Content=@n3)
   n5 = Get CustomInspectScreenBanner!(self:Fort Locker Category Config Object Reference=GetItemCategory!(self:Fort Item VM Object Reference=@n2) -> (ReturnValue:Fort Locker Category Config Object Reference)) -> (CustomInspectScreenBanner:User Widget Soft Class Reference)
   Branch(
    execute=@n2.LoopBody
    Condition=IsValidClass(Class=Resolve Soft Reference!(Input:User Widget Soft Class Reference=@n5) -> (Output:User Widget Class Reference))
   )
   > n6 = Set CustomBanner(CustomBanner:User Widget Class Reference) -> (Output_Get:User Widget Class Reference)
   HandleBanner!() -> (then, OfferVM:Fort Item Shop Offer VM Object Reference) {is=title K2Node_FunctionEntry}
  ]}
 HandleItemPreview {nodes [
   HandleItemPreview!() -> (then, PreviewedItem:Item Shop Preview Data Object Reference) {is=title K2Node_FunctionEntry}
   > Branch(Condition=NOT Boolean(A=Get IsShowingOfferInspect()))
   > n1 = Branch(Condition=ArePreviewsIdentical!(
     Lhs:Item Shop Preview Data Object Reference=Get CachedPreviewedItem() -> (CachedPreviewedItem:Item Shop Preview Data Object Reference)
     Rhs:Item Shop Preview Data Object Reference=Get PreviewedItem() -> (PreviewedItem:Item Shop Preview Data Object Reference)
    ) -> (ReturnValue:bool))
   RequestItemPreview!(execute=@n1.else, self:Fort MPItem Shop Screen Object Reference, InPreviewedItem:Item Shop Preview Data Object Reference=Get PreviewedItem()) -> (then)
   > Branch(Condition=IsValid(Object=Get PreviewedItem()))
   > Set IsItemPreviewReady(IsItemPreviewReady=false)
  ]}
 HandleNewAppliedFilters {nodes [
   HandleNewAppliedFilters() {is=title K2Node_FunctionEntry}
   n1 = Get FilterVM(self=Get FortItemShopVM())
   n2 = Get AppliedFilters!(self:Fort Item Shop Filter VM Object Reference=@n1) -> (AppliedFilters:Array of Fort Item Shop Offer Filter VM Object References)
   n3 = Get CachedAppliedFilters()
   n4 = Get CachedAppliedOptions()
   n5 = Get AppliedOptions!(self:Fort Item Shop Filter VM Object Reference=@n1) -> (AppliedOptions:Array of Fort Option Item VM Object References)
   > n6 = Branch(Condition=AND Boolean(
     A=Identical!(ArrayA:Array of Fort Item Shop Offer Filter VM Object References=@n3, ArrayB:Array of Fort Item Shop Offer Filter VM Object References=@n2) -> (ReturnValue:bool) {is=title K2Node_CallArrayFunction}
     B=Identical(ArrayA:Array of Fort Option Item VM Object References=@n4, ArrayB:Array of Fort Option Item VM Object References=@n5)
    ))
   Set FeaturedLayout(execute=@n6.else, self=Get FortItemShopVM())
   > Set ShouldResetFeaturedLayout(ShouldResetFeaturedLayout=true)
   > OnShopContentReady() {is=title K2Node_CallFunction}
  ]}
 HandlePcbShopFTUE {nodes [
   HandlePcbShopFTUE() {is=title K2Node_FunctionEntry}
   n1 = GetTemporaryPcBangItemNamesByType!(self:Fort Pc Bang VM Object Reference=Get FortPcBangVM() -> (FortPcBangVM:Fort Pc Bang VM Object Reference)) -> (ReturnValue:Array of Texts)
   > n2 = Branch(Condition="integer > integer"(A=Length(TargetArray:Array of Texts=@n1) {is=title K2Node_CallArrayFunction}))
   n3 = Get Button_PCB() -> (Button_PCB:WBP UIKit Button Quiet Object Reference)
   > Set Visibility(Visibility=SelfHitTestInvisible, self=@n3)
   > n4 = Branch(Condition=Get HasShownPcbModal() -> (HasShownPcbModal:bool))
   Set HasShownPcbModal(execute=@n4.else, HasShownPcbModal:bool=true) -> (Output_Get:bool)
   > SetupAndPushPcbModal()
   Set Visibility(execute=@n2.else, Visibility=Collapsed, self=@n3)
  ]}
 HasContextChanged {nodes [
   HasContextChanged() {is=title K2Node_FunctionEntry}
   n1 = Get FortItemShopVM()
   > Return Node!(
    execute
    ReturnValue:bool="Not Equal (GameplayTag)"(
     A:Gameplay Tag Structure=Get CachedItemShopProductTag() -> (CachedItemShopProductTag:Gameplay Tag Structure)
     B:Gameplay Tag Structure=GetItemShopContext(self=@n1)
    ) -> (ReturnValue:bool=~) {is=title K2Node_PromotableOperator}
   )
  ]}
 IsSectionNavigatorCollapsed {nodes [
   IsSectionNavigatorCollapsed() {is=title K2Node_FunctionEntry}
   > Return Node!(
    execute
    IsCollapsed:bool=GetIsCollapsed!(self:WBP Item Shop Side Navigation Object Reference=Get WBP_ItemShop_SectionNavigator_new()) -> (IsCollapsed:bool)
   )
  ]}
 IsUsingTouch {nodes [
   IsUsingTouch() {is=title K2Node_FunctionEntry}
   > Return Node!(
    execute
    ReturnValue:bool=IsUsingTouch!(PlayerController:Player Controller Object Reference=GetOwningPlayer()) -> (ReturnValue) {is=title K2Node_CallFunction}
   )
  ]}
 ManualSoftActivate {nodes [
   ManualSoftActivate() {is=title K2Node_FunctionEntry}
   > Fire Event Main View Opened!(execute, View:Widget Object Reference=Self-Reference(), AdditionalInfo:Flow Analytics Additional Info Structure, bIsOverlay:bool=false) -> (then)
   > SoftActivate()
  ]}
 MarkAllOfferGroupsAsItemShopVisited {nodes [
   MarkAllOfferGroupsAsItemShopVisited() {is=title K2Node_FunctionEntry}
   > n1 = For Each Loop(Array=Get AllOfferGroupsWithOffers(self=Get FortItemShopVM()))
   > Set SeenLevel!(execute, SeenLevel:EFortOfferSeenLevel Enum=ItemShopVisited, self:Fort Item Shop Offer Group VM Object Reference=@n1) -> (then, Output_Get:EFortOfferSeenLevel Enum=Unseen)
  ]}
 NavRightFromFilterButtons {nodes [
   NavRightFromFilterButtons!() -> (then, Navigation:EUINavigation Enum) {is=title K2Node_FunctionEntry}
   > n1 = Branch(Condition=IsVisible(self:Widget Object Reference=Get Button_ClearFilters() -> (Button_ClearFilters:WBP UIKit Button Generic Object Reference)))
   > Return Node(ReturnValue:Widget Object Reference=Get Button_ClearFilters())
   n2 = FocusContent!(execute=@n1.else, self, Navigation:EUINavigation Enum=Left) -> (then, ReturnValue:Widget Object Reference) {is=title K2Node_CallFunction}
   > Return Node(ReturnValue=@n2)
  ]}
 OnIsItemPreviewEnabledChanged {nodes [
   n1 = OnIsItemPreviewEnabledChanged!() -> (then, IsItemPreviewEnabled:bool=~) {is=title K2Node_FunctionEntry}
   > SetSize!(
    execute
    self=Get Spacer_CameraFrame() -> (Spacer_CameraFrame:Spacer Object Reference)
    InSize_X:double=SelectFloat(A=Get CameraFrameWidth() -> (CameraFrameWidth:double), B=Get EmptyCameraFrameWidth() -> (EmptyCameraFrameWidth:double), bPickA=@n1)
    InSize_Y:double=1
   ) -> (then)
  ]}
 OnItemsToPreviewChanged {nodes [
   OnItemsToPreviewChanged!() -> (then, ItemsToPreview:Array of Item Shop Preview Data Object References) {is=title K2Node_FunctionEntry}
   > n1 = Get!(execute:Exec) -> (then:Exec, else:Exec, FortItemShopVM:Fort Item Shop VM Object Reference)
   > Clear and Invalidate Timer by Handle(Handle=Get ItemPreviewTimer())
   > Set ItemPreviewIndex(ItemPreviewIndex:int=0) -> (Output_Get:int=0)
   n2 = Get ItemsToPreview!() -> (ItemsToPreview)
   n3 = Is Empty!(TargetArray:Array of Item Shop Preview Data Object References=@n2) -> (ReturnValue:bool)
   n4 = NOT Boolean(A=IsItemPreviewEnabledForSelectedOfferGroup!(SelectionVM:Fort Item Shop Selection VM Object Reference=Get FortItemShopSelectionVM()) -> (IsEnabled:bool))
   n5 = NOT Boolean(A=IsItemPreviewEnabled(self=@n1))
   > Set PreviewedItem!(
    execute
    PreviewedItem:Item Shop Preview Data Object Reference=Select(
     Option 0:Item Shop Preview Data Object Reference="Get (a copy)"(Array:Array of Item Shop Preview Data Object References=@n2) -> (Output:Item Shop Preview Data Object Reference)
     Option 1:Item Shop Preview Data Object Reference
     Index=OR Boolean!(A=@n3, B=@n5, C:bool=@n4) -> (ReturnValue)
    ) -> (ReturnValue:Item Shop Preview Data Object Reference)
    self:Fort Item Shop Selection VM Object Reference=Get FortItemShopSelectionVM()
   ) -> (then, Output_Get:Item Shop Preview Data Object Reference)
  ]}
 OnUpdateBackground {nodes [
   OnUpdateBackground!() -> (then, InLayout:Fort Item Shop Layout Structure) {is=title K2Node_FunctionEntry}
   > LoadAndUpdateBackground!(execute, self, ImmediateTransition:bool=false) -> (then) {is=title K2Node_CallFunction}
  ]}
 PreviewOfferFromID {nodes [
   n1 = PreviewOfferFromID!() -> (then, Offer ID:String) {is=title K2Node_FunctionEntry}
   n2 = GetOfferFromId(self=Get FortItemShopVM(), OfferId=@n1, bSearchFromCatalog=false)
   > Is Valid(InputObject=@n2)
   > n3 = Set Offer(Offer:Fort Item Shop Offer VM Object Reference=@n2) -> (Output_Get:Fort Item Shop Offer VM Object Reference)
   n4 = Get OverridePreviews!(self:Fort Item Shop Offer VM Object Reference=@n3) -> (OverridePreviews:Array of Item Shop Preview Data Object References)
   > n5 = Branch(Condition=Is Empty(TargetArray=@n4))
   > n6 = For Each Loop(Array:Array of Fort Item VM Object References=Get GrantedItems(self=@n3)) -> (Array Element:Fort Item VM Object Reference)
   > n7 = Cast To AthenaItemVM!(execute, Object=@n6) -> (then, CastFailed, AsAthena Item VM:Athena Item VM Object Reference) {is=title K2Node_DynamicCast}
   > Cast To AthenaCosmeticItemDefinition!(execute, Object=Get ItemDefinition(self=@n7)) -> (then, CastFailed, AsAthena Cosmetic Item Definition:Athena Cosmetic Item Definition Object Reference) {is=title K2Node_DynamicCast}
   > n8 = CreatePreviewDataFromItemVM!(execute, Context:Widget Object Reference, ItemVM:Fort Item VM Object Reference=@n7) -> (then, ReturnValue:Item Shop Preview Data Object Reference)
   > PreviewSingleItem(ItemToPreview=@n8)
   > Return Node!(execute)
   PreviewSingleItem(execute=@n5.else, ItemToPreview="Get (a copy)"(Array=@n4))
   > Return Node()
  ]}
 PreviewSingleItem {nodes [
   n1 = PreviewSingleItem!() -> (then, ItemToPreview:Item Shop Preview Data Object Reference) {is=title K2Node_FunctionEntry}
   > Set ItemsToPreview!(
    execute
    ItemsToPreview:Array of Item Shop Preview Data Object References=Select(
     Option 0:Array of Item Shop Preview Data Object References=Make Array!() -> (Array:Array of Item Shop Preview Data Object References)
     Option 1:Array of Item Shop Preview Data Object References=Make Array!("[0]":Item Shop Preview Data Object Reference=@n1) -> (Array)
     Index=IsValid(Object=@n1)
    ) -> (ReturnValue:Array of Item Shop Preview Data Object References)
    self:Fort Item Shop Selection VM Object Reference=Get FortItemShopSelectionVM()
   ) -> (then, Output_Get:Array of Item Shop Preview Data Object References)
   > Set bIsItemPreviewCycleEnabled!(execute, bIsItemPreviewCycleEnabled:bool=false, self:Fort Item Shop Selection VM Object Reference=Get FortItemShopSelectionVM()) -> (then, Output_Get:bool)
  ]}
 PushModal {nodes [
   n1 = PushModal!() -> (then, ActivatableWidget:Common Activatable Widget Object Reference) {is=title K2Node_FunctionEntry}
   n2 = Create Event()
   > Unbind Event from On Widget Deactivated(self:Common Activatable Widget Object Reference=@n1, Delegate=@n2) {is=title K2Node_RemoveDelegate}
   > Bind Event to On Widget Deactivated(self=@n1, Delegate=@n2)
   > SoftDeactivate()
   > TryPushWidgetAsModalPanel!(execute, PushWidget:User Widget Object Reference=@n1) -> (then)
  ]}
 RefreshWidgetCameraFrame {nodes [
   RefreshWidgetCameraFrame() {is=title K2Node_FunctionEntry}
   > n1 = GetRowWidgetByRowIndex!(
    execute
    self:WBP MPItem Shop Content Object Reference=Get WBP_MPItemShop_Content()
    RowIndex:int=Get CurrentRowIndex(self=Get WBP_MPItemShop_Content()) -> (CurrentRowIndex=0)
   ) -> (then, RowWidget:WBP MPItem Shop Row Wrapper Object Reference)
   > n2 = Is Valid(InputObject=@n1)
   > n3 = Set UseWidePreview(UseWidePreview:bool=Get bUseWidePreview!(self:Fort Item Shop Offer Group VM Object Reference=Get!(self:WBP MPItem Shop Row Wrapper Object Reference=@n1) -> (
      FortItemShopRow_OfferGroup:Fort Item Shop Offer Group VM Object Reference
      FortItemShopRow_Offers:Array of Fort Item Shop Offer VM Object References
      FortItemShopRow_bIsFirstRowInOfferGroup:bool=~
      FortItemShopRow_bIsLastRowInOfferGroup:bool=~
     )) -> (bUseWidePreview:bool)) -> (Output_Get:bool)
   n4 = IsPreviewingVehicle!(self:Fort MPItem Shop Screen Object Reference) -> (ReturnValue:bool)
   n5 = Set CameraFrameWidth(
    execute=[@n3.then, @n2.Is Not Valid]
    CameraFrameWidth:double=SelectFloat(A=SelectFloat(A=1000, B=926, bPickA=@n4), B=646, bPickA=Get UseWidePreview() -> (UseWidePreview:bool))
   ) -> (Output_Get:double)
   > Set!(execute, Size_X:double=@n5, Size_Y:double=0, self:Spacer Object Reference=Get Spacer_CameraFrame()) -> (then, Output_Get:Vector 2D Structure)
   n6 = IsPreviewingVehicle()
   n7 = Get UseWidePreview()
   > SetPadding!(
    execute
    self:Size Box Slot Object Reference=SlotAsSizeBoxSlot!(Widget:Widget Object Reference=Get ScaleBox_ItemPreview() -> (ScaleBox_ItemPreview:Scale Box Object Reference)) -> (ReturnValue:Size Box Slot Object Reference)
    InPadding_Left:float=0
    InPadding_Top:float=60
    InPadding_Right:float=SelectFloat(A=-110, B=0, bPickA=AND Boolean(A=@n6, B=@n7))
    InPadding_Bottom:float=60
   ) -> (then)
  ]}
 RequestMarkAllOfferGroupsAsItemShopVisited {nodes [
   RequestMarkAllOfferGroupsAsItemShopVisited() {is=title K2Node_FunctionEntry}
   > n1 = Branch(Condition="float > float"(A=Get MarkAllOfferGroupsAsItemShopVisitedDelay() -> (MarkAllOfferGroupsAsItemShopVisitedDelay:float)))
   n2 = Get MarkAllOfferGroupsAsItemShopVisitedHandler()
   > Branch(Condition=NOT Boolean(A=Is Timer Active by Handle!(Handle:Timer Handle Structure=@n2) -> (ReturnValue:bool)))
   > n3 = Set Timer by Event(Delegate=Create Event(), Time=Get MarkAllOfferGroupsAsItemShopVisitedDelay(), bLooping=false)
   > Set MarkAllOfferGroupsAsItemShopVisitedHandler(MarkAllOfferGroupsAsItemShopVisitedHandler:Timer Handle Structure=@n3) -> (Output_Get:Timer Handle Structure)
   MarkAllOfferGroupsAsItemShopVisited(execute=@n1.else) {is=title K2Node_CallFunction}
  ]}
 Set Gamepad Action Bar Style {nodes [
   Set Gamepad Action Bar Style() {is=title K2Node_FunctionEntry}
   n1 = Get CommonInputSubsystem()
   > Branch(Condition="Equal (Enum)"(A:ECommonInputType Enum=GetCurrentInputType(self=@n1), B:ECommonInputType Enum=Gamepad))
   > SetPadding(
    self:Border Slot Object Reference=[
     SlotAsBorderSlot!(Widget:Widget Object Reference=Get Button_OpenQuickNavTouch() -> (Button_OpenQuickNavTouch:WBP UIKit Button Quiet Object Reference)) -> (ReturnValue:Border Slot Object Reference)
     SlotAsBorderSlot(Widget=Get Button_ShowPreviewTouch() -> (Button_ShowPreviewTouch:WBP UIKit Button Quiet Object Reference))
    ]
    InPadding_Left=0
    InPadding_Top=0
    InPadding_Right=0
    InPadding_Bottom=0
   )
   > SetPadding(
    self:Horizontal Box Slot Object Reference=[
     SlotAsHorizontalBoxSlot!(Widget:Widget Object Reference=Get Button_SAC() -> (Button_SAC:WBP UIKit Button Quiet Object Reference)) -> (ReturnValue:Horizontal Box Slot Object Reference)
     SlotAsHorizontalBoxSlot(Widget=Get Button_BackToGame() -> (Button_BackToGame:WBP UIKit Button Quiet Object Reference))
    ]
    InPadding_Left=0
    InPadding_Top=0
    InPadding_Right=0
    InPadding_Bottom=0
   )
   > SetPadding(
    self:Overlay Slot Object Reference=SlotAsOverlaySlot(Widget=Get Button_AboutShop() -> (Button_AboutShop:WBP UIKit Button Quiet Object Reference))
    InPadding_Left=0
    InPadding_Top=0
    InPadding_Right=0
    InPadding_Bottom=0
   )
  ]}
 SetAllInteractionsEnabled {nodes [
   n1 = SetAllInteractionsEnabled!() -> (then, bInIsInteractionEnabled:bool=~) {is=title K2Node_FunctionEntry}
   n2 = Get Button_BackToGame()
   n3 = Get Button_PCB()
   n4 = Get Button_BackToGame_NoOffers() -> (Button_BackToGame_NoOffers:WBP UIKit Button Quiet Object Reference)
   n5 = Get Button_OpenQuickNavTouch()
   n6 = Get Button_SAC()
   n7 = Get Button_AboutShop()
   n8 = Get Button_ShowPreviewTouch()
   > SetIsInteractionEnabled!(execute, self:Common Button Base Object Reference=[@n2, @n3, @n4, @n5, @n6, @n7, @n8], bInIsInteractionEnabled:bool=@n1) -> (then)
   > SetIsEnabled(self=[@n2, @n3, @n4, @n5, @n6, @n7, @n8], bInIsEnabled=@n1)
  ]}
 SetButtonPromptsVisibility {nodes [
   n1 = SetButtonPromptsVisibility!() -> (then, ShouldShow:bool=~) {is=title K2Node_FunctionEntry}
   > SetVisibility(
    self=Get WBP_ActionBarContainer() -> (WBP_ActionBarContainer:WBP Action Bar Container Object Reference)
    InVisibility=Select(Option 0:ESlateVisibility Enum=Collapsed, Option 1:ESlateVisibility Enum=SelfHitTestInvisible, Index=@n1) -> (ReturnValue:ESlateVisibility Enum)
   )
  ]}
 SetButtonStyling {nodes [
   SetButtonStyling() {is=title K2Node_FunctionEntry}
   n1 = Make Literal Float(Value=8)
   n2 = GetCurrentCulture!() -> (ReturnValue:String)
   n3 = Select!(
    MouseAndKeyboard:double=@n1
    Gamepad:double=Make Literal Float(Value=42)
    Touch:double=@n1
    Count:double=@n1
    Index:ECommonInputType Enum=GetCurrentInputType(self=Get CommonInputSubsystem())
   ) -> (ReturnValue:double)
   > SetButtonPadding!(
    execute
    self:WBP UIKit Button Generic Base Object Reference=Get Button_ClearFilters()
    Padding:Margin Structure=Select!(
     Option 0_Left:float=@n3
     Option 0_Top:float=0
     Option 0_Right:float=@n1
     Option 0_Bottom:float=0
     Option 1_Left:float=@n1
     Option 1_Top:float=0
     Option 1_Right:float=@n3
     Option 1_Bottom:float=0
     Index:bool="Equal, Case Insensitive (String)"(A=@n2, B=ar)
    ) -> (ReturnValue:Margin Structure)
   ) -> (then)
  ]}
 SetCameraFrameCollapseAnimationValue {nodes [
   n1 = SetCameraFrameCollapseAnimationValue!() -> (then, Alpha:double=~) {is=title K2Node_FunctionEntry}
   > SetSize(self=Get Spacer_CameraFrame(), InSize_X=Lerp(A=Get CameraFrameWidth(), B=Get EmptyCameraFrameWidth(), Alpha=@n1), InSize_Y=1)
  ]}
 SetCompactUIStyle {nodes [
   SetCompactUIStyle() {is=title K2Node_FunctionEntry}
   > Branch(Condition=IsUsingCompactUI())
   > SetPadding(
    self:Horizontal Box Slot Object Reference=SlotAsHorizontalBoxSlot(Widget=Get WBP_ItemShop_SectionNavigator_new())
    InPadding_Left=0
    InPadding_Top=160
    InPadding_Right=0
    InPadding_Bottom=0
   )
   > SetPadding(
    self:Overlay Slot Object Reference=SlotAsOverlaySlot(Widget=Get BottomBar() -> (BottomBar:Overlay Object Reference))
    InPadding_Left=0
    InPadding_Top=0
    InPadding_Right=48
    InPadding_Bottom=24
   )
   > SetPadding(
    self=SlotAsOverlaySlot(Widget=Get ArrowButton() -> (ArrowButton:WBP UIKit View More Arrow Button Object Reference))
    InPadding_Left=0
    InPadding_Top=0
    InPadding_Right=0
    InPadding_Bottom=96
   )
   n1 = Get Spacer_Filter_Buttons()
   > SetVisibility(
    self=@n1
    InVisibility=Select!(
     Option 0:ESlateVisibility Enum=SelfHitTestInvisible
     Option 1:ESlateVisibility Enum=Collapsed
     Index=Get ShouldCollapse() -> (ShouldCollapse:bool)
    ) -> (ReturnValue:ESlateVisibility Enum)
   )
   > SetSize!(execute, self=@n1, InSize:Vector 2D Structure=(X=1,Y=192)) -> (then)
   n2 = Get HB_FilterButtons() -> (HB_FilterButtons:Horizontal Box Object Reference)
   > SetPadding(self=SlotAsOverlaySlot(Widget=@n2), InPadding_Left=0, InPadding_Top=160, InPadding_Right=0, InPadding_Bottom=24)
   > SetHorizontalAlignment(self=SlotAsOverlaySlot(Widget=@n2), InHorizontalAlignment=HAlign_Fill)
   > SetIconSize!(execute, self:WBP UIKit Button Generic Base Object Reference=Get Button_Filters(), Icon:Vector 2D Structure=(X=48,Y=48)) -> (then)
   n3 = Get Button_ClearFilters()
   > SetIconSize(self=@n3, Icon=(X=48,Y=48))
   > SetIconPadding!(
    execute
    self:WBP UIKit Button Generic Base Object Reference=@n3
    Padding_Left:float=16
    Padding_Top:float=0
    Padding_Right:float=16
    Padding_Bottom:float=0
   ) -> (then)
   n4 = Get CommonInputSubsystem()
   > Branch(Condition="Equal (Enum)"(A=GetCurrentInputType(self=@n4), B=Touch))
   > SetSize(self=Get Spacer_Filter_Buttons(), InSize=(X=440,Y=256))
  ]}
 SetItemPreviewCycleState {nodes [
   SetItemPreviewCycleState!() -> (then, IsCycleEnabled:bool=~) {is=title K2Node_FunctionEntry}
   > n1 = Branch(Condition=Is Valid Timer Handle!(Handle:Timer Handle Structure=Get ItemPreviewTimer()) -> (ReturnValue:bool))
   n2 = Get ItemPreviewTimer()
   > Branch(Condition="Not Equal (Boolean)"(A=Get IsCycleEnabled() -> (IsCycleEnabled:bool), B=Is Timer Active by Handle(Handle=@n2)) -> (ReturnValue=~))
   > n3 = Branch(Condition=Get IsCycleEnabled())
   > Unpause Timer by Handle!(execute, Handle:Timer Handle Structure=Get ItemPreviewTimer()) -> (then)
   Pause Timer by Handle!(execute=@n3.else, Handle:Timer Handle Structure=Get ItemPreviewTimer()) -> (then)
   Branch(execute=@n1.else, Condition=AND Boolean(A=Get IsCycleEnabled(), B=Get IsItemPreviewReady() -> (IsItemPreviewReady:bool)))
   > OnItemPreviewIsReady() {is=title K2Node_CallFunction}
  ]}
 SetNavigationRules {nodes [
   SetNavigationRules() {is=title K2Node_FunctionEntry}
   > SetNavigationRuleCustom!(
    execute
    self:Widget Object Reference=Get WBP_ItemShop_SectionNavigator_new()
    Direction:EUINavigation Enum=Right
    InCustomDelegate:Delegate=Create Event()
   ) -> (then)
   > SetNavigationRuleCustomBoundary!(
    execute
    self:Widget Object Reference=Get WBP_MPItemShop_Content()
    Direction:EUINavigation Enum=Left
    InCustomDelegate:Delegate=Create Event()
   ) -> (then)
   > n1 = Branch(Condition="Equal, Case Insensitive (String)"(A=GetCurrentCulture(), B=ar))
   > SetNavigationRuleCustomBoundary(self=Get Button_Filters(), Direction=Right, InCustomDelegate=Create Event())
   SetNavigationRuleCustomBoundary(execute=@n1.else, self=Get Button_Filters(), Direction=Right, InCustomDelegate=Create Event())
   > SetNavigationRuleCustomBoundary(self=Get Button_ClearFilters(), Direction=Right, InCustomDelegate=Create Event())
  ]}
 SetNextItemPreview {nodes [
   SetNextItemPreview() {is=title K2Node_FunctionEntry}
   n1 = Get ItemPreviewIndex() -> (ItemPreviewIndex:int=0)
   n2 = Length(TargetArray:Array of Item Shop Preview Data Object References=Get ItemsToPreview!(self:Fort Item Shop Selection VM Object Reference=Get FortItemShopSelectionVM()) -> (ItemsToPreview))
   > n3 = Set ItemPreviewIndex(ItemPreviewIndex="Wrap (Integer)"!(Value:int=int + int(A=@n1), Min:int=-1, Max:int=int - int(A=@n2)) -> (ReturnValue:int=0)) -> (Output_Get=0)
   n4 = Get FortItemShopSelectionVM()
   > Set PreviewedItem(PreviewedItem="Get (a copy)"(Array=Get ItemsToPreview(self=@n4), Dimension 1=@n3), self=Get FortItemShopSelectionVM())
  ]}
 SetSTWStyle {nodes [
   SetSTWStyle() {is=title K2Node_FunctionEntry}
   > Branch(Condition=IsInSubGame!(SubGame:ESubGame Enum=Campaign) -> (ReturnValue:bool))
   > SetVisibility(self=Get Background_Loading() -> (Background_Loading:Image Object Reference), InVisibility=Hidden)
   n1 = SlotAsHorizontalBoxSlot(Widget=Get WBP_ItemShop_SectionNavigator_new())
   n2 = Get!(self:Horizontal Box Slot Object Reference=@n1) -> (Padding_Left:float=~, Padding_Top:float=~, Padding_Right:float=~, Padding_Bottom:float=~)
   > Set!(
    execute
    Padding_Left:float=@n2
    Padding_Top:float=128
    Padding_Right:float=@n2.Padding_Right
    Padding_Bottom:float=@n2.Padding_Bottom
    self:Horizontal Box Slot Object Reference=@n1
   ) -> (then, Output_Get:Margin Structure)
  ]}
 SetSectionNavigatorCollapsed {nodes [
   n1 = SetSectionNavigatorCollapsed!() -> (then, ShouldCollapse:bool=~, IsImmediate:bool=~) {is=title K2Node_FunctionEntry}
   > Set ShouldCollapse(ShouldCollapse:bool=@n1) -> (Output_Get:bool)
   > Is Valid(InputObject=Get FortItemShopVM())
   > n2 = Sequence!(execute) -> (then_0, then_1, then_2:Exec)
   n3 = Get WBP_ItemShop_SectionNavigator_new()
   > Branch(Condition="Not Equal (Boolean)"(A=Get ShouldCollapse(), B=GetIsCollapsed(self=@n3)) -> (ReturnValue=~))
   > n4 = Branch(Condition=Get ShouldCollapse())
   > n5 = Branch(Condition=Get IsImmediate())
   > SetCameraFrameCollapseAnimationValue!(execute, self, Alpha:double=0) -> (then) {is=title K2Node_CallFunction}
   PlayAnimationReverse(execute=@n5.else, InAnimation=Get CollapseSpacerCameraFrame() -> (CollapseSpacerCameraFrame:Widget Animation Object Reference))
   n6 = Branch(execute=@n4.else, Condition=Get IsImmediate())
   > SetCameraFrameCollapseAnimationValue(Alpha=1)
   PlayAnimationForward(execute=@n6.else, InAnimation=Get CollapseSpacerCameraFrame())
   SetIsCollapsed!(
    execute=@n2.then_1
    self:WBP Item Shop Side Navigation Object Reference=Get WBP_ItemShop_SectionNavigator_new()
    ShouldCollapse:bool=Get ShouldCollapse()
    IsImmediate:bool=Get IsImmediate()
   ) -> (then)
   n7 = Get Button_Filters()
   n8 = Branch(
    execute=@n2.then_2
    Condition="Equal (Boolean)"(A:bool=Get ShouldCollapse(), B:bool=Get IsCollapsed!(self:WBP UIKit Button Collapsible Object Reference=@n7) -> (IsCollapsed:bool)) -> (ReturnValue:bool=~) {is=title K2Node_PromotableOperator}
   )
   SetIsCollapsed!(
    execute=@n8.else
    self:BPI UIKit Collapsible Interface=Get Button_Filters()
    IsCollapsed:bool=Get ShouldCollapse()
    SkipAnimation:bool=Get IsImmediate()
   ) -> (then)
  ]}
 SetSectionNavigatorFortItemShopVM {nodes [
   SetSectionNavigatorFortItemShopVM() {is=title K2Node_FunctionEntry}
   > Is Valid(InputObject=Get FortItemShopVM())
   > Set FortItemShopVM!(
    execute
    FortItemShopVM:Fort Item Shop VM Object Reference=Get FortItemShopVM()
    self:WBP Item Shop Side Navigation Object Reference=Get WBP_ItemShop_SectionNavigator_new()
   ) -> (then, Output_Get:Fort Item Shop VM Object Reference)
  ]}
 SetupAndPushInfoModal {nodes [
   SetupAndPushInfoModal() {is=title K2Node_FunctionEntry}
   > n1 = Construct Object from Class(Class=/ItemShop/MultiProduct/Data/DialogVM_ShopInfo.DialogVM_ShopInfo_C) -> (ReturnValue:Dialog VM Shop Info Object Reference)
   > n2 = Create Widget!(
    execute
    Class=&/WBP_ShopInfo_Dialog.WBP_ShopInfo_Dialog_C
    OwningPlayer
    DialogVM_ShopInfo:Dialog VM Shop Info Object Reference=@n1
    FortItemShopVM:Fort Item Shop VM Object Reference=Get FortItemShopVM()
   ) -> (then, ReturnValue:WBP Shop Info Dialog Object Reference)
   > PushModal(ActivatableWidget=@n2) {is=title K2Node_CallFunction}
  ]}
 SetupAndPushPcbModal {nodes [
   SetupAndPushPcbModal() {is=title K2Node_FunctionEntry}
   > n1 = Construct Object from Class(Class=/ItemShop/MultiProduct/Data/DialogVM_AboutPcb.DialogVM_AboutPcb_C) -> (ReturnValue:Dialog VM About Pcb Object Reference)
   > n2 = Create Widget!(
    execute
    Class=&/WBP_AboutPcb_Dialog.WBP_AboutPcb_Dialog_C
    OwningPlayer
    DialogVM_PcbShop:Dialog VM About Pcb Object Reference=@n1
    FortPcBangVM:Fort Pc Bang VM Object Reference=Get FortPcBangVM()
   ) -> (then, ReturnValue:WBP About Pcb Dialog Object Reference)
   > PushModal(ActivatableWidget=@n2)
  ]}
 SetupSectionNavigator {nodes [
   SetupSectionNavigator() {is=title K2Node_FunctionEntry}
   > SetNavigationRuleCustom(self=Get WBP_ItemShop_SectionNavigator_new(), Direction=Right, InCustomDelegate=Create Event())
  ]}
 SoftActivate {nodes [
   SoftActivate() {is=title K2Node_FunctionEntry}
   > UpdateSubscriptionData(self=Get FortSubscriptionVM())
   > SetAllInteractionsEnabled(bInIsInteractionEnabled=true) {is=title K2Node_CallFunction}
   > OnSoftActivation(self=Get WBP_MPItemShop_Content(), bRowsDirty=false)
   > Activate(self=Get WBP_Billboard_Background_Manager())
   > n1 = Branch(Condition=Get IsShowingOfferInspect())
   > LoadAndUpdateBackground(ImmediateTransition=true)
   > n2 = Set IsShowingOfferInspect(IsShowingOfferInspect=false)
   HandleItemPreview!(
    execute=[@n1.else, @n2.then]
    self
    PreviewedItem:Item Shop Preview Data Object Reference=Get PreviewedItem!(self:Fort Item Shop Selection VM Object Reference=Get FortItemShopSelectionVM()) -> (PreviewedItem)
   ) -> (then) {is=title K2Node_CallFunction}
   > HandleNewAppliedFilters() {is=title K2Node_CallFunction}
   > UpdateBestSellersOffers!(execute, self:Fort Item Shop VM Object Reference=Get FortItemShopVM(), bUpdateContentOrderingIfChanged:bool=true) -> (then)
  ]}
 SoftDeactivate {nodes [
   SoftDeactivate() {is=title K2Node_FunctionEntry}
   > OnSoftDeactivation(self:WBP MPItem Shop Content Object Reference=Get WBP_MPItemShop_Content())
   > Set CachedFeaturedLayout(CachedFeaturedLayout=Get FeaturedLayout(self=Get FortItemShopVM()))
   > Set CachedItemShopProductTag(CachedItemShopProductTag=GetItemShopContext(self=Get FortItemShopVM()))
   > SetAllInteractionsEnabled(bInIsInteractionEnabled=false)
   > UpdateSongInfoVisibility(ShouldShow=false)
   n1 = Get FilterVM(self=Get FortItemShopVM())
   > Set CachedAppliedFilters(CachedAppliedFilters:Array of Fort Item Shop Offer Filter VM Object References=Get AppliedFilters(self=@n1)) -> (Output_Get:Array of Fort Item Shop Offer Filter VM Object References)
   > Set CachedAppliedOptions(CachedAppliedOptions:Array of Fort Option Item VM Object References=Get AppliedOptions(self=@n1)) -> (Output_Get:Array of Fort Option Item VM Object References)
  ]}
 Update Preview Quick Nav Buttons Visibility {nodes [
   Update Preview Quick Nav Buttons Visibility() {is=title K2Node_FunctionEntry}
   > n1 = Branch(Condition=IsItemPreviewEnabled(self=Get FortItemShopVM()))
   > n2 = Branch(Condition=NOT Boolean(A=IsSectionNavigatorCollapsed() {is=title K2Node_CallFunction}))
   > SetVisibility(self=Get Button_OpenQuickNavTouch(), InVisibility=Collapsed)
   > SetVisibility(self=Get Button_ShowPreviewTouch(), InVisibility=Visible)
   > ClearItemPreview(self:Fort MPItem Shop Screen Object Reference)
   SetVisibility(execute=@n2.else, self=Get Button_OpenQuickNavTouch(), InVisibility=Visible)
   > SetVisibility(self=Get Button_ShowPreviewTouch(), InVisibility=Collapsed)
   SetVisibility(execute=@n1.else, self=Get Button_OpenQuickNavTouch(), InVisibility=Collapsed)
   > SetVisibility(self=Get Button_ShowPreviewTouch(), InVisibility=Collapsed)
  ]}
 UpdateArrowButton {nodes [
   UpdateArrowButton() {is=title K2Node_FunctionEntry}
   n1 = Get WBP_MPItemShop_Content()
   > n2 = GetRowWidgetByRowIndex(self=Get WBP_MPItemShop_Content(), RowIndex=int + int(A=Get CurrentRowIndex(self=@n1) -> (CurrentRowIndex=0)))
   > n3 = Is Valid(InputObject=@n2)
   n4 = Get!(self:WBP MPItem Shop Row Wrapper Object Reference=@n2) -> (
    FortItemShopRow_OfferGroup:Fort Item Shop Offer Group VM Object Reference
    FortItemShopRow_Offers:Array of Fort Item Shop Offer VM Object References
    FortItemShopRow_bIsFirstRowInOfferGroup:bool=~
    FortItemShopRow_bIsLastRowInOfferGroup:bool=~
   )
   > n5 = Is Valid(InputObject=@n4)
   > n6 = Branch(Condition="Equal (Enum)"(
     A:EFortItemShopOfferDisplayType Enum=GetDisplayType!(self:Fort Item Shop Offer Group VM Object Reference=@n4) -> (ReturnValue:EFortItemShopOfferDisplayType Enum=TileGrid)
     B:EFortItemShopOfferDisplayType Enum=Billboard
    ))
   StopAnimation(execute=[@n11.then, @n6.then], InAnimation=Get ArrowEaseOut() -> (ArrowEaseOut:Widget Animation Object Reference))
   > PlayAnimationForward(InAnimation=Get ArrowEaseIn() -> (ArrowEaseIn:Widget Animation Object Reference))
   > PlayPulseAnimation(self:WBP UIKit View More Arrow Button Object Reference=Get ArrowButton())
   n7 = GetCurrentRowWidget(execute=[@n6.else, @n5.Is Not Valid], self=Get WBP_MPItemShop_Content())
   > n8 = Is Valid(InputObject=@n7)
   n9 = Get(self=@n7) -> (FortItemShopRow_bIsFirstRowInOfferGroup=~, FortItemShopRow_bIsLastRowInOfferGroup=~)
   > n10 = Is Valid(InputObject=@n9)
   > n11 = Branch(Condition="Equal (Enum)"(A=GetDisplayType(self=@n9) -> (ReturnValue=TileGrid), B=Billboard))
   Branch(execute=[@n11.else, @n10.Is Not Valid, @n8.Is Not Valid, @n3.Is Not Valid], Condition=IsVisible(self=Get ArrowButton()))
   > StopAnimation(InAnimation=Get ArrowEaseIn())
   > PlayAnimationForward(InAnimation=Get ArrowEaseOut())
  ]}
 UpdateBannerVisibility {nodes [
   n1 = UpdateBannerVisibility!() -> (then, ShouldShow:bool=~) {is=title K2Node_FunctionEntry}
   > Set Visibility(
    Visibility=Select(Option 0=Collapsed, Option 1=SelfHitTestInvisible, Index=@n1)
    self=Get Border_Banner() -> (Border_Banner:Border Object Reference)
   )
  ]}
 UpdateHasSeenItemShop {nodes [
   UpdateHasSeenItemShop() {is=title K2Node_FunctionEntry}
   n1 = Get FortItemShopVM()
   > Branch(Condition=NOT Boolean(A=GetHasSeenItemShop!(self:Fort Item Shop VM Object Reference=@n1) -> (ReturnValue:bool)))
   > SetHasSeenItemShop!(execute, self:Fort Item Shop VM Object Reference=Get FortItemShopVM(), bHasSeen:bool=true) -> (then)
  ]}
 UpdateInGameButtons {nodes [
   UpdateInGameButtons() {is=title K2Node_FunctionEntry}
   n1 = GetGameState!() -> (ReturnValue:Game State Base Object Reference)
   > SetVisibility(
    self=[Get Button_BackToGame_NoOffers(), Get Button_BackToGame()]
    InVisibility=Select(
     Option 0=Collapsed
     Option 1=Visible
     Index=Cast To FortGameState_InGame!(Object=@n1) -> (AsFort Game State in Game:Fort Game State In Game Object Reference, bSuccess:bool=~) {is=title K2Node_DynamicCast}.bSuccess
    )
   )
  ]}
 UpdateItemPreview {nodes [
   UpdateItemPreview!() -> (then, ShouldShow:bool=~) {is=title K2Node_FunctionEntry}
   > n1 = Branch(Condition=Get IsShowingOfferInspect())
   n2 = Branch(execute=@n1.else, Condition=Get ShouldShow() -> (ShouldShow:bool))
   > n3 = Get!(execute:Exec) -> (then:Exec, else:Exec, FortItemShopSelectionVM:Fort Item Shop Selection VM Object Reference)
   > n4 = Branch(Condition=NOT Boolean(A=IsItemPreviewEnabledForSelectedOfferGroup(SelectionVM=@n3)))
   ClearItemPreview(execute=[@n4.then, @n3.else])
   n5 = Branch(execute=@n4.else, Condition=IsEmpty(InString=Get FocusedOfferId() -> (FocusedOfferId:String)))
   > n6 = FocusFirstOfferInCurrentRow() {is=title K2Node_CallFunction}
   PreviewOfferFromID!(execute=[@n6.then, @n5.else], self, Offer ID:String=Get FocusedOfferId()) -> (then) {is=title K2Node_CallFunction}
   Set ItemsToPreview(execute=@n2.else, ItemsToPreview=Make Array!() -> (Array), self=Get FortItemShopSelectionVM())
   > Set bIsItemPreviewCycleEnabled(bIsItemPreviewCycleEnabled=false, self=Get FortItemShopSelectionVM())
  ]}
 UpdateMainActiveWidget {nodes [
   UpdateMainActiveWidget() {is=title K2Node_FunctionEntry}
   n1 = Get FortItemShopVM()
   n2 = Get FortItemShopSelectionVM()
   > n3 = Branch(Condition=AND Boolean(A=IsValid(Object=@n1), B=IsValid(Object=@n2)))
   n4 = Get Switcher_Main() -> (Switcher_Main:Common Visibility Switcher Object Reference)
   n5 = GetNumRows!(self:Fort Item Shop Selection VM Object Reference=Get FortItemShopSelectionVM()) -> (ReturnValue:int=0)
   n6 = Get Overlay_Loading() -> (Overlay_Loading:Overlay Object Reference)
   n7 = Select!(
    Uninitialized:Widget Object Reference=@n6
    CatalogRefreshing:Widget Object Reference=@n6
    ContentOutdated:Widget Object Reference=@n6
    ContentUpdating:Widget Object Reference=@n6
    Ready:Widget Object Reference=Select!(
     Option 0:Widget Object Reference=Get SafeZone_Content() -> (SafeZone_Content:Safe Zone Object Reference)
     Option 1:Widget Object Reference=Get Overlay_NoOffers() -> (Overlay_NoOffers:Overlay Object Reference)
     Index:bool="Equal (Integer)"(A=@n5)
    ) -> (ReturnValue)
    Index:EFortItemShopLoadingState Enum=Get LoadingState!(self:Fort Item Shop VM Object Reference=Get FortItemShopVM()) -> (LoadingState:EFortItemShopLoadingState Enum=Uninitialized)
   ) -> (ReturnValue:Widget Object Reference)
   > Branch(Condition="Not Equal (Object)"(
     A:Object Reference=@n7
     B:Object Reference=GetActiveWidget!(self:Common Visibility Switcher Object Reference=@n4) -> (ReturnValue:Widget Object Reference)
    ) -> (ReturnValue:bool=~) {is=title K2Node_PromotableOperator})
   > SetActiveWidget(self=Get Switcher_Main(), Widget=@n7)
   > n8 = Switch on EFortItemShopLoadingState!(execute, Selection:EFortItemShopLoadingState Enum=Get LoadingState(self=Get FortItemShopVM()) -> (LoadingState=Uninitialized)) -> (Uninitialized:Exec, CatalogRefreshing:Exec, ContentOutdated:Exec, ContentUpdating:Exec, Ready:Exec) {is=title K2Node_SwitchEnum}
   OnCatalogRefreshing(execute=@n8.CatalogRefreshing) {is=title K2Node_CallFunction}
   OnContentUpdating(execute=@n8.ContentUpdating) {is=title K2Node_CallFunction}
   OnShopContentReady(execute=@n8.Ready)
   SetActiveWidget(execute=@n3.else, self=Get Switcher_Main(), Widget=Get Overlay_Loading())
  ]}
 UpdateMainActiveWidgetFromLoadingState {nodes [
   UpdateMainActiveWidgetFromLoadingState!() -> (then, LoadingState:EFortItemShopLoadingState Enum=Ready) {is=title K2Node_FunctionEntry}
   > UpdateMainActiveWidget() {is=title K2Node_CallFunction}
  ]}
 UpdateMainActiveWidgetFromNumRows {nodes [
   UpdateMainActiveWidgetFromNumRows!() -> (then, NumRows:int) {is=title K2Node_FunctionEntry}
   > UpdateMainActiveWidget()
  ]}
 UpdateSongInfoSizing {nodes [
   UpdateSongInfoSizing() {is=title K2Node_FunctionEntry}
   > n1 = GetSongPreviewActorCorners!(execute, self:Fort MPItem Shop Screen Object Reference) -> (then, LeftCorner2D:Vector 2D Structure, RightCorner2D:Vector 2D Structure)
   > Set LeftCorner(LeftCorner:Vector 2D Structure=@n1) -> (Output_Get:Vector 2D Structure)
   > Set RightCorner(RightCorner:Vector 2D Structure=@n1.RightCorner2D) -> (Output_Get:Vector 2D Structure)
   n2 = Get RightCorner() -> (RightCorner:Vector 2D Structure)
   n3 = Get LeftCorner() -> (LeftCorner:Vector 2D Structure)
   > n4 = Branch(Condition=AND Boolean(
     A=IsNearlyZero2D!(A:"Vector 2D Structure (by ref)"=@n3, Tolerance:float=0.0001) -> (ReturnValue:bool)
     B=IsNearlyZero2D(A=@n2, Tolerance=0.0001)
    ))
   > Return Node!(execute, IsValid:bool=false)
   n5 = Get!() -> (LeftCorner_X:double, LeftCorner_Y:double)
   Set(
    execute=@n4.else
    Padding_Left=@n5
    Padding_Top=float + float(A=@n5.LeftCorner_Y, B=24)
    Padding_Right=Get!() -> (RightCorner_X:double, RightCorner_Y:double)
    Padding_Bottom=0
    self:Overlay Slot Object Reference=SlotAsOverlaySlot(Widget=Get Border_SongInfo() -> (Border_SongInfo:Border Object Reference))
   )
   > Return Node(IsValid=true)
  ]}
 UpdateSongInfoVisibility {nodes [
   UpdateSongInfoVisibility!() -> (then, ShouldShow:bool=~) {is=title K2Node_FunctionEntry}
   > Set PreviousVisibility(PreviousVisibility:ESlateVisibility Enum=Get Visibility!(self:Widget Object Reference=Get Border_SongInfo()) -> (Visibility:ESlateVisibility Enum=Visible)) -> (Output_Get:ESlateVisibility Enum=Visible)
   > n1 = Set Visibility(
    Visibility=Select(Option 0:ESlateVisibility Enum=Collapsed, Option 1:ESlateVisibility Enum=SelfHitTestInvisible, Index=Get ShouldShow()) -> (ReturnValue:ESlateVisibility Enum)
    self=Get Border_SongInfo()
   )
   n2 = Get PreviousVisibility() -> (PreviousVisibility:ESlateVisibility Enum=Visible)
   > Branch(Condition=AND Boolean(A="Not Equal (Enum)"!(A:ESlateVisibility Enum=@n1, B:ESlateVisibility Enum=@n2) -> (ReturnValue:bool), B=Get ShouldShow()))
   > PlayAnimationForward(InAnimation=Get SongInfoReveal() -> (SongInfoReveal:Widget Animation Object Reference))
  ]}
 VM_ActiveFiltersNumUpdated {nodes [
   n1 = VM_ActiveFiltersNumUpdated!() -> (then, ActiveFiltersNum:int) {is=title K2Node_FunctionEntry}
   n2 = Get StatusIndicator!(self:WBP UIKit Button Generic Base Object Reference=Get Button_Filters()) -> (StatusIndicator:WBP UIKit Status Indicator Object Reference)
   > "SetText (Text)"(
    self=Get Indicator_Text!(self:WBP UIKit Status Indicator Base Object Reference=@n2) -> (Indicator_Text:Common Text Object Reference)
    InText="To Text (Integer)"(Value=@n1)
   )
   > Set Visibility(Visibility=Select(Option 0=SelfHitTestInvisible, Option 1=Collapsed, Index="Equal (Integer)"(A=@n1)), self=@n2)
  ]}
 VM_FilterIsActiveUpdated {nodes [
   VM_FilterIsActiveUpdated!() -> (then, Is Active:bool=~) {is=title K2Node_FunctionEntry}
   n1 = Get FortItemShopVM()
   > FlushImpressions(self=Get FortItemShopSelectionVM(), ImpressionsEventAdditionalInfo=GetLastFilterContextAsAdditionalInfo(self=Get FilterVM(self=@n1)))
  ]}
 VM_SelectedOfferGroupUpdated {nodes [
   n1 = VM_SelectedOfferGroupUpdated!() -> (then, SelectedOfferGroup:Fort Item Shop Offer Group VM Object Reference) {is=title K2Node_FunctionEntry}
   > UpdateArrowButton() {is=title K2Node_CallFunction}
   > Is Valid(InputObject=@n1)
   > n2 = Sequence!(execute) -> (then_0, then_1)
   n3 = GetDisplayType(self=@n1) -> (ReturnValue=TileGrid)
   > n4 = Branch(Condition="Equal (Enum)"(A=@n3, B=Billboard))
   > Set IsInBillboardRow(IsInBillboardRow:bool=true) -> (Output_Get:bool)
   Branch(execute=@n4.else, Condition=Get IsInBillboardRow() -> (IsInBillboardRow:bool))
   > LoadAndUpdateBackground(ImmediateTransition=true)
   > Set IsInBillboardRow(IsInBillboardRow=false)
   n5 = Branch(execute=@n2.then_1, Condition="Equal (Enum)"(A=@n3, B=ScrollableList))
   Set bCanNavigateToSideNav!(execute=@n5.else, bCanNavigateToSideNav:bool=true, self:Fort Item Shop Selection VM Object Reference=Get FortItemShopSelectionVM()) -> (then, Output_Get:bool)
  ]}
]
variables [
 bShouldQuickNavStartCollapsed {type=boolean, default=true}
 CachedAppliedFilters {type="array<FortItemShopOfferFilterVM>", default []}
 CachedAppliedOptions {type="array<FortOptionItemVM>", default []}
 CachedFeaturedLayout {type=FortItemShopLayout, default {sectionId="", subsectionId=""}}
 CachedItemShopProductTag {type=GameplayTag, default {tagName=None}}
 CameraFrameCollapseAnimationValue {type=number}
 CameraFrameWidth {type=number, default=646}
 ChatEventListener {type=AsyncAction_StartListeningToEvent, default=null}
 DelayBetweenItemPreviews {type=number, default=4}
 EmptyCameraFrameWidth {type=number, default=280}
 FocusedOfferId {type=string, default=""}
 HasCompletedFullActivation {type=boolean}
 HasShownAboutShopModal {type=boolean}
 HasShownPcbModal {type=boolean}
 Is Quick Scrolling {type=boolean}
 IsChatOpen {type=boolean}
 IsInBillboardRow {type=boolean}
 IsItemPreviewReady {type=boolean, default=true}
 IsNavigatingToOffer {type=boolean}
 IsShowingOfferInspect {type=boolean}
 IsSoftActivation {type=boolean}
 ItemPreviewIndex {type=integer}
 ItemPreviewTimer {type=TimerHandle}
 MarkAllOfferGroupsAsItemShopVisitedDelay {type=number, default=2}
 MarkAllOfferGroupsAsItemShopVisitedHandler {type=TimerHandle}
 MOTD_TrackingID {type=string, default=""}
 NavHideWhooshSound {type=SoundBase, default=/Game/Sounds/Fort_UI_Sounds/FrontEnd/ItemShop/Fort_UI_Shop_Whoosh_NavigatonHide_Cue}
 NavRevealWhooshSound {type=SoundBase, default=/Game/Sounds/Fort_UI_Sounds/FrontEnd/ItemShop/Fort_UI_Shop_Whoosh_NavigatorReveal_Cue}
 RequestedOfferGroupVM {type=FortItemShopOfferGroupVM, default=null}
 RequestedOfferVM {type=FortItemShopOfferVM, default=null}
 ShouldCollapse {type=boolean}
 ShouldResetFeaturedLayout {type=boolean}
]
widgets {
 info {
  inheritedWidgetCount=2
  namedSlotCount=2
  parentClass=/Script/FortniteUI.FortMPItemShopScreen
  rootWidgetClass=&Overlay
  widgetCount=60
 }
 widgets [
  Overlay_2 {widgetClassPath=&Overlay, widget=Overlay_2}
  CommonHardwareVisibilityBorder_0 {
   widgetClassPath=&CommonHardwareVisibilityBorder
   parent=Overlay_2
   widget=CommonHardwareVisibilityBorder_0
   properties {
    visibleType=Visible
    visibilityQuery {
     autoDescription=" NONE( Input.Gamepad )"
     tokenStreamVersion=0
     userDescription=NoGamepad
     queryTokenStream [0,1,3,1,0]
     tagDictionary [{tagName=Input.Gamepad}]
    }
   }
   slot {class=OverlaySlot, properties {horizontalAlignment=HAlign_Fill, verticalAlignment=VAlign_Fill}}
  }
  ScrollBox_Content {
   widgetClassPath=/Script/FortniteUI.AthenaScrollBox
   bInherited=true
   parent=CommonHardwareVisibilityBorder_0
   widget=ScrollBox_Content
   properties {
    bAllowRightClickDragScrolling=false
    bAnimateWheelScrolling=false
    bEnableFixedPageOffset=true
    consumeMouseWheel=Always
    navigationDestination=Center
    scrollBarVisibility=Collapsed
   }
   slot {class=BorderSlot, properties {padding=<ltrb 0 0 0 0>}}
  }
  WBP_Billboard_Background_Manager {
   widgetClassPath=/ItemShop/MultiProduct/Billboard/WBP_Billboard_Background_Manager
   bIsVariable=true
   parent=Overlay_2
   widget=WBP_Billboard_Background_Manager
   slot {class=OverlaySlot, properties {horizontalAlignment=HAlign_Fill, verticalAlignment=VAlign_Fill}}
  }
  PostBufferUpdate {
   widgetClassPath=/Script/UMG.PostBufferUpdate
   parent=Overlay_2
   widget=PostBufferUpdate
   properties {
    visibility=SelfHitTestInvisible
    updateBufferInfos [
     {
      bufferToUpdate=ESlatePostRT_0
      postParamUpdater=&self:WidgetTree.PostBufferUpdate.PostBufferBlurUpdater_0
     }
     {bufferToUpdate=ESlatePostRT_1, postParamUpdater=null}
    ]
   }
   slot {class=OverlaySlot, properties {horizontalAlignment=HAlign_Fill, verticalAlignment=VAlign_Fill}}
  }
  SafeZone_TopBottom {
   widgetClassPath=&SafeZone
   parent=Overlay_2
   widget=SafeZone_TopBottom
   properties {padLeft=false, padRight=false}
   slot {class=OverlaySlot, properties {horizontalAlignment=HAlign_Fill, verticalAlignment=VAlign_Fill}}
  }
  SizeBox_ItemPreview {
   widgetClassPath=&SizeBox
   parent=SafeZone_TopBottom
   widget=SizeBox_ItemPreview
   properties {bOverride_WidthOverride=true, widthOverride=1920}
   slot {class=SafeZoneSlot, properties {hAlign=HAlign_Center}}
  }
  ScaleBox_ItemPreview {
   widgetClassPath=&ScaleBox
   bIsVariable=true
   parent=SizeBox_ItemPreview
   widget=ScaleBox_ItemPreview
   properties {stretch=ScaleBySafeZone}
   slot {class=SizeBoxSlot, properties {horizontalAlignment=HAlign_Right, padding=<ltrb 0 60 0 60>}}
  }
  Border_CameraFrameContainer {
   widgetClassPath=&Border
   bIsVariable=true
   parent=ScaleBox_ItemPreview
   widget=Border_CameraFrameContainer
   properties {
    visibility=SelfHitTestInvisible
    background {
     bIsDynamicallyLoaded=false
     drawAs=NoDrawType
     imageType=FullColor
     mirroring=NoMirror
     resourceName=None
     resourceObject=null
     tiling=NoTile
     imageSize=<32 32>
     margin=<ltrb 0 0 0 0>
     outlineSettings {
      bUseBrushTransparency=false
      roundingType=HalfHeightRadius
      width=0
      color {colorUseRule=UseColor_Specified, specifiedColor=<rgba 0 0 0 0>}
      cornerRadii=<0 0 0 0>
     }
     tintColor {colorUseRule=UseColor_Specified, specifiedColor=<rgba 1 1 1 1>}
     uVRegion {bIsValid=false, max=<0 0>, min=<0 0>}
    }
    padding=<ltrb 0 0 0 0>
   }
   slot {class=ScaleBoxSlot, properties {horizontalAlignment=HAlign_Fill, verticalAlignment=VAlign_Fill}}
  }
  Widget_CameraFrame {
   widgetClassPath=&Overlay
   parent=Border_CameraFrameContainer
   widget=Widget_CameraFrame
   slot {class=BorderSlot, properties {padding=<ltrb 0 0 0 0>}}
  }
  FortItemPreviewWidget {
   widgetClassPath=/FortniteCoreUI/Frontend/ItemPreview/WBP_ItemPreview
   bInherited=true
   bIsVariable=true
   parent=Widget_CameraFrame
   widget=FortItemPreviewWidget
   properties {
    bAllowItemZooming=false
    bAllowPawnOverride=true
    bCanDisplayUserAccessories=false
    bRequireExclusiveLoadoutSatisfied=false
    bSpecialPreviewActionsEnabled=false
    bSupportsActivationFocus=false
    bTogglePreviewActionsEnabled=false
    itemViewObject=&self:WidgetTree.FortItemPreviewWidget.ItemViewObject
    previewSceneChangerTag {tagName=UI.ItemPreview.SceneChanger.MPItemShop}
   }
   slot {class=OverlaySlot, properties {horizontalAlignment=HAlign_Fill, verticalAlignment=VAlign_Fill}}
  }
  Spacer_CameraFrame {
   widgetClassPath=&Spacer
   bIsVariable=true
   parent=Widget_CameraFrame
   widget=Spacer_CameraFrame
   properties {size=<550 1>}
   slot {class=OverlaySlot, properties {verticalAlignment=VAlign_Fill}}
  }
  Border_SongInfo {
   widgetClassPath=&Border
   bIsVariable=true
   parent=Overlay_2
   widget=Border_SongInfo
   properties {verticalAlignment=VAlign_Top, brushColor=<rgba 1 0 0 0>, padding=<ltrb 0 0 0 0>}
   slot {class=OverlaySlot, properties {horizontalAlignment=HAlign_Fill}}
  }
  Overlay_0 {
   widgetClassPath=&Overlay
   parent=Border_SongInfo
   widget=Overlay_0
   slot {class=BorderSlot, properties {verticalAlignment=VAlign_Top, padding=<ltrb 0 0 0 0>}}
  }
  SongInfo {
   widgetClassPath=/FortUILibrary/UIKit/InfoBox/WBP_UIKit_InfoBox
   bIsVariable=true
   parent=Overlay_0
   widget=SongInfo
   properties {previewText="[PH] Song info", showIcon=false}
   slot {class=OverlaySlot, properties {horizontalAlignment=HAlign_Fill, verticalAlignment=VAlign_Fill}}
  }
  Switcher_Main {
   widgetClassPath=&CommonVisibilitySwitcher
   bIsVariable=true
   parent=Overlay_2
   widget=Switcher_Main
   slot {class=OverlaySlot, properties {horizontalAlignment=HAlign_Fill, verticalAlignment=VAlign_Fill}}
  }
  SafeZone_Content {
   widgetClassPath=&SafeZone
   bIsVariable=true
   parent=Switcher_Main
   widget=SafeZone_Content
   slot {class=CommonVisibilitySwitcherSlot}
  }
  Overlay_Content {widgetClassPath=&Overlay, bIsVariable=true, parent=SafeZone_Content, widget=Overlay_Content, slot {class=SafeZoneSlot}}
  SizeBox {
   widgetClassPath=&SizeBox
   parent=Overlay_Content
   widget=SizeBox
   properties {bOverride_WidthOverride=true, widthOverride=1920}
   slot {class=OverlaySlot, properties {horizontalAlignment=HAlign_Center, verticalAlignment=VAlign_Fill}}
  }
  Overlay_5 {widgetClassPath=&Overlay, parent=SizeBox, widget=Overlay_5, slot {class=SizeBoxSlot}}
  GridPanel_Content {
   widgetClassPath=&GridPanel
   bIsVariable=true
   parent=Overlay_5
   widget=GridPanel_Content
   properties {bOverride_Cursor=true, columnFill [0,1], rowFill [1]}
   slot {class=OverlaySlot, properties {verticalAlignment=VAlign_Fill}}
  }
  VerticalBox_SectionNavigator {
   widgetClassPath=&VerticalBox
   parent=GridPanel_Content
   widget=VerticalBox_SectionNavigator
   slot {class=GridSlot, properties {layer=1, verticalAlignment=VAlign_Top, padding=<ltrb 80 0 34 0>}}
  }
  Overlay_50 {widgetClassPath=&Overlay, parent=VerticalBox_SectionNavigator, widget=Overlay_50, slot {class=VerticalBoxSlot}}
  Spacer_Filter_Buttons {
   widgetClassPath=&Spacer
   bIsVariable=true
   parent=Overlay_50
   widget=Spacer_Filter_Buttons
   properties {size=<1 256>}
   slot {class=OverlaySlot}
  }
  HB_FilterButtons {
   widgetClassPath=&HorizontalBox
   bIsVariable=true
   parent=Overlay_50
   widget=HB_FilterButtons
   properties {flowDirectionPreference=Culture}
   slot {class=OverlaySlot, properties {horizontalAlignment=HAlign_Fill, verticalAlignment=VAlign_Bottom, padding=<ltrb 0 0 0 24>}}
  }
  Button_Filters {
   widgetClassPath=/FortUILibrary/UIKit/Buttons/WBP_UIKit_Button_Collapsible
   bIsVariable=true
   parent=HB_FilterButtons
   widget=Button_Filters
   properties {
    iconMaterial=/FortUILibrary/Textures/MaterialInstances/MI_UIKit_Icon_Filter
    minHeight=48
    mobileFontSizeMultiplierOverride=1.5
    mobileMinHeight=80
    statusIndicatorVerticalAlignment=VAlign_Center
    text=Filters
    useStatusIndicator=true
    buttonPadding=<ltrb 8 8 8 8>
    statusIndicatorPadding=<ltrb 0 0 8 0>
   }
   slot {class=HorizontalBoxSlot, properties {size {sizeRule=Fill, value=1}}}
  }
  Button_ClearFilters {
   widgetClassPath=&WBP_UIKit_Button_Generic
   bIsVariable=true
   parent=HB_FilterButtons
   widget=Button_ClearFilters
   properties {
    bHideInputActionWithKeyboard=true
    bShouldUseFallbackDefaultInputAction=false
    backgroundMaterial=/FortUILibrary/UIKit/Buttons/MaterialInstances/CTAButtons/MI_UIKit_BgRegular
    contentHorizontalAlignment=HAlign_Fill
    iconHorizontalAlignment=HAlign_Right
    iconMaterial=/FortUILibrary/Textures/MaterialInstances/MI_MPLocker_Icon_Clear
    minHeight=48
    text=المرشحات
    useTextBlock=false
    buttonPadding=<ltrb 48 0 8 0>
    iconSize=<32 32>
    promptNudge=<0 0>
    promptPadding=<ltrb 6 0 6 0>
    triggeringInputAction {dataTable=&InputActionDataTable, rowName=GenericRightTrigger}
   }
   slot {class=HorizontalBoxSlot, properties {padding=<ltrb 24 0 0 0>}}
  }
  WBP_ItemShop_SectionNavigator_new {
   widgetClassPath=/ItemShop/MultiProduct/SectionNavigator/WBP_ItemShop_SideNavigation
   bIsVariable=true
   parent=VerticalBox_SectionNavigator
   widget=WBP_ItemShop_SectionNavigator_new
   slot {class=VerticalBoxSlot}
  }
  WBP_MPItemShop_Content {
   widgetClassPath=&/WBP_MPItemShop_Content
   bIsVariable=true
   parent=GridPanel_Content
   widget=WBP_MPItemShop_Content
   properties {bSupportsActivationFocus=false}
   slot {class=GridSlot, properties {column=1, padding=<ltrb 16 0 0 0>}}
  }
  Border_Banner {
   widgetClassPath=&Border
   bIsVariable=true
   parent=Overlay_5
   widget=Border_Banner
   properties {visibility=HitTestInvisible, brushColor=<rgba 1 1 1 0>, padding=<ltrb 0 0 0 0>}
   slot {class=OverlaySlot, properties {horizontalAlignment=HAlign_Right, verticalAlignment=VAlign_Bottom, padding=<ltrb 0 0 80 140>}}
  }
  NamedSlot_Banner {
   widgetClassPath=&NamedSlot
   bIsVariable=true
   parent=Border_Banner
   widget=NamedSlot_Banner
   slot {class=BorderSlot, properties {padding=<ltrb 0 0 0 0>}}
  }
  ArrowButton {
   widgetClassPath=/FortUILibrary/UIKit/Buttons/ArrowButtons/WBP_UIKit_ViewMore_ArrowButton
   bIsVariable=true
   parent=Overlay_5
   widget=ArrowButton
   properties {bShouldUseFallbackDefaultInputAction=false, visibility=Collapsed}
   slot {class=OverlaySlot, properties {horizontalAlignment=HAlign_Center, verticalAlignment=VAlign_Bottom, padding=<ltrb 0 0 0 56>}}
  }
  BottomBar {
   widgetClassPath=&Overlay
   bIsVariable=true
   parent=Overlay_Content
   widget=BottomBar
   slot {class=OverlaySlot, properties {horizontalAlignment=HAlign_Right, verticalAlignment=VAlign_Bottom, padding=<ltrb 0 0 80 32>}}
  }
  WBP_ActionBarContainer {
   widgetClassPath=/Game/UI/Frontend/BottomBar/WBP_ActionBarContainer
   bIsVariable=true
   parent=BottomBar
   widget=WBP_ActionBarContainer
   slot {class=OverlaySlot}
  }
  HorizontalBox_ButtonPrompts {widgetClassPath=&HorizontalBox, namedSlotHost=WBP_ActionBarContainer, widget=HorizontalBox_ButtonPrompts}
  CommonHardwareVisibilityBorder_208 {
   widgetClassPath=&CommonHardwareVisibilityBorder
   parent=HorizontalBox_ButtonPrompts
   widget=CommonHardwareVisibilityBorder_208
   properties {
    visibleType=Visible
    padding=<ltrb 0 0 24 0>
    visibilityQuery {
     autoDescription=" ANY( Input.Touch )"
     tokenStreamVersion=0
     userDescription=""
     queryTokenStream [0,1,1,1,0]
     tagDictionary [{tagName=Input.Touch}]
    }
   }
   slot {class=HorizontalBoxSlot}
  }
  Button_ShowPreviewTouch {
   widgetClassPath=/FortUILibrary/UIKit/Buttons/WBP_UIKit_Button_Quiet
   bIsVariable=true
   parent=CommonHardwareVisibilityBorder_208
   widget=Button_ShowPreviewTouch
   properties {text=Show Preview}
   slot {class=BorderSlot, properties {padding=<ltrb 0 0 24 0>}}
  }
  CommonHardwareVisibilityBorder {
   widgetClassPath=&CommonHardwareVisibilityBorder
   parent=HorizontalBox_ButtonPrompts
   widget=CommonHardwareVisibilityBorder
   properties {
    visibleType=Visible
    padding=<ltrb 24 0 0 0>
    visibilityQuery {
     autoDescription=" ANY( Input.Touch )"
     tokenStreamVersion=0
     userDescription=""
     queryTokenStream [0,1,1,1,0]
     tagDictionary [{tagName=Input.Touch}]
    }
   }
   slot {class=HorizontalBoxSlot}
  }
  Button_OpenQuickNavTouch {
   widgetClassPath=&WBP_UIKit_Button_Quiet
   bIsVariable=true
   parent=CommonHardwareVisibilityBorder
   widget=Button_OpenQuickNavTouch
   properties {text=Open Quick Nav}
   slot {class=BorderSlot, properties {padding=<ltrb 24 0 0 0>}}
  }
  Button_SAC {
   widgetClassPath=&WBP_UIKit_Button_Quiet
   bIsVariable=true
   parent=HorizontalBox_ButtonPrompts
   widget=Button_SAC
   properties {
    bIsFocusable=false
    text="[PH] Support-a-creator: NONE"
    triggeringInputAction {dataTable=&InputActionDataTable, rowName=GenericFaceButtonLeft}
   }
   slot {class=HorizontalBoxSlot}
  }
  Button_AboutShop {
   widgetClassPath=&WBP_UIKit_Button_Quiet
   bIsVariable=true
   parent=HorizontalBox_ButtonPrompts
   widget=Button_AboutShop
   properties {
    bIsFocusable=false
    text=About Shop
    triggeringInputAction {dataTable=&InputActionDataTable, rowName=GenericFaceButtonTop}
   }
   slot {class=HorizontalBoxSlot, properties {padding=<ltrb 24 0 0 0>}}
  }
  Button_PCB {
   widgetClassPath=&WBP_UIKit_Button_Quiet
   bIsVariable=true
   parent=HorizontalBox_ButtonPrompts
   widget=Button_PCB
   properties {
    bIsFocusable=false
    text=About PCB
    triggeringInputAction {dataTable=&InputActionDataTable, rowName=GenericRightStickClick}
   }
   slot {class=HorizontalBoxSlot, properties {padding=<ltrb 24 0 0 0>}}
  }
  Button_BackToGame {
   widgetClassPath=&WBP_UIKit_Button_Quiet
   bIsVariable=true
   parent=HorizontalBox_ButtonPrompts
   widget=Button_BackToGame
   properties {bIsFocusable=false, text=Back To Game, triggeringInputAction {dataTable=&InputActionDataTable, rowName=BackToGame}}
   slot {class=HorizontalBoxSlot, properties {padding=<ltrb 24 0 0 0>}}
  }
  WBP_AthenaBottomBar_WithNotifications {
   widgetClassPath=/Game/UI/Frontend/BottomBar/WBP_AthenaBottomBar_WithNotifications
   bIsVariable=true
   parent=HorizontalBox_ButtonPrompts
   widget=WBP_AthenaBottomBar_WithNotifications
   slot {class=HorizontalBoxSlot, properties {padding=<ltrb 24 0 0 0>}}
  }
  QuickScrollDetector {
   widgetClassPath=/FortUILibrary/Blueprints/WBP_UIKit_QuickScrollDetector
   bIsVariable=true
   parent=Overlay_Content
   widget=QuickScrollDetector
   properties {bSupportsActivationFocus=false, scrollRequestsPerTimeRangeThreshold=4}
   slot {class=OverlaySlot, properties {horizontalAlignment=HAlign_Right, verticalAlignment=VAlign_Bottom, padding=<ltrb 0 0 100 200>}}
  }
  Overlay_NoOffers {
   widgetClassPath=&Overlay
   bIsVariable=true
   parent=Switcher_Main
   widget=Overlay_NoOffers
   slot {class=CommonVisibilitySwitcherSlot}
  }
  Background_NoOffers {
   widgetClassPath=&Image
   parent=Overlay_NoOffers
   widget=Background_NoOffers
   properties {brush {
     bIsDynamicallyLoaded=false
     drawAs=Image
     imageType=FullColor
     mirroring=NoMirror
     resourceName=None
     resourceObject=/FortniteCoreUI/Frontend/Art/M_UI_RadialGradient_ItemShop_BG
     tiling=NoTile
     imageSize=<32 32>
     margin=<ltrb 0 0 0 0>
     outlineSettings {
      bUseBrushTransparency=false
      roundingType=HalfHeightRadius
      width=0
      color {colorUseRule=UseColor_Specified, specifiedColor=<rgba 0 0 0 0>}
      cornerRadii=<0 0 0 0>
     }
     tintColor {colorUseRule=UseColor_Specified, specifiedColor=<rgba 1 1 1 1>}
     uVRegion {bIsValid=false, max=<0 0>, min=<0 0>}
    }}
   slot {class=OverlaySlot, properties {horizontalAlignment=HAlign_Fill, verticalAlignment=VAlign_Fill}}
  }
  VerticalBox_0 {
   widgetClassPath=&VerticalBox
   parent=Overlay_NoOffers
   widget=VerticalBox_0
   slot {class=OverlaySlot, properties {horizontalAlignment=HAlign_Center, verticalAlignment=VAlign_Center}}
  }
  Border {
   widgetClassPath=&CommonBorder
   parent=VerticalBox_0
   widget=Border
   properties {
    horizontalAlignment=HAlign_Center
    style=/FortUILibrary/Styles/Border/Border_StatRow_WhiteLowOpacity.Border_StatRow_WhiteLowOpacity_C
    padding=<ltrb 0 0 0 0>
   }
   slot {class=VerticalBoxSlot, properties {horizontalAlignment=HAlign_Center}}
  }
  SizeBox_0 {
   widgetClassPath=&SizeBox
   parent=Border
   widget=SizeBox_0
   properties {bOverride_HeightOverride=true, heightOverride=48}
   slot {class=BorderSlot, properties {horizontalAlignment=HAlign_Center, padding=<ltrb 0 0 0 0>}}
  }
  CommonTextBlock_2 {
   widgetClassPath=&CommonTextBlock
   parent=SizeBox_0
   widget=CommonTextBlock_2
   properties {
    justification=Center
    text=No Offers Available
    textTransformPolicy=ToUpper
    colorAndOpacity {colorUseRule=UseColor_Specified, specifiedColor=<rgba 1 1 1 0.7>}
    font {
     bForceMonospaced=false
     bMaterialIsStencil=false
     fontMaterial=null
     fontObject=&HeadingNow
     letterSpacing=0
     monospacedWidth=1
     size=18
     skewAmount=0
     typefaceFontName=76 Bold
     outlineSettings {
      bApplyOutlineToDropShadows=false
      bMiteredCorners=false
      bSeparateFillAlpha=false
      outlineMaterial=null
      outlineSize=0
      outlineColor=<rgba 0 0 0 1>
     }
    }
    shadowOffset=<0 0>
   }
   slot {class=SizeBoxSlot, properties {padding=<ltrb 32 11 32 8>}}
  }
  Text_EmptyShopDevReason {
   widgetClassPath=&CommonTextBlock
   parent=VerticalBox_0
   widget=Text_EmptyShopDevReason
   properties {
    bAutoCollapseWithEmptyText=true
    justification=Center
    text="[PH] Dev Reason: Something went wrong, and this is the text that would explain what the issue is."
    wrapTextAt=1000
    colorAndOpacity {colorUseRule=UseColor_Specified, specifiedColor=<rgba 1 1 1 0.7>}
    font {
     bForceMonospaced=false
     bMaterialIsStencil=false
     fontMaterial=null
     fontObject=&HeadingNow
     letterSpacing=0
     monospacedWidth=1
     size=18
     skewAmount=0
     typefaceFontName=74 Regular
     outlineSettings {
      bApplyOutlineToDropShadows=false
      bMiteredCorners=false
      bSeparateFillAlpha=false
      outlineMaterial=null
      outlineSize=0
      outlineColor=<rgba 0 0 0 1>
     }
    }
    shadowOffset=<0 0>
   }
   slot {class=VerticalBoxSlot, properties {padding=<ltrb 0 8 0 0>}}
  }
  SafeZone_Prompts_1 {
   widgetClassPath=&SafeZone
   bIsVariable=true
   parent=Overlay_NoOffers
   widget=SafeZone_Prompts_1
   properties {padLeft=false, padTop=false}
   slot {class=OverlaySlot, properties {horizontalAlignment=HAlign_Right, verticalAlignment=VAlign_Bottom, padding=<ltrb 0 0 80 32>}}
  }
  WBP_ActionBarContainer_1 {
   widgetClassPath=&WBP_ActionBarContainer
   bIsVariable=true
   parent=SafeZone_Prompts_1
   widget=WBP_ActionBarContainer_1
   slot {class=SafeZoneSlot}
  }
  HorizontalBox_ButtonPrompts_1 {
   widgetClassPath=&WrapBox
   namedSlotHost=WBP_ActionBarContainer_1
   widget=HorizontalBox_ButtonPrompts_1
   properties {bExplicitWrapSize=true, horizontalAlignment=HAlign_Fill, orientation=Orient_Vertical, wrapSize=0, innerSlotPadding=<24 0>}
  }
  Button_BackToGame_NoOffers {
   widgetClassPath=&WBP_UIKit_Button_Quiet
   bIsVariable=true
   parent=HorizontalBox_ButtonPrompts_1
   widget=Button_BackToGame_NoOffers
   properties {bIsFocusable=false, text=Back To Game, triggeringInputAction {dataTable=&InputActionDataTable, rowName=BackToGame}}
   slot {class=WrapBoxSlot}
  }
  Button_Chat {
   widgetClassPath=&WBP_UIKit_Button_Quiet
   bIsVariable=true
   parent=HorizontalBox_ButtonPrompts_1
   widget=Button_Chat
   properties {text=Chat, triggeringInputAction {dataTable=&InputActionDataTable, rowName=Chat}}
   slot {class=WrapBoxSlot}
  }
  Overlay_Loading {
   widgetClassPath=&Overlay
   bIsVariable=true
   parent=Switcher_Main
   widget=Overlay_Loading
   slot {class=CommonVisibilitySwitcherSlot}
  }
  Background_Loading {
   widgetClassPath=&Image
   bIsVariable=true
   parent=Overlay_Loading
   widget=Background_Loading
   properties {brush {
     bIsDynamicallyLoaded=false
     drawAs=Image
     imageType=FullColor
     mirroring=NoMirror
     resourceName=None
     resourceObject=&M_UI_RadialGradient_ItemShop_BG
     tiling=NoTile
     imageSize=<32 32>
     margin=<ltrb 0 0 0 0>
     outlineSettings {
      bUseBrushTransparency=false
      roundingType=HalfHeightRadius
      width=0
      color {colorUseRule=UseColor_Specified, specifiedColor=<rgba 0 0 0 0>}
      cornerRadii=<0 0 0 0>
     }
     tintColor {colorUseRule=UseColor_Specified, specifiedColor=<rgba 1 1 1 1>}
     uVRegion {bIsValid=false, max=<0 0>, min=<0 0>}
    }}
   slot {class=OverlaySlot, properties {horizontalAlignment=HAlign_Fill, verticalAlignment=VAlign_Fill}}
  }
  Image_DownloadSpinner {
   widgetClassPath=&Image
   parent=Overlay_Loading
   widget=Image_DownloadSpinner
   properties {
    visibility=SelfHitTestInvisible
    brush {
     bIsDynamicallyLoaded=false
     drawAs=Image
     imageType=NoImage
     mirroring=NoMirror
     resourceName=None
     resourceObject=/Game/UI/ContentOnDemand/ItemDownloadIndicator/MI_ItemDownloadIndicator_SpinnerMaterial
     tiling=NoTile
     imageSize=<100 100>
     margin=<ltrb 0 0 0 0>
     outlineSettings {
      bUseBrushTransparency=false
      roundingType=HalfHeightRadius
      width=0
      color {colorUseRule=UseColor_Specified, specifiedColor=<rgba 0 0 0 0>}
      cornerRadii=<0 0 0 0>
     }
     tintColor {colorUseRule=UseColor_Specified, specifiedColor=<rgba 1 1 1 1>}
     uVRegion {bIsValid=false, max=<0 0>, min=<0 0>}
    }
   }
   slot {class=OverlaySlot, properties {horizontalAlignment=HAlign_Center, verticalAlignment=VAlign_Center}}
  }
 ]
}
