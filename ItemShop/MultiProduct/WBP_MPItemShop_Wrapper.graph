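# Wrapper screen for the multi-product Item Shop that hosts the actual shop screen in a named slot.
# On first activation it asynchronously creates the override screen class, or the default WBP_MPItemShop_Screen, with input suspended, then forwards later activation and deactivation.
# Offer navigation requests go straight to the screen, or are cached until it exists, then replayed and cleared by Handle_MOTD_Request.
# Content setup shows the top bar and hides the bottom bar and chat.
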WidgetBlueprint /ItemShop/MultiProduct/WBP_MPItemShop_Wrapper
graphs [
 EventGraph {nodes [
   n1 = Event OnNavigateToOffer!() -> (OutputDelegate, then, OfferTrackingId:String, bShowOfferDetails:bool) {is=title K2Node_Event}
   > n2 = Get!(execute:Exec) -> (then:Exec, else:Exec, ItemShopInstance:Fort MPItem Shop Screen Object Reference)
   > OnNavigateToOffer!(
    execute
    self:Fort MPItem Shop Screen Object Reference=@n2
    OfferTrackingId:String=@n1.OfferTrackingId
    bShowOfferDetails:bool=@n1.bShowOfferDetails
   ) -> (then)
   > Set CachedOfferTrackingID(CachedOfferTrackingID:String) -> (Output_Get:String)
   > Set CachedShowOfferDetails(CachedShowOfferDetails:bool=false) -> (Output_Get:bool)
   Set CachedOfferTrackingID(execute=@n2.else, CachedOfferTrackingID=@n1.OfferTrackingId)
   > Set CachedShowOfferDetails(CachedShowOfferDetails=@n1.bShowOfferDetails)
   Event On Activated()
   > n3 = Get()
   > ActivateWidget(self=@n3)
   n4 = GetScreenOverrideClass!(self:Fort MPItem Shop Screen Wrapper Object Reference) -> (ReturnValue:Fort MPItem Shop Screen Soft Class Reference)
   n5 = CreateWidgetAsync!(
    execute:Exec=@n3.else
    UserWidgetSoftClass:User Widget Soft Class Reference=Select(
     Option 0:Fort MPItem Shop Screen Soft Class Reference=&/WBP_MPItemShop_Screen.WBP_MPItemShop_Screen_C
     Option 1:Fort MPItem Shop Screen Soft Class Reference=@n4
     Index=IsValidSoftClassReference!(SoftClassReference:"Object Soft Class Reference (by ref)"=@n4) -> (ReturnValue:bool)
    ) -> (ReturnValue:Fort MPItem Shop Screen Soft Class Reference)
    OwningPlayer:Player Controller Object Reference=GetOwningPlayer()
    bSuspendInputUntilComplete:bool=true
   ) -> (then:Exec, OnComplete:Exec, OnFailure:Exec, UserWidget:User Widget Object Reference) {is=title K2Node_AsyncAction}
   n6 = Cast To FortMPItemShopScreen!(execute=@n5.OnComplete, Object=@n5) -> (then, CastFailed, AsFort MPItem Shop Screen:Fort MPItem Shop Screen Object Reference) {is=title K2Node_DynamicCast}
   > n7 = Set ItemShopInstance(ItemShopInstance:Fort MPItem Shop Screen Object Reference=@n6) -> (Output_Get:Fort MPItem Shop Screen Object Reference)
   > AddChild(self=Get NamedSlot_ItemShop() -> (NamedSlot_ItemShop:Named Slot Object Reference), Content=@n7)
   > ActivateWidget(self=Get ItemShopInstance() -> (ItemShopInstance:Fort MPItem Shop Screen Object Reference))
   > Handle_MOTD_Request()
   Event On Deactivated()
   > n8 = Get()
   > DeactivateWidget(self=@n8)
   > Reroute Node()
  ]}
 GetPrimaryContentSetup {nodes [
   GetPrimaryContentSetup() {is=title K2Node_FunctionEntry}
   > Return Node!(
    execute
    ReturnValue_bShowBottomBar:bool=false
    ReturnValue_bShowChatWidget:bool=false
    ReturnValue_bEnableNewTextChat:bool=False
    ReturnValue_bShowTopBar:bool=True
    ReturnValue_bTabSupportsInterface:bool=true
   )
  ]}
 Handle_MOTD_Request {nodes [
   Handle_MOTD_Request() {is=title K2Node_FunctionEntry}
   > n1 = Branch(Condition=IsEmpty(InString=Get CachedOfferTrackingID() -> (CachedOfferTrackingID:String)))
   Set CachedOfferTrackingID(execute=[@n2.then, @n1.then])
   > Set CachedShowOfferDetails(CachedShowOfferDetails=false)
   n2 = OnNavigateToOffer(
    execute=@n1.else
    self=Get ItemShopInstance()
    OfferTrackingId=Get CachedOfferTrackingID()
    bShowOfferDetails=Get CachedShowOfferDetails() -> (CachedShowOfferDetails:bool)
   )
  ]}
]
variables [
 CachedOfferTrackingID {type=string, default=""}
 CachedShowOfferDetails {type=boolean}
 ItemShopInstance {type=FortMPItemShopScreen, default=null}
]
widgets {
 info {inheritedWidgetCount=0, parentClass=/Script/FortniteUI.FortMPItemShopScreenWrapper, rootWidgetClass=&NamedSlot, widgetCount=1}
 widgets [NamedSlot_ItemShop {widgetClassPath=&NamedSlot, bIsVariable=true, widget=NamedSlot_ItemShop}]
}
