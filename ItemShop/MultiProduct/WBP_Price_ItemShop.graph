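# Item Shop price widget derived from WBP_UIKit_Price, showing currency or platform icon, price, additional price text, and struck-through regular price with a suffix.
# Getter functions hand its child widgets to the parent class.
# PreConstruct runs UpdateVisuals, applying exposed icon size, font and padding variables to those widgets.
# SetAdditionalRegularPriceText collapses the discount disclaimer when the text is empty, then calls the parent implementation.
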WidgetBlueprint /ItemShop/MultiProduct/WBP_Price_ItemShop
graphs [
 EventGraph {nodes [
   Event PreConstruct()
   > UpdateVisuals()
  ]}
 GetAdditionalPriceTextRef {nodes [
   GetAdditionalPriceTextRef() {is=title K2Node_FunctionEntry}
   > Return Node!(execute, Widget:Common Text Object Reference=Get Text_AdditionalPrice() -> (Text_AdditionalPrice:Common Text Object Reference))
  ]}
 GetAdditionalRegularPriceTextRef_Optional {nodes [
   GetAdditionalRegularPriceTextRef_Optional() {is=title K2Node_FunctionEntry}
   > Return Node(Widget=Get Text_AdditionalRegularPrice() -> (Text_AdditionalRegularPrice:Common Text Object Reference))
  ]}
 GetPlatformIconRef {nodes [
   GetPlatformIconRef() {is=title K2Node_FunctionEntry}
   > Return Node(Widget:WBP Price Platform Store Icon Object Reference=Get PlatformIcon() -> (PlatformIcon:WBP Price Platform Store Icon Object Reference))
  ]}
 GetPriceTextRef {nodes [
   GetPriceTextRef() {is=title K2Node_FunctionEntry}
   > Return Node(Widget:Common Text Object Reference=Get Text_Price() -> (Text_Price:Common Text Object Reference))
  ]}
 GetRegularPriceContainerRef_Optional {nodes [
   GetRegularPriceContainerRef_Optional() {is=title K2Node_FunctionEntry}
   > Return Node(Widget:Widget Object Reference=Get HB_RegularPrice() -> (HB_RegularPrice:Horizontal Box Object Reference))
  ]}
 GetRegularPriceTextRef_Optional {nodes [
   GetRegularPriceTextRef_Optional() {is=title K2Node_FunctionEntry}
   > Return Node(Widget:Common Text Object Reference=Get Text_RegularPrice() -> (Text_RegularPrice:Common Text Object Reference))
  ]}
 GetVBuckIconRef {nodes [
   GetVBuckIconRef() {is=title K2Node_FunctionEntry}
   > Return Node(Widget:Widget Object Reference=Get Image_Currency() -> (Image_Currency:Image Object Reference))
  ]}
 SetAdditionalRegularPriceText {nodes [
   n1 = SetAdditionalRegularPriceText!() -> (then, AdditionalRegularPriceText:Text) {is=title K2Node_FunctionEntry}
   > SetDiscountDisclaimerVisibility!(
    execute
    self:WBP UIKit Price Object Reference
    Visibility:ESlateVisibility Enum=Select(Option 0=SelfHitTestInvisible, Option 1=Collapsed, Index=TextIsEmpty(InText=@n1))
   ) -> (then)
   > Parent: SetAdditionalRegularPriceText!(execute, AdditionalRegularPriceText:Text=@n1) -> (then) {is=title K2Node_CallParentFunction}
  ]}
 UpdateVisuals {nodes [
   UpdateVisuals() {is=title K2Node_FunctionEntry}
   > Set ImageSize!(
    execute
    ImageSize:Vector 2D Structure=Get IconSize() -> (IconSize:Vector 2D Structure)
    self:WBP Price Platform Store Icon Object Reference=Get PlatformIcon()
   ) -> (then, Output_Get:Vector 2D Structure)
   n1 = Get Image_Currency()
   n2 = Get!(self:Image Object Reference=@n1) -> (
    Brush_TintColor:Slate Color Structure
    Brush_DrawAs:ESlateBrushDrawType Enum
    Brush_Tiling:ESlateBrushTileType Enum
    Brush_Mirroring:ESlateBrushMirrorType Enum
    Brush_ImageSize:Vector 2D Structure
    Brush_Margin:Margin Structure
    Brush_ResourceObject:Object Reference
    Brush_OutlineSettings:Slate Brush Outline Settings Structure
   )
   > Set!(
    execute
    Brush_TintColor:Slate Color Structure=@n2
    Brush_DrawAs:ESlateBrushDrawType Enum=@n2.Brush_DrawAs
    Brush_Tiling:ESlateBrushTileType Enum=@n2.Brush_Tiling
    Brush_Mirroring:ESlateBrushMirrorType Enum=@n2.Brush_Mirroring
    Brush_ImageSize:Vector 2D Structure=Get IconSize()
    Brush_Margin:Margin Structure=@n2.Brush_Margin
    Brush_ResourceObject:Object Reference=@n2.Brush_ResourceObject
    Brush_OutlineSettings:Slate Brush Outline Settings Structure=@n2.Brush_OutlineSettings
    self:Image Object Reference=@n1
   ) -> (then, Output_Get:Slate Brush Structure)
   > Set Font!(
    execute
    Font:Slate Font Info Structure=Get PriceFont() -> (PriceFont:Slate Font Info Structure)
    self:Text Object Reference=Get Text_Price()
   ) -> (then, Output_Get:Slate Font Info Structure)
   > Set Font(Font=Get AdditionalPriceFont() -> (AdditionalPriceFont:Slate Font Info Structure), self=Get Text_AdditionalPrice())
   > Set Padding!(
    execute
    Padding:Margin Structure=Get AdditionalPricePadding() -> (AdditionalPricePadding:Margin Structure)
    self:Horizontal Box Slot Object Reference=SlotAsHorizontalBoxSlot!(Widget:Widget Object Reference=Get Text_AdditionalPrice()) -> (ReturnValue:Horizontal Box Slot Object Reference)
   ) -> (then, Output_Get:Margin Structure)
   > Set Font(Font=Get RegularPriceFont() -> (RegularPriceFont:Slate Font Info Structure), self=Get Text_RegularPrice())
   > Set Padding(
    Padding=Get RegularPricePadding() -> (RegularPricePadding:Margin Structure)
    self=SlotAsHorizontalBoxSlot(Widget=Get Overlay_RegularPrice() -> (Overlay_RegularPrice:Overlay Object Reference))
   )
   > Set Font(Font=Get AdditionalRegularPriceFont() -> (AdditionalRegularPriceFont:Slate Font Info Structure), self=Get Text_AdditionalRegularPrice())
  ]}
]
variables [
 AdditionalPriceFont {
  type=SlateFontInfo
  default {
   bForceMonospaced=false
   bMaterialIsStencil=false
   fontMaterial=null
   fontObject=&HeadingNow
   letterSpacing=0
   monospacedWidth=1
   size=11.25
   skewAmount=0
   typefaceFontName=74 Regular
   outlineSettings {
    bApplyOutlineToDropShadows=false
    bMiteredCorners=false
    bSeparateFillAlpha=false
    outlineMaterial=null
    outlineSize=0
    outlineColor=<rgba 0 0 0 1>
   }
  }
 }
 AdditionalPricePadding {type=Margin, default=<ltrb 0 0 0 4>}
 AdditionalRegularPriceFont {
  type=SlateFontInfo
  default {
   bForceMonospaced=false
   bMaterialIsStencil=false
   fontMaterial=null
   fontObject=&HeadingNow
   letterSpacing=0
   monospacedWidth=1
   size=15.75
   skewAmount=0
   typefaceFontName=75 Medium
   outlineSettings {
    bApplyOutlineToDropShadows=false
    bMiteredCorners=false
    bSeparateFillAlpha=false
    outlineMaterial=null
    outlineSize=0
    outlineColor=<rgba 0 0 0 1>
   }
  }
 }
 IconSize {type=Vector2D, default=<32 32>}
 PriceFont {
  type=SlateFontInfo
  default {
   bForceMonospaced=false
   bMaterialIsStencil=false
   fontMaterial=/FortUILibrary/Systems/ItemShopTile/Materials/PriceSwitcher/MI_UIKit_PriceSwitcher_Text
   fontObject=&HeadingNow
   letterSpacing=0
   monospacedWidth=1
   size=24
   skewAmount=0
   typefaceFontName=86 Bold Italic
   outlineSettings {
    bApplyOutlineToDropShadows=false
    bMiteredCorners=false
    bSeparateFillAlpha=false
    outlineMaterial=null
    outlineSize=0
    outlineColor=<rgba 0 0 0 1>
   }
  }
 }
 RegularPriceFont {
  type=SlateFontInfo
  default {
   bForceMonospaced=false
   bMaterialIsStencil=false
   fontMaterial=&MI_UIKit_PriceSwitcher_Text
   fontObject=&HeadingNow
   letterSpacing=0
   monospacedWidth=1
   size=24
   skewAmount=0
   typefaceFontName=86 Bold Italic
   outlineSettings {
    bApplyOutlineToDropShadows=false
    bMiteredCorners=false
    bSeparateFillAlpha=false
    outlineMaterial=null
    outlineSize=0
    outlineColor=<rgba 0 0 0 1>
   }
  }
 }
 RegularPricePadding {type=Margin, default=<ltrb 0 0 0 0>}
]
widgets {
 info {
  inheritedWidgetCount=0
  parentClass=/FortUILibrary/Systems/Price/WBP_UIKit_Price.WBP_UIKit_Price_C
  rootWidgetClass=&GridPanel
  widgetCount=14
 }
 widgets [
  GridPanel_0 {widgetClassPath=&GridPanel, widget=GridPanel_0, properties {bOverride_Cursor=true}}
  HB_PriceContainer {widgetClassPath=&HorizontalBox, parent=GridPanel_0, widget=HB_PriceContainer, slot {class=GridSlot}}
  Overlay_0 {
   widgetClassPath=&Overlay
   parent=HB_PriceContainer
   widget=Overlay_0
   slot {class=HorizontalBoxSlot, properties {horizontalAlignment=HAlign_Left, verticalAlignment=VAlign_Center}}
  }
  PlatformIcon {
   widgetClassPath=/FortUILibrary/Systems/ItemShopTile/WBP_Price_PlatformStoreIcon
   bIsVariable=true
   parent=Overlay_0
   widget=PlatformIcon
   properties {currency=EFortPriceCurrency_MAX}
   slot {class=OverlaySlot, properties {horizontalAlignment=HAlign_Center, verticalAlignment=VAlign_Center}}
  }
  Image_Currency {
   widgetClassPath=&Image
   bIsVariable=true
   parent=Overlay_0
   widget=Image_Currency
   properties {brush {
     bIsDynamicallyLoaded=false
     drawAs=Image
     imageType=FullColor
     mirroring=NoMirror
     resourceName=None
     resourceObject=/FortUILibrary/Textures/Icons/T_UI_Icon_UIKit_VBuck
     tiling=NoTile
     imageSize=<32 32>
     margin=<ltrb 0 0 0 0>
     outlineSettings {
      bUseBrushTransparency=false
      roundingType=HalfHeightRadius
      width=0
      color {colorUseRule=UseColor_Specified, specifiedColor=<rgba 0 0 0 0>}
      cornerRadii=<0 0 0 0>
     }
     tintColor {colorUseRule=UseColor_Specified, specifiedColor=<rgba 1 1 1 1>}
     uVRegion {bIsValid=false, max=<0 0>, min=<0 0>}
    }}
   slot {class=OverlaySlot, properties {horizontalAlignment=HAlign_Center, verticalAlignment=VAlign_Center, padding=<ltrb 1 0 4 0>}}
  }
  Spacer {
   widgetClassPath=&Spacer
   parent=HB_PriceContainer
   widget=Spacer
   slot {class=HorizontalBoxSlot, properties {horizontalAlignment=HAlign_Left}}
  }
  Text_Price {
   widgetClassPath=&CommonTextBlock
   bIsVariable=true
   parent=HB_PriceContainer
   widget=Text_Price
   properties {
    bAutoCollapseWithEmptyText=true
    justification=Center
    mobileFontSizeMultiplier=1.2
    text="1,800"
    textTransformPolicy=ToUpper
    font {
     bForceMonospaced=false
     bMaterialIsStencil=false
     fontMaterial=&MI_UIKit_PriceSwitcher_Text
     fontObject=&HeadingNow
     letterSpacing=0
     monospacedWidth=1
     size=24
     skewAmount=0
     typefaceFontName=86 Bold Italic
     outlineSettings {
      bApplyOutlineToDropShadows=false
      bMiteredCorners=false
      bSeparateFillAlpha=false
      outlineMaterial=null
      outlineSize=0
      outlineColor=<rgba 0 0 0 1>
     }
    }
    shadowOffset=<0 0>
   }
   slot {class=HorizontalBoxSlot, properties {horizontalAlignment=HAlign_Right, verticalAlignment=VAlign_Center, padding=<ltrb 0 0 4 0>}}
  }
  Text_AdditionalPrice {
   widgetClassPath=&CommonTextBlock
   bIsVariable=true
   parent=HB_PriceContainer
   widget=Text_AdditionalPrice
   properties {
    applyLineHeightToBottomLine=false
    bAutoCollapseWithEmptyText=true
    justification=Center
    mobileFontSizeMultiplier=1.2
    text=VAT incl.
    font {
     bForceMonospaced=false
     bMaterialIsStencil=false
     fontMaterial=null
     fontObject=&HeadingNow
     letterSpacing=0
     monospacedWidth=1
     size=11.25
     skewAmount=0
     typefaceFontName=74 Regular
     outlineSettings {
      bApplyOutlineToDropShadows=false
      bMiteredCorners=false
      bSeparateFillAlpha=false
      outlineMaterial=null
      outlineSize=0
      outlineColor=<rgba 0 0 0 1>
     }
    }
    shadowOffset=<0 0>
   }
   slot {class=HorizontalBoxSlot, properties {horizontalAlignment=HAlign_Right, verticalAlignment=VAlign_Bottom, padding=<ltrb 0 0 4 4>}}
  }
  Spacer_1 {
   widgetClassPath=&Spacer
   parent=HB_PriceContainer
   widget=Spacer_1
   properties {size=<2 1>}
   slot {class=HorizontalBoxSlot, properties {horizontalAlignment=HAlign_Left}}
  }
  HB_RegularPrice {
   widgetClassPath=&HorizontalBox
   bIsVariable=true
   parent=HB_PriceContainer
   widget=HB_RegularPrice
   slot {class=HorizontalBoxSlot, properties {horizontalAlignment=HAlign_Right, verticalAlignment=VAlign_Center}}
  }
  Overlay_RegularPrice {widgetClassPath=&Overlay, bIsVariable=true, parent=HB_RegularPrice, widget=Overlay_RegularPrice, slot {class=HorizontalBoxSlot}}
  Text_RegularPrice {
   widgetClassPath=&CommonTextBlock
   bIsVariable=true
   parent=Overlay_RegularPrice
   widget=Text_RegularPrice
   properties {
    bAutoCollapseWithEmptyText=true
    justification=Center
    mobileFontSizeMultiplier=1.2
    text="2,000"
    textTransformPolicy=ToUpper
    colorAndOpacity {colorUseRule=UseColor_Specified, specifiedColor=<rgba 1 1 1 0.5>}
    font {
     bForceMonospaced=false
     bMaterialIsStencil=false
     fontMaterial=&MI_UIKit_PriceSwitcher_Text
     fontObject=&HeadingNow
     letterSpacing=0
     monospacedWidth=1
     size=24
     skewAmount=0
     typefaceFontName=86 Bold Italic
     outlineSettings {
      bApplyOutlineToDropShadows=false
      bMiteredCorners=false
      bSeparateFillAlpha=false
      outlineMaterial=null
      outlineSize=0
      outlineColor=<rgba 0 0 0 1>
     }
    }
    shadowOffset=<0 0>
   }
   slot {class=OverlaySlot, properties {horizontalAlignment=HAlign_Fill, verticalAlignment=VAlign_Center, padding=<ltrb 4 0 4 0>}}
  }
  Image_Strike {
   widgetClassPath=&Image
   parent=Overlay_RegularPrice
   widget=Image_Strike
   properties {
    visibility=SelfHitTestInvisible
    brush {
     bIsDynamicallyLoaded=false
     drawAs=Image
     imageType=FullColor
     mirroring=NoMirror
     resourceName=None
     resourceObject=/FortUILibrary/Systems/ItemShopTile/Materials/PriceSwitcher/MI_UIKit_ItemShopTile_RegularPriceStrike
     tiling=NoTile
     imageSize=<32 4>
     margin=<ltrb 0 0 0 0>
     outlineSettings {
      bUseBrushTransparency=false
      roundingType=HalfHeightRadius
      width=0
      color {colorUseRule=UseColor_Specified, specifiedColor=<rgba 0 0 0 0>}
      cornerRadii=<0 0 0 0>
     }
     tintColor {colorUseRule=UseColor_Specified, specifiedColor=<rgba 1 1 1 1>}
     uVRegion {bIsValid=false, max=<0 0>, min=<0 0>}
    }
    renderTransform {angle=-3, scale=<1 1>, shear=<0 0>, translation=<0 0>}
   }
   slot {class=OverlaySlot, properties {horizontalAlignment=HAlign_Fill, verticalAlignment=VAlign_Center}}
  }
  Text_AdditionalRegularPrice {
   widgetClassPath=&CommonTextBlock
   bIsVariable=true
   parent=HB_RegularPrice
   widget=Text_AdditionalRegularPrice
   properties {
    bAutoCollapseWithEmptyText=true
    justification=Center
    mobileFontSizeMultiplier=1.2
    text=*
    font {
     bForceMonospaced=false
     bMaterialIsStencil=false
     fontMaterial=null
     fontObject=&HeadingNow
     letterSpacing=0
     monospacedWidth=1
     size=15.75
     skewAmount=0
     typefaceFontName=75 Medium
     outlineSettings {
      bApplyOutlineToDropShadows=false
      bMiteredCorners=false
      bSeparateFillAlpha=false
      outlineMaterial=null
      outlineSize=0
      outlineColor=<rgba 0 0 0 1>
     }
    }
    shadowOffset=<0 0>
   }
   slot {class=HorizontalBoxSlot, properties {horizontalAlignment=HAlign_Right, verticalAlignment=VAlign_Center, padding=<ltrb 4 0 4 0>}}
  }
 ]
}
